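# Material function for the post-match XP screen UI that draws a two-layer flare shape as a mask.
# Inner and outer layers each scale UVs about the center, mirror them with 2x2 symmetry and rotate them into a regular-polygon SDF.
# Both share rotation and anti-aliasing smoothing inputs, with separate inset and opacity.
# The layers are added, then a Step on the UV green channel clips the result to one vertical half.
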MaterialFunction /FortniteCoreUI/HUD/EndGame/PostGameFlowScreens/DefaultScreens/XPScreen/Materials/MF_PostMatch_Flare
graph {
 properties {previewBlendMode=BLEND_Translucent, previewMaterialDomain=MD_UI}
 nodes [
  n1 = ComponentMask(None=NamedRerouteUsage()) {g=true}
  n2 = NamedRerouteUsage()
  n3 = FunctionInput() {bUsePreviewValueAsDefault=true, inputName=Inner Flare Inset, inputType=FunctionInput_Scalar, sortPriority=1, previewValue=<0.5 0 0 1>}
  n4 = FunctionInput() {bUsePreviewValueAsDefault=true, inputName=Rotation, inputType=FunctionInput_Scalar, sortPriority=3, previewValue=<0.05 0 0 1>}
  n5 = MaterialFunctionCall(
   UVs=MaterialFunctionCall!(
    UV=MaterialFunctionCall(UVs=@n2, Texture Scale=@n3) {materialFunction=&ScaleUVsByCenter}.UVs
    Swizzle UV
   ) {materialFunction=/FortUILibrary/MaterialFunctions/Basics/MF_UI_Symmetry_2x2_New}.Symmetry
   "Rotation Angle (0-1)"=@n4
  ) {materialFunction=&CustomRotator}
  n6 = FunctionInput() {bUsePreviewValueAsDefault=true, inputName=Smoothing, inputType=FunctionInput_Scalar, sortPriority=4, previewValue=<0.05 0 0 1>}
  n7 = FunctionInput() {bUsePreviewValueAsDefault=true, inputName=Inner Flare Opacity, inputType=FunctionInput_Scalar, sortPriority=5, previewValue=<0.25 0 0 1>}
  n8 = Multiply(
   A=MaterialFunctionCall(
    DistanceField=MaterialFunctionCall!(UVs=@n5.Rotated Values, Size, Sides, InnerAngle, OutlineThickness, OutlineAlignment, GlowMin, GlowMax) {materialFunction=/FortUILibrary/MaterialFunctions/Shapes/MF_UI_RegularPolygon}.SDF
    Smoothing=@n6
   ) {materialFunction=&MF_UI_AntiAliasing}.Inner
   B=@n7
  ) {constB=0.5}
  n9 = NamedRerouteUsage()
  n10 = FunctionInput() {bUsePreviewValueAsDefault=true, inputName=Outer Flare Inset, inputType=FunctionInput_Scalar, sortPriority=2, previewValue=<1 0 0 1>}
  n11 = MaterialFunctionCall(
   UVs=MaterialFunctionCall(UV=MaterialFunctionCall(UVs=@n9, Texture Scale=@n10) {materialFunction=&ScaleUVsByCenter}.UVs) {materialFunction=&MF_UI_Symmetry_2x2_New}.Symmetry
   "Rotation Angle (0-1)"=@n4
  ) {materialFunction=&CustomRotator}
  n12 = Constant() {r=0.4}
  n13 = FunctionInput() {bUsePreviewValueAsDefault=true, inputName=Outer Flare Opacity, inputType=FunctionInput_Scalar, sortPriority=6, previewValue=<0.5 0 0 1>}
  n14 = Multiply(
   A=MaterialFunctionCall(
    DistanceField=MaterialFunctionCall(UVs=@n11.Rotated Values, Size=@n12) {materialFunction=&MF_UI_RegularPolygon}.SDF
    Smoothing=@n6
   ) {materialFunction=&MF_UI_AntiAliasing}.Inner
   B=@n13
  )
  FunctionOutput(None=Multiply(A=Step(X=@n1) {constX=0.5, constY=0.5}, B=Add(A=@n8, B=@n14)))
  NamedRerouteDeclaration(None=FunctionInput(Preview=TextureCoordinate()) {bUsePreviewValueAsDefault=true, inputName=UV, inputType=FunctionInput_Vector2}) {name=UV, nodeColor=<rgba 1 0.4219 0 1>}
  ScalarParameter() {defaultValue=0.05, parameterName=FlareSmoothing}
 ]
}
