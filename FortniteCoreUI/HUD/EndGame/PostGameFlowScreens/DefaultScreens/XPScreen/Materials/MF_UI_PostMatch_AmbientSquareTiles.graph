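# Material function for the post-match XP screen that outputs a scalar mask of ambient, animated square tiles.
# A particle grid splits the offset UVs into cells, and seeded random values set each tile's opacity and cull some tiles through a biased Ceil.
# A second grid call feeds time-seeded randomness into tile size.
# A rounded-box SDF with SmoothStep draws each tile, with edge softness driven by Blurriness and size.
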MaterialFunction /FortniteCoreUI/HUD/EndGame/PostGameFlowScreens/DefaultScreens/XPScreen/Materials/MF_UI_PostMatch_AmbientSquareTiles
graph {nodes [
  n1 = FunctionInput() {bUsePreviewValueAsDefault=true, inputName=Min Opacity, inputType=FunctionInput_Scalar, sortPriority=6, previewValue=<0.5 0 0 1>}
  n2 = FunctionInput() {bUsePreviewValueAsDefault=true, inputName=Max Opacity, inputType=FunctionInput_Scalar, sortPriority=5, previewValue=<1 0 0 1>}
  n3 = Add(
   A=FunctionInput(Preview=MaterialFunctionCall() {materialFunction=&GetUserInterfaceUV}.Normalized UV) {bUsePreviewValueAsDefault=true, inputName=UVs, inputType=FunctionInput_Vector2}
   B=FunctionInput() {bUsePreviewValueAsDefault=true, inputName=OffsetGrid, inputType=FunctionInput_Scalar, sortPriority=4}
  )
  n4 = FunctionInput() {bUsePreviewValueAsDefault=true, inputName=Tile Count, inputType=FunctionInput_Scalar, sortPriority=1, previewValue=<5 0 0 1>}
  n5 = Constant()
  n6 = FunctionInput() {
   bUsePreviewValueAsDefault=true
   inputName=Tile Subtraction Random Seed
   inputType=FunctionInput_Scalar
   sortPriority=4
   previewValue=<19.5479 0 0 1>
  }
  n7 = Saturate(None=MaterialFunctionCall!(UV=@n3, Global Offset, Global Rotation, Tiling=@n4, Cell Offset=@n5, Cell Scale, Cell Rotation=@n5, Random Seed=@n6) {materialFunction=/FortUILibrary/MaterialFunctions/Basics/MF_UI_ParticleGrid}.Random Cell Values)
  n8 = LinearInterpolate(A=@n1, B=@n2, Alpha=@n7) {constA=0.5}
  n9 = FunctionInput() {bUsePreviewValueAsDefault=true, inputName=Tile Subtraction Amount, inputType=FunctionInput_Scalar, sortPriority=3, previewValue=<2 0 0 1>}
  n10 = LinearInterpolate(A=Multiply(A=@n9) {constB=-1}, Alpha=@n7) {constA=-4}
  n11 = FunctionInput() {bUsePreviewValueAsDefault=true, inputName=Min Size, inputType=FunctionInput_Scalar, sortPriority=8, previewValue=<0.5 0 0 1>}
  n12 = Constant()
  n13 = MaterialFunctionCall(UV=@n3, Tiling=@n4, Cell Offset=@n12, Cell Rotation=@n12) {materialFunction=&MF_UI_ParticleGrid}
  n14 = FunctionInput() {bUsePreviewValueAsDefault=true, inputName=Speed, inputType=FunctionInput_Scalar, sortPriority=2, previewValue=<0.25 0 0 1>}
  n15 = Multiply(A=Time(), B=@n14) {constB=0.125}
  n16 = Saturate(None=LinearInterpolate(
    A=@n11
    Alpha=MaterialFunctionCall!(UV=@n13.Random Cell Values, Seed=@n15) {materialFunction=&MF_UI_Random}.Result
   ) {constA=0.5})
  n17 = FunctionInput() {bUsePreviewValueAsDefault=true, inputName=Blurriness, inputType=FunctionInput_Scalar, sortPriority=10, previewValue=<0.5 0 0 1>}
  n18 = FunctionInput() {bUsePreviewValueAsDefault=true, inputName=Corner Radius, inputType=FunctionInput_Scalar, sortPriority=9, previewValue=<0.05 0 0 1>}
  n19 = Multiply(A=Add(A=MaterialFunctionCall(UVs=@n13.Cells UV, Size=@n16, CornerRoundness=@n18) {materialFunction=&MF_UI_Box}.SDF)) {constB=0.5}
  n20 = SmoothStep(
   Min=Multiply(A=LinearInterpolate(Alpha=@n16) {constB=0.25}, B=@n17) {constB=0.5}
   Max=Multiply(A=LinearInterpolate(Alpha=@n16) {constA=1, constB=0.75}, B=@n17) {constB=0.5}
   Value=Saturate(None=OneMinus(None=@n19))
  ) {constMax=0.5, constMin=0.25}
  n21 = Saturate(None=Multiply(A=Ceil(None=@n10), B=Saturate(None=@n20)))
  FunctionOutput(None=Saturate(None=Multiply(A=@n8, B=@n21)))
 ]}
