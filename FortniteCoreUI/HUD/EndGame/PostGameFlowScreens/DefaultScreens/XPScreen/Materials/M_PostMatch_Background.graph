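# UI material that draws the animated background for the post-match XP screen in the end-game flow.
# Radial gradients build the color base and glows, with layered ambient square-tile cells at several scales modulating the gradient and vector noise soft-light blended in.
# A time-seeded StormTech grid, SDF ripples and a backplate mask from two unioned boxes, one skewed, add further layers.
# Scalar and vector parameters expose colors, speeds and shapes.
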Material /FortniteCoreUI/HUD/EndGame/PostGameFlowScreens/DefaultScreens/XPScreen/Materials/M_PostMatch_Background
graph {
 expressions {
  MaterialExpressionAdd_0 {
   class=MaterialExpressionAdd
   inputs [
    A {from=MaterialExpressionMaterialFunctionCall_13, output=Swizzled Radial Coordinates}
    B {from=MaterialExpressionLinearInterpolate_7}
   ]
  }
  MaterialExpressionAdd_1 {
   class=MaterialExpressionAdd
   inputs [
    A {from=MaterialExpressionMaterialFunctionCall_7, output=SDF}
    B {from=""}
   ]
  }
  MaterialExpressionAdd_12 {
   class=MaterialExpressionAdd
   inputs [
    A {from=MaterialExpressionLinearInterpolate_1}
    B {from=MaterialExpressionNamedRerouteUsage_43}
   ]
  }
  MaterialExpressionAdd_4 {
   class=MaterialExpressionAdd
   inputs [
    A {from=MaterialExpressionNamedRerouteUsage_8}
    B {from=MaterialExpressionConstant2Vector_0}
   ]
  }
  MaterialExpressionAdd_6 {
   class=MaterialExpressionAdd
   inputs [
    A {from=MaterialExpressionMaterialFunctionCall_17, output=Skewed UVs}
    B {from=MaterialExpressionConvert_2}
   ]
  }
  MaterialExpressionAdd_7 {
   class=MaterialExpressionAdd
   inputs [
    A {from=MaterialExpressionReroute_1}
    B {from=MaterialExpressionConvert_3}
   ]
  }
  MaterialExpressionClamp_1 {
   class=MaterialExpressionClamp
   inputs [
    None {from=MaterialExpressionMultiply_1, output=""}
    Min {from=MaterialExpressionFloor_1, output=""}
    Max {from=MaterialExpressionFloor_0, output=""}
   ]
   properties {maxDefault=2}
  }
  MaterialExpressionComponentMask_0 {class=MaterialExpressionComponentMask, inputs [None {from=MaterialExpressionAdd_0}], properties {r=true}}
  MaterialExpressionComponentMask_1 {class=MaterialExpressionComponentMask, inputs [None {from=MaterialExpressionNamedRerouteUsage_1}], properties {g=true}}
  MaterialExpressionComponentMask_2 {class=MaterialExpressionComponentMask, inputs [None {from=MaterialExpressionNamedRerouteUsage_24}], properties {g=true}}
  MaterialExpressionConstant2Vector_0 {class=MaterialExpressionConstant2Vector, properties {g=0.8}}
  MaterialExpressionConstant2Vector_1 {class=MaterialExpressionConstant2Vector, properties {g=0.85, r=0.925}}
  MaterialExpressionConstant2Vector_3 {class=MaterialExpressionConstant2Vector, properties {g=0.25, r=0.1}}
  MaterialExpressionConstant2Vector_4 {class=MaterialExpressionConstant2Vector, properties {g=0.5, r=0.75}}
  MaterialExpressionConstant2Vector_5 {class=MaterialExpressionConstant2Vector, properties {g=0.125}}
  MaterialExpressionConstantBiasScale_0 {class=MaterialExpressionConstantBiasScale, inputs [None {from=MaterialExpressionSine_1}]}
  MaterialExpressionConstantBiasScale_1 {class=MaterialExpressionConstantBiasScale, inputs [None {from=MaterialExpressionSine_0}]}
  MaterialExpressionConstant_0 {class=MaterialExpressionConstant}
  MaterialExpressionConstant_1 {class=MaterialExpressionConstant, properties {r=0.75}}
  MaterialExpressionConstant_10 {class=MaterialExpressionConstant, properties {r=30.4275}}
  MaterialExpressionConstant_11 {class=MaterialExpressionConstant, properties {r=0.3333}}
  MaterialExpressionConstant_12 {class=MaterialExpressionConstant}
  MaterialExpressionConstant_13 {class=MaterialExpressionConstant, properties {r=2}}
  MaterialExpressionConstant_14 {class=MaterialExpressionConstant, properties {r=0.1111}}
  MaterialExpressionConstant_15 {class=MaterialExpressionConstant, properties {r=4}}
  MaterialExpressionConstant_16 {class=MaterialExpressionConstant, properties {r=30.4275}}
  MaterialExpressionConstant_17 {class=MaterialExpressionConstant, properties {r=0.3333}}
  MaterialExpressionConstant_18 {class=MaterialExpressionConstant}
  MaterialExpressionConstant_19 {class=MaterialExpressionConstant, properties {r=1}}
  MaterialExpressionConstant_2 {class=MaterialExpressionConstant, properties {r=0.5}}
  MaterialExpressionConstant_20 {class=MaterialExpressionConstant, properties {r=-0.25}}
  MaterialExpressionConstant_21 {class=MaterialExpressionConstant, properties {r=1}}
  MaterialExpressionConstant_22 {class=MaterialExpressionConstant, properties {r=4}}
  MaterialExpressionConstant_23 {class=MaterialExpressionConstant, properties {r=0.05}}
  MaterialExpressionConstant_24 {class=MaterialExpressionConstant, properties {r=4}}
  MaterialExpressionConstant_25 {class=MaterialExpressionConstant, properties {r=18.6482}}
  MaterialExpressionConstant_26 {class=MaterialExpressionConstant, properties {r=0.5}}
  MaterialExpressionConstant_27 {class=MaterialExpressionConstant}
  MaterialExpressionConstant_28 {class=MaterialExpressionConstant, properties {r=0.5}}
  MaterialExpressionConstant_29 {class=MaterialExpressionConstant, properties {r=-0.25}}
  MaterialExpressionConstant_3 {class=MaterialExpressionConstant, properties {r=0.025}}
  MaterialExpressionConstant_30 {class=MaterialExpressionConstant, properties {r=0.3}}
  MaterialExpressionConstant_31 {class=MaterialExpressionConstant, properties {r=7}}
  MaterialExpressionConstant_32 {class=MaterialExpressionConstant, properties {r=0.0625}}
  MaterialExpressionConstant_33 {class=MaterialExpressionConstant, properties {r=8}}
  MaterialExpressionConstant_34 {class=MaterialExpressionConstant, properties {r=32.8625}}
  MaterialExpressionConstant_35 {class=MaterialExpressionConstant, properties {r=0.125}}
  MaterialExpressionConstant_36 {class=MaterialExpressionConstant}
  MaterialExpressionConstant_37 {class=MaterialExpressionConstant, properties {r=0.25}}
  MaterialExpressionConstant_38 {class=MaterialExpressionConstant, properties {r=-0.25}}
  MaterialExpressionConstant_39 {class=MaterialExpressionConstant, properties {r=0.5}}
  MaterialExpressionConstant_4 {class=MaterialExpressionConstant, properties {r=0.05}}
  MaterialExpressionConstant_40 {class=MaterialExpressionConstant, properties {r=11}}
  MaterialExpressionConstant_41 {class=MaterialExpressionConstant, properties {r=0.25}}
  MaterialExpressionConstant_42 {class=MaterialExpressionConstant, properties {r=16}}
  MaterialExpressionConstant_43 {class=MaterialExpressionConstant, properties {r=32.8625}}
  MaterialExpressionConstant_44 {class=MaterialExpressionConstant, properties {r=0.2}}
  MaterialExpressionConstant_45 {class=MaterialExpressionConstant}
  MaterialExpressionConstant_46 {class=MaterialExpressionConstant, properties {r=0.5}}
  MaterialExpressionConstant_47 {class=MaterialExpressionConstant, properties {r=-0.25}}
  MaterialExpressionConstant_48 {class=MaterialExpressionConstant, properties {r=0.4}}
  MaterialExpressionConstant_49 {class=MaterialExpressionConstant, properties {r=0.75}}
  MaterialExpressionConstant_5 {class=MaterialExpressionConstant, properties {r=2}}
  MaterialExpressionConstant_50 {class=MaterialExpressionConstant, properties {r=-0.25}}
  MaterialExpressionConstant_51 {class=MaterialExpressionConstant, properties {r=21}}
  MaterialExpressionConstant_52 {class=MaterialExpressionConstant, properties {r=0.2}}
  MaterialExpressionConstant_53 {class=MaterialExpressionConstant, properties {r=24}}
  MaterialExpressionConstant_54 {class=MaterialExpressionConstant, properties {r=4.8286}}
  MaterialExpressionConstant_55 {class=MaterialExpressionConstant, properties {r=0.125}}
  MaterialExpressionConstant_56 {class=MaterialExpressionConstant}
  MaterialExpressionConstant_57 {class=MaterialExpressionConstant, properties {r=0.5}}
  MaterialExpressionConstant_58 {class=MaterialExpressionConstant, properties {r=-0.25}}
  MaterialExpressionConstant_59 {class=MaterialExpressionConstant, properties {r=0.45}}
  MaterialExpressionConstant_6 {class=MaterialExpressionConstant}
  MaterialExpressionConstant_60 {class=MaterialExpressionConstant, properties {r=0.25}}
  MaterialExpressionConstant_61 {class=MaterialExpressionConstant, properties {r=0.25}}
  MaterialExpressionConstant_62 {class=MaterialExpressionConstant}
  MaterialExpressionConstant_63 {class=MaterialExpressionConstant}
  MaterialExpressionConstant_64 {class=MaterialExpressionConstant, properties {r=0.025}}
  MaterialExpressionConstant_7 {class=MaterialExpressionConstant, properties {r=-0.25}}
  MaterialExpressionConstant_8 {class=MaterialExpressionConstant, properties {r=0.1111}}
  MaterialExpressionConstant_9 {class=MaterialExpressionConstant, properties {r=2}}
  MaterialExpressionConvert_0 {
   class=MaterialExpressionConvert
   inputs [
    None {from=""}
    None {from=MaterialExpressionScalarParameter_5}
   ]
   properties {
    nodeName=Make Vector2
    convertInputs [
     {type=Scalar, expressionInput {expression=null, inputName=None, mask=0, maskA=0, maskB=0, maskG=0, maskR=0, outputIndex=0}}
     {
      type=Scalar
      expressionInput {
       expression=&self:MaterialExpressionScalarParameter_5
       inputName=None
       mask=0
       maskA=0
       maskB=0
       maskG=0
       maskR=0
       outputIndex=0
      }
     }
    ]
    convertOutputs [{type=Vector2}]
   }
  }
  MaterialExpressionConvert_1 {
   class=MaterialExpressionConvert
   inputs [
    None {from=""}
    None {from=MaterialExpressionScalarParameter_10}
   ]
   properties {
    nodeName=Make Vector2
    convertInputs [
     {type=Scalar, expressionInput {expression=null, inputName=None, mask=0, maskA=0, maskB=0, maskG=0, maskR=0, outputIndex=0}}
     {
      type=Scalar
      expressionInput {
       expression=&self:MaterialExpressionScalarParameter_10
       inputName=None
       mask=0
       maskA=0
       maskB=0
       maskG=0
       maskR=0
       outputIndex=0
      }
     }
    ]
    convertOutputs [{type=Vector2}]
   }
  }
  MaterialExpressionConvert_2 {
   class=MaterialExpressionConvert
   inputs [
    None {from=MaterialExpressionScalarParameter_12}
    None {from=MaterialExpressionScalarParameter_13}
   ]
   properties {
    nodeName=Make Vector2
    convertInputs [
     {
      type=Scalar
      expressionInput {
       expression=&self:MaterialExpressionScalarParameter_12
       inputName=None
       mask=0
       maskA=0
       maskB=0
       maskG=0
       maskR=0
       outputIndex=0
      }
     }
     {
      type=Scalar
      expressionInput {
       expression=&self:MaterialExpressionScalarParameter_13
       inputName=None
       mask=0
       maskA=0
       maskB=0
       maskG=0
       maskR=0
       outputIndex=0
      }
     }
    ]
    convertOutputs [{type=Vector2}]
   }
  }
  MaterialExpressionConvert_3 {
   class=MaterialExpressionConvert
   inputs [
    None {from=MaterialExpressionScalarParameter_14}
    None {from=MaterialExpressionScalarParameter_15}
   ]
   properties {
    nodeName=Make Vector2
    convertInputs [
     {
      type=Scalar
      expressionInput {
       expression=&self:MaterialExpressionScalarParameter_14
       inputName=None
       mask=0
       maskA=0
       maskB=0
       maskG=0
       maskR=0
       outputIndex=0
      }
     }
     {
      type=Scalar
      expressionInput {
       expression=&self:MaterialExpressionScalarParameter_15
       inputName=None
       mask=0
       maskA=0
       maskB=0
       maskG=0
       maskR=0
       outputIndex=0
      }
     }
    ]
    convertOutputs [{type=Vector2}]
   }
  }
  MaterialExpressionConvert_4 {
   class=MaterialExpressionConvert
   inputs [
    None {from=MaterialExpressionScalarParameter_18}
    None {from=MaterialExpressionScalarParameter_19}
   ]
   properties {
    nodeName=Make Vector2
    convertInputs [
     {
      type=Scalar
      expressionInput {
       expression=&self:MaterialExpressionScalarParameter_18
       inputName=None
       mask=0
       maskA=0
       maskB=0
       maskG=0
       maskR=0
       outputIndex=0
      }
     }
     {
      type=Scalar
      expressionInput {
       expression=&self:MaterialExpressionScalarParameter_19
       inputName=None
       mask=0
       maskA=0
       maskB=0
       maskG=0
       maskR=0
       outputIndex=0
      }
     }
    ]
    convertOutputs [{type=Vector2}]
   }
  }
  MaterialExpressionDivide_0 {
   class=MaterialExpressionDivide
   inputs [
    A {from=MaterialExpressionScalarParameter_1}
    B {from=MaterialExpressionMaterialFunctionCall_0, output=Viewport Scale Factor}
   ]
  }
  MaterialExpressionFloor_0 {class=MaterialExpressionFloor, inputs [None {from=MaterialExpressionScalarParameter_29, output=""}], properties {}}
  MaterialExpressionFloor_1 {class=MaterialExpressionFloor, inputs [None {from=MaterialExpressionScalarParameter_30, output=""}], properties {}}
  MaterialExpressionInverseLinearInterpolate_0 {
   class=MaterialExpressionInverseLinearInterpolate
   inputs [
    A {from=MaterialExpressionScalarParameter_33, output=""}
    B {from=MaterialExpressionScalarParameter_34, output=""}
    Value {from=MaterialExpressionMaterialFunctionCall_14, output=EaseInQuint}
   ]
   properties {constA=0.8, constB=1.5}
  }
  MaterialExpressionLinearInterpolate_0 {
   class=MaterialExpressionLinearInterpolate
   inputs [
    A {from=MaterialExpressionMultiply_1}
    B {from=MaterialExpressionClamp_1}
    Alpha {from=MaterialExpressionScalarParameter_31}
   ]
  }
  MaterialExpressionLinearInterpolate_1 {
   class=MaterialExpressionLinearInterpolate
   inputs [
    A {from=MaterialExpressionNamedRerouteUsage_30}
    B {from=MaterialExpressionMultiply_3}
    Alpha {from=MaterialExpressionMultiply_2}
   ]
   properties {constB=0.1}
  }
  MaterialExpressionLinearInterpolate_10 {
   class=MaterialExpressionLinearInterpolate
   inputs [
    A {from=MaterialExpressionMaterialFunctionCall_12, output=EaseInSine}
    B {from=""}
    Alpha {from=MaterialExpressionMultiply_10}
   ]
  }
  MaterialExpressionLinearInterpolate_11 {
   class=MaterialExpressionLinearInterpolate
   inputs [
    A {from=MaterialExpressionNamedRerouteUsage_6}
    B {from=""}
    Alpha {from=MaterialExpressionLinearInterpolate_12}
   ]
  }
  MaterialExpressionLinearInterpolate_12 {
   class=MaterialExpressionLinearInterpolate
   inputs [
    A {from=MaterialExpressionNamedRerouteUsage_16}
    B {from=""}
    Alpha {from=MaterialExpressionLinearInterpolate_15}
   ]
  }
  MaterialExpressionLinearInterpolate_13 {
   class=MaterialExpressionLinearInterpolate
   inputs [
    A {from=MaterialExpressionStaticSwitchParameter_3}
    B {from=MaterialExpressionScalarParameter_20}
    Alpha {from=MaterialExpressionNamedRerouteUsage_36}
   ]
  }
  MaterialExpressionLinearInterpolate_14 {
   class=MaterialExpressionLinearInterpolate
   inputs [
    A {from=""}
    B {from=""}
    Alpha {from=MaterialExpressionConstantBiasScale_1}
   ]
   properties {constA=1, constB=0.8}
  }
  MaterialExpressionLinearInterpolate_15 {
   class=MaterialExpressionLinearInterpolate
   inputs [
    A {from=MaterialExpressionNamedRerouteUsage_17}
    B {from=""}
    Alpha {from=MaterialExpressionLinearInterpolate_22}
   ]
  }
  MaterialExpressionLinearInterpolate_16 {
   class=MaterialExpressionLinearInterpolate
   inputs [
    A {from=MaterialExpressionNamedRerouteUsage_38}
    B {from=MaterialExpressionNamedRerouteDeclaration_16}
    Alpha {from=MaterialExpressionNamedRerouteUsage_37}
   ]
  }
  MaterialExpressionLinearInterpolate_17 {
   class=MaterialExpressionLinearInterpolate
   inputs [
    A {from=MaterialExpressionScalarParameter_22}
    B {from=MaterialExpressionScalarParameter_21}
    Alpha {from=MaterialExpressionNamedRerouteUsage_39}
   ]
  }
  MaterialExpressionLinearInterpolate_18 {
   class=MaterialExpressionLinearInterpolate
   inputs [
    A {from=""}
    B {from=""}
    Alpha {from=MaterialExpressionScalarParameter_38}
   ]
   properties {constA=1, constB=2}
  }
  MaterialExpressionLinearInterpolate_19 {
   class=MaterialExpressionLinearInterpolate
   inputs [
    A {from=MaterialExpressionVectorParameter_0, output=RGB}
    B {from=MaterialExpressionVectorParameter_9, output=RGB}
    Alpha {from=MaterialExpressionNamedRerouteUsage_41}
   ]
  }
  MaterialExpressionLinearInterpolate_2 {
   class=MaterialExpressionLinearInterpolate
   inputs [
    A {from=MaterialExpressionLinearInterpolate_5}
    B {from=MaterialExpressionVectorParameter_2, output=RGB}
    Alpha {from=MaterialExpressionMultiply_7}
   ]
  }
  MaterialExpressionLinearInterpolate_20 {
   class=MaterialExpressionLinearInterpolate
   inputs [
    A {from=MaterialExpressionScalarParameter_22}
    B {from=MaterialExpressionScalarParameter_23}
    Alpha {from=MaterialExpressionNamedRerouteUsage_42}
   ]
   properties {constB=0.4}
  }
  MaterialExpressionLinearInterpolate_21 {
   class=MaterialExpressionLinearInterpolate
   inputs [
    A {from=MaterialExpressionLinearInterpolate_1}
    B {from=MaterialExpressionAdd_12}
    Alpha {from=MaterialExpressionScalarParameter_24}
   ]
  }
  MaterialExpressionLinearInterpolate_22 {
   class=MaterialExpressionLinearInterpolate
   inputs [
    A {from=MaterialExpressionNamedRerouteUsage_19}
    B {from=""}
    Alpha {from=MaterialExpressionNamedRerouteUsage_25}
   ]
  }
  MaterialExpressionLinearInterpolate_23 {
   class=MaterialExpressionLinearInterpolate
   inputs [
    A {from=MaterialExpressionMaterialFunctionCall_16, output=Result}
    B {from=""}
    Alpha {from=MaterialExpressionMaterialFunctionCall_19, output=Result}
   ]
  }
  MaterialExpressionLinearInterpolate_24 {
   class=MaterialExpressionLinearInterpolate
   inputs [
    A {from=MaterialExpressionNamedRerouteUsage_22}
    B {from=MaterialExpressionVectorParameter_10, output=RGB}
    Alpha {from=MaterialExpressionNamedRerouteUsage_21}
   ]
  }
  MaterialExpressionLinearInterpolate_25 {
   class=MaterialExpressionLinearInterpolate
   inputs [
    A {from=MaterialExpressionNamedRerouteUsage_22}
    B {from=MaterialExpressionLinearInterpolate_24}
    Alpha {from=MaterialExpressionScalarParameter_39}
   ]
  }
  MaterialExpressionLinearInterpolate_3 {
   class=MaterialExpressionLinearInterpolate
   inputs [
    A {from=""}
    B {from=MaterialExpressionScalarParameter_37}
    Alpha {from=MaterialExpressionNamedRerouteDeclaration_6}
   ]
   properties {constA=0.95}
  }
  MaterialExpressionLinearInterpolate_4 {
   class=MaterialExpressionLinearInterpolate
   inputs [
    A {from=MaterialExpressionVectorParameter_6, output=RGB}
    B {from=MaterialExpressionVectorParameter_7, output=RGB}
    Alpha {from=MaterialExpressionNamedRerouteUsage_29}
   ]
  }
  MaterialExpressionLinearInterpolate_5 {
   class=MaterialExpressionLinearInterpolate
   inputs [
    A {from=MaterialExpressionNamedRerouteUsage_22}
    B {from=MaterialExpressionVectorParameter_8, output=RGB}
    Alpha {from=MaterialExpressionNamedRerouteUsage_32}
   ]
  }
  MaterialExpressionLinearInterpolate_6 {
   class=MaterialExpressionLinearInterpolate
   inputs [
    A {from=""}
    B {from=""}
    Alpha {from=MaterialExpressionNamedRerouteUsage_35}
   ]
   properties {constA=1, constB=0.5}
  }
  MaterialExpressionLinearInterpolate_7 {
   class=MaterialExpressionLinearInterpolate
   inputs [
    A {from=MaterialExpressionMultiply_5}
    B {from=MaterialExpressionScalarParameter_4}
    Alpha {from=MaterialExpressionScalarParameter_0}
   ]
  }
  MaterialExpressionLinearInterpolate_8 {
   class=MaterialExpressionLinearInterpolate
   inputs [
    A {from=MaterialExpressionLinearInterpolate_4}
    B {from=MaterialExpressionReroute_2}
    Alpha {from=MaterialExpressionMaterialFunctionCall_1, output=RadialGradientExponential}
   ]
  }
  MaterialExpressionLinearInterpolate_9 {
   class=MaterialExpressionLinearInterpolate
   inputs [
    A {from=MaterialExpressionVectorParameter_4, output=RGB}
    B {from=MaterialExpressionVectorParameter_3, output=RGB}
    Alpha {from=MaterialExpressionLinearInterpolate_10}
   ]
  }
  MaterialExpressionMaterialFunctionCall_0 {
   class=MaterialExpressionMaterialFunctionCall
   inputs []
   properties {materialFunction=/FortUILibrary/MaterialFunctions/Basics/MF_UI_ViewportScale}
  }
  MaterialExpressionMaterialFunctionCall_1 {
   class=MaterialExpressionMaterialFunctionCall
   inputs [
    UVs {from=MaterialExpressionNamedRerouteUsage_0}
    CenterPosition {from=""}
    Radius {from=MaterialExpressionScalarParameter_25}
    Density {from=MaterialExpressionScalarParameter_28}
    Invert Density {from=""}
   ]
   properties {materialFunction=&RadialGradientExponential}
  }
  MaterialExpressionMaterialFunctionCall_10 {
   class=MaterialExpressionMaterialFunctionCall
   inputs [
    UV {from=MaterialExpressionAdd_4}
    GradientStart {from=""}
    GradientEnd {from=MaterialExpressionLinearInterpolate_14}
   ]
   properties {materialFunction=&MF_UI_RadialGradient}
  }
  MaterialExpressionMaterialFunctionCall_11 {
   class=MaterialExpressionMaterialFunctionCall
   inputs [
    UVs {from=MaterialExpressionNamedRerouteUsage_18}
    Tile Count {from=MaterialExpressionConstant_40}
    Speed {from=MaterialExpressionConstant_41}
    Tile Subtraction Amount {from=MaterialExpressionConstant_42}
    Tile Subtraction Random Seed {from=MaterialExpressionConstant_43}
    OffsetGrid {from=""}
    Max Opacity {from=MaterialExpressionConstant_44}
    Min Opacity {from=MaterialExpressionConstant_45}
    Min Size {from=MaterialExpressionConstant_46}
    Corner Radius {from=MaterialExpressionConstant_47}
    Blurriness {from=MaterialExpressionConstant_48}
   ]
   properties {materialFunction=&/MF_UI_PostMatch_AmbientSquareTiles}
  }
  MaterialExpressionMaterialFunctionCall_12 {
   class=MaterialExpressionMaterialFunctionCall
   inputs ["Value (0-1)" {from=MaterialExpressionNamedRerouteUsage_5}]
   properties {materialFunction=&MF_UI_EaseCurves}
  }
  MaterialExpressionMaterialFunctionCall_13 {
   class=MaterialExpressionMaterialFunctionCall
   inputs [Vector or UVs {from=MaterialExpressionSubtract_0}]
   properties {materialFunction=/FortUILibrary/MaterialFunctions/Basics/MF_UI_CheaperVectorToRadial}
  }
  MaterialExpressionMaterialFunctionCall_14 {
   class=MaterialExpressionMaterialFunctionCall
   inputs ["Value (0-1)" {from=MaterialExpressionComponentMask_1}]
   properties {materialFunction=&MF_UI_EaseCurves}
  }
  MaterialExpressionMaterialFunctionCall_15 {
   class=MaterialExpressionMaterialFunctionCall
   inputs ["Value (0-1)" {from=MaterialExpressionSaturate_1}]
   properties {materialFunction=&MF_UI_EaseCurves}
  }
  MaterialExpressionMaterialFunctionCall_16 {
   class=MaterialExpressionMaterialFunctionCall
   inputs [
    UVs {from=MaterialExpressionNamedRerouteUsage_11}
    Tile Count {from=MaterialExpressionConstant_13}
    Speed {from=MaterialExpressionConstant_14}
    Tile Subtraction Amount {from=MaterialExpressionConstant_15}
    Tile Subtraction Random Seed {from=MaterialExpressionConstant_16}
    OffsetGrid {from=""}
    Max Opacity {from=MaterialExpressionConstant_17}
    Min Opacity {from=MaterialExpressionConstant_18}
    Min Size {from=MaterialExpressionConstant_1}
    Corner Radius {from=MaterialExpressionConstant_20}
    Blurriness {from=MaterialExpressionConstant_2}
   ]
   properties {materialFunction=&MF_UI_PostMatch_AmbientSquareTiles}
  }
  MaterialExpressionMaterialFunctionCall_17 {
   class=MaterialExpressionMaterialFunctionCall
   inputs [
    UV {from=MaterialExpressionSubtract_2}
    Skew Angle {from=MaterialExpressionConvert_4}
   ]
   properties {materialFunction=/FortUILibrary/MaterialFunctions/Basics/MF_UI_Skew}
  }
  MaterialExpressionMaterialFunctionCall_18 {
   class=MaterialExpressionMaterialFunctionCall
   inputs [
    UVs {from=MaterialExpressionNamedRerouteUsage_15}
    Tile Count {from=MaterialExpressionConstant_31}
    Speed {from=MaterialExpressionConstant_32}
    Tile Subtraction Amount {from=MaterialExpressionConstant_33}
    Tile Subtraction Random Seed {from=MaterialExpressionConstant_34}
    OffsetGrid {from=""}
    Max Opacity {from=MaterialExpressionConstant_35}
    Min Opacity {from=MaterialExpressionConstant_36}
    Min Size {from=MaterialExpressionConstant_37}
    Corner Radius {from=MaterialExpressionConstant_38}
    Blurriness {from=MaterialExpressionConstant_39}
   ]
   properties {materialFunction=&MF_UI_PostMatch_AmbientSquareTiles}
  }
  MaterialExpressionMaterialFunctionCall_19 {
   class=MaterialExpressionMaterialFunctionCall
   inputs [
    UVs {from=MaterialExpressionNamedRerouteUsage_26}
    Tile Count {from=MaterialExpressionConstant_5}
    Speed {from=MaterialExpressionConstant_8}
    Tile Subtraction Amount {from=MaterialExpressionConstant_9}
    Tile Subtraction Random Seed {from=MaterialExpressionConstant_10}
    OffsetGrid {from=MaterialExpressionConstant_61}
    Max Opacity {from=MaterialExpressionConstant_11}
    Min Opacity {from=MaterialExpressionConstant_12}
    Min Size {from=MaterialExpressionConstant_49}
    Corner Radius {from=MaterialExpressionConstant_50}
    Blurriness {from=MaterialExpressionConstant_60}
   ]
   properties {materialFunction=&MF_UI_PostMatch_AmbientSquareTiles}
  }
  MaterialExpressionMaterialFunctionCall_2 {
   class=MaterialExpressionMaterialFunctionCall
   inputs [
    DistanceField {from=MaterialExpressionSubtract_4}
    Threshold {from=""}
    Smoothing {from=""}
   ]
   properties {materialFunction=&MF_UI_AntiAliasing}
  }
  MaterialExpressionMaterialFunctionCall_20 {
   class=MaterialExpressionMaterialFunctionCall
   inputs [
    UV {from=MaterialExpressionNamedRerouteUsage_10}
    TextureSize {from=""}
    ScalePivotPoint {from=""}
    ContainerSize {from=""}
   ]
   properties {materialFunction=&MF_UI_UVScaling}
  }
  MaterialExpressionMaterialFunctionCall_21 {
   class=MaterialExpressionMaterialFunctionCall
   inputs [
    UVs {from=MaterialExpressionNamedRerouteUsage_20}
    Tile Count {from=MaterialExpressionConstant_51}
    Speed {from=MaterialExpressionConstant_52}
    Tile Subtraction Amount {from=MaterialExpressionConstant_53}
    Tile Subtraction Random Seed {from=MaterialExpressionConstant_54}
    OffsetGrid {from=""}
    Max Opacity {from=MaterialExpressionConstant_55}
    Min Opacity {from=MaterialExpressionConstant_56}
    Min Size {from=MaterialExpressionConstant_57}
    Corner Radius {from=MaterialExpressionConstant_58}
    Blurriness {from=MaterialExpressionConstant_59}
   ]
   properties {materialFunction=&MF_UI_PostMatch_AmbientSquareTiles}
  }
  MaterialExpressionMaterialFunctionCall_22 {
   class=MaterialExpressionMaterialFunctionCall
   inputs [
    A {from=MaterialExpressionVectorParameter_4, output=RGB}
    B {from=MaterialExpressionVectorParameter_5, output=RGB}
    C {from=MaterialExpressionVectorParameter_3, output=RGB}
    Alpha {from=MaterialExpressionMaterialFunctionCall_12, output=EaseInSine}
   ]
   properties {materialFunction=&3ColorBlend}
  }
  MaterialExpressionMaterialFunctionCall_23 {
   class=MaterialExpressionMaterialFunctionCall
   inputs [
    DistanceFieldA {from=MaterialExpressionMaterialFunctionCall_27, output=SDF}
    DistanceFieldB {from=MaterialExpressionMaterialFunctionCall_28, output=SDF}
    BlendAmount {from=""}
   ]
   properties {materialFunction=/FortUILibrary/MaterialFunctions/Shapes/Modifiers/MF_UI_SDF_Union}
  }
  MaterialExpressionMaterialFunctionCall_24 {
   class=MaterialExpressionMaterialFunctionCall
   inputs [
    DistanceField {from=MaterialExpressionMaterialFunctionCall_23, output=Exact}
    Threshold {from=""}
    Smoothing {from=""}
   ]
   properties {materialFunction=&MF_UI_AntiAliasing}
  }
  MaterialExpressionMaterialFunctionCall_27 {
   class=MaterialExpressionMaterialFunctionCall
   inputs [
    UVs {from=MaterialExpressionAdd_6}
    Size {from=MaterialExpressionConstant_19}
    CornerRoundness {from=""}
    KeepAspectRatio {from=""}
    PixelSize {from=""}
   ]
   properties {materialFunction=/FortUILibrary/MaterialFunctions/Shapes/SDFs/MF_SDF_Box}
  }
  MaterialExpressionMaterialFunctionCall_28 {
   class=MaterialExpressionMaterialFunctionCall
   inputs [
    UVs {from=MaterialExpressionAdd_7}
    Size {from=MaterialExpressionConstant_21}
    CornerRoundness {from=""}
    KeepAspectRatio {from=""}
    PixelSize {from=""}
   ]
   properties {materialFunction=&MF_SDF_Box}
  }
  MaterialExpressionMaterialFunctionCall_3 {
   class=MaterialExpressionMaterialFunctionCall
   inputs [
    Screen UV {from=""}
    Tiling {from=MaterialExpressionReroute_4}
    Rotation {from=""}
    Main Min/Max Opacity {from=MaterialExpressionReroute_5}
    Main Min/Max Scale {from=MaterialExpressionReroute_6}
    Main Min/Max Perspective {from=MaterialExpressionReroute_10}
    Main Outline Thickness {from=MaterialExpressionConstant_63}
    Main Speed {from=MaterialExpressionReroute_12}
    Main Reduction {from=MaterialExpressionReroute_14}
    Sub Min/Max Opacity {from=MaterialExpressionReroute_16}
    Sub Min/Max Scale {from=MaterialExpressionReroute_20}
    Sub Min/Max Perspective {from=MaterialExpressionReroute_22}
    Sub Outline Thickness {from=MaterialExpressionConstant_64}
    Sub Speed {from=MaterialExpressionReroute_18}
    Sub Reduction {from=MaterialExpressionReroute_24}
    Random Seed {from=MaterialExpressionMultiply_6}
    Time {from=""}
   ]
   properties {materialFunction=/FortUILibrary/MaterialFunctions/BR/MF_UI_BR_StormTech}
  }
  MaterialExpressionMaterialFunctionCall_4 {
   class=MaterialExpressionMaterialFunctionCall
   inputs [
    UVs {from=MaterialExpressionNamedRerouteUsage_12}
    CenterPosition {from=""}
    Radius {from=MaterialExpressionScalarParameter_2}
    Density {from=""}
    Invert Density {from=""}
   ]
   properties {materialFunction=&RadialGradientExponential}
  }
  MaterialExpressionMaterialFunctionCall_5 {
   class=MaterialExpressionMaterialFunctionCall
   inputs [
    UV {from=""}
    TextureSize {from=""}
    ScalePivotPoint {from=""}
    ContainerSize {from=""}
   ]
   properties {materialFunction=&MF_UI_UVScaling}
  }
  MaterialExpressionMaterialFunctionCall_6 {
   class=MaterialExpressionMaterialFunctionCall
   inputs [
    UV {from=MaterialExpressionNamedRerouteUsage_3}
    GradientStart {from=MaterialExpressionConstant_7}
    GradientEnd {from=""}
   ]
   properties {materialFunction=&MF_UI_RadialGradient}
  }
  MaterialExpressionMaterialFunctionCall_7 {
   class=MaterialExpressionMaterialFunctionCall
   inputs [
    UVs {from=MaterialExpressionNamedRerouteUsage_7}
    Size {from=MaterialExpressionScalarParameter_36}
    OutlineThickness {from=""}
    OutlineAlignment {from=""}
    GlowMin {from=""}
    GlowMax {from=""}
   ]
   properties {materialFunction=&MF_UI_Circle}
  }
  MaterialExpressionMaterialFunctionCall_8 {
   class=MaterialExpressionMaterialFunctionCall
   inputs [
    Base {from=MaterialExpressionMaterialFunctionCall_22, output=Result}
    Blend {from=MaterialExpressionMultiply_9}
   ]
   properties {materialFunction=/Engine/Functions/Engine_MaterialFunctions03/Blends/Blend_SoftLight}
  }
  MaterialExpressionMaterialFunctionCall_9 {
   class=MaterialExpressionMaterialFunctionCall
   inputs [
    UVs {from=MaterialExpressionNamedRerouteUsage_13}
    Tile Count {from=MaterialExpressionConstant_22}
    Speed {from=MaterialExpressionConstant_4}
    Tile Subtraction Amount {from=MaterialExpressionConstant_24}
    Tile Subtraction Random Seed {from=MaterialExpressionConstant_25}
    OffsetGrid {from=""}
    Max Opacity {from=MaterialExpressionConstant_26}
    Min Opacity {from=MaterialExpressionConstant_27}
    Min Size {from=MaterialExpressionConstant_28}
    Corner Radius {from=MaterialExpressionConstant_29}
    Blurriness {from=MaterialExpressionConstant_30}
   ]
   properties {materialFunction=&MF_UI_PostMatch_AmbientSquareTiles}
  }
  MaterialExpressionMax_2 {
   class=MaterialExpressionMax
   inputs [
    A {from=MaterialExpressionMaterialFunctionCall_24, output=Inner}
    B {from=MaterialExpressionStep_0}
   ]
  }
  MaterialExpressionMin_0 {
   class=MaterialExpressionMin
   inputs [
    A {from=MaterialExpressionScalarParameter_17, output=""}
    B {from=MaterialExpressionScalarParameter_16, output=""}
   ]
   properties {}
  }
  MaterialExpressionMultiply_0 {
   class=MaterialExpressionMultiply
   inputs [
    A {from=MaterialExpressionMaterialFunctionCall_3, output=Grid}
    B {from=MaterialExpressionOneMinus_1}
   ]
  }
  MaterialExpressionMultiply_1 {
   class=MaterialExpressionMultiply
   inputs [
    A {from=MaterialExpressionComponentMask_0}
    B {from=MaterialExpressionScalarParameter_27}
   ]
   properties {constB=4}
  }
  MaterialExpressionMultiply_10 {
   class=MaterialExpressionMultiply
   inputs [
    A {from=MaterialExpressionLinearInterpolate_11}
    B {from=MaterialExpressionLinearInterpolate_18}
   ]
   properties {constB=0.5}
  }
  MaterialExpressionMultiply_11 {
   class=MaterialExpressionMultiply
   inputs [
    A {from=MaterialExpressionLinearInterpolate_6}
    B {from=MaterialExpressionScalarParameter_22}
   ]
  }
  MaterialExpressionMultiply_14 {
   class=MaterialExpressionMultiply
   inputs [
    A {from=MaterialExpressionTime_0}
    B {from=MaterialExpressionScalarParameter_3}
   ]
  }
  MaterialExpressionMultiply_2 {
   class=MaterialExpressionMultiply
   inputs [
    A {from=MaterialExpressionNamedRerouteUsage_9}
    B {from=MaterialExpressionScalarParameter_32}
   ]
  }
  MaterialExpressionMultiply_3 {
   class=MaterialExpressionMultiply
   inputs [
    A {from=MaterialExpressionNamedRerouteUsage_30}
    B {from=""}
   ]
   properties {constB=3}
  }
  MaterialExpressionMultiply_4 {
   class=MaterialExpressionMultiply
   inputs [
    A {from=MaterialExpressionVectorParameter_1, output=RGB}
    B {from=MaterialExpressionVectorParameter_1, output=RGB}
   ]
   properties {constB=0.1}
  }
  MaterialExpressionMultiply_5 {
   class=MaterialExpressionMultiply
   inputs [
    A {from=MaterialExpressionTime_1}
    B {from=MaterialExpressionScalarParameter_6}
   ]
   properties {constB=-0.1}
  }
  MaterialExpressionMultiply_6 {
   class=MaterialExpressionMultiply
   inputs [
    A {from=MaterialExpressionTime_2}
    B {from=MaterialExpressionScalarParameter_7}
   ]
  }
  MaterialExpressionMultiply_7 {
   class=MaterialExpressionMultiply
   inputs [
    A {from=MaterialExpressionNamedRerouteUsage_2}
    B {from=MaterialExpressionScalarParameter_35}
   ]
  }
  MaterialExpressionMultiply_8 {
   class=MaterialExpressionMultiply
   inputs [
    A {from=MaterialExpressionAdd_1}
    B {from=""}
   ]
   properties {constB=0.5}
  }
  MaterialExpressionMultiply_9 {
   class=MaterialExpressionMultiply
   inputs [
    A {from=MaterialExpressionVectorNoise_0}
    B {from=""}
   ]
   properties {constB=0.0333}
  }
  MaterialExpressionNamedRerouteDeclaration_0 {
   class=MaterialExpressionNamedRerouteDeclaration
   inputs [None {from=MaterialExpressionMultiply_0}]
   properties {name=Grid Mask, nodeColor=<rgba 1 0 0.2578 1>}
  }
  MaterialExpressionNamedRerouteDeclaration_1 {
   class=MaterialExpressionNamedRerouteDeclaration
   inputs [None {from=MaterialExpressionTextureCoordinate_3}]
   properties {name=UV, nodeColor=<rgba 0.2656 1 0 1>}
  }
  MaterialExpressionNamedRerouteDeclaration_10 {
   class=MaterialExpressionNamedRerouteDeclaration
   inputs [None {from=MaterialExpressionMaterialFunctionCall_18, output=Result}]
   properties {name=Small Cells, nodeColor=<rgba 0 0 0 1>}
  }
  MaterialExpressionNamedRerouteDeclaration_11 {
   class=MaterialExpressionNamedRerouteDeclaration
   inputs [None {from=MaterialExpressionMaterialFunctionCall_11, output=Result}]
   properties {name=XSmall Cells, nodeColor=<rgba 0 0 0 1>}
  }
  MaterialExpressionNamedRerouteDeclaration_12 {
   class=MaterialExpressionNamedRerouteDeclaration
   inputs [None {from=MaterialExpressionMaterialFunctionCall_2, output=Inner}]
   properties {name=Ripples Mask, nodeColor=<rgba 0 0.8359 1 1>}
  }
  MaterialExpressionNamedRerouteDeclaration_13 {
   class=MaterialExpressionNamedRerouteDeclaration
   inputs [None {from=MaterialExpressionMax_2}]
   properties {name=Backplate Mask, nodeColor=<rgba 0 1 0.4609 1>}
  }
  MaterialExpressionNamedRerouteDeclaration_14 {
   class=MaterialExpressionNamedRerouteDeclaration
   inputs [None {from=MaterialExpressionLinearInterpolate_8}]
   properties {name=Background, nodeColor=<rgba 1 0.2109 0 1>}
  }
  MaterialExpressionNamedRerouteDeclaration_15 {
   class=MaterialExpressionNamedRerouteDeclaration
   inputs [None {from=MaterialExpressionMaterialFunctionCall_22, output=Result}]
   properties {name=FinalColor, nodeColor=<rgba 0 0.8359 1 1>}
  }
  MaterialExpressionNamedRerouteDeclaration_16 {
   class=MaterialExpressionNamedRerouteDeclaration
   inputs [None {from=MaterialExpressionLinearInterpolate_19}]
   properties {name=Grid, nodeColor=<rgba 0.5469 0 1 1>}
  }
  MaterialExpressionNamedRerouteDeclaration_17 {
   class=MaterialExpressionNamedRerouteDeclaration
   inputs [None {from=MaterialExpressionLinearInterpolate_3}]
   properties {name=FinalAlpha, nodeColor=<rgba 1 0.8438 0 1>}
  }
  MaterialExpressionNamedRerouteDeclaration_18 {
   class=MaterialExpressionNamedRerouteDeclaration
   inputs [None {from=MaterialExpressionMaterialFunctionCall_21, output=Result}]
   properties {name="XXSmall  Cells", nodeColor=<rgba 0 0 0 1>}
  }
  MaterialExpressionNamedRerouteDeclaration_2 {
   class=MaterialExpressionNamedRerouteDeclaration
   inputs [None {from=MaterialExpressionSaturate_0}]
   properties {name=Lower Glow Mask, nodeColor=<rgba 0 1 0.2969 1>}
  }
  MaterialExpressionNamedRerouteDeclaration_3 {
   class=MaterialExpressionNamedRerouteDeclaration
   inputs [None {from=MaterialExpressionMaterialFunctionCall_6, output=Linear}]
   properties {name=Background Mask, nodeColor=<rgba 0 0.1094 1 1>}
  }
  MaterialExpressionNamedRerouteDeclaration_4 {
   class=MaterialExpressionNamedRerouteDeclaration
   inputs [None {from=MaterialExpressionMaterialFunctionCall_10, output=Smooth}]
   properties {name=GlowMask, nodeColor=<rgba 0.4297 1 0 1>}
  }
  MaterialExpressionNamedRerouteDeclaration_5 {
   class=MaterialExpressionNamedRerouteDeclaration
   inputs [None {from=MaterialExpressionMaterialFunctionCall_20, output=Scaled to Fit}]
   properties {name=Scaled UV, nodeColor=<rgba 0.3594 1 0 1>}
  }
  MaterialExpressionNamedRerouteDeclaration_6 {
   class=MaterialExpressionNamedRerouteDeclaration
   inputs [None {from=MaterialExpressionSaturate_1}]
   properties {name=GradientMask, nodeColor=<rgba 0 0 0 1>}
  }
  MaterialExpressionNamedRerouteDeclaration_7 {
   class=MaterialExpressionNamedRerouteDeclaration
   inputs [None {from=MaterialExpressionLinearInterpolate_23}]
   properties {name=Large Cells, nodeColor=<rgba 0 0 0 1>}
  }
  MaterialExpressionNamedRerouteDeclaration_8 {
   class=MaterialExpressionNamedRerouteDeclaration
   inputs [None {from=MaterialExpressionMaterialFunctionCall_5, output=ScaleY}]
   properties {name=ScaledUVs, nodeColor=<rgba 0 0 0 1>}
  }
  MaterialExpressionNamedRerouteDeclaration_9 {
   class=MaterialExpressionNamedRerouteDeclaration
   inputs [None {from=MaterialExpressionMaterialFunctionCall_9, output=Result}]
   properties {name=Medium Cells, nodeColor=<rgba 0 0 0 1>}
  }
  MaterialExpressionNamedRerouteUsage_0 {class=MaterialExpressionNamedRerouteUsage}
  MaterialExpressionNamedRerouteUsage_1 {class=MaterialExpressionNamedRerouteUsage}
  MaterialExpressionNamedRerouteUsage_10 {class=MaterialExpressionNamedRerouteUsage}
  MaterialExpressionNamedRerouteUsage_11 {class=MaterialExpressionNamedRerouteUsage}
  MaterialExpressionNamedRerouteUsage_12 {class=MaterialExpressionNamedRerouteUsage}
  MaterialExpressionNamedRerouteUsage_13 {class=MaterialExpressionNamedRerouteUsage}
  MaterialExpressionNamedRerouteUsage_14 {class=MaterialExpressionNamedRerouteUsage}
  MaterialExpressionNamedRerouteUsage_15 {class=MaterialExpressionNamedRerouteUsage}
  MaterialExpressionNamedRerouteUsage_16 {class=MaterialExpressionNamedRerouteUsage}
  MaterialExpressionNamedRerouteUsage_17 {class=MaterialExpressionNamedRerouteUsage}
  MaterialExpressionNamedRerouteUsage_18 {class=MaterialExpressionNamedRerouteUsage}
  MaterialExpressionNamedRerouteUsage_19 {class=MaterialExpressionNamedRerouteUsage}
  MaterialExpressionNamedRerouteUsage_2 {class=MaterialExpressionNamedRerouteUsage}
  MaterialExpressionNamedRerouteUsage_20 {class=MaterialExpressionNamedRerouteUsage}
  MaterialExpressionNamedRerouteUsage_21 {class=MaterialExpressionNamedRerouteUsage}
  MaterialExpressionNamedRerouteUsage_22 {class=MaterialExpressionNamedRerouteUsage}
  MaterialExpressionNamedRerouteUsage_23 {class=MaterialExpressionNamedRerouteUsage}
  MaterialExpressionNamedRerouteUsage_24 {class=MaterialExpressionNamedRerouteUsage}
  MaterialExpressionNamedRerouteUsage_25 {class=MaterialExpressionNamedRerouteUsage}
  MaterialExpressionNamedRerouteUsage_26 {class=MaterialExpressionNamedRerouteUsage}
  MaterialExpressionNamedRerouteUsage_29 {class=MaterialExpressionNamedRerouteUsage}
  MaterialExpressionNamedRerouteUsage_3 {class=MaterialExpressionNamedRerouteUsage}
  MaterialExpressionNamedRerouteUsage_30 {class=MaterialExpressionNamedRerouteUsage}
  MaterialExpressionNamedRerouteUsage_32 {class=MaterialExpressionNamedRerouteUsage}
  MaterialExpressionNamedRerouteUsage_35 {class=MaterialExpressionNamedRerouteUsage}
  MaterialExpressionNamedRerouteUsage_36 {class=MaterialExpressionNamedRerouteUsage}
  MaterialExpressionNamedRerouteUsage_37 {class=MaterialExpressionNamedRerouteUsage}
  MaterialExpressionNamedRerouteUsage_38 {class=MaterialExpressionNamedRerouteUsage}
  MaterialExpressionNamedRerouteUsage_39 {class=MaterialExpressionNamedRerouteUsage}
  MaterialExpressionNamedRerouteUsage_4 {class=MaterialExpressionNamedRerouteUsage}
  MaterialExpressionNamedRerouteUsage_41 {class=MaterialExpressionNamedRerouteUsage}
  MaterialExpressionNamedRerouteUsage_42 {class=MaterialExpressionNamedRerouteUsage}
  MaterialExpressionNamedRerouteUsage_43 {class=MaterialExpressionNamedRerouteUsage}
  MaterialExpressionNamedRerouteUsage_5 {class=MaterialExpressionNamedRerouteUsage}
  MaterialExpressionNamedRerouteUsage_6 {class=MaterialExpressionNamedRerouteUsage}
  MaterialExpressionNamedRerouteUsage_7 {class=MaterialExpressionNamedRerouteUsage}
  MaterialExpressionNamedRerouteUsage_8 {class=MaterialExpressionNamedRerouteUsage}
  MaterialExpressionNamedRerouteUsage_9 {class=MaterialExpressionNamedRerouteUsage}
  MaterialExpressionOneMinus_0 {class=MaterialExpressionOneMinus, inputs [None {from=MaterialExpressionMultiply_8}]}
  MaterialExpressionOneMinus_1 {class=MaterialExpressionOneMinus, inputs [None {from=MaterialExpressionMaterialFunctionCall_4, output=RadialGradientExponential}]}
  MaterialExpressionReroute_0 {class=MaterialExpressionReroute, inputs [None {from=MaterialExpressionMultiply_4}]}
  MaterialExpressionReroute_1 {class=MaterialExpressionReroute, inputs [None {from=MaterialExpressionMaterialFunctionCall_17, output=Skewed UVs}]}
  MaterialExpressionReroute_10 {class=MaterialExpressionReroute, inputs [None {from=MaterialExpressionReroute_9}]}
  MaterialExpressionReroute_11 {class=MaterialExpressionReroute, inputs [None {from=MaterialExpressionConstant_23}]}
  MaterialExpressionReroute_12 {class=MaterialExpressionReroute, inputs [None {from=MaterialExpressionReroute_11}]}
  MaterialExpressionReroute_13 {class=MaterialExpressionReroute, inputs [None {from=MaterialExpressionScalarParameter_9}]}
  MaterialExpressionReroute_14 {class=MaterialExpressionReroute, inputs [None {from=MaterialExpressionReroute_13}]}
  MaterialExpressionReroute_15 {class=MaterialExpressionReroute, inputs [None {from=MaterialExpressionConstant2Vector_5}]}
  MaterialExpressionReroute_16 {class=MaterialExpressionReroute, inputs [None {from=MaterialExpressionReroute_15}]}
  MaterialExpressionReroute_17 {class=MaterialExpressionReroute, inputs [None {from=MaterialExpressionConstant_3}]}
  MaterialExpressionReroute_18 {class=MaterialExpressionReroute, inputs [None {from=MaterialExpressionReroute_17}]}
  MaterialExpressionReroute_19 {class=MaterialExpressionReroute, inputs [None {from=MaterialExpressionConstant2Vector_4}]}
  MaterialExpressionReroute_2 {class=MaterialExpressionReroute, inputs [None {from=MaterialExpressionReroute_0}]}
  MaterialExpressionReroute_20 {class=MaterialExpressionReroute, inputs [None {from=MaterialExpressionReroute_19}]}
  MaterialExpressionReroute_21 {class=MaterialExpressionReroute, inputs [None {from=MaterialExpressionConstant_0}]}
  MaterialExpressionReroute_22 {class=MaterialExpressionReroute, inputs [None {from=MaterialExpressionReroute_21}]}
  MaterialExpressionReroute_23 {class=MaterialExpressionReroute, inputs [None {from=MaterialExpressionScalarParameter_8}]}
  MaterialExpressionReroute_24 {class=MaterialExpressionReroute, inputs [None {from=MaterialExpressionReroute_23}]}
  MaterialExpressionReroute_3 {class=MaterialExpressionReroute, inputs [None {from=MaterialExpressionLinearInterpolate_20}]}
  MaterialExpressionReroute_4 {class=MaterialExpressionReroute, inputs [None {from=MaterialExpressionDivide_0}]}
  MaterialExpressionReroute_5 {class=MaterialExpressionReroute, inputs [None {from=MaterialExpressionReroute_7}]}
  MaterialExpressionReroute_6 {class=MaterialExpressionReroute, inputs [None {from=MaterialExpressionReroute_8}]}
  MaterialExpressionReroute_7 {class=MaterialExpressionReroute, inputs [None {from=MaterialExpressionConstant2Vector_3}]}
  MaterialExpressionReroute_8 {class=MaterialExpressionReroute, inputs [None {from=MaterialExpressionConstant2Vector_1}]}
  MaterialExpressionReroute_9 {class=MaterialExpressionReroute, inputs [None {from=MaterialExpressionConstant_6}]}
  MaterialExpressionSaturate_0 {class=MaterialExpressionSaturate, inputs [None {from=MaterialExpressionInverseLinearInterpolate_0}]}
  MaterialExpressionSaturate_1 {class=MaterialExpressionSaturate, inputs [None {from=MaterialExpressionOneMinus_0}]}
  MaterialExpressionScalarParameter_0 {class=MaterialExpressionScalarParameter, properties {group=Ripples, parameterName=Limit Ripple Count, sliderMax=1}}
  MaterialExpressionScalarParameter_1 {class=MaterialExpressionScalarParameter, properties {defaultValue=16, group=Pattern, parameterName=Tiling}}
  MaterialExpressionScalarParameter_10 {class=MaterialExpressionScalarParameter, properties {defaultValue=0.2, group=Backplate, parameterName=Backplate Y Offset}}
  MaterialExpressionScalarParameter_11 {class=MaterialExpressionScalarParameter, properties {defaultValue=0.2, group=Backplate, parameterName=Backplate Y Offset}}
  MaterialExpressionScalarParameter_12 {class=MaterialExpressionScalarParameter, properties {defaultValue=0.4, group=Backplate, parameterName=Left Fin X Offset}}
  MaterialExpressionScalarParameter_13 {class=MaterialExpressionScalarParameter, properties {group=Backplate, parameterName=Left Fin Y Offset}}
  MaterialExpressionScalarParameter_14 {class=MaterialExpressionScalarParameter, properties {defaultValue=-0.5, group=Backplate, parameterName=Right Fin X Offset}}
  MaterialExpressionScalarParameter_15 {class=MaterialExpressionScalarParameter, properties {defaultValue=-0.1, group=Backplate, parameterName=Right Fin Y Offset}}
  MaterialExpressionScalarParameter_16 {class=MaterialExpressionScalarParameter, properties {defaultValue=-0.1, group=Backplate, parameterName=Right Fin Y Offset}}
  MaterialExpressionScalarParameter_17 {class=MaterialExpressionScalarParameter, properties {group=Backplate, parameterName=Left Fin Y Offset}}
  MaterialExpressionScalarParameter_18 {class=MaterialExpressionScalarParameter, properties {defaultValue=20, group=Backplate, parameterName=Backplate X Skew}}
  MaterialExpressionScalarParameter_19 {class=MaterialExpressionScalarParameter, properties {defaultValue=4, group=Backplate, parameterName=Backplate Y Skew}}
  MaterialExpressionScalarParameter_2 {class=MaterialExpressionScalarParameter, properties {defaultValue=1, group=Pattern, parameterName=Grid Mask Radius}}
  MaterialExpressionScalarParameter_20 {class=MaterialExpressionScalarParameter, properties {defaultValue=1, group=Glow, parameterName=Glow Opacity, sliderMax=1}}
  MaterialExpressionScalarParameter_21 {class=MaterialExpressionScalarParameter, properties {defaultValue=1, group=Backplate, parameterName=Backplate Opacity, sliderMax=1}}
  MaterialExpressionScalarParameter_22 {class=MaterialExpressionScalarParameter, properties {defaultValue=1, group=Background, parameterName=Background Opacity, sliderMax=1}}
  MaterialExpressionScalarParameter_23 {class=MaterialExpressionScalarParameter, properties {defaultValue=1, group=Pattern, parameterName=Grid Opacity, sliderMax=1}}
  MaterialExpressionScalarParameter_24 {class=MaterialExpressionScalarParameter, properties {defaultValue=1, group=Pattern, parameterName=Grid Opacity, sliderMax=1}}
  MaterialExpressionScalarParameter_25 {class=MaterialExpressionScalarParameter, properties {defaultValue=0.5, group=Background, parameterName=Background Glow Radius}}
  MaterialExpressionScalarParameter_26 {class=MaterialExpressionScalarParameter, properties {defaultValue=0.05, group=Ripples, parameterName=Ripple Thickness}}
  MaterialExpressionScalarParameter_27 {class=MaterialExpressionScalarParameter, properties {defaultValue=4, group=Ripples, parameterName=Ripple Spacing}}
  MaterialExpressionScalarParameter_28 {class=MaterialExpressionScalarParameter, properties {defaultValue=2.33, group=Background, parameterName=Background Glow Density}}
  MaterialExpressionScalarParameter_29 {class=MaterialExpressionScalarParameter, properties {defaultValue=4, group=Ripples, parameterName=Ripple End}}
  MaterialExpressionScalarParameter_3 {class=MaterialExpressionScalarParameter, properties {defaultValue=0.75, group=Glow, parameterName=Glow Pulse Speed}}
  MaterialExpressionScalarParameter_30 {class=MaterialExpressionScalarParameter, properties {group=Ripples, parameterName=Ripple Start}}
  MaterialExpressionScalarParameter_31 {class=MaterialExpressionScalarParameter, properties {group=Ripples, parameterName=Limit Ripple Count, sliderMax=1}}
  MaterialExpressionScalarParameter_32 {class=MaterialExpressionScalarParameter, properties {defaultValue=1, group=Ripples, parameterName=Ripple Opacity, sliderMax=1}}
  MaterialExpressionScalarParameter_33 {class=MaterialExpressionScalarParameter, properties {group=Glow, parameterName=Lower Glow Start}}
  MaterialExpressionScalarParameter_34 {class=MaterialExpressionScalarParameter, properties {defaultValue=1, group=Glow, parameterName=Lower Glow End}}
  MaterialExpressionScalarParameter_35 {class=MaterialExpressionScalarParameter, properties {defaultValue=1, group=Glow, parameterName=Lower Glow Opacity, sliderMax=1}}
  MaterialExpressionScalarParameter_36 {class=MaterialExpressionScalarParameter, properties {defaultValue=1, parameterName=RadialGradient_Radius}}
  MaterialExpressionScalarParameter_37 {class=MaterialExpressionScalarParameter, properties {defaultValue=0.8, parameterName=Opacity}}
  MaterialExpressionScalarParameter_38 {class=MaterialExpressionScalarParameter, properties {parameterName=Anim_TileBrightness}}
  MaterialExpressionScalarParameter_39 {class=MaterialExpressionScalarParameter, properties {parameterName=UsePatternGrid}}
  MaterialExpressionScalarParameter_4 {class=MaterialExpressionScalarParameter, properties {defaultValue=0.1, group=Ripples, parameterName=Ripple Phase}}
  MaterialExpressionScalarParameter_5 {class=MaterialExpressionScalarParameter, properties {group=Circle, parameterName=Circle Y Offset}}
  MaterialExpressionScalarParameter_6 {class=MaterialExpressionScalarParameter, properties {defaultValue=-0.075, group=Ripples, parameterName=Ripple Speed}}
  MaterialExpressionScalarParameter_7 {class=MaterialExpressionScalarParameter, properties {defaultValue=0.125, group=Pattern, parameterName=Grid Speed}}
  MaterialExpressionScalarParameter_8 {class=MaterialExpressionScalarParameter, properties {defaultValue=8, group=Pattern, parameterName=Sub Reduction}}
  MaterialExpressionScalarParameter_9 {class=MaterialExpressionScalarParameter, properties {defaultValue=0.5, group=Pattern, parameterName=Main Reduction}}
  MaterialExpressionSine_0 {class=MaterialExpressionSine, inputs [None {from=MaterialExpressionMultiply_14}]}
  MaterialExpressionSine_1 {class=MaterialExpressionSine, inputs [None {from=MaterialExpressionLinearInterpolate_0}]}
  MaterialExpressionStaticSwitchParameter_2 {
   class=MaterialExpressionStaticSwitchParameter
   inputs [
    True {from=MaterialExpressionLinearInterpolate_16, output=""}
    False {from=MaterialExpressionLinearInterpolate_21, output=""}
   ]
   properties {group=Backplate, parameterName=Show Backplate}
  }
  MaterialExpressionStaticSwitchParameter_3 {
   class=MaterialExpressionStaticSwitchParameter
   inputs [
    True {from=MaterialExpressionLinearInterpolate_17, output=""}
    False {from=MaterialExpressionReroute_3, output=""}
   ]
   properties {group=Backplate, parameterName=Show Backplate}
  }
  MaterialExpressionStep_0 {
   class=MaterialExpressionStep
   inputs [
    Y {from=MaterialExpressionSubtract_3, output=""}
    X {from=MaterialExpressionComponentMask_2, output=""}
   ]
   properties {constX=0.2, constY=0.2}
  }
  MaterialExpressionSubtract_0 {
   class=MaterialExpressionSubtract
   inputs [
    A {from=MaterialExpressionNamedRerouteUsage_4}
    B {from=MaterialExpressionConvert_0}
   ]
  }
  MaterialExpressionSubtract_2 {
   class=MaterialExpressionSubtract
   inputs [
    A {from=MaterialExpressionNamedRerouteUsage_14}
    B {from=MaterialExpressionConvert_1}
   ]
  }
  MaterialExpressionSubtract_3 {
   class=MaterialExpressionSubtract
   inputs [
    A {from=MaterialExpressionScalarParameter_11}
    B {from=MaterialExpressionMin_0}
   ]
  }
  MaterialExpressionSubtract_4 {
   class=MaterialExpressionSubtract
   inputs [
    A {from=MaterialExpressionConstantBiasScale_0}
    B {from=MaterialExpressionScalarParameter_26}
   ]
   properties {constB=0.5}
  }
  MaterialExpressionTextureCoordinate_3 {class=MaterialExpressionTextureCoordinate}
  MaterialExpressionTime_0 {class=MaterialExpressionTime}
  MaterialExpressionTime_1 {class=MaterialExpressionTime}
  MaterialExpressionTime_2 {class=MaterialExpressionTime}
  MaterialExpressionVectorNoise_0 {class=MaterialExpressionVectorNoise, inputs [World Position {from="", output=""}], properties {}}
  MaterialExpressionVectorParameter_0 {
   class=MaterialExpressionVectorParameter
   properties {group=Backplate, parameterName=Backplate Color, defaultValue=<rgba 0.0204 0.0036 0.0906 1>}
  }
  MaterialExpressionVectorParameter_1 {
   class=MaterialExpressionVectorParameter
   properties {group=Background, parameterName=Background Glow Color, defaultValue=<rgba 0.2121 0.6403 1 1>}
  }
  MaterialExpressionVectorParameter_10 {class=MaterialExpressionVectorParameter, properties {parameterName=BrightColor, defaultValue=<rgba 0.0531 0.0125 0.5 1>}}
  MaterialExpressionVectorParameter_2 {class=MaterialExpressionVectorParameter, properties {group=Glow, parameterName=Lower Glow Color, defaultValue=<rgba 0.0518 0.015 0.75 1>}}
  MaterialExpressionVectorParameter_3 {class=MaterialExpressionVectorParameter, properties {parameterName=BrightColor, defaultValue=<rgba 0.0531 0.0125 0.5 1>}}
  MaterialExpressionVectorParameter_4 {class=MaterialExpressionVectorParameter, properties {parameterName=FalloffColor, defaultValue=<rgba 0 0 0.05 1>}}
  MaterialExpressionVectorParameter_5 {class=MaterialExpressionVectorParameter, properties {parameterName=MidColor, defaultValue=<rgba 0.0164 0.0055 0.222 1>}}
  MaterialExpressionVectorParameter_6 {
   class=MaterialExpressionVectorParameter
   properties {group=Background, parameterName=Background Outer Color, defaultValue=<rgba 0.0052 0.0012 0.0409 1>}
  }
  MaterialExpressionVectorParameter_7 {
   class=MaterialExpressionVectorParameter
   properties {group=Background, parameterName=Background Inner Color, defaultValue=<rgba 0.003 0.0048 0.0648 1>}
  }
  MaterialExpressionVectorParameter_8 {class=MaterialExpressionVectorParameter, properties {group=Glow, parameterName=Glow Color, defaultValue=<rgba 0.0206 0.0125 0.5 1>}}
  MaterialExpressionVectorParameter_9 {class=MaterialExpressionVectorParameter, properties {group=Pattern, parameterName=Grid Color, defaultValue=<rgba 0.2159 0 1 1>}}
 }
 properties {
  bAutomaticallySetUsageInEditor=false
  bUsedWithStaticMesh=false
  blendMode=BLEND_Translucent
  materialDomain=MD_UI
  usedShadingModels=Default Lit
 }
}
