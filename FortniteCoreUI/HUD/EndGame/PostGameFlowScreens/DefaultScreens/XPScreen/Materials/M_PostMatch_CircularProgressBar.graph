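# UI material for the circular progress bar on the post-match XP screen, with animated progress and delta change between start and end values.
# An antialiased SDF ring is masked by radial arc fills, with rotated SDF circles forming rounded and leading caps.
# Parameter-driven lerps toggle segments, step snapping, outline, glow and inner circle fill.
# Color comes from gradient, delta and empty-part colors, composited with alpha composite blending.
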Material /FortniteCoreUI/HUD/EndGame/PostGameFlowScreens/DefaultScreens/XPScreen/Materials/M_PostMatch_CircularProgressBar
graph {
 properties {blendMode=BLEND_AlphaComposite, materialDomain=MD_UI, pixelDepthOffsetMode=PDOM_Legacy, usedShadingModels=Default Lit}
 nodes [
  n1 = NamedRerouteUsage()
  n2 = OneMinus(None=Divide(A=@n1, B=NamedRerouteUsage()))
  n3 = Saturate(None=MaterialFunctionCall(DistanceField=NamedRerouteUsage()) {materialFunction=&MF_UI_AntiAliasing}.Inner)
  Add(A=@n2, B=@n3)
  n4 = NamedRerouteUsage()
  n5 = NamedRerouteUsage()
  n6 = LinearInterpolate(A=@n4, B=@n5, Alpha=NamedRerouteUsage())
  n7 = ScalarParameter() {
   defaultValue=1
   desc="Will show rounded cap for start and end edge of the progress bar. Not support delta functionality and snap to segments.\r\nbool"
   group=Progress_Layout
   parameterName=UseRoundedProgress
   sliderMax=1
   sortPriority=2
  }
  LinearInterpolate(A=@n5, B=@n6, Alpha=@n7) {constA=1}
  n8 = VectorParameter() {
   desc="Color for empty part of progress bar.\r\nAlpha defines opacity for empty part of progress bar."
   group=Color
   parameterName=EmptyColor
   sortPriority=10
   defaultValue=<rgba 0.3924 0.3924 0.3924 0.3>
  }
  n9 = ScalarParameter() {defaultValue=1, desc="Alpha for filled part of progress bar.\r\n0 -1", group=Color, parameterName=FillAlpha, sliderMax=1}
  n10 = Max(A=NamedRerouteUsage(), B=NamedRerouteUsage())
  n11 = MaterialFunctionCall(DistanceField=NamedRerouteUsage()) {materialFunction=&MF_UI_AntiAliasing}
  n12 = ScalarParameter() {
   defaultValue=1
   desc="Will show rounded cap for end edge of the progress bar.\r\nbool"
   group=Progress_Layout
   parameterName=UseLeadingCircle
   sliderMax=1
  }
  n13 = LinearInterpolate(A=LinearInterpolate(B=@n8.A, Alpha=NamedRerouteUsage()), B=@n9, Alpha=LinearInterpolate(A=@n10, B=Max(A=@n10, B=@n11.Inner), Alpha=@n12))
  n14 = NamedRerouteUsage()
  n15 = LinearInterpolate(
   B=LinearInterpolate(
    A=@n13
    B=Multiply(A=@n13, B=@n14) {constA=1}
    Alpha=ScalarParameter() {desc="Enable segments functionality.\r\nbool", group=Segments, parameterName=UseSegments, sliderMax=1, sortPriority=1}
   )
   Alpha=NamedRerouteUsage()
  )
  n16 = VectorParameter() {group=Behind Circle, parameterName=BehindCircleColor, sortPriority=2, defaultValue=<rgba 0.1493 0.1493 0.1493 0.1>}
  n17 = Add(
   A=@n16.A
   B=Multiply(A=VectorParameter() {group=Behind Circle, parameterName=BehindGlowColor, defaultValue=<rgba 0.3672 0.3124 1 1>}.A, B=NamedRerouteUsage())
  )
  n18 = ScalarParameter() {
   desc="Enables inner circle fill for the circular progress bar.\r\nbool"
   group=Behind Circle
   parameterName=UseBehindCircle
   sliderMax=1
   sortPriority=1
  }
  n19 = LinearInterpolate(A=@n15, B=LinearInterpolate(A=@n15, B=Saturate(None=@n17), Alpha=NamedRerouteUsage()), Alpha=@n18)
  n20 = MaterialFunctionCall("Value (0-1)"=NamedRerouteUsage()) {materialFunction=&MF_UI_EaseCurves}
  n21 = Multiply(A=OneMinus(None=ScalarParameter() {defaultValue=0.75, group=Glow, parameterName=GlowAdditiveRatio}), B=@n20.EaseInCubic)
  n22 = ScalarParameter() {desc="Enables glow around progress fill part.\r\nbool", group=Glow, parameterName=UseProgressGlow, sliderMax=1, sortPriority=1}
  LinearInterpolate(A=@n19, B=Saturate(None=Add(A=@n19, B=@n21)), Alpha=@n22)
  n23 = NamedRerouteUsage()
  n24 = VectorParameter() {desc=Start color in gradient, group=Color, parameterName=ColorA, sortPriority=1, defaultValue=<rgba 0 0 1 1>}
  n25 = VectorParameter() {desc=Central color in gradient, group=Color, parameterName=ColorB, sortPriority=2, defaultValue=<rgba 0 1 0 1>}
  n26 = VectorParameter() {desc=End color in gradient, group=Color, parameterName=ColorC, sortPriority=3, defaultValue=<rgba 1 0 0 1>}
  n27 = OneMinus(None=NamedRerouteUsage())
  n28 = Min(A=Round(None=@n27), B=@n3)
  n29 = MaterialFunctionCall("Value (0-1)"=NamedRerouteUsage()) {materialFunction=&MF_UI_EaseCurves}
  n30 = LinearInterpolate(
   A=MaterialFunctionCall(A=@n24.RGB, B=@n25.RGB, C=@n26.RGB, Alpha=Subtract(A=@n27, B=@n28)) {materialFunction=&3ColorBlend}.Result
   B=@n26.RGB
   Alpha=NamedRerouteDeclaration(None=Saturate(None=@n29.EaseInSine)) {name=ProgressCurved, nodeColor=<rgba 0 0 0 1>}
  )
  n31 = VectorParameter() {desc=Color for delta part of progress bar., group=Delta, parameterName=DeltaColor, sortPriority=20, defaultValue=<rgba 1 0 0.829 1>}
  n32 = VectorParameter() {group=Delta, parameterName=DeltaColorDecay, sortPriority=21, defaultValue=<rgba 0 1 1 1>}
  n33 = ScalarParameter() {
   desc="Property for animating Delta color decay\r\n(from DeltaColor to DeltaColorDecay)\r\n0 - 1"
   group=Delta
   parameterName=AnimateDeltaColor
   sliderMax=1
   sortPriority=34
  }
  n34 = NamedRerouteUsage()
  n35 = OneMinus(None=Max(A=@n34, B=NamedRerouteUsage()))
  n36 = NamedRerouteUsage()
  n37 = LinearInterpolate(
   A=LinearInterpolate(
    A=NamedRerouteDeclaration(None=@n30) {name=ArcGradientColor, nodeColor=<rgba 0 0 0 1>}
    B=NamedRerouteUsage()
    Alpha=ScalarParameter() {group=Color, parameterName=UseColorGradients, sliderMax=1}
   )
   B=LinearInterpolate(
    A=LinearInterpolate(A=@n31.RGB, B=@n32.RGB, Alpha=@n33)
    B=NamedRerouteUsage()
    Alpha=ScalarParameter() {group=Color, parameterName=UseColorGradients, sliderMax=1}
   )
   Alpha=LinearInterpolate(A=@n35, B=Multiply(A=@n36, B=@n35), Alpha=@n36)
  ) {constAlpha=0}
  n38 = NamedRerouteUsage()
  n39 = ScalarParameter() {
   defaultValue=1
   desc="Enables GlowColor for the glow effect, otherwise will use progress fill color.\r\nbool"
   group=Glow
   parameterName=UseGlowColor
   sliderMax=1
   sortPriority=2
  }
  n40 = LinearInterpolate(
   A=@n37
   B=LinearInterpolate(A=LinearInterpolate(A=@n24.RGB, B=@n25.RGB, Alpha=@n38), B=LinearInterpolate(A=@n25.RGB, B=@n26.RGB, Alpha=@n38), Alpha=NamedRerouteUsage())
   Alpha=@n39
  )
  n41 = NamedRerouteUsage()
  n42 = NamedRerouteUsage()
  n43 = NamedRerouteUsage()
  n44 = ScalarParameter() {
   desc="Enables inner circle fill for the circular progress bar.\r\nbool"
   group=Behind Circle
   parameterName=UseBehindCircle
   sliderMax=1
   sortPriority=1
  }
  n45 = LinearInterpolate(A=Ceil(None=@n41), B=@n41, Alpha=LinearInterpolate(A=@n42, B=Max(A=@n42, B=@n43), Alpha=@n44))
  n46 = ScalarParameter() {desc="Enables glow around progress fill part.\r\nbool", group=Glow, parameterName=UseProgressGlow, sliderMax=1, sortPriority=1}
  n47 = Max(A=NamedRerouteUsage(), B=NamedRerouteUsage())
  n48 = NamedRerouteUsage()
  n49 = LinearInterpolate(
   A=LinearInterpolate(A=@n23, B=LinearInterpolate(A=@n23, B=@n40, Alpha=@n45), Alpha=@n46)
   B=@n37
   Alpha=LinearInterpolate(
    A=@n47
    B=Multiply(A=@n47, B=@n48) {constA=1}
    Alpha=ScalarParameter() {desc="Enable segments functionality.\r\nbool", group=Segments, parameterName=UseSegments, sliderMax=1, sortPriority=1}
   )
  )
  n50 = VectorParameter() {desc=Start color in gradient, group=Color, parameterName=ColorA, sortPriority=1, defaultValue=<rgba 0 0 1 1>}
  n51 = VectorParameter() {desc=Central color in gradient, group=Color, parameterName=ColorB, sortPriority=2, defaultValue=<rgba 0 1 0 1>}
  n52 = VectorParameter() {desc=End color in gradient, group=Color, parameterName=ColorC, sortPriority=3, defaultValue=<rgba 1 0 0 1>}
  n53 = SmoothStep(Max=NamedRerouteUsage(), Value=NamedRerouteUsage()) {constMax=0.5}
  n54 = MaterialFunctionCall(
   A=@n50.RGB
   B=@n51.RGB
   C=@n52.RGB
   Alpha=NamedRerouteDeclaration(None=@n53) {name=CurrentProgressArcMask_0-1, nodeColor=<rgba 0.3125 1 0 1>}
  ) {materialFunction=&3ColorBlend}
  n55 = NamedRerouteUsage()
  n56 = ScalarParameter() {
   defaultValue=1
   desc="Will show rounded cap for end edge of the progress bar.\r\nbool"
   group=Progress_Layout
   parameterName=UseLeadingCircle
   sliderMax=1
  }
  LinearInterpolate(
   A=@n49
   B=LinearInterpolate(
    A=@n49
    B=@n54.Result
    Alpha=MaterialFunctionCall(DistanceField=@n55) {materialFunction=&MF_UI_AntiAliasing}.Inner
   )
   Alpha=@n56
  )
  n57 = NamedRerouteUsage()
  n58 = NamedRerouteUsage()
  n59 = Subtract(
   A=OneMinus(None=ScalarParameter() {defaultValue=0.5, parameterName=ProgressFillAdditiveRatio})
   B=Multiply(A=NamedRerouteUsage()) {constB=0.25}
  )
  LinearInterpolate(
   A=Multiply(A=Subtract(A=@n57, B=@n58), B=ScalarParameter() {defaultValue=0.5, parameterName=EmptyFillAlpha}) {constB=0.5}
   Alpha=LinearInterpolate(A=@n21, B=Saturate(None=@n59), Alpha=@n58)
  )
  n60 = VectorParameter() {desc=Color for delta part of progress bar., group=Delta, parameterName=DeltaColor2, sortPriority=20, defaultValue=<rgba 0 1 0.2141 1>}
  n61 = VectorParameter() {group=Delta, parameterName=DeltaColorDecay, sortPriority=21, defaultValue=<rgba 0 1 1 1>}
  n62 = ScalarParameter() {
   desc="Property for animating Delta color decay\r\n(from DeltaColor to DeltaColorDecay)\r\n0 - 1"
   group=Delta
   parameterName=AnimateDeltaColor
   sliderMax=1
   sortPriority=34
  }
  n63 = Max(A=NamedRerouteUsage(), B=@n34)
  LinearInterpolate(
   A=@n37
   B=LinearInterpolate(
    A=LinearInterpolate(A=@n60.RGB, B=@n61.RGB, Alpha=@n62)
    B=NamedRerouteUsage()
    Alpha=ScalarParameter() {group=Color, parameterName=UseColorGradients, sliderMax=1}
   )
   Alpha=Multiply(A=OneMinus(None=@n63), B=NamedRerouteUsage())
  )
  MaterialFunctionCall!(FloatValue, Speed) {materialFunction=/FortUILibrary/MaterialFunctions/Basics/MF_UI_LinearTime}
  MaterialFunctionCall!(DistanceFieldA=NamedRerouteUsage(), DistanceFieldB=@n4, BlendAmount) {materialFunction=/FortUILibrary/MaterialFunctions/Shapes/Modifiers/MF_UI_SDF_Union}
  n64 = Multiply(A=ScalarParameter() {defaultValue=5, group=Outline, parameterName=Outline_PX, sortPriority=2}) {constB=2}
  n65 = Add(
   A=NamedRerouteDeclaration(None=MaterialFunctionCall!(Value PX=@n64, UV Pixel Size, X or Y) {materialFunction=/FortUILibrary/MaterialFunctions/Basics/MF_UI_PixelsToNormalized}.Normalized) {name=OutlineThicknessNormalized, nodeColor=<rgba 0 0 0 1>}
   B=NamedRerouteUsage()
  )
  n66 = Subtract(A=OneMinus(None=@n65), B=NamedRerouteUsage())
  n67 = Multiply(A=@n65) {constB=0.55}
  n68 = Saturate(None=MaterialFunctionCall(DistanceField=MaterialFunctionCall(Size=@n66, OutlineThickness=@n67) {materialFunction=&MF_UI_Circle}.SDFOutline) {materialFunction=&MF_UI_AntiAliasing}.Outer)
  n69 = ScalarParameter() {defaultValue=1, desc="Will add outline for progress bar.\r\nbool", group=Outline, parameterName=UseOutline, sliderMax=1, sortPriority=1}
  n70 = MaterialFunctionCall(DistanceField=LinearInterpolate(
    A=NamedRerouteUsage()
    B=MaterialFunctionCall(Size=@n66, OutlineThickness=@n65) {materialFunction=&MF_UI_Circle}.SDFOutline
    Alpha=@n69
   )) {materialFunction=&MF_UI_AntiAliasing}
  Min(A=@n68, B=@n70.Outer)
  Multiply()
  n71 = Multiply(A=@n54.Result, B=OneMinus(None=ScalarParameter() {defaultValue=0.8, group=Glow, parameterName=GlowColorDim})) {constB=0.5}
  n72 = Multiply(A=NamedRerouteUsage(), B=SmoothStep(Min=NamedRerouteUsage(), Value=NamedRerouteUsage()) {constMax=0})
  n73 = LinearInterpolate(
   A=@n51.RGB
   B=LinearInterpolate(A=@n54.Result, B=@n50.RGB, Alpha=NamedRerouteDeclaration(None=@n72) {name=EdgeMask, nodeColor=<rgba 0 0.25 1 1>})
   Alpha=LinearInterpolate(Alpha=NamedRerouteUsage()) {constA=0.5}
  )
  n74 = NamedRerouteUsage()
  Multiply(
   A=VertexColor().A
   B=LinearInterpolate(A=LinearInterpolate(B=@n71, Alpha=@n20.EaseInCubic), B=LinearInterpolate(A=@n73, Alpha=@n74), Alpha=NamedRerouteUsage())
  )
  n75 = ScalarParameter() {defaultValue=20, desc=Adds space on the texture edge in pixels., group=Progress_Layout, parameterName=EdgeOffset_PX, sortPriority=10}
  n76 = Multiply(A=@n75) {constB=2}
  n77 = NamedRerouteDeclaration(None=MaterialFunctionCall(Value PX=ScalarParameter() {defaultValue=20, group=Progress_Layout, parameterName=BarThickness_PX, sliderMax=1000, sortPriority=0}) {materialFunction=&MF_UI_PixelsToNormalized}."Normalized (>1px)") {name=BarThicknessNormalized, nodeColor=<rgba 0 0 0 1>}
  n78 = Add(
   A=NamedRerouteDeclaration(None=MaterialFunctionCall(Value PX=@n76) {materialFunction=&MF_UI_PixelsToNormalized}.Normalized) {name=EdgeOffsetNormalized, nodeColor=<rgba 0 0 0 1>}
   B=@n77
  )
  n79 = NamedRerouteUsage()
  n80 = ScalarParameter() {defaultValue=1, desc="Will add outline for progress bar.\r\nbool", group=Outline, parameterName=UseOutline, sliderMax=1, sortPriority=1}
  n81 = MaterialFunctionCall(Size=Subtract(A=OneMinus(None=@n78), B=LinearInterpolate(B=@n79, Alpha=@n80)), OutlineThickness=@n77) {materialFunction=&MF_UI_Circle}
  NamedRerouteDeclaration(None=@n81.SDFOutline) {name=SDF_CircleBar, nodeColor=<rgba 0 0 0 1>}
  n82 = ScalarParameter() {desc="Angle from 0 point and move clockwise.\r\n0 - 1", group=Progress_Layout, parameterName=StartAngle, sliderMax=1, sortPriority=5}
  n83 = ScalarParameter() {desc="Angle from 0 point and move anticlockwise.\r\n0 - 1", group=Progress_Layout, parameterName=EndAngle, sliderMax=1, sortPriority=6}
  n84 = ScalarParameter() {
   defaultValue=1
   desc="Will fill the progress bar clockwise.\r\nbool"
   group=Progress_Layout
   parameterName=FillClockwise
   sliderMax=1
   sortPriority=40
  }
  n85 = ScalarParameter() {
   desc="When progress value changes, it defines where change starts.\r\n0 - 1"
   group=Progress
   parameterName=ProgressStart
   sliderMax=1
   sortPriority=1
  }
  n86 = LinearInterpolate(A=@n85, B=NamedRerouteUsage(), Alpha=NamedRerouteUsage())
  n87 = LinearInterpolate(
   A=@n86
   B=MaterialFunctionCall!("Value (0-1)"=@n86, Steps=NamedRerouteUsage()) {materialFunction=/FortUILibrary/MaterialFunctions/Basics/MF_UI_StepRatio}.Stepped Ratio
   Alpha=ScalarParameter() {desc="Change progress value only with segments chanks.\r\nbool", group=Segments, parameterName=SnapToSegments, sliderMax=1}
  )
  n88 = NamedRerouteDeclaration(None=MaterialFunctionCall!(
    UV=NamedRerouteUsage()
    "Arc Start Angle (0-1)"=@n82
    "Arc End Angle (0-1)"=@n83
    "Fill Ratio (0-1)"=OneMinus(None=@n87)
    Fill Clockwise=@n84
   ) {materialFunction=&/MF_PostMatch_RadialArc}.Radial Fill) {name=ArcDeltaGradient, nodeColor=<rgba 0 0 0 1>}
  n89 = ScalarParameter() {
   desc="Enables delta functionality for progress bar.\r\nShows delta after progress on decrease progress value and on encrease shows delta value before progress.\r\nbool"
   group=Delta
   parameterName=ShowDelta
   sliderMax=1
   sortPriority=0
  }
  NamedRerouteDeclaration(None=Multiply(
    A=MaterialFunctionCall(DistanceField=@n88) {materialFunction=&MF_UI_AntiAliasing}.Outer
    B=@n89
   )) {name=ArcDeltaMask, nodeColor=<rgba 0 0 0 1>}
  n90 = ScalarParameter() {desc="Angle from 0 point and move clockwise.\r\n0 - 1", group=Progress_Layout, parameterName=StartAngle, sliderMax=1, sortPriority=5}
  n91 = ScalarParameter() {desc="Angle from 0 point and move anticlockwise.\r\n0 - 1", group=Progress_Layout, parameterName=EndAngle, sliderMax=1, sortPriority=6}
  n92 = ScalarParameter() {
   defaultValue=1
   desc="Will fill the progress bar clockwise.\r\nbool"
   group=Progress_Layout
   parameterName=FillClockwise
   sliderMax=1
   sortPriority=40
  }
  n93 = ScalarParameter() {
   desc="When progress value changes, it defines where change starts.\r\n0 - 1"
   group=Progress
   parameterName=ProgressStart
   sliderMax=1
   sortPriority=1
  }
  n94 = LinearInterpolate(A=@n93, B=NamedRerouteUsage(), Alpha=NamedRerouteUsage())
  n95 = LinearInterpolate(
   A=@n94
   B=MaterialFunctionCall("Value (0-1)"=@n94, Steps=NamedRerouteUsage()) {materialFunction=&MF_UI_StepRatio}.Stepped Ratio
   Alpha=ScalarParameter() {desc="Change progress value only with segments chanks.\r\nbool", group=Segments, parameterName=SnapToSegments, sliderMax=1}
  )
  n96 = MaterialFunctionCall(
   UV=NamedRerouteUsage()
   "Arc Start Angle (0-1)"=@n90
   "Arc End Angle (0-1)"=@n91
   "Fill Ratio (0-1)"=OneMinus(None=@n95)
   Fill Clockwise=@n92
  ) {materialFunction=&MF_PostMatch_RadialArc}
  NamedRerouteDeclaration(None=@n96.Arc Mask) {name=ArcGradient, nodeColor=<rgba 0 0 0 1>}
  n97 = MaterialFunctionCall(Value PX=ScalarParameter() {defaultValue=4, desc=The size of gap between segements in px., group=Segments, parameterName=SeparatorThickness_PX}) {materialFunction=&MF_UI_PixelsToNormalized}
  n98 = Multiply(A=@n97."Normalized (>1px)", B=Multiply(A=NamedRerouteUsage()) {constB=0.25})
  n99 = ScalarParameter() {defaultValue=0.01, group=Progress_Layout, parameterName=SeparatorSharpness, sortPriority=42}
  n100 = ScalarParameter() {defaultValue=8, desc=Number of segments in progress bar, group=Segments, parameterName=SegmentsCount, sliderMin=1, sortPriority=2}
  n101 = Multiply(A=NamedRerouteUsage(), B=NamedRerouteDeclaration(None=Floor(None=@n100)) {name=SegmentsCount, nodeColor=<rgba 0 0 0 1>})
  n102 = Constant() {r=1}
  n103 = ConstantBiasScale(None=NamedRerouteUsage()) {bias=-0.5, scale=2}
  n104 = Saturate(None=Multiply(
    A=MaterialFunctionCall!(Value=@n101, Scale=@n102, Height) {materialFunction=/FortUILibrary/MaterialFunctions/Basics/MF_UI_PingPong}.Result
    B=Length(None=@n103) {is=class MaterialExpressionLength}
   ))
  NamedRerouteDeclaration(None=SmoothStep(Min=Subtract(A=@n98, B=@n99), Max=Add(A=@n98, B=@n99), Value=Subtract(A=@n104, B=@n97."Normalized (>1px)")) {constMax=0.01}) {name=SegmentsMask, nodeColor=<rgba 0 0 0 1>}
  NamedRerouteDeclaration(None=@n70.Inner) {name=CircleBarMask, nodeColor=<rgba 0 0 0 1>}
  n105 = NamedRerouteDeclaration(None=@n96.Radial Fill) {name=ArcProgressGradient, nodeColor=<rgba 0 0 0 1>}
  NamedRerouteDeclaration(None=MaterialFunctionCall(DistanceField=@n105) {materialFunction=&MF_UI_AntiAliasing}.Outer) {name=ArcProgressMask, nodeColor=<rgba 0 0 0 1>}
  n106 = NamedRerouteUsage()
  n107 = SmoothStep(
   Min=Multiply(A=ScalarParameter() {defaultValue=0.1, group=Glow, parameterName=GlowThicknessAdjustment}) {constB=-1}
   Max=ScalarParameter() {defaultValue=0.1, group=Glow, parameterName=GlowIntensity, sortPriority=4}
   Value=@n6
  ) {constMax=0.15}
  NamedRerouteDeclaration(None=Multiply(A=OneMinus(None=@n106), B=OneMinus(None=@n107))) {name=GlowMask, nodeColor=<rgba 0 0 0 1>}
  NamedRerouteDeclaration(None=@n87) {name=DeltaProgress, nodeColor=<rgba 0 0 0 1>}
  n108 = ScalarParameter() {desc="Angle from 0 point and move anticlockwise.\r\n0 - 1", group=Progress_Layout, parameterName=EndAngle, sliderMax=1, sortPriority=6}
  n109 = OneMinus(None=@n108)
  n110 = ScalarParameter() {desc="Angle from 0 point and move clockwise.\r\n0 - 1", group=Progress_Layout, parameterName=StartAngle, sliderMax=1, sortPriority=5}
  n111 = ScalarParameter() {
   defaultValue=1
   desc="Will fill the progress bar clockwise.\r\nbool"
   group=Progress_Layout
   parameterName=FillClockwise
   sliderMax=1
   sortPriority=40
  }
  n112 = LinearInterpolate(A=@n109, B=@n110, Alpha=@n111)
  n113 = ScalarParameter() {
   defaultValue=1
   desc="Will show rounded cap for start and end edge of the progress bar. Not support delta functionality and snap to segments.\r\nbool"
   group=Progress_Layout
   parameterName=UseRoundedProgress
   sliderMax=1
   sortPriority=2
  }
  n114 = ScalarParameter() {
   defaultValue=1
   desc="Will show rounded cap for end edge of the progress bar.\r\nbool"
   group=Progress_Layout
   parameterName=UseLeadingCircle
   sliderMax=1
  }
  n115 = LinearInterpolate(A=NamedRerouteUsage(), Alpha=Max(A=@n113, B=@n114))
  n116 = MaterialFunctionCall(UVs=NamedRerouteUsage(), "Rotation Angle (0-1)"=Subtract(A=OneMinus(None=@n112), B=@n115) {constB=0}) {materialFunction=&CustomRotator}
  n117 = ComponentMask(None=@n116.Rotated Values) {r=true}
  n118 = Add(A=ComponentMask(None=@n116.Rotated Values) {g=true}, B=NamedRerouteUsage()) {constB=0.47}
  NamedRerouteDeclaration(None=MaterialFunctionCall(UVs=AppendVector(A=@n117, B=@n118), Size=NamedRerouteUsage()) {materialFunction=&MF_UI_Circle}.SDF) {name=SDF_CircleStart, nodeColor=<rgba 0 0 0 1>}
  n119 = LinearInterpolate(A=@n110, B=@n109, Alpha=@n111)
  n120 = NamedRerouteUsage()
  n121 = MaterialFunctionCall(UVs=NamedRerouteUsage(), "Rotation Angle (0-1)"=OneMinus(None=LinearInterpolate(A=@n112, B=@n119, Alpha=@n120))) {materialFunction=&CustomRotator}
  n122 = ComponentMask(None=@n121.Rotated Values) {g=true}
  n123 = NamedRerouteUsage()
  n124 = MaterialFunctionCall(UVs=AppendVector(A=ComponentMask(None=@n121.Rotated Values) {r=true}, B=Add(A=@n122, B=@n123) {constB=0.47}), Size=NamedRerouteUsage()) {materialFunction=&MF_UI_Circle}
  NamedRerouteDeclaration(None=@n124.SDF) {name=SDF_CircleDeltaEnd, nodeColor=<rgba 0 0 0 1>}
  n125 = Multiply(A=NamedRerouteUsage(), B=NamedRerouteUsage())
  n126 = MaterialFunctionCall(DistanceField=MaterialFunctionCall(DistanceFieldA=NamedRerouteUsage(), DistanceFieldB=NamedRerouteUsage()) {materialFunction=&MF_UI_SDF_Union}.Exact) {materialFunction=&MF_UI_AntiAliasing}
  n127 = Max(A=@n125, B=@n126.Inner)
  n128 = ScalarParameter() {
   defaultValue=1
   desc="Will show rounded cap for start and end edge of the progress bar. Not support delta functionality and snap to segments.\r\nbool"
   group=Progress_Layout
   parameterName=UseRoundedProgress
   sliderMax=1
   sortPriority=2
  }
  n129 = ScalarParameter() {
   desc="Enables delta functionality for progress bar.\r\nShows delta after progress on decrease progress value and on encrease shows delta value before progress.\r\nbool"
   group=Delta
   parameterName=ShowDelta
   sliderMax=1
   sortPriority=0
  }
  NamedRerouteDeclaration(None=Multiply(A=LinearInterpolate(A=@n125, B=@n127, Alpha=@n128), B=@n129)) {name=DeltaMask, nodeColor=<rgba 0 0 0 1>}
  n130 = NamedRerouteUsage()
  NamedRerouteDeclaration(None=Multiply(A=@n81.Fill, B=OneMinus(None=@n130))) {name=BehindCircle_Inside, nodeColor=<rgba 0 0 0 1>}
  n131 = MaterialFunctionCall(DistanceField=@n105, Threshold=Constant() {r=0.0063}) {materialFunction=&MF_UI_AntiAliasing}
  n132 = Ceil(None=@n105)
  NamedRerouteDeclaration(None=Multiply(A=NamedRerouteUsage(), B=Min(A=@n131.Inner, B=@n132))) {name=ArcAngleCap, nodeColor=<rgba 0 0 0 1>}
  NamedRerouteDeclaration(None=MaterialFunctionCall!("Value (0-1)"=SmoothStep(Value=@n105) {constMax=0.0162}) {materialFunction=/FortUILibrary/MaterialFunctions/Basics/MF_UI_EaseCurves_01_Sine}.In) {name=EndCapShadow, nodeColor=<rgba 0 0.8594 1 1>}
  n133 = NamedRerouteUsage()
  n134 = MaterialFunctionCall(UVs=NamedRerouteUsage(), "Rotation Angle (0-1)"=OneMinus(None=LinearInterpolate(A=@n112, B=@n119, Alpha=@n133))) {materialFunction=&CustomRotator}
  n135 = ComponentMask(None=@n134.Rotated Values) {r=true}
  n136 = Add(A=ComponentMask(None=@n134.Rotated Values) {g=true}, B=NamedRerouteUsage()) {constB=0.47}
  NamedRerouteDeclaration(None=MaterialFunctionCall(UVs=AppendVector(A=@n135, B=@n136), Size=NamedRerouteUsage()) {materialFunction=&MF_UI_Circle}.SDF) {name=SDF_LeadingCircle, nodeColor=<rgba 0 0 0 1>}
  n137 = LinearInterpolate(A=@n112, B=@n119, Alpha=NamedRerouteUsage())
  n138 = MaterialFunctionCall(UVs=NamedRerouteUsage(), "Rotation Angle (0-1)"=Add(A=OneMinus(None=@n137), B=@n115) {constB=0}) {materialFunction=&CustomRotator}
  n139 = ComponentMask(None=@n138.Rotated Values) {r=true}
  n140 = Add(A=ComponentMask(None=@n138.Rotated Values) {g=true}, B=NamedRerouteUsage()) {constB=0.47}
  NamedRerouteDeclaration(None=MaterialFunctionCall(UVs=AppendVector(A=@n139, B=@n140), Size=NamedRerouteUsage()) {materialFunction=&MF_UI_Circle}.SDF) {name=SDF_CircleEnd, nodeColor=<rgba 0 0 0 1>}
  NamedRerouteDeclaration(None=MaterialFunctionCall() {materialFunction=&GetUserInterfaceUV}.Normalized UV) {name=UVs, nodeColor=<rgba 0 0 0 1>}
  NamedRerouteDeclaration(None=Multiply(A=@n68, B=@n70.Inner)) {name=BarEdgesMask, nodeColor=<rgba 0.1953 1 0 1>}
  n141 = Subtract(A=NamedRerouteUsage(), B=NamedRerouteUsage())
  n142 = NamedRerouteUsage()
  n143 = OneMinus(None=Divide(A=NamedRerouteUsage(), B=Multiply(A=@n141, B=@n142)) {constB=0.05})
  NamedRerouteDeclaration(None=LinearInterpolate(
    A=VectorParameter() {group=Color, parameterName=Delta 1 Color 1, defaultValue=<rgba 0 0.1737 1 1>}.RGB
    B=VectorParameter() {group=Color, parameterName=Delta 1 Color 2, defaultValue=<rgba 1 0.0174 0.2053 1>}.RGB
    Alpha=Saturate(None=@n143)
   )) {name=DeltaColorGradient, nodeColor=<rgba 0 0 0 1>}
  n144 = ScalarParameter() {
   desc="Rotate the start point of the progress bar.\r\n0 - top\r\n0.25 - left\r\n0.5 - bottom\r\n0.75 - right"
   group=Progress_Layout
   parameterName=RotationAngle
  }
  NamedRerouteDeclaration(None=MaterialFunctionCall(UVs=NamedRerouteUsage(), "Rotation Angle (0-1)"=@n144) {materialFunction=&CustomRotator}.Rotated Values) {name=RotatorUVs, nodeColor=<rgba 0 0 0 1>}
  NamedRerouteDeclaration(None=Ceil(None=@n96.Arc Mask)) {name=ArcEmptyPartMask, nodeColor=<rgba 0 0 0 1>}
  NamedRerouteDeclaration(None=LinearInterpolate(
    A=VectorParameter() {group=Color, parameterName=Progress Color 1, defaultValue=<rgba 0.9339 0 1 1>}.RGB
    B=VectorParameter() {group=Color, parameterName=Progress Color 2, defaultValue=<rgba 1 1 1 1>}.RGB
    Alpha=Saturate(None=@n2)
   )) {name=ProgressColorGradient, nodeColor=<rgba 0 0 0 0.2035>}
  n145 = VectorParameter() {
   desc="Color for empty part of progress bar.\r\nAlpha defines opacity for empty part of progress bar."
   group=Color
   parameterName=EmptyColor
   sortPriority=10
   defaultValue=<rgba 0.3924 0.3924 0.3924 0.3>
  }
  n146 = LinearInterpolate(B=@n145.RGB, Alpha=NamedRerouteUsage())
  n147 = VectorParameter() {group=Behind Circle, parameterName=BehindCircleColor, sortPriority=2, defaultValue=<rgba 0.1493 0.1493 0.1493 0.1>}
  n148 = ScalarParameter() {
   desc="Enables inner circle fill for the circular progress bar.\r\nbool"
   group=Behind Circle
   parameterName=UseBehindCircle
   sliderMax=1
   sortPriority=1
  }
  n149 = LinearInterpolate(A=@n146, B=LinearInterpolate(A=@n146, B=@n147.RGB, Alpha=NamedRerouteUsage()), Alpha=@n148)
  n150 = VectorParameter() {group=Behind Circle, parameterName=BehindGlowColor, defaultValue=<rgba 0.3672 0.3124 1 1>}
  n151 = NamedRerouteDeclaration(None=MaterialFunctionCall(
    UVs=NamedRerouteUsage()
    Radius=ScalarParameter() {defaultValue=0.35, group=Behind Circle, parameterName=BehindGlowRadius}
    Density=ScalarParameter() {defaultValue=2.33, group=Behind Circle, parameterName=BehindGlowDensity}
   ) {materialFunction=&RadialGradientExponential}.RadialGradientExponential) {name=BehindGlow, nodeColor=<rgba 0.5937 1 0 1>}
  n152 = Multiply(A=@n150.RGB, B=@n151)
  NamedRerouteDeclaration(None=LinearInterpolate(A=@n149, B=LinearInterpolate(A=@n149, B=@n152, Alpha=@n151), Alpha=@n150.A)) {name=BehindCircleColor, nodeColor=<rgba 0 1 0.9766 1>}
  n153 = ScalarParameter() {
   defaultValue=1
   desc="When progress value changes, it defines where change ends.\r\n0 - 1"
   group=Progress
   parameterName=ProgressEnd
   sliderMax=1
   sortPriority=2
  }
  NamedRerouteDeclaration(None=@n153) {name=ProgressEnd, nodeColor=<rgba 0 0.1094 1 1>}
  n154 = NamedRerouteUsage()
  n155 = ScalarParameter() {desc="Angle from 0 point and move clockwise.\r\n0 - 1", group=Progress_Layout, parameterName=StartAngle, sliderMax=1, sortPriority=5}
  n156 = ScalarParameter() {desc="Angle from 0 point and move anticlockwise.\r\n0 - 1", group=Progress_Layout, parameterName=EndAngle, sliderMax=1, sortPriority=6}
  n157 = ScalarParameter() {
   defaultValue=1
   desc="Will fill the progress bar clockwise.\r\nbool"
   group=Progress_Layout
   parameterName=FillClockwise
   sliderMax=1
   sortPriority=40
  }
  n158 = NamedRerouteUsage()
  n159 = ScalarParameter() {
   desc="When progress value changes, it defines where change starts.\r\n0 - 1"
   group=Progress
   parameterName=ProgressStart
   sliderMax=1
   sortPriority=1
  }
  n160 = LinearInterpolate(A=@n159, B=NamedRerouteUsage(), Alpha=Max(A=NamedRerouteUsage(), B=NamedRerouteUsage()))
  n161 = OneMinus(None=LinearInterpolate(
    A=@n160
    B=MaterialFunctionCall("Value (0-1)"=@n160, Steps=@n158) {materialFunction=&MF_UI_StepRatio}.Stepped Ratio
    Alpha=ScalarParameter() {desc="Change progress value only with segments chanks.\r\nbool", group=Segments, parameterName=SnapToSegments, sliderMax=1}
   ))
  NamedRerouteDeclaration(None=Saturate(None=MaterialFunctionCall(UV=@n154, "Arc Start Angle (0-1)"=@n155, "Arc End Angle (0-1)"=@n156, "Fill Ratio (0-1)"=@n161, Fill Clockwise=@n157) {materialFunction=&MF_PostMatch_RadialArc}.Radial Fill)) {name=ArcMaxGradient, nodeColor=<rgba 0 0 0 1>}
  n162 = Multiply(A=NamedRerouteUsage(), B=NamedRerouteUsage())
  n163 = ScalarParameter() {
   defaultValue=1
   desc="Will show rounded cap for start and end edge of the progress bar. Not support delta functionality and snap to segments.\r\nbool"
   group=Progress_Layout
   parameterName=UseRoundedProgress
   sliderMax=1
   sortPriority=2
  }
  NamedRerouteDeclaration(None=LinearInterpolate(A=@n162, B=Max(A=@n162, B=@n126.Inner), Alpha=@n163)) {name=ProgressMask, nodeColor=<rgba 0 0 0 1>}
  n164 = ScalarParameter() {
   desc="When progress value changes, it defines where change starts.\r\n0 - 1"
   group=Progress
   parameterName=ProgressStart
   sliderMax=1
   sortPriority=1
  }
  n165 = NamedRerouteUsage()
  n166 = NamedRerouteUsage()
  NamedRerouteDeclaration(None=LinearInterpolate(A=@n164, B=NamedRerouteUsage(), Alpha=Max(A=@n165, B=@n166))) {name=MaxStop, nodeColor=<rgba 0 0 0 1>}
  n167 = ScalarParameter() {
   defaultValue=0.5
   desc="Property for animating progress value change (from ProgressStart to ProgressEnd)\r\n0 - 1"
   group=Progress
   parameterName=AnimateProgress
   sliderMax=1
   sortPriority=3
  }
  NamedRerouteDeclaration(None=@n167) {name=AnimateProgress, nodeColor=<rgba 1 0 0.8203 1>}
  n168 = ScalarParameter() {
   defaultValue=1
   desc="Property for animating Delta value change (from ProgressStart to ProgressEnd)\r\n0 - 1"
   group=Delta
   parameterName=AnimateDelta
   sliderMax=1
   sortPriority=33
  }
  NamedRerouteDeclaration(None=@n168) {name=AnimateDelta, nodeColor=<rgba 1 0 0.9141 1>}
  NamedRerouteDeclaration(None=@n81.Outline) {name=CircleBar, nodeColor=<rgba 0 1 0.0391 1>}
  NamedRerouteDeclaration(None=@n124.Fill) {name=CircleDeltaEnd, nodeColor=<rgba 1 0.6563 0 1>}
  n169 = NamedRerouteUsage()
  n170 = NamedRerouteUsage()
  NamedRerouteDeclaration(None=Divide(A=Multiply(A=@n169) {constB=0.53}, B=Multiply(A=@n170) {constB=6.2832})) {name=HalfCapArcAngle, nodeColor=<rgba 0 0 0 1>}
  n171 = NamedRerouteUsage()
  n172 = NamedRerouteUsage()
  n173 = ScalarParameter() {defaultValue=1, desc="Will add outline for progress bar.\r\nbool", group=Outline, parameterName=UseOutline, sliderMax=1, sortPriority=1}
  n174 = Multiply(A=LinearInterpolate(A=@n171, B=Add(A=@n171, B=@n172), Alpha=@n173)) {constB=0.5}
  n175 = NamedRerouteUsage()
  NamedRerouteDeclaration(None=Subtract(A=Subtract(B=@n174) {constA=0.5}, B=Multiply(A=@n175) {constB=0.5})) {name=CapOffset, nodeColor=<rgba 0 0 0 1>}
  n176 = NamedRerouteUsage()
  n177 = Subtract(A=NamedRerouteUsage(), B=@n176)
  NamedRerouteDeclaration(None=Subtract(A=Step(Y=@n176, X=@n96.Radial Fill), B=Step(Y=@n177, X=@n96.Radial Fill))) {name=ArcProgressMinusCapsMask, nodeColor=<rgba 0 0 0 1>}
  NamedRerouteDeclaration(None=@n95) {name=Progress, nodeColor=<rgba 0 0 0 1>}
  NamedRerouteUsage()
  NamedRerouteUsage()
  OneMinus(None=@n53)
  Reroute(None=@n37)
  ScalarParameter() {defaultValue=0.5, parameterName=ProgressFillAdditiveRatio}
  Subtract()
  Subtract(A=@n1, B=@n3)
  VectorParameter() {group=Progress_Layout, parameterName=LeadingCircleColor, defaultValue=<rgba 1 1 1 1>}
 ]
}
