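# Translucent UI material for a glow or soft drop shadow on SDF font text, filed under the post-match XP screen.
# The font's smooth signed distance runs through an anti-aliasing function, whose outer result is remapped by a Min/Max SmoothStep and scaled by the Color alpha.
# X and Y offset parameters, multiplied by the UI viewport scale factor, are added to the screen-plane position to shift the effect.
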Material /FortniteCoreUI/HUD/EndGame/PostGameFlowScreens/DefaultScreens/XPScreen/Materials/M_PostMatch_FontGlow
graph {
 properties {
  bAutomaticallySetUsageInEditor=false
  bUsedWithStaticMesh=false
  blendMode=BLEND_Translucent
  materialDomain=MD_UI
  usedShadingModels=Default Lit
 }
 nodes [
  n1 = ScalarParameter() {parameterName=X Offset}
  n2 = ScalarParameter() {defaultValue=4, parameterName=Y Offset}
  Add(
   A=WorldPosition().XY
   B=Multiply(
    A=AppendVector(A=@n1, B=@n2)
    B=MaterialFunctionCall() {materialFunction=/FortUILibrary/MaterialFunctions/Basics/MF_UI_ViewportScale}.Viewport Scale Factor
   )
  )
  n3 = ScalarParameter() {parameterName=Min, sliderMax=1}
  n4 = ScalarParameter() {defaultValue=1, parameterName=Max, sliderMax=1}
  n5 = MaterialFunctionCall(DistanceField=FontSignedDistance().Smooth Signed Distance, Threshold=Constant() {r=0.5}, Smoothing=Constant() {r=1}) {materialFunction=&MF_UI_AntiAliasing}
  Multiply(A=VectorParameter() {parameterName=Color, defaultValue=<rgba 0 0 0 0.25>}.A, B=Saturate(None=SmoothStep(Min=@n3, Max=@n4, Value=@n5.Outer)))
 ]
}
