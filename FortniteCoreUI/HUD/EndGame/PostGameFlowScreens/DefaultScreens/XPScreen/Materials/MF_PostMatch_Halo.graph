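# UI material function that draws a procedural halo, apparently for the post-match XP screen.
# Rotated UVs feed an arctangent angular gradient, which Mask Start and Mask End shape into a mirrored sweep mask.
# That mask varies the anti-alias smoothing and opacity of an SDF circle outline (Outer).
# A fainter glare outline with adjustable smoothing (Inner) is added to it and saturated as Halo, with all three exposed as outputs.
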MaterialFunction /FortniteCoreUI/HUD/EndGame/PostGameFlowScreens/DefaultScreens/XPScreen/Materials/MF_PostMatch_Halo
graph {
 properties {previewBlendMode=BLEND_Translucent, previewMaterialDomain=MD_UI}
 nodes [
  n1 = FunctionInput() {bUsePreviewValueAsDefault=true, inputName=Halo Size, inputType=FunctionInput_Scalar, previewValue=<0.7 0 0 1>}
  n2 = MaterialFunctionCall(UVs=NamedRerouteUsage(), Size=@n1) {materialFunction=&MF_UI_Circle}
  n3 = FunctionInput() {bUsePreviewValueAsDefault=true, inputName=Mask End, inputType=FunctionInput_Scalar, sortPriority=1, previewValue=<1 0 0 1>}
  n4 = FunctionInput() {bUsePreviewValueAsDefault=true, inputName=Rotation, inputType=FunctionInput_Scalar, sortPriority=1, previewValue=<0.875 0 0 1>}
  MaterialExpressionConstantBiasScale_0 = ConstantBiasScale(None=MaterialFunctionCall(UVs=NamedRerouteUsage(), "Rotation Angle (0-1)"=@n4) {materialFunction=&CustomRotator}.Rotated Values) {bias=-0.5, scale=2}
  n5 = Convert(None=@MaterialExpressionConstantBiasScale_0) {
   nodeName=Break Float2
   convertInputs [{
     type=Vector2
     expressionInput {
      expression=&self:MaterialExpressionConstantBiasScale_0
      inputName=None
      mask=0
      maskA=0
      maskB=0
      maskG=0
      maskR=0
      outputIndex=0
     }
    }]
   convertOutputs [{type=Scalar}, {type=Scalar}]
  }
  n6 = Divide(A=Arctangent2Fast(Y=@n5, X=@n5), B=MaterialFunctionCall() {materialFunction=&Pi}.Result)
  n7 = NamedRerouteDeclaration(None=SmoothStep(
    Min=FunctionInput() {bUsePreviewValueAsDefault=true, inputName=Mask Start, inputType=FunctionInput_Scalar, sortPriority=1}
    Max=@n3
    Value=Abs(None=@n6)
   )) {name=HaloMask, nodeColor=<rgba 0 1 0.7188 1>}
  n8 = LinearInterpolate(Alpha=@n7) {constA=0.4, constB=0.6}
  n9 = FunctionOutput(None=Multiply(
    A=MaterialFunctionCall(DistanceField=@n2.SDFOutline, Smoothing=@n8) {materialFunction=&MF_UI_AntiAliasing}.Inner
    B=@n7
   )) {outputName=Outer, sortPriority=1}
  n10 = FunctionInput() {bUsePreviewValueAsDefault=true, inputName=Glare Size, inputType=FunctionInput_Scalar, sortPriority=1, previewValue=<0.5 0 0 1>}
  n11 = MaterialFunctionCall(UVs=NamedRerouteUsage(), Size=@n10) {materialFunction=&MF_UI_Circle}
  n12 = FunctionInput() {bUsePreviewValueAsDefault=true, inputName=Glow Smoothing, inputType=FunctionInput_Scalar, sortPriority=3, previewValue=<0.1 0 0 1>}
  n13 = OneMinus(None=NamedRerouteUsage())
  n14 = FunctionOutput(None=Multiply(
    A=MaterialFunctionCall(DistanceField=@n11.SDFOutline, Smoothing=@n12) {materialFunction=&MF_UI_AntiAliasing}.Inner
    B=LinearInterpolate(Alpha=@n13) {constA=0.05, constB=0.1}
   )) {outputName=Inner, sortPriority=2}
  FunctionOutput(None=Saturate(None=Add(A=@n9, B=@n14))) {outputName=Halo}
  NamedRerouteDeclaration(None=FunctionInput(Preview=TextureCoordinate()) {bUsePreviewValueAsDefault=true, inputName=UV, inputType=FunctionInput_Vector2}) {name=UV, nodeColor=<rgba 1 0 0.0234 1>}
 ]
}
