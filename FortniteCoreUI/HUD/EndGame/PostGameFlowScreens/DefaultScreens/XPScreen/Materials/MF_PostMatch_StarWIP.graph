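# UI material function for the post-match XP screen that procedurally draws a star sparkle with glow and streak.
# Centre-scaled, rotated UVs feed ProceduralCrossStar, whose result is remapped, saturated and multiplied by an opacity input.
# Two separately scaled exponential radial gradients form the glow and the streak, the latter stretched thin by its default size.
# Star, Glow and Streak are output individually, alongside their saturated sum.
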MaterialFunction /FortniteCoreUI/HUD/EndGame/PostGameFlowScreens/DefaultScreens/XPScreen/Materials/MF_PostMatch_StarWIP
graph {
 properties {previewBlendMode=BLEND_Translucent, previewMaterialDomain=MD_UI}
 nodes [
  n1 = TextureCoordinate()
  n2 = FunctionInput() {bUsePreviewValueAsDefault=true, inputName=Star Size, inputType=FunctionInput_Vector2, sortPriority=1, previewValue=<1.2 1 0 1>}
  n3 = FunctionInput() {bUsePreviewValueAsDefault=true, inputName=Star Scale, inputType=FunctionInput_Scalar, sortPriority=2, previewValue=<0.7 0 0 1>}
  n4 = NamedRerouteDeclaration(None=MaterialFunctionCall(
    UVs=FunctionInput(Preview=@n1) {bUsePreviewValueAsDefault=true, inputName=UV, inputType=FunctionInput_Vector2}
    Texture Scale=Multiply(A=@n2, B=@n3)
   ) {materialFunction=&ScaleUVsByCenter}.UVs) {name=StarUV, nodeColor=<rgba 0 1 0.7891 1>}
  n5 = FunctionInput() {bUsePreviewValueAsDefault=true, inputName=Star Rotation, inputType=FunctionInput_Scalar, sortPriority=3, previewValue=<0.125 0 0 1>}
  n6 = InverseLinearInterpolate(Value=MaterialFunctionCall!(
    Tighten
    UVIn=MaterialFunctionCall(UVs=@n4, "Rotation Angle (0-1)"=@n5) {materialFunction=&CustomRotator}.Rotated Values
    "Softness(0-1)"=Constant() {r=1}
    "Fix_AntiFlick(Increase This If The Tip Flicking)"
    Fix_TipWidth
   ) {materialFunction=/Game/Effects/Niagara/MasterMaterials/Functions/ProceduralCrossStar}.Result) {constB=0.4}
  n7 = FunctionInput() {bUsePreviewValueAsDefault=true, inputName=Star Opacity, inputType=FunctionInput_Scalar, sortPriority=4, previewValue=<0.2 0 0 1>}
  n8 = FunctionOutput(None=Multiply(A=Saturate(None=@n6), B=@n7) {constB=0.2}) {outputName=Star, sortPriority=1}
  n9 = FunctionInput() {bUsePreviewValueAsDefault=true, inputName=Glow Size, inputType=FunctionInput_Vector2, sortPriority=5, previewValue=<1.2 1 0 1>}
  n10 = MaterialFunctionCall(UVs=NamedRerouteUsage(), Texture Scale=@n9) {materialFunction=&ScaleUVsByCenter}
  n11 = FunctionInput() {bUsePreviewValueAsDefault=true, inputName=Glow Scale, inputType=FunctionInput_Scalar, sortPriority=6, previewValue=<0.3 0 0 1>}
  n12 = FunctionInput() {bUsePreviewValueAsDefault=true, inputName=Glow Density, inputType=FunctionInput_Scalar, sortPriority=7, previewValue=<2 0 0 1>}
  n13 = FunctionInput() {bUsePreviewValueAsDefault=true, inputName=Streak Size, inputType=FunctionInput_Vector2, sortPriority=8, previewValue=<0.6 0.005 0 1>}
  n14 = MaterialFunctionCall(UVs=NamedRerouteUsage(), Texture Scale=@n13) {materialFunction=&ScaleUVsByCenter}
  n15 = Constant() {r=1}
  n16 = Add(
   A=FunctionOutput(None=MaterialFunctionCall(UVs=@n10.UVs, Radius=@n11, Density=@n12) {materialFunction=&RadialGradientExponential}.RadialGradientExponential) {outputName=Glow, sortPriority=2}
   B=FunctionOutput(None=MaterialFunctionCall(UVs=@n14.UVs, Radius=@n15) {materialFunction=&RadialGradientExponential}.RadialGradientExponential) {outputName=Streak, sortPriority=3}
  )
  FunctionOutput(None=Saturate(None=Add(A=@n8, B=@n16)))
 ]
}
