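# Translucent UI material for the post-match level-up bloom, combining a star, glow, streak and halo.
# Fit-scaled UVs appear to feed the shared star function and a center-scaled halo function, whose rotation rocks slowly on a time-driven sine.
# Star and halo results are summed and saturated.
# Separately, a radial inner glow tinted by Glow Color is added to the base Color and multiplied by vertex alpha.
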Material /FortniteCoreUI/HUD/EndGame/PostGameFlowScreens/DefaultScreens/XPScreen/Materials/M_PostMatch_LevelUpBloom
graph {
 expressions {
  MaterialExpressionAdd_0 {
   class=MaterialExpressionAdd
   inputs [
    A {from=MaterialExpressionMultiply_0}
    B {from=MaterialExpressionVectorParameter_0, output=RGB}
   ]
  }
  MaterialExpressionAdd_1 {
   class=MaterialExpressionAdd
   inputs [
    A {from=MaterialExpressionMaterialFunctionCall_0, output=Result}
    B {from=MaterialExpressionMaterialFunctionCall_3, output=Halo}
   ]
  }
  MaterialExpressionConvert_0 {
   class=MaterialExpressionConvert
   inputs [
    None {from=MaterialExpressionScalarParameter_3}
    None {from=""}
   ]
   properties {
    nodeName=Make Vector2
    convertInputs [
     {
      type=Scalar
      expressionInput {
       expression=&self:MaterialExpressionScalarParameter_3
       inputName=None
       mask=0
       maskA=0
       maskB=0
       maskG=0
       maskR=0
       outputIndex=0
      }
     }
     {
      type=Scalar
      defaultValue=<rgba 0.005 0 0 1>
      expressionInput {expression=null, inputName=None, mask=0, maskA=0, maskB=0, maskG=0, maskR=0, outputIndex=0}
     }
    ]
    convertOutputs [{type=Vector2}]
   }
  }
  MaterialExpressionMaterialFunctionCall_0 {
   class=MaterialExpressionMaterialFunctionCall
   inputs [
    UV {from=MaterialExpressionNamedRerouteUsage_0}
    Star Size {from=MaterialExpressionScalarParameter_1}
    Star Scale {from=""}
    Star Rotation {from=""}
    Star Opacity {from=""}
    Glow Size {from=MaterialExpressionScalarParameter_2}
    Glow Scale {from=""}
    Glow Density {from=MaterialExpressionScalarParameter_0}
    Streak Size {from=MaterialExpressionConvert_0}
   ]
   properties {materialFunction=&/MF_PostMatch_StarWIP}
  }
  MaterialExpressionMaterialFunctionCall_1 {
   class=MaterialExpressionMaterialFunctionCall
   inputs [
    UV {from=""}
    TextureSize {from=""}
    ScalePivotPoint {from=""}
    ContainerSize {from=""}
   ]
   properties {materialFunction=&MF_UI_UVScaling}
  }
  MaterialExpressionMaterialFunctionCall_2 {
   class=MaterialExpressionMaterialFunctionCall
   inputs [
    UVs {from=MaterialExpressionNamedRerouteUsage_1}
    CenterPosition {from=""}
    Radius {from=MaterialExpressionScalarParameter_4}
    Density {from=""}
    Invert Density {from=""}
   ]
   properties {materialFunction=&RadialGradientExponential}
  }
  MaterialExpressionMaterialFunctionCall_3 {
   class=MaterialExpressionMaterialFunctionCall
   inputs [
    Halo Size {from=""}
    UV {from=MaterialExpressionMaterialFunctionCall_4, output=UVs}
    Rotation {from=MaterialExpressionMultiply_2}
    Mask End {from=""}
    Mask Start {from=""}
    Glare Size {from=""}
    Glow Smoothing {from=""}
   ]
   properties {materialFunction=&/MF_PostMatch_Halo}
  }
  MaterialExpressionMaterialFunctionCall_4 {
   class=MaterialExpressionMaterialFunctionCall
   inputs [
    UVs {from=MaterialExpressionNamedRerouteUsage_2}
    Texture Scale {from=MaterialExpressionScalarParameter_8}
   ]
   properties {materialFunction=&ScaleUVsByCenter}
  }
  MaterialExpressionMultiply_0 {
   class=MaterialExpressionMultiply
   inputs [
    A {from=MaterialExpressionVectorParameter_1, output=RGB}
    B {from=MaterialExpressionMaterialFunctionCall_2, output=RadialGradientExponential}
   ]
  }
  MaterialExpressionMultiply_1 {
   class=MaterialExpressionMultiply
   inputs [
    A {from=MaterialExpressionTime_0}
    B {from=""}
   ]
   properties {constB=0.1}
  }
  MaterialExpressionMultiply_2 {
   class=MaterialExpressionMultiply
   inputs [
    A {from=MaterialExpressionSine_0}
    B {from=""}
   ]
   properties {constB=0.1}
  }
  MaterialExpressionMultiply_3 {
   class=MaterialExpressionMultiply
   inputs [
    A {from=MaterialExpressionVertexColor_0, output=A}
    B {from=MaterialExpressionAdd_0}
   ]
  }
  MaterialExpressionNamedRerouteDeclaration_0 {
   class=MaterialExpressionNamedRerouteDeclaration
   inputs [None {from=MaterialExpressionMaterialFunctionCall_1, output=Scaled to Fit}]
   properties {name=UV, nodeColor=<rgba 0 1 0.8125 1>}
  }
  MaterialExpressionNamedRerouteUsage_0 {class=MaterialExpressionNamedRerouteUsage}
  MaterialExpressionNamedRerouteUsage_1 {class=MaterialExpressionNamedRerouteUsage}
  MaterialExpressionNamedRerouteUsage_2 {class=MaterialExpressionNamedRerouteUsage}
  MaterialExpressionSaturate_0 {class=MaterialExpressionSaturate, inputs [None {from=MaterialExpressionAdd_1}]}
  MaterialExpressionScalarParameter_0 {class=MaterialExpressionScalarParameter, properties {defaultValue=2, group=Glow, parameterName=Glow Density}}
  MaterialExpressionScalarParameter_1 {class=MaterialExpressionScalarParameter, properties {defaultValue=1, group=Star, parameterName=Star Size}}
  MaterialExpressionScalarParameter_2 {class=MaterialExpressionScalarParameter, properties {defaultValue=1, group=Glow, parameterName=Glow Size}}
  MaterialExpressionScalarParameter_3 {class=MaterialExpressionScalarParameter, properties {defaultValue=0.5, parameterName=Streak Size}}
  MaterialExpressionScalarParameter_4 {class=MaterialExpressionScalarParameter, properties {defaultValue=0.5, parameterName=Inner Glow Size}}
  MaterialExpressionScalarParameter_8 {class=MaterialExpressionScalarParameter, properties {defaultValue=0.5, parameterName=Halo Size}}
  MaterialExpressionSine_0 {class=MaterialExpressionSine, inputs [None {from=MaterialExpressionMultiply_1}]}
  MaterialExpressionTime_0 {class=MaterialExpressionTime}
  MaterialExpressionVectorParameter_0 {class=MaterialExpressionVectorParameter, properties {parameterName=Color, defaultValue=<rgba 1 1 1 0>}}
  MaterialExpressionVectorParameter_1 {class=MaterialExpressionVectorParameter, properties {parameterName=Glow Color, defaultValue=<rgba 1 1 1 0>}}
  MaterialExpressionVertexColor_0 {class=MaterialExpressionVertexColor, inputs [], properties {}}
 }
 properties {
  bAutomaticallySetUsageInEditor=false
  bUsedWithStaticMesh=false
  blendMode=BLEND_Translucent
  materialDomain=MD_UI
  usedShadingModels=Default Lit
 }
}
