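# UI material function that draws a field of randomized circular sparkles, located with the post-match XP screen materials.
# Rotated UVs are split into Count columns, each hashed with Seed to give random channels for per-column variation such as vertical position and dot size.
# Aspect-corrected column tiling and position offsets feed MF_UI_Circle, which outputs modulated Fill, SDF and UV.
# It has no time node, so any motion comes from inputs such as Offset.
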MaterialFunction /FortniteCoreUI/HUD/EndGame/PostGameFlowScreens/DefaultScreens/XPScreen/Materials/MF_PostMatch_Sparkles
graph {
 properties {previewBlendMode=BLEND_Translucent, previewMaterialDomain=MD_UI}
 nodes [
  n1 = NamedRerouteUsage()
  n2 = NamedRerouteUsage()
  n3 = FunctionInput() {bUsePreviewValueAsDefault=true, inputName=Min Size, inputType=FunctionInput_Scalar, sortPriority=3, previewValue=<0.1 0 0 1>}
  n4 = FunctionInput() {bUsePreviewValueAsDefault=true, inputName=Max Size, inputType=FunctionInput_Scalar, sortPriority=4, previewValue=<0.5 0 0 1>}
  n5 = Saturate(None=NamedRerouteUsage())
  n6 = FunctionInput() {bUsePreviewValueAsDefault=true, inputName=Scale, inputType=FunctionInput_Scalar, sortPriority=5, previewValue=<0.5 0 0 1>}
  n7 = MaterialFunctionCall(
   UVs=FunctionOutput(None=Add(A=@n1, B=@n2)) {outputName=UV, sortPriority=2}
   Size=Multiply(A=LinearInterpolate(A=@n3, B=@n4, Alpha=@n5) {constA=0.1, constB=0.5}, B=@n6)
  ) {materialFunction=&MF_UI_Circle}
  FunctionOutput(None=Multiply(A=@n7.Fill, B=NamedRerouteUsage())) {outputName=Fill}
  FunctionOutput(None=@n7.SDF) {outputName=SDF, sortPriority=1}
  NamedRerouteDeclaration(None=FunctionInput(Preview=TextureCoordinate()) {bUsePreviewValueAsDefault=true, inputName=UV, inputType=FunctionInput_Vector2}) {name=UV, nodeColor=<rgba 1 0.4219 0 1>}
  n8 = FunctionInput() {bUsePreviewValueAsDefault=true, inputName=Count, inputType=FunctionInput_Scalar, sortPriority=1, previewValue=<24 0 0 1>}
  NamedRerouteDeclaration(None=@n8) {name=Count, nodeColor=<rgba 1 0.8672 0 1>}
  n9 = NamedRerouteUsage()
  n10 = FunctionInput() {bUsePreviewValueAsDefault=true, inputName=Rotation, inputType=FunctionInput_Scalar, sortPriority=7}
  n11 = ComponentMask(None=NamedRerouteDeclaration(None=MaterialFunctionCall(UVs=@n9, "Rotation Angle (0-1)"=@n10) {materialFunction=&CustomRotator}.Rotated Values) {name=SparkleUV, nodeColor=<rgba 0.6406 0 1 1>}) {r=true}
  n12 = NamedRerouteUsage()
  n13 = Divide(A=Floor(None=Multiply(A=@n11, B=@n12)), B=@n12)
  n14 = ConstantBiasScale(None=MaterialFunctionCall(
    Seed=FunctionInput() {bUsePreviewValueAsDefault=true, inputName=Seed, inputType=FunctionInput_Scalar, sortPriority=2}
    Vec3=MaterialFunctionCall(R=@n13) {materialFunction=&AppendMany}.RGB
   ) {materialFunction=&MF_Hash}.Vec3)
  NamedRerouteDeclaration(None=ComponentMask(None=@n14) {r=true}) {name=ColumnsR, nodeColor=<rgba 1 0.8203 0 1>}
  n15 = AppendVector(A=NamedRerouteUsage(), B=Constant() {r=1})
  n16 = Frac(None=Multiply(A=@n15, B=NamedRerouteUsage()))
  n17 = FunctionInput() {bUsePreviewValueAsDefault=true, inputName=Offset, inputType=FunctionInput_Vector2, sortPriority=6, previewValue=<0 -0.5 0 1>}
  NamedRerouteDeclaration(None=Multiply(
    A=Add(A=@n16, B=@n17) {constB=-0.5}
    B=MaterialFunctionCall(XY=@n15) {materialFunction=&Swizzle}.YX
   )) {name=GridTiling, nodeColor=<rgba 0.5 0 1 1>}
  n18 = Multiply(A=Add(A=NamedRerouteUsage()) {constB=-2}) {constB=0.5}
  n19 = Multiply(A=@n18) {constB=-1}
  n20 = NamedRerouteUsage()
  NamedRerouteDeclaration(None=AppendVector(A=Constant(), B=LinearInterpolate(A=@n19, B=@n18, Alpha=@n20) {constA=-9, constB=9})) {name=PositionColumns, nodeColor=<rgba 0 1 0.1797 1>}
  NamedRerouteDeclaration(None=ComponentMask(None=@n14) {b=true}) {name=ColumnsB, nodeColor=<rgba 0.4297 0 1 1>}
  NamedRerouteDeclaration(None=ComponentMask(None=@n14) {g=true}) {name=ColumnsG, nodeColor=<rgba 0 0.8594 1 1>}
 ]
}
