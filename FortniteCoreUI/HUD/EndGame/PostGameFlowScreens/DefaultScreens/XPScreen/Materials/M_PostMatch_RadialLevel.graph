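# Translucent UI material for the radial level graphic on the post-match XP screen.
# It defines a circle mask and a linear gradient from a Circle Size parameter, plus colour blending between Color Left and Color Right and a Final Opacity multiplier.
# A sparkle function, with time-driven seed and rotation on center-scaled UVs, supplies a mask alongside a tintable sparkle colour.
# Named reroutes link these parts.
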Material /FortniteCoreUI/HUD/EndGame/PostGameFlowScreens/DefaultScreens/XPScreen/Materials/M_PostMatch_RadialLevel
graph {
 properties {
  bAutomaticallySetUsageInEditor=false
  bUsedWithStaticMesh=false
  blendMode=BLEND_Translucent
  materialDomain=MD_UI
  usedShadingModels=Default Lit
 }
 nodes [
  n1 = NamedRerouteUsage()
  LinearInterpolate(
   A=VectorParameter() {group=Colors, parameterName=Color Left, defaultValue=<rgba 0.0823 0.0012 0.2542 1>}.RGB
   B=VectorParameter() {group=Colors, parameterName=Color Right, defaultValue=<rgba 0.0075 0.0003 0.0176 1>}.RGB
   Alpha=@n1
  )
  Multiply(A=Multiply(A=@n1, B=NamedRerouteUsage()), B=ScalarParameter() {defaultValue=0.5, parameterName=Final Opacity, sliderMax=1})
  n2 = NamedRerouteUsage()
  n3 = Constant() {r=0.4}
  n4 = Time()
  NamedRerouteDeclaration(None=MaterialFunctionCall!(
    UV=MaterialFunctionCall(UVs=@n2, Texture Scale=@n3) {materialFunction=&ScaleUVsByCenter}.UVs
    Count=Constant() {r=32}
    Seed=Multiply(A=@n4) {constB=0.1}
    Min Size
    Max Size
    Scale=Constant() {r=0.3}
    Offset
    Rotation=Multiply(A=@n4) {constB=0.01}
   ) {materialFunction=&/MF_PostMatch_Sparkles}.Fill) {name=Sparkles Mask, nodeColor=<rgba 0.2422 1 0 1>}
  NamedRerouteDeclaration(None=TextureCoordinate()) {name=UV, nodeColor=<rgba 0.2656 1 0 1>}
  n5 = ScalarParameter() {defaultValue=1, parameterName=Circle Size}
  n6 = OneMinus(None=@n5)
  n7 = Subtract(A=@n6, B=Multiply(A=@n6) {constB=0.5})
  NamedRerouteDeclaration(None=MaterialFunctionCall(UVs=NamedRerouteUsage(), GradientStart=@n7, GradientEnd=Add(A=@n7, B=@n5)) {materialFunction=&MF_UI_LinearGradient}.Result) {name=Circle Gradient, nodeColor=<rgba 1 0.2578 0 1>}
  NamedRerouteDeclaration(None=Multiply(A=VectorParameter() {parameterName=Sparkles Color, defaultValue=<rgba 0.1356 0.6724 1 1>}.RGB, B=NamedRerouteUsage())) {name=Sparkles, nodeColor=<rgba 0 0.0625 1 1>}
  NamedRerouteDeclaration(None=MaterialFunctionCall(UVs=NamedRerouteUsage(), Size=ScalarParameter() {defaultValue=1, parameterName=Circle Size}) {materialFunction=&MF_UI_Circle}.Fill) {name=Circle Mask, nodeColor=<rgba 0.2422 1 0 1>}
 ]
}
