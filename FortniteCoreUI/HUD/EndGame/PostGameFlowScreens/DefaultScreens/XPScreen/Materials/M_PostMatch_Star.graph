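# M_PostMatch_Star is a translucent UI material for a star with glow and streak, used in the post-match XP screen.
# It wraps MF_PostMatch_Star, exposing scalar parameters for star scale, size, rotation and opacity, glow scale, size and density, and streak size.
# Separate X/Y parameters are appended into size vectors.
# A time-driven sine drives a narrow lerp that multiplies Glow Density, giving the glow a subtle pulse.
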Material /FortniteCoreUI/HUD/EndGame/PostGameFlowScreens/DefaultScreens/XPScreen/Materials/M_PostMatch_Star
graph {
 properties {
  bAutomaticallySetUsageInEditor=false
  bUsedWithStaticMesh=false
  blendMode=BLEND_Translucent
  materialDomain=MD_UI
  usedShadingModels=Default Lit
 }
 nodes [
  n1 = Sine(None=Multiply(A=Time()))
  MaterialFunctionCall!(
   Glow Size=AppendVector(A=ScalarParameter() {defaultValue=1.2, parameterName=Glow Size X}, B=ScalarParameter() {defaultValue=1, parameterName=Glow Size Y})
   Star Scale=ScalarParameter() {defaultValue=0.7, parameterName=Star Scale}
   Star Size=AppendVector(A=ScalarParameter() {defaultValue=1.2, parameterName=Star Size X}, B=ScalarParameter() {defaultValue=1, parameterName=Star Size Y})
   Star Rotation=ScalarParameter() {defaultValue=0.125, parameterName=Star Rotation, sliderMax=1}
   UV
   Glow Scale=ScalarParameter() {defaultValue=0.3, parameterName=Glow Scale}
   Glow Density=Multiply(A=ScalarParameter() {defaultValue=2, parameterName=Glow Density}, B=LinearInterpolate(Alpha=@n1) {constA=0.98})
   Streak Size=AppendVector(A=ScalarParameter() {defaultValue=0.6, parameterName=Streak Size X}, B=ScalarParameter() {defaultValue=0.005, parameterName=Streak Size Y})
   Star Opacity=ScalarParameter() {defaultValue=0.2, parameterName=Star Opacity, sliderMax=1}
  ) {materialFunction=&/MF_PostMatch_Star}
 ]
}
