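# UI material function that draws animated radial streaks for the post-match XP screen.
# UVs become radial coordinates, scaled by Count on one axis and rotated over time by Speed, then sampled by single-level value noise.
# A center-distance mask, optionally aspect-corrected, lowers the SmoothStep threshold outward, keeping the middle clear, with Line Density and Line Falloff shaping the streaks.
# It outputs both the thresholded streaks and the raw noise.
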MaterialFunction /FortniteCoreUI/HUD/EndGame/PostGameFlowScreens/DefaultScreens/XPScreen/Materials/MF_PostMatch_Streaks
graph {
 properties {previewBlendMode=BLEND_Translucent, previewMaterialDomain=MD_UI}
 nodes [
  n1 = FunctionInput() {bUsePreviewValueAsDefault=true, inputName=Mask Size, inputType=FunctionInput_Scalar, sortPriority=5, previewValue=<0.0625 0 0 1>}
  n2 = FunctionInput() {bUsePreviewValueAsDefault=true, inputName=Mask Edge, inputType=FunctionInput_Scalar, sortPriority=6, previewValue=<0.5 0 0 1>}
  n3 = Add(A=NamedRerouteUsage()) {constB=-0.5}
  n4 = Multiply(A=@n3, B=MaterialFunctionCall!(Dimensions) {materialFunction=/FortUILibrary/MaterialFunctions/Basics/MF_UI_GetAspectRatio}.H Ratio)
  n5 = FunctionInput(Preview=StaticBool()) {bUsePreviewValueAsDefault=true, inputName=Correct Ratio, inputType=FunctionInput_StaticBool, sortPriority=7}
  n6 = InverseLinearInterpolate(A=@n1, B=Add(A=@n1, B=@n2), Value=Length(None=StaticSwitch(True=@n4, False=@n3, Value=@n5)) {is=class MaterialExpressionLength})
  n7 = FunctionInput() {bUsePreviewValueAsDefault=true, inputName=Line Density, inputType=FunctionInput_Scalar, sortPriority=3, previewValue=<0.5 0 0 1>}
  n8 = OneMinus(None=Multiply(A=Saturate(None=@n6), B=@n7) {constB=0.5})
  n9 = FunctionInput() {bUsePreviewValueAsDefault=true, inputName=Line Falloff, inputType=FunctionInput_Scalar, sortPriority=4, previewValue=<0.25 0 0 1>}
  n10 = FunctionInput() {bUsePreviewValueAsDefault=true, inputName=Count, inputType=FunctionInput_Scalar, sortPriority=1, previewValue=<400 0 0 1>}
  n11 = Multiply(
   A=MaterialFunctionCall!(Vector or UVs=NamedRerouteUsage()) {materialFunction=/FortUILibrary/MaterialFunctions/Basics/MF_UI_CheaperVectorToRadial}.Swizzled Radial Coordinates
   B=AppendVector(A=Constant() {r=1}, B=@n10)
  )
  n12 = FunctionInput() {bUsePreviewValueAsDefault=true, inputName=Speed, inputType=FunctionInput_Scalar, sortPriority=2, previewValue=<0.25 0 0 1>}
  n13 = Multiply(A=Time(), B=@n12) {constB=0.25}
  n14 = Noise!(World Position=AppendVector(A=Rotator(Coordinate=@n11, Time=@n13) {centerX=0, centerY=0}, B=Constant()), FilterWidth) {levels=1, noiseFunction=NOISEFUNCTION_ValueALU, outputMax=0.8, outputMin=0.2}
  FunctionOutput(None=SmoothStep(Min=@n8, Max=Add(A=@n8, B=@n9) {constB=0.25}, Value=FunctionOutput(None=@n14) {outputName=Noise, sortPriority=1}))
  NamedRerouteDeclaration(None=FunctionInput(Preview=TextureCoordinate()) {bUsePreviewValueAsDefault=true, inputName=UV, inputType=FunctionInput_Vector2}) {name=UV, nodeColor=<rgba 1 0 0.1641 1>}
 ]
}
