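# Translucent UI material that draws animated, two-color streaks for the post-match XP screen.
# MF_PostMatch_Streaks builds the aspect-corrected streak pattern from shared UVs, with parameters for count, speed, density, falloff and masking.
# Separately, a time-driven Rotator spins the UVs into MF_UI_RadialSegments, whose triangle gradient blends the two color parameters.
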Material /FortniteCoreUI/HUD/EndGame/PostGameFlowScreens/DefaultScreens/XPScreen/Materials/M_PostMatch_Streaks
graph {
 properties {
  bAutomaticallySetUsageInEditor=false
  bUsedWithStaticMesh=false
  blendMode=BLEND_Translucent
  materialDomain=MD_UI
  usedShadingModels=Default Lit
 }
 nodes [
  n1 = NamedRerouteUsage()
  n2 = Multiply(A=Time(), B=ScalarParameter() {defaultValue=0.25, parameterName=Color Speed})
  LinearInterpolate(
   A=VectorParameter() {parameterName=Color 1, defaultValue=<rgba 0.0885 0.2911 1 1>}.RGB
   B=VectorParameter() {parameterName=Color 2, defaultValue=<rgba 0.2949 0.0322 0.5841 1>}.RGB
   Alpha=MaterialFunctionCall!(UV=Rotator(Coordinate=@n1, Time=@n2), Segment Count, Segment Spacing=Constant(), Fixed Spacing) {materialFunction=/FortUILibrary/MaterialFunctions/Basics/MF_UI_RadialSegments}.Segmented Triangles Gradient
  )
  MaterialFunctionCall!(
   UV=NamedRerouteUsage()
   Count=ScalarParameter() {defaultValue=400, parameterName=Count}
   Speed=ScalarParameter() {defaultValue=0.25, parameterName=Line Speed}
   Line Density=ScalarParameter() {defaultValue=0.5, parameterName=Line Density}
   Line Falloff=ScalarParameter() {defaultValue=0.25, parameterName=Line Falloff}
   Mask Size=ScalarParameter() {defaultValue=0.0625, parameterName=Mask Size}
   Mask Edge=ScalarParameter() {defaultValue=0.5, parameterName=Mask Edge}
   Correct Ratio=StaticBool() {value=true}
  ) {materialFunction=&/MF_PostMatch_Streaks}
  NamedRerouteDeclaration(None=TextureCoordinate()) {name=UV, nodeColor=<rgba 0.0781 0 1 1>}
 ]
}
