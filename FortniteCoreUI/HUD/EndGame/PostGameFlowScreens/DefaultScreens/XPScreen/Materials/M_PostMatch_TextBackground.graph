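# Additive UI material for text backgrounds on the post-match XP screen, drawing an angled streak with a colored glow.
# The streak comes from mirrored, rotated UVs thresholded with anti-aliasing for slanted edges, then trimmed by a reflected-gradient vertical cutout.
# A center-scaled exponential radial gradient forms the glow mask and drives a three-color blend.
# The glow mask and opacity blend streak into glow for both color and saturated alpha.
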Material /FortniteCoreUI/HUD/EndGame/PostGameFlowScreens/DefaultScreens/XPScreen/Materials/M_PostMatch_TextBackground
graph {
 properties {
  bAutomaticallySetUsageInEditor=false
  bUsedWithStaticMesh=false
  blendMode=BLEND_Additive
  materialDomain=MD_UI
  usedShadingModels=Default Lit
 }
 nodes [
  LinearInterpolate(
   A=VectorParameter() {group=Streak, parameterName=Streak Color, defaultValue=<rgba 1 1 1 0>}.RGB
   B=NamedRerouteUsage()
   Alpha=Multiply(A=NamedRerouteUsage(), B=ScalarParameter() {defaultValue=1, group=Glow, parameterName=Glow Opacity, sliderMax=1})
  )
  NamedRerouteDeclaration(None=TextureCoordinate()) {name=UV, nodeColor=<rgba 1 0 0.3516 1>}
  n1 = MaterialFunctionCall(UV=NamedRerouteUsage()) {materialFunction=&MF_UI_UVScaling}
  NamedRerouteDeclaration(None=@n1.ScaleX) {name=Scaled UV X, nodeColor=<rgba 1 0.0089 0 1>}
  NamedRerouteDeclaration(None=@n1.ScaleY) {name=Scaled UV Y, nodeColor=<rgba 0 1 0.2784 1>}
  n2 = ScalarParameter() {defaultValue=1, group=Glow, parameterName=Glow Scale X}
  n3 = ScalarParameter() {defaultValue=1, group=Glow, parameterName=Glow Scale Y}
  n4 = MaterialFunctionCall(
   UVs=MaterialFunctionCall(UVs=NamedRerouteUsage(), Texture Scale=AppendVector(A=@n2, B=@n3)) {materialFunction=&ScaleUVsByCenter}.UVs
   Radius=ScalarParameter() {defaultValue=0.5, group=Glow, parameterName=Glow Radius}
   Density=ScalarParameter() {defaultValue=1.2, group=Glow, parameterName=Glow Density}
  ) {materialFunction=&RadialGradientExponential}
  NamedRerouteDeclaration(None=MaterialFunctionCall(
    A=VectorParameter() {group=Glow, parameterName=Glow Color 1, defaultValue=<rgba 0.1248 0 0.8469 1>}.RGB
    B=VectorParameter() {group=Glow, parameterName=Glow Color 2, defaultValue=<rgba 0.5089 0.0086 1 1>}.RGB
    C=VectorParameter() {group=Glow, parameterName=Glow Color 3, defaultValue=<rgba 0.9387 0.7305 1 1>}.RGB
    Alpha=NamedRerouteDeclaration(None=@n4.RadialGradientExponential) {name=Glow Mask, nodeColor=<rgba 1 0.6094 0 1>}
   ) {materialFunction=&3ColorBlend}.Result) {name=Glow Color, nodeColor=<rgba 0.9219 0 1 1>}
  n5 = MaterialFunctionCall(
   UVs=MaterialFunctionCall!(UV=NamedRerouteUsage(), Mirror Y, Swizzle UV) {materialFunction=/FortUILibrary/MaterialFunctions/Basics/MF_UI_Symmetry_2x1_New}.Symmetry
   "Rotation Angle (0-1)"=ScalarParameter() {defaultValue=-0.1, group=Streak, parameterName=Corner Angle, sliderMax=0.25, sliderMin=-0.25}
  ) {materialFunction=&CustomRotator}
  n6 = Multiply(
   A=MaterialFunctionCall(
    DistanceField=ComponentMask(None=@n5.Rotated Values) {r=true}
    Threshold=ScalarParameter() {defaultValue=0.5, group=Streak, parameterName=Threshold}
    Smoothing=ScalarParameter() {defaultValue=0.05, group=Streak, parameterName=Corner Smoothing}
   ) {materialFunction=&MF_UI_AntiAliasing}.Outer
   B=ScalarParameter() {defaultValue=1, group=Streak, parameterName=Streak Opacity, sliderMax=1}
  )
  n7 = Step(
   Y=ScalarParameter() {defaultValue=0.5, parameterName=Vertical Cutout, sliderMax=1}
   X=MaterialFunctionCall!(UVs=NamedRerouteUsage(), Offset, RotationCenter, RotationAngle=Constant() {r=0.25}) {materialFunction=/FortUILibrary/MaterialFunctions/Gradients/MF_UI_ReflectedGradient}.Result
  )
  n8 = ScalarParameter() {defaultValue=1, group=Streak, parameterName=Streak Opacity, sliderMax=1}
  n9 = ScalarParameter() {defaultValue=1, group=Glow, parameterName=Glow Opacity, sliderMax=1}
  n10 = NamedRerouteUsage()
  n11 = ScalarParameter() {defaultValue=1, group=Glow, parameterName=Glow Opacity, sliderMax=1}
  Saturate(None=LinearInterpolate(A=Multiply(A=@n6, B=@n7), B=Add(A=@n8, B=@n9), Alpha=Multiply(A=@n10, B=@n11)))
 ]
}
