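# M_UI_PostMatchProgressBar is a material filed under the post-game XP screen, named for its post-match progress bar.
# Its graph holds no expressions, so the file shows no fill, color or animation logic.
# It lists Default Lit as the used shading model and has automatic usage setting in the editor turned off.
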Material /FortniteCoreUI/HUD/EndGame/PostGameFlowScreens/DefaultScreens/XPScreen/Materials/M_UI_PostMatchProgressBar
graph {expressions {}, properties {bAutomaticallySetUsageInEditor=false, usedShadingModels=Default Lit}}
