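# Translucent UI material that draws an animated, procedural background for the post-match XP screen.
# Ambient square-tile function calls make five cell layers at different scales and speeds, with lerps combining layers under an Anim_TileBrightness control.
# The tile mix modulates an eased value that drives a falloff, mid and bright color ramp, soft-light blended with faint vector noise.
# A circle SDF forms a radial gradient mask that interpolates toward the Opacity parameter.
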Material /FortniteCoreUI/HUD/EndGame/PostGameFlowScreens/DefaultScreens/XPScreen/Materials/M_UI_PostMatch_Background_Simple
graph {
 properties {
  bAutomaticallySetUsageInEditor=false
  bUsedWithStaticMesh=false
  blendMode=BLEND_Translucent
  materialDomain=MD_UI
  usedShadingModels=Default Lit
 }
 nodes [
  n1 = MaterialFunctionCall(UVs=NamedRerouteUsage(), Size=ScalarParameter() {defaultValue=1, parameterName=RadialGradient_Radius}) {materialFunction=&MF_UI_Circle}
  n2 = OneMinus(None=Multiply(A=Add(A=@n1.SDF)) {constB=0.5})
  LinearInterpolate(
   B=ScalarParameter() {defaultValue=0.8, parameterName=Opacity}
   Alpha=NamedRerouteDeclaration(None=Saturate(None=@n2)) {name=GradientMask, nodeColor=<rgba 0 0 0 1>}
  ) {constA=1}
  n3 = VectorParameter() {parameterName=FalloffColor, defaultValue=<rgba 0 0.001 0.03 1>}
  n4 = VectorParameter() {parameterName=BrightColor, defaultValue=<rgba 0.0531 0.0125 0.5 1>}
  n5 = NamedRerouteUsage()
  n6 = NamedRerouteUsage()
  n7 = NamedRerouteUsage()
  n8 = LinearInterpolate(A=NamedRerouteUsage(), Alpha=LinearInterpolate(A=NamedRerouteUsage(), Alpha=LinearInterpolate(A=@n6, Alpha=@n7)))
  n9 = ScalarParameter() {parameterName=Anim_TileBrightness}
  n10 = LinearInterpolate(
   A=MaterialFunctionCall("Value (0-1)"=NamedRerouteUsage()) {materialFunction=&MF_UI_EaseCurves}.EaseInSine
   Alpha=Multiply(A=LinearInterpolate(A=@n5, Alpha=@n8), B=LinearInterpolate(Alpha=@n9) {constA=0.25, constB=2}) {constB=0.5}
  )
  LinearInterpolate(A=@n3.RGB, B=@n4.RGB, Alpha=@n10)
  MaterialFunctionCall!(
   Base=MaterialFunctionCall(A=@n3.RGB, B=VectorParameter() {parameterName=MidColor, defaultValue=<rgba 0.009 0.0025 0.1 1>}.RGB, C=@n4.RGB, Alpha=@n10) {materialFunction=&3ColorBlend}.Result
   Blend=Multiply(A=VectorNoise()) {constB=0.0333}
  ) {materialFunction=/Engine/Functions/Engine_MaterialFunctions03/Blends/Blend_SoftLight}
  NamedRerouteDeclaration(None=MaterialFunctionCall!(
    UVs=NamedRerouteUsage()
    Tile Count=Constant() {r=3}
    Speed=Constant() {r=0.125}
    Tile Subtraction Amount=Constant() {r=1}
    Tile Subtraction Random Seed=Constant() {r=19.5479}
    OffsetGrid
    Max Opacity=Constant() {r=0.5}
    Min Opacity=Constant() {r=0.125}
    Min Size=Constant() {r=0.5}
    Corner Radius=Constant() {r=-0.25}
    Blurriness=Constant() {r=0.3333}
   ) {materialFunction=&/MF_UI_PostMatch_AmbientSquareTiles}.Result) {name=Large Cells, nodeColor=<rgba 0 0 0 1>}
  NamedRerouteDeclaration(None=MaterialFunctionCall() {materialFunction=&MF_UI_UVScaling}.ScaleY) {name=ScaledUVs, nodeColor=<rgba 0 0 0 1>}
  NamedRerouteDeclaration(None=MaterialFunctionCall(
    UVs=NamedRerouteUsage()
    Tile Count=Constant() {r=4}
    Speed=Constant() {r=0.0625}
    Tile Subtraction Amount=Constant() {r=2}
    Tile Subtraction Random Seed=Constant() {r=18.6482}
    Max Opacity=Constant() {r=0.25}
    Min Opacity=Constant() {r=0.125}
    Min Size=Constant() {r=0.25}
    Corner Radius=Constant() {r=-0.25}
    Blurriness=Constant() {r=0.3}
   ) {materialFunction=&MF_UI_PostMatch_AmbientSquareTiles}.Result) {name=Medium Cells, nodeColor=<rgba 0 0 0 1>}
  NamedRerouteDeclaration(None=MaterialFunctionCall(
    UVs=NamedRerouteUsage()
    Tile Count=Constant() {r=7}
    Speed=Constant() {r=0.0625}
    Tile Subtraction Amount=Constant() {r=16}
    Tile Subtraction Random Seed=Constant() {r=32.8625}
    Max Opacity=Constant() {r=0.25}
    Min Opacity=Constant() {r=0.125}
    Min Size=Constant() {r=0.125}
    Corner Radius=Constant() {r=-0.25}
    Blurriness=Constant() {r=0.5}
   ) {materialFunction=&MF_UI_PostMatch_AmbientSquareTiles}.Result) {name=Small Cells, nodeColor=<rgba 0 0 0 1>}
  NamedRerouteDeclaration(None=MaterialFunctionCall(
    UVs=NamedRerouteUsage()
    Tile Count=Constant() {r=11}
    Speed=Constant() {r=0.25}
    Tile Subtraction Amount=Constant() {r=10}
    Tile Subtraction Random Seed=Constant() {r=32.8625}
    Max Opacity=Constant() {r=0.2}
    Min Opacity=Constant() {r=0.05}
    Min Size=Constant() {r=0.125}
    Corner Radius=Constant() {r=-0.25}
    Blurriness=Constant() {r=0.3333}
   ) {materialFunction=&MF_UI_PostMatch_AmbientSquareTiles}.Result) {name=XSmall Cells, nodeColor=<rgba 0 0 0 1>}
  NamedRerouteDeclaration(None=MaterialFunctionCall(
    UVs=NamedRerouteUsage()
    Tile Count=Constant() {r=21}
    Speed=Constant() {r=0.2}
    Tile Subtraction Amount=Constant() {r=20}
    Tile Subtraction Random Seed=Constant() {r=4.8286}
    Max Opacity=Constant() {r=0.5}
    Min Opacity=Constant() {r=0.25}
    Min Size=Constant() {r=0.5}
    Corner Radius=Constant() {r=-0.25}
    Blurriness=Constant() {r=0.3333}
   ) {materialFunction=&MF_UI_PostMatch_AmbientSquareTiles}.Result) {name=XXSmall Cells, nodeColor=<rgba 0 0 0 1>}
 ]
}
