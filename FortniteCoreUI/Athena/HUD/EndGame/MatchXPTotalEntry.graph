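# End-of-match HUD widget, built on FortMatchXPTotalEntry, that shows one XP category or total with label, number and optional divider.
# On construct it reads a console variable for S39 post-match updates, which with bIsTotal selects one of four intro animations.
# Playing the intro triggers an XP popup sound for positive values, faded out when inactive.
# Compact state changes number size; skipping stops animations and sets final opacity.
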WidgetBlueprint /FortniteCoreUI/Athena/HUD/EndGame/MatchXPTotalEntry
animations [
 Intro {
  binding_count=4
  length_seconds=2
  object=&self:Intro
  sequence {bindings [
    CommonBorder_divider {
     id=#1
     possessed_class=CommonBorder
     tracks [{
       type=MovieSceneFloatVectorTrack
       display_name=Desired Size Scale
       sections [{
         type=MovieSceneFloatVectorSection
         channels [
          Vector.X {type=MovieSceneScriptingFloatChannel, keys [[27.0873,0],[40,1]]}
          Vector.Y {type=MovieSceneScriptingFloatChannel}
         ]
        }]
      }]
    }
    ScaleBox_0 {
     id=#2
     possessed_class=ScaleBox
     tracks [{
       type=MovieSceneFloatTrack
       display_name=User Specified Scale
       sections [{type=MovieSceneFloatSection, channels [None {type=MovieSceneScriptingFloatChannel, keys [[23.75,1],[27.0873,0.5]]}]}]
      }]
    }
    TextBlock_XPType {
     id=#3
     possessed_class=CommonTextBlock
     tracks [{
       type=MovieScene2DTransformTrack
       sections [{
         type=MovieScene2DTransformSection
         channels [
          Scale.X {type=MovieSceneScriptingFloatChannel, keys [[0,0],[6.1783,1]]}
          Scale.Y {type=MovieSceneScriptingFloatChannel, keys [[0,0],[6.1783,1]]}
         ]
        }]
      }]
    }
    ContainerStack {
     id=#4
     display_name="CommonText_XPType (3)"
     possessed_class=VerticalBox
     tracks [
      {
       type=MovieScene2DTransformTrack
       sections [{
         type=MovieScene2DTransformSection
         channels [
          Translation.X {type=MovieSceneScriptingFloatChannel}
          Translation.Y {type=MovieSceneScriptingFloatChannel, keys [[19,-411.1959],[20,-470.0952],[25.3567,-86.1582],[27.0873,-3.6188]]}
          Angle {type=MovieSceneScriptingFloatChannel}
          Scale.X {type=MovieSceneScriptingFloatChannel, thinned_from=6, keys [[0,10],[3.404,1.6251],[4.8307,1],[19,1],[27.0873,0.8]]}
          Scale.Y {type=MovieSceneScriptingFloatChannel, thinned_from=6, keys [[0,10],[3.404,1.6251],[4.8307,1],[19,1],[27.0873,0.8]]}
          Shear.X {type=MovieSceneScriptingFloatChannel}
          Shear.Y {type=MovieSceneScriptingFloatChannel}
         ]
        }]
      }
      {
       type=MovieSceneFloatTrack
       display_name=Render Opacity
       sections [{type=MovieSceneFloatSection, channels [None {type=MovieSceneScriptingFloatChannel, keys [[0,0],[3.1973,1]]}]}]
      }
     ]
    }
   ]}
 }
 IntroTotal {
  binding_count=4
  length_seconds=2
  object=&self:IntroTotal
  sequence {bindings [
    CommonBorder_divider {
     id=#1
     possessed_class=CommonBorder
     tracks [{
       type=MovieSceneFloatVectorTrack
       display_name=Desired Size Scale
       sections [{
         type=MovieSceneFloatVectorSection
         channels [
          Vector.X {type=MovieSceneScriptingFloatChannel, keys [[27.0873,0],[40,1]]}
          Vector.Y {type=MovieSceneScriptingFloatChannel}
         ]
        }]
      }]
    }
    ScaleBox_0 {
     id=#2
     possessed_class=ScaleBox
     tracks [{
       type=MovieSceneFloatTrack
       display_name=User Specified Scale
       sections [{type=MovieSceneFloatSection, channels [None {type=MovieSceneScriptingFloatChannel, keys [[23.75,1],[27.0873,0.5]]}]}]
      }]
    }
    TextBlock_XPType {
     id=#3
     possessed_class=CommonTextBlock
     tracks [{
       type=MovieScene2DTransformTrack
       sections [{
         type=MovieScene2DTransformSection
         channels [
          Scale.X {type=MovieSceneScriptingFloatChannel, keys [[0,0],[6.1783,1]]}
          Scale.Y {type=MovieSceneScriptingFloatChannel, keys [[0,0],[6.1783,1]]}
         ]
        }]
      }]
    }
    ContainerStack {
     id=#4
     display_name="CommonText_XPType (3)"
     possessed_class=VerticalBox
     tracks [
      {
       type=MovieScene2DTransformTrack
       sections [{
         type=MovieScene2DTransformSection
         channels [
          Translation.X {type=MovieSceneScriptingFloatChannel}
          Translation.Y {type=MovieSceneScriptingFloatChannel, keys [[19,-290],[20,-320],[25.3567,-86.1582],[27.0873,-3.6188]]}
          Angle {type=MovieSceneScriptingFloatChannel}
          Scale.X {type=MovieSceneScriptingFloatChannel, thinned_from=6, keys [[0,10],[3.404,1.6251],[4.8307,1],[19,1],[27.0873,0.8]]}
          Scale.Y {type=MovieSceneScriptingFloatChannel, thinned_from=6, keys [[0,10],[3.404,1.6251],[4.8307,1],[19,1],[27.0873,0.8]]}
          Shear.X {type=MovieSceneScriptingFloatChannel}
          Shear.Y {type=MovieSceneScriptingFloatChannel}
         ]
        }]
      }
      {
       type=MovieSceneFloatTrack
       display_name=Render Opacity
       sections [{type=MovieSceneFloatSection, channels [None {type=MovieSceneScriptingFloatChannel, keys [[0,0],[3.1973,1]]}]}]
      }
     ]
    }
   ]}
 }
 Anim_Outro {
  binding_count=1
  length_seconds=0.3
  object=&self:Anim_Outro
  sequence {bindings [MatchXPTotalEntry_C_6 {
     id=#5
     possessed_class=MatchXPTotalEntry_C
     tracks [
      {
       type=MovieScene2DTransformTrack
       sections [{
         type=MovieScene2DTransformSection
         end_frame=6
         start_frame=0
         channels [
          Translation.X {type=MovieSceneScriptingFloatChannel, keys [[0,0],[6,30]]}
          Translation.Y {type=MovieSceneScriptingFloatChannel, const [0,1]}
          Angle {type=MovieSceneScriptingFloatChannel, const [0,1]}
          Scale.X {type=MovieSceneScriptingFloatChannel, const [1,1]}
          Scale.Y {type=MovieSceneScriptingFloatChannel, const [1,1]}
          Shear.X {type=MovieSceneScriptingFloatChannel, const [0,1]}
          Shear.Y {type=MovieSceneScriptingFloatChannel, const [0,1]}
         ]
        }]
      }
      {
       type=MovieSceneFloatTrack
       display_name=Render Opacity
       sections [{type=MovieSceneFloatSection, end_frame=6, start_frame=0, channels [None {type=MovieSceneScriptingFloatChannel, keys [[0,1],[6,0]]}]}]
      }
     ]
    }]}
 }
 Intro_Alt {
  binding_count=1
  length_seconds=0.65
  object=&self:Intro_Alt
  sequence {bindings [Root {
     id=#6
     possessed_class=Overlay
     tracks [
      {
       type=MovieScene2DTransformTrack
       sections [{
         type=MovieScene2DTransformSection
         end_frame=13
         start_frame=0
         channels [
          Translation.X {type=MovieSceneScriptingFloatChannel, const [0,1]}
          Translation.Y {type=MovieSceneScriptingFloatChannel, keys [[0,0],[5.4167,-50],[8.6667,-50],[13,0]]}
          Angle {type=MovieSceneScriptingFloatChannel, const [0,1]}
          Scale.X {type=MovieSceneScriptingFloatChannel, keys [[0,1],[5.4167,1.3],[8.6667,1.3],[13,1]]}
          Scale.Y {type=MovieSceneScriptingFloatChannel, keys [[0,1],[5.4167,1.3],[8.6667,1.3],[13,1]]}
          Shear.X {type=MovieSceneScriptingFloatChannel, const [0,1]}
          Shear.Y {type=MovieSceneScriptingFloatChannel, const [0,1]}
         ]
        }]
      }
      {
       type=MovieSceneFloatTrack
       display_name=Render Opacity
       sections [{
         type=MovieSceneFloatSection
         end_frame=13
         start_frame=0
         channels [None {type=MovieSceneScriptingFloatChannel, keys [[0,0],[5.4167,1],[8.6667,1],[13,0.6]]}]
        }]
      }
     ]
    }]}
 }
 Intro_Alt_Total {
  binding_count=1
  length_seconds=0.65
  object=&self:Intro_Alt_Total
  sequence {bindings [Root {
     id=#6
     possessed_class=Overlay
     tracks [
      {
       type=MovieScene2DTransformTrack
       sections [{
         type=MovieScene2DTransformSection
         end_frame=13
         start_frame=0
         channels [
          Translation.X {type=MovieSceneScriptingFloatChannel, const [0,1]}
          Translation.Y {type=MovieSceneScriptingFloatChannel, keys [[0,0],[5.4167,-50],[8.6667,-50],[13,0]]}
          Angle {type=MovieSceneScriptingFloatChannel, const [0,1]}
          Scale.X {type=MovieSceneScriptingFloatChannel, keys [[0,1],[5.4167,1.3],[8.6667,1.3],[13,1]]}
          Scale.Y {type=MovieSceneScriptingFloatChannel, keys [[0,1],[5.4167,1.3],[8.6667,1.3],[13,1]]}
          Shear.X {type=MovieSceneScriptingFloatChannel, const [0,1]}
          Shear.Y {type=MovieSceneScriptingFloatChannel, const [0,1]}
         ]
        }]
      }
      {
       type=MovieSceneFloatTrack
       display_name=Render Opacity
       sections [{
         type=MovieSceneFloatSection
         end_frame=13
         start_frame=0
         channels [None {type=MovieSceneScriptingFloatChannel, thinned_from=4, keys [[0,0],[5.4167,1],[13,1]]}]
        }]
      }
     ]
    }]}
 }
]
graphs [
 DisplayCompact {nodes [
   DisplayCompact() {is=title K2Node_FunctionEntry}
   n1 = Get!() -> (
    NUMERICTEXT_FONT_FontObject:Object Reference
    NUMERICTEXT_FONT_FontMaterial:Object Reference
    NUMERICTEXT_FONT_OutlineSettings:Font Outline Settings Structure
    NUMERICTEXT_FONT_TypefaceFontName:Name
    NUMERICTEXT_FONT_Size:float=~
    NUMERICTEXT_FONT_LetterSpacing:int
    NUMERICTEXT_FONT_SkewAmount:float=~
    NUMERICTEXT_FONT_bForceMonospaced:bool=~
    NUMERICTEXT_FONT_bMaterialIsStencil:bool=~
    NUMERICTEXT_FONT_MonospacedWidth:float=~
   )
   > SetFont!(
    execute
    self=Get Text_Number() -> (Text_Number:Text Block Object Reference)
    InFontInfo_FontObject:Object Reference=@n1
    InFontInfo_FontMaterial:Object Reference=@n1.NUMERICTEXT_FONT_FontMaterial
    InFontInfo_OutlineSettings:Font Outline Settings Structure=@n1.NUMERICTEXT_FONT_OutlineSettings
    InFontInfo_TypefaceFontName:Name=@n1.NUMERICTEXT_FONT_TypefaceFontName
    InFontInfo_Size:float=Select(Option 0:float=30, Option 1:float=74.6, Index=Get IsCompact() -> (IsCompact:bool)) -> (ReturnValue:float)
    InFontInfo_LetterSpacing:int=@n1.NUMERICTEXT_FONT_LetterSpacing
    InFontInfo_SkewAmount:float=@n1.NUMERICTEXT_FONT_SkewAmount
    InFontInfo_bForceMonospaced:bool=@n1.NUMERICTEXT_FONT_bForceMonospaced
    InFontInfo_bMaterialIsStencil:bool=@n1.NUMERICTEXT_FONT_bMaterialIsStencil
    InFontInfo_MonospacedWidth:float=@n1.NUMERICTEXT_FONT_MonospacedWidth
   ) -> (then)
  ]}
 EventGraph {nodes [
   n1 = IntroFinished() {is=title K2Node_CustomEvent}
   > Set IsAnimationStarted(IsAnimationStarted:bool=false) -> (Output_Get:bool)
   Event PreConstruct()
   > SetVisibility(
    self=Get CommonBorder_divider() -> (CommonBorder_divider:Common Border Object Reference)
    InVisibility=Select(Option 0:ESlateVisibility Enum=Collapsed, Option 1:ESlateVisibility Enum=HitTestInvisible, Index=Get showDivider() -> (showDivider:bool)) -> (ReturnValue:ESlateVisibility Enum)
   )
   > SetRenderScale!(
    execute
    self:Widget Object Reference=Get SizeBox_Container() -> (SizeBox_Container:Size Box Object Reference)
    Scale:Vector 2D Structure=Select(Option 0:Vector 2D Structure=(X=1,Y=1), Option 1:Vector 2D Structure=(X=1.4,Y=1.4), Index=Get bIsTotal() -> (bIsTotal:bool)) -> (ReturnValue:Vector 2D Structure)
   ) -> (then)
   Event Construct()
   > UpdateS39PostMatchVisibility()
   > SetRenderOpacity(self=Get Root() -> (Root:Overlay Object Reference), InOpacity=0)
   > BindToAnimationFinished(Animation=GetIntroAnimation!(self) -> (ReturnValue:Widget Animation Object Reference), Delegate=@n1)
   Event OnIntroPlayed() {is=title K2Node_Event}
   > PlayIntro_VIZ()
   Event OnIsCompactChanged() {is=title K2Node_Event}
   > DisplayCompact() {is=title K2Node_CallFunction}
   Event OnEntryInitialized() {is=title K2Node_Event}
   > "SetText (Text)"(self=Get Text_Number(), InText="To Text (Integer)"(Value=GetXPValue!(self:Fort Match XPTotal Entry Object Reference) -> (ReturnValue:int=0)))
   > "SetText (Text)"(self=Get TextBlock_XPType() -> (TextBlock_XPType:Common Text Object Reference), InText=Get XPType() -> (XPType:Text))
   n2 = Event SetXPAnimationState!() -> (OutputDelegate, then, IsActive:bool) {is=title K2Node_Event}
   > SetIsAnimationActive!(execute, self, IsActive:bool=@n2.IsActive) -> (then)
  ]}
 GetIntroAnimation {nodes [
   GetIntroAnimation() {is=title K2Node_FunctionEntry}
   n1 = Get bIsTotal()
   n2 = Get Intro_Alt() -> (Intro_Alt:Widget Animation Object Reference)
   n3 = Get Intro_Alt_Total() -> (Intro_Alt_Total:Widget Animation Object Reference)
   n4 = Get Intro()
   n5 = Get bIsTotal()
   n6 = Get IntroTotal() -> (IntroTotal:Widget Animation Object Reference)
   > Return Node!(
    execute
    ReturnValue:Widget Animation Object Reference=Select(
     Option 0:Widget Animation Object Reference=Select(Option 0=@n4, Option 1=@n6, Index=@n5)
     Option 1:Widget Animation Object Reference=Select(Option 0=@n2, Option 1=@n3, Index=@n1)
     Index=Get IsUsingS39PostmatchUpdates() -> (IsUsingS39PostmatchUpdates:bool)
    ) -> (ReturnValue:Widget Animation Object Reference)
   )
  ]}
 ResetAnim {nodes [
   ResetAnim() {is=title K2Node_FunctionEntry}
   > StopAllAnimations(self:User Widget Object Reference)
   n1 = GetIntroAnimation() {is=title K2Node_CallFunction}
   > PlayAnimation(InAnimation=@n1, StartAtTime=GetEndTime(self=@n1), PlayMode=Reverse)
   > Set IsAnimationStarted(IsAnimationStarted=false)
  ]}
 SetIsAnimationActive {nodes [
   n1 = SetIsAnimationActive!() -> (then, IsActive:bool=~) {is=title K2Node_FunctionEntry}
   > n2 = Set bMatchXPTotalEntryActive(bMatchXPTotalEntryActive:bool=@n1) -> (Output_Get:bool)
   > SetIsAudioActive!(execute, self, IsAudioActive:bool=@n2) -> (then)
  ]}
 SetIsAudioActive {nodes [
   n1 = SetIsAudioActive!() -> (then, IsAudioActive:bool=~) {is=title K2Node_FunctionEntry}
   > n2 = Branch(Condition=Get IsAnimationStarted() -> (IsAnimationStarted:bool))
   > n3 = Branch(Condition=@n1)
   n4 = GetIntroAnimation()
   > Branch(Condition="float < float"(
     A=GetAnimationCurrentTime!(self:User Widget Object Reference, InAnimation:Widget Animation Object Reference=@n4) -> (ReturnValue:float)
     B=0.1
    ))
   > Branch(Condition="integer > integer"(A=GetXPValue() -> (ReturnValue=0), B=0) {is=title K2Node_CallFunction})
   > n5 = Get!(execute:Exec) -> (then:Exec, else:Exec, XpScoreAudio:Audio Component Object Reference)
   > n6 = FadeOut!(
    execute
    self:Audio Component Object Reference=@n5
    FadeOutDuration:float=0.5
    FadeVolumeLevel:float=0
    FadeCurve:EAudioFaderCurve Enum=Linear
   ) -> (then)
   n7 = SpawnSound2D!(
    execute=[@n6.then, @n8.then]
    Sound:Sound Base Object Reference=/Game/Sounds/Fort_UI_Sounds/Gameplay/XP/Fort_XP_Score_Popup_Cue
    VolumeMultiplier:float=1
    PitchMultiplier:float=1
    StartTime:float=0
    ConcurrencySettings:Sound Concurrency Object Reference
    bPersistAcrossLevelTransition:bool=false
    bAutoDestroy:bool=true
   ) -> (then, ReturnValue:Audio Component Object Reference)
   > Set XpScoreAudio(XpScoreAudio:Audio Component Object Reference=@n7) -> (Output_Get:Audio Component Object Reference)
   n8 = Branch(execute=@n5.else, Condition=AND Boolean(A=Get bMatchXPTotalEntryActive() -> (bMatchXPTotalEntryActive:bool), B=IsUsingCompactUI()))
   n9 = Get(execute=[@n2.else, @n3.else])
   > FadeOut(self=@n9, FadeOutDuration=0.5, FadeVolumeLevel=0, FadeCurve=Linear)
  ]}
 SkipVisualization {nodes [
   SkipVisualization() {is=title K2Node_FunctionEntry}
   > StopAllAnimations()
   > SetIsCompact!(execute, self:Fort Match XPTotal Entry Object Reference, IsCompactSize:bool=true) -> (then)
   > SetRenderOpacity(self=Get Root(), InOpacity=Select(Option 0:float=0.6, Option 1:float=1, Index=Get bIsTotal()) -> (ReturnValue:float))
   > SetIsAnimationActive(IsActive=false) {is=title K2Node_CallFunction}
  ]}
 UpdateS39PostMatchVisibility {nodes [
   UpdateS39PostMatchVisibility() {is=title K2Node_FunctionEntry}
   > n1 = GetConsoleVariableBoolValue!(execute, VariableName:String=Athena.MatchXPScreenWidget.S39EnablePostGameUpdates) -> (then, ReturnValue:bool)
   > Set IsUsingS39PostmatchUpdates(IsUsingS39PostmatchUpdates:bool=@n1) -> (Output_Get:bool)
  ]}
 Vizualization Graph {nodes [
   PlayIntro_VIZ() {is=title K2Node_CustomEvent}
   > SetRenderOpacity(self=Get Root(), InOpacity=1)
   n1 = GetIntroAnimation()
   > PlayAnimation(InAnimation=@n1)
   > Set IsAnimationStarted(IsAnimationStarted=true)
   > SetIsAudioActive(IsAudioActive=Get bMatchXPTotalEntryActive()) {is=title K2Node_CallFunction}
   > Branch(Condition=Get AnimateOutWhenFinished() -> (AnimateOutWhenFinished:bool))
   > BindToAnimationFinished(Animation=@n1, Delegate=Create Event())
   PlayOutro_VIZ() {is=title K2Node_CustomEvent}
   > PlayAnimation(InAnimation=Get Anim_Outro() -> (Anim_Outro:Widget Animation Object Reference))
  ]}
]
variables [
 AnimateOutWhenFinished {type=boolean}
 bMatchXPTotalEntryActive {type=boolean}
 HandleXpScoreAudio {type=TimerHandle}
 IsAnimationStarted {type=boolean}
 IsSmallSize {type=boolean}
 IsUsingS39PostmatchUpdates {type=boolean}
 NUMERICTEXT_FONT {
  type=SlateFontInfo
  default {
   bForceMonospaced=false
   bMaterialIsStencil=false
   fontMaterial=null
   fontObject=&Burbank
   letterSpacing=0
   monospacedWidth=1
   size=56
   skewAmount=0
   typefaceFontName=Burbank-Black
   outlineSettings {
    bApplyOutlineToDropShadows=false
    bMiteredCorners=false
    bSeparateFillAlpha=false
    outlineMaterial=null
    outlineSize=0
    outlineColor=<rgba 0 0 0 1>
   }
  }
 }
 showDivider {type=boolean}
 XpScoreAudio {type=AudioComponent, default=null}
]
widgets {
 info {inheritedWidgetCount=1, parentClass=/Script/FortniteUI.FortMatchXPTotalEntry, rootWidgetClass=&Overlay, widgetCount=12}
 widgets [
  Root {widgetClassPath=&Overlay, bIsVariable=true, widget=Root, properties {renderTransformPivot=<0.5 1>}}
  VerticalBox_0 {
   widgetClassPath=&VerticalBox
   parent=Root
   widget=VerticalBox_0
   slot {class=OverlaySlot, properties {horizontalAlignment=HAlign_Center, verticalAlignment=VAlign_Bottom}}
  }
  SizeBox_Container {
   widgetClassPath=&SizeBox
   bIsVariable=true
   parent=VerticalBox_0
   widget=SizeBox_Container
   properties {bOverride_WidthOverride=true, widthOverride=522.2969}
   slot {class=VerticalBoxSlot}
  }
  ContainerStack {
   widgetClassPath=&HorizontalBox
   parent=SizeBox_Container
   widget=ContainerStack
   properties {renderTransform {angle=0, scale=<0.8 0.8>, shear=<0 0>, translation=<0 0>}}
   slot {class=SizeBoxSlot, properties {horizontalAlignment=HAlign_Center, padding=<ltrb 0 -8 0 -8>}}
  }
  TextBlock_XPType {
   widgetClassPath=&CommonTextBlock
   bIsVariable=true
   parent=ContainerStack
   widget=TextBlock_XPType
   properties {
    text=Match Earned XP
    textTransformPolicy=ToUpper
    font {
     bForceMonospaced=false
     bMaterialIsStencil=false
     fontMaterial=null
     fontObject=&Burbank
     letterSpacing=0
     monospacedWidth=1
     size=20
     skewAmount=0
     typefaceFontName=BurbankRegular-Black
     outlineSettings {
      bApplyOutlineToDropShadows=false
      bMiteredCorners=false
      bSeparateFillAlpha=true
      outlineMaterial=null
      outlineSize=0
      outlineColor=<rgba 0 0 0 0>
     }
    }
    renderTransform {angle=0, scale=<1 1>, shear=<-12 0>, translation=<0 0>}
    shadowOffset=<0 0>
   }
   slot {class=HorizontalBoxSlot, properties {horizontalAlignment=HAlign_Left, verticalAlignment=VAlign_Center, padding=<ltrb 12 0 20 0>}}
  }
  ScaleBox_0 {
   widgetClassPath=&ScaleBox
   parent=ContainerStack
   widget=ScaleBox_0
   properties {stretch=UserSpecified}
   slot {class=HorizontalBoxSlot}
  }
  HorizontalBox_5 {widgetClassPath=&HorizontalBox, parent=ScaleBox_0, widget=HorizontalBox_5, slot {class=ScaleBoxSlot}}
  Text_Number {
   widgetClassPath=&UEFN_TextBlock
   bIsVariable=true
   parent=HorizontalBox_5
   widget=Text_Number
   properties {
    text=0
    font {
     bForceMonospaced=false
     bMaterialIsStencil=false
     fontMaterial=null
     fontObject=&Burbank
     letterSpacing=0
     monospacedWidth=1
     size=56.25
     skewAmount=0.15
     typefaceFontName=BurbankRegular-Black
     outlineSettings {
      bApplyOutlineToDropShadows=false
      bMiteredCorners=false
      bSeparateFillAlpha=false
      outlineMaterial=null
      outlineSize=0
      outlineColor=<rgba 0 0 0 1>
     }
    }
   }
   slot {class=HorizontalBoxSlot, properties {verticalAlignment=VAlign_Center, padding=<ltrb 0 0 15 0>}}
  }
  TextBlock_XPSymbol {
   widgetClassPath=&CommonTextBlock
   bIsVariable=true
   parent=HorizontalBox_5
   widget=TextBlock_XPSymbol
   properties {
    justification=Center
    text=XP
    textTransformPolicy=ToUpper
    font {
     bForceMonospaced=false
     bMaterialIsStencil=false
     fontMaterial=null
     fontObject=&BurbankSmall
     letterSpacing=0
     monospacedWidth=1
     size=32
     skewAmount=0
     typefaceFontName=Black
     outlineSettings {
      bApplyOutlineToDropShadows=false
      bMiteredCorners=false
      bSeparateFillAlpha=true
      outlineMaterial=null
      outlineSize=6
      outlineColor=<rgba 0 0 0 0.0938>
     }
    }
    renderTransform {angle=0, scale=<1 1>, shear=<-12 0>, translation=<0 -6>}
    shadowColorAndOpacity=<rgba 0 0 0 1>
    shadowOffset=<0 0>
   }
   slot {class=HorizontalBoxSlot, properties {verticalAlignment=VAlign_Center, padding=<ltrb 0 0 0 -7>}}
  }
  CommonBorder_divider {
   widgetClassPath=&CommonBorder
   bIsVariable=true
   parent=VerticalBox_0
   widget=CommonBorder_divider
   properties {horizontalAlignment=HAlign_Center, desiredSizeScale=<0.6146 1>, padding=<ltrb 0 2 0 6>}
   slot {class=VerticalBoxSlot, properties {horizontalAlignment=HAlign_Center}}
  }
  Image_divider {
   widgetClassPath=&Image
   bIsVariable=true
   parent=CommonBorder_divider
   widget=Image_divider
   properties {
    brush {
     bIsDynamicallyLoaded=false
     drawAs=Image
     imageType=NoImage
     mirroring=NoMirror
     resourceName=None
     resourceObject=null
     tiling=NoTile
     imageSize=<430 2>
     margin=<ltrb 0 0 0 0>
     outlineSettings {
      bUseBrushTransparency=false
      roundingType=HalfHeightRadius
      width=0
      color {colorUseRule=UseColor_Specified, specifiedColor=<rgba 0 0 0 0>}
      cornerRadii=<0 0 0 0>
     }
     tintColor {colorUseRule=UseColor_Specified, specifiedColor=<rgba 1 1 1 1>}
     uVRegion {bIsValid=false, max=<0 0>, min=<0 0>}
    }
    colorAndOpacity=<rgba 1 1 1 0.15>
   }
   slot {class=BorderSlot, properties {horizontalAlignment=HAlign_Center, padding=<ltrb 0 2 0 6>}}
  }
  TutorialHighlight {widgetClassPath=&TutorialHighlight}
 ]
}
