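# Full-screen minigame scoreboard widget with a "FINAL SCORES" header, from the post-game XP screen folder.
# On show, it configures stat columns from the minigame, then fills the scroll box with sorted team or player rows.
# Row widgets are reused by index, with surplus ones collapsed, and the list scrolls to the owning or spectated player.
# A re-armed timer refreshes stats while the minigame runs; PreConstruct builds design-time preview rows from TestData.
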WidgetBlueprint /FortniteCoreUI/Athena/HUD/EndGame/PostGameFlowScreens/DefaultScreens/XPScreen/WBP_MatchXPScoreboard
graphs [
 EventGraph {nodes [
   UpdateActiveStats() {is=title K2Node_CustomEvent}
   > UpdateStats!(execute, self, ScoreboardState:EMinigameScoreboardStates Enum=Max) -> (then)
   > Refresh Active Stats()
   Event DoShowUI() {is=title K2Node_Event}
   > Set bInGameScoreboard(bInGameScoreboard:bool=true) -> (Output_Get:bool)
   > SetVisibility(InVisibility=SelfHitTestInvisible)
   > SetStatsColumns()
   > UpdateStats(ScoreboardState=Max)
   > ShowTargetPlayerStateInScrollBox()
   > Refresh Active Stats()
   Event DoHideUI() {is=title K2Node_Event}
   > SetVisibility(InVisibility=Collapsed)
   Event Construct()
   > SetVisibility(
    self=Get Header() -> (Header:Horizontal Box Object Reference)
    InVisibility=Select(Option 0=Collapsed, Option 1=SelfHitTestInvisible, Index=Get bShowHeader() -> (bShowHeader:bool))
   )
   n1 = Event PreConstruct()
   > Branch(Condition=@n1.IsDesignTime)
   n2 = Get TestData() -> (TestData:Array of Minigame Scoreboard Player Row Structures)
   > n3 = For Each Loop(Array:Array of Minigame Scoreboard Player Row Structures=@n2) -> (Array Element:Minigame Scoreboard Player Row Structure)
   > n4 = MakePlayerScoreboardRow!(
    execute
    self
    PlayerRow:Minigame Scoreboard Player Row Structure=@n3
    PlayerIndex:int=@n3.Array Index
    ShowWins:bool=Get IsRoundsGame() -> (IsRoundsGame:bool)
    NumPlayers:int=Length(TargetArray:Array of Minigame Scoreboard Player Row Structures=@n2)
    Standing:int=@n3.Array Index
   ) -> (then, ScoreboardRow:Scoreboard UI Row Object Reference)
   > AddChild(self=Get DataScrollBox() -> (DataScrollBox:Scroll Box Object Reference), Content=@n4)
  ]}
 GetScoreboardStateNames {nodes [
   GetScoreboardStateNames() {is=title K2Node_FunctionEntry}
   > n1 = Get!(execute:Exec) -> (then:Exec, else:Exec, CurrentMinigame:Fort Minigame Object Reference)
   n2 = GetMinigameStatsComponent!(ContextActor:Actor Object Reference=Get Volume!(self:Fort Minigame Object Reference=@n1) -> (Volume:Actor Object Reference)) -> (ReturnValue:Fort Minigame Stats Component Object Reference)
   > Is Valid(InputObject=@n2)
   > n3 = For Each Loop(Array:Array of Fort Minigame Stat Filter Object References=GetTrackedStatsForScoreboardColumns!(self:Fort Minigame Stats Component Object Reference=@n2, bOnlyPersistentStats:bool=Get bIsCareerScoreboard() -> (bIsCareerScoreboard:bool)) -> (OutScoreboardColumnSortedTrackedStats:Array of Fort Minigame Stat Filter Object References)) -> (Array Element:Fort Minigame Stat Filter Object Reference)
   Return Node!(execute=@n3.Completed, OutNames:Array of Texts=Get NamesArray() -> (NamesArray:Array of Texts))
   > Add(
    TargetArray:Array of Texts=Get NamesArray()
    NewItem:"Text (by ref)"=Get HudShortDescription!(self:Fort Minigame Stat Filter Object Reference=@n3) -> (HudShortDescription:Text)
   ) {is=title K2Node_CallArrayFunction}
  ]}
 GetTargetPlayerState {nodes [
   GetTargetPlayerState() {is=title K2Node_FunctionEntry}
   > n1 = Set OwningPlayer(OwningPlayer:Player Controller Object Reference=GetOwningPlayer()) -> (Output_Get:Player Controller Object Reference)
   > n2 = Is Valid(InputObject=@n1)
   Return Node!(execute=@n2.Is Not Valid, PlayerState:Player State Object Reference)
   > n3 = Set SpectatorPawn(SpectatorPawn:Spectator Pawn Object Reference=GetSpectatorPawn!(self:Player Controller Object Reference=Get OwningPlayer() -> (OwningPlayer:Player Controller Object Reference)) -> (ReturnValue:Spectator Pawn Object Reference)) -> (Output_Get:Spectator Pawn Object Reference)
   > n4 = Is Valid(InputObject=@n3)
   > Return Node(PlayerState=Get PlayerState!(self:Pawn Object Reference=Get SpectatorPawn() -> (SpectatorPawn:Spectator Pawn Object Reference)) -> (PlayerState:Player State Object Reference))
   Return Node(execute=@n4.Is Not Valid, PlayerState=Get PlayerState(self:Controller Object Reference=Get OwningPlayer()))
  ]}
 GetTargetScoreRow {nodes [
   GetTargetScoreRow!() -> (then, TargetPlayerState:Player State Object Reference) {is=title K2Node_FunctionEntry}
   > n1 = For Each Loop(Array:Array of Widget Object References=GetAllChildren(self=Get DataScrollBox())) -> (Array Element:Widget Object Reference)
   Return Node!(execute=@n1.Completed, Widget:Widget Object Reference)
   > n2 = Branch(Condition=Get bIsTeamGame() -> (bIsTeamGame:bool))
   > n3 = Cast To Scoreboard_Team_Widget!(execute, Object=@n1) -> (then, CastFailed, AsScoreboard Team Widget:Scoreboard Team Widget Object Reference) {is=title K2Node_DynamicCast}
   > n4 = GetWidgetForPlayerState!(
    execute
    self:Scoreboard Team Widget Object Reference=@n3
    PlayerState:Player State Object Reference=Get TargetPlayerState() -> (TargetPlayerState:Player State Object Reference)
   ) -> (then, Widget:Widget Object Reference)
   > Is Valid(InputObject=@n4)
   > Return Node(Widget=@n3)
   n5 = Cast To Scoreboard_UI_Row!(execute=@n2.else, Object=@n1) -> (then, CastFailed, AsScoreboard UI Row:Scoreboard UI Row Object Reference) {is=title K2Node_DynamicCast}
   n6 = Get PlayerBucketData!(self:Scoreboard UI Row Object Reference=@n5) -> (PlayerBucketData:Minigame Scoreboard Player Row Structure)
   > Branch(Condition="Equal (Object)"(
     A=Get TargetPlayerState()
     B=Break Minigame Scoreboard Player Row!(MinigameScoreboardPlayerRow:"Minigame Scoreboard Player Row Structure (by ref)"=@n6) -> (PlayerState:Fort Player State Object Reference, bHasLeftGame:bool=~, bHideFromDisplay:bool=~) {is=title K2Node_BreakStruct}
    ) {is=title K2Node_CallFunction})
   > Return Node(Widget=@n5)
  ]}
 HandleScoreboardColumnsData {nodes [
   n1 = HandleScoreboardColumnsData!() -> (then, ScoreBoardStatNames:Array of Texts, bTeamGame:bool=~, bDisplayingWins:bool=~) {is=title K2Node_FunctionEntry}
   > n2 = Is Valid(InputObject=Get ScoreboardColumnWidget() -> (ScoreboardColumnWidget:Scoreboard UI Column Labels Object Reference))
   n3 = Create Widget!(
    execute:Exec=@n2.Is Not Valid
    Class:User Widget Class Reference=/Game/Items/Traps/Blueprints/Toys/Scoreboard_UI_ColumnLabels.Scoreboard_UI_ColumnLabels_C
    OwningPlayer:Player Controller Object Reference=GetOwningPlayer()
    DataText:Array of Texts=@n1
    bIsTeamGame:bool=@n1.bTeamGame
    bDisplayingWins:bool=@n1.bDisplayingWins
    bIngameScoreboard:bool=Get bInGameScoreboard() -> (bInGameScoreboard:bool)
    WinCondition:EMinigameWinCondition Enum=Get WinCondition() -> (WinCondition:EMinigameWinCondition Enum=MostRoundWins)
    bShowCumulativeScoreboard:bool=Get bShowCumulativeScoreboard() -> (bShowCumulativeScoreboard:bool)
   ) -> (then:Exec, ReturnValue:Scoreboard UI Column Labels Object Reference)
   > n4 = Set ScoreboardColumnWidget(ScoreboardColumnWidget:Scoreboard UI Column Labels Object Reference=@n3) -> (Output_Get:Scoreboard UI Column Labels Object Reference)
   > AddChild(self=Get ColumnHolder() -> (ColumnHolder:Border Object Reference), Content=@n4)
   n5 = Get ScoreboardColumnWidget()
   Set DataText!(
    execute=@n2.Is Valid
    DataText:Array of Texts=Get ScoreBoardStatNames() -> (ScoreBoardStatNames:Array of Texts)
    self:Scoreboard UI Column Labels Object Reference=@n5
   ) -> (then, Output_Get:Array of Texts)
   > Set bIsTeamGame!(execute, bIsTeamGame:bool=Get bTeamGame() -> (bTeamGame:bool), self:Scoreboard UI Column Labels Object Reference=@n5) -> (then, Output_Get:bool)
   > Set bDisplayingWins!(execute, bDisplayingWins:bool=Get bDisplayingWins() -> (bDisplayingWins:bool), self:Scoreboard UI Column Labels Object Reference=@n5) -> (then, Output_Get:bool)
   > Set WinCondition!(
    execute
    WinCondition:EMinigameWinCondition Enum=Get WinCondition() -> (WinCondition=MostRoundWins)
    self:Scoreboard UI Column Labels Object Reference=@n5
   ) -> (then, Output_Get:EMinigameWinCondition Enum=MostRoundWins)
   > Set bShowCumulativeScoreboard!(execute, bShowCumulativeScoreboard:bool=Get bShowCumulativeScoreboard(), self:Scoreboard UI Column Labels Object Reference=@n5) -> (then, Output_Get:bool)
   > UpdateColumnDisplay(self:Scoreboard UI Column Labels Object Reference=@n5)
  ]}
 HideUnusedRows {nodes [
   HideUnusedRows!() -> (then, bTeamGame:bool=~, NumBuckets:int) {is=title K2Node_FunctionEntry}
   > n1 = Sequence()
   > n2 = For Each Loop(Array:Array of Scoreboard Team Widget Object References=Get TeamScoreDisplays() -> (TeamScoreDisplays:Array of Scoreboard Team Widget Object References)) -> (Array Element:Scoreboard Team Widget Object Reference)
   n3 = "integer < integer"!(A:int=@n2.Array Index, B:int=Get NumBuckets() -> (NumBuckets:int=0)) -> (ReturnValue:bool) {is=title K2Node_CallFunction}
   n4 = Get bIsTeamGame()
   > SetVisibility(self=@n2, InVisibility=Select(Option 0=Collapsed, Option 1=HitTestInvisible, Index=AND Boolean(A=@n3, B=@n4)))
   n5 = For Each Loop(
    Exec=@n1.then_1
    Array:Array of Scoreboard UI Row Object References=Get PlayerScoreDisplays() -> (PlayerScoreDisplays:Array of Scoreboard UI Row Object References)
   ) -> (Array Element:Scoreboard UI Row Object Reference)
   n6 = "integer < integer"(A=@n5.Array Index, B=Get NumBuckets() -> (NumBuckets=0))
   n7 = NOT Boolean(A=Get bIsTeamGame())
   > SetVisibility(self=@n5, InVisibility=Select(Option 0=Collapsed, Option 1=HitTestInvisible, Index=AND Boolean(A=@n6, B=@n7)))
  ]}
 IsMinigameTeamBased {nodes [
   n1 = IsMinigameTeamBased!() -> (then, MinigameToCheck:Fort Minigame Object Reference, MinigameTeamComponent:Fort Minigame Team Component Object Reference) {is=title K2Node_FunctionEntry}
   n2 = Get bIsCareerScoreboard()
   n3 = Get bForceIndividualScoresInTeamGame!(self:Fort Minigame Scoreboard Component Object Reference=Get ScoreboardComponent!(self:Fort Minigame Object Reference=@n1) -> (ScoreboardComponent:Fort Minigame Scoreboard Component Object Reference)) -> (bForceIndividualScoresInTeamGame:bool)
   > Return Node!(
    execute
    ReturnValue:bool=AND Boolean!(
     A=NOT Boolean(A=@n3)
     B=GetTeamMinigame!(self:Fort Minigame Team Component Object Reference=@n1.MinigameTeamComponent) -> (ReturnValue:bool)
     C:bool=NOT Boolean(A=@n2)
    ) -> (ReturnValue)
   )
  ]}
 MakePlayerScoreboardRow {nodes [
   n1 = MakePlayerScoreboardRow!() -> (then, PlayerRow:Minigame Scoreboard Player Row Structure, PlayerIndex:int, ShowWins:bool=~, NumPlayers:int, Standing:int) {is=title K2Node_FunctionEntry}
   > n2 = Create Widget!(
    execute
    Class=/Game/Items/Traps/Blueprints/Toys/Scoreboard_UI_Row.Scoreboard_UI_Row_C
    OwningPlayer=GetOwningPlayer()
    bDisplayWins:bool=@n1.ShowWins
    PlayerBucketData:Minigame Scoreboard Player Row Structure=@n1
    bIngameScorerow:bool=Get bInGameScoreboard()
    bIsTeamGame=Get bIsTeamGame()
    MedalStanding:int=@n1.Standing
    TeamColorParam:double=0
    WinCondition=Get WinCondition() -> (WinCondition=MostRoundWins)
    bShowCumulativeScoreboard=Get bShowCumulativeScoreboard()
   ) -> (then, ReturnValue:Scoreboard UI Row Object Reference)
   > Return Node!(execute, ScoreboardRow:Scoreboard UI Row Object Reference=@n2)
  ]}
 MakeTeamScoreboardRow {nodes [
   n1 = MakeTeamScoreboardRow!() -> (then, MinigameScoreboardBucketRow:"Minigame Scoreboard Bucket Row Structure (by ref)", IsRoundsGame:bool=~) {is=title K2Node_FunctionEntry}
   > n2 = Create Widget!(
    execute
    Class=/Game/Scoreboard_Team_Widget.Scoreboard_Team_Widget_C
    OwningPlayer=GetOwningPlayer()
    bShowVictoryScore:bool=Get IsRoundsGame()
    TeamScoreboardBucket:Minigame Scoreboard Bucket Row Structure=Get MinigameScoreboardBucketRow() -> (MinigameScoreboardBucketRow:"Minigame Scoreboard Bucket Row Structure (by ref)")
    bIngameTeamScore:bool=Get bInGameScoreboard()
    TeamColorParam="To Float (Integer)"(InInt=Break Minigame Scoreboard Bucket Row!(MinigameScoreboardBucketRow:"Minigame Scoreboard Bucket Row Structure (by ref)"=@n1) -> (
      Standing:int
      BucketIndex:int
      PlayerRows:Array of Minigame Scoreboard Player Row Structures
      Name:Text
      TeamColorIndex:int
      Values:Array of Minigame Scoreboard Value Structures
      WinCount:int
      bHighlight:bool=~
     ) {is=title K2Node_BreakStruct}.TeamColorIndex)
   ) -> (then, ReturnValue:Scoreboard Team Widget Object Reference)
   > Return Node!(execute, ReturnValue:Scoreboard Team Widget Object Reference=@n2)
  ]}
 OnBlockMapActionByKey {nodes [
   OnBlockMapActionByKey!() -> (then, InKey:"Key Structure (by ref)") {is=title K2Node_FunctionEntry}
   > Return Node!(execute, bOutMapActionBlocked:bool=false, ReturnValue:bool=false)
  ]}
 OnHandleMapActionKey {nodes [
   OnHandleMapActionKey!() -> (then, InKey:"Key Structure (by ref)", bKeyDown:bool) {is=title K2Node_FunctionEntry}
   > Return Node!(execute, bOutMapActionHandled:bool=false, ReturnValue=false)
  ]}
 OnHandleMapBackAction {nodes [
   OnHandleMapBackAction() {is=title K2Node_FunctionEntry}
   > Return Node!(execute, bOutBackActionHandled:bool=false, ReturnValue=false)
  ]}
 OnShouldCreateMapScreenTab {nodes [
   OnShouldCreateMapScreenTab!() -> (then, WorldContextObject:Object Reference) {is=title K2Node_FunctionEntry}
   > Return Node!(execute, bOutShouldCreateMapScreenTab:bool=false, ReturnValue=false)
  ]}
 OnShouldDisplayBackActionInActionBar {nodes [
   OnShouldDisplayBackActionInActionBar() {is=title K2Node_FunctionEntry}
   > Return Node!(execute, bOutShouldDisplayBackActionInActionBar:bool=false, ReturnValue=false)
  ]}
 OnShouldShowTabOnMapScreenForMinigame {nodes [
   OnShouldShowTabOnMapScreenForMinigame!() -> (then, Minigame:Fort Minigame Object Reference) {is=title K2Node_FunctionEntry}
   > Return Node!(execute, bOutShouldShowTabOnMapScreenForMinigame:bool=false, ReturnValue=false)
  ]}
 Refresh Active Stats {nodes [
   Refresh Active Stats() {is=title K2Node_FunctionEntry}
   > n1 = Get()
   > Branch(Condition=IsInProgress!(self:Fort Minigame Object Reference=@n1) -> (ReturnValue:bool))
   > Set Timer by Event(Delegate=Create Event(), Time=Get StatUpdateRate() -> (StatUpdateRate:double), bLooping=false)
  ]}
 SetStatsColumns {nodes [
   SetStatsColumns() {is=title K2Node_FunctionEntry}
   > Set ScoreboardStatNames(ScoreboardStatNames:Array of Texts=GetScoreboardStateNames!() -> (OutNames:Array of Texts) {is=title K2Node_CallFunction}) -> (Output_Get:Array of Texts)
   > n1 = Get()
   n2 = GetMinigameTeamComponent!(self:Fort Minigame Object Reference=@n1) -> (ReturnValue:Fort Minigame Team Component Object Reference)
   > Is Valid(InputObject=@n2)
   > n3 = Set bIsTeamGame!(
    execute
    bIsTeamGame=IsMinigameTeamBased!(MinigameToCheck:Fort Minigame Object Reference=@n1, MinigameTeamComponent:Fort Minigame Team Component Object Reference=@n2) -> (ReturnValue:bool) {is=title K2Node_CallFunction}
   ) -> (then, Output_Get)
   > Set IsRoundsGame(IsRoundsGame:bool="integer > integer"(A=Get TotalRounds!(self:Fort Minigame Object Reference=@n1) -> (TotalRounds:int=0), B=1) {is=title K2Node_CallFunction}) -> (Output_Get:bool)
   > Set WinCondition!(execute, WinCondition=Get WinCondition!(self:Fort Minigame Object Reference=@n1) -> (WinCondition=MostRoundWins)) -> (then, Output_Get=MostRoundWins)
   > Set bShowCumulativeScoreboard!(
    execute
    bShowCumulativeScoreboard=Get bShowCumulativeScoreboard!(self:Fort Minigame Scoreboard Component Object Reference=Get ScoreboardComponent(self=@n1)) -> (bShowCumulativeScoreboard)
   ) -> (then, Output_Get)
   > HandleScoreboardColumnsData!(
    execute
    self
    ScoreBoardStatNames:Array of Texts=Get ScoreboardStatNames() -> (ScoreboardStatNames:Array of Texts)
    bTeamGame:bool=@n3
    bDisplayingWins:bool=ShowWinsColumn!() -> (ReturnValue:bool)
   ) -> (then) {is=title K2Node_CallFunction}
  ]}
 ShowTargetPlayerStateInScrollBox {nodes [
   ShowTargetPlayerStateInScrollBox() {is=title K2Node_FunctionEntry}
   > n1 = GetTargetPlayerState!(execute, self) -> (then, PlayerState:Player State Object Reference) {is=title K2Node_CallFunction}
   > Is Valid(InputObject=@n1)
   > n2 = GetTargetScoreRow!(execute, self, TargetPlayerState:Player State Object Reference=@n1) -> (then, Widget:Widget Object Reference) {is=title K2Node_CallFunction}
   > Is Valid(InputObject=@n2)
   > ScrollWidgetIntoView!(
    execute
    self:Scroll Box Object Reference=Get DataScrollBox()
    WidgetToFind:Widget Object Reference=@n2
    AnimateScroll:bool=true
    ScrollDestination:EDescendantScrollDestination Enum=IntoView
    Padding:float=0
   ) -> (then)
  ]}
 ShowWinsColumn {nodes [
   ShowWinsColumn() {is=title K2Node_FunctionEntry}
   > n1 = Branch(Condition=Get IsRoundsGame())
   > n2 = Branch(Condition=Get bIsCareerScoreboard())
   > n3 = Get()
   n4 = Get Volume(self=@n3)
   > Return Node!(
    execute
    ReturnValue=IsStatPersistent!(
     self:Fort Minigame Stats Component Object Reference=GetMinigameStatsComponent(ContextActor=@n4)
     InFilterClass:Fort Minigame Stat Filter Class Reference=/Game/Athena/Playset/Minigames/StatFilters/MinigameStat_RoundWins.MinigameStat_RoundWins_C
    ) -> (ReturnValue:bool)
   )
   Return Node(execute=@n3.else, ReturnValue=false)
   Return Node(execute=@n2.else, ReturnValue=true)
   Return Node(execute=@n1.else, ReturnValue=false)
  ]}
 UpdateStats {nodes [
   n1 = UpdateStats!() -> (then, ScoreboardState:EMinigameScoreboardStates Enum=RoundEndDisplayWinner) {is=title K2Node_FunctionEntry}
   > n2 = Get()
   > n3 = Get!(execute, self:Fort Minigame Object Reference=@n2) -> (then, else, ScoreboardComponent:Fort Minigame Scoreboard Component Object Reference)
   n4 = GetMinigameStatsComponent(ContextActor=Get Volume(self=@n2))
   > Is Valid(InputObject=@n4)
   n5 = Get bIsCareerScoreboard()
   n6 = IsPostGame!(self:Fort Minigame Object Reference=@n2) -> (ReturnValue:bool)
   n7 = Get bIsCareerScoreboard()
   > n8 = Set ScoreboardBucketRows(ScoreboardBucketRows:Array of Minigame Scoreboard Bucket Row Structures=GetColumnSortedScoreboardRows!(
     self:Fort Minigame Scoreboard Component Object Reference=@n3
     RowSortOrder:Array of Fort Minigame Stat Filter Object References=Get ScoreboardStatObjects!(self:Fort Minigame Stats Component Object Reference=@n4) -> (ScoreboardStatObjects:Array of Fort Minigame Stat Filter Object References)
     ColumnSortOrder:Array of Fort Minigame Stat Filter Object References=GetTrackedStatsForScoreboardColumns(self=@n4, bOnlyPersistentStats=@n5)
     bGameTotal:bool=OR Boolean(A=@n6, B=@n7)
     bHighlighting:bool=true
     bCountWins:bool=true
     bIsCareerScoreboard:bool=@n7
     ScoreboardState:EMinigameScoreboardStates Enum=@n1
    ) -> (OutScoreboardRows:Array of Minigame Scoreboard Bucket Row Structures)) -> (Output_Get:Array of Minigame Scoreboard Bucket Row Structures)
   > n9 = Set NumBuckets(NumBuckets:int=Length(TargetArray:Array of Minigame Scoreboard Bucket Row Structures=@n8)) -> (Output_Get:int=0)
   > HideUnusedRows!(execute, self, bTeamGame:bool=Get bIsTeamGame(), NumBuckets:int=@n9) -> (then) {is=title K2Node_CallFunction}
   > n10 = For Each Loop(Array:Array of Minigame Scoreboard Bucket Row Structures=@n8) -> (Array Element:Minigame Scoreboard Bucket Row Structure)
   > Set CurrentBucket(CurrentBucket:Minigame Scoreboard Bucket Row Structure=@n10) -> (Output_Get:Minigame Scoreboard Bucket Row Structure)
   > Set CurrentIndex(CurrentIndex:int=@n10.Array Index) -> (Output_Get:int=0)
   > n11 = Branch(Condition=Get bIsTeamGame())
   > n12 = Branch(Condition=Is Valid Index!(
     TargetArray:Array of Scoreboard Team Widget Object References=Get TeamScoreDisplays()
     IndexToTest:int=Get CurrentIndex() -> (CurrentIndex:int=0)
    ) -> (ReturnValue:bool))
   > UpdateTeamData!(
    execute
    self:Scoreboard Team Widget Object Reference="Get (a copy)"(Array:Array of Scoreboard Team Widget Object References=Get TeamScoreDisplays(), Dimension 1=Get CurrentIndex() -> (CurrentIndex=0)) -> (Output:Scoreboard Team Widget Object Reference)
    TeamBucketData:Minigame Scoreboard Bucket Row Structure=Get CurrentBucket() -> (CurrentBucket:Minigame Scoreboard Bucket Row Structure)
    ShowVictoryScore:bool=Get IsRoundsGame()
   ) -> (then)
   n13 = MakeTeamScoreboardRow!(
    execute=@n12.else
    self
    MinigameScoreboardBucketRow:"Minigame Scoreboard Bucket Row Structure (by ref)"=Get CurrentBucket()
    IsRoundsGame:bool=Get IsRoundsGame()
   ) -> (then, ReturnValue:Scoreboard Team Widget Object Reference) {is=title K2Node_CallFunction}
   > Add(
    TargetArray:Array of Scoreboard Team Widget Object References=Get TeamScoreDisplays()
    NewItem:"Scoreboard Team Widget Object Reference (by ref)"=@n13
   )
   > AddChild(self=Get DataScrollBox(), Content=@n13)
   n14 = Break Minigame Scoreboard Bucket Row(MinigameScoreboardBucketRow=Get CurrentBucket()) -> (bHighlight=~)
   n15 = For Each Loop(Exec=@n11.else, Array:Array of Minigame Scoreboard Player Row Structures=@n14.PlayerRows) -> (Array Element:Minigame Scoreboard Player Row Structure)
   > n16 = Branch(Condition=Is Valid Index(TargetArray:Array of Scoreboard UI Row Object References=Get PlayerScoreDisplays(), IndexToTest=Get CurrentIndex() -> (CurrentIndex=0)))
   > UpdatePlayerData!(
    execute
    self:Scoreboard UI Row Object Reference="Get (a copy)"(Array:Array of Scoreboard UI Row Object References=Get PlayerScoreDisplays(), Dimension 1=Get CurrentIndex() -> (CurrentIndex=0)) -> (Output:Scoreboard UI Row Object Reference)
    PlayerBucketData:Minigame Scoreboard Player Row Structure=@n15
   ) -> (then)
   n17 = MakePlayerScoreboardRow(
    execute=@n16.else
    PlayerRow=@n15
    PlayerIndex=Get CurrentIndex() -> (CurrentIndex=0)
    ShowWins=ShowWinsColumn() {is=title K2Node_CallFunction}
    NumPlayers=Get NumBuckets() -> (NumBuckets=0)
    Standing=@n14
   ) {is=title K2Node_CallFunction}
   > Add(
    TargetArray:Array of Scoreboard UI Row Object References=Get PlayerScoreDisplays()
    NewItem:"Scoreboard UI Row Object Reference (by ref)"=@n17
   )
   > AddChild(self=Get DataScrollBox(), Content=@n17)
  ]}
]
variables [
 bInGameScoreboard {type=boolean}
 bIsCareerScoreboard {type=boolean}
 bIsTeamGame {type=boolean}
 bShowCumulativeScoreboard {type=boolean}
 bShowHeader {type=boolean, default=true}
 IsRoundsGame {type=boolean}
 PlayerScoreDisplays {type="array<Scoreboard_UI_Row_C>", default []}
 ScoreboardColumnWidget {type=Scoreboard_UI_ColumnLabels_C, default=null}
 StatUpdateRate {type=number, default=0.5}
 TeamScoreDisplays {type="array<Scoreboard_Team_Widget_C>", default []}
 TestData {
  type="array<MinigameScoreboardPlayerRow>"
  default [
   "" {bHasLeftGame=false, bHighlight=false, playerState=null, teamColorIndex=-1, winCount=0, values []}
   "" {bHasLeftGame=false, bHighlight=false, playerState=null, teamColorIndex=-1, winCount=0, values []}
   "" {bHasLeftGame=false, bHighlight=false, playerState=null, teamColorIndex=-1, winCount=0, values []}
   "" {bHasLeftGame=false, bHighlight=false, playerState=null, teamColorIndex=-1, winCount=0, values []}
   "" {bHasLeftGame=false, bHighlight=false, playerState=null, teamColorIndex=-1, winCount=0, values []}
   "" {bHasLeftGame=false, bHighlight=false, playerState=null, teamColorIndex=-1, winCount=0, values []}
   "" {bHasLeftGame=false, bHighlight=false, playerState=null, teamColorIndex=-1, winCount=0, values []}
   "" {bHasLeftGame=false, bHighlight=false, playerState=null, teamColorIndex=-1, winCount=0, values []}
   "" {bHasLeftGame=false, bHighlight=false, playerState=null, teamColorIndex=-1, winCount=0, values []}
   "" {bHasLeftGame=false, bHighlight=false, playerState=null, teamColorIndex=-1, winCount=0, values []}
   "" {bHasLeftGame=false, bHighlight=false, playerState=null, teamColorIndex=-1, winCount=0, values []}
   "" {bHasLeftGame=false, bHighlight=false, playerState=null, teamColorIndex=-1, winCount=0, values []}
   "" {bHasLeftGame=false, bHighlight=false, playerState=null, teamColorIndex=-1, winCount=0, values []}
   "" {bHasLeftGame=false, bHighlight=false, playerState=null, teamColorIndex=-1, winCount=0, values []}
   "" {bHasLeftGame=false, bHighlight=false, playerState=null, teamColorIndex=-1, winCount=0, values []}
   "" {bHasLeftGame=false, bHighlight=false, playerState=null, teamColorIndex=-1, winCount=0, values []}
   "" {bHasLeftGame=false, bHighlight=false, playerState=null, teamColorIndex=-1, winCount=0, values []}
   "" {bHasLeftGame=false, bHighlight=false, playerState=null, teamColorIndex=-1, winCount=0, values []}
  ]
 }
 WinCondition {type=EMinigameWinCondition, default=MostRoundWins}
]
widgets {
 info {inheritedWidgetCount=2, parentClass=/Script/FortniteUI.AthenaFullScreenScoreboardPanel, rootWidgetClass=&Border, widgetCount=11}
 widgets [
  Border_4 {
   widgetClassPath=&Border
   widget=Border_4
   properties {
    background {
     bIsDynamicallyLoaded=false
     drawAs=Image
     imageType=NoImage
     mirroring=NoMirror
     resourceName=None
     resourceObject=/Game/Items/Traps/Blueprints/Toys/UI_Art_Assets/M_UI_CreativeScoreboard_Backing_Inst
     tiling=NoTile
     imageSize=<32 32>
     margin=<ltrb 0 0 0 0>
     outlineSettings {
      bUseBrushTransparency=false
      roundingType=HalfHeightRadius
      width=0
      color {colorUseRule=UseColor_Specified, specifiedColor=<rgba 0 0 0 0>}
      cornerRadii=<0 0 0 0>
     }
     tintColor {colorUseRule=UseColor_Specified, specifiedColor=<rgba 1 1 1 0>}
     uVRegion {bIsValid=false, max=<0 0>, min=<0 0>}
    }
    padding=<ltrb 0 0 0 0>
   }
  }
  Overlay_1 {widgetClassPath=&Overlay, parent=Border_4, widget=Overlay_1, slot {class=BorderSlot, properties {padding=<ltrb 0 0 0 0>}}}
  HeaderDisplay {
   widgetClassPath=&VerticalBox
   parent=Overlay_1
   widget=HeaderDisplay
   slot {class=OverlaySlot, properties {horizontalAlignment=HAlign_Fill, verticalAlignment=VAlign_Fill}}
  }
  Header {widgetClassPath=&HorizontalBox, bIsVariable=true, parent=HeaderDisplay, widget=Header, slot {class=VerticalBoxSlot}}
  HeaderText {widgetClassPath=&HorizontalBox, parent=Header, widget=HeaderText, slot {class=HorizontalBoxSlot}}
  GameStats {
   widgetClassPath=&CommonTextBlock
   bIsVariable=true
   parent=HeaderText
   widget=GameStats
   properties {
    text=FINAL SCORES
    textTransformPolicy=ToUpper
    font {
     bForceMonospaced=false
     bMaterialIsStencil=false
     fontMaterial=null
     fontObject=&Burbank
     letterSpacing=0
     monospacedWidth=1
     size=56
     skewAmount=0
     typefaceFontName=Burbank-Black
     outlineSettings {
      bApplyOutlineToDropShadows=false
      bMiteredCorners=false
      bSeparateFillAlpha=false
      outlineMaterial=null
      outlineSize=0
      outlineColor=<rgba 0 0 0 1>
     }
    }
    renderTransform {angle=0, scale=<1 1>, shear=<-8 0>, translation=<0 0>}
    shadowOffset=<0 0>
   }
   slot {class=HorizontalBoxSlot}
  }
  ScoreboardDisplay {
   widgetClassPath=&VerticalBox
   bIsVariable=true
   parent=HeaderDisplay
   widget=ScoreboardDisplay
   slot {class=VerticalBoxSlot, properties {padding=<ltrb 0 40 0 0>, size {sizeRule=Fill, value=1}}}
  }
  ColumnHolder {
   widgetClassPath=&Border
   bIsVariable=true
   parent=ScoreboardDisplay
   widget=ColumnHolder
   properties {
    background {
     bIsDynamicallyLoaded=false
     drawAs=NoDrawType
     imageType=NoImage
     mirroring=NoMirror
     resourceName=None
     resourceObject=null
     tiling=NoTile
     imageSize=<32 32>
     margin=<ltrb 0 0 0 0>
     outlineSettings {
      bUseBrushTransparency=false
      roundingType=HalfHeightRadius
      width=0
      color {colorUseRule=UseColor_Specified, specifiedColor=<rgba 0 0 0 0>}
      cornerRadii=<0 0 0 0>
     }
     tintColor {colorUseRule=UseColor_Specified, specifiedColor=<rgba 1 1 1 1>}
     uVRegion {bIsValid=false, max=<0 0>, min=<0 0>}
    }
    padding=<ltrb 0 0 0 0>
   }
   slot {class=VerticalBoxSlot, properties {verticalAlignment=VAlign_Top}}
  }
  CanvasPanel_2 {
   widgetClassPath=&CanvasPanel
   parent=ScoreboardDisplay
   widget=CanvasPanel_2
   slot {class=VerticalBoxSlot, properties {size {sizeRule=Fill, value=1}}}
  }
  DataScrollBox {
   widgetClassPath=&ScrollBox
   bInherited=true
   bIsVariable=true
   parent=CanvasPanel_2
   widget=DataScrollBox
   properties {allowOverscroll=false}
   slot {
    class=CanvasPanelSlot
    properties {bAutoSize=true, layoutData {alignment=<0 0>, anchors {maximum=<1 1>, minimum=<0 0>}, offsets=<ltrb 0 0 0 0>}}
   }
  }
  TutorialHighlight {widgetClassPath=&TutorialHighlight}
 ]
}
