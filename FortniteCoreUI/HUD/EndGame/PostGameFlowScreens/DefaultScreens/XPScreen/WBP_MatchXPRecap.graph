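# Post-match XP recap screen built on FortMatchXPRecap, with XP totals, quest recap, survivor tier badge and scoreboard.
# Pausable timers chain entry intros from the list's last child upward, skipping zero-XP rows.
# Completion broadcasts an event, conditionally showing the scoreboard from minigame settings and a Return to Island button that ends the minigame and resets post-match UI.
# A gameplay-tagged widget host slot accepts injected extension content.
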WidgetBlueprint /FortniteCoreUI/HUD/EndGame/PostGameFlowScreens/DefaultScreens/XPScreen/WBP_MatchXPRecap
animations [
 Anim_IntroTransition {
  binding_count=7
  length_seconds=0.5
  object=&self:Anim_IntroTransition
  sequence {bindings [
    MatchXP_CombatBonus {
     id=#1
     possessed_class=MatchXPTotalEntry_C
     tracks [{
       type=MovieSceneFloatTrack
       display_name=Render Opacity
       sections [{type=MovieSceneFloatSection, end_frame=10, start_frame=0, channels [None {type=MovieSceneScriptingFloatChannel, keys [[0,1],[7,1],[8,0]]}]}]
      }]
    }
    MatchXP_MedalEntry {
     id=#2
     possessed_class=MatchXPTotalEntry_C
     tracks [{
       type=MovieSceneFloatTrack
       display_name=Render Opacity
       sections [{type=MovieSceneFloatSection, end_frame=10, start_frame=0, channels [None {type=MovieSceneScriptingFloatChannel, keys [[0,1],[7,1],[8,0]]}]}]
      }]
    }
    MatchCash_TotalEntry {
     id=#3
     possessed_class=MatchXPTotalEntry_C
     tracks [{
       type=MovieSceneFloatTrack
       display_name=Render Opacity
       sections [{type=MovieSceneFloatSection, end_frame=10, start_frame=0, channels [None {type=MovieSceneScriptingFloatChannel, keys [[0,1],[7,1],[8,0]]}]}]
      }]
    }
    VerticalBox_XPTotalList {
     id=#4
     possessed_class=VerticalBox
     tracks [{
       type=MovieSceneFloatTrack
       display_name=Render Opacity
       sections [{type=MovieSceneFloatSection, end_frame=10, start_frame=0, channels [None {type=MovieSceneScriptingFloatChannel, keys [[0,1],[7,1],[10,0]]}]}]
      }]
    }
    MatchXP_MatchXp {
     id=#5
     possessed_class=MatchXPTotalEntry_C
     tracks [{
       type=MovieSceneFloatTrack
       display_name=Render Opacity
       sections [{type=MovieSceneFloatSection, end_frame=10, start_frame=0, channels [None {type=MovieSceneScriptingFloatChannel, keys [[0,1],[7,1],[8,0]]}]}]
      }]
    }
    MatchXP_SurvivalBonus {
     id=#6
     possessed_class=MatchXPTotalEntry_C
     tracks [{
       type=MovieSceneFloatTrack
       display_name=Render Opacity
       sections [{type=MovieSceneFloatSection, end_frame=10, start_frame=0, channels [None {type=MovieSceneScriptingFloatChannel, keys [[0,1],[7,1],[8,0]]}]}]
      }]
    }
    MatchXP_ChallengeXp {
     id=#7
     possessed_class=MatchXPTotalEntry_C
     tracks [{
       type=MovieSceneFloatTrack
       display_name=Render Opacity
       sections [{type=MovieSceneFloatSection, end_frame=10, start_frame=0, channels [None {type=MovieSceneScriptingFloatChannel, keys [[0,1],[7,1],[8,0]]}]}]
      }]
    }
   ]}
 }
 Anim_Finish {
  binding_count=8
  length_seconds=0.25
  object=&self:Anim_Finish
  sequence {bindings [
    MatchXP_CombatBonus {
     id=#1
     possessed_class=MatchXPTotalEntry_C
     tracks [{
       type=MovieSceneFloatTrack
       display_name=Render Opacity
       sections [{type=MovieSceneFloatSection, end_frame=5, start_frame=0, channels [None {type=MovieSceneScriptingFloatChannel, const [0,2]}]}]
      }]
    }
    MatchXP_MedalEntry {
     id=#2
     possessed_class=MatchXPTotalEntry_C
     tracks [{
       type=MovieSceneFloatTrack
       display_name=Render Opacity
       sections [{type=MovieSceneFloatSection, end_frame=5, start_frame=0, channels [None {type=MovieSceneScriptingFloatChannel, const [0,2]}]}]
      }]
    }
    MatchCash_TotalEntry {
     id=#3
     possessed_class=MatchXPTotalEntry_C
     tracks [{
       type=MovieSceneFloatTrack
       display_name=Render Opacity
       sections [{type=MovieSceneFloatSection, end_frame=5, start_frame=0, channels [None {type=MovieSceneScriptingFloatChannel, const [0,2]}]}]
      }]
    }
    VerticalBox_XPTotalList {
     id=#4
     possessed_class=VerticalBox
     tracks [{
       type=MovieSceneFloatTrack
       display_name=Render Opacity
       sections [{type=MovieSceneFloatSection, end_frame=5, start_frame=0, channels [None {type=MovieSceneScriptingFloatChannel, keys [[1,0],[5,1]]}]}]
      }]
    }
    MatchXP_MatchXp {
     id=#5
     possessed_class=MatchXPTotalEntry_C
     tracks [{
       type=MovieSceneFloatTrack
       display_name=Render Opacity
       sections [{type=MovieSceneFloatSection, end_frame=5, start_frame=0, channels [None {type=MovieSceneScriptingFloatChannel, const [0,2]}]}]
      }]
    }
    OverlaySlot_14 {
     id=#8
     parent_id=#4
     possessed_class=OverlaySlot
     tracks [
      {
       type=MovieSceneByteTrack
       display_name=Horizontal Alignment
       sections [{type=MovieSceneByteSection, end_frame=5, start_frame=0, channels [None {type=MovieSceneScriptingByteChannel, const [1,2]}]}]
      }
      {
       type=MovieSceneByteTrack
       display_name=Vertical Alignment
       sections [{type=MovieSceneByteSection, end_frame=5, start_frame=0, channels [None {type=MovieSceneScriptingByteChannel, const [3,2]}]}]
      }
      {
       type=MovieSceneMarginTrack
       display_name=Padding
       sections [{
         type=MovieSceneMarginSection
         end_frame=5
         start_frame=0
         channels [
          Left {type=MovieSceneScriptingFloatChannel, const [100,1]}
          Top {type=MovieSceneScriptingFloatChannel, const [0,1]}
          Right {type=MovieSceneScriptingFloatChannel, const [0,1]}
          Bottom {type=MovieSceneScriptingFloatChannel, const [0,1]}
         ]
        }]
      }
     ]
    }
    MatchXP_SurvivalBonus {
     id=#6
     possessed_class=MatchXPTotalEntry_C
     tracks [{
       type=MovieSceneFloatTrack
       display_name=Render Opacity
       sections [{type=MovieSceneFloatSection, end_frame=5, start_frame=0, channels [None {type=MovieSceneScriptingFloatChannel, const [0,2]}]}]
      }]
    }
    MatchXP_ChallengeXp {
     id=#7
     possessed_class=MatchXPTotalEntry_C
     tracks [{
       type=MovieSceneFloatTrack
       display_name=Render Opacity
       sections [{type=MovieSceneFloatSection, end_frame=5, start_frame=0, channels [None {type=MovieSceneScriptingFloatChannel, const [0,2]}]}]
      }]
    }
   ]}
 }
]
graphs [
 CacheMinigame {nodes [
   CacheMinigame() {is=title K2Node_FunctionEntry}
   > n1 = Cast To FortPlayerControllerAthena(Object=GetOwningPlayer())
   n2 = GetMinigame!(self:Fort Player Controller Gameplay Object Reference=@n1) -> (ReturnValue:Fort Minigame Object Reference)
   > Is Valid(InputObject=@n2)
   > Set Minigame(Minigame:Fort Minigame Object Reference=@n2) -> (Output_Get:Fort Minigame Object Reference)
  ]}
 CanChangeMinigameState {nodes [
   CanChangeMinigameState() {is=title K2Node_FunctionEntry}
   > n1 = Cast To FortGameStateAthena!(execute, Object=GetGameState!() -> (ReturnValue:Game State Base Object Reference)) -> (then, CastFailed, AsFort Game State Athena:Fort Game State Athena Object Reference) {is=title K2Node_DynamicCast}
   > n2 = Branch(Condition=IsExperiencePlayOnly!(self:Fort Game State Athena Object Reference=@n1) -> (ReturnValue:bool))
   Return Node!(execute=[@n3.CastFailed, @n2.then, @n1.CastFailed], CanChange:bool=false)
   n3 = Cast To FortPlayerControllerAthena(execute=@n2.else, Object=GetOwningPlayer())
   > Return Node(CanChange=CanOperateOnMinigame!(PlayerController:Fort Player Controller Object Reference=@n3) -> (ReturnValue:bool))
  ]}
 EventGraph {nodes [
   n1 = "On Clicked (ReturnToIslandButton)"!() -> (OutputDelegate, then, Button:Common Button Base Object Reference) {is=title K2Node_ComponentBoundEvent}
   n2 = Sequence(execute=[@n1.then, @n6.then])
   > n3 = Cast To FortPlayerControllerAthena(Object=GetOwningPlayer())
   > ServerEndMinigame!(execute, self:Fort Player Controller Athena Object Reference=@n3, bAbandon:bool=true) -> (then)
   ResetPostMatch(execute=@n2.then_1)
   Event Post Minigame Reset() {is=title K2Node_Event}
   > ResetPostMatch()
   n4 = Event SetXPAnimationState!() -> (OutputDelegate, then, IsActive:bool) {is=title K2Node_Event}
   > SetXPAnimationActive!(execute, self, IsAnimationActive:bool=@n4.IsActive) -> (then)
   Event OnInitialized()
   > SetVisibility(
    self=[
     Get MatchXPScoreboard() -> (MatchXPScoreboard:WBP Match XPScoreboard Object Reference)
     Get ReturnToIslandButton() -> (ReturnToIslandButton:WBP UIKit Button Loud Object Reference)
    ]
    InVisibility=Collapsed
   )
   > ReturnToIslandButtonShown!(execute, self:Fort Match XPRecap Object Reference, bShown:bool=false) -> (then)
   > Set IsXpAnimationActive(IsXpAnimationActive:bool=false) -> (Output_Get:bool)
   Event On Toggle Spectating() {is=title K2Node_Event}
   > Update Show Scoreboard!(execute, self) -> (then, ShowingScoreboard:bool)
   Event AnimateXPTotals() {is=title K2Node_Event}
   > SetVisibility(self=[Get MatchXPScoreboard(), Get ReturnToIslandButton()], InVisibility=Collapsed)
   > ReturnToIslandButtonShown(bShown=false)
   > HandlePostMatchVisibility(self:Fort Athena Survivor Tier Badge Object Reference=Get W_SurvivorTierBadge() -> (W_SurvivorTierBadge:W Survivor Tier Badge Object Reference))
   > CacheMinigame() {is=title K2Node_CallFunction}
   > Delay(Duration=0.2)
   n5 = Get VerticalBox_XPTotalList() -> (VerticalBox_XPTotalList:Vertical Box Object Reference)
   > Set CurrentXpEntryIndex(CurrentXpEntryIndex:int=Last Index!(TargetArray:Array of Widget Object References=GetAllChildren(self=@n5)) -> (ReturnValue=0)) -> (Output_Get:int=0)
   > PlayNextEntryIntro()
   n6 = Event OnReturnToIslandInputAction() {is=title K2Node_Event}
   Event Destruct()
   > UnregisterWidgetHost!(
    execute
    WidgetHost:Widget Object Reference
    WidgetHostTag:"Gameplay Tag Structure (by ref)"=Get MatchRecapExtensionSlot() -> (MatchRecapExtensionSlot:Gameplay Tag Structure)
   ) -> (then)
   n7 = Event Add Content To Host Widget!() -> (
    OutputDelegate
    then
    WidgetHostTag:"Gameplay Tag Structure (by ref)"
    ContentWidget:Widget Object Reference
    Padding:"Margin Structure (by ref)"
    bFill:bool
   ) {is=title K2Node_Event}
   > Branch(Condition=MatchesTag!(TagOne:Gameplay Tag Structure=@n7.WidgetHostTag, TagTwo:Gameplay Tag Structure=Get MatchRecapExtensionSlot(), bExactMatch:bool=true) -> (ReturnValue:bool))
   > AddChildToVerticalBox(
    self=Get VerticalBox_PrimaryMatchXPExtension() -> (VerticalBox_PrimaryMatchXPExtension:Vertical Box Object Reference)
    Content=@n7.ContentWidget
   )
   Event Construct()
   > RegisterWidgetHost!(
    execute
    WidgetHost:Widget Object Reference
    WidgetHostTag:"Gameplay Tag Structure (by ref)"=Get MatchRecapExtensionSlot()
    IsPlayerSpecific:bool=true
   ) -> (then)
  ]}
 FindPreviousValidEntry {nodes [
   n1 = FindPreviousValidEntry!() -> (then, StartingIndex:int) {is=title K2Node_FunctionEntry}
   > n2 = Set CurrentIndex(CurrentIndex:int=@n1) -> (Output_Get:int=0)
   n3 = Decrement Int(" "=[@n2.then, @n6.else], Value=@n2)
   > n4 = Branch(Condition="integer >= integer"(A=@n3))
   > n5 = GetEntryAtIndex!(execute, self, Index:int=Get CurrentIndex() -> (CurrentIndex:int=0)) -> (then, IsValid:bool, EntryWidget:Fort Match XPTotal Entry Object Reference)
   > n6 = Branch(Condition=@n5)
   > Return Node!(
    execute
    IsSuccess:bool=true
    FoundIndex:int=Get CurrentIndex() -> (CurrentIndex=0)
    EntryWidget:Fort Match XPTotal Entry Object Reference=@n5.EntryWidget
   )
   Return Node(execute=@n4.else, IsSuccess=false, FoundIndex=0)
  ]}
 GetEntryAtIndex {nodes [
   n1 = GetEntryAtIndex!() -> (then, Index:int) {is=title K2Node_FunctionEntry}
   n2 = GetAllChildren(self=Get VerticalBox_XPTotalList())
   > n3 = Branch(Condition=Is Valid Index!(TargetArray:Array of Widget Object References=@n2, IndexToTest:int=@n1) -> (ReturnValue:bool))
   > n4 = Cast To FortMatchXPTotalEntry!(
    execute
    Object="Get (a copy)"(Array:Array of Widget Object References=@n2, Dimension 1=Get Index() -> (Index:int=0)) -> (Output:Widget Object Reference)
   ) -> (then, CastFailed, AsFort Match XPTotal Entry:Fort Match XPTotal Entry Object Reference) {is=title K2Node_DynamicCast}
   > n5 = Branch(Condition="integer > integer"(A=GetXPValue!(self:Fort Match XPTotal Entry Object Reference=@n4) -> (ReturnValue:int=0), B=0) {is=title K2Node_CallFunction})
   > Return Node!(execute, IsValid:bool=true, EntryWidget=@n4)
   Return Node(execute=[@n5.else, @n3.else], IsValid=false)
  ]}
 IsWidgetClassUnique {nodes [
   n1 = IsWidgetClassUnique!() -> (then, ExistingWidgets:Array of Widget Object References, NewWidget:Widget Object Reference) {is=title K2Node_FunctionEntry}
   > n2 = For Each Loop(Array:Array of Widget Object References=@n1) -> (Array Element:Widget Object Reference)
   Return Node!(execute=@n2.Completed, IsUnique:bool=true)
   n3 = Get NewWidget() -> (NewWidget:Widget Object Reference)
   > Branch(Condition=IsA!(Object:Object Reference=@n2, ObjectClass:Object Class Reference=Get Class(Object=@n3) -> (ReturnValue:Widget Class Reference)) -> (ReturnValue:bool))
   > Return Node(IsUnique=false)
  ]}
 OnAllXPAnimationsFinish {nodes [
   OnAllXPAnimationsFinish() {is=title K2Node_FunctionEntry}
   > n1 = Sequence!(execute) -> (then_0, then_1, then_2:Exec)
   > Branch(Condition=CanChangeMinigameState!() -> (CanChange:bool) {is=title K2Node_CallFunction})
   > SetVisibility(self=Get ReturnToIslandButton(), InVisibility=SelfHitTestInvisible)
   > ReturnToIslandButtonShown(bShown=true)
   n2 = Update Show Scoreboard(execute=@n1.then_1)
   > Branch(Condition=@n2)
   > FadeOutOverlay(self:Fort Match XPRecap Object Reference)
   Call On XPRundown Animations Finished!(execute:Exec=@n1.then_2, self:Fort Match XPRecap Object Reference) -> (then:Exec) {is=title K2Node_CallDelegate}
  ]}
 OnLastXPAnimationFinished {nodes [
   OnLastXPAnimationFinished() {is=title K2Node_FunctionEntry}
   > n1 = Set Timer by Event(
    Delegate=Create Event()
    Time=float - float(A=Get OnXPAnimationsFinishedDelay() -> (OnXPAnimationsFinishedDelay:double), B=Get XPEntryDelay() -> (XPEntryDelay:double))
    bLooping=false
   )
   > Set AnimationSequenceTimerHandle(AnimationSequenceTimerHandle:Timer Handle Structure=@n1) -> (Output_Get:Timer Handle Structure)
   n2 = Get W_SurvivorTierBadge()
   > Branch(Condition=IsVisible(self=@n2))
   > PlaySwapAnimation(self:W Survivor Tier Badge Object Reference=@n2)
  ]}
 PlayNextEntryIntro {nodes [
   PlayNextEntryIntro() {is=title K2Node_FunctionEntry}
   > n1 = GetEntryAtIndex(Index=Get CurrentXpEntryIndex() -> (CurrentXpEntryIndex:int=0)) {is=title K2Node_CallFunction}
   > n2 = Branch(Condition=@n1)
   > n3 = Set CurrentEntryWidget(CurrentEntryWidget:Fort Match XPTotal Entry Object Reference=@n1.EntryWidget) -> (Output_Get:Fort Match XPTotal Entry Object Reference)
   PlayIntro(
    execute=[@n13.then, @n3.then]
    self:Fort Match XPTotal Entry Object Reference=Get CurrentEntryWidget() -> (CurrentEntryWidget:Fort Match XPTotal Entry Object Reference)
   )
   > n4 = FindPreviousValidEntry!(execute, self, StartingIndex:int=Get CurrentXpEntryIndex() -> (CurrentXpEntryIndex=0)) -> (then, IsSuccess:bool, FoundIndex:int=0, EntryWidget:Fort Match XPTotal Entry Object Reference) {is=title K2Node_CallFunction}
   > Set CurrentXpEntryIndex(CurrentXpEntryIndex=@n4.FoundIndex) -> (Output_Get=0)
   > n5 = Branch(Condition=@n4)
   > n6 = Set Timer by Event(Delegate=Create Event(), Time=Get XPEntryDelay(), bLooping=false)
   > n7 = Set AnimationSequenceTimerHandle(AnimationSequenceTimerHandle=@n6)
   n8 = Branch(execute=[@n10.then, @n7.then], Condition=Get IsXpAnimationActive() -> (IsXpAnimationActive:bool))
   Pause Timer by Handle!(execute=@n8.else, Handle:Timer Handle Structure=Get AnimationSequenceTimerHandle() -> (AnimationSequenceTimerHandle:Timer Handle Structure)) -> (then)
   n9 = Set Timer by Event(execute=[@n5.else, @n12.else], Delegate=Create Event(), Time=Get XPEntryDelay(), bLooping=false)
   > n10 = Set AnimationSequenceTimerHandle(AnimationSequenceTimerHandle=@n9)
   n11 = FindPreviousValidEntry(execute=@n2.else, StartingIndex=Get CurrentXpEntryIndex() -> (CurrentXpEntryIndex=0)) -> (FoundIndex=0)
   > n12 = Branch(Condition=@n11)
   > Set CurrentXpEntryIndex(CurrentXpEntryIndex=@n11.FoundIndex) -> (Output_Get=0)
   > n13 = Set CurrentEntryWidget(CurrentEntryWidget=@n11.EntryWidget)
  ]}
 RemoveContentFromWidgetHost {nodes [
   n1 = Remove Content From Host Widget!() -> (
    then
    WidgetHostTag:"Gameplay Tag Structure (by ref)"
    ContentWidget:Widget Object Reference
    RemovalType:EWidgetSlottingRemovalType Enum=Unparent
   ) {is=title K2Node_FunctionEntry}
   > Branch(Condition=MatchesTag(TagOne=@n1, TagTwo=Get MatchRecapExtensionSlot(), bExactMatch=true))
   > n2 = Switch on EWidgetSlottingRemovalType!(execute, Selection:EWidgetSlottingRemovalType Enum=@n1.RemovalType) -> (Unparent:Exec, Collapse:Exec) {is=title K2Node_SwitchEnum}
   SetVisibility(execute=@n2.Collapse, self=@n1.ContentWidget, InVisibility=Collapsed)
   n3 = RemoveChild(execute=@n2.Unparent, self=Get VerticalBox_PrimaryMatchXPExtension(), Content=@n1.ContentWidget)
   > Return Node!(execute, ReturnValue:bool=@n3)
  ]}
 ResetPostMatch {nodes [
   ResetPostMatch() {is=title K2Node_FunctionEntry}
   > PopContentWidget!(execute) -> (then)
   > RecoverFromPostGameMenu(self:Athena HUDContext Object Reference=Get AthenaHUDContext!() -> (ReturnValue:Athena HUDContext Object Reference) {is=title K2Node_GetBlueprintContextEx})
   n1 = Get FortHUDContext!() -> (ReturnValue:Fort HUDContext Object Reference) {is=title K2Node_GetBlueprintContextEx}
   > SetCursorModeLocked!(execute, self:Fort HUDContext Object Reference=@n1, bLocked:bool=false) -> (then)
   > EnterCameraMode(self:Fort HUDContext Object Reference=@n1)
  ]}
 SetXPAnimationActive {nodes [
   n1 = SetXPAnimationActive!() -> (then, IsAnimationActive:bool=~) {is=title K2Node_FunctionEntry}
   > n2 = Set IsXpAnimationActive(IsXpAnimationActive=@n1)
   > n3 = Sequence!(execute) -> (then_0, then_1)
   > n4 = For Each Loop(Array=GetAllChildren(self=Get VerticalBox_XPTotalList()))
   > n5 = Cast To MatchXPTotalEntry!(execute, Object=@n4) -> (then, CastFailed, AsMatch XPTotal Entry:Match XPTotal Entry Object Reference) {is=title K2Node_DynamicCast}
   > SetIsAnimationActive!(execute, self:Match XPTotal Entry Object Reference=@n5, IsActive:bool=Get IsXpAnimationActive()) -> (then)
   n6 = Branch(execute=@n3.then_1, Condition=@n2)
   n7 = Get AnimationSequenceTimerHandle()
   > Unpause Timer by Handle!(execute, Handle:Timer Handle Structure=@n7) -> (then)
   Pause Timer by Handle(execute=@n6.else, Handle=@n7)
  ]}
 Update Show Scoreboard {nodes [
   Update Show Scoreboard() {is=title K2Node_FunctionEntry}
   > n1 = Sequence!(execute) -> (then_0, then_1, then_2:Exec)
   n2 = GetMinigameSettingsMachine!(ContextActor:Actor Object Reference=Get Minigame() -> (Minigame:Fort Minigame Object Reference)) -> (ReturnValue:Fort Minigame Settings Building Object Reference)
   > Is Valid(InputObject=@n2)
   n3 = Break Minigame Post Game Settings!(MinigamePostGameSettings:"Minigame Post Game Settings Structure (by ref)"=GetPostGameSettings!(self:Fort Minigame Settings Building Object Reference=@n2) -> (ReturnValue:"Minigame Post Game Settings Structure (by ref)")) -> (
    PostGameType:EMinigameGameEndPostGameType Enum
    PostGameDelay:float=~
    bCustomShowScoreboard:bool=~
    VictoryAnimationSettings:Minigame Post Game Animation Settings Structure
    DefeatAnimationSettings:Minigame Post Game Animation Settings Structure
    TieAnimationSettings:Minigame Post Game Animation Settings Structure
   ) {is=title K2Node_BreakStruct}
   n4 = IsSpectating!(self:Fort Match XPRecap Object Reference) -> (ReturnValue:bool)
   > Set bShowScoreboard(bShowScoreboard:bool=AND Boolean!(
     A="Equal (Enum)"(A:EMinigameGameEndPostGameType Enum=@n3, B:EMinigameGameEndPostGameType Enum=Custom)
     B=@n3.bCustomShowScoreboard
     C:bool=NOT Boolean(A=@n4)
    ) -> (ReturnValue)) -> (Output_Get:bool)
   n5 = Branch(execute=@n1.then_1, Condition=Get bShowScoreboard() -> (bShowScoreboard:bool))
   n6 = Get MatchXpScoreboard() -> (MatchXpScoreboard:WBP Match XPScoreboard Object Reference)
   > DoShowUI!(execute:Exec, self:Object Reference=@n6) -> (then:Exec) {is=title K2Node_Message}
   DoHideUI!(execute:Exec=@n5.else, self:Object Reference=@n6) -> (then:Exec) {is=title K2Node_Message}
   Return Node!(execute=@n1.then_2, ShowingScoreboard:bool=Get bShowScoreboard())
  ]}
]
variables [
 AnimationSequenceTimerHandle {type=TimerHandle}
 CompletedQuestEntries {type="array<RundownEntryDisplay>", default []}
 CurrentXpEntryIndex {type=integer}
 InProgressQuestEntries {type="array<RundownEntryDisplay>", default []}
 IsXpAnimationActive {type=boolean}
 MatchRecapExtensionSlot {type=GameplayTag, default {tagName=DynamicUI.HUD.Fort.Athena.PostMatch.Slot}}
 Minigame {type=FortMinigame, default=null}
 OnXPAnimationsFinishedDelay {type=number, default=2}
 PARENT_PADDING {type=Margin, default=<ltrb -45 0 -45 -112>}
 QuestRundownItems {type="array<WBP_PostMatch_RundownItem_C>", default []}
 XPEntryDelay {type=number, default=1.1}
]
widgets {
 info {inheritedWidgetCount=9, parentClass=/Script/FortniteUI.FortMatchXPRecap, rootWidgetClass=&Overlay, widgetCount=22}
 widgets [
  Overlay_XPRecap {widgetClassPath=&Overlay, widget=Overlay_XPRecap}
  Text_MemberOptedOut {
   widgetClassPath=&CommonTextBlock
   parent=Overlay_XPRecap
   widget=Text_MemberOptedOut
   properties {
    autoWrapText=true
    justification=Center
    text=Your party left the team
    textTransformPolicy=ToUpper
    visibility=Collapsed
    font {
     bForceMonospaced=false
     bMaterialIsStencil=false
     fontMaterial=null
     fontObject=&Burbank
     letterSpacing=0
     monospacedWidth=1
     size=34
     skewAmount=0
     typefaceFontName=Burbank-PlayerPower
     outlineSettings {
      bApplyOutlineToDropShadows=false
      bMiteredCorners=false
      bSeparateFillAlpha=false
      outlineMaterial=null
      outlineSize=0
      outlineColor=<rgba 0 0 0 1>
     }
    }
    renderTransform {angle=0, scale=<1 1>, shear=<-8 0>, translation=<0 0>}
    shadowOffset=<0 0>
   }
   slot {class=OverlaySlot, properties {horizontalAlignment=HAlign_Center, padding=<ltrb 0 290 0 0>}}
  }
  SizeBox_Content {
   widgetClassPath=&SizeBox
   parent=Overlay_XPRecap
   widget=SizeBox_Content
   properties {bOverride_HeightOverride=true, bOverride_WidthOverride=true, heightOverride=300, widthOverride=300}
   slot {class=OverlaySlot, properties {padding=<ltrb 150 150 0 0>}}
  }
  ScaleBox_Content {
   widgetClassPath=&ScaleBox
   parent=SizeBox_Content
   widget=ScaleBox_Content
   properties {stretchDirection=DownOnly}
   slot {class=SizeBoxSlot}
  }
  W_SurvivorTierBadge {
   widgetClassPath=/Game/UI/SurvivorTiers/W_SurvivorTierBadge
   bIsVariable=true
   parent=ScaleBox_Content
   widget=W_SurvivorTierBadge
   properties {badgeSize=Large, visibility=Collapsed}
   slot {class=ScaleBoxSlot}
  }
  VerticalBox_PrimaryMatchXPExtension {
   widgetClassPath=&VerticalBox
   bIsVariable=true
   parent=Overlay_XPRecap
   widget=VerticalBox_PrimaryMatchXPExtension
   slot {class=OverlaySlot, properties {verticalAlignment=VAlign_Center, padding=<ltrb 150 20 20 20>}}
  }
  Overlay_ChallengeSection {
   widgetClassPath=&Overlay
   bIsVariable=true
   parent=Overlay_XPRecap
   widget=Overlay_ChallengeSection
   slot {class=OverlaySlot, properties {verticalAlignment=VAlign_Bottom, padding=<ltrb 20 20 20 20>}}
  }
  Widget_QuestRecap {
   widgetClassPath=/FortniteCoreUI/HUD/EndGame/EndOfMatch_QuestProgessRecap
   bInherited=true
   bIsVariable=true
   parent=Overlay_ChallengeSection
   widget=Widget_QuestRecap
   properties {exclusiveFilterTags {gameplayTags [{tagName=QuestCategory.BR.Habanero}, {tagName=QuestCategory.BR.Ranked.Urgent}]}}
   slot {class=OverlaySlot}
  }
  SizeBox_0 {
   widgetClassPath=&SizeBox
   parent=Overlay_XPRecap
   widget=SizeBox_0
   properties {bOverride_HeightOverride=true, bOverride_WidthOverride=true, heightOverride=550, widthOverride=1200}
   slot {class=OverlaySlot, properties {padding=<ltrb 12 140 0 0>}}
  }
  MatchXPScoreboard {
   widgetClassPath=/FortniteCoreUI/Athena/HUD/EndGame/PostGameFlowScreens/DefaultScreens/XPScreen/WBP_MatchXPScoreboard
   bIsVariable=true
   parent=SizeBox_0
   widget=MatchXPScoreboard
   properties {bShowHeader=false, elementTag {tagName=Athena.HideForCapture}, hUDElementTag {gameplayTags []}}
   slot {class=SizeBoxSlot}
  }
  VerticalBox_0 {
   widgetClassPath=&VerticalBox
   parent=Overlay_XPRecap
   widget=VerticalBox_0
   slot {class=OverlaySlot, properties {horizontalAlignment=HAlign_Center, verticalAlignment=VAlign_Bottom, padding=<ltrb 0 0 0 35>}}
  }
  ReturnToIslandButton {
   widgetClassPath=/FortUILibrary/UIKit/Buttons/WBP_UIKit_Button_Loud
   bIsVariable=true
   parent=VerticalBox_0
   widget=ReturnToIslandButton
   properties {
    bIsFocusable=false
    buttonSize=Large
    navigation=&self:WidgetTree.ReturnToIslandButton.WidgetNavigation_0
    text=RETURN TO ISLAND
    visibility=Collapsed
    triggeringInputAction {dataTable=&InputActionDataTable, rowName=CreativeScoreboard_Abort}
   }
   slot {class=VerticalBoxSlot, properties {padding=<ltrb 0 0 0 16>}}
  }
  AthenaWinnerList {
   widgetClassPath=/Game/Athena/HUD/EndGame/AthenaWinnerList
   parent=Overlay_XPRecap
   widget=AthenaWinnerList
   properties {visibility=Collapsed}
   slot {class=OverlaySlot, properties {horizontalAlignment=HAlign_Center, padding=<ltrb 0 355 0 0>}}
  }
  VerticalBox_XPTotalList {
   widgetClassPath=&VerticalBox
   bIsVariable=true
   parent=Overlay_XPRecap
   widget=VerticalBox_XPTotalList
   properties {renderTransform {angle=0, scale=<1 1>, shear=<0 0>, translation=<0 25>}}
   slot {class=OverlaySlot, properties {horizontalAlignment=HAlign_Center, verticalAlignment=VAlign_Center, padding=<ltrb 0 300 0 0>}}
  }
  MatchXP_TotalEntry {
   widgetClassPath=/FortniteCoreUI/Athena/HUD/EndGame/MatchXPTotalEntry
   bInherited=true
   bIsVariable=true
   parent=VerticalBox_XPTotalList
   widget=MatchXP_TotalEntry
   properties {bIsTotal=true, xPType=Match Total}
   slot {class=VerticalBoxSlot, properties {horizontalAlignment=HAlign_Center, verticalAlignment=VAlign_Center}}
  }
  MatchCash_TotalEntry {
   widgetClassPath=&MatchXPTotalEntry
   bInherited=true
   bIsVariable=true
   parent=VerticalBox_XPTotalList
   widget=MatchCash_TotalEntry
   properties {showDivider=true, xPType=collected}
   slot {class=VerticalBoxSlot, properties {horizontalAlignment=HAlign_Center, verticalAlignment=VAlign_Top, padding=<ltrb 0 0 0 5>}}
  }
  MatchXP_MedalEntry {
   widgetClassPath=&MatchXPTotalEntry
   bInherited=true
   bIsVariable=true
   parent=VerticalBox_XPTotalList
   widget=MatchXP_MedalEntry
   properties {xPType=Medal Bonus}
   slot {class=VerticalBoxSlot, properties {horizontalAlignment=HAlign_Center}}
  }
  MatchXP_SurvivalBonus {
   widgetClassPath=&MatchXPTotalEntry
   bInherited=true
   bIsVariable=true
   parent=VerticalBox_XPTotalList
   widget=MatchXP_SurvivalBonus
   properties {xPType=Survival}
   slot {class=VerticalBoxSlot, properties {horizontalAlignment=HAlign_Center}}
  }
  MatchXP_CombatBonus {
   widgetClassPath=&MatchXPTotalEntry
   bInherited=true
   bIsVariable=true
   parent=VerticalBox_XPTotalList
   widget=MatchXP_CombatBonus
   properties {xPType=Combat}
   slot {class=VerticalBoxSlot, properties {horizontalAlignment=HAlign_Center}}
  }
  MatchXP_MatchXp {
   widgetClassPath=&MatchXPTotalEntry
   bInherited=true
   bIsVariable=true
   parent=VerticalBox_XPTotalList
   widget=MatchXP_MatchXp
   properties {xPType=Match}
   slot {class=VerticalBoxSlot, properties {horizontalAlignment=HAlign_Center}}
  }
  MatchXP_ChallengeXp {
   widgetClassPath=&MatchXPTotalEntry
   bInherited=true
   bIsVariable=true
   parent=VerticalBox_XPTotalList
   widget=MatchXP_ChallengeXp
   properties {xPType=Quest}
   slot {class=VerticalBoxSlot, properties {horizontalAlignment=HAlign_Center, verticalAlignment=VAlign_Center}}
  }
  MatchXP_GameplayXp {widgetClassPath=/Script/FortniteUI.FortMatchXPTotalEntry, bInherited=true, widget=null}
 ]
}
