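# Widget Blueprint for the Athena end-game HUD ready-up tooltip, derived from the native AthenaReadyUpTooltipWidget class.
# It lays out eight AthenaReadyUpTooltipTickWidget instances in a horizontal row, each exposed as a variable.
# An Overlay and a down-only ScaleBox wrap the row, centered horizontally.
# The event graph holds only empty PreConstruct, Construct and Tick events, so the Blueprint adds no logic of its own.
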WidgetBlueprint /FortniteCoreUI/Athena/HUD/EndGame/AthenaReadyUpTooltipWidget
graphs [EventGraph {nodes [
   Event PreConstruct()
   Event Construct()
   Event Tick()
  ]}]
variables []
widgets {
 info {inheritedWidgetCount=0, parentClass=/Script/FortniteUI.AthenaReadyUpTooltipWidget, rootWidgetClass=&Overlay, widgetCount=12}
 widgets [
  Overlay_1 {widgetClassPath=&Overlay, widget=Overlay_1}
  ScaleBox_0 {
   widgetClassPath=&ScaleBox
   parent=Overlay_1
   widget=ScaleBox_0
   properties {stretch=None, stretchDirection=DownOnly}
   slot {class=OverlaySlot, properties {horizontalAlignment=HAlign_Center, verticalAlignment=VAlign_Fill}}
  }
  Overlay_0 {
   widgetClassPath=&Overlay
   parent=ScaleBox_0
   widget=Overlay_0
   slot {class=ScaleBoxSlot, properties {horizontalAlignment=HAlign_Fill, verticalAlignment=VAlign_Fill}}
  }
  HorizontalBox_68 {
   widgetClassPath=&HorizontalBox
   parent=Overlay_0
   widget=HorizontalBox_68
   slot {class=OverlaySlot, properties {horizontalAlignment=HAlign_Fill, verticalAlignment=VAlign_Fill}}
  }
  AthenaReadyUpTooltipTickWidget {
   widgetClassPath=&/AthenaReadyUpTooltipTickWidget
   bIsVariable=true
   parent=HorizontalBox_68
   widget=AthenaReadyUpTooltipTickWidget
   slot {class=HorizontalBoxSlot, properties {horizontalAlignment=HAlign_Center}}
  }
  AthenaReadyUpTooltipTickWidget_1 {
   widgetClassPath=&AthenaReadyUpTooltipTickWidget
   bIsVariable=true
   parent=HorizontalBox_68
   widget=AthenaReadyUpTooltipTickWidget_1
   slot {class=HorizontalBoxSlot}
  }
  AthenaReadyUpTooltipTickWidget_2 {
   widgetClassPath=&AthenaReadyUpTooltipTickWidget
   bIsVariable=true
   parent=HorizontalBox_68
   widget=AthenaReadyUpTooltipTickWidget_2
   slot {class=HorizontalBoxSlot}
  }
  AthenaReadyUpTooltipTickWidget_3 {
   widgetClassPath=&AthenaReadyUpTooltipTickWidget
   bIsVariable=true
   parent=HorizontalBox_68
   widget=AthenaReadyUpTooltipTickWidget_3
   slot {class=HorizontalBoxSlot}
  }
  AthenaReadyUpTooltipTickWidget_4 {
   widgetClassPath=&AthenaReadyUpTooltipTickWidget
   bIsVariable=true
   parent=HorizontalBox_68
   widget=AthenaReadyUpTooltipTickWidget_4
   slot {class=HorizontalBoxSlot}
  }
  AthenaReadyUpTooltipTickWidget_5 {
   widgetClassPath=&AthenaReadyUpTooltipTickWidget
   bIsVariable=true
   parent=HorizontalBox_68
   widget=AthenaReadyUpTooltipTickWidget_5
   slot {class=HorizontalBoxSlot}
  }
  AthenaReadyUpTooltipTickWidget_6 {
   widgetClassPath=&AthenaReadyUpTooltipTickWidget
   bIsVariable=true
   parent=HorizontalBox_68
   widget=AthenaReadyUpTooltipTickWidget_6
   slot {class=HorizontalBoxSlot}
  }
  AthenaReadyUpTooltipTickWidget_7 {
   widgetClassPath=&AthenaReadyUpTooltipTickWidget
   bIsVariable=true
   parent=HorizontalBox_68
   widget=AthenaReadyUpTooltipTickWidget_7
   slot {class=HorizontalBoxSlot}
  }
 ]
}
