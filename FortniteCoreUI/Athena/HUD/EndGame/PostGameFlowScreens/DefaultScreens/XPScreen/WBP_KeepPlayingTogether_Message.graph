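# Keep Playing Together message widget for the post-game XP screen, showing a party-themed banner with primary and secondary text.
# A party icon badge sits beside a bordered alert icon and primary line, with secondary text below, collapsed by default.
# SetMemberOptedOutState skips all updates when Primary is empty; otherwise it sets both texts and collapses the secondary line if empty.
# Its Status input is unused, and the event graph is empty.
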WidgetBlueprint /FortniteCoreUI/Athena/HUD/EndGame/PostGameFlowScreens/DefaultScreens/XPScreen/WBP_KeepPlayingTogether_Message
graphs [
 EventGraph {links [], nodes []}
 SetMemberOptedOutState {nodes [
   n1 = SetMemberOptedOutState!() -> (then, Status:EKeepPlayingTogetherVotingStatus Enum=Undecided, Primary:Text, Secondary:Text) {is=title K2Node_FunctionEntry}
   > n2 = Branch(Condition=TextIsEmpty(InText=@n1.Primary))
   "SetText (Text)"(execute=@n2.else, self=Get Text_Primary() -> (Text_Primary:Common Text Object Reference), InText=@n1.Primary)
   > SetVisibility(
    self=Get Text_Secondary() -> (Text_Secondary:Common Text Object Reference)
    InVisibility=Select(Option 0=SelfHitTestInvisible, Option 1=Collapsed, Index=TextIsEmpty(InText=@n1.Secondary))
   )
   > "SetText (Text)"(self=Get Text_Secondary(), InText=@n1.Secondary)
  ]}
]
variables []
widgets {
 info {inheritedWidgetCount=0, parentClass=&UserWidget, rootWidgetClass=&VerticalBox, widgetCount=9}
 widgets [
  VerticalBox_20 {widgetClassPath=&VerticalBox, widget=VerticalBox_20}
  HorizontalBox_1 {
   widgetClassPath=&HorizontalBox
   parent=VerticalBox_20
   widget=HorizontalBox_1
   slot {class=VerticalBoxSlot, properties {padding=<ltrb 0 0 0 3>}}
  }
  IconParty {
   widgetClassPath=&Border
   parent=HorizontalBox_1
   widget=IconParty
   properties {
    horizontalAlignment=HAlign_Center
    verticalAlignment=VAlign_Center
    background {
     bIsDynamicallyLoaded=false
     drawAs=Box
     imageType=NoImage
     mirroring=NoMirror
     resourceName=None
     resourceObject=/Game/Athena/HUD/EndGame/PostGameFlowScreens/DefaultScreens/XPScreen/Art/MI_UI_KeepPlayingTogetherMessage_Icon
     tiling=NoTile
     imageSize=<32 32>
     margin=<ltrb 0.3 0 0.3 0>
     outlineSettings {
      bUseBrushTransparency=false
      roundingType=HalfHeightRadius
      width=0
      color {colorUseRule=UseColor_Specified, specifiedColor=<rgba 0 0 0 0>}
      cornerRadii=<0 0 0 0>
     }
     tintColor {colorUseRule=UseColor_Specified, specifiedColor=<rgba 1 1 1 1>}
     uVRegion {bIsValid=false, max=<0 0>, min=<0 0>}
    }
    padding=<ltrb 9 3 9 3>
   }
   slot {class=HorizontalBoxSlot, properties {horizontalAlignment=HAlign_Left, padding=<ltrb 0 -20 -5 20>}}
  }
  Image_247 {
   widgetClassPath=&Image
   bIsVariable=true
   parent=IconParty
   widget=Image_247
   properties {brush {
     bIsDynamicallyLoaded=false
     drawAs=Box
     imageType=NoImage
     mirroring=NoMirror
     resourceName=None
     resourceObject=/Game/Athena/HUD/EndGame/PostGameFlowScreens/DefaultScreens/XPScreen/Art/T_UI_Icon_KeepPlayingTogether_Party
     tiling=NoTile
     imageSize=<64 64>
     margin=<ltrb 0 0 0 0>
     outlineSettings {
      bUseBrushTransparency=false
      roundingType=HalfHeightRadius
      width=0
      color {colorUseRule=UseColor_Specified, specifiedColor=<rgba 0 0 0 0>}
      cornerRadii=<0 0 0 0>
     }
     tintColor {colorUseRule=UseColor_Specified, specifiedColor=<rgba 1 1 1 1>}
     uVRegion {bIsValid=false, max=<0 0>, min=<0 0>}
    }}
   slot {class=BorderSlot, properties {horizontalAlignment=HAlign_Center, verticalAlignment=VAlign_Center, padding=<ltrb 9 3 9 3>}}
  }
  Border_32 {
   widgetClassPath=&Border
   parent=HorizontalBox_1
   widget=Border_32
   properties {
    verticalAlignment=VAlign_Center
    background {
     bIsDynamicallyLoaded=false
     drawAs=Box
     imageType=NoImage
     mirroring=NoMirror
     resourceName=None
     resourceObject=/Game/Athena/HUD/EndGame/PostGameFlowScreens/DefaultScreens/XPScreen/Art/MI_UI_KeepPlayingTogetherMessage
     tiling=NoTile
     imageSize=<32 32>
     margin=<ltrb 0.3 0 0.3 0>
     outlineSettings {
      bUseBrushTransparency=false
      roundingType=HalfHeightRadius
      width=0
      color {colorUseRule=UseColor_Specified, specifiedColor=<rgba 0 0 0 0>}
      cornerRadii=<0 0 0 0>
     }
     tintColor {colorUseRule=UseColor_Specified, specifiedColor=<rgba 1 1 1 1>}
     uVRegion {bIsValid=false, max=<0 0>, min=<0 0>}
    }
    padding=<ltrb 0 3 0 3.5>
   }
   slot {class=HorizontalBoxSlot, properties {verticalAlignment=VAlign_Bottom, size {sizeRule=Fill, value=1}}}
  }
  HorizontalBox_0 {
   widgetClassPath=&HorizontalBox
   parent=Border_32
   widget=HorizontalBox_0
   slot {class=BorderSlot, properties {verticalAlignment=VAlign_Center, padding=<ltrb 0 3 0 3.5>}}
  }
  Image_Icon {
   widgetClassPath=&Image
   parent=HorizontalBox_0
   widget=Image_Icon
   properties {brush {
     bIsDynamicallyLoaded=false
     drawAs=Image
     imageType=NoImage
     mirroring=NoMirror
     resourceName=None
     resourceObject=/Game/Athena/HUD/EndGame/PostGameFlowScreens/DefaultScreens/XPScreen/Art/T_UI_Icon_KeepPlayingTogether_PartyAlert
     tiling=NoTile
     imageSize=<64 64>
     margin=<ltrb 0 0 0 0>
     outlineSettings {
      bUseBrushTransparency=false
      roundingType=HalfHeightRadius
      width=0
      color {colorUseRule=UseColor_Specified, specifiedColor=<rgba 0 0 0 0>}
      cornerRadii=<0 0 0 0>
     }
     tintColor {colorUseRule=UseColor_Specified, specifiedColor=<rgba 1 1 1 1>}
     uVRegion {bIsValid=false, max=<0 0>, min=<0 0>}
    }}
   slot {class=HorizontalBoxSlot, properties {horizontalAlignment=HAlign_Left, verticalAlignment=VAlign_Center, padding=<ltrb 3 0 0 0>}}
  }
  Text_Primary {
   widgetClassPath=&CommonTextBlock
   bIsVariable=true
   parent=HorizontalBox_0
   widget=Text_Primary
   properties {
    autoWrapText=true
    lineHeightPercentage=0.9
    text=You cannot stay with team
    textTransformPolicy=ToUpper
    font {
     bForceMonospaced=false
     bMaterialIsStencil=false
     fontMaterial=null
     fontObject=&Burbank
     letterSpacing=0
     monospacedWidth=1
     size=21
     skewAmount=0
     typefaceFontName=BurbankRegular-Black
     outlineSettings {
      bApplyOutlineToDropShadows=false
      bMiteredCorners=false
      bSeparateFillAlpha=false
      outlineMaterial=null
      outlineSize=0
      outlineColor=<rgba 0 0 0 1>
     }
    }
    shadowOffset=<0 0>
   }
   slot {class=HorizontalBoxSlot, properties {verticalAlignment=VAlign_Center, padding=<ltrb 2 0 8 0>, size {sizeRule=Fill, value=1}}}
  }
  Text_Secondary {
   widgetClassPath=&CommonTextBlock
   bIsVariable=true
   parent=VerticalBox_20
   widget=Text_Secondary
   properties {
    autoWrapText=true
    justification=Center
    style=/FortUILibrary/Styles/Typography/Foundation/TextStyle-BurbankSmall-S-Teal.TextStyle-BurbankSmall-S-Teal_C
    text=This is long test text which is not localized.
    textTransformPolicy=ToUpper
    visibility=Collapsed
    colorAndOpacity {colorUseRule=UseColor_Specified, specifiedColor=<rgba 0.0275 0.8157 0.9922 1>}
    font {
     bForceMonospaced=false
     bMaterialIsStencil=false
     fontMaterial=null
     fontObject=&BurbankSmall
     letterSpacing=0
     monospacedWidth=1
     size=17
     skewAmount=0
     typefaceFontName=Black
     outlineSettings {
      bApplyOutlineToDropShadows=false
      bMiteredCorners=false
      bSeparateFillAlpha=false
      outlineMaterial=null
      outlineSize=0
      outlineColor=<rgba 0 0 0 1>
     }
    }
    shadowOffset=<0 0>
   }
   slot {class=VerticalBoxSlot}
  }
 ]
}
