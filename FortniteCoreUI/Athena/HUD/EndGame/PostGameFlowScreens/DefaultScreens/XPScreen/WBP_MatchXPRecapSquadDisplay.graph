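# Squad display widget for the post-game match XP recap screen, derived from AthenaIGMSquadDisplayWidget.
# It lays out an initially collapsed local-player slot beside a DynamicEntryBox of squad member entries, both using WBP_MatchXPRecapSquadMemberWidget and bound to inherited parent variables.
# PreConstruct applies a configurable scale variable to the enclosing ScaleBox, resizing the whole row.
# The graph holds no population logic.
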WidgetBlueprint /FortniteCoreUI/Athena/HUD/EndGame/PostGameFlowScreens/DefaultScreens/XPScreen/WBP_MatchXPRecapSquadDisplay
graphs [EventGraph {nodes [
   Event PreConstruct()
   > SetUserSpecifiedScale!(
    execute
    self:Scale Box Object Reference=Get ScaleBox_SquadMemberEntries() -> (ScaleBox_SquadMemberEntries:Scale Box Object Reference)
    InUserSpecifiedScale:float=Get In User Specified Scale() -> (In User Specified Scale:double)
   ) -> (then)
   Event Construct()
   Event Tick()
  ]}]
variables [In User Specified Scale {type=number, default=0.75}]
widgets {
 info {inheritedWidgetCount=2, parentClass=/Script/FortniteUI.AthenaIGMSquadDisplayWidget, rootWidgetClass=&Overlay, widgetCount=6}
 widgets [
  Overlay_0 {widgetClassPath=&Overlay, widget=Overlay_0}
  ScaleBox_SquadMemberEntries {
   widgetClassPath=&ScaleBox
   bIsVariable=true
   parent=Overlay_0
   widget=ScaleBox_SquadMemberEntries
   properties {stretch=UserSpecified}
   slot {class=OverlaySlot, properties {horizontalAlignment=HAlign_Fill, verticalAlignment=VAlign_Fill, padding=<ltrb 15 8 15 -12>}}
  }
  HorizontalBox_66 {
   widgetClassPath=&HorizontalBox
   parent=ScaleBox_SquadMemberEntries
   widget=HorizontalBox_66
   properties {renderTransform {angle=0, scale=<1 1>, shear=<0 0>, translation=<0 -11>}}
   slot {class=ScaleBoxSlot}
  }
  LocalPlayerSquadMemberWidget {
   widgetClassPath=&/WBP_MatchXPRecapSquadMemberWidget
   bInherited=true
   bIsVariable=true
   parent=HorizontalBox_66
   widget=LocalPlayerSquadMemberWidget
   properties {bIsLocalPlayerSlot=true, visibility=Collapsed}
   slot {class=HorizontalBoxSlot}
  }
  ScaleBox_Players {
   widgetClassPath=&ScaleBox
   parent=HorizontalBox_66
   widget=ScaleBox_Players
   properties {stretch=UserSpecified}
   slot {class=HorizontalBoxSlot, properties {verticalAlignment=VAlign_Bottom, padding=<ltrb 10 0 0 0>}}
  }
  EntryBox_SquadMemberWidgets {
   widgetClassPath=&DynamicEntryBox
   bInherited=true
   bIsVariable=true
   parent=ScaleBox_Players
   widget=EntryBox_SquadMemberWidgets
   properties {
    entryWidgetClass=&/WBP_MatchXPRecapSquadMemberWidget.WBP_MatchXPRecapSquadMemberWidget_C
    entrySpacing=<60 0>
   }
   slot {class=ScaleBoxSlot}
  }
 ]
}
