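# Translucent UI material for a rematch button bracket graphic on the post-game XP screen.
# Antialiased sine and rotated linear-gradient masks combine through Min/Max into the bracket shape, then add the Rivals icon texture's alpha.
# A static Flip switch, on by default, mirrors the shape's UVs horizontally.
# SmoothVersion blends toward a soft variant by changing UV scale, edge smoothing and an eased falloff.
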Material /FortniteCoreUI/HUD/EndGame/PostGameFlowScreens/DefaultScreens/XPScreen/Materials/M_UI_RematchButton_Bracket
graph {
 properties {
  bAutomaticallySetUsageInEditor=false
  bUsedWithStaticMesh=false
  blendMode=BLEND_Translucent
  materialDomain=MD_UI
  usedShadingModels=Default Lit
 }
 nodes [
  n1 = Sine(None=Add(A=NamedRerouteUsage(), B=Constant2Vector() {g=-0.0125})) {period=2}
  n2 = Saturate(None=MaterialFunctionCall(DistanceField=ComponentMask(None=@n1) {g=true}, Threshold=Constant() {r=0.375}, Smoothing=NamedRerouteUsage()) {materialFunction=&MF_UI_AntiAliasing}.Inner)
  n3 = Add(A=NamedRerouteUsage(), B=AppendVector(A=Constant() {r=0.3}, B=Constant()))
  n4 = MaterialFunctionCall!(Degrees=Constant()) {materialFunction=/Engine/Functions/Engine_MaterialFunctions01/Math/DegreesToRadians}
  n5 = Saturate(None=MaterialFunctionCall(
    DistanceField=MaterialFunctionCall(UVs=@n3, RotationAngle=@n4.Radians) {materialFunction=&MF_UI_LinearGradient}.Result
    Threshold=Constant() {r=0.5}
    Smoothing=NamedRerouteUsage()
   ) {materialFunction=&MF_UI_AntiAliasing}.Inner)
  n6 = Add(A=NamedRerouteUsage(), B=Constant2Vector() {r=0.4})
  n7 = MaterialFunctionCall(Degrees=Constant()) {materialFunction=&DegreesToRadians}
  n8 = MaterialFunctionCall(
   DistanceField=OneMinus(None=MaterialFunctionCall(UVs=@n6, RotationAngle=@n7.Radians) {materialFunction=&MF_UI_LinearGradient}.Result)
   Threshold=Constant() {r=0.5}
   Smoothing=NamedRerouteUsage()
  ) {materialFunction=&MF_UI_AntiAliasing}
  n9 = NamedRerouteUsage()
  n10 = Constant2Vector() {g=-0.35}
  n11 = Constant() {r=45.5}
  n12 = MaterialFunctionCall(
   DistanceField=MaterialFunctionCall(
    UVs=Add(A=@n9, B=@n10)
    RotationAngle=MaterialFunctionCall(Degrees=@n11) {materialFunction=&DegreesToRadians}.Radians
   ) {materialFunction=&MF_UI_LinearGradient}.Result
   Threshold=Constant() {r=0.5}
   Smoothing=NamedRerouteUsage()
  ) {materialFunction=&MF_UI_AntiAliasing}
  n13 = Add(A=NamedRerouteUsage(), B=Constant2Vector() {g=0.31})
  n14 = MaterialFunctionCall(Degrees=Constant() {r=40}) {materialFunction=&DegreesToRadians}
  n15 = MaterialFunctionCall(
   DistanceField=OneMinus(None=MaterialFunctionCall(UVs=@n13, RotationAngle=@n14.Radians) {materialFunction=&MF_UI_LinearGradient}.Result)
   Threshold=Constant() {r=0.5}
   Smoothing=NamedRerouteUsage()
  ) {materialFunction=&MF_UI_AntiAliasing}
  n16 = Min(A=Min(A=Max(A=@n2, B=@n5), B=Saturate(None=@n8.Inner)), B=Min(A=Saturate(None=@n12.Inner), B=Saturate(None=@n15.Inner)))
  n17 = Multiply(
   A=MaterialFunctionCall("Value (0-1)"=@n16) {materialFunction=&MF_UI_EaseCurves}.EaseInQuint
   B=MaterialFunctionCall("Value (0-1)"=NamedRerouteUsage()) {materialFunction=&MF_UI_EaseCurves}.EaseOutSine
  )
  n18 = ScalarParameter() {parameterName=SmoothVersion}
  n19 = Constant()
  n20 = Add(A=TextureCoordinate(), B=AppendVector(A=Constant() {r=0.13}, B=Constant() {r=-0.01}))
  n21 = Constant() {r=0.65}
  Add(
   A=Multiply(A=LinearInterpolate(A=@n16, B=@n17, Alpha=@n18) {constA=1, constB=0.75}, B=OneMinus(None=@n19))
   B=TextureSample(UVs=MaterialFunctionCall(UVs=@n20, Texture Scale=@n21) {materialFunction=&ScaleUVsByCenter}.UVs) {texture=/FortniteCoreUI/HUD/EndGame/PostGameFlowScreens/DefaultScreens/XPScreen/Textures/T_UI_Rivals_Icon}.A
  )
  Constant2Vector() {r=0.03}
  Constant4Vector() {constant=<rgba 0.9301 0.9911 0.0103 1>}
  MaterialExpressionTextureCoordinate_5 = TextureCoordinate()
  n22 = Convert(None=@MaterialExpressionTextureCoordinate_5) {
   nodeName=Break Float2
   convertInputs [{
     type=Vector2
     expressionInput {
      expression=&self:MaterialExpressionTextureCoordinate_5
      inputName=None
      mask=0
      maskA=0
      maskB=0
      maskG=0
      maskR=0
      outputIndex=0
     }
    }]
   convertOutputs [{type=Scalar}, {type=Scalar}]
  }
  n23 = AppendVector(A=OneMinus(None=@n22), B=@n22)
  n24 = ScalarParameter() {parameterName=SmoothVersion}
  NamedRerouteDeclaration(None=MaterialFunctionCall(
    UVs=StaticSwitchParameter(True=@n23, False=@MaterialExpressionTextureCoordinate_5) {defaultValue=true, parameterName=Flip}
    Texture Scale=LinearInterpolate(Alpha=@n24) {constA=1, constB=0.5}
   ) {materialFunction=&ScaleUVsByCenter}.UVs) {name=UV, nodeColor=<rgba 0 0 0 1>}
  NamedRerouteDeclaration(None=LinearInterpolate(B=ScalarParameter() {defaultValue=0.75, parameterName=Smoothing}, Alpha=ScalarParameter() {parameterName=SmoothVersion})) {name=Smoothing, nodeColor=<rgba 0 0 0 1>}
  NamedRerouteDeclaration(None=@n5) {name=SmoothGradientFalloffMask, nodeColor=<rgba 0 0 0 1>}
 ]
}
