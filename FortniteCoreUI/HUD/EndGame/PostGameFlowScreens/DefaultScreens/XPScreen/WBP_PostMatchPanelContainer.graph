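# Post-match XP screen container widget, derived from AthenaMatchXPScreenWidget, hosting tabs, matchmaking buttons, spectate controls and text chat.
# Finishing or skipping the opening sequence reveals tabs and buttons via TabFadeIn, broadcasting router events.
# Console variables and a spatial-experience check choose the S39 or older layout, reparenting ReadyUp and optionally skipping the XP rundown.
# Spectating hides the main panels and post-match audio mix; rematch state drives button styles and icons.
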WidgetBlueprint /FortniteCoreUI/HUD/EndGame/PostGameFlowScreens/DefaultScreens/XPScreen/WBP_PostMatchPanelContainer
animations [
 TabFadeIn {
  binding_count=3
  length_seconds=0.2492
  object=&self:TabFadeIn
  sequence {bindings [
    Overlay_TopBar {
     id=#1
     possessed_class=Overlay
     tracks [{
       type=MovieScene2DTransformTrack
       sections [{
         type=MovieScene2DTransformSection
         end_frame=4
         start_frame=0
         channels [
          Translation.X {type=MovieSceneScriptingFloatChannel, const [0,1]}
          Translation.Y {type=MovieSceneScriptingFloatChannel, keys [[0,-75],[4.9837,0]]}
         ]
        }]
      }]
    }
    SizeBoxNavigationOptions {
     id=#2
     possessed_class=SizeBox
     tracks [{
       type=MovieSceneFloatTrack
       display_name=Render Opacity
       sections [{type=MovieSceneFloatSection, end_frame=4, start_frame=0, channels [None {type=MovieSceneScriptingFloatChannel, keys [[0,0],[4.9837,1]]}]}]
      }]
    }
    SizeBox_BottomBar {
     id=#3
     possessed_class=SizeBox
     tracks [
      {
       type=MovieScene2DTransformTrack
       sections [{
         type=MovieScene2DTransformSection
         end_frame=4
         start_frame=0
         channels [Translation.Y {type=MovieSceneScriptingFloatChannel, keys [[0,50],[4.9837,0]]}]
        }]
      }
      {
       type=MovieSceneFloatTrack
       display_name=Render Opacity
       sections [{type=MovieSceneFloatSection, end_frame=4, start_frame=0, channels [None {type=MovieSceneScriptingFloatChannel, keys [[0,0],[4.9837,1]]}]}]
      }
     ]
    }
   ]}
 }
 Intro {
  binding_count=1
  length_seconds=0.5
  object=&self:Intro
  sequence {bindings [Image_Background {
     id=#4
     possessed_class=Image
     tracks [
      {
       type=MovieSceneWidgetMaterialTrack
       sections [{
         type=MovieSceneParameterSection
         end_frame=10
         start_frame=0
         channels [Transition Mask Scale {type=MovieSceneScriptingFloatChannel, keys [[0,0],[10,1]]}]
        }]
      }
      {
       type=MovieSceneFloatTrack
       display_name=Render Opacity
       sections [{type=MovieSceneFloatSection, end_frame=10, start_frame=0, channels [None {type=MovieSceneScriptingFloatChannel, keys [[0,0],[9.9997,1]]}]}]
      }
     ]
    }]}
 }
]
graphs [
 AudioMixHandling {nodes [
   AudioMixHandling!() -> (then, ApplyMix:bool=~) {is=title K2Node_FunctionEntry}
   > n1 = Branch(Condition=Get ApplyMix() -> (ApplyMix:bool))
   n2 = Get FortSoundModulatorGroupSubsystem() -> (ReturnValue:Fort Sound Modulator Group Subsystem Object Reference) {is=title K2Node_GetSubsystem}
   > n3 = ActivateMix!(
    execute
    self:Fort Sound Modulator Group Subsystem Object Reference=@n2
    InMix:Sound Modulator Group Mix Object Reference=/Game/Sounds/Modulation/SoundSelectorPresets/Mix_PostMatchScreen
   ) -> (then, ReturnValue:Sound Control Bus Mix Object Reference)
   > Set PostMatchControlBusMix(PostMatchControlBusMix:Sound Control Bus Mix Object Reference=@n3) -> (Output_Get:Sound Control Bus Mix Object Reference)
   UnregisterModulatorGroup!(
    execute=@n1.else
    self:Fort Sound Modulator Group Subsystem Object Reference=@n2
    InGroup:Sound Selector Preset Object Reference=Get SelectorModulationGroup() -> (SelectorModulationGroup:Sound Selector Preset Object Reference)
   ) -> (then)
   > UnregisterModulatorGroup(self=@n2, InGroup=/Game/Sounds/Modulation/SoundSelectorPresets/SSP_PostMatchScreen_Muted)
   > Deactivate Control Bus Mix!(
    execute
    Mix:Sound Control Bus Mix Object Reference=Get PostMatchControlBusMix() -> (PostMatchControlBusMix:Sound Control Bus Mix Object Reference)
   ) -> (then)
  ]}
 BP_ShouldHandleTabChangeAudio {nodes [
   ShouldHandleTabChangeAudio() {is=title K2Node_FunctionEntry}
   > Return Node!(execute, ReturnValue:bool=Get ShouldPlayTabChangeAudio() -> (ShouldPlayTabChangeAudio:bool))
  ]}
 BP_ShouldHandleTabChangeInput {nodes [
   ShouldHandleTabChangeInput() {is=title K2Node_FunctionEntry}
   > n1 = Parent: ShouldHandleTabChangeInput!(execute) -> (then, ReturnValue:bool) {is=title K2Node_CallParentFunction}
   n2 = IsSpectating!(self:Athena Match XPScreen Widget Object Reference) -> (ReturnValue:bool)
   > Return Node(ReturnValue=Select(Option 0:bool=@n1, Option 1:bool=NOT Boolean(A=@n2), Index=Get CancelTabInputWhenSpectating() -> (CancelTabInputWhenSpectating:bool)) -> (ReturnValue:bool))
  ]}
 EventGraph {nodes [
   "On Clicked (Button_AutoAdvance_Continue)"!() -> (OutputDelegate, then, Button:Common Button Base Object Reference) {is=title K2Node_ComponentBoundEvent}
   > AutoAdvanceToNextTab(self:Athena Match XPScreen Widget Object Reference)
   > SetVisibility(self=Get Button_AutoAdvance_Continue() -> (Button_AutoAdvance_Continue:WBP UIKit Button Quiet Object Reference), InVisibility=Collapsed)
   "On Clicked (Button_ReportPlayer)"!() -> (OutputDelegate, then, Button:Common Button Base Object Reference) {is=title K2Node_ComponentBoundEvent}
   > HandleReportPlayerClicked(self:Athena Match XPScreen Widget Object Reference)
   "On Clicked (Button_Quests)"!() -> (OutputDelegate, then, Button:Common Button Base Object Reference) {is=title K2Node_ComponentBoundEvent}
   > NavigateToQuestScreen!(
    execute
    Context:Widget Object Reference=Self-Reference()
    StartingCategoryTag:"Gameplay Tag Structure (by ref)"=MakeLiteralGameplayTag!(Value:Gameplay Tag Structure=(TagName="QuestCategory.All")) -> (ReturnValue:Gameplay Tag Structure)
   ) -> (then)
   "On Clicked (Button_Skip)"!() -> (OutputDelegate, then, Button:Common Button Base Object Reference) {is=title K2Node_ComponentBoundEvent}
   > HandleVisualizationSkipped()
   "On Clicked (Button_Spectate)"!() -> (OutputDelegate, then, Button:Common Button Base Object Reference) {is=title K2Node_ComponentBoundEvent}
   > HandleSpectatingClicked(self:Athena Match XPScreen Widget Object Reference)
   > AudioMixHandling!(execute, self, ApplyMix:bool=false) -> (then) {is=title K2Node_CallFunction}
   OnTabFadeInAnimationFinished() {is=title K2Node_CustomEvent}
   > SetTabsShown!(execute, self:Athena Match XPScreen Widget Object Reference, bInTabsShown:bool=true) -> (then)
   n1 = OnOpenTextChat!() -> (OutputDelegate, then, IsTextChatOpen:bool=~) {is=title K2Node_CustomEvent}
   > n2 = Branch(Condition=@n1.IsTextChatOpen)
   n3 = SlotAsOverlaySlot(Widget=Get SafeZone_TextChat() -> (SafeZone_TextChat:Safe Zone Object Reference))
   > SetPadding!(
    execute
    self:Overlay Slot Object Reference=@n3
    InPadding_Left:float=0
    InPadding_Top:float=0
    InPadding_Right:float=0
    InPadding_Bottom:float=0
   ) -> (then)
   SetPadding(execute=@n2.else, self=@n3, InPadding_Left=20, InPadding_Top=-75, InPadding_Right=65, InPadding_Bottom=20)
   UnlockPostMatch() {is=title K2Node_CustomEvent}
   > ShowTabsAndButtons()
   n4 = Get WBP_TextChat_InputListener() -> (WBP_TextChat_InputListener:WBP Text Chat Input Listener Object Reference)
   > SetBlocksLegacyInput!(execute, self:WBP Text Chat Input Listener Object Reference=@n4, BlockWhileActive:bool=true, BlockWhileInactive:bool=false) -> (then)
   > SetBottomBarButtonsVisible!(
    execute
    self:WBP Wrapper Text Chat Object Reference=Get WBP_Wrapper_Text_Chat!(self:WBP Text Chat Input Listener Object Reference=@n4) -> (WBP_Wrapper_Text_Chat:WBP Wrapper Text Chat Object Reference)
    Visible:bool=true
   ) -> (then)
   > n5 = StartListeningToEvent!(
    execute:Exec
    Target:Gameplay Event Router Component Object Reference=GetGlobalEventRouter()
    EventType:Script Struct Object Reference=/Script/FortniteUI.MatchXPScreenEvent_ToggledFullscreenMap
    Context:Object Reference
    EventBubblingRule:EEventBubblingRule Enum
   ) -> (
    then:Exec
    AsyncTaskProxy:Async Action Start Listening To Event Object Reference
    OnEventReceived:Exec
    Payload:Match XPScreen Event Toggled Fullscreen Map Structure
   )
   > Set ToggleFullscreenListener(ToggleFullscreenListener:Async Action Start Listening To Event Object Reference=@n5) -> (Output_Get:Async Action Start Listening To Event Object Reference)
   > n6 = Branch(Condition=IsMobileGame())
   ActivateWidget(execute=@n6.else, self=Get WBP_Wrapper_Text_Chat(self=Get WBP_TextChat_InputListener()))
   n7 = Break Match XPScreen Event Toggled Fullscreen Map!(MatchXPScreenEvent_ToggledFullscreenMap:"Match XPScreen Event Toggled Fullscreen Map Structure (by ref)"=@n5.Payload) -> (bFullscreenMapVisible:bool=~) {is=title K2Node_BreakStruct}
   SetBottomBarButtonsVisible(execute=@n5.OnEventReceived, self=Get WBP_Wrapper_Text_Chat(self=Get WBP_TextChat_InputListener()), Visible=NOT Boolean(A=@n7))
   > Branch(Condition=@n7)
   > CloseTextChat(self:WBP Text Chat Input Listener Object Reference=Get WBP_TextChat_InputListener())
   InitialTabShowFinished() {is=title K2Node_CustomEvent}
   > Broadcast Stateful Event!(
    execute
    self:Gameplay Event Router Component Object Reference=GetGlobalEventRouter()
    EventData:Match XPScreen Event XPRundown Sequence Complete Structure=Make Match XPScreen Event XPRundown Sequence Complete!() -> (MatchXPScreenEvent_XPRundownSequenceComplete:Match XPScreen Event XPRundown Sequence Complete Structure) {is=title K2Node_MakeStruct}
    EventContext:Object Reference=Self-Reference()
   ) -> (then, ReturnValue:bool)
   n8 = Event PreConstruct()
   > Branch(Condition=@n8.IsDesignTime)
   > ClearChildren(self=Get TabButtonsContainer() -> (TabButtonsContainer:Horizontal Box Object Reference))
   > For Loop(FirstIndex=1, LastIndex=Get DesignTimeTabCount() -> (DesignTimeTabCount:int=0))
   > n9 = Create Widget!(
    execute:Exec
    Class:User Widget Class Reference=/Game/Athena/HUD/BP_SkewedTabButton.BP_SkewedTabButton_C
    OwningPlayer:Player Controller Object Reference
    Style:Common Button Style Class Reference=/FortUILibrary/Foundation/Buttons/ButtonStyles/Athena/Athena_ButtonStyle_CollectionTab.Athena_ButtonStyle_CollectionTab_C
    bApplyAlphaOnDisable:bool=True
    bLocked:bool=False
    bSelectable:bool=True
    bShouldSelectUponReceivingFocus:bool=False
    bInteractableWhenSelected:bool=False
    bToggleable:bool=False
    bTriggerClickedAfterSelection:bool=False
    bDisplayInputActionWhenNotInteractable:bool=True
    bHideInputActionWithKeyboard:bool=False
    bShouldUseFallbackDefaultInputAction:bool=True
    bRequiresHold:bool=False
    HoldData:Common UIHold Data Class Reference
    ClickMethod:EButtonClickMethod Enum=DownAndUp
    InputPriority:int=0
    TriggeringInputAction:Data Table Row Handle Structure
   ) -> (then:Exec, ReturnValue:BP Skewed Tab Button Object Reference)
   > AddChildToHorizontalBox!(execute, self:Horizontal Box Object Reference=Get TabButtonsContainer(), Content:Widget Object Reference=@n9) -> (then, ReturnValue:Horizontal Box Slot Object Reference)
   n10 = Event OnAutoAdvanceTimeRemainingUpdated!() -> (OutputDelegate, then, TimeRemaining:int=0) {is=title K2Node_Event}
   > n11 = Branch(Condition="integer > integer"(A=@n10.TimeRemaining))
   > SetVisibility(self=Get Button_AutoAdvance_Continue(), InVisibility=SelfHitTestInvisible)
   > SetText!(
    execute
    self:Fort CTAButton Object Reference=Get Button_AutoAdvance_Continue()
    ButtonText:"Text (by ref)"=Format Text!(Format="LOCTEXT(\"Continue ({Remaining})\")", Remaining:int=@n10.TimeRemaining) -> (Result)
   ) -> (then)
   SetVisibility(execute=@n11.else, self=Get Button_AutoAdvance_Continue(), InVisibility=Collapsed)
   Event OnAutoAdvanceSequenceFinished() {is=title K2Node_Event}
   > Branch(Condition=NOT Boolean(A=Get UsingS39PostMatchUpdates() -> (UsingS39PostMatchUpdates:bool)))
   > UnlockPostMatch() {is=title K2Node_CallFunction}
   Event On Deactivated()
   > Set ShouldPlayTabChangeAudio(ShouldPlayTabChangeAudio:bool=false) -> (Output_Get:bool)
   > AudioMixHandling(ApplyMix=false)
   Event OnReplayRundownClicked() {is=title K2Node_Event}
   > StopAnimation(InAnimation=Get TabFadeIn() -> (TabFadeIn:Widget Animation Object Reference))
   > PlayAnimationTimeRange!(
    execute
    self:User Widget Object Reference
    InAnimation:Widget Animation Object Reference=Get TabFadeIn()
    StartAtTime:float=0
    EndAtTime:float=0.01
    NumLoopsToPlay:int=1
    PlayMode:EUMGSequencePlayMode Enum
    PlaybackSpeed:float=1
    bRestoreState:bool=false
   ) -> (then, ReturnValue:Widget Animation Handle Structure)
   n12 = Event RematchStatusChanged!() -> (OutputDelegate, then, bRematchAvailable:bool, StayWithTeamButton:Fort CTAButton Object Reference, bShouldShowStayWithTeam:bool) {is=title K2Node_Event}
   > Set Stay with Team Button(Stay with Team Button:Fort CTAButton Object Reference=@n12.StayWithTeamButton) -> (Output_Get:Fort CTAButton Object Reference)
   > Set bRematchAvailable(bRematchAvailable:bool=@n12.bRematchAvailable) -> (Output_Get:bool)
   > Set bShouldStayWithTeam(bShouldStayWithTeam:bool=@n12.bShouldShowStayWithTeam) -> (Output_Get:bool)
   > n13 = Sequence()
   n14 = Get!(execute:Exec=[@n13.then_1, @n24.then]) -> (then:Exec, else:Exec, Stay with Team Button:Fort CTAButton Object Reference)
   n15 = Get StayWithSquad_StyleTags_Default() -> (StayWithSquad_StyleTags_Default:Array of Gameplay Tag Structures)
   n16 = Get StayWithSquad_StyleTags_Rematch() -> (StayWithSquad_StyleTags_Rematch:Array of Gameplay Tag Structures)
   n17 = Get bRematchAvailable() -> (bRematchAvailable:bool)
   > UpdateWidgetStyleTags!(
    execute
    InWidget:Widget Object Reference=@n14
    InAddedTags:Array of Gameplay Tag Structures=Select(Option 0:Array of Gameplay Tag Structures=@n15, Option 1:Array of Gameplay Tag Structures=@n16, Index=@n17) -> (ReturnValue:Array of Gameplay Tag Structures)
    InRemovedTags:Array of Gameplay Tag Structures=Select(Option 0=@n16, Option 1=@n15, Index=@n17)
   ) -> (then, ReturnValue:bool)
   n18 = Get bShouldStayWithTeam() -> (bShouldStayWithTeam:bool)
   n19 = Get bRematchAvailable()
   n20 = AND Boolean(A=@n18, B=@n19)
   > SetVisibility(
    self=Get FVA_Squad_RematchIcon() -> (FVA_Squad_RematchIcon:Fort Visual Attachment Object Reference)
    InVisibility=Select(Option 0:ESlateVisibility Enum=Collapsed, Option 1:ESlateVisibility Enum=HitTestInvisible, Index=@n20) -> (ReturnValue:ESlateVisibility Enum)
   )
   n21 = Get UsingS39PostMatchUpdates()
   SetVisibility(
    execute=[@n13.then_0, @n26.then]
    self=Select(
     Option 0:Fort Visual Attachment Object Reference=Get FVA_ReadyUp_RematchIcon_Old() -> (FVA_ReadyUp_RematchIcon_Old:Fort Visual Attachment Object Reference)
     Option 1:Fort Visual Attachment Object Reference=Get FVA_ReadyUp_RematchIcon_New() -> (FVA_ReadyUp_RematchIcon_New:Fort Visual Attachment Object Reference)
     Index=NOT Boolean(A=@n21)
    ) -> (ReturnValue:Fort Visual Attachment Object Reference)
    InVisibility=Collapsed
   )
   n22 = NOT Boolean(A=@n18)
   > SetVisibility(
    self=Select(Option 0=Get FVA_ReadyUp_RematchIcon_Old(), Option 1=Get FVA_ReadyUp_RematchIcon_New(), Index=Get UsingS39PostMatchUpdates())
    InVisibility=Select(Option 0:ESlateVisibility Enum=Collapsed, Option 1:ESlateVisibility Enum=HitTestInvisible, Index=AND Boolean(A=@n22, B=@n19)) -> (ReturnValue:ESlateVisibility Enum)
   )
   n23 = Event StayWithTeamButtonRematchUpdate!() -> (OutputDelegate, then, bRematchAvailable:bool, StayWithTeamButton:Fort CTAButton Object Reference, bShouldShowStayWithTeam:bool) {is=title K2Node_Event}
   > Set Stay with Team Button(Stay with Team Button=@n23.StayWithTeamButton)
   > Set bRematchAvailable(bRematchAvailable=@n23.bRematchAvailable)
   > n24 = Set bShouldStayWithTeam(bShouldStayWithTeam=@n23.bShouldShowStayWithTeam)
   n25 = Event ReadyUpButtonRematchUpdate!() -> (OutputDelegate, then, bRematchAvailable:bool, bShouldShowStayWithTeam:bool) {is=title K2Node_Event}
   > Set bRematchAvailable(bRematchAvailable=@n25.bRematchAvailable)
   > n26 = Set bShouldStayWithTeam(bShouldStayWithTeam=@n25.bShouldShowStayWithTeam)
   n27 = Event On Show Tabs Set!() -> (OutputDelegate, then, bShowTabs:bool) {is=title K2Node_Event}
   > SetVisibility(
    self=Get HBox_Tabs() -> (HBox_Tabs:Horizontal Box Object Reference)
    InVisibility=Select(Option 0=Collapsed, Option 1=SelfHitTestInvisible, Index=@n27.bShowTabs)
   )
   n28 = Event SetMemberOptedOutState!() -> (
    OutputDelegate
    then
    Status:"EKeepPlayingTogetherVotingStatus Enum (by ref)"=Undecided
    MainContext:"Text (by ref)"
    SecondaryContext:"Text (by ref)"
   ) {is=title K2Node_Event}
   > n29 = Cast To WBP_KeepPlayingTogether_Message!(execute, Object=Get Content_MemberOptedOut() -> (Content_MemberOptedOut:Widget Object Reference)) -> (then, CastFailed, AsWBP Keep Playing Together Message:WBP Keep Playing Together Message Object Reference) {is=title K2Node_DynamicCast}
   > SetMemberOptedOutState!(
    execute
    self:WBP Keep Playing Together Message Object Reference=@n29
    Status:EKeepPlayingTogetherVotingStatus Enum=@n28.Status
    Primary:Text=@n28.MainContext
    Secondary:Text=@n28.SecondaryContext
   ) -> (then)
   Event On Toggle Spectating() {is=title K2Node_Event}
   > UpdateSpectateStateVisibility()
   > AudioMixHandling(ApplyMix=NOT Boolean(A=IsSpectating()))
   Event OnDefaultTabWidgetReady() {is=title K2Node_Event}
   > UnlockPostMatch()
   Event Construct()
   > Parent: Construct()
   > SetInputAction!(
    execute
    self:Common Action Widget Object Reference=Get NextTabAction() -> (NextTabAction:Common Action Widget Object Reference)
    InputActionRow:Data Table Row Handle Structure=Get NextTabInputAction() -> (NextTabInputAction:Data Table Row Handle Structure)
   ) -> (then)
   > SetInputAction(
    self=Get PreviousTabAction() -> (PreviousTabAction:Common Action Widget Object Reference)
    InputActionRow=Get PreviousTabInputAction() -> (PreviousTabInputAction:Data Table Row Handle Structure)
   )
   > BindToAnimationFinished(Animation=Get TabFadeIn(), Delegate=Create Event())
   > PlayAnimation(InAnimation=Get Intro())
   > n30 = Branch(Condition=IsMobileGame())
   > SetWidthOverride(self=Get SizeBox_Mobile() -> (SizeBox_Mobile:Size Box Object Reference), InWidthOverride=400)
   n31 = SlotAsOverlaySlot(Widget=Get SafeZone_TextChat())
   > SetVerticalAlignment(self=@n31, InVerticalAlignment=VAlign_Fill)
   > SetHorizontalAlignment(self=@n31, InHorizontalAlignment=HAlign_Fill)
   > SetPadding(self=@n31, InPadding_Left=20, InPadding_Top=-75, InPadding_Right=65, InPadding_Bottom=20)
   > Bind Event to Text Chat Open Finished(self:WBP Wrapper Text Chat Object Reference=Get WBP_Wrapper_Text_Chat(self=Get WBP_TextChat_InputListener()), Delegate=Create Event()) {is=title K2Node_AddDelegate}
   ClearWidthOverride(execute=@n30.else, self:Size Box Object Reference=Get SizeBox_Mobile())
   Event Destruct()
   > AudioMixHandling(ApplyMix=false)
   > StopListeningToEvent(self:Async Action Start Listening To Event Object Reference=Get ToggleFullscreenListener() -> (ToggleFullscreenListener:Async Action Start Listening To Event Object Reference))
   > Branch(Condition=IsMobileGame())
   > Unbind Event from Text Chat Open Finished(self:WBP Wrapper Text Chat Object Reference=Get WBP_Wrapper_Text_Chat(self=Get WBP_TextChat_InputListener()), Delegate=@n1) {is=title K2Node_RemoveDelegate}
   Event On Activated()
   > AudioMixHandling(ApplyMix=true)
   > UpdateS39PostMatchVisibility()
   Event OnInitialized()
   > SetVisibility(
    self=[
     Get Overlay_TopBar() -> (Overlay_TopBar:Overlay Object Reference)
     Get SizeBox_BottomBar() -> (SizeBox_BottomBar:Size Box Object Reference)
     Get SizeBoxNavigationOptions() -> (SizeBoxNavigationOptions:Size Box Object Reference)
    ]
    InVisibility=Collapsed
   )
   > SetRenderOpacity(self=Get SizeBoxNavigationOptions(), InOpacity=0)
   "To String (Boolean)"(InBool=@n20)
  ]}
 GetToggleSpectateVisibility {nodes [
   GetToggleSpectateVisibility() {is=title K2Node_FunctionEntry}
   n1 = IsUsingCompactUI()
   n2 = IsSpectating()
   > Return Node!(
    execute
    ShowBottomButtonGroup:bool=AND Boolean!(A=NOT Boolean(A=@n2), B=NOT Boolean(A=@n1), C:bool=Get UsingS39PostMatchUpdates()) -> (ReturnValue)
   )
  ]}
 HandleVisualizationSkipped {nodes [
   HandleVisualizationSkipped() {is=title K2Node_FunctionEntry}
   > SkipTabVisualization(self:Athena Match XPScreen Widget Object Reference)
   > ShowTabsAndButtons()
  ]}
 IsTextChatOpen {nodes [
   IsTextChatOpen() {is=title K2Node_FunctionEntry}
   > Return Node!(
    execute
    ReturnValue:bool=IsTextChatOpen!(self:WBP Text Chat Input Listener Object Reference=Get WBP_TextChat_InputListener()) -> (IsOpen:bool) {is=title K2Node_CallFunction}
   )
  ]}
 NavigateToPostGameOptions {nodes [
   NavigateToPostGameOptions!() -> (then, Navigation:EUINavigation Enum) {is=title K2Node_FunctionEntry}
   > Return Node(ReturnValue:Widget Object Reference=GetFirstEntry!(self:Fort Post Game Menu List View Object Reference=Get ListView_PostGameMenu() -> (ListView_PostGameMenu:Fort Post Game Menu List View Object Reference)) -> (ReturnValue:Fort Post Game Menu Entry Object Reference))
  ]}
 ShouldShowSpectateButton {nodes [
   ShouldShowSpectateButton() {is=title K2Node_FunctionEntry}
   > n1 = ShouldShowSpectate!(execute) -> (then, ReturnValue:bool)
   > Return Node!(
    execute
    ShouldShow:bool=AND Boolean(A=@n1, B=GetToggleSpectateVisibility!() -> (ShowBottomButtonGroup:bool) {is=title K2Node_CallFunction}, C=Get UsingS39PostMatchUpdates())
   )
  ]}
 ShowTabsAndButtons {nodes [
   ShowTabsAndButtons() {is=title K2Node_FunctionEntry}
   > Set IsScreenReady(IsScreenReady:bool=true) -> (Output_Get:bool)
   > SetVisibility(self=Get Button_Skip() -> (Button_Skip:WBP UIKit Button Quiet Object Reference), InVisibility=Collapsed)
   > SetVisibility(self=[Get Overlay_TopBar(), Get SizeBoxNavigationOptions(), Get SizeBox_BottomBar()], InVisibility=SelfHitTestInvisible)
   > UpdateSpectateStateVisibility()
   n1 = Get TabFadeIn()
   > PlayAnimation(InAnimation=@n1)
   > Set Timer by Event(Delegate=Create Event(), Time=GetEndTime(self=@n1), bLooping=false)
   > Set ShouldPlayTabChangeAudio(ShouldPlayTabChangeAudio=true)
   > Broadcast Event!(
    execute
    self:Gameplay Event Router Component Object Reference=GetGlobalEventRouter()
    EventData:Fort Event Fortnite Core UIPost Match Started Showing Tabs Structure=Make Fort Event Fortnite Core UIPost Match Started Showing Tabs!() -> (FortEvent_FortniteCoreUIPostMatchStartedShowingTabs:Fort Event Fortnite Core UIPost Match Started Showing Tabs Structure) {is=title K2Node_MakeStruct}
    EventContext:Object Reference=GetOwningPlayer()
   ) -> (then, ReturnValue:bool)
  ]}
 UpdateS39PostMatchVisibility {nodes [
   UpdateS39PostMatchVisibility() {is=title K2Node_FunctionEntry}
   > n1 = GetConsoleVariableBoolValue!(execute, VariableName:String=Athena.MatchXPScreenWidget.S39SuppressXPRundown) -> (then, ReturnValue:bool)
   > Set IsSuppressingRundown(IsSuppressingRundown:bool=OR Boolean(A=@n1, B=IsSpatialGameplayExperience!() -> (ReturnValue:bool))) -> (Output_Get:bool)
   > n2 = GetConsoleVariableBoolValue(VariableName=Athena.MatchXPScreenWidget.S39SuppressPostGameBattlepass)
   > Set IsSuppressingBattlepass(IsSuppressingBattlepass:bool=@n2) -> (Output_Get:bool)
   > n3 = GetConsoleVariableBoolValue(VariableName=Athena.MatchXPScreenWidget.S39EnablePostGameUpdates)
   n4 = IsSpatialGameplayExperience()
   > n5 = Set UsingS39PostMatchUpdates(UsingS39PostMatchUpdates:bool=AND Boolean!(A=@n3, B=NOT Boolean(A=@n4)) -> (ReturnValue)) -> (Output_Get:bool)
   > n6 = Branch(Condition=@n5)
   > UpdateSpectateStateVisibility()
   n7 = SlotAsOverlaySlot(Widget=Get SizeBoxNavigationOptions())
   n8 = Get!() -> (CTA_StackMargin_S39_Left:float=~, CTA_StackMargin_S39_Top:float=~, CTA_StackMargin_S39_Right:float=~, CTA_StackMargin_S39_Bottom:float=~)
   > SetPadding(
    self=@n7
    InPadding_Left=@n8
    InPadding_Top=@n8.CTA_StackMargin_S39_Top
    InPadding_Right=@n8.CTA_StackMargin_S39_Right
    InPadding_Bottom=Select(
     Option 0:float=float + float(A=@n8.CTA_StackMargin_S39_Bottom, B=80)
     Option 1:float=@n8.CTA_StackMargin_S39_Bottom
     Index=Get IsSuppressingBattlepass() -> (IsSuppressingBattlepass:bool)
    ) -> (ReturnValue:float)
   )
   > SetVerticalAlignment(self=@n7, InVerticalAlignment=VAlign_Center)
   > SetBrushFromSoftMaterial!(
    execute
    self:Image Object Reference=Get Image_Background() -> (Image_Background:Image Object Reference)
    SoftMaterial:Material Interface Soft Object Reference=/FortniteCoreUI/HUD/EndGame/PostGameFlowScreens/DefaultScreens/XPScreen/Materials/MI_PostMatch_Background_XPRundown
   ) -> (then)
   > SetRenderOpacity(self=Get S18_PlayerBattlePassXpAndReward() -> (S18_PlayerBattlePassXpAndReward:Fort Season Data Widget Object Reference), InOpacity=0)
   > SetVisibility(
    self=[Get TabBackground() -> (TabBackground:Image Object Reference), Get HBox_SocialArea() -> (HBox_SocialArea:Horizontal Box Object Reference)]
    InVisibility=Collapsed
   )
   > AddChild(
    self=Get NSlot_ReadyUp_NewPlacement() -> (NSlot_ReadyUp_NewPlacement:Named Slot Object Reference)
    Content=Get Button_ReadyUp() -> (Button_ReadyUp:Fort CTAButton Object Reference)
   )
   > Branch(Condition=OR Boolean(A=Get IsSuppressingRundown() -> (IsSuppressingRundown:bool), B=IsSpectating()))
   > HandleVisualizationSkipped() {is=title K2Node_CallFunction}
   SetVisibility(
    execute=@n6.else
    self=[
     Get Button_Spectate() -> (Button_Spectate:WBP UIKit Button Quiet Object Reference)
     Get Button_ReportPlayer() -> (Button_ReportPlayer:WBP UIKit Button Quiet Object Reference)
     Get Button_Quests() -> (Button_Quests:WBP UIKit Button Quiet Object Reference)
    ]
    InVisibility=Collapsed
   )
   n9 = Get SizeBoxNavigationOptions()
   > SetVisibility(self=@n9, InVisibility=SelfHitTestInvisible)
   n10 = SlotAsOverlaySlot(Widget=@n9)
   > SetPadding!(execute, self=@n10, InPadding:Margin Structure=Get CTA_StackMargin() -> (CTA_StackMargin:Margin Structure)) -> (then)
   > SetVerticalAlignment(self=@n10, InVerticalAlignment=VAlign_Bottom)
   > SetBrushFromSoftMaterial(self=Get Image_Background(), SoftMaterial=/Game/Athena/HUD/EndGame/PostGameFlowScreens/DefaultScreens/XPScreen/Art/MI_UI_Overlay_Grid)
   > SetRenderOpacity(self=Get S18_PlayerBattlePassXpAndReward(), InOpacity=1)
   > RepresentLocalPlayer(self:Fort Social Avatar Icon Object Reference=Get SocialAvatarIcon() -> (SocialAvatarIcon:Fort Social Avatar Icon Object Reference))
   > SetVisibility(self=[Get TabBackground(), Get HBox_SocialArea()], InVisibility=SelfHitTestInvisible)
  ]}
 UpdateSpectateStateVisibility {nodes [
   UpdateSpectateStateVisibility() {is=title K2Node_FunctionEntry}
   > Branch(Condition=Get IsScreenReady() -> (IsScreenReady:bool))
   > n1 = Sequence!(execute) -> (then_0, then_1, then_2:Exec)
   n2 = IsSpectating()
   > SetVisibility(
    self=[
     Get Image_Background()
     Get SizeBoxNavigationOptions()
     Get Overlay_TopBar()
     Get SizeBox_SwitcherContainer() -> (SizeBox_SwitcherContainer:Size Box Object Reference)
    ]
    InVisibility=Select(Option 0:ESlateVisibility Enum=SelfHitTestInvisible, Option 1:ESlateVisibility Enum=Collapsed, Index=@n2) -> (ReturnValue:ESlateVisibility Enum)
   )
   > SetVisibility(
    self=Get Button_Spectate()
    InVisibility=Select(Option 0=Collapsed, Option 1=Visible, Index=ShouldShowSpectateButton!() -> (ShouldShow:bool) {is=title K2Node_CallFunction})
   )
   > SetVisibility(
    self=[Get Button_Quests(), Get Button_ReportPlayer()]
    InVisibility=Select(Option 0=Collapsed, Option 1=SelfHitTestInvisible, Index=GetToggleSpectateVisibility())
   )
   n3 = Get WBP_TextChat_InputListener()
   SetMobileLayout!(execute=@n1.then_2, self:WBP Text Chat Input Listener Object Reference=@n3, Text Chat Open:bool=IsTextChatOpen(self=@n3)) -> (then)
   n4 = Get WBP_TextChat_InputListener()
   Set MobileClosedChatHorizontalAligment!(
    execute=@n1.then_1
    MobileClosedChatHorizontalAligment:EHorizontalAlignment Enum=Select(Option 0:EHorizontalAlignment Enum=HAlign_Left, Option 1:EHorizontalAlignment Enum=HAlign_Right, Index=@n2) -> (ReturnValue:EHorizontalAlignment Enum)
    self:WBP Text Chat Input Listener Object Reference=@n4
   ) -> (then, Output_Get:EHorizontalAlignment Enum=HAlign_Fill)
   > n5 = GetConsoleVariableBoolValue(VariableName=Athena.MatchXPScreenWidget.S39EnablePostGameUpdates)
   > Set!(
    execute
    MobileClosedChatPadding_Left:float=0
    MobileClosedChatPadding_Top:float=SelectFloat(A=SelectFloat(A=125, B=405, bPickA=@n5), B=0, bPickA=@n2)
    MobileClosedChatPadding_Right:float=0
    MobileClosedChatPadding_Bottom:float=0
    self:WBP Text Chat Input Listener Object Reference=@n4
   ) -> (then, Output_Get:Margin Structure)
  ]}
]
variables [
 BattlepassChaseList {type=CommonActivatableWidget, default=null}
 bRematchAvailable {type=boolean}
 bShouldStayWithTeam {type=boolean}
 CancelTabInputWhenSpectating {type=boolean, default=true}
 CTA_StackMargin {type=Margin, default=<ltrb 32 0 97 150>}
 CTA_StackMargin_S39 {type=Margin, default=<ltrb 32 0 100 0>}
 DesignTimeTabCount {type=integer, default=3}
 IsScreenReady {type=boolean}
 IsSuppressingRundown {type=boolean}
 Mutator {type=StatefulAudioMutator, default=null}
 PostMatchControlBusMix {type=SoundControlBusMix, default=/Game/Sounds/Modulation/Mixes/CBM_DuckOnResultsScreen}
 SelectorModulationGroup {type=SoundSelectorPresetAsset, default=/Game/Sounds/Modulation/SoundSelectorPresets/SSP_PostMatchScreen_Ducked}
 ShouldPlayTabChangeAudio {type=boolean}
 Stay with Team Button {type=FortCTAButton, default=null}
 StayWithSquad_StyleTags_Default {type="array<GameplayTag>", default [{tagName=StyleSheet.Style.Regular}]}
 StayWithSquad_StyleTags_Rematch {type="array<GameplayTag>", default [{tagName=StyleSheet.Style.Loud}]}
 ToggleFullscreenListener {type=AsyncAction_StartListeningToEvent, default=null}
 UseMutators {type=boolean, default=true}
 UsingS39PostMatchUpdates {type=boolean, default=true}
]
widgets {
 info {
  inheritedWidgetCount=25
  namedSlotCount=1
  parentClass=/Script/FortniteUI.AthenaMatchXPScreenWidget
  rootWidgetClass=&Overlay
  widgetCount=75
 }
 widgets [
  Overlay_0 {widgetClassPath=&Overlay, widget=Overlay_0}
  Image_Background {
   widgetClassPath=&Image
   bIsVariable=true
   parent=Overlay_0
   widget=Image_Background
   properties {brush {
     bIsDynamicallyLoaded=false
     drawAs=Image
     imageType=FullColor
     mirroring=NoMirror
     resourceName=None
     resourceObject=&MI_PostMatch_Background_XPRundown
     tiling=NoTile
     imageSize=<32 32>
     margin=<ltrb 0 0 0 0>
     outlineSettings {
      bUseBrushTransparency=false
      roundingType=HalfHeightRadius
      width=0
      color {colorUseRule=UseColor_Specified, specifiedColor=<rgba 0 0 0 0>}
      cornerRadii=<0 0 0 0>
     }
     tintColor {colorUseRule=UseColor_Specified, specifiedColor=<rgba 1 1 1 1>}
     uVRegion {bIsValid=false, max=<0 0>, min=<0 0>}
    }}
   slot {class=OverlaySlot, properties {horizontalAlignment=HAlign_Fill, verticalAlignment=VAlign_Fill}}
  }
  SizeBox_SwitcherContainer {
   widgetClassPath=&SizeBox
   bIsVariable=true
   parent=Overlay_0
   widget=SizeBox_SwitcherContainer
   slot {class=OverlaySlot, properties {horizontalAlignment=HAlign_Fill, verticalAlignment=VAlign_Fill}}
  }
  VisibilitySwitcher_Content {
   widgetClassPath=&CommonVisibilitySwitcher
   bInherited=true
   bIsVariable=true
   parent=SizeBox_SwitcherContainer
   widget=VisibilitySwitcher_Content
   properties {bAutoActivateSlot=false}
   slot {class=SizeBoxSlot}
  }
  SafeZone_AllSides {
   widgetClassPath=&SafeZone
   parent=Overlay_0
   widget=SafeZone_AllSides
   slot {class=OverlaySlot, properties {horizontalAlignment=HAlign_Fill, verticalAlignment=VAlign_Fill}}
  }
  Overlay_74 {widgetClassPath=&Overlay, parent=SafeZone_AllSides, widget=Overlay_74, slot {class=SafeZoneSlot}}
  SizeBoxNavigationOptions {
   widgetClassPath=&SizeBox
   bInherited=true
   bIsVariable=true
   parent=Overlay_74
   widget=SizeBoxNavigationOptions
   properties {bOverride_WidthOverride=true, widthOverride=480}
   slot {class=OverlaySlot, properties {horizontalAlignment=HAlign_Right, verticalAlignment=VAlign_Center, padding=<ltrb 0 0 100 0>}}
  }
  Leto_IGMMWrapper {
   widgetClassPath=&VerticalBox
   bInherited=true
   parent=SizeBoxNavigationOptions
   widget=Leto_IGMMWrapper
   slot {class=SizeBoxSlot, properties {verticalAlignment=VAlign_Top}}
  }
  VBox_Matchmaking {widgetClassPath=&VerticalBox, bInherited=true, parent=Leto_IGMMWrapper, widget=VBox_Matchmaking, slot {class=VerticalBoxSlot}}
  SizeBox_Mobile {
   widgetClassPath=&SizeBox
   bInherited=true
   bIsVariable=true
   parent=VBox_Matchmaking
   widget=SizeBox_Mobile
   properties {widthOverride=400}
   slot {class=VerticalBoxSlot, properties {verticalAlignment=VAlign_Bottom}}
  }
  ScaleBox_1 {
   widgetClassPath=&ScaleBox
   parent=SizeBox_Mobile
   widget=ScaleBox_1
   properties {stretchDirection=DownOnly}
   slot {class=SizeBoxSlot}
  }
  SquadDisplayWidget {
   widgetClassPath=/FortniteCoreUI/Athena/HUD/EndGame/PostGameFlowScreens/DefaultScreens/XPScreen/WBP_MatchXPRecapSquadDisplay
   bInherited=true
   bIsVariable=true
   parent=ScaleBox_1
   widget=SquadDisplayWidget
   slot {class=ScaleBoxSlot, properties {verticalAlignment=VAlign_Bottom}}
  }
  Text_MatchmakingStatus {
   widgetClassPath=&CommonTextBlock
   bInherited=true
   parent=VBox_Matchmaking
   widget=Text_MatchmakingStatus
   properties {
    autoWrapText=true
    justification=Center
    text="##Matchmaking status"
    textTransformPolicy=ToUpper
    visibility=Collapsed
    font {
     bForceMonospaced=false
     bMaterialIsStencil=false
     fontMaterial=null
     fontObject=&Burbank
     letterSpacing=0
     monospacedWidth=1
     size=24
     skewAmount=0
     typefaceFontName=Burbank-PlayerPower
     outlineSettings {
      bApplyOutlineToDropShadows=false
      bMiteredCorners=false
      bSeparateFillAlpha=false
      outlineMaterial=null
      outlineSize=0
      outlineColor=<rgba 0 0 0 1>
     }
    }
    renderTransform {angle=0, scale=<1 1>, shear=<-8 0>, translation=<0 0>}
    shadowOffset=<0 0>
   }
   slot {class=VerticalBoxSlot}
  }
  Content_MemberOptedOut {
   widgetClassPath=/FortniteCoreUI/Athena/HUD/EndGame/PostGameFlowScreens/DefaultScreens/XPScreen/WBP_KeepPlayingTogether_Message
   bInherited=true
   bIsVariable=true
   parent=VBox_Matchmaking
   widget=Content_MemberOptedOut
   properties {visibility=Collapsed}
   slot {class=VerticalBoxSlot, properties {padding=<ltrb 0 25 0 24>}}
  }
  Overlay_2 {
   widgetClassPath=&Overlay
   parent=Leto_IGMMWrapper
   widget=Overlay_2
   slot {class=VerticalBoxSlot, properties {verticalAlignment=VAlign_Top}}
  }
  WBP_CaptureForPostBufferUpdate {
   widgetClassPath=&WBP_CaptureForPostBufferUpdate
   bIsVariable=true
   parent=Overlay_2
   slot {class=OverlaySlot, properties {horizontalAlignment=HAlign_Fill, verticalAlignment=VAlign_Fill}}
  }
  VBox_NavButtons {
   widgetClassPath=&VerticalBox
   bInherited=true
   parent=Overlay_2
   widget=VBox_NavButtons
   slot {class=OverlaySlot, properties {horizontalAlignment=HAlign_Fill, verticalAlignment=VAlign_Fill}}
  }
  Slot_AboveReadyUp {
   widgetClassPath=/Script/WidgetSlotting.WidgetSlottingHostOverlay
   parent=VBox_NavButtons
   widget=Slot_AboveReadyUp
   properties {
    slottedWidgetDefaultHorizontalAlignment=HAlign_Fill
    slottedWidgetDefaultVerticalAlignment=VAlign_Fill
    hostTag {tagName=DynamicUI.HUD.Fort.Athena.PostMatch.NavButtonsTopSlot}
   }
   slot {class=VerticalBoxSlot}
  }
  Overlay_4 {widgetClassPath=&Overlay, parent=VBox_NavButtons, widget=Overlay_4, slot {class=VerticalBoxSlot}}
  NSlot_ReadyUp_OldPlacement {
   widgetClassPath=&NamedSlot
   bIsVariable=true
   parent=Overlay_4
   widget=NSlot_ReadyUp_OldPlacement
   slot {class=OverlaySlot, properties {horizontalAlignment=HAlign_Fill, verticalAlignment=VAlign_Fill}}
  }
  Button_ReadyUp {
   widgetClassPath=/FortUILibrary/UIKit/Buttons/WBP_UIKit_Button_Loud
   bInherited=true
   bIsVariable=true
   parent=NSlot_ReadyUp_OldPlacement
   widget=Button_ReadyUp
   properties {
    buttonSize=Large
    navigation=&self:WidgetTree.Button_ReadyUp.WidgetNavigation_0
    visibility=Collapsed
    padding=<ltrb 0 0 0 12>
    triggeringInputAction {dataTable=&InputActionDataTable, rowName=GenericFaceButtonTop}
   }
   slot {class=PanelSlot}
  }
  FVA_ReadyUp_RematchIcon_Old {
   widgetClassPath=&FortVisualAttachment
   bIsVariable=true
   parent=Overlay_4
   widget=FVA_ReadyUp_RematchIcon_Old
   properties {visibility=Collapsed, contentAnchor=<1 0>}
   slot {class=OverlaySlot, properties {horizontalAlignment=HAlign_Fill}}
  }
  Image_ReadyUp_RematchIcon {
   widgetClassPath=&Image
   parent=FVA_ReadyUp_RematchIcon_Old
   widget=Image_ReadyUp_RematchIcon
   properties {brush {
     bIsDynamicallyLoaded=false
     drawAs=Image
     imageType=FullColor
     mirroring=NoMirror
     resourceName=None
     resourceObject=/FortniteCoreUI/HUD/EndGame/PostGameFlowScreens/DefaultScreens/XPScreen/Materials/M_UI_RematchButton_Bracket
     tiling=NoTile
     imageSize=<86 86>
     margin=<ltrb 0 0 0 0>
     outlineSettings {
      bUseBrushTransparency=false
      roundingType=HalfHeightRadius
      width=0
      color {colorUseRule=UseColor_Specified, specifiedColor=<rgba 0 0 0 0>}
      cornerRadii=<0 0 0 0>
     }
     tintColor {colorUseRule=UseColor_Specified, specifiedColor=<rgba 1 1 1 1>}
     uVRegion {bIsValid=false, max=<0 0>, min=<0 0>}
    }}
   slot {class=SizeBoxSlot, properties {horizontalAlignment=HAlign_Left, verticalAlignment=VAlign_Center}}
  }
  Slot_BelowReadyUp {
   widgetClassPath=&WidgetSlottingHostOverlay
   parent=VBox_NavButtons
   widget=Slot_BelowReadyUp
   properties {bOverride_Cursor=true, hostTag {tagName=DynamicUI.HUD.Fort.Athena.PostMatch.BelowReadyUp}}
   slot {class=VerticalBoxSlot}
  }
  Overlay_5 {widgetClassPath=&Overlay, parent=VBox_NavButtons, widget=Overlay_5, slot {class=VerticalBoxSlot}}
  MenuEntry_DisabledStayWithSquad {
   widgetClassPath=&WBP_UIKit_Button_Regular
   bInherited=true
   bIsVariable=true
   parent=Overlay_5
   widget=MenuEntry_DisabledStayWithSquad
   properties {
    bIsEnabled=false
    bIsFocusable=false
    minHeight=56
    navigation=&self:WidgetTree.MenuEntry_DisabledStayWithSquad.WidgetNavigation_0
    visibility=Collapsed
    padding=<ltrb 0 0 0 12>
   }
   slot {class=OverlaySlot, properties {horizontalAlignment=HAlign_Fill, verticalAlignment=VAlign_Fill}}
  }
  FVA_Squad_RematchIcon {
   widgetClassPath=&FortVisualAttachment
   bIsVariable=true
   parent=Overlay_5
   widget=FVA_Squad_RematchIcon
   properties {bOverride_HeightOverride=true, bOverride_WidthOverride=true, heightOverride=56, visibility=Collapsed, widthOverride=56, contentAnchor=<1 0>}
   slot {class=OverlaySlot, properties {horizontalAlignment=HAlign_Fill}}
  }
  Image_Squad_RematchIcon {
   widgetClassPath=&Image
   parent=FVA_Squad_RematchIcon
   widget=Image_Squad_RematchIcon
   properties {brush {
     bIsDynamicallyLoaded=false
     drawAs=Image
     imageType=FullColor
     mirroring=NoMirror
     resourceName=None
     resourceObject=&M_UI_RematchButton_Bracket
     tiling=NoTile
     imageSize=<86 86>
     margin=<ltrb 0 0 0 0>
     outlineSettings {
      bUseBrushTransparency=false
      roundingType=HalfHeightRadius
      width=0
      color {colorUseRule=UseColor_Specified, specifiedColor=<rgba 0 0 0 0>}
      cornerRadii=<0 0 0 0>
     }
     tintColor {colorUseRule=UseColor_Specified, specifiedColor=<rgba 1 1 1 1>}
     uVRegion {bIsValid=false, max=<0 0>, min=<0 0>}
    }}
   slot {class=SizeBoxSlot, properties {horizontalAlignment=HAlign_Left, verticalAlignment=VAlign_Center}}
  }
  ListView_PostGameMenu {
   widgetClassPath=/Script/FortniteUI.FortPostGameMenuListView
   bInherited=true
   bIsVariable=true
   parent=VBox_NavButtons
   widget=ListView_PostGameMenu
   properties {
    allowOverscroll=false
    bAllowDragging=false
    bEnableRightClickScrolling=false
    clipping=Inherit
    entryClass=/Game/Athena/HUD/EndGame/PostGameFlowScreens/DefaultScreens/XPScreen/Menu/FortPostGameMenuEntry.FortPostGameMenuEntry_C
    entrySpacing=0
    entryWidgetClass=&FortPostGameMenuEntry_C
    itemShopEntryClass=/Game/Athena/HUD/EndGame/PostGameFlowScreens/DefaultScreens/XPScreen/Menu/FortPostGameMenuItemShopEntry.FortPostGameMenuItemShopEntry_C
    navigation=&self:WidgetTree.ListView_PostGameMenu.WidgetNavigation_0
    numDesignerPreviewEntries=2
   }
   slot {class=VerticalBoxSlot}
  }
  Overlay_3 {widgetClassPath=&Overlay, parent=VBox_NavButtons, widget=Overlay_3, slot {class=VerticalBoxSlot}}
  NSlot_ReadyUp_NewPlacement {
   widgetClassPath=&NamedSlot
   bIsVariable=true
   parent=Overlay_3
   widget=NSlot_ReadyUp_NewPlacement
   slot {class=OverlaySlot, properties {horizontalAlignment=HAlign_Fill, verticalAlignment=VAlign_Fill}}
  }
  FVA_ReadyUp_RematchIcon_New {
   widgetClassPath=&FortVisualAttachment
   bIsVariable=true
   parent=Overlay_3
   widget=FVA_ReadyUp_RematchIcon_New
   properties {visibility=Collapsed, contentAnchor=<1 0>}
   slot {class=OverlaySlot, properties {horizontalAlignment=HAlign_Fill}}
  }
  Image_ReadyUp_RematchIcon_New {
   widgetClassPath=&Image
   parent=FVA_ReadyUp_RematchIcon_New
   widget=Image_ReadyUp_RematchIcon_New
   properties {brush {
     bIsDynamicallyLoaded=false
     drawAs=Image
     imageType=FullColor
     mirroring=NoMirror
     resourceName=None
     resourceObject=&M_UI_RematchButton_Bracket
     tiling=NoTile
     imageSize=<86 86>
     margin=<ltrb 0 0 0 0>
     outlineSettings {
      bUseBrushTransparency=false
      roundingType=HalfHeightRadius
      width=0
      color {colorUseRule=UseColor_Specified, specifiedColor=<rgba 0 0 0 0>}
      cornerRadii=<0 0 0 0>
     }
     tintColor {colorUseRule=UseColor_Specified, specifiedColor=<rgba 1 1 1 1>}
     uVRegion {bIsValid=false, max=<0 0>, min=<0 0>}
    }}
   slot {class=SizeBoxSlot, properties {horizontalAlignment=HAlign_Left, verticalAlignment=VAlign_Center}}
  }
  SizeBox_ReadyUpToolTip {
   widgetClassPath=&SizeBox
   bInherited=true
   parent=VBox_NavButtons
   widget=SizeBox_ReadyUpToolTip
   properties {heightOverride=50, widthOverride=400}
   slot {class=VerticalBoxSlot}
  }
  AthenaReadyUpTooltipWidget {
   widgetClassPath=/FortniteCoreUI/Athena/HUD/EndGame/AthenaReadyUpTooltipWidget
   parent=SizeBox_ReadyUpToolTip
   widget=AthenaReadyUpTooltipWidget
   slot {class=SizeBoxSlot, properties {horizontalAlignment=HAlign_Center, padding=<ltrb 0 10 0 0>}}
  }
  SizeBox_BottomBar {
   widgetClassPath=&SizeBox
   bIsVariable=true
   parent=Overlay_74
   widget=SizeBox_BottomBar
   properties {bOverride_MinDesiredHeight=true, heightOverride=45, minDesiredHeight=45}
   slot {class=OverlaySlot, properties {horizontalAlignment=HAlign_Fill, verticalAlignment=VAlign_Bottom, padding=<ltrb 32 0 32 16>}}
  }
  BottomBarOverlay {
   widgetClassPath=&Overlay
   parent=SizeBox_BottomBar
   widget=BottomBarOverlay
   slot {class=SizeBoxSlot, properties {verticalAlignment=VAlign_Center}}
  }
  WBP_CaptureForPostBufferUpdate_1 {
   widgetClassPath=&WBP_CaptureForPostBufferUpdate
   bIsVariable=true
   parent=BottomBarOverlay
   widget=WBP_CaptureForPostBufferUpdate_1
   slot {class=OverlaySlot, properties {horizontalAlignment=HAlign_Fill, verticalAlignment=VAlign_Fill}}
  }
  HorizontalBox_LeftButtons {
   widgetClassPath=&HorizontalBox
   parent=BottomBarOverlay
   widget=HorizontalBox_LeftButtons
   slot {class=OverlaySlot, properties {horizontalAlignment=HAlign_Fill, verticalAlignment=VAlign_Bottom}}
  }
  Button_Back {
   widgetClassPath=/FortUILibrary/UIKit/Buttons/WBP_UIKit_Button_Quiet
   bInherited=true
   bIsVariable=true
   parent=HorizontalBox_LeftButtons
   widget=Button_Back
   properties {text=Back, triggeringInputAction {dataTable=&InputActionDataTable, rowName=EoMBack}}
   slot {class=HorizontalBoxSlot, properties {padding=<ltrb 0 0 32 0>}}
  }
  Button_Cancel {
   widgetClassPath=&WBP_UIKit_Button_Quiet
   bInherited=true
   bIsVariable=true
   parent=HorizontalBox_LeftButtons
   widget=Button_Cancel
   properties {text=Return to lobby, visibility=Collapsed, triggeringInputAction {dataTable=&InputActionDataTable, rowName=EoMBack}}
   slot {class=HorizontalBoxSlot, properties {padding=<ltrb 0 0 32 0>}}
  }
  HorizontalBox_CenterButtons {
   widgetClassPath=&HorizontalBox
   parent=BottomBarOverlay
   widget=HorizontalBox_CenterButtons
   slot {class=OverlaySlot, properties {horizontalAlignment=HAlign_Center}}
  }
  Button_PrevPlayer {
   widgetClassPath=&WBP_UIKit_Button_Quiet
   bInherited=true
   bIsVariable=true
   parent=HorizontalBox_CenterButtons
   widget=Button_PrevPlayer
   properties {
    text=Previous Player
    visibility=Collapsed
    triggeringInputAction {dataTable=&InputActionDataTable, rowName=PreviousPlayer}
   }
   slot {class=HorizontalBoxSlot, properties {padding=<ltrb 0 0 32 0>}}
  }
  Button_NextPlayer {
   widgetClassPath=&WBP_UIKit_Button_Quiet
   bInherited=true
   bIsVariable=true
   parent=HorizontalBox_CenterButtons
   widget=Button_NextPlayer
   properties {text=Next Player, visibility=Collapsed, triggeringInputAction {dataTable=&InputActionDataTable, rowName=NextPlayer}}
   slot {class=HorizontalBoxSlot}
  }
  Button_PrevTeam {
   widgetClassPath=&WBP_UIKit_Button_Quiet
   bInherited=true
   bIsVariable=true
   parent=HorizontalBox_CenterButtons
   widget=Button_PrevTeam
   properties {
    text=Previous Teammate
    visibility=Collapsed
    triggeringInputAction {dataTable=&InputActionDataTable, rowName=PreviousTeammate}
   }
   slot {class=HorizontalBoxSlot, properties {padding=<ltrb 0 0 32 0>}}
  }
  Button_NextTeam {
   widgetClassPath=&WBP_UIKit_Button_Quiet
   bInherited=true
   bIsVariable=true
   parent=HorizontalBox_CenterButtons
   widget=Button_NextTeam
   properties {text=Next Teammate, visibility=Collapsed, triggeringInputAction {dataTable=&InputActionDataTable, rowName=NextTeammate}}
   slot {class=HorizontalBoxSlot}
  }
  HorizontalBox_RightButtons {
   widgetClassPath=&HorizontalBox
   parent=BottomBarOverlay
   widget=HorizontalBox_RightButtons
   slot {class=OverlaySlot, properties {horizontalAlignment=HAlign_Right, verticalAlignment=VAlign_Bottom}}
  }
  VisibilitySwitcher_NextMatch {
   widgetClassPath=&CommonVisibilitySwitcher
   bInherited=true
   parent=HorizontalBox_RightButtons
   widget=VisibilitySwitcher_NextMatch
   properties {visibility=Collapsed}
   slot {class=HorizontalBoxSlot, properties {padding=<ltrb 32 0 0 0>}}
  }
  Button_NextMatch {
   widgetClassPath=&WBP_UIKit_Button_Quiet
   bInherited=true
   bIsVariable=true
   parent=VisibilitySwitcher_NextMatch
   widget=Button_NextMatch
   properties {text=Ready Up, triggeringInputAction {dataTable=&InputActionDataTable, rowName=GenericFaceButtonTop}}
   slot {class=CommonVisibilitySwitcherSlot}
  }
  Button_PlayWithNewTeam {
   widgetClassPath=&WBP_UIKit_Button_Quiet
   bInherited=true
   bIsVariable=true
   parent=VisibilitySwitcher_NextMatch
   widget=Button_PlayWithNewTeam
   properties {text=Play With New Team, triggeringInputAction {dataTable=&InputActionDataTable, rowName=GenericFaceButtonTop}}
   slot {class=CommonVisibilitySwitcherSlot}
  }
  StayTogetherContainer {
   widgetClassPath=&SizeBox
   parent=HorizontalBox_RightButtons
   widget=StayTogetherContainer
   slot {class=HorizontalBoxSlot, properties {padding=<ltrb 32 0 0 0>}}
  }
  Button_StayTogether {
   widgetClassPath=/Game/Athena/HUD/EndGame/IGM/StayTogetherButton
   bInherited=true
   parent=StayTogetherContainer
   widget=Button_StayTogether
   slot {class=SizeBoxSlot}
  }
  Button_Quests {
   widgetClassPath=&WBP_UIKit_Button_Quiet
   bIsVariable=true
   parent=HorizontalBox_RightButtons
   widget=Button_Quests
   properties {bIsFocusable=false, minWidth=120, mobileMinWidth=200, text="Map && Quests"}
   slot {class=HorizontalBoxSlot, properties {horizontalAlignment=HAlign_Right}}
  }
  KeybindMap {
   widgetClassPath=/Game/UI/InGame/KeybindWidget/GlobalActionKeybindWidget
   namedSlotHost=Button_Quests
   widget=KeybindMap
   properties {
    globalAction=ToggleFullScreenMap
    keyBindTextBorder {
     bIsDynamicallyLoaded=false
     drawAs=Box
     imageType=FullColor
     mirroring=NoMirror
     resourceName=None
     resourceObject=/Game/UI/InGame/KeybindWidget/Materials/MI_Buttonborder_DF_Black
     tiling=NoTile
     imageSize=<24 24>
     margin=<ltrb 0.25 0.25 0.25 0.25>
     outlineSettings {
      bUseBrushTransparency=false
      roundingType=HalfHeightRadius
      width=0
      color {colorUseRule=UseColor_Specified, specifiedColor=<rgba 0 0 0 0>}
      cornerRadii=<0 0 0 0>
     }
     tintColor {colorUseRule=UseColor_Specified, specifiedColor=<rgba 1 1 1 1>}
     uVRegion {bIsValid=false, max=<0 0>, min=<0 0>}
    }
   }
  }
  Button_ReportPlayer {
   widgetClassPath=&WBP_UIKit_Button_Quiet
   bIsVariable=true
   parent=HorizontalBox_RightButtons
   widget=Button_ReportPlayer
   properties {
    bIsFocusable=false
    text=Report Player
    visibility=Collapsed
    triggeringInputAction {dataTable=&InputActionDataTable, rowName=Report}
   }
   slot {class=HorizontalBoxSlot, properties {horizontalAlignment=HAlign_Right, verticalAlignment=VAlign_Bottom, padding=<ltrb 32 0 0 0>}}
  }
  Button_Spectate {
   widgetClassPath=&WBP_UIKit_Button_Quiet
   bIsVariable=true
   parent=HorizontalBox_RightButtons
   widget=Button_Spectate
   properties {
    bIsFocusable=false
    text=Spectate
    visibility=Collapsed
    triggeringInputAction {dataTable=&InputActionDataTable, rowName=GenericFaceButtonLeft}
   }
   slot {class=HorizontalBoxSlot, properties {horizontalAlignment=HAlign_Right, verticalAlignment=VAlign_Bottom, padding=<ltrb 32 0 0 0>}}
  }
  Button_AutoAdvance_Continue {
   widgetClassPath=&WBP_UIKit_Button_Quiet
   bIsVariable=true
   parent=HorizontalBox_RightButtons
   widget=Button_AutoAdvance_Continue
   properties {
    bIsFocusable=false
    text=Continue
    visibility=Collapsed
    triggeringInputAction {dataTable=&InputActionDataTable, rowName=EoMNext}
   }
   slot {class=HorizontalBoxSlot, properties {horizontalAlignment=HAlign_Right, verticalAlignment=VAlign_Bottom, padding=<ltrb 32 0 0 0>}}
  }
  WBP_AthenaBottomBar_WithNotifications {
   widgetClassPath=/Game/UI/Frontend/BottomBar/WBP_AthenaBottomBar_WithNotifications
   bIsVariable=true
   parent=HorizontalBox_RightButtons
   widget=WBP_AthenaBottomBar_WithNotifications
   slot {class=HorizontalBoxSlot, properties {horizontalAlignment=HAlign_Left, verticalAlignment=VAlign_Top, padding=<ltrb 32 0 0 0>}}
  }
  Button_Skip {
   widgetClassPath=&WBP_UIKit_Button_Quiet
   bIsVariable=true
   parent=Overlay_74
   widget=Button_Skip
   properties {
    bIsFocusable=false
    text=Skip
    visibility=Visible
    triggeringInputAction {dataTable=&InputActionDataTable, rowName=EoMNext}
   }
   slot {class=OverlaySlot, properties {horizontalAlignment=HAlign_Right, verticalAlignment=VAlign_Bottom, padding=<ltrb 32 0 32 16>}}
  }
  Overlay_TopBar {
   widgetClassPath=&Overlay
   bIsVariable=true
   parent=Overlay_0
   widget=Overlay_TopBar
   slot {class=OverlaySlot, properties {horizontalAlignment=HAlign_Fill}}
  }
  TabBackground {
   widgetClassPath=&Image
   bIsVariable=true
   parent=Overlay_TopBar
   widget=TabBackground
   properties {brush {
     bIsDynamicallyLoaded=false
     drawAs=Image
     imageType=NoImage
     mirroring=NoMirror
     resourceName=None
     resourceObject=null
     tiling=NoTile
     imageSize=<32 32>
     margin=<ltrb 0 0 0 0>
     outlineSettings {
      bUseBrushTransparency=false
      roundingType=HalfHeightRadius
      width=0
      color {colorUseRule=UseColor_Specified, specifiedColor=<rgba 0 0 0 0>}
      cornerRadii=<0 0 0 0>
     }
     tintColor {colorUseRule=UseColor_Specified, specifiedColor=<rgba 0.0033 0.0144 0.0704 1>}
     uVRegion {bIsValid=false, max=<0 0>, min=<0 0>}
    }}
   slot {class=OverlaySlot, properties {horizontalAlignment=HAlign_Fill, verticalAlignment=VAlign_Fill}}
  }
  SafeZone_TabButtons {
   widgetClassPath=&SafeZone
   parent=Overlay_TopBar
   widget=SafeZone_TabButtons
   properties {padBottom=false}
   slot {class=OverlaySlot, properties {horizontalAlignment=HAlign_Fill, padding=<ltrb 36 0 36 0>}}
  }
  Overlay_1 {widgetClassPath=&Overlay, parent=SafeZone_TabButtons, widget=Overlay_1, slot {class=SafeZoneSlot}}
  HBox_Tabs {
   widgetClassPath=&HorizontalBox
   bIsVariable=true
   parent=Overlay_1
   widget=HBox_Tabs
   slot {class=OverlaySlot, properties {verticalAlignment=VAlign_Center}}
  }
  PreviousTabVisibilityWidget {
   widgetClassPath=&CommonVisibilityWidgetLegacy
   parent=HBox_Tabs
   widget=PreviousTabVisibilityWidget
   properties {bShowForMouseAndKeyboard=false, verticalAlignment=VAlign_Center, padding=<ltrb 0 0 0 0>}
   slot {class=HorizontalBoxSlot, properties {verticalAlignment=VAlign_Center, padding=<ltrb 0 0 10 0>}}
  }
  PreviousTabAction {
   widgetClassPath=&CommonActionWidget
   bIsVariable=true
   parent=PreviousTabVisibilityWidget
   widget=PreviousTabAction
   slot {class=BorderSlot, properties {verticalAlignment=VAlign_Center, padding=<ltrb 0 0 0 0>}}
  }
  TabButtonsContainer {
   widgetClassPath=&HorizontalBox
   bInherited=true
   bIsVariable=true
   parent=HBox_Tabs
   widget=TabButtonsContainer
   slot {class=HorizontalBoxSlot, properties {horizontalAlignment=HAlign_Left, verticalAlignment=VAlign_Top}}
  }
  NextTabVisibilityWidget {
   widgetClassPath=&CommonVisibilityWidgetLegacy
   parent=HBox_Tabs
   widget=NextTabVisibilityWidget
   properties {bShowForMouseAndKeyboard=false, verticalAlignment=VAlign_Center, padding=<ltrb 0 0 0 0>}
   slot {class=HorizontalBoxSlot, properties {verticalAlignment=VAlign_Center, padding=<ltrb 10 0 0 0>}}
  }
  NextTabAction {
   widgetClassPath=&CommonActionWidget
   bIsVariable=true
   parent=NextTabVisibilityWidget
   widget=NextTabAction
   slot {class=BorderSlot, properties {verticalAlignment=VAlign_Center, padding=<ltrb 0 0 0 0>}}
  }
  Spacer_207 {widgetClassPath=&Spacer, parent=HBox_Tabs, widget=Spacer_207, slot {class=HorizontalBoxSlot}}
  HBox_SocialArea {
   widgetClassPath=&HorizontalBox
   bIsVariable=true
   parent=Overlay_1
   widget=HBox_SocialArea
   slot {class=OverlaySlot, properties {horizontalAlignment=HAlign_Right, verticalAlignment=VAlign_Center}}
  }
  SocialAvatarIcon {
   widgetClassPath=/Script/FortniteUI.FortSocialAvatarIcon
   bIsVariable=true
   parent=HBox_SocialArea
   widget=SocialAvatarIcon
   properties {
    avatarMaterial=/Game/UI/SocialPanel/Art/Materials/UI_UserPortrait_Buffet_Small
    materialTextureParamName=SmallPreview
    brush {
     bIsDynamicallyLoaded=false
     drawAs=Image
     imageType=NoImage
     mirroring=NoMirror
     resourceName=None
     resourceObject=&UI_UserPortrait_Buffet_Small
     tiling=NoTile
     imageSize=<55 55>
     margin=<ltrb 0 0 0 0>
     outlineSettings {
      bUseBrushTransparency=false
      roundingType=HalfHeightRadius
      width=0
      color {colorUseRule=UseColor_Specified, specifiedColor=<rgba 0 0 0 0>}
      cornerRadii=<0 0 0 0>
     }
     tintColor {colorUseRule=UseColor_Specified, specifiedColor=<rgba 1 1 1 1>}
     uVRegion {bIsValid=false, max=<0 0>, min=<0 0>}
    }
   }
   slot {class=HorizontalBoxSlot, properties {horizontalAlignment=HAlign_Left, verticalAlignment=VAlign_Top, padding=<ltrb 0 0 20 0>}}
  }
  S18_PlayerBattlePassXpAndReward {
   widgetClassPath=/FortniteCoreUI/UI/Progression/S18_PlayerBattlePassXpAndReward
   bInherited=true
   bIsVariable=true
   parent=HBox_SocialArea
   widget=S18_PlayerBattlePassXpAndReward
   slot {class=HorizontalBoxSlot, properties {horizontalAlignment=HAlign_Right, verticalAlignment=VAlign_Center}}
  }
  SafeZone_TextChat {
   widgetClassPath=&SafeZone
   bIsVariable=true
   parent=Overlay_0
   widget=SafeZone_TextChat
   slot {class=OverlaySlot, properties {horizontalAlignment=HAlign_Fill, verticalAlignment=VAlign_Bottom, padding=<ltrb 20 20 20 225>}}
  }
  WBP_TextChat_InputListener {
   widgetClassPath=/FortniteCoreUI/TextChat/UI/WBP_TextChat_InputListener
   bIsVariable=true
   parent=SafeZone_TextChat
   widget=WBP_TextChat_InputListener
   properties {bAllowLookWhileActive=false, mobileClosedChatVerticalAligment=VAlign_Top}
   slot {class=SafeZoneSlot}
  }
 ]
}
