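# Ranked variant of the post-game elimination list entry, a row showing a player's placement, avatar, rank icon and name.
# Initialize looks up tier display data with overrides keyed by rank type, rank, MMR and position to feed the Habanero icon, then sets the name.
# LoadAvatar async-loads the avatar texture into the portrait's dynamic material.
# PlayIntroAnimation adds a slide, fade and vertical-scale animation with a tick sound to the parent's behavior.
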WidgetBlueprint /FortniteCoreUI/Athena/HUD/EndGame/PostGameFlowScreens/DefaultScreens/XPScreen/WBP_EliminationListEntry_RankedRecap
animations [IntroAnimation {
  binding_count=1
  length_seconds=0.2956
  object=&self:IntroAnimation
  sequence {bindings [Border_ScaleAnim {
     id=#1
     possessed_class=Border
     tracks [
      {
       type=MovieScene2DTransformTrack
       sections [{
         type=MovieScene2DTransformSection
         channels [
          Translation.X {type=MovieSceneScriptingFloatChannel, keys [[0.0717,50],[5.984,0]]}
          Translation.Y {type=MovieSceneScriptingFloatChannel, const [0,1]}
         ]
        }]
      }
      {
       type=MovieSceneFloatTrack
       display_name=Render Opacity
       sections [{type=MovieSceneFloatSection, channels [None {type=MovieSceneScriptingFloatChannel, keys [[0.0717,0],[5.9837,1]]}]}]
      }
      {
       type=MovieSceneDoubleVectorTrack
       display_name=Desired Size Scale
       sections [{
         type=MovieSceneDoubleVectorSection
         channels [
          Vector.X {type=MovieSceneScriptingDoubleChannel, const [1,1]}
          Vector.Y {type=MovieSceneScriptingDoubleChannel, keys [[0.0717,0],[4,1]]}
         ]
        }]
      }
     ]
    }]}
 }]
graphs [
 EventGraph {nodes [
   n1 = LoadAvatar!() -> (OutputDelegate, then, SoftTexture:Texture 2D Soft Object Reference) {is=title K2Node_CustomEvent}
   > n2 = Async Load Asset(Asset=@n1.SoftTexture)
   n3 = Cast To Texture2D(execute=@n2.Completed, Object=@n2)
   > n4 = GetDynamicMaterial(self=Get Image_Avatar() -> (Image_Avatar:Image Object Reference))
   > SetTextureParameterValue(self=@n4, ParameterName=SmallPreview, Value=@n3)
  ]}
 Initialize {nodes [
   n1 = Initialize!() -> (then, Name:String, RankType:String, Rank:int, Current MMR:int, Current Position:int, Avatar:Texture 2D Soft Object Reference, Placement:int) {is=title K2Node_FunctionEntry}
   > SetVisibility(InVisibility=HitTestInvisible)
   > "SetText (Text)"(
    self=Get PlayerPlacement() -> (PlayerPlacement:Common Text Object Reference)
    InText=Format Text!(Format="LOCTEXT(\"##{placement}\")", placement:int=@n1.Placement) -> (Result)
   )
   n2 = String To Name(InString=Get RankType() -> (RankType:String))
   n3 = Get Current MMR() -> (Current MMR:int=0)
   n4 = Get Current Position() -> (Current Position:int=0)
   n5 = Get Rank() -> (Rank:int=0)
   n6 = Break Fort Habanero Tier!(FortHabaneroTier:"Fort Habanero Tier Structure (by ref)"=GetHabaneroDisplayDataOverride!(
     InDisplayData:Fort Habanero Display Data Object Reference=Get RankData() -> (RankData:Fort Habanero Display Data Object Reference)
     InOverrideData:Fort Compete UIHabanero Tier Override List Object Reference=Get RankInfoOverrides() -> (RankInfoOverrides:Fort Compete UIHabanero Tier Override List Object Reference)
     InKey:"Fort Compete UIHabanero Tier Override Key Structure (by ref)"=Make Fort Compete UIHabanero Tier Override Key!(RankedTrack:Name=@n2, Season:String, Rank:int=@n5, MatchmakingRating:int=@n3, CurrentPosition:int=@n4) -> (FortCompeteUIHabaneroTierOverrideKey:Fort Compete UIHabanero Tier Override Key Structure) {is=title K2Node_MakeStruct}
    ) -> (OutData:Fort Habanero Tier Override Structure, ReturnValue:bool)) -> (
    ForegroundIcon:Material Instance Soft Object Reference
    ForegroundIconSmall:Material Instance Soft Object Reference
    BackgroundIcon:Material Instance Soft Object Reference
    BackgroundIconSmall:Material Instance Soft Object Reference
    WhiteIcon:Texture 2D Soft Object Reference
    SimpleIconSmall:Texture 2D Soft Object Reference
    Name:Text
    TierType:Fort Habanero Tier Type Data Object Reference
    AdditionalLeftIcon:Material Instance Soft Object Reference
    AdditionalRightIcon:Material Instance Soft Object Reference
   ) {is=title K2Node_BreakStruct}
   > UpdateDetails!(
    execute
    self:WBP Habanero Icon Object Reference=Get Habanero_Icon() -> (Habanero_Icon:WBP Habanero Icon Object Reference)
    ForegroundIcon:Material Instance Soft Object Reference=@n6.ForegroundIconSmall
    BackgroundIcon:Material Instance Soft Object Reference=@n6.BackgroundIconSmall
   ) -> (then)
   > "SetText (Text)"(self=Get PlayerName() -> (PlayerName:Common Text Object Reference), InText="To Text (String)"(InString=Get Name() -> (Name:String)))
   > LoadAvatar!(execute, self, SoftTexture:Texture 2D Soft Object Reference=Get Avatar() -> (Avatar:Texture 2D Soft Object Reference)) -> (then) {is=title K2Node_CallFunction}
  ]}
 PlayIntroAnimation {nodes [
   PlayIntroAnimation() {is=title K2Node_FunctionEntry}
   > Parent: PlayIntroAnimation() {is=title K2Node_CallParentFunction}
   > PlayAnimation(InAnimation=Get IntroAnimation() -> (IntroAnimation:Widget Animation Object Reference))
   > PlaySound2D(Sound=/FortUIAudioContent/SoundCues/Fort_UI_PostGame_Ranked_Tick_Elim_Tracker_Cue)
  ]}
]
variables [
 RankData {type=FortHabaneroDisplayData, default=/Game/UI/Habanero/HabaneroDisplayData}
 RankInfoOverrides {type=FortCompeteUIHabaneroTierOverrideList, default=/Game/UI/Habanero/DA_RankDisplayDataOverride}
]
widgets {
 info {
  inheritedWidgetCount=0
  parentClass=/Game/Athena/HUD/EndGame/PostGameFlowScreens/DefaultScreens/XPScreen/WBP_EliminationListEntry.WBP_EliminationListEntry_C
  rootWidgetClass=&Border
  widgetCount=8
 }
 widgets [
  Border_ScaleAnim {
   widgetClassPath=&Border
   widget=Border_ScaleAnim
   properties {
    background {
     bIsDynamicallyLoaded=false
     drawAs=NoDrawType
     imageType=NoImage
     mirroring=NoMirror
     resourceName=None
     resourceObject=null
     tiling=NoTile
     imageSize=<32 32>
     margin=<ltrb 0 0 0 0>
     outlineSettings {
      bUseBrushTransparency=false
      roundingType=HalfHeightRadius
      width=0
      color {colorUseRule=UseColor_Specified, specifiedColor=<rgba 0 0 0 0>}
      cornerRadii=<0 0 0 0>
     }
     tintColor {colorUseRule=UseColor_Specified, specifiedColor=<rgba 1 1 1 1>}
     uVRegion {bIsValid=false, max=<0 0>, min=<0 0>}
    }
    padding=<ltrb 0 4 0 4>
   }
  }
  HorizontalBox_87 {
   widgetClassPath=&HorizontalBox
   parent=Border_ScaleAnim
   widget=HorizontalBox_87
   slot {class=BorderSlot, properties {padding=<ltrb 0 4 0 4>}}
  }
  SizeBox {
   widgetClassPath=&SizeBox
   parent=HorizontalBox_87
   widget=SizeBox
   properties {bOverride_WidthOverride=true, widthOverride=90}
   slot {class=HorizontalBoxSlot, properties {verticalAlignment=VAlign_Center}}
  }
  PlayerPlacement {
   widgetClassPath=&CommonTextBlock
   bIsVariable=true
   parent=SizeBox
   widget=PlayerPlacement
   properties {
    style=/FortUILibrary/Styles/Typography/Foundation/TextStyle_BurbankSmall_S_White.TextStyle_BurbankSmall_S_White_C
    text="##3"
    font {
     bForceMonospaced=false
     bMaterialIsStencil=false
     fontMaterial=null
     fontObject=&BurbankSmall
     letterSpacing=0
     monospacedWidth=1
     size=18
     skewAmount=0
     typefaceFontName=Black
     outlineSettings {
      bApplyOutlineToDropShadows=false
      bMiteredCorners=false
      bSeparateFillAlpha=false
      outlineMaterial=null
      outlineSize=0
      outlineColor=<rgba 0 0 0 1>
     }
    }
    shadowOffset=<0 0>
   }
   slot {class=SizeBoxSlot, properties {horizontalAlignment=HAlign_Center}}
  }
  Image_Avatar {
   widgetClassPath=&Image
   bIsVariable=true
   parent=HorizontalBox_87
   widget=Image_Avatar
   properties {brush {
     bIsDynamicallyLoaded=false
     drawAs=Image
     imageType=NoImage
     mirroring=NoMirror
     resourceName=None
     resourceObject=/Game/UI/SocialPanel/Art/Materials/UI_UserPortrait_Buffet_Small
     tiling=NoTile
     imageSize=<56 56>
     margin=<ltrb 0 0 0 0>
     outlineSettings {
      bUseBrushTransparency=false
      roundingType=HalfHeightRadius
      width=0
      color {colorUseRule=UseColor_Specified, specifiedColor=<rgba 0 0 0 0>}
      cornerRadii=<0 0 0 0>
     }
     tintColor {colorUseRule=UseColor_Specified, specifiedColor=<rgba 1 1 1 1>}
     uVRegion {bIsValid=false, max=<0 0>, min=<0 0>}
    }}
   slot {class=HorizontalBoxSlot}
  }
  Spacer_37 {widgetClassPath=&Spacer, parent=HorizontalBox_87, widget=Spacer_37, properties {size=<15 1>}, slot {class=HorizontalBoxSlot}}
  Habanero_Icon {
   widgetClassPath=/Game/UI/Frontend/Lobby/Habanero/WBP_Habanero_Icon
   bIsVariable=true
   parent=HorizontalBox_87
   widget=Habanero_Icon
   properties {
    debugBackground=/Game/UI/Frontend/Lobby/Habanero/Habanero_Materials/MaterialInstances/Icons/Background/T02/MI_UI_Habanero_Background_01
    debugForeground=/Game/UI/Frontend/Lobby/Habanero/Habanero_Materials/MaterialInstances/Icons/Foreground/T01/MI_UI_Habanero_Foreground_T01_01
    iconMinSize=<50 50>
   }
   slot {class=HorizontalBoxSlot, properties {verticalAlignment=VAlign_Center}}
  }
  PlayerName {
   widgetClassPath=&CommonTextBlock
   bIsVariable=true
   parent=HorizontalBox_87
   widget=PlayerName
   properties {
    style=&TextStyle_BurbankSmall_S_White_C
    text=playername
    font {
     bForceMonospaced=false
     bMaterialIsStencil=false
     fontMaterial=null
     fontObject=&BurbankSmall
     letterSpacing=0
     monospacedWidth=1
     size=18
     skewAmount=0
     typefaceFontName=Black
     outlineSettings {
      bApplyOutlineToDropShadows=false
      bMiteredCorners=false
      bSeparateFillAlpha=false
      outlineMaterial=null
      outlineSize=0
      outlineColor=<rgba 0 0 0 1>
     }
    }
    shadowOffset=<0 0>
   }
   slot {class=HorizontalBoxSlot, properties {verticalAlignment=VAlign_Center}}
  }
 ]
}
