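# Survival accolade variant of WBP_EliminationListEntry for the post-game XP screen, showing a title such as a Survivor rank over a tinted flare, glow and mirrored triangle-spiral decorations.
# PlayIntroAnimation calls the parent, then plays IntroAnimation and the ranked survivor tick sound.
# IntroAnimation drives the root border's vertical desired-size scale, opacity and X translation, so the entry expands, fades and slides in.
# SetText feeds the supplied text to the uppercase title block.
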WidgetBlueprint /FortniteCoreUI/Athena/HUD/EndGame/PostGameFlowScreens/DefaultScreens/XPScreen/WBP_EliminationListEntry_SurvivalAccolade
animations [IntroAnimation {
  binding_count=1
  length_seconds=0.2956
  object=&self:IntroAnimation
  sequence {bindings [Border_ScaleAnim {
     id=#1
     possessed_class=Border
     tracks [
      {
       type=MovieSceneDoubleVectorTrack
       display_name=Desired Size Scale
       sections [{
         type=MovieSceneDoubleVectorSection
         channels [
          Vector.X {type=MovieSceneScriptingDoubleChannel, const [1,1]}
          Vector.Y {type=MovieSceneScriptingDoubleChannel, keys [[0.0717,0],[4,1]]}
         ]
        }]
      }
      {
       type=MovieSceneFloatTrack
       display_name=Render Opacity
       sections [{type=MovieSceneFloatSection, channels [None {type=MovieSceneScriptingFloatChannel, keys [[0.0717,0],[5.9837,1]]}]}]
      }
      {
       type=MovieScene2DTransformTrack
       sections [{
         type=MovieScene2DTransformSection
         channels [
          Translation.X {type=MovieSceneScriptingFloatChannel, keys [[0.0717,50],[5.984,0]]}
          Translation.Y {type=MovieSceneScriptingFloatChannel, const [0,1]}
         ]
        }]
      }
     ]
    }]}
 }]
graphs [
 EventGraph {nodes [
   Event PreConstruct()
   Event Construct()
   Event Tick()
  ]}
 PlayIntroAnimation {nodes [
   PlayIntroAnimation() {is=title K2Node_FunctionEntry}
   > Parent: PlayIntroAnimation() {is=title K2Node_CallParentFunction}
   > PlayAnimation(InAnimation=Get IntroAnimation() -> (IntroAnimation:Widget Animation Object Reference))
   > PlaySound2D(Sound=/FortUIAudioContent/SoundCues/Fort_UI_PostGame_Ranked_Tick_Survivor_Cue)
  ]}
 SetText {nodes [
   n1 = SetText!() -> (then, InText:Text) {is=title K2Node_FunctionEntry}
   > "SetText (Text)"(self=Get Text_Title() -> (Text_Title:Common Text Object Reference), InText=@n1)
  ]}
]
variables []
widgets {
 info {
  inheritedWidgetCount=0
  parentClass=/Game/Athena/HUD/EndGame/PostGameFlowScreens/DefaultScreens/XPScreen/WBP_EliminationListEntry.WBP_EliminationListEntry_C
  rootWidgetClass=&Border
  widgetCount=10
 }
 widgets [
  Border_ScaleAnim {
   widgetClassPath=&Border
   widget=Border_ScaleAnim
   properties {
    background {
     bIsDynamicallyLoaded=false
     drawAs=NoDrawType
     imageType=NoImage
     mirroring=NoMirror
     resourceName=None
     resourceObject=null
     tiling=NoTile
     imageSize=<32 32>
     margin=<ltrb 0 0 0 0>
     outlineSettings {
      bUseBrushTransparency=false
      roundingType=HalfHeightRadius
      width=0
      color {colorUseRule=UseColor_Specified, specifiedColor=<rgba 0 0 0 0>}
      cornerRadii=<0 0 0 0>
     }
     tintColor {colorUseRule=UseColor_Specified, specifiedColor=<rgba 1 1 1 1>}
     uVRegion {bIsValid=false, max=<0 0>, min=<0 0>}
    }
    padding=<ltrb 0 0 0 0>
   }
  }
  Overlay_Title {
   widgetClassPath=&Overlay
   parent=Border_ScaleAnim
   widget=Overlay_Title
   slot {class=BorderSlot, properties {padding=<ltrb 0 0 0 0>}}
  }
  FortVisualAttachment_0 {
   widgetClassPath=&FortVisualAttachment
   parent=Overlay_Title
   widget=FortVisualAttachment_0
   properties {contentAnchor=<0.5 0.5>}
   slot {class=OverlaySlot, properties {horizontalAlignment=HAlign_Center, verticalAlignment=VAlign_Center}}
  }
  SizeBoxTriangles {
   widgetClassPath=&SizeBox
   parent=FortVisualAttachment_0
   widget=SizeBoxTriangles
   properties {bOverride_HeightOverride=true, bOverride_WidthOverride=true, heightOverride=50, widthOverride=120}
   slot {class=SizeBoxSlot}
  }
  HorizontalBox_0 {widgetClassPath=&HorizontalBox, parent=SizeBoxTriangles, widget=HorizontalBox_0, slot {class=SizeBoxSlot}}
  ImageTriangleL {
   widgetClassPath=&Image
   parent=HorizontalBox_0
   widget=ImageTriangleL
   properties {
    brush {
     bIsDynamicallyLoaded=false
     drawAs=Image
     imageType=NoImage
     mirroring=NoMirror
     resourceName=None
     resourceObject=/Game/UI/HUD/Materials/M_FX_TriangleSpiral
     tiling=NoTile
     imageSize=<125 85>
     margin=<ltrb 0 0 0 0>
     outlineSettings {
      bUseBrushTransparency=false
      roundingType=HalfHeightRadius
      width=0
      color {colorUseRule=UseColor_Specified, specifiedColor=<rgba 0 0 0 0>}
      cornerRadii=<0 0 0 0>
     }
     tintColor {colorUseRule=UseColor_Specified, specifiedColor=<rgba 0.9427 1 0 1>}
     uVRegion {bIsValid=false, max=<0 0>, min=<0 0>}
    }
    renderTransform {angle=0, scale=<-1 1>, shear=<0 0>, translation=<0 0>}
    renderTransformPivot=<0.5 1>
   }
   slot {class=HorizontalBoxSlot, properties {verticalAlignment=VAlign_Center, padding=<ltrb -200 0 0 0>, size {sizeRule=Fill, value=0.5}}}
  }
  ImageTriangleR {
   widgetClassPath=&Image
   parent=HorizontalBox_0
   widget=ImageTriangleR
   properties {
    brush {
     bIsDynamicallyLoaded=false
     drawAs=Image
     imageType=NoImage
     mirroring=NoMirror
     resourceName=None
     resourceObject=&M_FX_TriangleSpiral
     tiling=NoTile
     imageSize=<125 85>
     margin=<ltrb 0 0 0 0>
     outlineSettings {
      bUseBrushTransparency=false
      roundingType=HalfHeightRadius
      width=0
      color {colorUseRule=UseColor_Specified, specifiedColor=<rgba 0 0 0 0>}
      cornerRadii=<0 0 0 0>
     }
     tintColor {colorUseRule=UseColor_Specified, specifiedColor=<rgba 0.9427 1 0 1>}
     uVRegion {bIsValid=false, max=<0 0>, min=<0 0>}
    }
    renderTransformPivot=<0.5 1>
   }
   slot {class=HorizontalBoxSlot, properties {verticalAlignment=VAlign_Center, padding=<ltrb 0 0 -200 0>, size {sizeRule=Fill, value=0.5}}}
  }
  Image_TitleBG {
   widgetClassPath=&Image
   parent=Overlay_Title
   widget=Image_TitleBG
   properties {
    renderOpacity=0.8
    brush {
     bIsDynamicallyLoaded=false
     drawAs=Image
     imageType=NoImage
     mirroring=NoMirror
     resourceName=None
     resourceObject=/Game/Athena/HUD/EndGame/PostGameFlowScreens/DefaultScreens/XPScreen/Art/M_UI_AnamorphicFlare_RankedSurvivalAccolade
     tiling=NoTile
     imageSize=<32 32>
     margin=<ltrb 0 0 0 0>
     outlineSettings {
      bUseBrushTransparency=false
      roundingType=HalfHeightRadius
      width=0
      color {colorUseRule=UseColor_Specified, specifiedColor=<rgba 0 0 0 0>}
      cornerRadii=<0 0 0 0>
     }
     tintColor {colorUseRule=UseColor_Specified, specifiedColor=<rgba 1 1 1 1>}
     uVRegion {bIsValid=false, max=<0 0>, min=<0 0>}
    }
    colorAndOpacity=<rgba 1 0.4 0 1>
    renderTransform {angle=0, scale=<1 1>, shear=<0 0>, translation=<32 0>}
   }
   slot {class=OverlaySlot, properties {horizontalAlignment=HAlign_Fill, verticalAlignment=VAlign_Center, padding=<ltrb -125 0 -125 0>}}
  }
  Image_Glow {
   widgetClassPath=&Image
   parent=Overlay_Title
   widget=Image_Glow
   properties {
    renderOpacity=0.05
    brush {
     bIsDynamicallyLoaded=false
     drawAs=Image
     imageType=NoImage
     mirroring=NoMirror
     resourceName=None
     resourceObject=/Game/Athena/HUD/LTM/Wax/Materials/M_UI_TextGlow_Add
     tiling=NoTile
     imageSize=<32 32>
     margin=<ltrb 0 0 0 0>
     outlineSettings {
      bUseBrushTransparency=false
      roundingType=HalfHeightRadius
      width=0
      color {colorUseRule=UseColor_Specified, specifiedColor=<rgba 0 0 0 0>}
      cornerRadii=<0 0 0 0>
     }
     tintColor {colorUseRule=UseColor_Specified, specifiedColor=<rgba 1 0.4 0 1>}
     uVRegion {bIsValid=false, max=<0 0>, min=<0 0>}
    }
   }
   slot {class=OverlaySlot, properties {horizontalAlignment=HAlign_Fill, verticalAlignment=VAlign_Fill, padding=<ltrb -25 -10 -25 -10>}}
  }
  Text_Title {
   widgetClassPath=&CommonTextBlock
   bIsVariable=true
   parent=Overlay_Title
   widget=Text_Title
   properties {
    mobileFontSizeMultiplier=1.05
    text=Survivor III - Top 10 players
    textTransformPolicy=ToUpper
    font {
     bForceMonospaced=false
     bMaterialIsStencil=false
     fontMaterial=null
     fontObject=&BurbankSmall
     letterSpacing=0
     monospacedWidth=1
     size=18
     skewAmount=0
     typefaceFontName=Bold
     outlineSettings {
      bApplyOutlineToDropShadows=false
      bMiteredCorners=false
      bSeparateFillAlpha=false
      outlineMaterial=null
      outlineSize=0
      outlineColor=<rgba 0 0 0 1>
     }
    }
    shadowColorAndOpacity=<rgba 0.0052 0.0052 0.0056 0.8>
    shadowOffset=<1 2>
   }
   slot {class=OverlaySlot, properties {horizontalAlignment=HAlign_Center, verticalAlignment=VAlign_Center, padding=<ltrb 0 5 0 5>}}
  }
 ]
}
