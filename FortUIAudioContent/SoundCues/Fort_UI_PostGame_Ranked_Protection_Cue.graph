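# Sound cue that plays the post-game ranked rank protection stinger in the UI.
# A single wave player plays the stinger wave directly through the SFX sound class, with a slightly boosted volume and virtualization disabled.
# Overridden concurrency limits simultaneous copies and stops the oldest when exceeded.
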SoundCue /FortUIAudioContent/SoundCues/Fort_UI_PostGame_Ranked_Protection_Cue
root=WavePlayer_0
nodes [{id=WavePlayer_0, properties {soundWaveAssetPtr=/FortUIAudioContent/Waves/UI_Ranked_Post_Rank_Protection_Stinger_01}}]
properties {
 bOverrideConcurrency=true
 maxDistance=2097152
 soundClassObject=&SFX
 virtualizationMode=Disabled
 volumeMultiplier=1.15
 concurrencyOverrides {maxCount=2, resolutionRule=StopOldest}
}
