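# Sound cue for the whoosh in a Ranked UI flash transition.
# A single wave player feeds the cue root directly with the Ranked close-whoosh wave, with no randomization, mixing or modulation.
# The cue routes to the SFX sound class and overrides concurrency to allow one instance, stopping the oldest so retriggers do not stack.
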SoundCue /FortUIAudioContent/SoundCues/Fortnite_UI_Ranked_Flash_Transition_Whoosh_Cue
root=WavePlayer_0
nodes [{id=WavePlayer_0, properties {soundWaveAssetPtr=/Game/Sounds/Fort_UI_Sounds/FrontEnd/New_FrontEnd/Ranked/wav/UI_Ranked_CloseWhoosh}}]
properties {
 bOverrideConcurrency=true
 maxDistance=2097152
 soundClassObject=&SFX
 volumeMultiplier=1
 concurrencyOverrides {maxCount=1, resolutionRule=StopOldest}
}
