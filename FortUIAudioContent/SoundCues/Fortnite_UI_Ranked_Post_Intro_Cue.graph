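# Sound cue for a ranked-mode UI intro in the front end, judging by its name and wave paths.
# A Mixer layers a menu intro whoosh with a "modes unlocked" intro wave at equal volume, feeding a Modulator left at default settings as the root.
# It plays in the SFX class and overrides concurrency to allow one instance, stopping the oldest, with a retrigger time set.
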SoundCue /FortUIAudioContent/SoundCues/Fortnite_UI_Ranked_Post_Intro_Cue
root=Modulator_0
nodes [
 {id=Mixer_0, children [WavePlayer_0,WavePlayer_1], properties {inputVolume [1,1]}}
 {id=Modulator_0, children [Mixer_0]}
 {id=WavePlayer_0, properties {soundWaveAssetPtr=/Game/Sounds/Fort_UI_Sounds/FrontEnd/New_FrontEnd/Ranked/wav/UI_Ranked_Menu_Intro_Whoosh}}
 {id=WavePlayer_1, properties {soundWaveAssetPtr=/Game/Sounds/Fort_UI_Sounds/FrontEnd/New_FrontEnd/Ranked/wav/UI_Ranked_ModesUnlockedIntro}}
]
properties {
 bOverrideConcurrency=true
 maxDistance=2097152
 soundClassObject=&SFX
 volumeMultiplier=1
 concurrencyOverrides {maxCount=1, resolutionRule=StopOldest, retriggerTime=2}
}
