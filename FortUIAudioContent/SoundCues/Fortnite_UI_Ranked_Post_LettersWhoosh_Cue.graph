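# Sound cue for a UI whoosh effect, named for the letters whoosh in the Ranked post screen.
# A single wave player feeds the UI_Ranked_Post_LettersWhoosh wave through a root Modulator node left at default settings.
# The cue uses the SFX sound class with a slight volume boost and a very large max distance.
# Its concurrency override allows one instance at a time, stopping the oldest when a new one starts.
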SoundCue /FortUIAudioContent/SoundCues/Fortnite_UI_Ranked_Post_LettersWhoosh_Cue
root=Modulator_0
nodes [
 {id=Modulator_0, children [WavePlayer_1]}
 {id=WavePlayer_1, properties {soundWaveAssetPtr=/FortUIAudioContent/Waves/UI_Ranked_Post_LettersWhoosh}}
]
properties {
 bOverrideConcurrency=true
 maxDistance=2097152
 soundClassObject=&SFX
 volumeMultiplier=1.2
 concurrencyOverrides {maxCount=1, resolutionRule=StopOldest}
}
