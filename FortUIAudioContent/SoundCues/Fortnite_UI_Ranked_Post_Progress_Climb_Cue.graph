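# Sound cue for the ranked post-match UI, named for the progress climb.
# A single wave player feeds the root with the ranked progress fill wave, routed to the SFX sound class with a slight volume boost.
# Its own concurrency override allows one instance at a time, stopping the oldest, with a short retrigger time.
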SoundCue /FortUIAudioContent/SoundCues/Fortnite_UI_Ranked_Post_Progress_Climb_Cue
root=WavePlayer_1
nodes [{id=WavePlayer_1, properties {soundWaveAssetPtr=/FortUIAudioContent/Waves/UI_Ranked_Post_Progress_Fill_lp}}]
properties {
 bOverrideConcurrency=true
 maxDistance=2097152
 soundClassObject=&SFX
 volumeMultiplier=1.2
 concurrencyOverrides {maxCount=1, resolutionRule=StopOldest, retriggerTime=0.2}
}
