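# Sound cue for the ranked post-match progress end whoosh in the Fortnite UI, as its name indicates.
# A Mixer layers two wave players, the ranked menu intro whoosh and the display toggle whoosh, at equal input volume.
# It plays on the SFX sound class with a concurrency override that allows one instance and stops the oldest.
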SoundCue /FortUIAudioContent/SoundCues/Fortnite_UI_Ranked_Post_Progress_EndWhoosh_Cue
root=Mixer_0
nodes [
 {id=Mixer_0, children [WavePlayer_1,WavePlayer_2], properties {inputVolume [1,1]}}
 {id=WavePlayer_1, properties {soundWaveAssetPtr=/Game/Sounds/Fort_UI_Sounds/FrontEnd/New_FrontEnd/Ranked/wav/UI_Ranked_Menu_Intro_Whoosh}}
 {id=WavePlayer_2, properties {soundWaveAssetPtr=/Game/Sounds/Fort_UI_Sounds/FrontEnd/New_FrontEnd/Ranked/wav/UI_Ranked_DisplayToggleWhoosh}}
]
properties {
 bOverrideConcurrency=true
 maxDistance=2097152
 soundClassObject=&SFX
 volumeMultiplier=1
 concurrencyOverrides {maxCount=1, resolutionRule=StopOldest}
}
