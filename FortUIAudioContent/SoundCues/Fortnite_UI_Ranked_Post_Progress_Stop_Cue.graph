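# Sound cue for the ranked post-progress UI, playing the fill-end sound that marks the progress stopping.
# A single wave player feeds a Modulator left at default settings, at slightly reduced cue volume in the SFX sound class.
# A concurrency override allows one instance at a time, stopping the oldest when a new one starts.
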SoundCue /FortUIAudioContent/SoundCues/Fortnite_UI_Ranked_Post_Progress_Stop_Cue
root=Modulator_0
nodes [
 {id=Modulator_0, children [WavePlayer_0]}
 {id=WavePlayer_0, properties {soundWaveAssetPtr=/FortUIAudioContent/Waves/UI_Ranked_Post_Progress_Fill_End}}
]
properties {
 bOverrideConcurrency=true
 maxDistance=2097152
 soundClassObject=&SFX
 volumeMultiplier=0.8
 concurrencyOverrides {maxCount=1, resolutionRule=StopOldest}
}
