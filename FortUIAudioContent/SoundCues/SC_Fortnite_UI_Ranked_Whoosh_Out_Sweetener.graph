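# SoundCue for the Fortnite Ranked UI that plays a whoosh-out sweetener sound.
# A single wave player node feeds the UI_Ranked_Whoosh_Out_Sweetener wave straight to the output under the SFX sound class.
# A concurrency override allows one instance at a time, so a new play stops the oldest.
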SoundCue /FortUIAudioContent/SoundCues/SC_Fortnite_UI_Ranked_Whoosh_Out_Sweetener
root=WavePlayer_0
nodes [{id=WavePlayer_0, properties {soundWaveAssetPtr=/FortUIAudioContent/Waves/UI_Ranked_Whoosh_Out_Sweetener}}]
properties {
 bOverrideConcurrency=true
 maxDistance=2097152
 soundClassObject=&SFX
 volumeMultiplier=1
 concurrencyOverrides {maxCount=1, resolutionRule=StopOldest}
}
