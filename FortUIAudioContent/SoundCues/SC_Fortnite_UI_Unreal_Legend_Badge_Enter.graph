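# Sound cue for a Fortnite UI sound, named for the entrance of the Unreal Legend badge.
# A single wave player feeds a modulator whose pitch and volume ranges are pinned at unity, so playback has no random variation.
# The cue uses the SFX sound class with a boosted volume multiplier.
# A concurrency override allows one instance at a time, stopping the oldest when it retriggers.
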SoundCue /FortUIAudioContent/SoundCues/SC_Fortnite_UI_Unreal_Legend_Badge_Enter
root=Modulator_0
nodes [
 {id=Modulator_0, children [WavePlayer_1], properties {pitchMax=1, pitchMin=1, volumeMax=1, volumeMin=1}}
 {id=WavePlayer_1, properties {soundWaveAssetPtr=/FortUIAudioContent/Waves/UI_Unreal_Legend_Badge_Enter_2}}
]
properties {
 bOverrideConcurrency=true
 maxDistance=2097152
 soundClassObject=&SFX
 volumeMultiplier=1.6
 concurrencyOverrides {maxCount=1, resolutionRule=StopOldest}
}
