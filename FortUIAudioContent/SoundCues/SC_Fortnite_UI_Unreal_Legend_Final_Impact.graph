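# Sound cue for the Unreal Legend final impact UI sound in Fortnite.
# A single wave player feeds the output directly with no randomization or modulation, boosted in volume and routed to the SFX sound class.
# Overridden concurrency allows one instance at a time, stopping the oldest when a new one starts.
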SoundCue /FortUIAudioContent/SoundCues/SC_Fortnite_UI_Unreal_Legend_Final_Impact
root=WavePlayer_1
nodes [{id=WavePlayer_1, properties {soundWaveAssetPtr=/FortUIAudioContent/Waves/UI_Unreal_Legend_Final_Impact}}]
properties {
 bOverrideConcurrency=true
 maxDistance=2097152
 soundClassObject=&SFX
 volumeMultiplier=1.7
 concurrencyOverrides {maxCount=1, resolutionRule=StopOldest}
}
