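# Timeline entry widget for the post-game XP screen that marks either an elimination or a storm circle, with a rank icon for eliminations.
# Initialize uses IsStormCircle to switch between the red elimination bar and purple storm bar, hiding the Habanero rank icon for storm entries.
# For eliminations, it builds a key from rank type, rank, MMR and position, resolves it through rank display data and overrides, and feeds the small tier icons to the icon widget.
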WidgetBlueprint /Game/Athena/HUD/EndGame/PostGameFlowScreens/DefaultScreens/XPScreen/WBP_EliminationTimelineEntry
graphs [
 EventGraph {nodes [
   Event PreConstruct()
   Event Construct()
   Event Tick()
  ]}
 Initialize {nodes [
   n1 = Initialize!() -> (then, Name:String, RankType:String, Rank:int, Current MMR:int, Current Position:int, IsStormCircle:bool=~) {is=title K2Node_FunctionEntry}
   > SetActiveWidgetIndex(
    self:Common Visibility Switcher Object Reference=Get Switcher_Marker() -> (Switcher_Marker:Common Visibility Switcher Object Reference)
    Index=SelectInt(A=1, B=0, bPickA=@n1.IsStormCircle)
   )
   > n2 = Branch(Condition=Get IsStormCircle() -> (IsStormCircle:bool))
   n3 = Get Habanero_Icon() -> (Habanero_Icon:WBP Habanero Icon Object Reference)
   > SetVisibility(self=@n3, InVisibility=Hidden)
   SetVisibility(execute=@n2.else, self=@n3, InVisibility=HitTestInvisible)
   n4 = String To Name(InString=Get RankType() -> (RankType:String))
   n5 = Get Rank() -> (Rank:int=0)
   n6 = Get Current MMR() -> (Current MMR:int=0)
   n7 = Get Current Position() -> (Current Position:int=0)
   n8 = Break Fort Habanero Tier!(FortHabaneroTier:"Fort Habanero Tier Structure (by ref)"=GetHabaneroDisplayDataOverride!(
     InDisplayData:Fort Habanero Display Data Object Reference=Get RankData() -> (RankData:Fort Habanero Display Data Object Reference)
     InOverrideData:Fort Compete UIHabanero Tier Override List Object Reference=Get Rank Info Overrides() -> (Rank Info Overrides:Fort Compete UIHabanero Tier Override List Object Reference)
     InKey:"Fort Compete UIHabanero Tier Override Key Structure (by ref)"=Make Fort Compete UIHabanero Tier Override Key!(RankedTrack:Name=@n4, Season:String, Rank:int=@n5, MatchmakingRating:int=@n6, CurrentPosition:int=@n7) -> (FortCompeteUIHabaneroTierOverrideKey:Fort Compete UIHabanero Tier Override Key Structure) {is=title K2Node_MakeStruct}
    ) -> (OutData:Fort Habanero Tier Override Structure, ReturnValue:bool)) -> (
    ForegroundIcon:Material Instance Soft Object Reference
    ForegroundIconSmall:Material Instance Soft Object Reference
    BackgroundIcon:Material Instance Soft Object Reference
    BackgroundIconSmall:Material Instance Soft Object Reference
    WhiteIcon:Texture 2D Soft Object Reference
    SimpleIconSmall:Texture 2D Soft Object Reference
    Name:Text
    TierType:Fort Habanero Tier Type Data Object Reference
    AdditionalLeftIcon:Material Instance Soft Object Reference
    AdditionalRightIcon:Material Instance Soft Object Reference
   ) {is=title K2Node_BreakStruct}
   > UpdateDetails!(
    execute
    self:WBP Habanero Icon Object Reference=Get Habanero_Icon()
    ForegroundIcon:Material Instance Soft Object Reference=@n8.ForegroundIconSmall
    BackgroundIcon:Material Instance Soft Object Reference=@n8.BackgroundIconSmall
   ) -> (then)
  ]}
]
variables [
 Rank Info Overrides {type=FortCompeteUIHabaneroTierOverrideList, default=/Game/UI/Habanero/DA_RankDisplayDataOverride}
 RankData {type=FortHabaneroDisplayData, default=/Game/UI/Habanero/HabaneroDisplayData}
]
widgets {
 info {inheritedWidgetCount=0, parentClass=&UserWidget, rootWidgetClass=&Overlay, widgetCount=7}
 widgets [
  Overlay_0 {widgetClassPath=&Overlay, widget=Overlay_0}
  SizeBox_1 {
   widgetClassPath=&SizeBox
   parent=Overlay_0
   widget=SizeBox_1
   properties {
    bOverride_HeightOverride=true
    bOverride_WidthOverride=true
    heightOverride=50
    widthOverride=5
    renderTransform {angle=0, scale=<1 1>, shear=<0 0>, translation=<0 32>}
   }
   slot {class=OverlaySlot, properties {horizontalAlignment=HAlign_Center, verticalAlignment=VAlign_Center}}
  }
  Switcher_Marker {
   widgetClassPath=&CommonVisibilitySwitcher
   bIsVariable=true
   parent=SizeBox_1
   widget=Switcher_Marker
   slot {class=SizeBoxSlot}
  }
  Image_EliminationMarker {
   widgetClassPath=&Image
   bIsVariable=true
   parent=Switcher_Marker
   widget=Image_EliminationMarker
   properties {brush {
     bIsDynamicallyLoaded=false
     drawAs=Image
     imageType=NoImage
     mirroring=NoMirror
     resourceName=None
     resourceObject=null
     tiling=NoTile
     imageSize=<32 32>
     margin=<ltrb 0 0 0 0>
     outlineSettings {
      bUseBrushTransparency=false
      roundingType=HalfHeightRadius
      width=0
      color {colorUseRule=UseColor_Specified, specifiedColor=<rgba 0 0 0 0>}
      cornerRadii=<0 0 0 0>
     }
     tintColor {colorUseRule=UseColor_Specified, specifiedColor=<rgba 0.7031 0.1501 0.1619 1>}
     uVRegion {bIsValid=false, max=<0 0>, min=<0 0>}
    }}
   slot {class=CommonVisibilitySwitcherSlot}
  }
  Image_StormMarker {
   widgetClassPath=&Image
   bIsVariable=true
   parent=Switcher_Marker
   widget=Image_StormMarker
   properties {brush {
     bIsDynamicallyLoaded=false
     drawAs=Image
     imageType=NoImage
     mirroring=NoMirror
     resourceName=None
     resourceObject=null
     tiling=NoTile
     imageSize=<32 32>
     margin=<ltrb 0 0 0 0>
     outlineSettings {
      bUseBrushTransparency=false
      roundingType=HalfHeightRadius
      width=0
      color {colorUseRule=UseColor_Specified, specifiedColor=<rgba 0 0 0 0>}
      cornerRadii=<0 0 0 0>
     }
     tintColor {colorUseRule=UseColor_Specified, specifiedColor=<rgba 0.4314 0.1012 0.5885 1>}
     uVRegion {bIsValid=false, max=<0 0>, min=<0 0>}
    }}
   slot {class=CommonVisibilitySwitcherSlot}
  }
  SizeBox {
   widgetClassPath=&SizeBox
   parent=Overlay_0
   widget=SizeBox
   properties {bOverride_HeightOverride=true, bOverride_WidthOverride=true, heightOverride=32, widthOverride=32}
   slot {class=OverlaySlot, properties {horizontalAlignment=HAlign_Center}}
  }
  Habanero_Icon {
   widgetClassPath=/Game/UI/Frontend/Lobby/Habanero/WBP_Habanero_Icon
   bIsVariable=true
   parent=SizeBox
   widget=Habanero_Icon
   properties {
    debugBackground=/Game/UI/Frontend/Lobby/Habanero/Habanero_Materials/MaterialInstances/Icons/Background/T02/MI_UI_Habanero_Background_01
    debugForeground=/Game/UI/Frontend/Lobby/Habanero/Habanero_Materials/MaterialInstances/Icons/Foreground/T01/MI_UI_Habanero_Foreground_T01_01
    iconMinSize=<64 64>
    renderTransform {angle=0, scale=<1 1>, shear=<0 0>, translation=<0 3.5>}
   }
   slot {class=SizeBoxSlot, properties {horizontalAlignment=HAlign_Center, verticalAlignment=VAlign_Center}}
  }
 ]
}
