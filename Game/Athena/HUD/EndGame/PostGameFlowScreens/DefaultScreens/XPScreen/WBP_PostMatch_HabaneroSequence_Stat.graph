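# Post-match stat widget for the Habanero sequence on the XP screen, showing a centered number above a dimmed uppercase title.
# SetStat writes an integer to the value text, optionally prefixed with a hash sign through a bool.
# SetStatTitle stores and displays the title.
# PreConstruct applies the stored title and wraps it at a fixed width on mobile only.
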WidgetBlueprint /Game/Athena/HUD/EndGame/PostGameFlowScreens/DefaultScreens/XPScreen/WBP_PostMatch_HabaneroSequence_Stat
graphs [
 EventGraph {nodes [
   Event PreConstruct()
   > "SetText (Text)"(self=Get Text_Title() -> (Text_Title:Common Text Object Reference), InText=Get Title() -> (Title:Text))
   > SetWrapTextWidth!(execute, self:Common Text Object Reference=Get Text_Title(), InWrapTextAt:int=SelectInt(A=220, B=0, bPickA=IsMobileGame())) -> (then)
  ]}
 SetStat {nodes [
   n1 = SetStat!() -> (then, Value:int, "Show##ForValue":bool=~) {is=title K2Node_FunctionEntry}
   n2 = "To Text (Integer)"(Value=@n1)
   > "SetText (Text)"(
    self=Get Text_Stat() -> (Text_Stat:Common Text Object Reference)
    InText=Select(Option 0:Text=@n2, Option 1:Text=Format Text!(Format="LOCTEXT(\"##{Value}\")", Value:Text=@n2) -> (Result), Index=@n1."Show##ForValue") -> (ReturnValue:Text)
   )
  ]}
 SetStatTitle {nodes [
   n1 = SetStatTitle!() -> (then, StatTitle:Text) {is=title K2Node_FunctionEntry}
   > n2 = Set Title(Title:Text=@n1) -> (Output_Get:Text)
   > "SetText (Text)"(self=Get Text_Title(), InText=@n2)
  ]}
]
variables [Title {type=string, default="{Stat}"}]
widgets {
 info {inheritedWidgetCount=0, parentClass=&CommonUserWidget, rootWidgetClass=&Overlay, widgetCount=4}
 widgets [
  Overlay {widgetClassPath=&Overlay, widget=Overlay}
  VerticalBox_0 {
   widgetClassPath=&VerticalBox
   parent=Overlay
   widget=VerticalBox_0
   slot {class=OverlaySlot, properties {horizontalAlignment=HAlign_Fill, padding=<ltrb 15 15 15 15>}}
  }
  Text_Stat {
   widgetClassPath=&CommonTextBlock
   bIsVariable=true
   parent=VerticalBox_0
   widget=Text_Stat
   properties {
    justification=Center
    renderOpacity=0.85
    text=0
    font {
     bForceMonospaced=false
     bMaterialIsStencil=false
     fontMaterial=null
     fontObject=&BurbankSmall
     letterSpacing=0
     monospacedWidth=1
     size=42
     skewAmount=0
     typefaceFontName=Black
     outlineSettings {
      bApplyOutlineToDropShadows=false
      bMiteredCorners=false
      bSeparateFillAlpha=false
      outlineMaterial=null
      outlineSize=0
      outlineColor=<rgba 0 0 0 1>
     }
    }
    shadowOffset=<0 0>
   }
   slot {class=VerticalBoxSlot}
  }
  Text_Title {
   widgetClassPath=&CommonTextBlock
   bIsVariable=true
   parent=VerticalBox_0
   widget=Text_Title
   properties {
    justification=Center
    renderOpacity=0.6
    style=/FortUILibrary/Styles/Typography/_TextStyle_SubH4._TextStyle_SubH4_C
    text="{Stat}"
    textTransformPolicy=ToUpper
    font {
     bForceMonospaced=false
     bMaterialIsStencil=false
     fontMaterial=null
     fontObject=&BurbankSmall
     letterSpacing=0
     monospacedWidth=1
     size=24
     skewAmount=0
     typefaceFontName=Black
     outlineSettings {
      bApplyOutlineToDropShadows=false
      bMiteredCorners=false
      bSeparateFillAlpha=false
      outlineMaterial=null
      outlineSize=0
      outlineColor=<rgba 0 0 0 1>
     }
    }
    shadowOffset=<0 0>
    strikeBrush {
     bIsDynamicallyLoaded=false
     drawAs=NoDrawType
     imageType=NoImage
     mirroring=NoMirror
     resourceName=None
     resourceObject=null
     tiling=NoTile
     imageSize=<32 32>
     margin=<ltrb 0 0 0 0>
     outlineSettings {
      bUseBrushTransparency=false
      roundingType=HalfHeightRadius
      width=0
      color {colorUseRule=UseColor_Specified, specifiedColor=<rgba 0 0 0 0>}
      cornerRadii=<0 0 0 0>
     }
     tintColor {colorUseRule=UseColor_Specified, specifiedColor=<rgba 1 1 1 1>}
     uVRegion {bIsValid=false, max=<0 0>, min=<0 0>}
    }
   }
   slot {class=VerticalBoxSlot}
  }
 ]
}
