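# Small tournament banner widget for the post-game XP screen, showing the next tournament's name, thumbnail and start countdown.
# InitializeTournamentData sets the name, starts FortCountdownTimerVM toward StartTime with "LIVE" as live text, then updates the thumbnail.
# Thumbnail texture and loading state drive dynamic material parameters; a missing texture keeps the loading state on.
# HandleCountdownTimerStateChanged maps timer state to highlight visibility, collapsing it for None.
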WidgetBlueprint /Game/Athena/HUD/EndGame/PostGameFlowScreens/DefaultScreens/XPScreen/WBP_TournamentSmallBanner
graphs [
 EventGraph {nodes [
   n1 = GetMcpRelativeUtcNow!(
    execute
    FortPC:Fort Player Controller Object Reference=GetFirstLocalFortPlayerController!(bIncludeNonSpectators:bool=true, bIncludeSpectators:bool=false) -> (ReturnValue:Fort Player Controller Object Reference)
   ) -> (then, ReturnValue:Date Time Structure)
   n2 = InitializeTournamentData!() -> (OutputDelegate, then, NextTournamentData:Next Tournament Data Structure) {is=title K2Node_CustomEvent}
   n3 = Break Next Tournament Data!(NextTournamentData:"Next Tournament Data Structure (by ref)"=@n2.NextTournamentData) -> (
    TournamentName:String
    StartTime:Date Time Structure
    EndTime:Date Time Structure
    bIsLoadingImage:bool=~
    TournamentTexture:Texture 2D Object Reference
   ) {is=title K2Node_BreakStruct}
   InitializeCountdown!(execute=[@n1.then, @n6.then], self, InFutureTime:"Date Time Structure (by ref)"=@n3.StartTime) -> (then)
   > SetThumbnailLoadingState!(execute, self, IsLoading:bool=@n3.bIsLoadingImage) -> (then)
   > SetThumbnail!(execute, self, Texture:Texture Object Reference=@n3.TournamentTexture) -> (then)
   n4 = InitializeCountdown!() -> (OutputDelegate, then, InFutureTime:"Date Time Structure (by ref)") {is=title K2Node_CustomEvent}
   n5 = Get FortCountdownTimerVM() -> (FortCountdownTimerVM:Countdown Timer View Model Object Reference)
   > SetupViewModel!(
    execute
    self:Countdown Timer View Model Object Reference=@n5
    InFutureTime:"Date Time Structure (by ref)"=@n4.InFutureTime
    InLiveText:"Text (by ref)"=MakeLiteralText(Value="LOCTEXT(\"LIVE\")")
   ) -> (then)
   > StartTimer(self:Countdown Timer View Model Object Reference=@n5)
   DateTime + Timespan!(
    A:Date Time Structure=@n1
    B:Timespan Structure=MakeTimespan!(Days:int=0, Hours:int=0, Minutes:int=1, Seconds:int=10, Milliseconds:int=0) -> (ReturnValue:Timespan Structure)
   ) -> (ReturnValue:Date Time Structure)
   n6 = "SetText (Text)"(execute=@n2.then, self=Get Text_Highlight() -> (Text_Highlight:Common Text Object Reference), InText="To Text (String)"(InString=@n3))
  ]}
 HandleCountdownTimerStateChanged {nodes [
   n1 = HandleCountdownTimerStateChanged!() -> (then, TimerState:ECountdownTimerState Enum=None) {is=title K2Node_FunctionEntry}
   > SetVisibility(
    self=Get WBP_UIKit_Highlight() -> (WBP_UIKit_Highlight:WBP UIKit Highlight Object Reference)
    InVisibility=Select!(
     None:ESlateVisibility Enum=Collapsed
     Countdown:ESlateVisibility Enum=SelfHitTestInvisible
     Live:ESlateVisibility Enum=SelfHitTestInvisible
     Countup:ESlateVisibility Enum=Visible
     Index:ECountdownTimerState Enum=@n1
    ) -> (ReturnValue)
   )
  ]}
 SetThumbnail {nodes [
   n1 = SetThumbnail!() -> (then, Texture:Texture Object Reference) {is=title K2Node_FunctionEntry}
   > n2 = Is Valid(InputObject=@n1)
   > n3 = GetDynamicMaterial(self=Get TileThumbnail() -> (TileThumbnail:Image Object Reference))
   > SetTextureParameterValue(self=@n3, ParameterName=Texture, Value=@n1)
   > SetThumbnailLoadingState(IsLoading=false)
   SetThumbnailLoadingState(execute=@n2.Is Not Valid, IsLoading=true)
  ]}
 SetThumbnailLoadingState {nodes [
   n1 = SetThumbnailLoadingState!() -> (then, IsLoading:bool=~) {is=title K2Node_FunctionEntry}
   > n2 = GetDynamicMaterial(self=Get TileThumbnail())
   > SetScalarParameterValue(self=@n2, ParameterName=IsLoading, Value="To Float (Boolean)"(InBool=@n1))
  ]}
]
variables []
widgets {
 info {inheritedWidgetCount=0, parentClass=&UserWidget, rootWidgetClass=&Overlay, widgetCount=6}
 widgets [
  Overlay_0 {widgetClassPath=&Overlay, widget=Overlay_0}
  SizeBox_0 {
   widgetClassPath=&SizeBox
   parent=Overlay_0
   widget=SizeBox_0
   properties {bOverride_HeightOverride=true, bOverride_WidthOverride=true, heightOverride=252, widthOverride=472}
   slot {class=OverlaySlot, properties {horizontalAlignment=HAlign_Center, verticalAlignment=VAlign_Center}}
  }
  ThumbnailOverlay {widgetClassPath=&Overlay, parent=SizeBox_0, widget=ThumbnailOverlay, slot {class=SizeBoxSlot}}
  TileThumbnail {
   widgetClassPath=&Image
   bIsVariable=true
   parent=ThumbnailOverlay
   widget=TileThumbnail
   properties {
    visibility=SelfHitTestInvisible
    brush {
     bIsDynamicallyLoaded=false
     drawAs=Image
     imageType=FullColor
     mirroring=NoMirror
     resourceName=None
     resourceObject=/Game/Athena/HUD/EndGame/PostGameFlowScreens/DefaultScreens/XPScreen/Art/M_TournamentSmallBannerThumbnail
     tiling=NoTile
     imageSize=<32 32>
     margin=<ltrb 0 0 0 0>
     outlineSettings {
      bUseBrushTransparency=false
      roundingType=HalfHeightRadius
      width=0
      color {colorUseRule=UseColor_Specified, specifiedColor=<rgba 0 0 0 0>}
      cornerRadii=<0 0 0 0>
     }
     tintColor {colorUseRule=UseColor_Specified, specifiedColor=<rgba 1 1 1 1>}
     uVRegion {bIsValid=false, max=<0 0>, min=<0 0>}
    }
   }
   slot {class=OverlaySlot, properties {horizontalAlignment=HAlign_Fill, verticalAlignment=VAlign_Fill}}
  }
  Text_Highlight {
   widgetClassPath=&CommonTextBlock
   bIsVariable=true
   parent=ThumbnailOverlay
   widget=Text_Highlight
   properties {
    applyLineHeightToBottomLine=false
    bAutoCollapseWithEmptyText=true
    scrollStyle=&TextScrollStyle-Base_C
    text=Tournament
    font {
     bForceMonospaced=false
     bMaterialIsStencil=false
     fontMaterial=null
     fontObject=&HeadingNow
     letterSpacing=0
     monospacedWidth=1
     size=20.25
     skewAmount=0
     typefaceFontName=75 Medium
     outlineSettings {
      bApplyOutlineToDropShadows=false
      bMiteredCorners=false
      bSeparateFillAlpha=false
      outlineMaterial=null
      outlineSize=0
      outlineColor=<rgba 0 0 0 1>
     }
    }
    shadowOffset=<0 0>
   }
   slot {class=OverlaySlot, properties {horizontalAlignment=HAlign_Fill, verticalAlignment=VAlign_Bottom, padding=<ltrb 18 0 12 26>}}
  }
  WBP_UIKit_Highlight {
   widgetClassPath=/FortUILibrary/UIKit/Highlight/WBP_UIKit_Highlight
   bIsVariable=true
   parent=ThumbnailOverlay
   widget=WBP_UIKit_Highlight
   properties {highlightText=00:00}
   slot {class=OverlaySlot, properties {padding=<ltrb 8 8 0 0>}}
  }
 ]
}
