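# Additive UI material that draws animated triangles for the Habanero post-match screen in the frontend lobby.
# Three channels of one mask texture pan at separate speeds over rotatable UVs, one of them scaled about center.
# SmoothStep thresholds driven by a horizontal parabola and TriScale vary triangle size across the width, then the layers merge by Max and are sharpened.
# A horizontal ramp and an inverted, offset radial gradient fade the result.
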Material /Game/UI/Frontend/Lobby/Habanero/Habanero_Materials/M_UI_Habanero_PostMatch_Triangles
graph {
 properties {bUsedWithStaticMesh=false, blendMode=BLEND_Additive, materialDomain=MD_UI, pixelDepthOffsetMode=PDOM_Legacy, usedShadingModels=Default Lit}
 nodes [
  n1 = MaterialFunctionCall!(
   Rotation_Fraction=LinearInterpolate(Alpha=ScalarParameter() {defaultValue=1, parameterName=DirectionRotationFraction}) {constB=0.5}
   TexCoord=TextureCoordinate()
  ) {materialFunction=/FortUILibrary/Foundation/Art_ToolKit/AngledBoxes/MF_UI_Rotator_Fractions}
  n2 = AppendVector(A=ScalarParameter() {defaultValue=0.25, parameterName=Speed1}, B=Constant())
  n3 = TextureSampleParameter2D(UVs=MaterialFunctionCall(UVs=Panner(Coordinate=@n1.Result, Speed=@n2) {speedX=0.25}, Texture Scale=Constant() {r=0.6}) {materialFunction=&ScaleUVsByCenter}.UVs) {
   parameterName=TextureMask
   samplerType=SAMPLERTYPE_LinearColor
   texture=/Game/UI/Frontend/Lobby/Habanero/Habanero_Textures/T_PostMatch_Triangles
  }
  n4 = ComponentMask(None=@n1.Result) {r=true}
  n5 = Multiply(A=@n4, B=OneMinus(None=@n4))
  n6 = Multiply(A=Constant() {r=4}, B=ScalarParameter() {defaultValue=1, parameterName=TriScale})
  n7 = Saturate(None=OneMinus(None=Multiply(A=@n5, B=@n6) {constB=4}))
  n8 = ScalarParameter() {defaultValue=0.15, parameterName=Speed2}
  n9 = Constant()
  n10 = TextureSampleParameter2D(UVs=Panner(Coordinate=@n1.Result, Speed=AppendVector(A=@n8, B=@n9)) {speedX=0.15}) {
   parameterName=TextureMask
   samplerType=SAMPLERTYPE_LinearColor
   texture=&T_PostMatch_Triangles
  }
  n11 = ScalarParameter() {defaultValue=0.5, parameterName=Speed3}
  n12 = Constant()
  n13 = TextureSampleParameter2D(UVs=Panner(Coordinate=@n1.Result, Speed=AppendVector(A=@n11, B=@n12)) {speedX=0.5}) {
   parameterName=TextureMask
   samplerType=SAMPLERTYPE_LinearColor
   texture=&T_PostMatch_Triangles
  }
  n14 = Max(
   A=Max(
    A=MaterialFunctionCall(Alpha=@n3.B, Min=@n7) {materialFunction=&SmoothStep}.Result
    B=MaterialFunctionCall(Alpha=@n10.G, Min=@n7) {materialFunction=&SmoothStep}.Result
   )
   B=MaterialFunctionCall(Alpha=@n13.R, Min=@n7) {materialFunction=&SmoothStep}.Result
  )
  n15 = Saturate(None=Multiply(A=Subtract(A=@n14) {constB=0.65}) {constB=8})
  n16 = Multiply(A=@n4) {constB=3}
  n17 = NamedRerouteUsage()
  n18 = Constant2Vector() {r=-0.7}
  n19 = MaterialFunctionCall(
   UVs=MaterialFunctionCall(UV=Add(A=@n17, B=@n18) {constB=0.5}) {materialFunction=&MF_UI_UVScaling}.Scaled to Fit
   Radius=Constant() {r=0.7}
   Density=Constant() {r=5}
  ) {materialFunction=&RadialGradientExponential}
  Multiply(A=Multiply(A=LinearInterpolate(B=@n15, Alpha=@n16), B=OneMinus(None=@n19.RadialGradientExponential))) {constB=0.8}
  NamedRerouteDeclaration(None=@n1.Result) {name=UV, nodeColor=<rgba 1 0.9141 0 1>}
  VectorParameter() {parameterName=Color, defaultValue=<rgba 1 1 1 1>}
 ]
}
