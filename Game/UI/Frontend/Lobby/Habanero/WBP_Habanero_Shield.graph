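# Frontend lobby widget that displays the Habanero shield as a single Image driven by the M_UI_Habanero_Shield material.
# Its roughly three-second Intro animation fades and scales the shield in, holds it, then flares the glow before shrinking and fading out.
# The animation keys material parameters for scale, stripe sweep and glow directly, alongside render opacity.
# The event graph holds only empty default events, so the widget does not play Intro itself.
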WidgetBlueprint /Game/UI/Frontend/Lobby/Habanero/WBP_Habanero_Shield
animations [Intro {
  binding_count=1
  length_seconds=3.003
  object=&self:Intro
  sequence {
   bindings [Image_36 {
     id=#1
     possessed_class=Image
     tracks [
      {
       type=MovieSceneWidgetMaterialTrack
       sections [{
         type=MovieSceneParameterSection
         end_frame=90
         start_frame=0
         channels [
          GlowHighlightRotation {type=MovieSceneScriptingFloatChannel, keys [[0,-0.2],[77,1.0189],[85,1.05]]}
          GlowHighlightSize {type=MovieSceneScriptingFloatChannel, keys [[0,0.25],[74,0.75],[77,1.1],[85,1.2]]}
          GlowHighlight {type=MovieSceneScriptingFloatChannel, keys [[0,1],[74,1],[77,1.33],[81,0.25],[85,0]]}
          Scale {type=MovieSceneScriptingFloatChannel, keys [[0,0],[8,1],[11,1],[13,0.7],[74,0.7],[77,0.88],[81,0.88],[85,0]]}
          StripeHighlightPosition {type=MovieSceneScriptingFloatChannel, keys [[11,0],[33,2]]}
         ]
        }]
      }
      {
       type=MovieSceneFloatTrack
       display_name=Render Opacity
       sections [{
         type=MovieSceneFloatSection
         end_frame=85
         start_frame=0
         channels [None {type=MovieSceneScriptingFloatChannel, keys [[0,0],[8,1],[81,1],[85,0]]}]
        }]
      }
     ]
    }]
   display_rate {denominator=1001, numerator=30000}
  }
 }]
graphs [EventGraph {nodes [
   Event PreConstruct()
   Event Construct()
   Event Tick()
  ]}]
variables []
widgets {
 info {inheritedWidgetCount=0, parentClass=&UserWidget, rootWidgetClass=&Image, widgetCount=1}
 widgets [Image_36 {
   widgetClassPath=&Image
   bIsVariable=true
   widget=Image_36
   properties {brush {
     bIsDynamicallyLoaded=false
     drawAs=Image
     imageType=NoImage
     mirroring=NoMirror
     resourceName=None
     resourceObject=/Game/UI/Frontend/Lobby/Habanero/Habanero_Materials/M_UI_Habanero_Shield
     tiling=NoTile
     imageSize=<160 160>
     margin=<ltrb 0 0 0 0>
     outlineSettings {
      bUseBrushTransparency=false
      roundingType=HalfHeightRadius
      width=0
      color {colorUseRule=UseColor_Specified, specifiedColor=<rgba 0 0 0 0>}
      cornerRadii=<0 0 0 0>
     }
     tintColor {colorUseRule=UseColor_Specified, specifiedColor=<rgba 1 1 1 1>}
     uVRegion {bIsValid=false, max=<0 0>, min=<0 0>}
    }}
  }]
}
