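# Post-match widget that plays the ranked (Habanero) sequence, showing progress toward the next tier or a rank-up.
# Separate timelines cover intro, progress, rank-up start and end, flash fade-out and final rank, moving between current and next badges, tier text and progress bar.
# They animate the widget's own blend values for delta, progress and bar color, drive material parameters for rift, flare and glow effects, and fire timed events.
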WidgetBlueprint /FortniteCoreUI/HUD/EndGame/PostGameFlowScreens/DefaultScreens/XPScreen/WBP_PostMatch_HabaneroSequence
animations [
 Anim_Intro {
  binding_count=13
  length_seconds=0.3
  object=&self:Anim_Intro
  sequence {
   bindings [
    Text_RankUpMessage {
     id=#1
     possessed_class=CommonTextBlock
     tracks [{
       type=MovieSceneFloatTrack
       display_name=Render Opacity
       sections [{type=MovieSceneFloatSection, end_frame=6, start_frame=0, channels [None {type=MovieSceneScriptingFloatChannel, const [0,2]}]}]
      }]
    }
    Text_TierName {
     id=#2
     possessed_class=CommonTextBlock
     tracks [{
       type=MovieSceneFloatTrack
       display_name=Render Opacity
       sections [{type=MovieSceneFloatSection, end_frame=0, start_frame=0, channels [None {type=MovieSceneScriptingFloatChannel, const [0,1]}]}]
      }]
    }
    NextBadge {
     id=#3
     possessed_class=Overlay
     tracks [{
       type=MovieSceneFloatTrack
       display_name=Render Opacity
       sections [{type=MovieSceneFloatSection, end_frame=0, start_frame=0, channels [None {type=MovieSceneScriptingFloatChannel, const [0,1]}]}]
      }]
    }
    HorizontalBox_RankDelta {
     id=#4
     possessed_class=HorizontalBox
     tracks [{
       type=MovieSceneFloatTrack
       display_name=Render Opacity
       sections [{type=MovieSceneFloatSection, end_frame=0, start_frame=0, channels [None {type=MovieSceneScriptingFloatChannel, const [0,1]}]}]
      }]
    }
    RadialGlow {
     id=#5
     possessed_class=Image
     tracks [{
       type=MovieSceneFloatTrack
       display_name=Render Opacity
       sections [{type=MovieSceneFloatSection, end_frame=0, start_frame=0, channels [None {type=MovieSceneScriptingFloatChannel, const [0,1]}]}]
      }]
    }
    CurrentBadge {
     id=#6
     possessed_class=Overlay
     tracks [{
       type=MovieSceneFloatTrack
       display_name=Render Opacity
       sections [{type=MovieSceneFloatSection, end_frame=0, start_frame=0, channels [None {type=MovieSceneScriptingFloatChannel, const [0,1]}]}]
      }]
    }
    WBP_TournamentSmallBanner {
     id=#7
     possessed_class=WBP_TournamentSmallBanner_C
     tracks [{
       type=MovieSceneFloatTrack
       display_name=Render Opacity
       sections [{type=MovieSceneFloatSection, end_frame=6, start_frame=0, channels [None {type=MovieSceneScriptingFloatChannel, const [0,2]}]}]
      }]
    }
    Overlay_TriangleEnergy {
     id=#8
     possessed_class=Overlay
     tracks [{
       type=MovieSceneFloatTrack
       display_name=Render Opacity
       sections [{type=MovieSceneFloatSection, end_frame=0, start_frame=0, channels [None {type=MovieSceneScriptingFloatChannel, const [0,1]}]}]
      }]
    }
    Background {
     id=#9
     possessed_class=Image
     tracks [{
       type=MovieSceneWidgetMaterialTrack
       sections [{
         type=MovieSceneParameterSection
         end_frame=6
         start_frame=0
         channels [Background Opacity {type=MovieSceneScriptingFloatChannel, keys [[0,0],[6,0.6]]}]
        }]
      }]
    }
    Img_Flare {
     id=#10
     possessed_class=Image
     tracks [{
       type=MovieSceneFloatTrack
       display_name=Render Opacity
       sections [{type=MovieSceneFloatSection, end_frame=0, start_frame=0, channels [None {type=MovieSceneScriptingFloatChannel, const [0,1]}]}]
      }]
    }
    ProgressBar {
     id=#11
     possessed_class=Image
     tracks [{
       type=MovieSceneFloatTrack
       display_name=Render Opacity
       sections [{type=MovieSceneFloatSection, end_frame=0, start_frame=0, channels [None {type=MovieSceneScriptingFloatChannel, const [0,1]}]}]
      }]
    }
    Particles {
     id=#12
     possessed_class=GridPanel
     tracks [{
       type=MovieSceneFloatTrack
       display_name=Render Opacity
       sections [{type=MovieSceneFloatSection, end_frame=0, start_frame=0, channels [None {type=MovieSceneScriptingFloatChannel, const [0,1]}]}]
      }]
    }
    TextWrapper_RankUp {
     id=#13
     possessed_class=Border
     tracks [{
       type=MovieSceneFloatTrack
       display_name=Render Opacity
       sections [{type=MovieSceneFloatSection, end_frame=0, start_frame=0, channels [None {type=MovieSceneScriptingFloatChannel, const [0,1]}]}]
      }]
    }
   ]
   tracks [{type=MovieSceneEventTrack, sections [{type=MovieSceneEventTriggerSection, channels [None {type=MovieSceneScriptingEventChannel}]}]}]
  }
 }
 Anim_RankUp_Start {
  binding_count=15
  length_seconds=5
  object=&self:Anim_RankUp_Start
  sequence {
   bindings [
    FortVisualAttachment_Elims {
     id=#14
     possessed_class=FortVisualAttachment
     tracks [{
       type=MovieScene2DTransformTrack
       sections [{
         type=MovieScene2DTransformSection
         end_frame=207
         start_frame=0
         channels [
          Translation.X {type=MovieSceneScriptingFloatChannel}
          Translation.Y {type=MovieSceneScriptingFloatChannel, const [111,1]}
          Angle {type=MovieSceneScriptingFloatChannel}
          Scale.X {type=MovieSceneScriptingFloatChannel}
          Scale.Y {type=MovieSceneScriptingFloatChannel}
          Shear.X {type=MovieSceneScriptingFloatChannel}
          Shear.Y {type=MovieSceneScriptingFloatChannel}
         ]
        }]
      }]
    }
    TextWrapper_RankUp {
     id=#15
     possessed_class=Border
     tracks [{
       type=MovieSceneFloatTrack
       display_name=Render Opacity
       sections [{type=MovieSceneFloatSection, end_frame=300, start_frame=0, channels [None {type=MovieSceneScriptingFloatChannel, const [0,1]}]}]
      }]
    }
    WBP_PostMatch_HabaneroSequence_C_22 {
     id=#16
     possessed_class=WBP_PostMatch_HabaneroSequence_C
     tracks [
      {
       type=MovieSceneDoubleTrack
       display_name=Delta Value Blend
       sections [{
         type=MovieSceneDoubleSection
         end_frame=231
         start_frame=0
         channels [None {type=MovieSceneScriptingDoubleChannel, keys [[0,0],[39,0],[231,1]]}]
        }]
      }
      {
       type=MovieSceneDoubleTrack
       display_name=Progress Bar Color Blend
       sections [{type=MovieSceneDoubleSection, end_frame=294, start_frame=0, channels [None {type=MovieSceneScriptingDoubleChannel, keys [[0,0],[294,1]]}]}]
      }
      {
       type=MovieSceneDoubleTrack
       display_name=Progress Value Blend
       sections [{
         type=MovieSceneDoubleSection
         end_frame=264
         start_frame=0
         channels [None {type=MovieSceneScriptingDoubleChannel, keys [[0,0],[237,0],[264,1]]}]
        }]
      }
     ]
    }
    Particles {
     id=#17
     possessed_class=GridPanel
     tracks [{
       type=MovieSceneFloatTrack
       display_name=Render Opacity
       sections [{type=MovieSceneFloatSection, channels [None {type=MovieSceneScriptingFloatChannel, const [0,1]}]}]
      }]
    }
    NextBadge {
     id=#18
     possessed_class=Overlay
     tracks [{
       type=MovieSceneFloatTrack
       display_name=Render Opacity
       sections [{type=MovieSceneFloatSection, end_frame=300, start_frame=0, channels [None {type=MovieSceneScriptingFloatChannel, const [0,1]}]}]
      }]
    }
    ProgressBar {
     id=#19
     possessed_class=Image
     tracks [
      {
       type=MovieSceneWidgetMaterialTrack
       sections [{
         type=MovieSceneParameterSection
         end_frame=300
         start_frame=0
         channels [
          Glow Opacity {type=MovieSceneScriptingFloatChannel, keys [[0,0],[39,0],[119,0.25],[264,0.4],[268,0]]}
          Override Edge Gradient {type=MovieSceneScriptingFloatChannel, const [0,1]}
         ]
        }]
      }
      {
       type=MovieSceneFloatTrack
       display_name=Render Opacity
       sections [{
         type=MovieSceneFloatSection
         end_frame=268
         start_frame=0
         channels [None {type=MovieSceneScriptingFloatChannel, thinned_from=6, keys [[0,0],[18,0],[39,1],[265,1],[268,0]]}]
        }]
      }
      {
       type=MovieScene2DTransformTrack
       sections [{
         type=MovieScene2DTransformSection
         end_frame=273
         start_frame=0
         channels [
          Translation.Y {type=MovieSceneScriptingFloatChannel, const [96,1]}
          Scale.X {type=MovieSceneScriptingFloatChannel, keys [[0,0.25],[18,0.25],[39,1],[264,1],[273,4]]}
          Scale.Y {type=MovieSceneScriptingFloatChannel, keys [[0,0.25],[18,0.25],[39,1],[264,1],[273,4]]}
         ]
        }]
      }
     ]
    }
    Overlay_TriangleEnergy {
     id=#20
     possessed_class=Overlay
     tracks [{
       type=MovieSceneFloatTrack
       display_name=Render Opacity
       sections [{
         type=MovieSceneFloatSection
         end_frame=300
         start_frame=0
         channels [None {type=MovieSceneScriptingFloatChannel, keys [[0,0],[108,0],[231,0.05],[249,0],[300,0]]}]
        }]
      }]
    }
    Text_TierName {
     id=#21
     possessed_class=CommonTextBlock
     tracks [
      {
       type=MovieSceneFloatTrack
       display_name=Render Opacity
       sections [{
         type=MovieSceneFloatSection
         end_frame=300
         start_frame=0
         channels [None {type=MovieSceneScriptingFloatChannel, thinned_from=6, keys [[0,0],[27,0],[51,1],[264,1],[268,0]]}]
        }]
      }
      {
       type=MovieScene2DTransformTrack
       sections [{type=MovieScene2DTransformSection, channels [Translation.Y {type=MovieSceneScriptingFloatChannel, keys [[0,128],[300,0]]}]}]
      }
     ]
    }
    CurrentBadge_Icon {
     id=#22
     possessed_class=Image
     tracks [
      {
       type=MovieSceneWidgetMaterialTrack
       sections [{
         type=MovieSceneParameterSection
         end_frame=300
         start_frame=0
         channels [
          Badge Hotness {type=MovieSceneScriptingFloatChannel, keys [[0,0],[111,0],[228,2],[249,2],[264,1.75],[297,8.2439]]}
          Gem Hotness {type=MovieSceneScriptingFloatChannel, keys [[0,0],[111,0],[228,20],[249,20],[265,1]]}
          Crest Brightness {type=MovieSceneScriptingFloatChannel, keys [[0,12],[111,12],[228,6]]}
         ]
        }]
      }
      {
       type=MovieScene2DTransformTrack
       sections [{
         type=MovieScene2DTransformSection
         end_frame=257
         start_frame=0
         channels [
          Translation.X {type=MovieSceneScriptingFloatChannel, keys [[0,8],[210,0],[219.308,-4],[228.621,4],[237.933,8],[249,-8],[249.106,-4],[257.994,0]]}
          Translation.Y {type=MovieSceneScriptingFloatChannel, keys [[0,8],[210,0],[219.308,-4],[228.621,-4],[237.933,8],[249,8],[249.106,4],[257.994,0]]}
         ]
        }]
      }
     ]
    }
    FortVisualAttachment_Placement {
     id=#23
     possessed_class=FortVisualAttachment
     tracks [{
       type=MovieScene2DTransformTrack
       sections [{
         type=MovieScene2DTransformSection
         end_frame=207
         start_frame=0
         channels [
          Translation.X {type=MovieSceneScriptingFloatChannel}
          Translation.Y {type=MovieSceneScriptingFloatChannel, const [111,1]}
          Angle {type=MovieSceneScriptingFloatChannel}
          Scale.X {type=MovieSceneScriptingFloatChannel}
          Scale.Y {type=MovieSceneScriptingFloatChannel}
          Shear.X {type=MovieSceneScriptingFloatChannel}
          Shear.Y {type=MovieSceneScriptingFloatChannel}
         ]
        }]
      }]
    }
    CurrentBadge {
     id=#24
     possessed_class=Overlay
     tracks [
      {
       type=MovieSceneFloatTrack
       display_name=Render Opacity
       sections [{
         type=MovieSceneFloatSection
         end_frame=300
         start_frame=0
         channels [None {type=MovieSceneScriptingFloatChannel, keys [[0,0],[9,0],[39,1],[264,1],[267,0]]}]
        }]
      }
      {
       type=MovieScene2DTransformTrack
       sections [{
         type=MovieScene2DTransformSection
         end_frame=281
         start_frame=0
         channels [
          Translation.Y {type=MovieSceneScriptingFloatChannel, const [96,1]}
          Scale.X {type=MovieSceneScriptingFloatChannel, keys [[0,2],[27,0.6],[72,0.6],[258,0.7],[274,0.1]]}
          Scale.Y {type=MovieSceneScriptingFloatChannel, keys [[0,2],[27,0.6],[72,0.6],[258,0.7],[274,0.1]]}
         ]
        }]
      }
     ]
    }
    Background {
     id=#25
     possessed_class=Image
     tracks [{
       type=MovieSceneWidgetMaterialTrack
       sections [{
         type=MovieSceneParameterSection
         channels [
          Color Swap {type=MovieSceneScriptingFloatChannel, keys [[0,0],[282,0],[288,0.5]]}
          Hot Spot Intensity {type=MovieSceneScriptingFloatChannel, keys [[0,0.5],[261,0.5],[300,8]]}
          Cloud Opacity {type=MovieSceneScriptingFloatChannel, keys [[0,0],[249,0],[273,1],[300,1]]}
          Triangle Opacity {type=MovieSceneScriptingFloatChannel, keys [[0,0],[300,0.6]]}
          Triangles Tiling {type=MovieSceneScriptingFloatChannel, keys [[0,4],[282,4],[300,12]]}
          Rift Intensity {type=MovieSceneScriptingFloatChannel, keys [[0,0],[261,0],[300,20]]}
          Rift Cracks Removal {type=MovieSceneScriptingFloatChannel, keys [[0,0.6863],[267,0.6863],[273,0.1347],[282,-0.17]]}
          Invert Radial {type=MovieSceneScriptingFloatChannel, keys [[0,0.6],[39,0.6],[215,0],[288,1]]}
          Spot Scale {type=MovieSceneScriptingFloatChannel, keys [[0,1],[39,1],[219,1.75]]}
         ]
        }]
      }]
    }
    RadialGlow {
     id=#26
     possessed_class=Image
     tracks [
      {
       type=MovieSceneFloatTrack
       display_name=Render Opacity
       sections [{
         type=MovieSceneFloatSection
         end_frame=285
         start_frame=0
         channels [None {type=MovieSceneScriptingFloatChannel, keys [[0,0],[150,0],[201,1],[285,0]]}]
        }]
      }
      {
       type=MovieSceneWidgetMaterialTrack
       sections [{
         type=MovieSceneParameterSection
         end_frame=300
         start_frame=0
         channels [
          Opacity {type=MovieSceneScriptingFloatChannel, const [0.35,1]}
          HexGridOpacity {type=MovieSceneScriptingFloatChannel, const [0,1]}
          Glow ColorR {type=MovieSceneScriptingFloatChannel, const [1,1]}
          Glow ColorG {type=MovieSceneScriptingFloatChannel, const [1,1]}
          Glow ColorB {type=MovieSceneScriptingFloatChannel, const [1,1]}
          Glow ColorA {type=MovieSceneScriptingFloatChannel, const [1,1]}
         ]
        }]
      }
     ]
    }
    Box_PercentageText {
     id=#27
     display_name=PercentageTextwrapper
     possessed_class=HorizontalBox
     tracks [
      {
       type=MovieScene2DTransformTrack
       sections [{type=MovieScene2DTransformSection, channels [Translation.Y {type=MovieSceneScriptingFloatChannel, keys [[0,128],[300,0]]}]}]
      }
      {
       type=MovieSceneFloatTrack
       display_name=Render Opacity
       sections [{type=MovieSceneFloatSection, end_frame=45, start_frame=0, channels [None {type=MovieSceneScriptingFloatChannel, keys [[0,0],[29,0],[45,1]]}]}]
      }
     ]
    }
    Img_Flare {
     id=#28
     possessed_class=Image
     tracks [{
       type=MovieSceneWidgetMaterialTrack
       sections [{
         type=MovieSceneParameterSection
         end_frame=300
         start_frame=0
         channels [
          Opacity {type=MovieSceneScriptingFloatChannel, thinned_from=5, keys [[0,0],[264,0],[291,1],[297,0]]}
          Flare Strength {type=MovieSceneScriptingFloatChannel, keys [[0,24],[264,24],[300,150]]}
         ]
        }]
      }]
    }
   ]
   display_rate {numerator=60}
   tracks [{
     type=MovieSceneEventTrack
     sections [{
       type=MovieSceneEventTriggerSection
       channels [None {
         type=MovieSceneScriptingEventChannel
         keys [
          [0, "<Struct 'MovieSceneEvent' (0x4676113e880) {}>"]
          [51, "<Struct 'MovieSceneEvent' (0x4676113e920) {}>"]
          [171, "<Struct 'MovieSceneEvent' (0x4676113e100) {}>"]
          [260, "<Struct 'MovieSceneEvent' (0x4676113e600) {}>"]
          [264, "<Struct 'MovieSceneEvent' (0x4676113dc00) {}>"]
         ]
        }]
      }]
    }]
  }
 }
 Anim_RankUp_End {
  binding_count=17
  length_seconds=3.45
  object=&self:Anim_RankUp_End
  sequence {
   bindings [
    TextWrapper_RankUp {
     id=#15
     possessed_class=Border
     tracks [
      {
       type=MovieSceneFloatTrack
       display_name=Render Opacity
       sections [{
         type=MovieSceneFloatSection
         end_frame=204
         channels [None {type=MovieSceneScriptingFloatChannel, keys [[0,0],[33,0],[48,1],[180.128,1],[204.207,0]]}]
        }]
      }
      {
       type=MovieScene2DTransformTrack
       sections [{
         type=MovieScene2DTransformSection
         end_frame=63
         channels [Translation.Y {type=MovieSceneScriptingFloatChannel, keys [[0,32],[33,32],[48,0]]}]
        }]
      }
     ]
    }
    Particles {
     id=#29
     possessed_class=GridPanel
     tracks [{
       type=MovieSceneFloatTrack
       display_name=Render Opacity
       sections [{type=MovieSceneFloatSection, end_frame=72, start_frame=0, channels [None {type=MovieSceneScriptingFloatChannel, keys [[0,0],[42,0],[72,1]]}]}]
      }]
    }
    NextBadge {
     id=#18
     possessed_class=Overlay
     tracks [
      {
       type=MovieSceneFloatTrack
       display_name=Render Opacity
       sections [{type=MovieSceneFloatSection, channels [None {type=MovieSceneScriptingFloatChannel, keys [[0,0],[3,0],[24,1]]}]}]
      }
      {
       type=MovieScene2DTransformTrack
       sections [{
         type=MovieScene2DTransformSection
         end_frame=99
         channels [
          Translation.Y {type=MovieSceneScriptingFloatChannel, thinned_from=4, keys [[0,82],[24,82],[51,0]]}
          Scale.X {type=MovieSceneScriptingFloatChannel, keys [[0,0.9],[3,0.9],[24,1],[66,1],[81,1.2],[99,0.6]]}
          Scale.Y {type=MovieSceneScriptingFloatChannel, keys [[0,0.9],[3,0.9],[24,1],[66,1],[81,1.2],[99,0.6]]}
         ]
        }]
      }
     ]
    }
    ProgressBar {
     id=#19
     possessed_class=Image
     tracks [
      {
       type=MovieSceneFloatTrack
       display_name=Render Opacity
       sections [{
         type=MovieSceneFloatSection
         end_frame=99
         channels [None {type=MovieSceneScriptingFloatChannel, thinned_from=4, keys [[0,0],[81,0],[99,1]]}]
        }]
      }
      {
       type=MovieScene2DTransformTrack
       sections [{
         type=MovieScene2DTransformSection
         end_frame=103
         start_frame=0
         channels [
          Translation.X {type=MovieSceneScriptingFloatChannel, const [0,1]}
          Translation.Y {type=MovieSceneScriptingFloatChannel, const [0,1]}
          Scale.X {type=MovieSceneScriptingFloatChannel, const [1,1]}
          Scale.Y {type=MovieSceneScriptingFloatChannel, const [1,1]}
         ]
        }]
      }
      {
       type=MovieSceneWidgetMaterialTrack
       sections [{
         type=MovieSceneParameterSection
         end_frame=207
         start_frame=0
         channels [
          Glow Opacity {type=MovieSceneScriptingFloatChannel, const [0.25,1]}
          Progress Background Opacity {type=MovieSceneScriptingFloatChannel, const [1,1]}
          Override Edge Gradient {type=MovieSceneScriptingFloatChannel, const [0,1]}
         ]
        }]
      }
     ]
    }
    Overlay_TriangleEnergy {
     id=#20
     possessed_class=Overlay
     tracks [{
       type=MovieSceneFloatTrack
       display_name=Render Opacity
       sections [{type=MovieSceneFloatSection, channels [None {type=MovieSceneScriptingFloatChannel, const [0,1]}]}]
      }]
    }
    Text_TierName {
     id=#21
     possessed_class=CommonTextBlock
     tracks [
      {
       type=MovieSceneFloatTrack
       display_name=Render Opacity
       sections [{type=MovieSceneFloatSection, end_frame=63, channels [None {type=MovieSceneScriptingFloatChannel, keys [[0,0],[45,0],[51,1]]}]}]
      }
      {
       type=MovieScene2DTransformTrack
       sections [{
         type=MovieScene2DTransformSection
         end_frame=63
         channels [
          Scale.X {type=MovieSceneScriptingFloatChannel, keys [[0,4],[45,4],[54,1],[57,1.05],[60,1]]}
          Scale.Y {type=MovieSceneScriptingFloatChannel, keys [[0,2],[45,2],[54,1],[57,1.05],[60,1]]}
         ]
        }]
      }
     ]
    }
    WBP_PostMatch_HabaneroSequence_C_5 {
     id=#30
     possessed_class=WBP_PostMatch_HabaneroSequence_C
     tracks [
      {
       type=MovieSceneDoubleTrack
       display_name=Delta Value Blend
       sections [{
         type=MovieSceneDoubleSection
         end_frame=180
         start_frame=0
         channels [None {type=MovieSceneScriptingDoubleChannel, thinned_from=5, keys [[0,0],[81,0],[180,1]]}]
        }]
      }
      {
       type=MovieSceneDoubleTrack
       display_name=Progress Bar Color Blend
       sections [{type=MovieSceneDoubleSection, end_frame=0, start_frame=0, channels [None {type=MovieSceneScriptingDoubleChannel, const [1,1]}]}]
      }
      {
       type=MovieSceneDoubleTrack
       display_name=Progress Value Blend
       sections [{
         type=MovieSceneDoubleSection
         end_frame=207
         start_frame=0
         channels [None {type=MovieSceneScriptingDoubleChannel, keys [[0,0],[180,0],[207,1]]}]
        }]
      }
     ]
    }
    NextBadge_Icon {
     id=#31
     possessed_class=Image
     tracks [{
       type=MovieSceneWidgetMaterialTrack
       sections [{type=MovieSceneParameterSection, channels [Badge Hotness {type=MovieSceneScriptingFloatChannel, keys [[0,4],[3,4],[24,0]]}]}]
      }]
    }
    Text_RankUpMessage {
     id=#32
     possessed_class=CommonTextBlock
     tracks [
      {
       type=MovieSceneFloatTrack
       display_name=Render Opacity
       sections [{
         type=MovieSceneFloatSection
         end_frame=207
         start_frame=0
         channels [None {type=MovieSceneScriptingFloatChannel, thinned_from=4, keys [[45,0],[47,1],[207,1]]}]
        }]
      }
      {
       type=MovieScene2DTransformTrack
       sections [{
         type=MovieScene2DTransformSection
         end_frame=207
         start_frame=0
         channels [
          Scale.X {type=MovieSceneScriptingFloatChannel, keys [[45,1.25],[53,0.95],[56,1]]}
          Scale.Y {type=MovieSceneScriptingFloatChannel, keys [[45,1.25],[53,0.95],[56,1]]}
         ]
        }]
      }
     ]
    }
    FortVisualAttachment_Placement {
     id=#33
     possessed_class=FortVisualAttachment
     tracks [{
       type=MovieScene2DTransformTrack
       sections [{
         type=MovieScene2DTransformSection
         end_frame=0
         start_frame=0
         channels [
          Translation.X {type=MovieSceneScriptingFloatChannel, const [0,1]}
          Translation.Y {type=MovieSceneScriptingFloatChannel, const [20,1]}
          Angle {type=MovieSceneScriptingFloatChannel, const [0,1]}
          Scale.X {type=MovieSceneScriptingFloatChannel, const [1,1]}
          Scale.Y {type=MovieSceneScriptingFloatChannel, const [1,1]}
          Shear.X {type=MovieSceneScriptingFloatChannel, const [0,1]}
          Shear.Y {type=MovieSceneScriptingFloatChannel, const [0,1]}
         ]
        }]
      }]
    }
    WidgetSwitcherSlot_RankedFeedback {
     id=#34
     possessed_class=CommonAnimatedSwitcher
     tracks [{
       type=MovieSceneFloatTrack
       display_name=Render Opacity
       sections [{type=MovieSceneFloatSection, end_frame=0, start_frame=0, channels [None {type=MovieSceneScriptingFloatChannel, const [1,1]}]}]
      }]
    }
    CurrentBadge {
     id=#24
     possessed_class=Overlay
     tracks [{
       type=MovieSceneFloatTrack
       display_name=Render Opacity
       sections [{type=MovieSceneFloatSection, channels [None {type=MovieSceneScriptingFloatChannel, const [0,1]}]}]
      }]
    }
    Background {
     id=#25
     possessed_class=Image
     tracks [{
       type=MovieSceneWidgetMaterialTrack
       sections [{
         type=MovieSceneParameterSection
         channels [
          Triangle Opacity {type=MovieSceneScriptingFloatChannel, const [0.6,1]}
          Invert Radial {type=MovieSceneScriptingFloatChannel, const [1,1]}
          Color Swap {type=MovieSceneScriptingFloatChannel, const [1,1]}
         ]
        }]
      }]
    }
    Box_PercentageText {
     id=#35
     display_name=PercentageTextwrapper
     possessed_class=HorizontalBox
     tracks [{
       type=MovieSceneFloatTrack
       display_name=Render Opacity
       sections [{type=MovieSceneFloatSection, end_frame=99, start_frame=0, channels [None {type=MovieSceneScriptingFloatChannel, keys [[0,0],[81,0],[99,1]]}]}]
      }]
    }
    RadialGlow {
     id=#26
     possessed_class=Image
     tracks [{
       type=MovieSceneFloatTrack
       display_name=Render Opacity
       sections [{type=MovieSceneFloatSection, end_frame=63, channels [None {type=MovieSceneScriptingFloatChannel, const [0,1]}]}]
      }]
    }
    Img_Flare {
     id=#28
     possessed_class=Image
     tracks [{
       type=MovieSceneWidgetMaterialTrack
       sections [{type=MovieSceneParameterSection, end_frame=63, channels [Opacity {type=MovieSceneScriptingFloatChannel, const [0,1]}]}]
      }]
    }
    FortVisualAttachment_Elims {
     id=#36
     possessed_class=FortVisualAttachment
     tracks [{
       type=MovieScene2DTransformTrack
       sections [{
         type=MovieScene2DTransformSection
         end_frame=0
         start_frame=0
         channels [
          Translation.X {type=MovieSceneScriptingFloatChannel, const [0,1]}
          Translation.Y {type=MovieSceneScriptingFloatChannel, const [20,1]}
          Angle {type=MovieSceneScriptingFloatChannel, const [0,1]}
          Scale.X {type=MovieSceneScriptingFloatChannel, const [1,1]}
          Scale.Y {type=MovieSceneScriptingFloatChannel, const [1,1]}
          Shear.X {type=MovieSceneScriptingFloatChannel, const [0,1]}
          Shear.Y {type=MovieSceneScriptingFloatChannel, const [0,1]}
         ]
        }]
      }]
    }
   ]
   display_rate {numerator=60}
   tracks [{
     type=MovieSceneEventTrack
     sections [{
       type=MovieSceneEventTriggerSection
       channels [None {
         type=MovieSceneScriptingEventChannel
         keys [
          [33, "<Struct 'MovieSceneEvent' (0x466efbd7d00) {}>"]
          [76.2, "<Struct 'MovieSceneEvent' (0x46722902b20) {}>"]
          [91, "<Struct 'MovieSceneEvent' (0x46761250000) {}>"]
          [183, "<Struct 'MovieSceneEvent' (0x467612500a0) {}>"]
         ]
        }]
      }]
    }]
  }
 }
 Anim_RankUp_FlashFadeOut {
  binding_count=1
  length_seconds=0.45
  object=&self:Anim_RankUp_FlashFadeOut
  sequence {
   bindings [Background {
     id=#25
     possessed_class=Image
     tracks [{
       type=MovieSceneWidgetMaterialTrack
       sections [{
         type=MovieSceneParameterSection
         channels [
          Color Swap {type=MovieSceneScriptingFloatChannel, keys [[0,0.5],[27,1]]}
          Hot Spot Intensity {type=MovieSceneScriptingFloatChannel, keys [[0,8],[24,0.5]]}
          Cloud Opacity {type=MovieSceneScriptingFloatChannel, keys [[0,1],[24,0]]}
          Triangles Tiling {type=MovieSceneScriptingFloatChannel, keys [[0,12],[15,4]]}
          Rift Intensity {type=MovieSceneScriptingFloatChannel, keys [[0,20],[3,20],[6,0]]}
         ]
        }]
      }]
    }]
   display_rate {numerator=60}
  }
 }
 Anim_Progress {
  binding_count=16
  length_seconds=4.4
  object=&self:Anim_Progress
  sequence {
   bindings [
    FortVisualAttachment_Elims {
     id=#37
     possessed_class=FortVisualAttachment
     tracks [{
       type=MovieScene2DTransformTrack
       sections [{
         type=MovieScene2DTransformSection
         end_frame=0
         start_frame=0
         channels [
          Translation.X {type=MovieSceneScriptingFloatChannel, const [0,1]}
          Translation.Y {type=MovieSceneScriptingFloatChannel, const [15,1]}
          Angle {type=MovieSceneScriptingFloatChannel, const [0,1]}
          Scale.X {type=MovieSceneScriptingFloatChannel, const [1,1]}
          Scale.Y {type=MovieSceneScriptingFloatChannel, const [1,1]}
          Shear.X {type=MovieSceneScriptingFloatChannel, const [0,1]}
          Shear.Y {type=MovieSceneScriptingFloatChannel, const [0,1]}
         ]
        }]
      }]
    }
    TextWrapper_RankUp {
     id=#15
     possessed_class=Border
     tracks [{
       type=MovieSceneFloatTrack
       display_name=Render Opacity
       sections [{type=MovieSceneFloatSection, end_frame=119, start_frame=0, channels [None {type=MovieSceneScriptingFloatChannel, const [0,1]}]}]
      }]
    }
    NextBadge {
     id=#18
     possessed_class=Overlay
     tracks [{
       type=MovieSceneFloatTrack
       display_name=Render Opacity
       sections [{type=MovieSceneFloatSection, end_frame=39, start_frame=0, channels [None {type=MovieSceneScriptingFloatChannel, const [0,1]}]}]
      }]
    }
    FortVisualAttachment_Placement {
     id=#38
     possessed_class=FortVisualAttachment
     tracks [{
       type=MovieScene2DTransformTrack
       sections [{
         type=MovieScene2DTransformSection
         end_frame=0
         start_frame=0
         channels [
          Translation.X {type=MovieSceneScriptingFloatChannel, const [0,1]}
          Translation.Y {type=MovieSceneScriptingFloatChannel, const [15,1]}
          Angle {type=MovieSceneScriptingFloatChannel, const [0,1]}
          Scale.X {type=MovieSceneScriptingFloatChannel, const [1,1]}
          Scale.Y {type=MovieSceneScriptingFloatChannel, const [1,1]}
          Shear.X {type=MovieSceneScriptingFloatChannel, const [0,1]}
          Shear.Y {type=MovieSceneScriptingFloatChannel, const [0,1]}
         ]
        }]
      }]
    }
    ProgressBar {
     id=#19
     possessed_class=Image
     tracks [
      {
       type=MovieSceneWidgetMaterialTrack
       sections [{
         type=MovieSceneParameterSection
         end_frame=119
         start_frame=0
         channels [
          Glow Opacity {type=MovieSceneScriptingFloatChannel, keys [[0,0],[39,0],[119,0.25]]}
          Progress Background Opacity {type=MovieSceneScriptingFloatChannel, const [1,1]}
          Override Edge Gradient {type=MovieSceneScriptingFloatChannel, const [1,1]}
         ]
        }]
      }
      {
       type=MovieSceneFloatTrack
       display_name=Render Opacity
       sections [{
         type=MovieSceneFloatSection
         end_frame=119
         start_frame=0
         channels [None {type=MovieSceneScriptingFloatChannel, keys [[0,0],[18,0],[39,1],[119,1]]}]
        }]
      }
      {
       type=MovieScene2DTransformTrack
       sections [{
         type=MovieScene2DTransformSection
         end_frame=39
         start_frame=0
         channels [
          Translation.Y {type=MovieSceneScriptingFloatChannel, const [0,1]}
          Scale.X {type=MovieSceneScriptingFloatChannel, keys [[0,0.25],[18,0.25],[39,1]]}
          Scale.Y {type=MovieSceneScriptingFloatChannel, keys [[0,0.25],[18,0.25],[39,1]]}
         ]
        }]
      }
     ]
    }
    Overlay_TriangleEnergy {
     id=#20
     possessed_class=Overlay
     tracks [{
       type=MovieSceneFloatTrack
       display_name=Render Opacity
       sections [{type=MovieSceneFloatSection, end_frame=119, start_frame=0, channels [None {type=MovieSceneScriptingFloatChannel, const [0,1]}]}]
      }]
    }
    Particles {
     id=#39
     possessed_class=GridPanel
     tracks [{
       type=MovieSceneFloatTrack
       display_name=Render Opacity
       sections [{type=MovieSceneFloatSection, end_frame=0, start_frame=0, channels [None {type=MovieSceneScriptingFloatChannel, const [0,1]}]}]
      }]
    }
    Box_PercentageText {
     id=#40
     display_name=PercentageTextwrapper
     possessed_class=HorizontalBox
     tracks [{
       type=MovieSceneFloatTrack
       display_name=Render Opacity
       sections [{type=MovieSceneFloatSection, end_frame=51, start_frame=0, channels [None {type=MovieSceneScriptingFloatChannel, keys [[0,0],[27,0],[51,1]]}]}]
      }]
    }
    Text_TierName {
     id=#21
     possessed_class=CommonTextBlock
     tracks [{
       type=MovieSceneFloatTrack
       display_name=Render Opacity
       sections [{
         type=MovieSceneFloatSection
         end_frame=119
         start_frame=0
         channels [None {type=MovieSceneScriptingFloatChannel, keys [[0,0],[27,0],[51,1],[119,1]]}]
        }]
      }]
    }
    CurrentBadge_Icon {
     id=#22
     possessed_class=Image
     tracks [{
       type=MovieScene2DTransformTrack
       sections [{
         type=MovieScene2DTransformSection
         end_frame=72
         start_frame=0
         channels [
          Translation.X {type=MovieSceneScriptingFloatChannel, const [8,1]}
          Translation.Y {type=MovieSceneScriptingFloatChannel, const [8,1]}
         ]
        }]
      }]
    }
    WBP_PostMatch_HabaneroSequence_C_4 {
     id=#41
     possessed_class=WBP_PostMatch_HabaneroSequence_C
     tracks [
      {
       type=MovieSceneDoubleTrack
       display_name=Delta Value Blend
       sections [{
         type=MovieSceneDoubleSection
         end_frame=231
         start_frame=0
         channels [None {type=MovieSceneScriptingDoubleChannel, keys [[0,0],[39,0],[231,1]]}]
        }]
      }
      {
       type=MovieSceneDoubleTrack
       display_name=Progress Bar Color Blend
       sections [{type=MovieSceneDoubleSection, end_frame=264, start_frame=0, channels [None {type=MovieSceneScriptingDoubleChannel, const [1,1]}]}]
      }
      {
       type=MovieSceneDoubleTrack
       display_name=Progress Value Blend
       sections [{
         type=MovieSceneDoubleSection
         end_frame=264
         start_frame=0
         channels [None {type=MovieSceneScriptingDoubleChannel, keys [[0,0],[237,0],[264,1]]}]
        }]
      }
     ]
    }
    CurrentBadge {
     id=#24
     possessed_class=Overlay
     tracks [
      {
       type=MovieSceneFloatTrack
       display_name=Render Opacity
       sections [{type=MovieSceneFloatSection, end_frame=39, start_frame=0, channels [None {type=MovieSceneScriptingFloatChannel, keys [[0,0],[9,0],[39,1]]}]}]
      }
      {
       type=MovieScene2DTransformTrack
       sections [{
         type=MovieScene2DTransformSection
         end_frame=72
         start_frame=0
         channels [
          Translation.Y {type=MovieSceneScriptingFloatChannel, const [0,1]}
          Scale.X {type=MovieSceneScriptingFloatChannel, keys [[0,2],[27,0.6],[72,0.6]]}
          Scale.Y {type=MovieSceneScriptingFloatChannel, keys [[0,2],[27,0.6],[72,0.6]]}
         ]
        }]
      }
     ]
    }
    WidgetSwitcherSlot_RankedFeedback {
     id=#42
     possessed_class=CommonAnimatedSwitcher
     tracks [{
       type=MovieSceneFloatTrack
       display_name=Render Opacity
       sections [{type=MovieSceneFloatSection, end_frame=0, start_frame=0, channels [None {type=MovieSceneScriptingFloatChannel, const [1,1]}]}]
      }]
    }
    Background {
     id=#25
     possessed_class=Image
     tracks [{
       type=MovieSceneWidgetMaterialTrack
       sections [{
         type=MovieSceneParameterSection
         end_frame=150
         channels [
          Color Swap {type=MovieSceneScriptingFloatChannel, const [1,1]}
          Hot Spot Intensity {type=MovieSceneScriptingFloatChannel, const [0.5,1]}
          Cloud Opacity {type=MovieSceneScriptingFloatChannel, const [0,1]}
          Triangle Opacity {type=MovieSceneScriptingFloatChannel, keys [[0,0],[150,0.25]]}
          Triangles Tiling {type=MovieSceneScriptingFloatChannel, const [4,1]}
          Rift Intensity {type=MovieSceneScriptingFloatChannel, const [0,1]}
          Rift Cracks Removal {type=MovieSceneScriptingFloatChannel, const [0.6863,1]}
          Invert Radial {type=MovieSceneScriptingFloatChannel, const [0.6,2]}
          Spot Scale {type=MovieSceneScriptingFloatChannel, const [1,2]}
         ]
        }]
      }]
    }
    RadialGlow {
     id=#26
     possessed_class=Image
     tracks [{
       type=MovieSceneFloatTrack
       display_name=Render Opacity
       sections [{type=MovieSceneFloatSection, end_frame=119, start_frame=0, channels [None {type=MovieSceneScriptingFloatChannel, const [0,1]}]}]
      }]
    }
    Img_Flare {
     id=#28
     possessed_class=Image
     tracks [{
       type=MovieSceneWidgetMaterialTrack
       sections [{type=MovieSceneParameterSection, end_frame=119, start_frame=0, channels [Opacity {type=MovieSceneScriptingFloatChannel, const [0,1]}]}]
      }]
    }
   ]
   display_rate {numerator=60}
   tracks [{
     type=MovieSceneEventTrack
     sections [{
       type=MovieSceneEventTriggerSection
       channels [None {
         type=MovieSceneScriptingEventChannel
         keys [[21, "<Struct 'MovieSceneEvent' (0x46761252940) {}>"], [234, "<Struct 'MovieSceneEvent' (0x467612528a0) {}>"]]
        }]
      }]
    }]
  }
 }
 Anim_RankUp_FinalRank {
  binding_count=15
  length_seconds=2.0001
  object=&self:Anim_RankUp_FinalRank
  sequence {
   bindings [
    TextWrapper_RankUp {
     id=#15
     possessed_class=Border
     tracks [
      {
       type=MovieSceneFloatTrack
       display_name=Render Opacity
       sections [{
         type=MovieSceneFloatSection
         end_frame=120
         channels [None {type=MovieSceneScriptingFloatChannel, keys [[0,0],[34.572,0],[50.286,1],[101.976,1],[120.006,0]]}]
        }]
      }
      {
       type=MovieScene2DTransformTrack
       sections [{
         type=MovieScene2DTransformSection
         end_frame=66
         channels [Translation.Y {type=MovieSceneScriptingFloatChannel, keys [[0,32],[34.572,32],[50.286,0]]}]
        }]
      }
     ]
    }
    FortVisualAttachment_Elims {
     id=#43
     possessed_class=FortVisualAttachment
     tracks [{
       type=MovieScene2DTransformTrack
       sections [{
         type=MovieScene2DTransformSection
         end_frame=12
         start_frame=0
         channels [
          Translation.X {type=MovieSceneScriptingFloatChannel}
          Translation.Y {type=MovieSceneScriptingFloatChannel, const [15,1]}
          Angle {type=MovieSceneScriptingFloatChannel}
          Scale.X {type=MovieSceneScriptingFloatChannel}
          Scale.Y {type=MovieSceneScriptingFloatChannel}
          Shear.X {type=MovieSceneScriptingFloatChannel}
          Shear.Y {type=MovieSceneScriptingFloatChannel}
         ]
        }]
      }]
    }
    FortVisualAttachment_Placement {
     id=#44
     possessed_class=FortVisualAttachment
     tracks [{
       type=MovieScene2DTransformTrack
       sections [{
         type=MovieScene2DTransformSection
         end_frame=12
         start_frame=0
         channels [
          Translation.X {type=MovieSceneScriptingFloatChannel}
          Translation.Y {type=MovieSceneScriptingFloatChannel, const [15,1]}
          Angle {type=MovieSceneScriptingFloatChannel}
          Scale.X {type=MovieSceneScriptingFloatChannel}
          Scale.Y {type=MovieSceneScriptingFloatChannel}
          Shear.X {type=MovieSceneScriptingFloatChannel}
          Shear.Y {type=MovieSceneScriptingFloatChannel}
         ]
        }]
      }]
    }
    NextBadge {
     id=#18
     possessed_class=Overlay
     tracks [
      {
       type=MovieSceneFloatTrack
       display_name=Render Opacity
       sections [{type=MovieSceneFloatSection, end_frame=66, channels [None {type=MovieSceneScriptingFloatChannel, keys [[0,0],[3,0],[24,1]]}]}]
      }
      {
       type=MovieScene2DTransformTrack
       sections [{
         type=MovieScene2DTransformSection
         end_frame=66
         channels [
          Translation.Y {type=MovieSceneScriptingFloatChannel, thinned_from=4, keys [[0,82],[24,82],[51,0]]}
          Scale.X {type=MovieSceneScriptingFloatChannel, keys [[0,0.9],[3,0.9],[24,1],[66,1]]}
          Scale.Y {type=MovieSceneScriptingFloatChannel, keys [[0,0.9],[3,0.9],[24,1],[66,1]]}
         ]
        }]
      }
     ]
    }
    ProgressBar {
     id=#19
     possessed_class=Image
     tracks [{
       type=MovieSceneFloatTrack
       display_name=Render Opacity
       sections [{type=MovieSceneFloatSection, end_frame=66, channels [None {type=MovieSceneScriptingFloatChannel, const [0,1]}]}]
      }]
    }
    Box_PercentageText {
     id=#45
     display_name=PercentageTextwrapper
     possessed_class=HorizontalBox
     tracks [
      {
       type=MovieSceneFloatTrack
       display_name=Render Opacity
       sections [{type=MovieSceneFloatSection, end_frame=45, start_frame=0, channels [None {type=MovieSceneScriptingFloatChannel, keys [[0,0],[30,0],[45,1]]}]}]
      }
      {
       type=MovieScene2DTransformTrack
       sections [{
         type=MovieScene2DTransformSection
         end_frame=45
         start_frame=0
         channels [
          Translation.X {type=MovieSceneScriptingFloatChannel}
          Translation.Y {type=MovieSceneScriptingFloatChannel, keys [[0,32],[30,32],[45,0]]}
         ]
        }]
      }
     ]
    }
    Overlay_TriangleEnergy {
     id=#20
     possessed_class=Overlay
     tracks [{
       type=MovieSceneFloatTrack
       display_name=Render Opacity
       sections [{type=MovieSceneFloatSection, channels [None {type=MovieSceneScriptingFloatChannel, const [0,1]}]}]
      }]
    }
    Text_TierName {
     id=#21
     possessed_class=CommonTextBlock
     tracks [
      {
       type=MovieSceneFloatTrack
       display_name=Render Opacity
       sections [{type=MovieSceneFloatSection, end_frame=66, channels [None {type=MovieSceneScriptingFloatChannel, keys [[0,0],[47.144,0],[53.429,1]]}]}]
      }
      {
       type=MovieScene2DTransformTrack
       sections [{
         type=MovieScene2DTransformSection
         end_frame=66
         channels [
          Scale.X {type=MovieSceneScriptingFloatChannel, keys [[0,4],[47.144,4],[56.572,1],[59.715,1.05],[62.858,1]]}
          Scale.Y {type=MovieSceneScriptingFloatChannel, keys [[0,2],[47.144,2],[56.572,1],[59.715,1.05],[62.858,1]]}
         ]
        }]
      }
     ]
    }
    NextBadge_Icon {
     id=#31
     possessed_class=Image
     tracks [{
       type=MovieSceneWidgetMaterialTrack
       sections [{type=MovieSceneParameterSection, channels [Badge Hotness {type=MovieSceneScriptingFloatChannel, keys [[0,4],[3,4],[24,0]]}]}]
      }]
    }
    Box_SkillRating {
     id=#46
     possessed_class=HorizontalBox
     tracks [
      {
       type=MovieSceneFloatTrack
       display_name=Render Opacity
       sections [{
         type=MovieSceneFloatSection
         end_frame=66
         start_frame=6
         channels [None {type=MovieSceneScriptingFloatChannel, thinned_from=4, keys [[6,0],[53.294,0],[65.294,1]]}]
        }]
      }
      {
       type=MovieScene2DTransformTrack
       sections [{
         type=MovieScene2DTransformSection
         end_frame=66
         start_frame=27
         channels [
          Translation.X {type=MovieSceneScriptingFloatChannel}
          Translation.Y {type=MovieSceneScriptingFloatChannel, keys [[28,32],[53.294,32],[65.294,0]]}
         ]
        }]
      }
     ]
    }
    Particles {
     id=#47
     possessed_class=GridPanel
     tracks [{
       type=MovieSceneFloatTrack
       display_name=Render Opacity
       sections [{
         type=MovieSceneFloatSection
         end_frame=66
         start_frame=0
         channels [None {type=MovieSceneScriptingFloatChannel, keys [[0,0],[38.501,0],[66.001,1]]}]
        }]
      }]
    }
    CurrentBadge {
     id=#24
     possessed_class=Overlay
     tracks [{
       type=MovieSceneFloatTrack
       display_name=Render Opacity
       sections [{type=MovieSceneFloatSection, channels [None {type=MovieSceneScriptingFloatChannel, const [0,1]}]}]
      }]
    }
    Background {
     id=#25
     possessed_class=Image
     tracks [{
       type=MovieSceneWidgetMaterialTrack
       sections [{
         type=MovieSceneParameterSection
         channels [
          Triangle Opacity {type=MovieSceneScriptingFloatChannel, const [0.6,1]}
          Invert Radial {type=MovieSceneScriptingFloatChannel, const [1,1]}
          Color Swap {type=MovieSceneScriptingFloatChannel, const [1,1]}
         ]
        }]
      }]
    }
    RadialGlow {
     id=#26
     possessed_class=Image
     tracks [{
       type=MovieSceneFloatTrack
       display_name=Render Opacity
       sections [{type=MovieSceneFloatSection, end_frame=66, channels [None {type=MovieSceneScriptingFloatChannel, const [0,1]}]}]
      }]
    }
    Img_Flare {
     id=#28
     possessed_class=Image
     tracks [{
       type=MovieSceneWidgetMaterialTrack
       sections [{type=MovieSceneParameterSection, end_frame=66, channels [Opacity {type=MovieSceneScriptingFloatChannel, const [0,1]}]}]
      }]
    }
   ]
   display_rate {numerator=60}
   tracks [{
     type=MovieSceneEventTrack
     sections [{
       type=MovieSceneEventTriggerSection
       channels [None {type=MovieSceneScriptingEventChannel, const ["<Struct 'MovieSceneEvent' (0x46761254740) {}>", 1]}]
      }]
    }]
  }
 }
 Anim_GoDown {
  binding_count=15
  length_seconds=1.9834
  object=&self:Anim_GoDown
  sequence {
   bindings [
    TextWrapper_RankUp {
     id=#15
     possessed_class=Border
     tracks [{
       type=MovieSceneFloatTrack
       display_name=Render Opacity
       sections [{type=MovieSceneFloatSection, end_frame=119, start_frame=0, channels [None {type=MovieSceneScriptingFloatChannel, const [0,1]}]}]
      }]
    }
    Particles {
     id=#48
     possessed_class=GridPanel
     tracks [{
       type=MovieSceneFloatTrack
       display_name=Render Opacity
       sections [{type=MovieSceneFloatSection, end_frame=72, start_frame=0, channels [None {type=MovieSceneScriptingFloatChannel, keys [[0,0],[42,0],[72,1]]}]}]
      }]
    }
    FortVisualAttachment_Elims {
     id=#49
     possessed_class=FortVisualAttachment
     tracks [{
       type=MovieScene2DTransformTrack
       sections [{
         type=MovieScene2DTransformSection
         end_frame=0
         start_frame=0
         channels [
          Translation.X {type=MovieSceneScriptingFloatChannel, const [0,1]}
          Translation.Y {type=MovieSceneScriptingFloatChannel, const [15,1]}
          Angle {type=MovieSceneScriptingFloatChannel, const [0,1]}
          Scale.X {type=MovieSceneScriptingFloatChannel, const [1,1]}
          Scale.Y {type=MovieSceneScriptingFloatChannel, const [1,1]}
          Shear.X {type=MovieSceneScriptingFloatChannel, const [0,1]}
          Shear.Y {type=MovieSceneScriptingFloatChannel, const [0,1]}
         ]
        }]
      }]
    }
    Box_PercentageText {
     id=#50
     display_name=PercentageTextwrapper
     possessed_class=HorizontalBox
     tracks [
      {
       type=MovieScene2DTransformTrack
       sections [{type=MovieScene2DTransformSection, channels [Translation.Y {type=MovieSceneScriptingFloatChannel, const [0,1]}]}]
      }
      {
       type=MovieSceneFloatTrack
       display_name=Render Opacity
       sections [{type=MovieSceneFloatSection, end_frame=51, start_frame=0, channels [None {type=MovieSceneScriptingFloatChannel, keys [[0,0],[30,0],[51,1]]}]}]
      }
     ]
    }
    Overlay_TriangleEnergy {
     id=#51
     possessed_class=Overlay
     tracks [{
       type=MovieSceneFloatTrack
       display_name=Render Opacity
       sections [{type=MovieSceneFloatSection, channels [None {type=MovieSceneScriptingFloatChannel, const [0,1]}]}]
      }]
    }
    NextBadge {
     id=#18
     possessed_class=Overlay
     tracks [{
       type=MovieSceneFloatTrack
       display_name=Render Opacity
       sections [{type=MovieSceneFloatSection, end_frame=39, start_frame=0, channels [None {type=MovieSceneScriptingFloatChannel, const [0,1]}]}]
      }]
    }
    ProgressBar {
     id=#19
     possessed_class=Image
     tracks [
      {
       type=MovieSceneWidgetMaterialTrack
       sections [{
         type=MovieSceneParameterSection
         end_frame=119
         start_frame=0
         channels [
          Glow Opacity {type=MovieSceneScriptingFloatChannel, const [0.25,1]}
          Progress Background Opacity {type=MovieSceneScriptingFloatChannel, const [1,1]}
          Override Edge Gradient {type=MovieSceneScriptingFloatChannel, const [1,1]}
         ]
        }]
      }
      {
       type=MovieSceneFloatTrack
       display_name=Render Opacity
       sections [{
         type=MovieSceneFloatSection
         end_frame=119
         start_frame=0
         channels [None {type=MovieSceneScriptingFloatChannel, keys [[0,0],[18,0],[39,1],[119,1]]}]
        }]
      }
      {
       type=MovieScene2DTransformTrack
       sections [{
         type=MovieScene2DTransformSection
         end_frame=39
         start_frame=0
         channels [
          Translation.Y {type=MovieSceneScriptingFloatChannel, const [0,1]}
          Scale.X {type=MovieSceneScriptingFloatChannel, keys [[0,0.25],[18,0.25],[39,1]]}
          Scale.Y {type=MovieSceneScriptingFloatChannel, keys [[0,0.25],[18,0.25],[39,1]]}
         ]
        }]
      }
     ]
    }
    Text_TierName {
     id=#21
     possessed_class=CommonTextBlock
     tracks [
      {
       type=MovieSceneFloatTrack
       display_name=Render Opacity
       sections [{
         type=MovieSceneFloatSection
         end_frame=119
         start_frame=0
         channels [None {type=MovieSceneScriptingFloatChannel, keys [[0,0],[27,0],[51,1],[119,1]]}]
        }]
      }
      {
       type=MovieScene2DTransformTrack
       sections [{type=MovieScene2DTransformSection, channels [Translation.Y {type=MovieSceneScriptingFloatChannel, const [0,1]}]}]
      }
     ]
    }
    WBP_PostMatch_HabaneroSequence_C_3 {
     id=#52
     possessed_class=WBP_PostMatch_HabaneroSequence_C
     tracks [
      {
       type=MovieSceneDoubleTrack
       display_name=Delta Value Blend
       sections [{type=MovieSceneDoubleSection, end_frame=104, start_frame=0, channels [None {type=MovieSceneScriptingDoubleChannel, const [0,1]}]}]
      }
      {
       type=MovieSceneDoubleTrack
       display_name=Progress Bar Color Blend
       sections [{type=MovieSceneDoubleSection, end_frame=119, start_frame=0, channels [None {type=MovieSceneScriptingDoubleChannel, const [1,1]}]}]
      }
      {
       type=MovieSceneDoubleTrack
       display_name=Progress Value Blend
       sections [{type=MovieSceneDoubleSection, end_frame=119, start_frame=0, channels [None {type=MovieSceneScriptingDoubleChannel, const [0,1]}]}]
      }
     ]
    }
    CurrentBadge_Icon {
     id=#22
     possessed_class=Image
     tracks [{
       type=MovieScene2DTransformTrack
       sections [{
         type=MovieScene2DTransformSection
         end_frame=72
         start_frame=0
         channels [
          Translation.X {type=MovieSceneScriptingFloatChannel, const [8,1]}
          Translation.Y {type=MovieSceneScriptingFloatChannel, const [8,1]}
         ]
        }]
      }]
    }
    FortVisualAttachment_Placement {
     id=#53
     possessed_class=FortVisualAttachment
     tracks [{
       type=MovieScene2DTransformTrack
       sections [{
         type=MovieScene2DTransformSection
         end_frame=0
         start_frame=0
         channels [
          Translation.X {type=MovieSceneScriptingFloatChannel, const [0,1]}
          Translation.Y {type=MovieSceneScriptingFloatChannel, const [15,1]}
          Angle {type=MovieSceneScriptingFloatChannel, const [0,1]}
          Scale.X {type=MovieSceneScriptingFloatChannel, const [1,1]}
          Scale.Y {type=MovieSceneScriptingFloatChannel, const [1,1]}
          Shear.X {type=MovieSceneScriptingFloatChannel, const [0,1]}
          Shear.Y {type=MovieSceneScriptingFloatChannel, const [0,1]}
         ]
        }]
      }]
    }
    CurrentBadge {
     id=#24
     possessed_class=Overlay
     tracks [
      {
       type=MovieSceneFloatTrack
       display_name=Render Opacity
       sections [{type=MovieSceneFloatSection, end_frame=39, start_frame=0, channels [None {type=MovieSceneScriptingFloatChannel, keys [[0,0],[9,0],[39,1]]}]}]
      }
      {
       type=MovieScene2DTransformTrack
       sections [{
         type=MovieScene2DTransformSection
         end_frame=72
         start_frame=0
         channels [
          Translation.Y {type=MovieSceneScriptingFloatChannel, const [0,1]}
          Scale.X {type=MovieSceneScriptingFloatChannel, keys [[0,2],[27,0.6],[72,0.6]]}
          Scale.Y {type=MovieSceneScriptingFloatChannel, keys [[0,2],[27,0.6],[72,0.6]]}
         ]
        }]
      }
     ]
    }
    Background {
     id=#25
     possessed_class=Image
     tracks [{
       type=MovieSceneWidgetMaterialTrack
       sections [{
         type=MovieSceneParameterSection
         end_frame=120
         channels [
          Color Swap {type=MovieSceneScriptingFloatChannel, const [1,1]}
          Hot Spot Intensity {type=MovieSceneScriptingFloatChannel, const [0.5,1]}
          Cloud Opacity {type=MovieSceneScriptingFloatChannel, const [0,1]}
          Triangle Opacity {type=MovieSceneScriptingFloatChannel, keys [[0,0],[119,0.25]]}
          Triangles Tiling {type=MovieSceneScriptingFloatChannel, const [4,1]}
          Rift Intensity {type=MovieSceneScriptingFloatChannel, const [0,1]}
          Rift Cracks Removal {type=MovieSceneScriptingFloatChannel, const [0.6863,1]}
          Invert Radial {type=MovieSceneScriptingFloatChannel, const [0.6,2]}
          Spot Scale {type=MovieSceneScriptingFloatChannel, const [1,2]}
         ]
        }]
      }]
    }
    RadialGlow {
     id=#26
     possessed_class=Image
     tracks [{
       type=MovieSceneFloatTrack
       display_name=Render Opacity
       sections [{type=MovieSceneFloatSection, end_frame=119, start_frame=0, channels [None {type=MovieSceneScriptingFloatChannel, const [0,1]}]}]
      }]
    }
    Img_Flare {
     id=#28
     possessed_class=Image
     tracks [{
       type=MovieSceneWidgetMaterialTrack
       sections [{type=MovieSceneParameterSection, end_frame=119, start_frame=0, channels [Opacity {type=MovieSceneScriptingFloatChannel, const [0,1]}]}]
      }]
    }
   ]
   display_rate {numerator=60}
   tracks [{type=MovieSceneEventTrack, sections [{type=MovieSceneEventTriggerSection, channels [None {type=MovieSceneScriptingEventChannel}]}]}]
  }
 }
 Anim_ShowStats_OnSequenceFinished {
  binding_count=3
  length_seconds=0.5
  object=&self:Anim_ShowStats_OnSequenceFinished
  sequence {
   bindings [
    StatsLeftContainer_VerticalBox {
     id=#54
     possessed_class=VerticalBox
     tracks [{
       type=MovieSceneFloatTrack
       display_name=Render Opacity
       sections [{type=MovieSceneFloatSection, channels [None {type=MovieSceneScriptingFloatChannel, keys [[0,0],[30,1]]}]}]
      }]
    }
    WidgetSwitcherSlot_RankedFeedback {
     id=#55
     possessed_class=CommonAnimatedSwitcher
     tracks [{
       type=MovieSceneFloatTrack
       display_name=Render Opacity
       sections [{type=MovieSceneFloatSection, end_frame=30, start_frame=0, channels [None {type=MovieSceneScriptingFloatChannel, const [1,2]}]}]
      }]
    }
    StatRightContainer_VerticalBox {
     id=#56
     possessed_class=VerticalBox
     tracks [{
       type=MovieSceneFloatTrack
       display_name=Render Opacity
       sections [{type=MovieSceneFloatSection, channels [None {type=MovieSceneScriptingFloatChannel, keys [[0,0],[30,1]]}]}]
      }]
    }
   ]
   display_rate {numerator=60}
  }
 }
 Anim_OnSequenceFinished {
  binding_count=2
  length_seconds=0.5
  object=&self:Anim_OnSequenceFinished
  sequence {bindings [
    HorizontalBox_RankDelta {
     id=#57
     possessed_class=HorizontalBox
     tracks [
      {
       type=MovieSceneFloatTrack
       display_name=Render Opacity
       sections [{type=MovieSceneFloatSection, channels [None {type=MovieSceneScriptingFloatChannel, keys [[0,0],[10,1]]}]}]
      }
      {
       type=MovieSceneByteTrack
       display_name=Visibility
       sections [{type=MovieSceneByteSection, end_frame=0, start_frame=0, channels [None {type=MovieSceneScriptingByteChannel, const [3,1]}]}]
      }
     ]
    }
    WBP_TournamentSmallBanner {
     id=#58
     possessed_class=WBP_TournamentSmallBanner_C
     tracks [{
       type=MovieSceneFloatTrack
       display_name=Render Opacity
       sections [{type=MovieSceneFloatSection, end_frame=10, start_frame=0, channels [None {type=MovieSceneScriptingFloatChannel, keys [[0,0],[6,1],[10,1]]}]}]
      }]
    }
   ]}
 }
 Anim_Stats_Intro {
  binding_count=3
  length_seconds=0.5
  object=&self:Anim_Stats_Intro
  sequence {
   bindings [
    StatRightContainer_VerticalBox {
     id=#59
     possessed_class=VerticalBox
     tracks [
      {
       type=MovieSceneFloatTrack
       display_name=Render Opacity
       sections [{type=MovieSceneFloatSection, channels [None {type=MovieSceneScriptingFloatChannel, keys [[0,0],[30,1]]}]}]
      }
      {
       type=MovieScene2DTransformTrack
       sections [{
         type=MovieScene2DTransformSection
         end_frame=30
         start_frame=0
         channels [
          Translation.X {type=MovieSceneScriptingFloatChannel}
          Translation.Y {type=MovieSceneScriptingFloatChannel, keys [[0,64],[30,0]]}
          Angle {type=MovieSceneScriptingFloatChannel}
          Scale.X {type=MovieSceneScriptingFloatChannel}
          Scale.Y {type=MovieSceneScriptingFloatChannel}
          Shear.X {type=MovieSceneScriptingFloatChannel}
          Shear.Y {type=MovieSceneScriptingFloatChannel}
         ]
        }]
      }
     ]
    }
    WidgetSwitcherSlot_RankedFeedback {
     id=#60
     possessed_class=CommonAnimatedSwitcher
     tracks [{
       type=MovieSceneFloatTrack
       display_name=Render Opacity
       sections [{type=MovieSceneFloatSection, end_frame=0, start_frame=0, channels [None {type=MovieSceneScriptingFloatChannel, const [1,1]}]}]
      }]
    }
    StatsLeftContainer_VerticalBox {
     id=#61
     possessed_class=VerticalBox
     tracks [
      {
       type=MovieSceneFloatTrack
       display_name=Render Opacity
       sections [{type=MovieSceneFloatSection, channels [None {type=MovieSceneScriptingFloatChannel, const [1,2]}]}]
      }
      {
       type=MovieScene2DTransformTrack
       sections [{
         type=MovieScene2DTransformSection
         end_frame=30
         start_frame=0
         channels [
          Translation.X {type=MovieSceneScriptingFloatChannel}
          Translation.Y {type=MovieSceneScriptingFloatChannel, keys [[0,64],[30,0]]}
          Angle {type=MovieSceneScriptingFloatChannel}
          Scale.X {type=MovieSceneScriptingFloatChannel}
          Scale.Y {type=MovieSceneScriptingFloatChannel}
          Shear.X {type=MovieSceneScriptingFloatChannel}
          Shear.Y {type=MovieSceneScriptingFloatChannel}
         ]
        }]
      }
     ]
    }
   ]
   display_rate {numerator=60}
  }
 }
 Anim_Stats_Hide {
  binding_count=3
  length_seconds=0.3
  object=&self:Anim_Stats_Hide
  sequence {
   bindings [
    StatRightContainer_VerticalBox {
     id=#62
     possessed_class=VerticalBox
     tracks [{
       type=MovieSceneFloatTrack
       display_name=Render Opacity
       sections [{type=MovieSceneFloatSection, channels [None {type=MovieSceneScriptingFloatChannel, keys [[0,1],[18,0]]}]}]
      }]
    }
    StatsLeftContainer_VerticalBox {
     id=#63
     possessed_class=VerticalBox
     tracks [{
       type=MovieSceneFloatTrack
       display_name=Render Opacity
       sections [{type=MovieSceneFloatSection, channels [None {type=MovieSceneScriptingFloatChannel, keys [[0,1],[18,0]]}]}]
      }]
    }
    WidgetSwitcherSlot_RankedFeedback {
     id=#64
     possessed_class=CommonAnimatedSwitcher
     tracks [{
       type=MovieSceneFloatTrack
       display_name=Render Opacity
       sections [{type=MovieSceneFloatSection, end_frame=18, start_frame=0, channels [None {type=MovieSceneScriptingFloatChannel, keys [[0,1],[18,0]]}]}]
      }]
    }
   ]
   display_rate {numerator=60}
  }
 }
 Anim_ShowHideSpinner {
  binding_count=1
  length_seconds=0.64
  object=&self:Anim_ShowHideSpinner
  sequence {bindings [Image_WaitingSpinner {
     id=#65
     possessed_class=Image
     tracks [
      {
       type=MovieSceneWidgetMaterialTrack
       sections [{
         type=MovieSceneParameterSection
         channels [
          Solid_Animate {type=MovieSceneScriptingFloatChannel, keys [[6.4003,1],[12.8,0]]}
          Fill_Animate {type=MovieSceneScriptingFloatChannel, keys [[6.4003,1],[12.8,0]]}
          Speed {type=MovieSceneScriptingFloatChannel, const [1,1]}
         ]
        }]
      }
      {
       type=MovieSceneFloatTrack
       display_name=Render Opacity
       sections [{type=MovieSceneFloatSection, channels [None {type=MovieSceneScriptingFloatChannel, keys [[1.6,0],[6.4,1]]}]}]
      }
      {
       type=MovieScene2DTransformTrack
       sections [{
         type=MovieScene2DTransformSection
         channels [
          Scale.X {type=MovieSceneScriptingFloatChannel, keys [[1.6,0.875],[6.4003,1]]}
          Scale.Y {type=MovieSceneScriptingFloatChannel, keys [[1.6,0.875],[6.4003,1]]}
         ]
        }]
      }
      {
       type=MovieSceneByteTrack
       display_name=Visibility
       sections [{type=MovieSceneByteSection, channels [None {type=MovieSceneScriptingByteChannel, keys [[0,1],[1.6,3]]}]}]
      }
     ]
    }]}
 }
 Anim_TimeLineIntro {
  binding_count=1
  length_seconds=0.5
  object=&self:Anim_TimeLineIntro
  sequence {bindings [Overlay_EliminationTimeline {
     id=#66
     possessed_class=Overlay
     tracks [{
       type=MovieSceneFloatTrack
       display_name=Render Opacity
       sections [{type=MovieSceneFloatSection, end_frame=10, start_frame=0, channels [None {type=MovieSceneScriptingFloatChannel, keys [[0,0],[10,1]]}]}]
      }]
    }]}
 }
 Anim_TimeLineProgress {
  binding_count=1
  length_seconds=1
  object=&self:Anim_TimeLineProgress
  sequence {bindings [Timeline {
     id=#67
     possessed_class=Image
     tracks [{
       type=MovieSceneWidgetMaterialTrack
       sections [{
         type=MovieSceneParameterSection
         end_frame=20
         start_frame=0
         channels [AnimateProgress {type=MovieSceneScriptingFloatChannel, keys [[0,0],[20,1]]}]
        }]
      }]
    }]}
 }
 Anim_TimeLineFadeOut {
  binding_count=1
  length_seconds=1
  object=&self:Anim_TimeLineFadeOut
  sequence {bindings [Overlay_EliminationTimeline {
     id=#68
     possessed_class=Overlay
     tracks [{
       type=MovieSceneFloatTrack
       display_name=Render Opacity
       sections [{type=MovieSceneFloatSection, end_frame=20, start_frame=0, channels [None {type=MovieSceneScriptingFloatChannel, keys [[0,1],[20,0]]}]}]
      }]
    }]}
 }
 Anim_GoDownProtected {
  binding_count=18
  length_seconds=3.5
  object=&self:Anim_GoDownProtected
  sequence {
   bindings [
    RankProtectedBox_Switcher {
     id=#69
     possessed_class=CommonAnimatedSwitcher
     tracks [{
       type=MovieSceneIntegerTrack
       display_name=Active Widget Index
       sections [{type=MovieSceneIntegerSection, end_frame=210, start_frame=0, channels [None {type=MovieSceneScriptingIntegerChannel, keys [[0,1],[201,0]]}]}]
      }]
    }
    TextWrapper_RankUp {
     id=#15
     possessed_class=Border
     tracks [{
       type=MovieSceneFloatTrack
       display_name=Render Opacity
       sections [{type=MovieSceneFloatSection, end_frame=119, start_frame=0, channels [None {type=MovieSceneScriptingFloatChannel, const [0,1]}]}]
      }]
    }
    Particles {
     id=#70
     possessed_class=GridPanel
     tracks [{
       type=MovieSceneFloatTrack
       display_name=Render Opacity
       sections [{type=MovieSceneFloatSection, end_frame=72, start_frame=0, channels [None {type=MovieSceneScriptingFloatChannel, keys [[0,0],[42,0],[72,1]]}]}]
      }]
    }
    FortVisualAttachment_Elims {
     id=#49
     possessed_class=FortVisualAttachment
     tracks [{
       type=MovieScene2DTransformTrack
       sections [{
         type=MovieScene2DTransformSection
         end_frame=0
         start_frame=0
         channels [
          Translation.X {type=MovieSceneScriptingFloatChannel, const [0,1]}
          Translation.Y {type=MovieSceneScriptingFloatChannel, const [15,1]}
          Angle {type=MovieSceneScriptingFloatChannel, const [0,1]}
          Scale.X {type=MovieSceneScriptingFloatChannel, const [1,1]}
          Scale.Y {type=MovieSceneScriptingFloatChannel, const [1,1]}
          Shear.X {type=MovieSceneScriptingFloatChannel, const [0,1]}
          Shear.Y {type=MovieSceneScriptingFloatChannel, const [0,1]}
         ]
        }]
      }]
    }
    Box_PercentageText {
     id=#50
     display_name=PercentageTextwrapper
     possessed_class=HorizontalBox
     tracks [
      {
       type=MovieScene2DTransformTrack
       sections [{type=MovieScene2DTransformSection, channels [Translation.Y {type=MovieSceneScriptingFloatChannel, const [0,1]}]}]
      }
      {
       type=MovieSceneFloatTrack
       display_name=Render Opacity
       sections [{type=MovieSceneFloatSection, end_frame=51, start_frame=0, channels [None {type=MovieSceneScriptingFloatChannel, keys [[0,0],[30,0],[51,1]]}]}]
      }
     ]
    }
    Text_RankProtected {
     id=#71
     possessed_class=CommonTextBlock
     tracks [{
       type=MovieSceneFloatTrack
       display_name=Render Opacity
       sections [{type=MovieSceneFloatSection, end_frame=119, start_frame=0, channels [None {type=MovieSceneScriptingFloatChannel, keys [[15,0],[27,1]]}]}]
      }]
    }
    NextBadge {
     id=#18
     possessed_class=Overlay
     tracks [{
       type=MovieSceneFloatTrack
       display_name=Render Opacity
       sections [{type=MovieSceneFloatSection, end_frame=39, start_frame=0, channels [None {type=MovieSceneScriptingFloatChannel, const [0,1]}]}]
      }]
    }
    ProgressBar {
     id=#19
     possessed_class=Image
     tracks [
      {
       type=MovieSceneWidgetMaterialTrack
       sections [{
         type=MovieSceneParameterSection
         end_frame=119
         start_frame=0
         channels [
          Glow Opacity {type=MovieSceneScriptingFloatChannel, const [0.25,1]}
          Progress Background Opacity {type=MovieSceneScriptingFloatChannel, const [1,1]}
          Override Edge Gradient {type=MovieSceneScriptingFloatChannel, const [1,1]}
         ]
        }]
      }
      {
       type=MovieSceneFloatTrack
       display_name=Render Opacity
       sections [{
         type=MovieSceneFloatSection
         end_frame=119
         start_frame=0
         channels [None {type=MovieSceneScriptingFloatChannel, keys [[0,0],[18,0],[39,1],[119,1]]}]
        }]
      }
      {
       type=MovieScene2DTransformTrack
       sections [{
         type=MovieScene2DTransformSection
         end_frame=39
         start_frame=0
         channels [
          Translation.Y {type=MovieSceneScriptingFloatChannel, const [0,1]}
          Scale.X {type=MovieSceneScriptingFloatChannel, keys [[0,0.25],[18,0.25],[39,1]]}
          Scale.Y {type=MovieSceneScriptingFloatChannel, keys [[0,0.25],[18,0.25],[39,1]]}
         ]
        }]
      }
     ]
    }
    Text_TierName {
     id=#21
     possessed_class=CommonTextBlock
     tracks [
      {
       type=MovieSceneFloatTrack
       display_name=Render Opacity
       sections [{
         type=MovieSceneFloatSection
         end_frame=119
         start_frame=0
         channels [None {type=MovieSceneScriptingFloatChannel, keys [[0,0],[27,0],[51,1],[119,1]]}]
        }]
      }
      {
       type=MovieScene2DTransformTrack
       sections [{type=MovieScene2DTransformSection, channels [Translation.Y {type=MovieSceneScriptingFloatChannel, const [0,1]}]}]
      }
     ]
    }
    Overlay_TriangleEnergy {
     id=#72
     possessed_class=Overlay
     tracks [{
       type=MovieSceneFloatTrack
       display_name=Render Opacity
       sections [{type=MovieSceneFloatSection, channels [None {type=MovieSceneScriptingFloatChannel, const [0,1]}]}]
      }]
    }
    WBP_PostMatch_HabaneroSequence_C_3 {
     id=#52
     possessed_class=WBP_PostMatch_HabaneroSequence_C
     tracks [
      {
       type=MovieSceneDoubleTrack
       display_name=Delta Value Blend
       sections [{type=MovieSceneDoubleSection, end_frame=104, start_frame=0, channels [None {type=MovieSceneScriptingDoubleChannel, const [0,1]}]}]
      }
      {
       type=MovieSceneDoubleTrack
       display_name=Progress Bar Color Blend
       sections [{type=MovieSceneDoubleSection, end_frame=119, start_frame=0, channels [None {type=MovieSceneScriptingDoubleChannel, const [1,1]}]}]
      }
      {
       type=MovieSceneDoubleTrack
       display_name=Progress Value Blend
       sections [{type=MovieSceneDoubleSection, end_frame=119, start_frame=0, channels [None {type=MovieSceneScriptingDoubleChannel, const [0,1]}]}]
      }
     ]
    }
    CurrentBadge_Icon {
     id=#22
     possessed_class=Image
     tracks [{
       type=MovieScene2DTransformTrack
       sections [{
         type=MovieScene2DTransformSection
         end_frame=72
         start_frame=0
         channels [
          Translation.X {type=MovieSceneScriptingFloatChannel, const [8,1]}
          Translation.Y {type=MovieSceneScriptingFloatChannel, const [8,1]}
         ]
        }]
      }]
    }
    FortVisualAttachment_Placement {
     id=#53
     possessed_class=FortVisualAttachment
     tracks [{
       type=MovieScene2DTransformTrack
       sections [{
         type=MovieScene2DTransformSection
         end_frame=0
         start_frame=0
         channels [
          Translation.X {type=MovieSceneScriptingFloatChannel, const [0,1]}
          Translation.Y {type=MovieSceneScriptingFloatChannel, const [15,1]}
          Angle {type=MovieSceneScriptingFloatChannel, const [0,1]}
          Scale.X {type=MovieSceneScriptingFloatChannel, const [1,1]}
          Scale.Y {type=MovieSceneScriptingFloatChannel, const [1,1]}
          Shear.X {type=MovieSceneScriptingFloatChannel, const [0,1]}
          Shear.Y {type=MovieSceneScriptingFloatChannel, const [0,1]}
         ]
        }]
      }]
    }
    CurrentBadge {
     id=#24
     possessed_class=Overlay
     tracks [
      {
       type=MovieSceneFloatTrack
       display_name=Render Opacity
       sections [{type=MovieSceneFloatSection, end_frame=39, start_frame=0, channels [None {type=MovieSceneScriptingFloatChannel, keys [[0,0],[9,0],[39,1]]}]}]
      }
      {
       type=MovieScene2DTransformTrack
       sections [{
         type=MovieScene2DTransformSection
         end_frame=72
         start_frame=0
         channels [
          Translation.Y {type=MovieSceneScriptingFloatChannel, const [0,1]}
          Scale.X {type=MovieSceneScriptingFloatChannel, keys [[0,2],[27,0.6],[72,0.6]]}
          Scale.Y {type=MovieSceneScriptingFloatChannel, keys [[0,2],[27,0.6],[72,0.6]]}
         ]
        }]
      }
     ]
    }
    Background {
     id=#25
     possessed_class=Image
     tracks [{
       type=MovieSceneWidgetMaterialTrack
       sections [{
         type=MovieSceneParameterSection
         end_frame=120
         channels [
          Color Swap {type=MovieSceneScriptingFloatChannel, const [1,1]}
          Hot Spot Intensity {type=MovieSceneScriptingFloatChannel, const [0.5,1]}
          Cloud Opacity {type=MovieSceneScriptingFloatChannel, const [0,1]}
          Triangle Opacity {type=MovieSceneScriptingFloatChannel, keys [[0,0],[119,0.25]]}
          Triangles Tiling {type=MovieSceneScriptingFloatChannel, const [4,1]}
          Rift Intensity {type=MovieSceneScriptingFloatChannel, const [0,1]}
          Rift Cracks Removal {type=MovieSceneScriptingFloatChannel, const [0.6863,1]}
          Invert Radial {type=MovieSceneScriptingFloatChannel, const [0.6,2]}
          Spot Scale {type=MovieSceneScriptingFloatChannel, const [1,2]}
         ]
        }]
      }]
    }
    RadialGlow {
     id=#26
     possessed_class=Image
     tracks [
      {
       type=MovieSceneFloatTrack
       display_name=Render Opacity
       sections [{
         type=MovieSceneFloatSection
         end_frame=119
         start_frame=0
         channels [None {type=MovieSceneScriptingFloatChannel, keys [[0,0],[14,0.4672],[23,0.75],[30,1]]}]
        }]
      }
      {
       type=MovieSceneWidgetMaterialTrack
       sections [{
         type=MovieSceneParameterSection
         end_frame=210
         start_frame=0
         channels [
          HexGridOpacity {type=MovieSceneScriptingFloatChannel, keys [[14,0],[23,0.02],[40,1],[180,1],[194,0]]}
          Opacity {type=MovieSceneScriptingFloatChannel, const [0.5,1]}
          Glow ColorR {type=MovieSceneScriptingFloatChannel, keys [[14,0],[23,0.02],[30,0]]}
          Glow ColorG {type=MovieSceneScriptingFloatChannel, keys [[14,0.3],[23,0.5],[30,0.1]]}
          Glow ColorB {type=MovieSceneScriptingFloatChannel, const [1,3]}
          Glow ColorA {type=MovieSceneScriptingFloatChannel, const [1,3]}
         ]
        }]
      }
      {
       type=MovieScene2DTransformTrack
       sections [{
         type=MovieScene2DTransformSection
         end_frame=210
         start_frame=0
         channels [
          Translation.X {type=MovieSceneScriptingFloatChannel, const [0,1]}
          Translation.Y {type=MovieSceneScriptingFloatChannel, keys [[0,12],[199,12],[205,0]]}
         ]
        }]
      }
     ]
    }
    Img_Flare {
     id=#28
     possessed_class=Image
     tracks [{
       type=MovieSceneWidgetMaterialTrack
       sections [{type=MovieSceneParameterSection, end_frame=119, start_frame=0, channels [Opacity {type=MovieSceneScriptingFloatChannel, const [0,1]}]}]
      }]
    }
    WBP_Habanero_Shield {
     id=#73
     possessed_class=WBP_Habanero_Shield_C
     tracks [
      {
       type=MovieSceneFloatTrack
       display_name=Render Opacity
       sections [{type=MovieSceneFloatSection, end_frame=210, start_frame=0, channels [None {type=MovieSceneScriptingFloatChannel, keys [[0,0],[33,1]]}]}]
      }
      {
       type=MovieSceneEventTrack
       sections [{
         type=MovieSceneEventTriggerSection
         channels [None {type=MovieSceneScriptingEventChannel, const ["<Struct 'MovieSceneEvent' (0x4676125b9a0) {}>", 1]}]
        }]
      }
     ]
    }
   ]
   display_rate {numerator=60}
   tracks [{type=MovieSceneEventTrack, sections [{type=MovieSceneEventTriggerSection, channels [None {type=MovieSceneScriptingEventChannel}]}]}]
  }
 }
 Anim_BonusFeedback_Elim {
  binding_count=2
  length_seconds=1.1042
  object=&self:Anim_BonusFeedback_Elim
  sequence {
   bindings [
    ProgressBar {
     id=#74
     possessed_class=Image
     tracks [{
       type=MovieSceneWidgetMaterialTrack
       sections [{
         type=MovieSceneParameterSection
         end_frame=48
         start_frame=0
         channels [
          Glow Range {type=MovieSceneScriptingFloatChannel, keys [[0,0.17],[8.104,0],[13.0912,0.17],[40.5192,0],[48,0.17]]}
          Bar ColorR {type=MovieSceneScriptingFloatChannel, keys [[0,0.0075],[10.5976,0.6],[32.9976,0.6],[48,0.0075]]}
          Bar ColorG {type=MovieSceneScriptingFloatChannel, keys [[0,0.4233],[10.5976,0],[32.9976,0],[48,0.4233]]}
          Bar ColorB {type=MovieSceneScriptingFloatChannel, keys [[0,0.8632],[10.5976,0],[32.9976,0],[48,0.8632]]}
          Bar ColorA {type=MovieSceneScriptingFloatChannel, const [1,2]}
          Bar Glow ColorR {type=MovieSceneScriptingFloatChannel, keys [[0,0],[10.5976,0.7391],[32.9976,0.05],[48,0]]}
          Bar Glow ColorG {type=MovieSceneScriptingFloatChannel, keys [[0,0.3014],[10.5976,0],[32.9976,0],[48,0.3014]]}
          Bar Glow ColorB {type=MovieSceneScriptingFloatChannel, keys [[0,1],[10.5976,0],[32.9976,0],[48,1]]}
          Bar Glow ColorA {type=MovieSceneScriptingFloatChannel, const [1,4]}
          Bar Outline ColorR {type=MovieSceneScriptingFloatChannel, keys [[0,0],[10.5976,0.7391],[32.9976,0.05],[48,0]]}
          Bar Outline ColorG {type=MovieSceneScriptingFloatChannel, keys [[0,0.3014],[10.5976,0],[32.9976,0],[48,0.3014]]}
          Bar Outline ColorB {type=MovieSceneScriptingFloatChannel, keys [[0,1],[10.5976,0],[32.9976,0],[48,1]]}
          Bar Outline ColorA {type=MovieSceneScriptingFloatChannel, const [1,4]}
          Progress Background ColorR {type=MovieSceneScriptingFloatChannel, keys [[0,0],[10.5976,0.7391],[32.9976,0.05],[48,0]]}
          Progress Background ColorG {type=MovieSceneScriptingFloatChannel, keys [[0,0.3014],[10.5976,0],[32.9976,0],[48,0.3014]]}
          Progress Background ColorB {type=MovieSceneScriptingFloatChannel, keys [[0,1],[10.5976,0],[32.9976,0],[48,1]]}
          Progress Background ColorA {type=MovieSceneScriptingFloatChannel, const [1,4]}
          Inner ColorR {type=MovieSceneScriptingFloatChannel, keys [[0,0],[10.5976,0.7391],[32.9976,0.05],[48,0]]}
          Inner ColorG {type=MovieSceneScriptingFloatChannel, keys [[0,0.3014],[10.5976,0],[32.9976,0],[48,0.3014]]}
          Inner ColorB {type=MovieSceneScriptingFloatChannel, keys [[0,1],[10.5976,0],[32.9976,0],[48,1]]}
          Inner ColorA {type=MovieSceneScriptingFloatChannel, const [1,4]}
         ]
        }]
      }]
    }
    Image_ElimGlow {
     id=#75
     possessed_class=Image
     tracks [
      {
       type=MovieSceneWidgetMaterialTrack
       sections [{type=MovieSceneParameterSection, channels [Glow Y Movement {type=MovieSceneScriptingFloatChannel, keys [[0,1.4],[29.6,-0.5]]}]}]
      }
      {
       type=MovieSceneFloatTrack
       display_name=Render Opacity
       sections [{type=MovieSceneFloatSection, channels [None {type=MovieSceneScriptingFloatChannel, keys [[0,0],[2.4,1],[30.4,1],[33.6,0],[53,0]]}]}]
      }
     ]
    }
   ]
   display_rate {numerator=48}
  }
 }
 Anim_BonusFeedback_Elim_V02 {
  binding_count=3
  length_seconds=1.9833
  object=&self:Anim_BonusFeedback_Elim_V02
  sequence {
   bindings [
    Image_ElimGlow {
     id=#76
     possessed_class=Image
     tracks [
      {
       type=MovieSceneWidgetMaterialTrack
       sections [{type=MovieSceneParameterSection, channels [Glow Y Movement {type=MovieSceneScriptingFloatChannel, keys [[0,1.4],[37,-0.5]]}]}]
      }
      {
       type=MovieSceneFloatTrack
       display_name=Render Opacity
       sections [{type=MovieSceneFloatSection, channels [None {type=MovieSceneScriptingFloatChannel, keys [[0,0],[37,0.5],[66.25,0]]}]}]
      }
     ]
    }
    RadialGlow {
     id=#26
     possessed_class=Image
     tracks [{
       type=MovieSceneFloatTrack
       display_name=Render Opacity
       sections [{type=MovieSceneFloatSection, end_frame=119, start_frame=0, channels [None {type=MovieSceneScriptingFloatChannel, const [0,1]}]}]
      }]
    }
    Img_Flare {
     id=#28
     possessed_class=Image
     tracks [{
       type=MovieSceneWidgetMaterialTrack
       sections [{type=MovieSceneParameterSection, end_frame=119, start_frame=0, channels [Opacity {type=MovieSceneScriptingFloatChannel, const [0,1]}]}]
      }]
    }
   ]
   display_rate {numerator=60}
  }
 }
 Anim_BonusFeedback_Placement {
  binding_count=2
  length_seconds=1.1042
  object=&self:Anim_BonusFeedback_Placement
  sequence {
   bindings [
    ProgressBar {
     id=#74
     possessed_class=Image
     tracks [{
       type=MovieSceneWidgetMaterialTrack
       sections [{
         type=MovieSceneParameterSection
         end_frame=48
         start_frame=0
         channels [
          Glow Range {type=MovieSceneScriptingFloatChannel, keys [[0,0.17],[8.104,0],[13.0912,0.17],[40.5192,0],[48,0.17]]}
          Bar ColorR {type=MovieSceneScriptingFloatChannel, default=0.0075, keys [[0,0.0075],[10.5976,0.15],[32.8,0.15],[48,0.0075]]}
          Bar ColorG {type=MovieSceneScriptingFloatChannel, default=0.4233, keys [[0,0.4233],[10.5976,0.3],[32.8,0.3],[48,0.4233]]}
          Bar ColorB {type=MovieSceneScriptingFloatChannel, default=0.8632, keys [[0,0.8632],[10.5976,1],[32.8,1],[48,0.8632]]}
          Bar ColorA {type=MovieSceneScriptingFloatChannel, default=1, const [1,2]}
          Bar Glow ColorR {type=MovieSceneScriptingFloatChannel, default=0, keys [[0,0],[10.5976,0.2],[32.8,0.2],[48,0]]}
          Bar Glow ColorG {type=MovieSceneScriptingFloatChannel, default=0.3014, keys [[0,0.3014],[10.5976,0.3],[32.8,0.3],[48,0.3014]]}
          Bar Glow ColorB {type=MovieSceneScriptingFloatChannel, default=1, const [1,4]}
          Bar Glow ColorA {type=MovieSceneScriptingFloatChannel, default=1, const [1,4]}
          Bar Outline ColorR {type=MovieSceneScriptingFloatChannel, default=0, thinned_from=5, keys [[0,0.0622],[10.4,0],[32.8,0],[48,0.0622]]}
          Bar Outline ColorG {type=MovieSceneScriptingFloatChannel, default=0.3014, thinned_from=5, keys [[0,0.8521],[10.4,0.2],[32.8,0.2],[48,0.8521]]}
          Bar Outline ColorB {type=MovieSceneScriptingFloatChannel, default=1, thinned_from=5, keys [[0,0.9948],[10.5976,1],[32.8,1],[48,0.9948]]}
          Bar Outline ColorA {type=MovieSceneScriptingFloatChannel, default=1, const [1,5]}
          Progress Background ColorR {type=MovieSceneScriptingFloatChannel, default=0, keys [[0,0.0012],[10.5976,0],[32.8,0],[48,0.0012]]}
          Progress Background ColorG {type=MovieSceneScriptingFloatChannel, default=0.3014, keys [[0,0.0123],[10.5976,0.01],[32.8,0.01],[48,0.0123]]}
          Progress Background ColorB {type=MovieSceneScriptingFloatChannel, default=1, keys [[0,0.0513],[10.5976,0.04],[32.8,0.04],[48,0.0513]]}
          Progress Background ColorA {type=MovieSceneScriptingFloatChannel, default=1, const [1,4]}
          Inner ColorR {type=MovieSceneScriptingFloatChannel, default=0, keys [[0,0.0021],[10.5976,0],[32.8,0],[48,0.0021]]}
          Inner ColorG {type=MovieSceneScriptingFloatChannel, default=0.3014, keys [[0,0.0185],[10.5976,0.03],[32.8,0.03],[48,0.0185]]}
          Inner ColorB {type=MovieSceneScriptingFloatChannel, default=1, keys [[0,0.0648],[10.5976,0.2],[32.8,0.2],[48,0.0648]]}
          Inner ColorA {type=MovieSceneScriptingFloatChannel, default=1, const [1,4]}
         ]
        }]
      }]
    }
    Image_PlacementGlow {
     id=#77
     possessed_class=Image
     tracks [
      {
       type=MovieSceneWidgetMaterialTrack
       sections [{type=MovieSceneParameterSection, channels [Glow Y Movement {type=MovieSceneScriptingFloatChannel, keys [[0,1.4],[29.6,-0.5]]}]}]
      }
      {
       type=MovieSceneFloatTrack
       display_name=Render Opacity
       sections [{type=MovieSceneFloatSection, channels [None {type=MovieSceneScriptingFloatChannel, keys [[0,0],[2.4,1],[30.4,1],[33.6,0],[53,0]]}]}]
      }
     ]
    }
   ]
   display_rate {numerator=48}
  }
 }
 Anim_BonusFeedback_Placement_V02 {
  binding_count=3
  length_seconds=1.9833
  object=&self:Anim_BonusFeedback_Placement_V02
  sequence {
   bindings [
    Image_PlacementGlow {
     id=#78
     possessed_class=Image
     tracks [
      {
       type=MovieSceneWidgetMaterialTrack
       sections [{type=MovieSceneParameterSection, channels [Glow Y Movement {type=MovieSceneScriptingFloatChannel, keys [[0,1.4],[37,-0.5]]}]}]
      }
      {
       type=MovieSceneFloatTrack
       display_name=Render Opacity
       sections [{type=MovieSceneFloatSection, channels [None {type=MovieSceneScriptingFloatChannel, keys [[0,0],[37,0.5],[66.25,0]]}]}]
      }
     ]
    }
    RadialGlow {
     id=#26
     possessed_class=Image
     tracks [{
       type=MovieSceneFloatTrack
       display_name=Render Opacity
       sections [{type=MovieSceneFloatSection, end_frame=119, start_frame=0, channels [None {type=MovieSceneScriptingFloatChannel, const [0,1]}]}]
      }]
    }
    Img_Flare {
     id=#28
     possessed_class=Image
     tracks [{
       type=MovieSceneWidgetMaterialTrack
       sections [{type=MovieSceneParameterSection, end_frame=119, start_frame=0, channels [Opacity {type=MovieSceneScriptingFloatChannel, const [0,1]}]}]
      }]
    }
   ]
   display_rate {numerator=60}
  }
 }
 Anim_UnrealLegends_Start {
  binding_count=13
  length_seconds=1.75
  object=&self:Anim_UnrealLegends_Start
  sequence {
   bindings [
    Overlay_TriangleEnergy {
     id=#79
     possessed_class=Overlay
     tracks [{
       type=MovieSceneFloatTrack
       display_name=Render Opacity
       sections [{type=MovieSceneFloatSection, end_frame=0, start_frame=0, channels [None {type=MovieSceneScriptingFloatChannel, const [0,1]}]}]
      }]
    }
    Particles {
     id=#80
     possessed_class=GridPanel
     tracks [{
       type=MovieSceneFloatTrack
       display_name=Render Opacity
       sections [{type=MovieSceneFloatSection, end_frame=52, start_frame=0, channels [None {type=MovieSceneScriptingFloatChannel, const [0,2]}]}]
      }]
    }
    CurrentBadge {
     id=#81
     possessed_class=Overlay
     tracks [
      {
       type=MovieSceneFloatTrack
       display_name=Render Opacity
       sections [{
         type=MovieSceneFloatSection
         end_frame=52
         start_frame=0
         channels [None {type=MovieSceneScriptingFloatChannel, default=0, keys [[0,0],[15,1],[37.5,1],[52.5,0]]}]
        }]
      }
      {
       type=MovieScene2DTransformTrack
       sections [{
         type=MovieScene2DTransformSection
         end_frame=52
         start_frame=0
         channels [
          Translation.X {type=MovieSceneScriptingFloatChannel, const [0,2]}
          Translation.Y {type=MovieSceneScriptingFloatChannel, const [0,2]}
          Angle {type=MovieSceneScriptingFloatChannel, const [0,2]}
          Scale.X {type=MovieSceneScriptingFloatChannel, const [0.85,2]}
          Scale.Y {type=MovieSceneScriptingFloatChannel, const [0.85,2]}
          Shear.X {type=MovieSceneScriptingFloatChannel, const [0,2]}
          Shear.Y {type=MovieSceneScriptingFloatChannel, const [0,2]}
         ]
        }]
      }
     ]
    }
    Text_TierName {
     id=#82
     possessed_class=CommonTextBlock
     tracks [
      {
       type=MovieSceneFloatTrack
       display_name=Render Opacity
       sections [{
         type=MovieSceneFloatSection
         end_frame=49
         start_frame=-1
         channels [None {type=MovieSceneScriptingFloatChannel, keys [[3,0],[15,1],[31.5,1],[46.5,0]]}]
        }]
      }
      {
       type=MovieScene2DTransformTrack
       sections [{
         type=MovieScene2DTransformSection
         end_frame=16
         start_frame=0
         channels [
          Translation.X {type=MovieSceneScriptingFloatChannel, const [0,1]}
          Translation.Y {type=MovieSceneScriptingFloatChannel, const [0,1]}
          Scale.X {type=MovieSceneScriptingFloatChannel, keys [[3,0.75],[15,1]]}
          Scale.Y {type=MovieSceneScriptingFloatChannel, keys [[3,0.75],[15,1]]}
         ]
        }]
      }
     ]
    }
    Box_PercentageText {
     id=#83
     possessed_class=HorizontalBox
     tracks [
      {
       type=MovieSceneFloatTrack
       display_name=Render Opacity
       sections [{
         type=MovieSceneFloatSection
         end_frame=45
         start_frame=0
         channels [None {type=MovieSceneScriptingFloatChannel, keys [[4,0],[16,1],[30,1],[45,0]]}]
        }]
      }
      {
       type=MovieScene2DTransformTrack
       sections [{
         type=MovieScene2DTransformSection
         end_frame=18
         start_frame=0
         channels [
          Translation.X {type=MovieSceneScriptingFloatChannel, const [0,1]}
          Translation.Y {type=MovieSceneScriptingFloatChannel, const [0,1]}
          Scale.X {type=MovieSceneScriptingFloatChannel, keys [[4,0.75],[16,1]]}
          Scale.Y {type=MovieSceneScriptingFloatChannel, keys [[4,0.75],[16,1]]}
         ]
        }]
      }
     ]
    }
    Unreal_BG_Overlay {
     id=#84
     possessed_class=Image
     tracks [
      {
       type=MovieSceneWidgetMaterialTrack
       sections [{
         type=MovieSceneParameterSection
         end_frame=52
         start_frame=0
         channels [
          Blobs Opacity {type=MovieSceneScriptingFloatChannel, keys [[0.0165,0],[4.6665,0],[10.5,0.0298],[52.5,-0.3],[89.9885,0.1],[139.4735,0.6],[224.946,0.1]]}
          UnderGlow_Opacity {type=MovieSceneScriptingFloatChannel, keys [[0.0165,0],[4.6665,0],[10.5,0.3562],[52.5,0.7],[89.9885,1],[139.4735,1],[224.946,0]]}
          Back Opacity {type=MovieSceneScriptingFloatChannel, thinned_from=6, keys [[0.0165,0],[4.6665,0],[10.5,1],[195.7055,1],[224.946,0]]}
          Center Blob Opacity {type=MovieSceneScriptingFloatChannel, default=0.1, const [0,3]}
          Inverted Center Blob Opacity {type=MovieSceneScriptingFloatChannel, default=0, thinned_from=6, keys [[0.0165,0],[67.495,0],[78.742,0.15],[139.4735,0],[224.946,0]]}
          Centre Blob Position {type=MovieSceneScriptingFloatChannel, keys [[52.5,0],[58.4985,0],[74.243,-0.25]]}
          Centre Blob Scale {type=MovieSceneScriptingFloatChannel, thinned_from=5, keys [[10.5035,0],[52.5025,1],[125.9775,4],[128.2265,4]]}
          Color 4R {type=MovieSceneScriptingFloatChannel, const [0.0091,2]}
          Color 4G {type=MovieSceneScriptingFloatChannel, const [0.0122,2]}
          Color 4B {type=MovieSceneScriptingFloatChannel, const [0.05,2]}
          Color 4A {type=MovieSceneScriptingFloatChannel, const [1,2]}
         ]
        }]
      }
      {
       type=MovieSceneFloatTrack
       display_name=Render Opacity
       sections [{type=MovieSceneFloatSection, end_frame=52, start_frame=0, channels [None {type=MovieSceneScriptingFloatChannel, keys [[0,0],[14,1]]}]}]
      }
     ]
    }
    Box_SkillRating {
     id=#85
     possessed_class=HorizontalBox
     tracks [{
       type=MovieSceneFloatTrack
       display_name=Render Opacity
       sections [{type=MovieSceneFloatSection, end_frame=7, start_frame=0, channels [None {type=MovieSceneScriptingFloatChannel, const [0,1]}]}]
      }]
    }
    HorizontalBox_0 {
     id=#86
     possessed_class=HorizontalBox
     tracks [{
       type=MovieScene2DTransformTrack
       sections [{
         type=MovieScene2DTransformSection
         end_frame=52
         start_frame=0
         channels [
          Translation.Y {type=MovieSceneScriptingFloatChannel, keys [[0,-525],[18,-125],[31.5,-125],[51,125]]}
          Scale.X {type=MovieSceneScriptingFloatChannel, keys [[0,0.5],[15,1],[36,1],[51,3.2]]}
          Scale.Y {type=MovieSceneScriptingFloatChannel, keys [[0,0.5],[15,1],[36,1],[51,3.2]]}
         ]
        }]
      }]
    }
    TextWrapper_RankUp {
     id=#87
     possessed_class=Border
     tracks [
      {
       type=MovieSceneFloatTrack
       display_name=Render Opacity
       sections [{
         type=MovieSceneFloatSection
         end_frame=45
         start_frame=-1
         channels [None {type=MovieSceneScriptingFloatChannel, keys [[4.75,0],[16.75,1],[28.5,1],[43.5,0]]}]
        }]
      }
      {
       type=MovieScene2DTransformTrack
       sections [{
         type=MovieScene2DTransformSection
         end_frame=19
         start_frame=0
         channels [
          Translation.X {type=MovieSceneScriptingFloatChannel, const [0,1]}
          Translation.Y {type=MovieSceneScriptingFloatChannel, const [0,1]}
          Scale.X {type=MovieSceneScriptingFloatChannel, keys [[4.75,0.75],[16.75,1]]}
          Scale.Y {type=MovieSceneScriptingFloatChannel, keys [[4.75,0.75],[16.75,1]]}
         ]
        }]
      }
     ]
    }
    NextBadge {
     id=#88
     possessed_class=Overlay
     tracks [{
       type=MovieSceneFloatTrack
       display_name=Render Opacity
       sections [{type=MovieSceneFloatSection, end_frame=52, start_frame=-1, channels [None {type=MovieSceneScriptingFloatChannel, const [0,2]}]}]
      }]
    }
    CurrentBadge_Icon {
     id=#89
     possessed_class=Image
     tracks [{
       type=MovieSceneColorTrack
       display_name=Color and Opacity
       sections [{
         type=MovieSceneColorSection
         end_frame=52
         start_frame=0
         channels [
          Color.R {type=MovieSceneScriptingFloatChannel, keys [[0,0],[12,1],[33,1],[49.5,0]]}
          Color.G {type=MovieSceneScriptingFloatChannel, keys [[0,0],[12,1],[33,1],[49.5,0]]}
          Color.B {type=MovieSceneScriptingFloatChannel, keys [[0,0],[12,1],[33,1],[49.5,0]]}
          Color.A {type=MovieSceneScriptingFloatChannel, const [1,4]}
         ]
        }]
      }]
    }
    RadialGlow {
     id=#90
     possessed_class=Image
     tracks [{
       type=MovieSceneColorTrack
       display_name=Color and Opacity
       sections [{
         type=MovieSceneColorSection
         end_frame=52
         start_frame=0
         channels [
          Color.R {type=MovieSceneScriptingFloatChannel, const [1,2]}
          Color.G {type=MovieSceneScriptingFloatChannel, const [1,2]}
          Color.B {type=MovieSceneScriptingFloatChannel, const [1,2]}
          Color.A {type=MovieSceneScriptingFloatChannel, const [0,2]}
         ]
        }]
      }]
    }
    Img_Flare {
     id=#91
     possessed_class=Image
     tracks [{
       type=MovieSceneFloatTrack
       display_name=Render Opacity
       sections [{type=MovieSceneFloatSection, end_frame=0, start_frame=0, channels [None {type=MovieSceneScriptingFloatChannel, const [0,1]}]}]
      }]
    }
   ]
   display_rate {numerator=30}
   tracks [
    {
     type=MovieSceneEventTrack
     sections [{
       type=MovieSceneEventTriggerSection
       channels [None {
         type=MovieSceneScriptingEventChannel
         keys [[0, "<Struct 'MovieSceneEvent' (0x4676125d840) {}>"], [27, "<Struct 'MovieSceneEvent' (0x4676125d7a0) {}>"]]
        }]
      }]
    }
    {
     type=MovieSceneEventTrack
     sections [{
       type=MovieSceneEventTriggerSection
       channels [None {
         type=MovieSceneScriptingEventChannel
         keys [[0, "<Struct 'MovieSceneEvent' (0x4676125d2a0) {}>"], [27, "<Struct 'MovieSceneEvent' (0x4676125d200) {}>"]]
        }]
      }]
    }
   ]
  }
 }
 Anim_UnrealLegends_End {
  binding_count=17
  length_seconds=6.2
  object=&self:Anim_UnrealLegends_End
  sequence {
   bindings [
    Text_TierName {
     id=#92
     possessed_class=CommonTextBlock
     tracks [
      {
       type=MovieSceneFloatTrack
       display_name=Render Opacity
       sections [{type=MovieSceneFloatSection, end_frame=61, channels [None {type=MovieSceneScriptingFloatChannel, keys [[0.4995,0],[49.4505,0],[59.9401,1]]}]}]
      }
      {
       type=MovieScene2DTransformTrack
       sections [{
         type=MovieScene2DTransformSection
         end_frame=71
         channels [
          Translation.Y {type=MovieSceneScriptingFloatChannel, keys [[49.4505,50],[65.9341,0],[110.0455,0]]}
          Scale.X {type=MovieSceneScriptingFloatChannel, keys [[0,0.7],[42.957,0.7],[49.4505,0.7283],[59.9401,0.9743],[64.9351,1]]}
          Scale.Y {type=MovieSceneScriptingFloatChannel, default=2, keys [[0,0.7],[42.957,0.7],[49.4505,0.7283],[59.9401,0.9743],[64.9351,1]]}
         ]
        }]
      }
     ]
    }
    TextWrapper_RankUp {
     id=#93
     possessed_class=Border
     tracks [
      {
       type=MovieSceneFloatTrack
       display_name=Render Opacity
       sections [{
         type=MovieSceneFloatSection
         end_frame=84
         channels [None {type=MovieSceneScriptingFloatChannel, keys [[0,0],[57.9416,0],[63.9356,1],[76.0849,1]]}]
        }]
      }
      {
       type=MovieScene2DTransformTrack
       sections [{
         type=MovieScene2DTransformSection
         end_frame=66
         channels [Translation.Y {type=MovieSceneScriptingFloatChannel, keys [[0,32],[57.9416,32],[66.9331,0]]}]
        }]
      }
     ]
    }
    Overlay_TriangleEnergy {
     id=#94
     possessed_class=Overlay
     tracks [{
       type=MovieSceneFloatTrack
       display_name=Render Opacity
       sections [{type=MovieSceneFloatSection, channels [None {type=MovieSceneScriptingFloatChannel, const [0,1]}]}]
      }]
    }
    FortVisualAttachment_Elims {
     id=#95
     possessed_class=FortVisualAttachment
     tracks [{
       type=MovieScene2DTransformTrack
       sections [{
         type=MovieScene2DTransformSection
         end_frame=46
         start_frame=40
         channels [
          Translation.X {type=MovieSceneScriptingFloatChannel}
          Translation.Y {type=MovieSceneScriptingFloatChannel, const [15,1]}
          Angle {type=MovieSceneScriptingFloatChannel}
          Scale.X {type=MovieSceneScriptingFloatChannel}
          Scale.Y {type=MovieSceneScriptingFloatChannel}
          Shear.X {type=MovieSceneScriptingFloatChannel}
          Shear.Y {type=MovieSceneScriptingFloatChannel}
         ]
        }]
      }]
    }
    Particles {
     id=#96
     possessed_class=GridPanel
     tracks [
      {
       type=MovieSceneFloatTrack
       display_name=Render Opacity
       sections [{
         type=MovieSceneFloatSection
         end_frame=69
         start_frame=0
         channels [None {type=MovieSceneScriptingFloatChannel, keys [[44.955,0],[60.9391,0],[69.9301,1]]}]
        }]
      }
      {
       type=MovieScene2DTransformTrack
       sections [{
         type=MovieScene2DTransformSection
         end_frame=185
         start_frame=0
         channels [
          Translation.X {type=MovieSceneScriptingFloatChannel, const [0,1]}
          Translation.Y {type=MovieSceneScriptingFloatChannel, const [-250,1]}
          Angle {type=MovieSceneScriptingFloatChannel, const [0,1]}
          Scale.X {type=MovieSceneScriptingFloatChannel, keys [[53.4466,0.01],[68.9311,1]]}
          Scale.Y {type=MovieSceneScriptingFloatChannel, keys [[53.4466,0.01],[68.9311,1]]}
          Shear.X {type=MovieSceneScriptingFloatChannel, const [0,1]}
          Shear.Y {type=MovieSceneScriptingFloatChannel, const [0,1]}
         ]
        }]
      }
     ]
    }
    Box_PercentageText {
     id=#97
     possessed_class=HorizontalBox
     tracks [
      {
       type=MovieSceneFloatTrack
       display_name=Render Opacity
       sections [{
         type=MovieSceneFloatSection
         end_frame=76
         start_frame=0
         channels [None {type=MovieSceneScriptingFloatChannel, thinned_from=4, keys [[0,0],[54.9446,0],[60.9386,1]]}]
        }]
      }
      {
       type=MovieScene2DTransformTrack
       sections [{
         type=MovieScene2DTransformSection
         end_frame=64
         start_frame=0
         channels [
          Translation.X {type=MovieSceneScriptingFloatChannel}
          Translation.Y {type=MovieSceneScriptingFloatChannel, keys [[42.3102,32],[54.9446,32],[63.9361,0]]}
         ]
        }]
      }
     ]
    }
    Img_Flare {
     id=#98
     possessed_class=Image
     tracks [
      {
       type=MovieSceneWidgetMaterialTrack
       sections [{
         type=MovieSceneParameterSection
         end_frame=185
         channels [Opacity {type=MovieSceneScriptingFloatChannel, keys [[0,0],[57.4426,1],[116.8831,1],[185.8142,0]]}]
        }]
      }
      {
       type=MovieScene2DTransformTrack
       sections [{
         type=MovieScene2DTransformSection
         end_frame=185
         start_frame=0
         channels [
          Translation.X {type=MovieSceneScriptingFloatChannel, default=0}
          Translation.Y {type=MovieSceneScriptingFloatChannel, default=0, const [-265,2]}
          Scale.X {type=MovieSceneScriptingFloatChannel, default=1, keys [[58.9411,0.2],[74.9251,1.4]]}
          Scale.Y {type=MovieSceneScriptingFloatChannel, default=1, keys [[57.4426,0.5],[67.4326,1.5],[121.3786,1.4]]}
          Shear.X {type=MovieSceneScriptingFloatChannel, default=0}
          Shear.Y {type=MovieSceneScriptingFloatChannel, default=0}
         ]
        }]
      }
      {
       type=MovieSceneColorTrack
       display_name=Color and Opacity
       sections [{
         type=MovieSceneColorSection
         end_frame=145
         start_frame=58
         channels [
          Color.R {type=MovieSceneScriptingFloatChannel, keys [[58.9406,1],[66.4331,0.1955],[119.8801,0.1791]]}
          Color.G {type=MovieSceneScriptingFloatChannel, keys [[58.9406,1],[66.4331,0.1768],[119.8801,0]]}
          Color.B {type=MovieSceneScriptingFloatChannel, const [1,3]}
          Color.A {type=MovieSceneScriptingFloatChannel, const [1,3]}
         ]
        }]
      }
      {
       type=MovieSceneBoolTrack
       display_name=Is Enabled
       sections [{type=MovieSceneBoolSection, end_frame=185, start_frame=0, channels [None {type=MovieSceneScriptingBoolChannel, const [true,2]}]}]
      }
      {
       type=MovieSceneFloatTrack
       display_name=Render Opacity
       sections [{
         type=MovieSceneFloatSection
         end_frame=185
         start_frame=0
         channels [None {type=MovieSceneScriptingFloatChannel, keys [[0,0],[57.4426,0],[64.4356,1],[116.8831,0],[185.8142,0]]}]
        }]
      }
     ]
    }
    NextBadge_Bg {
     id=#99
     possessed_class=Image
     tracks [{
       type=MovieSceneFloatTrack
       display_name=Render Opacity
       sections [{type=MovieSceneFloatSection, end_frame=19, start_frame=0, channels [None {type=MovieSceneScriptingFloatChannel, keys [[0,0],[19.973,1]]}]}]
      }]
    }
    CurrentBadge {
     id=#100
     possessed_class=Overlay
     tracks [{
       type=MovieSceneFloatTrack
       display_name=Render Opacity
       sections [{type=MovieSceneFloatSection, channels [None {type=MovieSceneScriptingFloatChannel, const [0,1]}]}]
      }]
    }
    HorizontalBox_0 {
     id=#101
     possessed_class=HorizontalBox
     tracks [{
       type=MovieScene2DTransformTrack
       sections [{
         type=MovieScene2DTransformSection
         end_frame=130
         start_frame=0
         channels [
          Translation.X {type=MovieSceneScriptingFloatChannel, default=0}
          Translation.Y {type=MovieSceneScriptingFloatChannel, keys [[0,-525],[17.982,0],[41.958,0],[47.952,16.9547],[58.9411,-265],[67.4326,-265]]}
          Angle {type=MovieSceneScriptingFloatChannel, const [0,2]}
          Scale.X {type=MovieSceneScriptingFloatChannel, keys [[0,0.5],[17.982,1]]}
          Scale.Y {type=MovieSceneScriptingFloatChannel, keys [[0,0.5],[17.982,1]]}
          Shear.X {type=MovieSceneScriptingFloatChannel, const [0,2]}
          Shear.Y {type=MovieSceneScriptingFloatChannel, const [0,2]}
         ]
        }]
      }]
    }
    Background {
     id=#102
     possessed_class=Image
     tracks [{
       type=MovieSceneWidgetMaterialTrack
       sections [{
         type=MovieSceneParameterSection
         channels [
          Triangle Opacity {type=MovieSceneScriptingFloatChannel, const [0.6,1]}
          Invert Radial {type=MovieSceneScriptingFloatChannel, const [1,1]}
         ]
        }]
      }]
    }
    Box_SkillRating {
     id=#103
     possessed_class=HorizontalBox
     tracks [
      {
       type=MovieSceneFloatTrack
       display_name=Render Opacity
       sections [{
         type=MovieSceneFloatSection
         end_frame=79
         start_frame=0
         channels [None {type=MovieSceneScriptingFloatChannel, thinned_from=4, keys [[0,0],[56.4431,0],[62.4371,1]]}]
        }]
      }
      {
       type=MovieScene2DTransformTrack
       sections [{
         type=MovieScene2DTransformSection
         end_frame=78
         start_frame=9
         channels [
          Translation.X {type=MovieSceneScriptingFloatChannel, const [0,1]}
          Translation.Y {type=MovieSceneScriptingFloatChannel, keys [[42.3102,32],[56.4431,32],[65.4346,0],[78.9206,0]]}
         ]
        }]
      }
     ]
    }
    Unreal_BG_Overlay {
     id=#104
     possessed_class=Image
     tracks [
      {
       type=MovieSceneWidgetMaterialTrack
       sections [{
         type=MovieSceneParameterSection
         end_frame=151
         start_frame=0
         channels [
          Blobs Opacity {type=MovieSceneScriptingFloatChannel, keys [[0,-0.3],[37.994,0.1],[57.4745,0.6],[149.8501,0.1]]}
          UnderGlow_Opacity {type=MovieSceneScriptingFloatChannel, keys [[-0.0005,0.3562],[0,0.7],[37.994,1],[57.4745,1],[149.8501,0]]}
          Back Opacity {type=MovieSceneScriptingFloatChannel, keys [[0,1],[107.8921,1],[149.8501,0]]}
          Center Blob Opacity {type=MovieSceneScriptingFloatChannel, default=0.1, keys [[0,0],[5.4945,0],[13.4865,10],[45.954,0],[149.8501,0]]}
          Inverted Center Blob Opacity {type=MovieSceneScriptingFloatChannel, default=0, keys [[0,0],[7,0.15],[27.1778,0.15],[57.4745,0],[149.8501,0]]}
          Centre Blob Position {type=MovieSceneScriptingFloatChannel, const [0,3]}
          Centre Blob Scale {type=MovieSceneScriptingFloatChannel, keys [[0,0],[32.967,4],[34.4655,4]]}
          Color 4R {type=MovieSceneScriptingFloatChannel, const [0.0091,1]}
          Color 4G {type=MovieSceneScriptingFloatChannel, const [0.0122,1]}
          Color 4B {type=MovieSceneScriptingFloatChannel, const [0.05,1]}
          Color 4A {type=MovieSceneScriptingFloatChannel, const [1,1]}
         ]
        }]
      }
      {
       type=MovieSceneFloatTrack
       display_name=Render Opacity
       sections [{type=MovieSceneFloatSection, end_frame=0, start_frame=0, channels [None {type=MovieSceneScriptingFloatChannel, const [1,1]}]}]
      }
     ]
    }
    NextBadge_Icon {
     id=#105
     possessed_class=Image
     tracks [
      {
       type=MovieSceneWidgetMaterialTrack
       sections [{type=MovieSceneParameterSection, channels [Badge Hotness {type=MovieSceneScriptingFloatChannel, keys [[0,4],[16.976,0]]}]}]
      }
      {
       type=MovieSceneColorTrack
       display_name=Color and Opacity
       sections [{
         type=MovieSceneColorSection
         end_frame=29
         start_frame=0
         channels [
          Color.R {type=MovieSceneScriptingFloatChannel, keys [[0,0],[4.4955,0],[29.97,1]]}
          Color.G {type=MovieSceneScriptingFloatChannel, keys [[0,0],[4.4955,0],[29.97,1]]}
          Color.B {type=MovieSceneScriptingFloatChannel, keys [[0,0],[4.4955,0],[29.97,1]]}
          Color.A {type=MovieSceneScriptingFloatChannel, const [1,2]}
         ]
        }]
      }
     ]
    }
    ProgressBar {
     id=#106
     possessed_class=Image
     tracks [{
       type=MovieSceneFloatTrack
       display_name=Render Opacity
       sections [{type=MovieSceneFloatSection, end_frame=0, start_frame=0, channels [None {type=MovieSceneScriptingFloatChannel, const [0,1]}]}]
      }]
    }
    NextBadge {
     id=#107
     possessed_class=Overlay
     tracks [
      {
       type=MovieSceneFloatTrack
       display_name=Render Opacity
       sections [{type=MovieSceneFloatSection, end_frame=73, start_frame=0, channels [None {type=MovieSceneScriptingFloatChannel, keys [[0,0],[10.4895,1]]}]}]
      }
      {
       type=MovieScene2DTransformTrack
       sections [{
         type=MovieScene2DTransformSection
         end_frame=64
         start_frame=0
         channels [
          Translation.X {type=MovieSceneScriptingFloatChannel, const [0,1]}
          Translation.Y {type=MovieSceneScriptingFloatChannel, const [0,1]}
          Angle {type=MovieSceneScriptingFloatChannel, const [0,1]}
          Scale.X {type=MovieSceneScriptingFloatChannel, keys [[0,1.4],[38.961,1.4],[59.9401,0.7666],[64.9351,0.8]]}
          Scale.Y {type=MovieSceneScriptingFloatChannel, keys [[0,1.4],[38.961,1.4],[59.9401,0.7666],[64.9351,0.8]]}
          Shear.X {type=MovieSceneScriptingFloatChannel, const [0,1]}
          Shear.Y {type=MovieSceneScriptingFloatChannel, const [0,1]}
         ]
        }]
      }
     ]
    }
    FortVisualAttachment_Placement {
     id=#108
     possessed_class=FortVisualAttachment
     tracks [{
       type=MovieScene2DTransformTrack
       sections [{
         type=MovieScene2DTransformSection
         end_frame=46
         start_frame=40
         channels [
          Translation.X {type=MovieSceneScriptingFloatChannel}
          Translation.Y {type=MovieSceneScriptingFloatChannel, const [15,1]}
          Angle {type=MovieSceneScriptingFloatChannel}
          Scale.X {type=MovieSceneScriptingFloatChannel}
          Scale.Y {type=MovieSceneScriptingFloatChannel}
          Shear.X {type=MovieSceneScriptingFloatChannel}
          Shear.Y {type=MovieSceneScriptingFloatChannel}
         ]
        }]
      }]
    }
    RadialGlow {
     id=#109
     possessed_class=Image
     tracks [{
       type=MovieSceneFloatTrack
       display_name=Render Opacity
       sections [{type=MovieSceneFloatSection, end_frame=0, start_frame=0, channels [None {type=MovieSceneScriptingFloatChannel, const [0,1]}]}]
      }]
    }
   ]
   display_rate {denominator=1001, numerator=30000}
   tracks [
    {
     type=MovieSceneEventTrack
     sections [{
       type=MovieSceneEventTriggerSection
       channels [None {
         type=MovieSceneScriptingEventChannel
         keys [[0, "<Struct 'MovieSceneEvent' (0x4676125ee20) {}>"], [40.959, "<Struct 'MovieSceneEvent' (0x4676125ed80) {}>"]]
        }]
      }]
    }
    {
     type=MovieSceneEventTrack
     sections [{
       type=MovieSceneEventTriggerSection
       channels [None {
         type=MovieSceneScriptingEventChannel
         keys [[0, "<Struct 'MovieSceneEvent' (0x4676125eba0) {}>"], [33.4665, "<Struct 'MovieSceneEvent' (0x4676125ec40) {}>"]]
        }]
      }]
    }
   ]
  }
 }
 Anim_GoDown_FinalRank {
  binding_count=15
  length_seconds=1.9834
  object=&self:Anim_GoDown_FinalRank
  sequence {
   bindings [
    TextWrapper_RankUp {
     id=#15
     possessed_class=Border
     tracks [{
       type=MovieSceneFloatTrack
       display_name=Render Opacity
       sections [{type=MovieSceneFloatSection, end_frame=119, start_frame=0, channels [None {type=MovieSceneScriptingFloatChannel, const [0,1]}]}]
      }]
    }
    Particles {
     id=#48
     possessed_class=GridPanel
     tracks [{
       type=MovieSceneFloatTrack
       display_name=Render Opacity
       sections [{type=MovieSceneFloatSection, end_frame=72, start_frame=0, channels [None {type=MovieSceneScriptingFloatChannel, keys [[0,0],[42,0],[72,1]]}]}]
      }]
    }
    FortVisualAttachment_Elims {
     id=#49
     possessed_class=FortVisualAttachment
     tracks [{
       type=MovieScene2DTransformTrack
       sections [{
         type=MovieScene2DTransformSection
         end_frame=0
         start_frame=0
         channels [
          Translation.X {type=MovieSceneScriptingFloatChannel, const [0,1]}
          Translation.Y {type=MovieSceneScriptingFloatChannel, const [15,1]}
          Angle {type=MovieSceneScriptingFloatChannel, const [0,1]}
          Scale.X {type=MovieSceneScriptingFloatChannel, const [1,1]}
          Scale.Y {type=MovieSceneScriptingFloatChannel, const [1,1]}
          Shear.X {type=MovieSceneScriptingFloatChannel, const [0,1]}
          Shear.Y {type=MovieSceneScriptingFloatChannel, const [0,1]}
         ]
        }]
      }]
    }
    Box_PercentageText {
     id=#50
     display_name=PercentageTextwrapper
     possessed_class=HorizontalBox
     tracks [
      {
       type=MovieScene2DTransformTrack
       sections [{type=MovieScene2DTransformSection, channels [Translation.Y {type=MovieSceneScriptingFloatChannel, const [0,1]}]}]
      }
      {
       type=MovieSceneFloatTrack
       display_name=Render Opacity
       sections [{type=MovieSceneFloatSection, end_frame=51, start_frame=0, channels [None {type=MovieSceneScriptingFloatChannel, keys [[0,0],[30,0],[51,1]]}]}]
      }
     ]
    }
    Overlay_TriangleEnergy {
     id=#51
     possessed_class=Overlay
     tracks [{
       type=MovieSceneFloatTrack
       display_name=Render Opacity
       sections [{type=MovieSceneFloatSection, channels [None {type=MovieSceneScriptingFloatChannel, const [0,1]}]}]
      }]
    }
    NextBadge {
     id=#18
     possessed_class=Overlay
     tracks [{
       type=MovieSceneFloatTrack
       display_name=Render Opacity
       sections [{type=MovieSceneFloatSection, end_frame=39, start_frame=0, channels [None {type=MovieSceneScriptingFloatChannel, const [0,1]}]}]
      }]
    }
    ProgressBar {
     id=#19
     possessed_class=Image
     tracks [
      {
       type=MovieSceneWidgetMaterialTrack
       sections [{
         type=MovieSceneParameterSection
         end_frame=119
         start_frame=0
         channels [
          Glow Opacity {type=MovieSceneScriptingFloatChannel, const [0.25,1]}
          Progress Background Opacity {type=MovieSceneScriptingFloatChannel, const [1,1]}
          Override Edge Gradient {type=MovieSceneScriptingFloatChannel, const [1,1]}
         ]
        }]
      }
      {
       type=MovieSceneFloatTrack
       display_name=Render Opacity
       sections [{
         type=MovieSceneFloatSection
         end_frame=119
         start_frame=0
         channels [None {type=MovieSceneScriptingFloatChannel, keys [[0,0],[18,0],[39,1],[119,1]]}]
        }]
      }
      {
       type=MovieScene2DTransformTrack
       sections [{
         type=MovieScene2DTransformSection
         end_frame=39
         start_frame=0
         channels [
          Translation.Y {type=MovieSceneScriptingFloatChannel, const [0,1]}
          Scale.X {type=MovieSceneScriptingFloatChannel, keys [[0,0.25],[18,0.25],[39,1]]}
          Scale.Y {type=MovieSceneScriptingFloatChannel, keys [[0,0.25],[18,0.25],[39,1]]}
         ]
        }]
      }
     ]
    }
    Text_TierName {
     id=#21
     possessed_class=CommonTextBlock
     tracks [
      {
       type=MovieSceneFloatTrack
       display_name=Render Opacity
       sections [{
         type=MovieSceneFloatSection
         end_frame=119
         start_frame=0
         channels [None {type=MovieSceneScriptingFloatChannel, keys [[0,0],[27,0],[51,1],[119,1]]}]
        }]
      }
      {
       type=MovieScene2DTransformTrack
       sections [{type=MovieScene2DTransformSection, channels [Translation.Y {type=MovieSceneScriptingFloatChannel, const [0,1]}]}]
      }
     ]
    }
    WBP_PostMatch_HabaneroSequence_C_3 {
     id=#52
     possessed_class=WBP_PostMatch_HabaneroSequence_C
     tracks [
      {
       type=MovieSceneDoubleTrack
       display_name=Delta Value Blend
       sections [{type=MovieSceneDoubleSection, end_frame=104, start_frame=0, channels [None {type=MovieSceneScriptingDoubleChannel, const [0,1]}]}]
      }
      {
       type=MovieSceneDoubleTrack
       display_name=Progress Bar Color Blend
       sections [{type=MovieSceneDoubleSection, end_frame=119, start_frame=0, channels [None {type=MovieSceneScriptingDoubleChannel, const [1,1]}]}]
      }
      {
       type=MovieSceneDoubleTrack
       display_name=Progress Value Blend
       sections [{type=MovieSceneDoubleSection, end_frame=119, start_frame=0, channels [None {type=MovieSceneScriptingDoubleChannel, const [0,1]}]}]
      }
     ]
    }
    CurrentBadge_Icon {
     id=#22
     possessed_class=Image
     tracks [{
       type=MovieScene2DTransformTrack
       sections [{
         type=MovieScene2DTransformSection
         end_frame=72
         start_frame=0
         channels [
          Translation.X {type=MovieSceneScriptingFloatChannel, const [8,1]}
          Translation.Y {type=MovieSceneScriptingFloatChannel, const [8,1]}
         ]
        }]
      }]
    }
    FortVisualAttachment_Placement {
     id=#53
     possessed_class=FortVisualAttachment
     tracks [{
       type=MovieScene2DTransformTrack
       sections [{
         type=MovieScene2DTransformSection
         end_frame=0
         start_frame=0
         channels [
          Translation.X {type=MovieSceneScriptingFloatChannel, const [0,1]}
          Translation.Y {type=MovieSceneScriptingFloatChannel, const [15,1]}
          Angle {type=MovieSceneScriptingFloatChannel, const [0,1]}
          Scale.X {type=MovieSceneScriptingFloatChannel, const [1,1]}
          Scale.Y {type=MovieSceneScriptingFloatChannel, const [1,1]}
          Shear.X {type=MovieSceneScriptingFloatChannel, const [0,1]}
          Shear.Y {type=MovieSceneScriptingFloatChannel, const [0,1]}
         ]
        }]
      }]
    }
    CurrentBadge {
     id=#24
     possessed_class=Overlay
     tracks [
      {
       type=MovieSceneFloatTrack
       display_name=Render Opacity
       sections [{type=MovieSceneFloatSection, end_frame=39, start_frame=0, channels [None {type=MovieSceneScriptingFloatChannel, keys [[0,0],[9,0],[39,1]]}]}]
      }
      {
       type=MovieScene2DTransformTrack
       sections [{
         type=MovieScene2DTransformSection
         end_frame=72
         start_frame=0
         channels [
          Translation.Y {type=MovieSceneScriptingFloatChannel, const [0,1]}
          Scale.X {type=MovieSceneScriptingFloatChannel, keys [[0,2],[27,1],[72,1]]}
          Scale.Y {type=MovieSceneScriptingFloatChannel, keys [[0,2],[27,1],[72,1]]}
         ]
        }]
      }
     ]
    }
    Background {
     id=#25
     possessed_class=Image
     tracks [{
       type=MovieSceneWidgetMaterialTrack
       sections [{
         type=MovieSceneParameterSection
         end_frame=120
         channels [
          Color Swap {type=MovieSceneScriptingFloatChannel, const [1,1]}
          Hot Spot Intensity {type=MovieSceneScriptingFloatChannel, const [0.5,1]}
          Cloud Opacity {type=MovieSceneScriptingFloatChannel, const [0,1]}
          Triangle Opacity {type=MovieSceneScriptingFloatChannel, keys [[0,0],[119,0.25]]}
          Triangles Tiling {type=MovieSceneScriptingFloatChannel, const [4,1]}
          Rift Intensity {type=MovieSceneScriptingFloatChannel, const [0,1]}
          Rift Cracks Removal {type=MovieSceneScriptingFloatChannel, const [0.6863,1]}
          Invert Radial {type=MovieSceneScriptingFloatChannel, const [0.6,2]}
          Spot Scale {type=MovieSceneScriptingFloatChannel, const [1,2]}
         ]
        }]
      }]
    }
    RadialGlow {
     id=#26
     possessed_class=Image
     tracks [{
       type=MovieSceneFloatTrack
       display_name=Render Opacity
       sections [{type=MovieSceneFloatSection, end_frame=119, start_frame=0, channels [None {type=MovieSceneScriptingFloatChannel, const [0,1]}]}]
      }]
    }
    Img_Flare {
     id=#28
     possessed_class=Image
     tracks [{
       type=MovieSceneWidgetMaterialTrack
       sections [{type=MovieSceneParameterSection, end_frame=119, start_frame=0, channels [Opacity {type=MovieSceneScriptingFloatChannel, const [0,1]}]}]
      }]
    }
   ]
   display_rate {numerator=60}
   tracks [{type=MovieSceneEventTrack, sections [{type=MovieSceneEventTriggerSection, channels [None {type=MovieSceneScriptingEventChannel}]}]}]
  }
 }
 Anim_GoDownProtected_FinalRank {
  binding_count=18
  length_seconds=3.5
  object=&self:Anim_GoDownProtected_FinalRank
  sequence {
   bindings [
    RankProtectedBox_Switcher {
     id=#69
     possessed_class=CommonAnimatedSwitcher
     tracks [{
       type=MovieSceneIntegerTrack
       display_name=Active Widget Index
       sections [{type=MovieSceneIntegerSection, end_frame=210, start_frame=0, channels [None {type=MovieSceneScriptingIntegerChannel, keys [[0,1],[201,0]]}]}]
      }]
    }
    TextWrapper_RankUp {
     id=#15
     possessed_class=Border
     tracks [{
       type=MovieSceneFloatTrack
       display_name=Render Opacity
       sections [{type=MovieSceneFloatSection, end_frame=119, start_frame=0, channels [None {type=MovieSceneScriptingFloatChannel, const [0,1]}]}]
      }]
    }
    Particles {
     id=#70
     possessed_class=GridPanel
     tracks [{
       type=MovieSceneFloatTrack
       display_name=Render Opacity
       sections [{type=MovieSceneFloatSection, end_frame=72, start_frame=0, channels [None {type=MovieSceneScriptingFloatChannel, keys [[0,0],[42,0],[72,1]]}]}]
      }]
    }
    FortVisualAttachment_Elims {
     id=#49
     possessed_class=FortVisualAttachment
     tracks [{
       type=MovieScene2DTransformTrack
       sections [{
         type=MovieScene2DTransformSection
         end_frame=0
         start_frame=0
         channels [
          Translation.X {type=MovieSceneScriptingFloatChannel, const [0,1]}
          Translation.Y {type=MovieSceneScriptingFloatChannel, const [15,1]}
          Angle {type=MovieSceneScriptingFloatChannel, const [0,1]}
          Scale.X {type=MovieSceneScriptingFloatChannel, const [1,1]}
          Scale.Y {type=MovieSceneScriptingFloatChannel, const [1,1]}
          Shear.X {type=MovieSceneScriptingFloatChannel, const [0,1]}
          Shear.Y {type=MovieSceneScriptingFloatChannel, const [0,1]}
         ]
        }]
      }]
    }
    Box_PercentageText {
     id=#50
     display_name=PercentageTextwrapper
     possessed_class=HorizontalBox
     tracks [
      {
       type=MovieScene2DTransformTrack
       sections [{type=MovieScene2DTransformSection, channels [Translation.Y {type=MovieSceneScriptingFloatChannel, const [0,1]}]}]
      }
      {
       type=MovieSceneFloatTrack
       display_name=Render Opacity
       sections [{type=MovieSceneFloatSection, end_frame=51, start_frame=0, channels [None {type=MovieSceneScriptingFloatChannel, keys [[0,0],[30,0],[51,1]]}]}]
      }
     ]
    }
    Text_RankProtected {
     id=#71
     possessed_class=CommonTextBlock
     tracks [{
       type=MovieSceneFloatTrack
       display_name=Render Opacity
       sections [{type=MovieSceneFloatSection, end_frame=119, start_frame=0, channels [None {type=MovieSceneScriptingFloatChannel, keys [[15,0],[27,1]]}]}]
      }]
    }
    NextBadge {
     id=#18
     possessed_class=Overlay
     tracks [{
       type=MovieSceneFloatTrack
       display_name=Render Opacity
       sections [{type=MovieSceneFloatSection, end_frame=39, start_frame=0, channels [None {type=MovieSceneScriptingFloatChannel, const [0,1]}]}]
      }]
    }
    ProgressBar {
     id=#19
     possessed_class=Image
     tracks [
      {
       type=MovieSceneWidgetMaterialTrack
       sections [{
         type=MovieSceneParameterSection
         end_frame=119
         start_frame=0
         channels [
          Glow Opacity {type=MovieSceneScriptingFloatChannel, const [0.25,1]}
          Progress Background Opacity {type=MovieSceneScriptingFloatChannel, const [1,1]}
          Override Edge Gradient {type=MovieSceneScriptingFloatChannel, const [1,1]}
         ]
        }]
      }
      {
       type=MovieSceneFloatTrack
       display_name=Render Opacity
       sections [{
         type=MovieSceneFloatSection
         end_frame=119
         start_frame=0
         channels [None {type=MovieSceneScriptingFloatChannel, keys [[0,0],[18,0],[39,1],[119,1]]}]
        }]
      }
      {
       type=MovieScene2DTransformTrack
       sections [{
         type=MovieScene2DTransformSection
         end_frame=39
         start_frame=0
         channels [
          Translation.Y {type=MovieSceneScriptingFloatChannel, const [0,1]}
          Scale.X {type=MovieSceneScriptingFloatChannel, keys [[0,0.25],[18,0.25],[39,1]]}
          Scale.Y {type=MovieSceneScriptingFloatChannel, keys [[0,0.25],[18,0.25],[39,1]]}
         ]
        }]
      }
     ]
    }
    Text_TierName {
     id=#21
     possessed_class=CommonTextBlock
     tracks [
      {
       type=MovieSceneFloatTrack
       display_name=Render Opacity
       sections [{
         type=MovieSceneFloatSection
         end_frame=119
         start_frame=0
         channels [None {type=MovieSceneScriptingFloatChannel, keys [[0,0],[27,0],[51,1],[119,1]]}]
        }]
      }
      {
       type=MovieScene2DTransformTrack
       sections [{type=MovieScene2DTransformSection, channels [Translation.Y {type=MovieSceneScriptingFloatChannel, const [0,1]}]}]
      }
     ]
    }
    Overlay_TriangleEnergy {
     id=#72
     possessed_class=Overlay
     tracks [{
       type=MovieSceneFloatTrack
       display_name=Render Opacity
       sections [{type=MovieSceneFloatSection, channels [None {type=MovieSceneScriptingFloatChannel, const [0,1]}]}]
      }]
    }
    WBP_PostMatch_HabaneroSequence_C_3 {
     id=#52
     possessed_class=WBP_PostMatch_HabaneroSequence_C
     tracks [
      {
       type=MovieSceneDoubleTrack
       display_name=Delta Value Blend
       sections [{type=MovieSceneDoubleSection, end_frame=104, start_frame=0, channels [None {type=MovieSceneScriptingDoubleChannel, const [0,1]}]}]
      }
      {
       type=MovieSceneDoubleTrack
       display_name=Progress Bar Color Blend
       sections [{type=MovieSceneDoubleSection, end_frame=119, start_frame=0, channels [None {type=MovieSceneScriptingDoubleChannel, const [1,1]}]}]
      }
      {
       type=MovieSceneDoubleTrack
       display_name=Progress Value Blend
       sections [{type=MovieSceneDoubleSection, end_frame=119, start_frame=0, channels [None {type=MovieSceneScriptingDoubleChannel, const [0,1]}]}]
      }
     ]
    }
    CurrentBadge_Icon {
     id=#22
     possessed_class=Image
     tracks [{
       type=MovieScene2DTransformTrack
       sections [{
         type=MovieScene2DTransformSection
         end_frame=72
         start_frame=0
         channels [
          Translation.X {type=MovieSceneScriptingFloatChannel, const [8,1]}
          Translation.Y {type=MovieSceneScriptingFloatChannel, const [8,1]}
         ]
        }]
      }]
    }
    FortVisualAttachment_Placement {
     id=#53
     possessed_class=FortVisualAttachment
     tracks [{
       type=MovieScene2DTransformTrack
       sections [{
         type=MovieScene2DTransformSection
         end_frame=0
         start_frame=0
         channels [
          Translation.X {type=MovieSceneScriptingFloatChannel, const [0,1]}
          Translation.Y {type=MovieSceneScriptingFloatChannel, const [15,1]}
          Angle {type=MovieSceneScriptingFloatChannel, const [0,1]}
          Scale.X {type=MovieSceneScriptingFloatChannel, const [1,1]}
          Scale.Y {type=MovieSceneScriptingFloatChannel, const [1,1]}
          Shear.X {type=MovieSceneScriptingFloatChannel, const [0,1]}
          Shear.Y {type=MovieSceneScriptingFloatChannel, const [0,1]}
         ]
        }]
      }]
    }
    CurrentBadge {
     id=#24
     possessed_class=Overlay
     tracks [
      {
       type=MovieSceneFloatTrack
       display_name=Render Opacity
       sections [{type=MovieSceneFloatSection, end_frame=39, start_frame=0, channels [None {type=MovieSceneScriptingFloatChannel, keys [[0,0],[9,0],[39,1]]}]}]
      }
      {
       type=MovieScene2DTransformTrack
       sections [{
         type=MovieScene2DTransformSection
         end_frame=72
         start_frame=0
         channels [
          Translation.Y {type=MovieSceneScriptingFloatChannel, const [0,1]}
          Scale.X {type=MovieSceneScriptingFloatChannel, keys [[0,2],[27,1],[72,1]]}
          Scale.Y {type=MovieSceneScriptingFloatChannel, keys [[0,2],[27,1],[72,1]]}
         ]
        }]
      }
     ]
    }
    Background {
     id=#25
     possessed_class=Image
     tracks [{
       type=MovieSceneWidgetMaterialTrack
       sections [{
         type=MovieSceneParameterSection
         end_frame=120
         channels [
          Color Swap {type=MovieSceneScriptingFloatChannel, const [1,1]}
          Hot Spot Intensity {type=MovieSceneScriptingFloatChannel, const [0.5,1]}
          Cloud Opacity {type=MovieSceneScriptingFloatChannel, const [0,1]}
          Triangle Opacity {type=MovieSceneScriptingFloatChannel, keys [[0,0],[119,0.25]]}
          Triangles Tiling {type=MovieSceneScriptingFloatChannel, const [4,1]}
          Rift Intensity {type=MovieSceneScriptingFloatChannel, const [0,1]}
          Rift Cracks Removal {type=MovieSceneScriptingFloatChannel, const [0.6863,1]}
          Invert Radial {type=MovieSceneScriptingFloatChannel, const [0.6,2]}
          Spot Scale {type=MovieSceneScriptingFloatChannel, const [1,2]}
         ]
        }]
      }]
    }
    RadialGlow {
     id=#26
     possessed_class=Image
     tracks [
      {
       type=MovieSceneFloatTrack
       display_name=Render Opacity
       sections [{
         type=MovieSceneFloatSection
         end_frame=119
         start_frame=0
         channels [None {type=MovieSceneScriptingFloatChannel, keys [[0,0],[14,0.4672],[23,0.75],[30,1]]}]
        }]
      }
      {
       type=MovieSceneWidgetMaterialTrack
       sections [{
         type=MovieSceneParameterSection
         end_frame=210
         start_frame=0
         channels [
          HexGridOpacity {type=MovieSceneScriptingFloatChannel, keys [[14,0],[23,0.02],[40,1],[180,1],[194,0]]}
          Opacity {type=MovieSceneScriptingFloatChannel, const [0.5,1]}
          Glow ColorR {type=MovieSceneScriptingFloatChannel, keys [[14,0],[23,0.02],[30,0]]}
          Glow ColorG {type=MovieSceneScriptingFloatChannel, keys [[14,0.3],[23,0.5],[30,0.1]]}
          Glow ColorB {type=MovieSceneScriptingFloatChannel, const [1,3]}
          Glow ColorA {type=MovieSceneScriptingFloatChannel, const [1,3]}
         ]
        }]
      }
      {
       type=MovieScene2DTransformTrack
       sections [{
         type=MovieScene2DTransformSection
         end_frame=210
         start_frame=0
         channels [
          Translation.X {type=MovieSceneScriptingFloatChannel, const [0,1]}
          Translation.Y {type=MovieSceneScriptingFloatChannel, keys [[0,12],[199,12],[205,0]]}
         ]
        }]
      }
     ]
    }
    Img_Flare {
     id=#28
     possessed_class=Image
     tracks [{
       type=MovieSceneWidgetMaterialTrack
       sections [{type=MovieSceneParameterSection, end_frame=119, start_frame=0, channels [Opacity {type=MovieSceneScriptingFloatChannel, const [0,1]}]}]
      }]
    }
    WBP_Habanero_Shield {
     id=#73
     possessed_class=WBP_Habanero_Shield_C
     tracks [
      {
       type=MovieSceneFloatTrack
       display_name=Render Opacity
       sections [{type=MovieSceneFloatSection, end_frame=210, start_frame=0, channels [None {type=MovieSceneScriptingFloatChannel, keys [[0,0],[33,1]]}]}]
      }
      {
       type=MovieSceneEventTrack
       sections [{
         type=MovieSceneEventTriggerSection
         channels [None {type=MovieSceneScriptingEventChannel, const ["<Struct 'MovieSceneEvent' (0x467612f6c20) {}>", 1]}]
        }]
      }
     ]
    }
   ]
   display_rate {numerator=60}
   tracks [{type=MovieSceneEventTrack, sections [{type=MovieSceneEventTriggerSection, channels [None {type=MovieSceneScriptingEventChannel}]}]}]
  }
 }
]
graphs [
 AddTimerValues {nodes [
   AddTimerValues() {is=title K2Node_FunctionEntry}
   > Branch(Condition="float < float"(A=Get CurrentTimer() -> (CurrentTimer:double), B=Get TotalMatchTime() -> (TotalMatchTime:double)))
   > n1 = Increment Float!(" ":Exec, Value:"Float (double-precision) (by ref)"=Get CurrentTimer()) -> ("  ":Exec, Result:double=~) {is=title K2Node_MacroInstance}
   > Set CurrentTimer(CurrentTimer:double=@n1) -> (Output_Get:double)
   > n2 = Increment Float(Value=Get SecondsValue() -> (SecondsValue:double)) -> (Result=~)
   > n3 = Set SecondsValue(SecondsValue:double=@n2) -> (Output_Get:double)
   > n4 = Branch(Condition="float >= float"(A=@n3, B=60))
   > Set SecondsValue(SecondsValue=0)
   > n5 = Increment Float(Value=Get MinutesValue() -> (MinutesValue:double)) -> (Result=~)
   > n6 = Set MinutesValue(MinutesValue:double=@n5) -> (Output_Get:double)
   Set Text(
    execute=[@n6.then, @n4.else]
    Text="To Text (Float)"(Value=Get MinutesValue(), MinimumIntegralDigits=1, MaximumIntegralDigits=324, MinimumFractionalDigits=0, MaximumFractionalDigits=3)
    self=Get Minutes() -> (Minutes:Common Text Object Reference)
   )
   > Set Text(
    Text="To Text (Float)"(Value=Get SecondsValue(), MinimumIntegralDigits=1, MaximumIntegralDigits=324, MinimumFractionalDigits=0, MaximumFractionalDigits=3)
    self=Get Seconds() -> (Seconds:Common Text Object Reference)
   )
  ]}
 AddWidgetSlotToCanvas {nodes [
   n1 = AddWidgetSlotToCanvas!() -> (then, Target:Widget Switcher Object Reference, Widget:Widget Object Reference) {is=title K2Node_FunctionEntry}
   > AddChild(self=@n1, Content=@n1.Widget)
   > SetActiveWidget(self:Widget Switcher Object Reference=@n1, Widget=@n1.Widget)
  ]}
 CacheDisplayedTierData {nodes [
   CacheDisplayedTierData() {is=title K2Node_FunctionEntry}
   n1 = GetFortHabaneroTierAtIndex!(Index:int=Get CurrentDisplayedTierIndex() -> (CurrentDisplayedTierIndex:int=0)) -> (HabaneroTierData:Fort Habanero Tier Structure)
   n2 = Get CurrentTopTier() -> (CurrentTopTier:bool)
   n3 = AND Boolean(
    A="Equal (Boolean)"(A:bool=Get InitialTopTier() -> (InitialTopTier:bool), B:bool=@n2) -> (ReturnValue:bool=~) {is=title K2Node_PromotableOperator}
    B=@n2
   )
   n4 = Get Top Tier Data() -> (Top Tier Data:Fort Habanero Tier Structure)
   > n5 = Cast To BP_UI_Habanero_TierTypeDataBase!(
    execute
    Object=Break Fort Habanero Tier!(FortHabaneroTier:"Fort Habanero Tier Structure (by ref)"=Select(Option 0:Fort Habanero Tier Structure=@n1, Option 1:Fort Habanero Tier Structure=@n4, Index=@n3) -> (ReturnValue:Fort Habanero Tier Structure)) -> (
     ForegroundIcon:Material Instance Soft Object Reference
     ForegroundIconSmall:Material Instance Soft Object Reference
     BackgroundIcon:Material Instance Soft Object Reference
     BackgroundIconSmall:Material Instance Soft Object Reference
     WhiteIcon:Texture 2D Soft Object Reference
     SimpleIconSmall:Texture 2D Soft Object Reference
     Name:Text
     TierType:Fort Habanero Tier Type Data Object Reference
     AdditionalLeftIcon:Material Instance Soft Object Reference
     AdditionalRightIcon:Material Instance Soft Object Reference
    ) {is=title K2Node_BreakStruct}.TierType
   ) -> (then, CastFailed, AsBP UI Habanero Tier Type Data Base:BP UI Habanero Tier Type Data Base Object Reference) {is=title K2Node_DynamicCast}
   > n6 = Set CurrentDisplayedTierCategoryData(CurrentDisplayedTierCategoryData:BP UI Habanero Tier Type Data Base Object Reference=@n5) -> (Output_Get:BP UI Habanero Tier Type Data Base Object Reference)
   n7 = Last Index!(TargetArray:Array of Fort Habanero Tier Structures=Get ActiveTierList() -> (ActiveTierList:Array of Fort Habanero Tier Structures)) -> (ReturnValue=0)
   n8 = Get CurrentDisplayedTierIndex() -> (CurrentDisplayedTierIndex=0)
   n9 = Get TargetDisplayedTierIndex() -> (TargetDisplayedTierIndex:int=0)
   n10 = int + int(A=@n8, B="Sign (Integer)"!(A:int=int - int(A=@n9, B=@n8)) -> (ReturnValue:int=0))
   n11 = Set NextTierIndex(
    execute=[@n6.then, @n5.CastFailed]
    NextTierIndex:int=Select(
     Option 0:int="Min (Integer)"(A=@n10, B=@n7) -> (ReturnValue=0)
     Option 1:int=Get TargetDisplayedTierIndex() -> (TargetDisplayedTierIndex=0)
     Index=WasUnranked!(self) -> (WasUnranked:bool)
    ) -> (ReturnValue:int)
   ) -> (Output_Get:int=0)
   n12 = GetFortHabaneroTierAtIndex(Index=@n11)
   n13 = Get CurrentTopTier()
   n14 = Get Top Tier Data()
   > n15 = Cast To BP_UI_Habanero_TierTypeDataBase(Object=Break Fort Habanero Tier(FortHabaneroTier=Select(Option 0:Fort Habanero Tier Structure=@n12, Option 1:Fort Habanero Tier Structure=@n14, Index=@n13) -> (ReturnValue:Fort Habanero Tier Structure)).TierType)
   > Set NextDisplayedTierCategoryData(NextDisplayedTierCategoryData:BP UI Habanero Tier Type Data Base Object Reference=@n15) -> (Output_Get:BP UI Habanero Tier Type Data Base Object Reference)
  ]}
 CalculateXOffset {nodes [
   n1 = CalculateXOffset!() -> (then, TotalMatchTime:double=~, EliminationMatchTime:double=~) {is=title K2Node_FunctionEntry}
   > Return Node!(
    execute
    ReturnValue:double=float * float(A=float / float(A=@n1.EliminationMatchTime, B=@n1), B=Get EliminationTimelineSize() -> (EliminationTimelineSize:double)) {is=title K2Node_PromotableOperator}
   )
  ]}
 CalculateXYOffset_Radial {nodes [
   n1 = CalculateXYOffset_Radial!() -> (then, TotalMatchTime:double=~, EliminationMatchTime:double=~) {is=title K2Node_FunctionEntry}
   > Set LinePosition(LinePosition:double=float / float(A=@n1.EliminationMatchTime, B=@n1)) -> (Output_Get:double)
   > Set LineAngle(LineAngle:double=float * float(A=Get LinePosition() -> (LinePosition:double), B=360)) -> (Output_Get:double)
   n2 = Get LineAngle() -> (LineAngle:double)
   > n3 = Set CartesianY(CartesianY:double=float * float(A="Cos (Degrees)"!(A:double=@n2) -> (ReturnValue:double), B=Get Radius() -> (Radius:double))) -> (Output_Get:double)
   n4 = Get LineAngle()
   > n5 = Set CartesianX(CartesianX:double=float * float(A="Sin (Degrees)"!(A:double=@n4) -> (ReturnValue:double), B=Get Radius())) -> (Output_Get:double)
   > Return Node!(execute, XOffset:double=@n5, YOffset:double=@n3, Angle:double=Get LineAngle())
  ]}
 ChangeLayoutForTournamentBanner {nodes [
   n1 = ChangeLayoutForTournamentBanner!() -> (then, DisplayTournamentBanner:bool=~) {is=title K2Node_FunctionEntry}
   > n2 = Branch(Condition=@n1)
   n3 = Get Stat_Right_0() -> (Stat_Right_0:WBP Post Match Habanero Sequence Stat Object Reference)
   > AddChild(self=Get StatsLeftContainer_VerticalBox() -> (StatsLeftContainer_VerticalBox:Vertical Box Object Reference), Content=@n3)
   n4 = Get DisplayTournamentBanner_Switcher() -> (DisplayTournamentBanner_Switcher:Common Visibility Switcher Object Reference)
   > SetActiveWidgetIndex(self:Common Visibility Switcher Object Reference=@n4, Index=1)
   AddChild(execute=@n2.else, self=Get StatRightContainer_VerticalBox() -> (StatRightContainer_VerticalBox:Vertical Box Object Reference), Content=@n3)
   > SetActiveWidgetIndex(self=@n4, Index=0)
  ]}
 DetermineStartingSequence {nodes [
   DetermineStartingSequence() {is=title K2Node_FunctionEntry}
   > Set ChangedTierIndex(ChangedTierIndex:bool="Not Equal (Integer)"(A=Get TargetDisplayedTierIndex() -> (TargetDisplayedTierIndex=0), B=Get CurrentDisplayedTierIndex() -> (CurrentDisplayedTierIndex=0))) -> (Output_Get:bool)
   n1 = "Equal (Integer)"(A=Get TargetDisplayedTierIndex() -> (TargetDisplayedTierIndex=0), B=Get CurrentDisplayedTierIndex() -> (CurrentDisplayedTierIndex=0))
   n2 = Select(
    Option 0:bool="float <= float"(A=Get TargetTierProgress() -> (TargetTierProgress:float), B=Get CurrentTierStartingProgress() -> (CurrentTierStartingProgress:float)) -> (ReturnValue=~)
    Option 1:bool="integer >= integer"(A=Get TargetPlayerPosition() -> (TargetPlayerPosition:int=0), B=Get CurrentPlayerPosition() -> (CurrentPlayerPosition:int=0))
    Index=IsFinalRankDisplayed!(self) -> (IsFinalRank:bool)
   ) -> (ReturnValue:bool)
   n3 = Get CurrentDisplayedTierIndex() -> (CurrentDisplayedTierIndex=0)
   n4 = Get TargetDisplayedTierIndex() -> (TargetDisplayedTierIndex=0)
   > n5 = Branch(Condition=OR Boolean(A="integer < integer"(A=@n4, B=@n3), B=AND Boolean(A=@n1, B=@n2)))
   > StartDecreasedProgessEvent()
   CacheDisplayedTierData(execute=@n5.else) {is=title K2Node_CallFunction}
   > RefreshVisuals()
   > n6 = Branch(Condition="Equal (Integer)"(A=Get TargetDisplayedTierIndex() -> (TargetDisplayedTierIndex=0), B=Get CurrentDisplayedTierIndex() -> (CurrentDisplayedTierIndex=0)))
   > StartIncreasedProgressEvent()
   StartLevelUpProgressEvent(execute=@n6.else)
  ]}
 EventGraph {nodes [
   Event PreConstruct()
   n1 = IsMobileGame()
   > SetPadding!(
    execute
    self:Vertical Box Slot Object Reference=SlotAsVerticalBoxSlot(Widget=Get StatsLeftContainer_VerticalBox())
    InPadding_Left:float=0
    InPadding_Top:float=0
    InPadding_Right:float=SelectFloat(A=-120, B=0, bPickA=@n1)
    InPadding_Bottom:float=0
   ) -> (then)
   > SetPadding(
    self=SlotAsVerticalBoxSlot(Widget=Get StatRightContainer_VerticalBox())
    InPadding_Left=SelectFloat(A=-120, B=0, bPickA=@n1)
    InPadding_Top=0
    InPadding_Right=0
    InPadding_Bottom=0
   )
   Event Construct()
   > Parent: Construct()
   > InitializeRankedViewModal(self:Ranked View Model Object Reference=Get FortRankedViewModel() -> (FortRankedViewModel:Ranked View Model Object Reference))
   > RegisterWidgetHost!(
    execute
    WidgetHost:Widget Object Reference
    WidgetHostTag:"Gameplay Tag Structure (by ref)"=Get RankedFeedbackTag() -> (RankedFeedbackTag:Gameplay Tag Structure)
    IsPlayerSpecific:bool=true
   ) -> (then)
   Event OnInitialized()
   > Sequence()
   > BindToAnimationFinished(Animation=Get Anim_GoDown() -> (Anim_GoDown:Widget Animation Object Reference), Delegate=Create Event())
   > BindToAnimationFinished(Animation=Get Anim_GoDown_FinalRank() -> (Anim_GoDown_FinalRank:Widget Animation Object Reference), Delegate=Create Event())
   > BindToAnimationFinished(Animation=Get Anim_Progress() -> (Anim_Progress:Widget Animation Object Reference), Delegate=Create Event())
   > BindToAnimationFinished(Animation=Get Anim_RankUp_Start() -> (Anim_RankUp_Start:Widget Animation Object Reference), Delegate=Create Event())
   > BindToAnimationFinished(Animation=Get Anim_RankUp_End() -> (Anim_RankUp_End:Widget Animation Object Reference), Delegate=Create Event())
   > BindToAnimationFinished(Animation=Get Anim_RankUp_FinalRank() -> (Anim_RankUp_FinalRank:Widget Animation Object Reference), Delegate=Create Event())
   > BindToAnimationFinished(Animation=Get Anim_TimeLineProgress() -> (Anim_TimeLineProgress:Widget Animation Object Reference), Delegate=Create Event())
   > BindToAnimationFinished(Animation=Get Anim_TimeLineFadeOut() -> (Anim_TimeLineFadeOut:Widget Animation Object Reference), Delegate=Create Event())
   > BindToAnimationFinished(Animation=Get Anim_UnrealLegends_End() -> (Anim_UnrealLegends_End:Widget Animation Object Reference), Delegate=Create Event())
   > BindToAnimationFinished(Animation=Get Anim_GoDownProtected() -> (Anim_GoDownProtected:Widget Animation Object Reference), Delegate=Create Event())
   > BindToAnimationFinished(
    Animation=Get Anim_GoDownProtected_FinalRank() -> (Anim_GoDownProtected_FinalRank:Widget Animation Object Reference)
    Delegate=Create Event()
   )
   > BindToAnimationFinished(Animation=Get Anim_UnrealLegends_Start() -> (Anim_UnrealLegends_Start:Widget Animation Object Reference), Delegate=Create Event())
   > Set AreStatsVisible(AreStatsVisible:bool=true) -> (Output_Get:bool)
   > SetVisibility(self=Get ContentWrapper() -> (ContentWrapper:Overlay Object Reference), InVisibility=Hidden)
   > SetVisibility(
    self=[
     Get StatsLeftContainer_VerticalBox()
     Get TextWrapper_RankUp() -> (TextWrapper_RankUp:Border Object Reference)
     Get StatRightContainer_VerticalBox()
     Get Image_PositionChange() -> (Image_PositionChange:Image Object Reference)
     Get Overlay_EliminationTimeline() -> (Overlay_EliminationTimeline:Overlay Object Reference)
    ]
    InVisibility=Collapsed
   )
   n2 = Event Add Content To Host Widget!() -> (
    OutputDelegate
    then
    WidgetHostTag:"Gameplay Tag Structure (by ref)"
    ContentWidget:Widget Object Reference
    Padding:"Margin Structure (by ref)"
    bFill:bool
   ) {is=title K2Node_Event}
   > Branch(Condition=MatchesTag!(TagOne:Gameplay Tag Structure=@n2.WidgetHostTag, TagTwo:Gameplay Tag Structure=Get RankedFeedbackTag(), bExactMatch:bool=false) -> (ReturnValue:bool))
   > AddWidgetSlotToCanvas!(
    execute
    self
    Target:Widget Switcher Object Reference=Get WidgetSwitcherSlot_RankedFeedback() -> (WidgetSwitcherSlot_RankedFeedback:Common Animated Switcher Object Reference)
    Widget:Widget Object Reference=@n2.ContentWidget
   ) -> (then) {is=title K2Node_CallFunction}
   Event Destruct()
   > UnregisterWidgetHost!(execute, WidgetHost:Widget Object Reference, WidgetHostTag:"Gameplay Tag Structure (by ref)"=Get RankedFeedbackTag()) -> (then)
  ]}
 GetFortHabaneroTierAtIndex {nodes [
   n1 = GetFortHabaneroTierAtIndex!() -> (then, Index:int) {is=title K2Node_FunctionEntry}
   > n2 = Branch(Condition="integer >= integer"(A=@n1))
   > Return Node!(
    execute
    HabaneroTierData:"Fort Habanero Tier Structure (by ref)"="Get (a ref)"(Array:Array of Fort Habanero Tier Structures=Get ActiveTierList(), Dimension 1=@n1) -> (Output:"Fort Habanero Tier Structure (by ref)")
   )
   Return Node(execute=@n2.else, HabaneroTierData=Get UnrankedTierData() -> (UnrankedTierData:Fort Habanero Tier Structure))
  ]}
 GetRankUpDataTable {nodes [
   GetRankUpDataTable() {is=title K2Node_FunctionEntry}
   n1 = Get RankType!(self:Ranked View Model Object Reference=Get FortRankedViewModel()) -> (RankType:String)
   > n2 = Branch(Condition="Equal Exactly (String)"(A=@n1, B=ranked-br))
   > Return Node!(execute, ReturnValue:Data Table Object Reference=Get RankUpData_BattleRoyale() -> (RankUpData_BattleRoyale:Data Table Object Reference))
   n3 = Branch(execute=@n2.else, Condition="Equal Exactly (String)"(A=@n1, B=ranked-zb))
   > Return Node(ReturnValue=Get RankUpData_ZeroBuild() -> (RankUpData_ZeroBuild:Data Table Object Reference))
   Return Node(execute=@n3.else)
  ]}
 GetRankUpMessage {nodes [
   GetRankUpMessage() {is=title K2Node_FunctionEntry}
   > n1 = GetConsoleVariableBoolValue!(execute, VariableName:String=Fort.Rollback.RankUpMessageEnabled) -> (then, ReturnValue:bool)
   > n2 = Branch(Condition=@n1)
   > n3 = GetRankUpDataTable!(execute, self) -> (then, ReturnValue:Data Table Object Reference) {is=title K2Node_CallFunction}
   > n4 = Is Valid(InputObject=@n3)
   Return Node!(execute=@n4.Is Not Valid, RankUpMessage:Text)
   > n5 = For Loop(
    FirstIndex=Get CurrentDisplayedTierIndex() -> (CurrentDisplayedTierIndex=0)
    LastIndex=Get TargetDisplayedTierIndex() -> (TargetDisplayedTierIndex=0)
   )
   > Branch(Condition="Not Equal (Integer)"(A=Get CurrentDisplayedTierIndex() -> (CurrentDisplayedTierIndex=0), B=@n5))
   > n6 = Get Data Table Row!(execute:Exec, DataTable:Data Table Object Reference=@n3, RowName:Name=String To Name(InString="To String (Integer)"(InInt=@n5))) -> (then:Exec, RowNotFound:Exec, ReturnValue:Fort Rank Up Data Entry Structure)
   n7 = Break Fort Rank Up Data Entry!(FortRankUpDataEntry:"Fort Rank Up Data Entry Structure (by ref)"=@n6) -> (RankUpMessage:Text) {is=title K2Node_BreakStruct}
   > Branch(Condition=NOT Boolean(A=TextIsEmpty(InText=@n7)))
   > Set MaxRankUpMessage(MaxRankUpMessage:Text=@n7) -> (Output_Get:Text)
   Return Node(execute=@n5.Completed, RankUpMessage=Get MaxRankUpMessage() -> (MaxRankUpMessage:Text))
   Return Node(execute=@n2.else)
  ]}
 Init Particles Colors {nodes [
   Init Particles Colors!() -> (then, TierCategoryData:BP UI Habanero Tier Type Data Base Object Reference) {is=title K2Node_FunctionEntry}
   > n1 = GetDynamicMaterial(self=Get RankParticles_Left() -> (RankParticles_Left:Image Object Reference))
   > n2 = GetDynamicMaterial(self=Get RankParticles_Right() -> (RankParticles_Right:Image Object Reference))
   n3 = Get NextDisplayedTierCategoryData() -> (NextDisplayedTierCategoryData:BP UI Habanero Tier Type Data Base Object Reference)
   > SetVectorParameterValue(
    self=[@n1, @n2]
    ParameterName=Get Particles Color() -> (Particles Color:Name=None)
    Value="To LinearColor (Color)"!(InColor:Color Structure=Get ParticlesOutward_Color!(self:BP UI Habanero Tier Type Data Base Object Reference=@n3) -> (ParticlesOutward_Color:Color Structure)) -> (ReturnValue:Linear Color Structure)
   )
   > n4 = GetDynamicMaterial(self=Get Particles_Inward_Medium() -> (Particles_Inward_Medium:Image Object Reference))
   n5 = Get TierCategoryData() -> (TierCategoryData:BP UI Habanero Tier Type Data Base Object Reference)
   > SetVectorParameterValue(
    self=@n4
    ParameterName=Get Particles Color() -> (Particles Color=None)
    Value="To LinearColor (Color)"(InColor=Get ParticlesInwardMedium_Color!(self:BP UI Habanero Tier Type Data Base Object Reference=@n5) -> (ParticlesInwardMedium_Color:Color Structure))
   )
   > n6 = GetDynamicMaterial(self=Get Particles_Inward_Small() -> (Particles_Inward_Small:Image Object Reference))
   n7 = Get TierCategoryData()
   > SetVectorParameterValue(
    self=@n6
    ParameterName=Get Particles Color() -> (Particles Color=None)
    Value="To LinearColor (Color)"(InColor=Get ParticlesInwardSmall_Color!(self:BP UI Habanero Tier Type Data Base Object Reference=@n7) -> (ParticlesInwardSmall_Color:Color Structure))
   )
   > n8 = GetDynamicMaterial(self=Get Particles_Inward_Big() -> (Particles_Inward_Big:Image Object Reference))
   n9 = Get TierCategoryData()
   > SetVectorParameterValue(
    self=@n8
    ParameterName=Get Particles Color() -> (Particles Color=None)
    Value="To LinearColor (Color)"(InColor=Get ParticlesInwardBig_Color!(self:BP UI Habanero Tier Type Data Base Object Reference=@n9) -> (ParticlesInwardBig_Color:Color Structure))
   )
   > n10 = GetDynamicMaterial(self=Get Img_Flare() -> (Img_Flare:Image Object Reference))
   n11 = Get TierCategoryData()
   > SetVectorParameterValue(
    self=@n10
    ParameterName=Get Flare Color() -> (Flare Color:Name=None)
    Value="To LinearColor (Color)"(InColor=Get Flare_Color!(self:BP UI Habanero Tier Type Data Base Object Reference=@n11) -> (Flare_Color:Color Structure))
   )
  ]}
 InitBackgroundColors {nodes [
   InitBackgroundColors() {is=title K2Node_FunctionEntry}
   > n1 = GetDynamicMaterial(self=Get Background() -> (Background:Image Object Reference))
   n2 = Get CurrentDisplayedTierCategoryData() -> (CurrentDisplayedTierCategoryData:BP UI Habanero Tier Type Data Base Object Reference)
   > SetVectorParameterValue(
    self=@n1
    ParameterName=Get Background Current Color() -> (Background Current Color:Name=None)
    Value="To LinearColor (Color)"(InColor=Get Background_Color!(self:BP UI Habanero Tier Type Data Base Object Reference=@n2) -> (Background_Color:Color Structure))
   )
   > SetVectorParameterValue(
    self=@n1
    ParameterName=Get Background Next Color() -> (Background Next Color:Name=None)
    Value=Get FinalBgColor() -> (FinalBgColor:Linear Color Structure)
   )
   > n3 = GetDynamicMaterial(self=Get ProgressBar() -> (ProgressBar:Image Object Reference))
   n4 = Get NextDisplayedTierCategoryData()
   > SetVectorParameterValue(
    self=@n3
    ParameterName=Edge Color
    Value="To LinearColor (Color)"(InColor=Get EdgeBar_Color!(self:BP UI Habanero Tier Type Data Base Object Reference=@n4) -> (EdgeBar_Color:Color Structure))
   )
   "To LinearColor (Color)"(InColor=Get Background_Color(self=Get NextDisplayedTierCategoryData()))
  ]}
 InitializeAllStatsAsCollapsed {nodes [
   InitializeAllStatsAsCollapsed() {is=title K2Node_FunctionEntry}
   > SetVisibility(
    self=[
     Get Stat_Left_1() -> (Stat_Left_1:WBP Post Match Habanero Sequence Stat Object Reference)
     Get Stat_Left_0() -> (Stat_Left_0:WBP Post Match Habanero Sequence Stat Object Reference)
     Get Stat_Right_1() -> (Stat_Right_1:WBP Post Match Habanero Sequence Stat Object Reference)
     Get Stat_Right_0()
    ]
    InVisibility=Collapsed
   )
  ]}
 InitializeTournamentData {nodes [
   n1 = InitializeTournamentData!() -> (then, NextTournamentData:Next Tournament Data Structure) {is=title K2Node_FunctionEntry}
   > InitializeTournamentData!(
    execute
    self:WBP Tournament Small Banner Object Reference=Get WBP_TournamentSmallBanner() -> (WBP_TournamentSmallBanner:WBP Tournament Small Banner Object Reference)
    NextTournamentData:Next Tournament Data Structure=@n1
   ) -> (then) {is=title K2Node_CallFunction}
   n2 = Break Next Tournament Data!(NextTournamentData:"Next Tournament Data Structure (by ref)"=@n1) -> (
    TournamentName:String
    StartTime:Date Time Structure
    EndTime:Date Time Structure
    bIsLoadingImage:bool=~
    TournamentTexture:Texture 2D Object Reference
   ) {is=title K2Node_BreakStruct}
   > ChangeLayoutForTournamentBanner!(execute, self, DisplayTournamentBanner:bool=NOT Boolean(A=IsEmpty(InString=@n2))) -> (then) {is=title K2Node_CallFunction}
  ]}
 InitilizeTimeline {nodes [
   n1 = InitilizeTimeline!() -> (then, EliminationEntries:Array of Elimination Timeline Entry Structures, TotalMatchTime:double=~) {is=title K2Node_FunctionEntry}
   > Set EliminationEntries(EliminationEntries:Array of Elimination Timeline Entry Structures=@n1) -> (Output_Get:Array of Elimination Timeline Entry Structures)
   > Set TotalMatchTime(TotalMatchTime:double=@n1.TotalMatchTime) -> (Output_Get:double)
   > ClearChildren(self=Get Overlay_TimelineEntries() -> (Overlay_TimelineEntries:Overlay Object Reference))
   n2 = "integer > integer"(A=Length(TargetArray:Array of Elimination Timeline Entry Structures=Get EliminationEntries() -> (EliminationEntries:Array of Elimination Timeline Entry Structures)))
   > SetVisibility(
    self=Get Overlay_EliminationTimeline()
    InVisibility=Select(Option 0:ESlateVisibility Enum=Collapsed, Option 1:ESlateVisibility Enum=HitTestInvisible, Index=@n2) -> (ReturnValue:ESlateVisibility Enum)
   )
   > Branch(Condition=@n2)
   > PlayAnimationForward(InAnimation=Get Anim_TimeLineProgress(), PlaybackSpeed=0.25)
   > Set Timer by Event(Delegate=Create Event(), Time=float / float(A=4, B=Get TotalMatchTime()), bLooping=true)
   n3 = Break Elimination Timeline Entry!(EliminationTimelineEntry:"Elimination Timeline Entry Structure (by ref)"="Get (a copy)"(Array:Array of Elimination Timeline Entry Structures=Get EliminationEntries()) -> (Output:Elimination Timeline Entry Structure)) -> (
    PlayerName:String
    PlayerRankType:String
    PlayerRank:int
    PlayerCurrentPosition:int
    PlayerMMR:int
    EliminatedAvatarBrushPtr:Texture 2D Soft Object Reference
    PlayersLeft:int
    PlayersTotal:int
    Time:float=~
    bLocalElimination:bool=~
    bPlayersLeftEntry:bool=~
   ) {is=title K2Node_BreakStruct}
   n4 = Get TotalMatchTime()
   > Set Timer by Event(Delegate=Create Event(), Time=float * float(A=float / float(A=@n3.Time, B=@n4), B=4), bLooping=false)
  ]}
 IsFinalRankDisplayed {nodes [
   IsFinalRankDisplayed() {is=title K2Node_FunctionEntry}
   n1 = Get ActiveTierList()
   > Return Node!(
    execute
    IsFinalRank:bool="Equal (Integer)"(A=Get CurrentDisplayedTierIndex() -> (CurrentDisplayedTierIndex=0), B=Last Index(TargetArray=@n1) -> (ReturnValue=0))
   )
  ]}
 OnAnimGoDownFinished {nodes [
   OnAnimGoDownFinished() {is=title K2Node_FunctionEntry}
   > StartSequenceFinished()
   > Sequence IMF Animations()
  ]}
 OnAnimProgressFinished {nodes [
   OnAnimProgressFinished() {is=title K2Node_FunctionEntry}
   > PlayProgressAudio!(execute, self, bShouldPlay:bool=false, bSouldShake:bool=false) -> (then)
   > StartSequenceFinished()
  ]}
 OnAnimRankScale {nodes [OnAnimRankScale() {is=title K2Node_FunctionEntry}]}
 OnAnimRankUpEndFinished {nodes [
   OnAnimRankUpEndFinished() {is=title K2Node_FunctionEntry}
   > StartSequenceFinished()
   > PlayProgressAudio(bShouldPlay=false, bSouldShake=false)
  ]}
 OnAnimRankUpFinalRankFinished {nodes [
   OnAnimRankUpFinalRankFinished() {is=title K2Node_FunctionEntry}
   > PlayProgressAudio(bShouldPlay=false, bSouldShake=false)
   > StartSequenceFinished()
   > Sequence IMF Animations()
  ]}
 OnAnimRankUpStartFinished {nodes [
   OnAnimRankUpStartFinished() {is=title K2Node_FunctionEntry}
   > Set CurrentDisplayedTierIndex(CurrentDisplayedTierIndex:int=Get NextTierIndex() -> (NextTierIndex:int=0)) -> (Output_Get:int=0)
   > Set CurrentTierStartingProgress(CurrentTierStartingProgress:float=0) -> (Output_Get:float)
   > CacheDisplayedTierData()
   > RefreshVisuals()
   > UpdateSequenceProgress()
   > PlayAnimation(InAnimation=Get Anim_RankUp_FlashFadeOut() -> (Anim_RankUp_FlashFadeOut:Widget Animation Object Reference))
   > PlaySound2D(Sound=Get FlashTransitionSound() -> (FlashTransitionSound:Sound Base Object Reference))
  ]}
 OnSequenceFinished {nodes [OnSequenceFinished() {is=title K2Node_FunctionEntry}]}
 OnSkiilRatingVisibilityChanged {nodes [OnSkiilRatingVisibilityChanged!() -> (then, IsVisible:bool=~) {is=title K2Node_FunctionEntry}]}
 OnTopTierStartAnimationFinished {nodes [
   OnTopTierStartAnimationFinished() {is=title K2Node_FunctionEntry}
   > Set InitialTopTier(InitialTopTier:bool=true) -> (Output_Get:bool)
   > CacheDisplayedTierData()
   > RefreshVisuals()
   > PlayAnimation(InAnimation=Get Anim_UnrealLegends_End())
  ]}
 PlayProgressAudio {nodes [PlayProgressAudio!() -> (then, bShouldPlay:bool=true, bSouldShake:bool=~) {is=title K2Node_FunctionEntry}]}
 RefreshVisuals {nodes [
   RefreshVisuals() {is=title K2Node_FunctionEntry}
   > Init Particles Colors!(execute, self, TierCategoryData:BP UI Habanero Tier Type Data Base Object Reference=Get CurrentDisplayedTierCategoryData()) -> (then) {is=title K2Node_CallFunction}
   > InitBackgroundColors() {is=title K2Node_CallFunction}
   n1 = Get CurrentDisplayedTierIndex() -> (CurrentDisplayedTierIndex=0)
   n2 = Get CurrentTopTier()
   n3 = "Equal (Boolean)"(A=Get InitialTopTier(), B=@n2) -> (ReturnValue=~)
   n4 = Break Fort Habanero Tier(FortHabaneroTier=Select(
     Option 0:Fort Habanero Tier Structure=GetFortHabaneroTierAtIndex(Index=@n1) {is=title K2Node_CallFunction}
     Option 1:Fort Habanero Tier Structure=Get Top Tier Data()
     Index=AND Boolean(A=@n3, B=@n2)
    ) -> (ReturnValue:Fort Habanero Tier Structure))
   n5 = Get NextTierIndex() -> (NextTierIndex=0)
   n6 = Get CurrentTopTier()
   n7 = "Equal (Boolean)"(A=Get InitialTopTier(), B=@n6) -> (ReturnValue=~)
   n8 = Break Fort Habanero Tier(FortHabaneroTier=Select(Option 0=GetFortHabaneroTierAtIndex(Index=@n5), Option 1=Get Top Tier Data(), Index=AND Boolean(A=@n7, B=@n6)))
   > SetBadgeIcons!(
    execute
    self
    CurrentBadgeIcon:Material Instance Soft Object Reference=@n4
    NextBadgeIcon:Material Instance Soft Object Reference=@n8
    NextBadgeBg:Material Instance Soft Object Reference=@n8.BackgroundIcon
   ) -> (then)
   > "SetText (Text)"(self=Get Text_TierName() -> (Text_TierName:Common Text Object Reference), InText=@n4.Name)
   > n9 = Sequence()
   n10 = GetCurrentLanguage!() -> (ReturnValue:String)
   > SetLineHeightPercentage!(
    execute
    self:Common Text Object Reference=Get Text_TierName()
    InLineHeightPercentage:float=SelectFloat(A=1.2, B=0.9, bPickA="Equal, Case Insensitive (String)"(A=@n10, B=ar))
   ) -> (then)
   UpdateProgressText!(execute=@n9.then_1, self, Percent:double=Get CurrentTierStartingProgress()) -> (then)
   > SetActiveWidgetIndex(
    self:Widget Switcher Object Reference=Get RankProtectedBox_Switcher() -> (RankProtectedBox_Switcher:Common Animated Switcher Object Reference)
    Index=0
   )
   > SetRenderOpacity(self=Get Text_RankUpMessage() -> (Text_RankUpMessage:Common Text Object Reference), InOpacity=0)
  ]}
 RemoveContentFromWidgetHost {nodes [
   n1 = Remove Content From Host Widget!() -> (
    then
    WidgetHostTag:"Gameplay Tag Structure (by ref)"
    ContentWidget:Widget Object Reference
    RemovalType:EWidgetSlottingRemovalType Enum=Unparent
   ) {is=title K2Node_FunctionEntry}
   > n2 = Branch(Condition=MatchesTag(TagOne=@n1, TagTwo=Get RankedFeedbackTag(), bExactMatch=false))
   n3 = Get WidgetSwitcherSlot_RankedFeedback()
   > n4 = RemoveChild(self=@n3, Content=@n1.ContentWidget)
   > n5 = Branch(Condition=@n4)
   > n6 = SetActiveWidgetIndex(self=@n3, Index=0)
   Return Node!(execute=[@n6.then, @n5.else], ReturnValue:bool=@n4)
   Return Node(execute=@n2.else, ReturnValue=false)
  ]}
 Sequencer Events {nodes [
   SequenceEvent_HideStats() {is=title K2Node_CustomEvent}
   > Branch(Condition=Get AreStatsVisible() -> (AreStatsVisible:bool))
   n1 = Get VM_RankedFeedbackAnimationTrigger() -> (VM_RankedFeedbackAnimationTrigger:VM Ranked Feedback Animation Trigger Object Reference)
   > Set with Broadcast RankedFeedbackBonusAnimationFired!(
    execute
    RankedFeedbackBonusAnimationFired:bool=InvertBool!(bValue:bool=Get RankedFeedbackBonusAnimationFired!(self:VM Ranked Feedback Animation Trigger Object Reference=@n1) -> (RankedFeedbackBonusAnimationFired:bool)) -> (ReturnValue:bool)
    self:VM Ranked Feedback Animation Trigger Object Reference=@n1
   ) -> (then, Output_Get:bool)
   > PlayAnimation(InAnimation=Get Anim_Stats_Hide() -> (Anim_Stats_Hide:Widget Animation Object Reference))
   > Set AreStatsVisible(AreStatsVisible=false)
   StartProgress_01() {is=title K2Node_CustomEvent}
   > n2 = Branch(Condition=Get bForward() -> (bForward:bool))
   n3 = SpawnSound2D!(
    execute=[@n10.then, @n2.then]
    Sound:Sound Base Object Reference=Get LoopProgressSound() -> (LoopProgressSound:Sound Base Object Reference)
    VolumeMultiplier:float=1
    PitchMultiplier:float=1
    StartTime:float=0
    ConcurrencySettings:Sound Concurrency Object Reference
    bPersistAcrossLevelTransition:bool=false
    bAutoDestroy:bool=true
   ) -> (then, ReturnValue:Audio Component Object Reference)
   > Set ProgressAudioComponent(ProgressAudioComponent:Audio Component Object Reference=@n3) -> (Output_Get:Audio Component Object Reference)
   ProgressBarQuickWhoosh() {is=title K2Node_CustomEvent}
   > PlaySound2D(Sound=Get ProgressBarEndWhoosh() -> (ProgressBarEndWhoosh:Sound Base Object Reference))
   BadgeAppear() {is=title K2Node_CustomEvent}
   > Branch(Condition=Get bForward())
   > PlaySound2D(Sound=Select!(
     Option 0:Sound Base Object Reference=Get BadgeAppearSound_01() -> (BadgeAppearSound_01:Sound Base Object Reference)
     Option 1:Sound Base Object Reference=Get BadgeAppearSound_02() -> (BadgeAppearSound_02:Sound Base Object Reference)
     Option 2:Sound Base Object Reference=Get BadgeAppearSound_03() -> (BadgeAppearSound_03:Sound Base Object Reference)
     Index:int=Get BadgeCounter() -> (BadgeCounter:int=0)
    ) -> (ReturnValue:Sound Base Object Reference))
   > n4 = Increment Int(Value=Get BadgeCounter() -> (BadgeCounter=0))
   > Branch(Condition="integer >= integer"(A=@n4, B=3))
   > Set BadgeCounter(BadgeCounter:int=0) -> (Output_Get:int=0)
   StartShakeSound() {is=title K2Node_CustomEvent}
   > Branch(Condition=Get bForward())
   > SpawnSound2D(
    Sound=Get ShakeSound() -> (ShakeSound:Sound Base Object Reference)
    VolumeMultiplier=1
    PitchMultiplier=1
    StartTime=0
    bPersistAcrossLevelTransition=false
    bAutoDestroy=true
   )
   ExplodeSound() {is=title K2Node_CustomEvent}
   > Branch(Condition=Get bForward())
   > SpawnSound2D(
    Sound=Get BurstCompletedSound() -> (BurstCompletedSound:Sound Base Object Reference)
    VolumeMultiplier=1
    PitchMultiplier=1
    StartTime=0
    bPersistAcrossLevelTransition=false
    bAutoDestroy=true
   )
   > n5 = Get!(execute:Exec) -> (then:Exec, else:Exec, ProgressAudioComponent:Audio Component Object Reference)
   > Stop(self:Audio Component Object Reference=@n5)
   Sequence IMF Animations() {is=title K2Node_CustomEvent}
   > Play IMF Placement Animation()
   > Delay(Duration=0.5)
   > Play IMF Elim Animation()
   PlayLevelUpEndAnimations() {is=title K2Node_CustomEvent}
   n6 = Get Anim_RankUp_End()
   > PlayAnimation(InAnimation=@n6)
   > Delay(Duration=1.5)
   > Sequence IMF Animations() {is=title K2Node_CallFunction}
   > n7 = IsAnimationPlayingForward!(execute, self:User Widget Object Reference, InAnimation:Widget Animation Object Reference=@n6) -> (then, ReturnValue:bool)
   > Set bForward(bForward:bool=@n7) -> (Output_Get:bool)
   UnrealBadgeEnter() {is=title K2Node_CustomEvent}
   > Branch(Condition=Get bForward())
   > SpawnSound2D(
    Sound=Get UnrealBadgeEnterSound() -> (UnrealBadgeEnterSound:Sound Base Object Reference)
    VolumeMultiplier=1
    PitchMultiplier=1
    StartTime=0
    bPersistAcrossLevelTransition=false
    bAutoDestroy=true
   )
   UnrealBadgeEnterWhoosh() {is=title K2Node_CustomEvent}
   > Branch(Condition=Get bForward())
   > SpawnSound2D(
    Sound=Get UnrealBadgeEnterWhooshSound() -> (UnrealBadgeEnterWhooshSound:Sound Base Object Reference)
    VolumeMultiplier=1
    PitchMultiplier=1
    StartTime=0
    bPersistAcrossLevelTransition=false
    bAutoDestroy=true
   )
   UnrealBadgeWhooshOut() {is=title K2Node_CustomEvent}
   > Branch(Condition=Get bForward())
   > SpawnSound2D(
    Sound=Get UnrealBadgeWhooshOutSound() -> (UnrealBadgeWhooshOutSound:Sound Base Object Reference)
    VolumeMultiplier=1
    PitchMultiplier=1
    StartTime=0
    bPersistAcrossLevelTransition=false
    bAutoDestroy=true
   )
   UnrealLegendBadgeEnter() {is=title K2Node_CustomEvent}
   > Branch(Condition=Get bForward())
   > SpawnSound2D(
    Sound=Get UnrealLegendBadgeEnterSound() -> (UnrealLegendBadgeEnterSound:Sound Base Object Reference)
    VolumeMultiplier=1
    PitchMultiplier=1
    StartTime=0
    bPersistAcrossLevelTransition=false
    bAutoDestroy=true
   )
   n8 = LettersAppearAudio_01() {is=title K2Node_CustomEvent}
   Branch(execute=[@n8.then, @n11.then], Condition=Get bForward())
   > SpawnSound2D(
    Sound=Get LettersAppearSound() -> (LettersAppearSound:Sound Base Object Reference)
    VolumeMultiplier=1
    PitchMultiplier=1
    StartTime=0
    bPersistAcrossLevelTransition=false
    bAutoDestroy=true
   )
   UnrealLegendFinalImpact() {is=title K2Node_CustomEvent}
   > Branch(Condition=Get bForward())
   > SpawnSound2D(
    Sound=Get UnrealLegendFinalImpactSound() -> (UnrealLegendFinalImpactSound:Sound Base Object Reference)
    VolumeMultiplier=1
    PitchMultiplier=1
    StartTime=0
    bPersistAcrossLevelTransition=false
    bAutoDestroy=true
   )
   UnrealLegendBadgeEnterWhoosh() {is=title K2Node_CustomEvent}
   > Branch(Condition=Get bForward())
   > SpawnSound2D(
    Sound=Get UnrealBadgeEnterWhooshSound()
    VolumeMultiplier=1
    PitchMultiplier=1
    StartTime=0
    bPersistAcrossLevelTransition=false
    bAutoDestroy=true
   )
   UnrealLegendWhooshSweetener() {is=title K2Node_CustomEvent}
   > Branch(Condition=Get bForward())
   > SpawnSound2D(
    Sound=Get UnrealLegendWhooshSweetenerSound() -> (UnrealLegendWhooshSweetenerSound:Sound Base Object Reference)
    VolumeMultiplier=1
    PitchMultiplier=1
    StartTime=0
    bPersistAcrossLevelTransition=false
    bAutoDestroy=true
   )
   UnrealWhooshOutSweetener() {is=title K2Node_CustomEvent}
   > Branch(Condition=Get bForward())
   > SpawnSound2D(
    Sound=Get UnrealWhooshOutSweetenerSound() -> (UnrealWhooshOutSweetenerSound:Sound Base Object Reference)
    VolumeMultiplier=1
    PitchMultiplier=1
    StartTime=0
    bPersistAcrossLevelTransition=false
    bAutoDestroy=true
   )
   n9 = StartProgress_03() {is=title K2Node_CustomEvent}
   Branch(execute=[@n9.then, @n20.then], Condition=Get bForward())
   > n10 = SpawnSound2D(
    Sound=Get LoopStartSound() -> (LoopStartSound:Sound Base Object Reference)
    VolumeMultiplier=1
    PitchMultiplier=1
    StartTime=0
    bPersistAcrossLevelTransition=false
    bAutoDestroy=true
   )
   n11 = LettersAppearAudio_02() {is=title K2Node_CustomEvent}
   n12 = WBP_Habanero_Shield_Event!() -> (OutputDelegate, then, WBP_Habanero_Shield:WBP Habanero Shield Object Reference) {is=title K2Node_CustomEvent}
   > PlayAnimation(self=@n12.WBP_Habanero_Shield, InAnimation=Get Intro!(self:WBP Habanero Shield Object Reference=@n12.WBP_Habanero_Shield) -> (Intro))
   > PlaySound2D(Sound=/FortUIAudioContent/SoundCues/Fort_UI_PostGame_Ranked_Protection_Cue)
   Play Progress Animations() {is=title K2Node_CustomEvent}
   n13 = float - float(A=Get CurrentTierFinalProgess() -> (CurrentTierFinalProgess:float), B=Get CurrentTierStartingProgress())
   n14 = Get StartScalingAnimationBelowProgress() -> (StartScalingAnimationBelowProgress:double)
   n15 = Get MaxProgressAnimationScale() -> (MaxProgressAnimationScale:double)
   n16 = Lerp(
    A=@n15
    B=1
    Alpha="Min (Float)"(A:double=float / float(A=@n13, B=@n14), B:double=1) -> (ReturnValue:double) {is=title K2Node_CommutativeAssociativeBinaryOperator}
   )
   n17 = Get Anim_Progress()
   > PlayAnimation(InAnimation=@n17, PlaybackSpeed=@n16)
   > Delay(Duration=Lerp(A=0.8, B=0.56, Alpha=RemapValueToInterval!(Input:double=@n16, InputLow:double=1, InputHigh:double=@n15) -> (Result:double)))
   > Sequence IMF Animations()
   > n18 = IsAnimationPlayingForward(InAnimation=@n17)
   > Set bForward(bForward=@n18)
   Play IMF Elim Animation() {is=title K2Node_CustomEvent}
   > PlayAnimation(InAnimation=Get Anim_BonusFeedback_Elim_V02() -> (Anim_BonusFeedback_Elim_V02:Widget Animation Object Reference))
   n19 = Get VM_RankedFeedbackAnimationTrigger()
   > Set with Broadcast IMF_KillsAnimTrigger!(execute, IMF_KillsAnimTrigger:bool=true, self:VM Ranked Feedback Animation Trigger Object Reference=@n19) -> (then, Output_Get:bool)
   n20 = StartProgress_02() {is=title K2Node_CustomEvent}
   Play IMF Placement Animation() {is=title K2Node_CustomEvent}
   > PlayAnimation(InAnimation=Get Anim_BonusFeedback_Placement_V02() -> (Anim_BonusFeedback_Placement_V02:Widget Animation Object Reference))
   > Set with Broadcast IMF_PlacementAnimTrigger!(execute, IMF_PlacementAnimTrigger:bool=true, self:VM Ranked Feedback Animation Trigger Object Reference=@n19) -> (then, Output_Get:bool)
  ]}
 SetBadgeIcons {nodes [
   n1 = SetBadgeIcons!() -> (
    then
    CurrentBadgeIcon:Material Instance Soft Object Reference
    NextBadgeIcon:Material Instance Soft Object Reference
    NextBadgeBg:Material Instance Soft Object Reference
   ) {is=title K2Node_FunctionEntry}
   > SetBrushFromSoftMaterial!(
    execute
    self:Image Object Reference=Get CurrentBadge_Icon() -> (CurrentBadge_Icon:Image Object Reference)
    SoftMaterial:Material Interface Soft Object Reference=@n1
   ) -> (then)
   > SetBrushFromSoftMaterial(self=Get NextBadge_Icon() -> (NextBadge_Icon:Image Object Reference), SoftMaterial=@n1.NextBadgeIcon)
   > SetBrushFromSoftMaterial(self=Get NextBadge_Bg() -> (NextBadge_Bg:Image Object Reference), SoftMaterial=@n1.NextBadgeBg)
  ]}
 SetDeltaValue_Blend {nodes [
   n1 = SetDeltaValue_Blend!() -> (then, NewParam:double=~) {is=title K2Node_FunctionEntry}
   > n2 = GetDynamicMaterial(self=Get ProgressBar())
   n3 = Lerp(A=Get CurrentTierStartingProgress(), B=Get CurrentTierFinalProgess(), Alpha=@n1)
   > SetScalarParameterValue(self=@n2, ParameterName=Get Delta Value Name() -> (Delta Value Name:Name=None), Value=@n3)
   > UpdateProgressText(Percent=@n3)
  ]}
 SetProgressBarColor_Blend {nodes [
   SetProgressBarColor_Blend!() -> (then, Progress:double=~) {is=title K2Node_FunctionEntry}
   > n1 = GetDynamicMaterial(self=Get ProgressBar())
   n2 = Get ProgressBar_Color!(self:BP UI Habanero Tier Type Data Base Object Reference=Get CurrentDisplayedTierCategoryData()) -> (ProgressBar_Color:Color Structure)
   n3 = Get ProgressBar_Color(self=Get NextDisplayedTierCategoryData())
   > SetVectorParameterValue(
    self=@n1
    ParameterName=Get Progress Bar Color() -> (Progress Bar Color:Name=None)
    Value="Lerp (LinearColor)"(A="To LinearColor (Color)"(InColor=@n2), B="To LinearColor (Color)"(InColor=@n3), Alpha=Get Progress() -> (Progress:double))
   )
   n4 = Get ProgressBackground_Color!(self:BP UI Habanero Tier Type Data Base Object Reference=Get CurrentDisplayedTierCategoryData()) -> (ProgressBackground_Color:Color Structure)
   > SetVectorParameterValue(
    self=@n1
    ParameterName=Get Progress Background Color() -> (Progress Background Color:Name=None)
    Value="Lerp (LinearColor)"(A="To LinearColor (Color)"(InColor=@n4), B=Get ProgressBgColor() -> (ProgressBgColor:Linear Color Structure), Alpha=Get Progress())
   )
   n5 = Get DeltaBar_Color!(self:BP UI Habanero Tier Type Data Base Object Reference=Get CurrentDisplayedTierCategoryData()) -> (DeltaBar_Color:Color Structure)
   n6 = Get DeltaBar_Color(self=Get NextDisplayedTierCategoryData())
   > SetVectorParameterValue(
    self=@n1
    ParameterName=Get Delta Bar Color() -> (Delta Bar Color:Name=None)
    Value="Lerp (LinearColor)"(A="To LinearColor (Color)"(InColor=@n5), B="To LinearColor (Color)"(InColor=@n6), Alpha=Get Progress())
   )
   n7 = Get BarGlow_Color!(self:BP UI Habanero Tier Type Data Base Object Reference=Get CurrentDisplayedTierCategoryData()) -> (BarGlow_Color:Color Structure)
   n8 = Get BarGlow_Color(self=Get NextDisplayedTierCategoryData())
   > SetVectorParameterValue(
    self=@n1
    ParameterName=Get Bar Glow Color() -> (Bar Glow Color:Name=None)
    Value="Lerp (LinearColor)"(A="To LinearColor (Color)"(InColor=@n7), B="To LinearColor (Color)"(InColor=@n8), Alpha=Get Progress())
   )
   n9 = Get ProgressOutline_Color!(self:BP UI Habanero Tier Type Data Base Object Reference=Get CurrentDisplayedTierCategoryData()) -> (ProgressOutline_Color:Color Structure)
   n10 = Get ProgressOutline_Color(self=Get NextDisplayedTierCategoryData())
   > SetVectorParameterValue(
    self=@n1
    ParameterName=Get Progress Outline Color() -> (Progress Outline Color:Name=None)
    Value="Lerp (LinearColor)"(A="To LinearColor (Color)"(InColor=@n9), B="To LinearColor (Color)"(InColor=@n10), Alpha=Get Progress())
   )
   n11 = Get InnerBackground_Color!(self:BP UI Habanero Tier Type Data Base Object Reference=Get CurrentDisplayedTierCategoryData()) -> (InnerBackground_Color:Color Structure)
   n12 = Get InnerBackground_Color(self=Get NextDisplayedTierCategoryData())
   > SetVectorParameterValue(
    self=@n1
    ParameterName=Get Inner Color() -> (Inner Color:Name=None)
    Value="Lerp (LinearColor)"(A="To LinearColor (Color)"(InColor=@n11), B="To LinearColor (Color)"(InColor=@n12), Alpha=Get Progress())
   )
   "To LinearColor (Color)"(InColor=Get ProgressBackground_Color(self=Get NextDisplayedTierCategoryData()))
  ]}
 SetProgressValue_Blend {nodes [
   n1 = SetProgressValue_Blend!() -> (then, NewParam:double=~) {is=title K2Node_FunctionEntry}
   > n2 = GetDynamicMaterial(self=Get ProgressBar())
   > SetScalarParameterValue(
    self=@n2
    ParameterName=Get Progress Value Name() -> (Progress Value Name:Name=None)
    Value=Lerp(A=Get CurrentTierStartingProgress(), B=Get CurrentTierFinalProgess(), Alpha=@n1)
   )
  ]}
 SetStatAtPosition {nodes [
   n1 = SetStatAtPosition!() -> (then, StatTitle:Text, StatValue:int, StatPosition:ENUM_HabaneroSequence_StatPosition Enum=NewEnumerator0) {is=title K2Node_FunctionEntry}
   n2 = Select!(
    NewEnumerator0:WBP Post Match Habanero Sequence Stat Object Reference=Get Stat_Left_0()
    NewEnumerator1:WBP Post Match Habanero Sequence Stat Object Reference=Get Stat_Right_0()
    NewEnumerator2:WBP Post Match Habanero Sequence Stat Object Reference=Get Stat_Left_1()
    NewEnumerator3:WBP Post Match Habanero Sequence Stat Object Reference=Get Stat_Right_1()
    Index:ENUM_HabaneroSequence_StatPosition Enum=@n1.StatPosition
   ) -> (ReturnValue:WBP Post Match Habanero Sequence Stat Object Reference)
   > SetStatWidget!(
    execute
    self
    StatWidget:WBP Post Match Habanero Sequence Stat Object Reference=@n2
    StatTitle:Text=@n1
    StatValue:int=@n1.StatValue
    "Show##ForValue":bool=Select(NewEnumerator0:bool=true, NewEnumerator1:bool=false, NewEnumerator2:bool=false, NewEnumerator3:bool=false, Index=@n1.StatPosition) -> (ReturnValue:bool)
   ) -> (then)
   > SetVisibility(self=@n2, InVisibility=SelfHitTestInvisible)
  ]}
 SetStatWidget {nodes [
   n1 = SetStatWidget!() -> (then, StatWidget:WBP Post Match Habanero Sequence Stat Object Reference, StatTitle:Text, StatValue:int, "Show##ForValue":bool=~) {is=title K2Node_FunctionEntry}
   > SetStatTitle!(execute, self:WBP Post Match Habanero Sequence Stat Object Reference=@n1, StatTitle:Text=@n1.StatTitle) -> (then)
   > SetStat!(execute, self:WBP Post Match Habanero Sequence Stat Object Reference=@n1, Value:int=@n1.StatValue, "Show##ForValue":bool=@n1."Show##ForValue") -> (then)
   > SetVisibility(
    self=@n1
    InVisibility=Select!(Option 0:ESlateVisibility Enum=SelfHitTestInvisible, Option 1:ESlateVisibility Enum=Collapsed, Index:bool=TextIsEmpty(InText=@n1.StatTitle)) -> (ReturnValue:ESlateVisibility Enum)
   )
  ]}
 SetTextColor {nodes [SetTextColor() {is=title K2Node_FunctionEntry}]}
 SetTournamentThumbnail {nodes [
   n1 = SetTournamentThumbnail!() -> (then, Texture:Texture Object Reference) {is=title K2Node_FunctionEntry}
   > SetThumbnail!(execute, self:WBP Tournament Small Banner Object Reference=Get WBP_TournamentSmallBanner(), Texture:Texture Object Reference=@n1) -> (then)
  ]}
 ShowDisplayRankedPercentageBasedOnDelta {nodes [
   ShowDisplayRankedPercentageBasedOnDelta() {is=title K2Node_FunctionEntry}
   n1 = "Equal (Integer)"(A=Get DisplayedPercentDifferenceRounded() -> (DisplayedPercentDifferenceRounded:int=0))
   n2 = "integer > integer"(A=Get DisplayedPercentDifferenceRounded() -> (DisplayedPercentDifferenceRounded=0), B=200)
   n3 = "integer < integer"(A=Get DisplayedPercentDifferenceRounded() -> (DisplayedPercentDifferenceRounded=0), B=-50)
   > Return Node!(execute, ShowDisplayPercentage:bool=NOT Boolean(A=OR Boolean!(A=@n1, B=@n2, C:bool=@n3) -> (ReturnValue)))
  ]}
 ShowSpinnerAndMessage {nodes [
   n1 = ShowSpinnerAndMessage!() -> (then, ShouldShow:bool=~) {is=title K2Node_FunctionEntry}
   > Branch(Condition="Not Equal (Boolean)"(A=@n1, B=Get IsSpinnerShowing() -> (IsSpinnerShowing:bool)) -> (ReturnValue=~))
   > n2 = Set IsSpinnerShowing(IsSpinnerShowing:bool=@n1) -> (Output_Get:bool)
   > n3 = Branch(Condition=@n2)
   > SetVisibility(self=Get Image_WaitingSpinner() -> (Image_WaitingSpinner:Image Object Reference), InVisibility=SelfHitTestInvisible)
   > PlayAnimationForward(InAnimation=Get Anim_ShowHideSpinner() -> (Anim_ShowHideSpinner:Widget Animation Object Reference))
   > SetVisibility(self=Get SizeBox_Error() -> (SizeBox_Error:Size Box Object Reference), InVisibility=SelfHitTestInvisible)
   > "SetText (Text)"(self=Get Text_ErrorText() -> (Text_ErrorText:Common Text Object Reference), InText=Get FetchingText() -> (FetchingText:Text))
   PlayAnimationReverse(execute=@n3.else, InAnimation=Get Anim_ShowHideSpinner())
   > SetVisibility(self=Get SizeBox_Error(), InVisibility=Collapsed)
  ]}
 StartDecreasedProgessEvent {nodes [
   StartDecreasedProgessEvent() {is=title K2Node_FunctionEntry}
   > Set CurrentDisplayedTierIndex(CurrentDisplayedTierIndex=Get TargetDisplayedTierIndex() -> (TargetDisplayedTierIndex=0)) -> (Output_Get=0)
   > Set CurrentTierStartingProgress(CurrentTierStartingProgress=Get TargetTierProgress())
   > Set CurrentPlayerPosition(CurrentPlayerPosition:int=Get TargetPlayerPosition() -> (TargetPlayerPosition=0)) -> (Output_Get:int=0)
   > CacheDisplayedTierData()
   > SetColorAndOpacity(
    self=Get Text_PercentProgress() -> (Text_PercentProgress:Common Text Object Reference)
    InColorAndOpacity=Make Slate Color(SpecifiedColor=(R=0.0078,G=0.4353,B=0.8706,A=1))
   )
   > RefreshVisuals() {is=title K2Node_CallFunction}
   > n1 = Sequence()
   > Branch(Condition=IsFinalRankDisplayed() {is=title K2Node_CallFunction})
   > SetVisibility(self=Get ProgressBar(), InVisibility=Collapsed)
   n2 = Get Image_PositionChange()
   > SetRenderTransformAngle!(execute, self:Widget Object Reference=@n2, Angle:float=180) -> (then)
   > SetVisibility(self=@n2, InVisibility=SelfHitTestInvisible)
   n3 = Get RawProgressPercentDifference() -> (RawProgressPercentDifference:double)
   n4 = Get ChangedTierIndex() -> (ChangedTierIndex:bool)
   n5 = Branch(
    execute=@n1.then_1
    Condition=AND Boolean!(
     A=Get HasRankLossProtection() -> (HasRankLossProtection:bool)
     B="Nearly Equal (Float)"(A=@n3, B=0, ErrorTolerance=0.0001)
     C:bool=NOT Boolean(A=@n4)
    ) -> (ReturnValue)
   )
   > PlayAnimation(InAnimation=Select(
     Option 0:Widget Animation Object Reference=Get Anim_GoDownProtected()
     Option 1:Widget Animation Object Reference=Get Anim_GoDownProtected_FinalRank()
     Index=IsFinalRankDisplayed()
    ) -> (ReturnValue:Widget Animation Object Reference))
   > SetVisibility(self=[Get Image_PositionChange(), Get Text_PercentChange() -> (Text_PercentChange:Common Text Object Reference)], InVisibility=Collapsed)
   PlayAnimation(execute=@n5.else, InAnimation=Select(Option 0=Get Anim_GoDown(), Option 1=Get Anim_GoDown_FinalRank(), Index=IsFinalRankDisplayed()))
  ]}
 StartFailureSequence {nodes [
   n1 = StartFailureSequence!() -> (then, ErrorMessage:Text) {is=title K2Node_FunctionEntry}
   > ShowSpinnerAndMessage!(execute, self, ShouldShow:bool=false) -> (then) {is=title K2Node_CallFunction}
   > n2 = Branch(Condition=TextIsEmpty(InText=@n1))
   SetVisibility(execute=@n2.else, self=Get SizeBox_Error(), InVisibility=SelfHitTestInvisible)
   > "SetText (Text)"(self=Get Text_ErrorText(), InText=@n1)
  ]}
 StartIncreasedProgressEvent {nodes [
   StartIncreasedProgressEvent() {is=title K2Node_FunctionEntry}
   > Set CurrentTierFinalProgess(CurrentTierFinalProgess:float=Get TargetTierProgress()) -> (Output_Get:float)
   > Set CurrentPlayerPosition(CurrentPlayerPosition=Get TargetPlayerPosition() -> (TargetPlayerPosition=0)) -> (Output_Get=0)
   > SetColorAndOpacity(self=Get Text_PercentProgress(), InColorAndOpacity=Make Slate Color(SpecifiedColor=(R=0.0078,G=0.4353,B=0.8706,A=1)))
   > n1 = Branch(Condition=IsFinalRankDisplayed())
   > SetVisibility(self=Get TextWrapper_RankUp(), InVisibility=Collapsed)
   > SetVisibility(self=Get Image_PositionChange(), InVisibility=SelfHitTestInvisible)
   n2 = Get InitialTopTier()
   n3 = Get CurrentTopTier()
   n4 = Select(
    Option 0=Get Anim_RankUp_FinalRank()
    Option 1=Get Anim_UnrealLegends_Start()
    Index=AND Boolean!(A="Not Equal (Boolean)"(A=@n2, B=@n3) -> (ReturnValue=~), B=Get CurrentTopTier()) -> (ReturnValue)
   )
   > PlayAnimation(InAnimation=@n4)
   > n5 = IsAnimationPlayingForward(InAnimation=@n4)
   > Set bForward(bForward=@n5)
   Play Progress Animations(execute=@n1.else) {is=title K2Node_CallFunction}
  ]}
 StartLevelUpEndEvent {nodes [
   StartLevelUpEndEvent() {is=title K2Node_FunctionEntry}
   > Set CurrentTierFinalProgess(CurrentTierFinalProgess=Get TargetTierProgress())
   > "SetText (Text)"(
    self=Get Text_EventTitle() -> (Text_EventTitle:Common Text Object Reference)
    InText=Select(
     Option 0:Text=Get RankUpEventText() -> (RankUpEventText:Text)
     Option 1:Text=Get RankRevealedEventText() -> (RankRevealedEventText:Text)
     Index=WasUnranked()
    ) -> (ReturnValue:Text)
   )
   > SetVisibility(self=Get TextWrapper_RankUp(), InVisibility=SelfHitTestInvisible)
   > SetColorAndOpacity(self=Get Text_PercentProgress(), InColorAndOpacity=Make Slate Color(SpecifiedColor=(R=0.0078,G=0.4353,B=0.8706,A=1)))
   > n1 = Branch(Condition=IsFinalRankDisplayed())
   > SetVisibility(self=Get Text_PercentChange(), InVisibility=Collapsed)
   n2 = Get InitialTopTier()
   n3 = Get CurrentTopTier()
   n4 = Select(
    Option 0:Widget Animation Object Reference=Get Anim_RankUp_FinalRank()
    Option 1:Widget Animation Object Reference=Get Anim_UnrealLegends_Start()
    Index=AND Boolean(A="Not Equal (Boolean)"(A=@n2, B=@n3) -> (ReturnValue=~), B=Get CurrentTopTier())
   ) -> (ReturnValue:Widget Animation Object Reference)
   > PlayAnimation(InAnimation=@n4)
   > n5 = IsAnimationPlayingForward(InAnimation=@n4)
   > Set bForward(bForward=@n5)
   PlayLevelUpEndAnimations(execute=@n1.else) {is=title K2Node_CallFunction}
  ]}
 StartLevelUpProgressEvent {nodes [
   StartLevelUpProgressEvent() {is=title K2Node_FunctionEntry}
   > Set CurrentTierFinalProgess(CurrentTierFinalProgess=1)
   > SetColorAndOpacity(self=Get Text_PercentProgress(), InColorAndOpacity=Make Slate Color(SpecifiedColor=(R=1,G=1,B=1,A=1)))
   > "SetText (Text)"(self=Get Text_RankUpMessage(), InText=GetRankUpMessage!() -> (RankUpMessage:Text) {is=title K2Node_CallFunction})
   n1 = Get Anim_RankUp_Start()
   > PlayAnimation(InAnimation=@n1)
   > n2 = IsAnimationPlayingForward(InAnimation=@n1)
   > Set bForward(bForward=@n2)
  ]}
 StartRankSequence {nodes [
   InitilizeTimeline!(
    execute
    self
    EliminationEntries:Array of Elimination Timeline Entry Structures=Get EliminationTimeline() -> (EliminationTimeline:Array of Elimination Timeline Entry Structures)
    TotalMatchTime:double=Get MatchTime() -> (MatchTime:double)
   ) -> (then) {is=title K2Node_CallFunction}
   n1 = Get EliminationTimeline()
   > n2 = Branch(Condition="integer > integer"(A=Length(TargetArray=@n1)))
   DetermineStartingSequence(execute=[@n12.then, @n2.else]) {is=title K2Node_CallFunction}
   n3 = StartRankSequence!() -> (
    then
    ActiveTierList:Array of Fort Habanero Tier Structures
    InitialTierIndex:int
    InitialProgress:float=~
    InitialPlayerPosition:int
    TargetTierIndex:int
    TargetProgress:float=~
    TargetPlayerPosition:int
    IsUnrankedTierValid:bool=~
    UnrankedTierData:Fort Habanero Tier Structure
    EliminationTimeline:Array of Elimination Timeline Entry Structures
    MatchTime:double=~
    IsInitialTopTier:bool=~
    IsCurrentTopTier:bool=~
    InInitialMMR:int
    InCurrentMMR:int
    TopTierData:Fort Habanero Tier Structure
   ) {is=title K2Node_FunctionEntry}
   > Set ActiveTierList(ActiveTierList:Array of Fort Habanero Tier Structures=@n3) -> (Output_Get:Array of Fort Habanero Tier Structures)
   > n4 = Set InitialTierIndex(InitialTierIndex:int=@n3.InitialTierIndex) -> (Output_Get:int=0)
   > n5 = Set CurrentDisplayedTierIndex(CurrentDisplayedTierIndex=@n4) -> (Output_Get=0)
   > n6 = Set CurrentTierStartingProgress(CurrentTierStartingProgress=SelectFloat(A=@n3.InitialProgress, B=0, bPickA="integer >= integer"(A=@n5)))
   > Set CurrentPlayerPosition(CurrentPlayerPosition=@n3.InitialPlayerPosition) -> (Output_Get=0)
   > Set TargetDisplayedTierIndex(TargetDisplayedTierIndex:int=@n3.TargetTierIndex) -> (Output_Get:int=0)
   > Set RawProgressPercentDifference(RawProgressPercentDifference:double=float - float(A=@n3.TargetProgress, B=@n6)) -> (Output_Get:double)
   n7 = "Clamp (Integer)"(Value=Round(A=float * float(A=@n3.TargetProgress, B=100)) {is=title K2Node_CallFunction}, Min=1, Max=99) -> (ReturnValue=0)
   > Set TargetTierProgress(TargetTierProgress:float=float / float(A="To Float (Integer)"(InInt=@n7), B=100)) -> (Output_Get:float)
   > Set TargetPlayerPosition(TargetPlayerPosition:int=@n3.TargetPlayerPosition) -> (Output_Get:int=0)
   n8 = Get ActiveTierList()
   > Set UnrankedTierData(UnrankedTierData:Fort Habanero Tier Structure=Select(
     Option 0:Fort Habanero Tier Structure="Get (a ref)"(Array=@n8, Dimension 1=0)
     Option 1:Fort Habanero Tier Structure=@n3.UnrankedTierData
     Index=@n3.IsUnrankedTierValid
    ) -> (ReturnValue:Fort Habanero Tier Structure)) -> (Output_Get:Fort Habanero Tier Structure)
   > Set Top Tier Data(Top Tier Data:Fort Habanero Tier Structure=@n3.TopTierData) -> (Output_Get:Fort Habanero Tier Structure)
   > Set InitialTopTier(InitialTopTier=@n3.IsInitialTopTier)
   > Set CurrentTopTier(CurrentTopTier:bool=@n3.IsCurrentTopTier) -> (Output_Get:bool)
   > Set InitialMMR(InitialMMR:int=@n3.InInitialMMR) -> (Output_Get:int=0)
   > Set CurrentMMR(CurrentMMR:int=@n3.InCurrentMMR) -> (Output_Get:int=0)
   > ShowSpinnerAndMessage(ShouldShow=false)
   > SetVisibility(self=Get ContentWrapper(), InVisibility=SelfHitTestInvisible)
   > Set HasSequenceStarted(HasSequenceStarted:bool=true) -> (Output_Get:bool)
   n9 = Get Anim_Intro() -> (Anim_Intro:Widget Animation Object Reference)
   > PlayAnimation(InAnimation=@n9)
   > n10 = IsAnimationPlayingForward(InAnimation=@n9)
   > Set bForward(bForward=@n10)
   n11 = Get VM_RankedFeedbackAnimationTrigger()
   > Set with Broadcast IMF_KillsAnimTrigger(IMF_KillsAnimTrigger=false, self=@n11)
   > Set with Broadcast IMF_PlacementAnimTrigger(IMF_PlacementAnimTrigger=false, self=@n11)
   > Set with Broadcast RankedFeedbackBonusAnimationFired(RankedFeedbackBonusAnimationFired=false, self=@n11)
   > PlayAnimation(InAnimation=Get Anim_Stats_Intro() -> (Anim_Stats_Intro:Widget Animation Object Reference))
   > Set AreStatsVisible(AreStatsVisible=true)
   > n12 = UpdateTextPercentChange()
  ]}
 StartSequenceFinished {nodes [
   StartSequenceFinished() {is=title K2Node_FunctionEntry}
   > n1 = Sequence()
   > SetVisibility(self=Get TextWrapper_RankUp(), InVisibility=Collapsed)
   > SetRenderOpacity(self=Get Text_RankUpMessage(), InOpacity=0)
   > Call On Sequence Finished!(execute:Exec, self:WBP Post Match Habanero Sequence Object Reference) -> (then:Exec) {is=title K2Node_CallDelegate}
   > PlayAnimation(InAnimation=Get Anim_OnSequenceFinished() -> (Anim_OnSequenceFinished:Widget Animation Object Reference))
   > n2 = Branch(Condition=Get AreStatsVisible())
   PlayAnimation(
    execute=@n2.else
    InAnimation=Get Anim_ShowStats_OnSequenceFinished() -> (Anim_ShowStats_OnSequenceFinished:Widget Animation Object Reference)
   )
   > Set AreStatsVisible(AreStatsVisible=true)
   PlaySound2D(execute=@n1.then_1, Sound=Get LoopStopSound() -> (LoopStopSound:Sound Base Object Reference))
   > n3 = Get()
   > FadeOut!(
    execute
    self:Audio Component Object Reference=@n3
    FadeOutDuration:float=0.1
    FadeVolumeLevel:float=0
    FadeCurve:EAudioFaderCurve Enum=Linear
   ) -> (then)
  ]}
 StartWaitingSequence {nodes [
   StartWaitingSequence!() -> (then, TimeoutDuration:double=~) {is=title K2Node_FunctionEntry}
   > SpawnSound2D(
    Sound=Get BlueScreenWhoosh() -> (BlueScreenWhoosh:Sound Base Object Reference)
    VolumeMultiplier=1
    PitchMultiplier=1
    StartTime=0
    bPersistAcrossLevelTransition=false
    bAutoDestroy=true
   )
   > SetVisibility(self=Get ContentWrapper(), InVisibility=Hidden)
   > ShowSpinnerAndMessage(ShouldShow=true)
   > Set Timer by Event(Delegate=Create Event(), Time=float / float(A=Get TimeoutDuration() -> (TimeoutDuration:double), B=2), bLooping=false)
  ]}
 TimelineAddLinesToProgress {nodes [
   TimelineAddLinesToProgress() {is=title K2Node_FunctionEntry}
   n1 = Length(TargetArray=Get EliminationEntries())
   n2 = Length(TargetArray=Get EliminationEntries())
   n3 = Get EliminationEntryIndex() -> (EliminationEntryIndex:int=0)
   > Branch(Condition=AND Boolean(A="integer > integer"(A=@n1), B="integer > integer"(A=@n2, B=@n3)))
   > n4 = Create Widget!(
    execute:Exec
    Class:User Widget Class Reference=/Game/Athena/HUD/EndGame/PostGameFlowScreens/DefaultScreens/XPScreen/WBP_EliminationTimelineEntry.WBP_EliminationTimelineEntry_C
    OwningPlayer:Player Controller Object Reference
   ) -> (then:Exec, ReturnValue:WBP Elimination Timeline Entry Object Reference)
   > n5 = Set WidgetRef(WidgetRef:WBP Elimination Timeline Entry Object Reference=@n4) -> (Output_Get:WBP Elimination Timeline Entry Object Reference)
   n6 = Break Elimination Timeline Entry(EliminationTimelineEntry="Get (a copy)"(Array=Get EliminationEntries(), Dimension 1=Get EliminationEntryIndex() -> (EliminationEntryIndex=0))) -> (Time=~, bLocalElimination=~, bPlayersLeftEntry=~)
   > Initialize!(
    execute
    self:WBP Elimination Timeline Entry Object Reference=@n5
    Name:String=@n6
    RankType:String=@n6.PlayerRankType
    Rank:int=@n6.PlayerRank
    Current MMR:int=@n6.PlayerMMR
    Current Position:int=@n6.PlayerCurrentPosition
    IsStormCircle:bool=@n6.bPlayersLeftEntry
   ) -> (then)
   > n7 = AddChildToOverlay!(execute, self:Overlay Object Reference=Get Overlay_TimelineEntries(), Content:Widget Object Reference=@n5) -> (then, ReturnValue:Overlay Slot Object Reference)
   > n8 = CalculateXOffset!(execute, self, TotalMatchTime:double=Get TotalMatchTime(), EliminationMatchTime:double=@n6.Time) -> (then, ReturnValue:double) {is=title K2Node_CallFunction}
   > SetPadding!(execute, self:Overlay Slot Object Reference=@n7, InPadding:Margin Structure=Make Margin(Left=@n8, Top=0, Right=0, Bottom=48)) -> (then)
   > n9 = Increment Int(Value=Get EliminationEntryIndex() -> (EliminationEntryIndex=0))
   > Set EliminationEntryIndex(EliminationEntryIndex:int=@n9) -> (Output_Get:int=0)
   n10 = int - int(A=Get EliminationEntryIndex() -> (EliminationEntryIndex=0))
   n11 = Get EliminationEntries()
   n12 = Get EliminationEntries()
   n13 = Get EliminationEntryIndex() -> (EliminationEntryIndex=0)
   n14 = float - float(
    A=Break Elimination Timeline Entry(EliminationTimelineEntry="Get (a copy)"(Array=@n12, Dimension 1=@n13)) -> (Time=~, bLocalElimination=~, bPlayersLeftEntry=~).Time
    B=Break Elimination Timeline Entry(EliminationTimelineEntry="Get (a copy)"(Array=@n11, Dimension 1=@n10)) -> (Time=~, bLocalElimination=~, bPlayersLeftEntry=~).Time
   )
   n15 = Get TotalMatchTime()
   > Set Timer by Event(Delegate=Create Event(), Time=float * float(A=float / float(A=@n14, B=@n15), B=4), bLooping=false)
   PlayAnimationForward(InAnimation=Get Anim_TimeLineProgress())
   Reroute Node(InputPin:Wildcard) -> (OutputPin:Wildcard)
  ]}
 TimelineFadeOut {nodes [
   TimelineFadeOut() {is=title K2Node_FunctionEntry}
   > PlayAnimationForward(InAnimation=Get Anim_TimeLineFadeOut())
  ]}
 UpdateProgressText {nodes [
   UpdateProgressText!() -> (then, Percent:double=~) {is=title K2Node_FunctionEntry}
   > n1 = Branch(Condition=IsFinalRankDisplayed())
   n2 = "To Text (Integer)"(Value=Get TargetPlayerPosition() -> (TargetPlayerPosition=0))
   n3 = Get UnrealRank() -> (UnrealRank:Text)
   n4 = Get TargetPlayerPosition() -> (TargetPlayerPosition=0)
   > "SetText (Text)"(
    self=Get Text_PercentProgress()
    InText=Select(
     Option 0:Text=Format Text!(Format=@n3, rank:Text=@n2) -> (Result)
     Option 1:Text=Get UnrealRankCapText() -> (UnrealRankCapText:Text)
     Index="integer >= integer"(A=@n4, B=100000)
    ) -> (ReturnValue:Text)
   )
   n5 = float * float(A=Get Percent() -> (Percent:double), B=100)
   "SetText (Text)"(
    execute=@n1.else
    self=Get Text_PercentProgress()
    InText=Format Text!(Format="LOCTEXT(\"{PercentProgress}%\")", PercentProgress:int=Round(A=@n5)) -> (Result)
   )
  ]}
 UpdateSequenceProgress {nodes [
   UpdateSequenceProgress() {is=title K2Node_FunctionEntry}
   > n1 = Branch(Condition="Equal (Integer)"(A=Get TargetDisplayedTierIndex() -> (TargetDisplayedTierIndex=0), B=Get CurrentDisplayedTierIndex() -> (CurrentDisplayedTierIndex=0)))
   > StartLevelUpEndEvent() {is=title K2Node_CallFunction}
   StartLevelUpProgressEvent(execute=@n1.else) {is=title K2Node_CallFunction}
  ]}
 UpdateTextPercentChange {nodes [
   UpdateTextPercentChange() {is=title K2Node_FunctionEntry}
   n1 = Get CurrentMMR() -> (CurrentMMR:int=0)
   n2 = AND Boolean(A=Get CurrentTopTier(), B="integer > integer"(A=@n1))
   > Call On Skiil Rating Visibility Changed!(execute:Exec, self:WBP Post Match Habanero Sequence Object Reference, IsVisible:bool=@n2) -> (then:Exec) {is=title K2Node_CallDelegate}
   > SetVisibility(
    self=Get Box_SkillRating() -> (Box_SkillRating:Horizontal Box Object Reference)
    InVisibility=Select(Option 0:ESlateVisibility Enum=Collapsed, Option 1:ESlateVisibility Enum=SelfHitTestInvisible, Index=@n2) -> (ReturnValue:ESlateVisibility Enum)
   )
   > "SetText (Text)"(self=Get Text_SkillRating() -> (Text_SkillRating:Common Text Object Reference), InText="To Text (Integer)"(Value=@n1))
   n3 = int - int(A=Get TargetDisplayedTierIndex() -> (TargetDisplayedTierIndex=0), B=Get InitialTierIndex() -> (InitialTierIndex:int=0))
   n4 = Round(A=float * float(A=Get CurrentTierStartingProgress(), B=100))
   n5 = Round(A=float * float(A=Get TargetTierProgress(), B=100))
   > n6 = Set DisplayedPercentDifferenceRounded(DisplayedPercentDifferenceRounded:int=int + int(A=int * int(A=@n3, B=100), B=int - int(A=@n5, B=@n4))) -> (Output_Get:int=0)
   n7 = "Absolute (Integer)"!(A:int=int - int(A=Get TargetPlayerPosition() -> (TargetPlayerPosition=0), B=Get CurrentPlayerPosition() -> (CurrentPlayerPosition=0))) -> (ReturnValue:int=0)
   n8 = "To Text (Integer)"(Value=@n6, bAlwaysSign=true)
   > "SetText (Text)"(
    self=Get Text_PercentChange()
    InText=Select(
     Option 0:Text=Format Text!(Format="LOCTEXT(\"{percent change}%\")", percent change:Text=@n8) -> (Result)
     Option 1:Text="To Text (Integer)"(Value=@n7)
     Index=IsFinalRankDisplayed()
    ) -> (ReturnValue:Text)
   )
   > n9 = Branch(Condition=IsFinalRankDisplayed())
   n10 = ShowDisplayRankedPercentageBasedOnDelta!(execute=@n9.else, self) -> (then, ShowDisplayPercentage:bool) {is=title K2Node_CallFunction}
   > SetVisibility(
    self=Get Text_PercentChange()
    InVisibility=Select(Option 0:ESlateVisibility Enum=Collapsed, Option 1:ESlateVisibility Enum=SelfHitTestInvisible, Index=@n10) -> (ReturnValue:ESlateVisibility Enum)
   )
  ]}
 UpdateWaitingMessage {nodes [
   UpdateWaitingMessage() {is=title K2Node_FunctionEntry}
   > n1 = Branch(Condition=Get HasSequenceStarted() -> (HasSequenceStarted:bool))
   "SetText (Text)"(execute=@n1.else, self=Get Text_ErrorText(), InText=Get StillFetchingText() -> (StillFetchingText:Text))
  ]}
 WasUnranked {nodes [
   WasUnranked() {is=title K2Node_FunctionEntry}
   > Return Node!(execute, WasUnranked:bool="integer < integer"(A=Get InitialTierIndex() -> (InitialTierIndex=0)))
  ]}
]
variables [
 ActiveTierList {type="array<FortHabaneroTier>", default []}
 AreStatsVisible {type=boolean}
 Background Current Color {type=string, default=Current Color}
 Background Next Color {type=string, default=Next Color}
 BadgeAppearSound_01 {type=SoundBase, default=/FortUIAudioContent/SoundCues/Fortnite_UI_Ranked_Post_Badge_Appear_01_Cue}
 BadgeAppearSound_02 {type=SoundBase, default=/FortUIAudioContent/SoundCues/Fortnite_UI_Ranked_Post_Badge_Appear_02_Cue}
 BadgeAppearSound_03 {type=SoundBase, default=/FortUIAudioContent/SoundCues/Fortnite_UI_Ranked_Post_Badge_Appear_03_Cue}
 BadgeCounter {type=integer}
 Bar Glow Color {type=string, default=Bar Glow Color}
 bForward {type=boolean, default=true}
 BlueScreenWhoosh {type=SoundBase, default=/FortUIAudioContent/SoundCues/Fortnite_UI_Ranked_Post_BlueIntro_Cue}
 BurstCompletedSound {type=SoundBase, default=/FortUIAudioContent/SoundCues/Fortnite_UI_Ranked_Post_Progress_Complete_Cue}
 ChangedTierIndex {type=boolean}
 CurrentDisplayedTierCategoryData {type=BP_UI_Habanero_TierTypeDataBase_C, default=null}
 CurrentDisplayedTierIndex {type=integer}
 CurrentMMR {type=integer}
 CurrentPlayerPosition {type=integer}
 CurrentTierFinalProgess {type=number, default=1}
 CurrentTierStartingProgress {type=number}
 CurrentTimer {type=number}
 CurrentTopTier {type=boolean}
 Debug {type=integer, default=6}
 Delta Bar Color {type=string, default=Delta Bar Color}
 Delta Value Name {type=string, default=Delta Progress}
 DeltaValue_Blend {type=number}
 DisplayedPercentDifferenceRounded {type=integer}
 EliminationEntries {type="array<EliminationTimelineEntry>", default []}
 EliminationEntryIndex {type=integer}
 EliminationTimelineSize {type=number, default=750}
 FetchingText {type=string, default="Fetching results…"}
 Flare Color {type=string, default=Color}
 FlashTransitionSound {type=SoundBase, default=/FortUIAudioContent/SoundCues/Fortnite_UI_Ranked_Flash_Transition_Whoosh_Cue}
 HasRankLossProtection {type=boolean}
 HasSequenceStarted {type=boolean}
 InitialMMR {type=integer}
 InitialTierIndex {type=integer}
 InitialTopTier {type=boolean}
 Inner Color {type=string, default=Inner Color}
 IsLevelUp {type=boolean}
 IsSpinnerShowing {type=boolean}
 LettersAppearSound {type=SoundBase, default=/FortUIAudioContent/SoundCues/Fortnite_UI_Ranked_Post_LettersWhoosh_Cue}
 LoopProgressSound {type=SoundBase, default=/FortUIAudioContent/SoundCues/Fortnite_UI_Ranked_Post_Progress_Climb_Cue}
 LoopStartSound {type=SoundBase, default=/FortUIAudioContent/SoundCues/Fortnite_UI_Ranked_Post_Progress_Start_Cue}
 LoopStopSound {type=SoundBase, default=/FortUIAudioContent/SoundCues/Fortnite_UI_Ranked_Post_Progress_Stop_Cue}
 MaxProgressAnimationScale {type=number, default=2.8}
 MinutesValue {type=number}
 NextDisplayedTierCategoryData {type=BP_UI_Habanero_TierTypeDataBase_C, default=null}
 NextTierIndex {type=integer}
 OnSequenceFinished {}
 OnSkiilRatingVisibilityChanged {}
 Particles Color {type=string, default=Color}
 Progress Background Color {type=string, default=Progress Background Color}
 Progress Bar Color {type=string, default=Bar Color}
 Progress Outline Color {type=string, default=Bar Outline Color}
 Progress Value Name {type=string, default=Progress}
 ProgressAudioComponent {type=AudioComponent, default=null}
 ProgressBarColor_Blend {type=number}
 ProgressBarEndWhoosh {type=SoundBase, default=/FortUIAudioContent/SoundCues/Fortnite_UI_Ranked_Post_Progress_EndWhoosh_Cue}
 ProgressValue_Blend {type=number}
 RankedFeedbackTag {type=GameplayTag, default {tagName=Hud.Athena.Habanero.PostMatchFeedback}}
 RankRevealedEventText {type=string, default="Rank Revealed!"}
 RankUpData_BattleRoyale {type=DataTable, default=/FortniteCoreUI/UI/Habanero/DataTables/DT_RankUpData_BattleRoyale}
 RankUpData_ZeroBuild {type=DataTable, default=/FortniteCoreUI/UI/Habanero/DataTables/DT_RankUpData_ZeroBuild}
 RankUpEventText {type=string, default="Rank up!"}
 RawProgressPercentDifference {type=number}
 SecondsValue {type=number}
 ShakeSound {type=SoundBase, default=/FortUIAudioContent/SoundCues/Fortnite_UI_Ranked_Post_Badge_Shake_Cue}
 StartScalingAnimationBelowProgress {type=number, default=0.15}
 StillFetchingText {type=string, default="Fetching results…\r\nThis is taking longer than usual."}
 TargetDisplayedTierIndex {type=integer}
 TargetPlayerPosition {type=integer}
 TargetTierProgress {type=number}
 Top Tier Data {
  type=FortHabaneroTier
  default {
   name=""
   additionalLeftIcon=null
   additionalRightIcon=null
   backgroundIcon=null
   backgroundIconSmall=null
   foregroundIcon=null
   foregroundIconSmall=null
   simpleIconSmall=null
   tierType=null
   whiteIcon=null
  }
 }
 TotalMatchTime {type=number}
 UnrankedTierData {
  type=FortHabaneroTier
  default {
   name=""
   additionalLeftIcon=null
   additionalRightIcon=null
   backgroundIcon=null
   backgroundIconSmall=null
   foregroundIcon=null
   foregroundIconSmall=null
   simpleIconSmall=null
   tierType=null
   whiteIcon=null
  }
 }
 UnrealBadgeEnterSound {type=SoundBase, default=/FortUIAudioContent/SoundCues/SC_Fortnite_UI_Ranked_Post_Badge_Appear_Unreal}
 UnrealBadgeEnterWhooshSound {type=SoundBase, default=&Fortnite_UI_Ranked_Flash_Transition_Whoosh_Cue}
 UnrealBadgeWhooshOutSound {type=SoundBase, default=/FortUIAudioContent/SoundCues/Fortnite_UI_Ranked_Post_Intro_Cue}
 UnrealLegendBadgeEnterSound {type=SoundBase, default=/FortUIAudioContent/SoundCues/SC_Fortnite_UI_Unreal_Legend_Badge_Enter}
 UnrealLegendFinalImpactSound {type=SoundBase, default=/FortUIAudioContent/SoundCues/SC_Fortnite_UI_Unreal_Legend_Final_Impact}
 UnrealLegendWhooshSweetenerSound {type=SoundBase, default=/FortUIAudioContent/SoundCues/SC_Fortnite_UI_Unreal_Legend_Whoosh_Sweetener}
 UnrealRank {type=string, default="##{rank}"}
 UnrealRankCapText {type=string, default="##100k+"}
 UnrealWhooshOutSweetenerSound {type=SoundBase, default=/FortUIAudioContent/SoundCues/SC_Fortnite_UI_Ranked_Whoosh_Out_Sweetener}
]
widgets {
 info {inheritedWidgetCount=0, parentClass=&CommonUserWidget, rootWidgetClass=&Overlay, widgetCount=68}
 widgets [
  Overlay_0 {widgetClassPath=&Overlay, widget=Overlay_0}
  ContentWrapper {
   widgetClassPath=&Overlay
   bIsVariable=true
   parent=Overlay_0
   widget=ContentWrapper
   slot {class=OverlaySlot, properties {horizontalAlignment=HAlign_Fill, verticalAlignment=VAlign_Fill}}
  }
  Background {
   widgetClassPath=&Image
   bIsVariable=true
   parent=ContentWrapper
   widget=Background
   properties {
    visibility=SelfHitTestInvisible
    brush {
     bIsDynamicallyLoaded=false
     drawAs=Image
     imageType=NoImage
     mirroring=NoMirror
     resourceName=None
     resourceObject=/Game/UI/Frontend/Lobby/Habanero/Habanero_Materials/MaterialInstances/Background/MI_UI_Habanero_Background
     tiling=NoTile
     imageSize=<32 32>
     margin=<ltrb 0 0 0 0>
     outlineSettings {
      bUseBrushTransparency=false
      roundingType=HalfHeightRadius
      width=0
      color {colorUseRule=UseColor_Specified, specifiedColor=<rgba 0 0 0 0>}
      cornerRadii=<0 0 0 0>
     }
     tintColor {colorUseRule=UseColor_Specified, specifiedColor=<rgba 1 1 1 1>}
     uVRegion {bIsValid=false, max=<0 0>, min=<0 0>}
    }
   }
   slot {class=OverlaySlot, properties {horizontalAlignment=HAlign_Fill, verticalAlignment=VAlign_Fill}}
  }
  Unreal_BG_Overlay {
   widgetClassPath=&Image
   bIsVariable=true
   parent=ContentWrapper
   widget=Unreal_BG_Overlay
   properties {
    renderOpacity=0
    brush {
     bIsDynamicallyLoaded=false
     drawAs=Image
     imageType=FullColor
     mirroring=NoMirror
     resourceName=None
     resourceObject=/FortniteCoreUI/HUD/EndGame/PostGameFlowScreens/DefaultScreens/XPScreen/Materials/M_UI_PostMatch_UnrealLegends_BG
     tiling=NoTile
     imageSize=<32 32>
     margin=<ltrb 0 0 0 0>
     outlineSettings {
      bUseBrushTransparency=false
      roundingType=HalfHeightRadius
      width=0
      color {colorUseRule=UseColor_Specified, specifiedColor=<rgba 0 0 0 0>}
      cornerRadii=<0 0 0 0>
     }
     tintColor {colorUseRule=UseColor_Specified, specifiedColor=<rgba 1 1 1 1>}
     uVRegion {bIsValid=false, max=<0 0>, min=<0 0>}
    }
   }
   slot {class=OverlaySlot, properties {horizontalAlignment=HAlign_Fill, verticalAlignment=VAlign_Fill}}
  }
  Img_Flare {
   widgetClassPath=&Image
   bIsVariable=true
   parent=ContentWrapper
   widget=Img_Flare
   properties {
    visibility=SelfHitTestInvisible
    brush {
     bIsDynamicallyLoaded=false
     drawAs=Image
     imageType=NoImage
     mirroring=NoMirror
     resourceName=None
     resourceObject=/Game/UI/Frontend/Lobby/Habanero/Habanero_Materials/MaterialInstances/Flares/MI_UI_Habanero_Flare
     tiling=NoTile
     imageSize=<1024 256>
     margin=<ltrb 0 0 0 0>
     outlineSettings {
      bUseBrushTransparency=false
      roundingType=HalfHeightRadius
      width=0
      color {colorUseRule=UseColor_Specified, specifiedColor=<rgba 0 0 0 0>}
      cornerRadii=<0 0 0 0>
     }
     tintColor {colorUseRule=UseColor_Specified, specifiedColor=<rgba 1 1 1 1>}
     uVRegion {bIsValid=false, max=<0 0>, min=<0 0>}
    }
   }
   slot {class=OverlaySlot, properties {horizontalAlignment=HAlign_Center, verticalAlignment=VAlign_Center}}
  }
  Particles {
   widgetClassPath=&GridPanel
   parent=ContentWrapper
   widget=Particles
   properties {renderTransform {angle=0, scale=<1 1>, shear=<0 0>, translation=<0 -250>}}
   slot {class=OverlaySlot, properties {horizontalAlignment=HAlign_Center, verticalAlignment=VAlign_Center}}
  }
  RankParticles_Left {
   widgetClassPath=&Image
   bIsVariable=true
   parent=Particles
   widget=RankParticles_Left
   properties {brush {
     bIsDynamicallyLoaded=false
     drawAs=Image
     imageType=FullColor
     mirroring=NoMirror
     resourceName=None
     resourceObject=/Game/UI/Frontend/Lobby/Habanero/Habanero_Materials/MaterialInstances/MI_UI_Habanero_PostMatch_Triangles
     tiling=NoTile
     imageSize=<250 250>
     margin=<ltrb 0 0 0 0>
     outlineSettings {
      bUseBrushTransparency=false
      roundingType=HalfHeightRadius
      width=0
      color {colorUseRule=UseColor_Specified, specifiedColor=<rgba 0 0 0 0>}
      cornerRadii=<0 0 0 0>
     }
     tintColor {colorUseRule=UseColor_Specified, specifiedColor=<rgba 1 1 1 1>}
     uVRegion {bIsValid=false, max=<0 0>, min=<0 0>}
    }}
   slot {class=GridSlot}
  }
  RankParticles_Right {
   widgetClassPath=&Image
   bIsVariable=true
   parent=Particles
   widget=RankParticles_Right
   properties {brush {
     bIsDynamicallyLoaded=false
     drawAs=Image
     imageType=FullColor
     mirroring=NoMirror
     resourceName=None
     resourceObject=/Game/UI/Frontend/Lobby/Habanero/Habanero_Materials/M_UI_Habanero_PostMatch_Triangles
     tiling=NoTile
     imageSize=<250 250>
     margin=<ltrb 0 0 0 0>
     outlineSettings {
      bUseBrushTransparency=false
      roundingType=HalfHeightRadius
      width=0
      color {colorUseRule=UseColor_Specified, specifiedColor=<rgba 0 0 0 0>}
      cornerRadii=<0 0 0 0>
     }
     tintColor {colorUseRule=UseColor_Specified, specifiedColor=<rgba 1 1 1 1>}
     uVRegion {bIsValid=false, max=<0 0>, min=<0 0>}
    }}
   slot {class=GridSlot, properties {column=1, padding=<ltrb 150 0 0 0>}}
  }
  Overlay_TriangleEnergy {
   widgetClassPath=&Overlay
   parent=ContentWrapper
   widget=Overlay_TriangleEnergy
   properties {renderTransform {angle=0, scale=<1 1>, shear=<0 0>, translation=<0 -130>}}
   slot {class=OverlaySlot, properties {horizontalAlignment=HAlign_Fill, verticalAlignment=VAlign_Center}}
  }
  Particles_Inward_Small {
   widgetClassPath=&Image
   bIsVariable=true
   parent=Overlay_TriangleEnergy
   widget=Particles_Inward_Small
   properties {
    visibility=SelfHitTestInvisible
    brush {
     bIsDynamicallyLoaded=false
     drawAs=Image
     imageType=NoImage
     mirroring=NoMirror
     resourceName=None
     resourceObject=/Game/UI/Frontend/Lobby/Habanero/Habanero_Materials/MaterialInstances/Background/MI_UI_Habanero_TrianglesEnergy_Inward_Small
     tiling=NoTile
     imageSize=<2048 256>
     margin=<ltrb 0 0 0 0>
     outlineSettings {
      bUseBrushTransparency=false
      roundingType=HalfHeightRadius
      width=0
      color {colorUseRule=UseColor_Specified, specifiedColor=<rgba 0 0 0 0>}
      cornerRadii=<0 0 0 0>
     }
     tintColor {colorUseRule=UseColor_Specified, specifiedColor=<rgba 1 1 1 1>}
     uVRegion {bIsValid=false, max=<0 0>, min=<0 0>}
    }
   }
   slot {class=OverlaySlot, properties {horizontalAlignment=HAlign_Fill, verticalAlignment=VAlign_Fill}}
  }
  Particles_Inward_Medium {
   widgetClassPath=&Image
   bIsVariable=true
   parent=Overlay_TriangleEnergy
   widget=Particles_Inward_Medium
   properties {
    visibility=SelfHitTestInvisible
    brush {
     bIsDynamicallyLoaded=false
     drawAs=Image
     imageType=NoImage
     mirroring=NoMirror
     resourceName=None
     resourceObject=/Game/UI/Frontend/Lobby/Habanero/Habanero_Materials/MaterialInstances/Background/MI_UI_Habanero_TrianglesEnergy_Inward
     tiling=NoTile
     imageSize=<2048 256>
     margin=<ltrb 0 0 0 0>
     outlineSettings {
      bUseBrushTransparency=false
      roundingType=HalfHeightRadius
      width=0
      color {colorUseRule=UseColor_Specified, specifiedColor=<rgba 0 0 0 0>}
      cornerRadii=<0 0 0 0>
     }
     tintColor {colorUseRule=UseColor_Specified, specifiedColor=<rgba 1 1 1 1>}
     uVRegion {bIsValid=false, max=<0 0>, min=<0 0>}
    }
   }
   slot {class=OverlaySlot, properties {horizontalAlignment=HAlign_Fill, verticalAlignment=VAlign_Fill}}
  }
  Particles_Inward_Big {
   widgetClassPath=&Image
   bIsVariable=true
   parent=Overlay_TriangleEnergy
   widget=Particles_Inward_Big
   properties {
    visibility=SelfHitTestInvisible
    brush {
     bIsDynamicallyLoaded=false
     drawAs=Image
     imageType=NoImage
     mirroring=NoMirror
     resourceName=None
     resourceObject=/Game/UI/Frontend/Lobby/Habanero/Habanero_Materials/MaterialInstances/Background/MI_UI_Habanero_TrianglesEnergy_Inward_Big
     tiling=NoTile
     imageSize=<2048 256>
     margin=<ltrb 0 0 0 0>
     outlineSettings {
      bUseBrushTransparency=false
      roundingType=HalfHeightRadius
      width=0
      color {colorUseRule=UseColor_Specified, specifiedColor=<rgba 0 0 0 0>}
      cornerRadii=<0 0 0 0>
     }
     tintColor {colorUseRule=UseColor_Specified, specifiedColor=<rgba 1 1 1 1>}
     uVRegion {bIsValid=false, max=<0 0>, min=<0 0>}
    }
   }
   slot {class=OverlaySlot, properties {horizontalAlignment=HAlign_Fill, verticalAlignment=VAlign_Fill}}
  }
  HorizontalBox_0 {
   widgetClassPath=&HorizontalBox
   parent=ContentWrapper
   widget=HorizontalBox_0
   properties {flowDirectionPreference=Culture, renderTransform {angle=0, scale=<1 1>, shear=<0 0>, translation=<0 -245>}}
   slot {class=OverlaySlot, properties {horizontalAlignment=HAlign_Center, verticalAlignment=VAlign_Center}}
  }
  FortVisualAttachment_Placement {
   widgetClassPath=&FortVisualAttachment
   parent=HorizontalBox_0
   widget=FortVisualAttachment_Placement
   properties {contentAnchor=<1 0>, renderTransform {angle=0, scale=<1 1>, shear=<0 0>, translation=<1 15>}}
   slot {class=HorizontalBoxSlot, properties {verticalAlignment=VAlign_Top}}
  }
  WidgetSwitcherSlot_RankedFeedback {
   widgetClassPath=/Script/CommonUI.CommonAnimatedSwitcher
   bIsVariable=true
   parent=FortVisualAttachment_Placement
   widget=WidgetSwitcherSlot_RankedFeedback
   properties {bOverride_Cursor=true, renderTransform {angle=0, scale=<1 1>, shear=<0 0>, translation=<100 200>}}
   slot {class=SizeBoxSlot, properties {horizontalAlignment=HAlign_Left, verticalAlignment=VAlign_Top}}
  }
  StatsLeftContainer_VerticalBox {
   widgetClassPath=&VerticalBox
   bIsVariable=true
   parent=WidgetSwitcherSlot_RankedFeedback
   widget=StatsLeftContainer_VerticalBox
   properties {renderTransform {angle=0, scale=<1 1>, shear=<0 0>, translation=<-100 -200>}}
   slot {class=WidgetSwitcherSlot}
  }
  Stat_Left_1 {
   widgetClassPath=/Game/Athena/HUD/EndGame/PostGameFlowScreens/DefaultScreens/XPScreen/WBP_PostMatch_HabaneroSequence_Stat
   bIsVariable=true
   parent=StatsLeftContainer_VerticalBox
   widget=Stat_Left_1
   properties {visibility=Collapsed}
   slot {class=VerticalBoxSlot}
  }
  Stat_Left_0 {
   widgetClassPath=&WBP_PostMatch_HabaneroSequence_Stat
   bIsVariable=true
   parent=StatsLeftContainer_VerticalBox
   widget=Stat_Left_0
   slot {class=VerticalBoxSlot}
  }
  VerticalBox_1 {
   widgetClassPath=&VerticalBox
   parent=HorizontalBox_0
   widget=VerticalBox_1
   slot {class=HorizontalBoxSlot, properties {padding=<ltrb 350 0 350 0>}}
  }
  BadgesContainer {
   widgetClassPath=&Overlay
   parent=VerticalBox_1
   widget=BadgesContainer
   slot {class=VerticalBoxSlot, properties {horizontalAlignment=HAlign_Center, verticalAlignment=VAlign_Center}}
  }
  Spacer_14 {
   widgetClassPath=&Spacer
   parent=BadgesContainer
   widget=Spacer_14
   properties {size=<370 370>}
   slot {class=OverlaySlot, properties {horizontalAlignment=HAlign_Fill, verticalAlignment=VAlign_Fill}}
  }
  ProgressBar {
   widgetClassPath=&Image
   bIsVariable=true
   parent=BadgesContainer
   widget=ProgressBar
   properties {
    visibility=SelfHitTestInvisible
    brush {
     bIsDynamicallyLoaded=false
     drawAs=Image
     imageType=NoImage
     mirroring=NoMirror
     resourceName=None
     resourceObject=/Game/UI/Frontend/Lobby/Habanero/Habanero_Materials/MaterialInstances/ProgressBars/MI_UI_Habanero_ProgressBar_UpAnim
     tiling=NoTile
     imageSize=<32 32>
     margin=<ltrb 0 0 0 0>
     outlineSettings {
      bUseBrushTransparency=false
      roundingType=HalfHeightRadius
      width=0
      color {colorUseRule=UseColor_Specified, specifiedColor=<rgba 0 0 0 0>}
      cornerRadii=<0 0 0 0>
     }
     tintColor {colorUseRule=UseColor_Specified, specifiedColor=<rgba 1 1 1 1>}
     uVRegion {bIsValid=false, max=<0 0>, min=<0 0>}
    }
   }
   slot {class=OverlaySlot, properties {horizontalAlignment=HAlign_Fill, verticalAlignment=VAlign_Fill}}
  }
  RadialGlow {
   widgetClassPath=&Image
   parent=BadgesContainer
   widget=RadialGlow
   properties {
    visibility=SelfHitTestInvisible
    brush {
     bIsDynamicallyLoaded=false
     drawAs=Image
     imageType=NoImage
     mirroring=NoMirror
     resourceName=None
     resourceObject=/Game/UI/Frontend/Lobby/Habanero/Habanero_Materials/MaterialInstances/Flares/MI_UI_Habanero_RadialGlow
     tiling=NoTile
     imageSize=<32 32>
     margin=<ltrb 0 0 0 0>
     outlineSettings {
      bUseBrushTransparency=false
      roundingType=HalfHeightRadius
      width=0
      color {colorUseRule=UseColor_Specified, specifiedColor=<rgba 0 0 0 0>}
      cornerRadii=<0 0 0 0>
     }
     tintColor {colorUseRule=UseColor_Specified, specifiedColor=<rgba 1 1 1 1>}
     uVRegion {bIsValid=false, max=<0 0>, min=<0 0>}
    }
    renderTransform {angle=0, scale=<1.25 1.25>, shear=<0 0>, translation=<0 96>}
   }
   slot {class=OverlaySlot, properties {horizontalAlignment=HAlign_Fill, verticalAlignment=VAlign_Fill}}
  }
  CurrentBadge {
   widgetClassPath=&Overlay
   parent=BadgesContainer
   widget=CurrentBadge
   properties {renderTransform {angle=0, scale=<0.6 0.6>, shear=<0 0>, translation=<0 0>}}
   slot {class=OverlaySlot, properties {horizontalAlignment=HAlign_Fill, verticalAlignment=VAlign_Fill}}
  }
  CurrentBadge_Icon {
   widgetClassPath=&Image
   bIsVariable=true
   parent=CurrentBadge
   widget=CurrentBadge_Icon
   properties {visibility=SelfHitTestInvisible, renderTransform {angle=0, scale=<1 1>, shear=<0 0>, translation=<8 8>}}
   slot {class=OverlaySlot, properties {horizontalAlignment=HAlign_Fill, verticalAlignment=VAlign_Fill}}
  }
  NextBadge {
   widgetClassPath=&Overlay
   parent=BadgesContainer
   widget=NextBadge
   properties {renderTransform {angle=0, scale=<0.6 0.6>, shear=<0 0>, translation=<0 0>}}
   slot {class=OverlaySlot, properties {horizontalAlignment=HAlign_Fill, verticalAlignment=VAlign_Fill}}
  }
  NextBadge_Bg {
   widgetClassPath=&Image
   bIsVariable=true
   parent=NextBadge
   widget=NextBadge_Bg
   properties {
    visibility=SelfHitTestInvisible
    brush {
     bIsDynamicallyLoaded=false
     drawAs=Image
     imageType=NoImage
     mirroring=NoMirror
     resourceName=None
     resourceObject=/Game/UI/Frontend/Lobby/Habanero/Habanero_Materials/MaterialInstances/Icons/Background/MI_UI_Habanero_Background_08
     tiling=NoTile
     imageSize=<32 32>
     margin=<ltrb 0 0 0 0>
     outlineSettings {
      bUseBrushTransparency=false
      roundingType=HalfHeightRadius
      width=0
      color {colorUseRule=UseColor_Specified, specifiedColor=<rgba 0 0 0 0>}
      cornerRadii=<0 0 0 0>
     }
     tintColor {colorUseRule=UseColor_Specified, specifiedColor=<rgba 1 1 1 1>}
     uVRegion {bIsValid=false, max=<0 0>, min=<0 0>}
    }
    renderTransform {angle=0, scale=<1.125 1.125>, shear=<0 0>, translation=<0 0>}
   }
   slot {class=OverlaySlot, properties {horizontalAlignment=HAlign_Fill, verticalAlignment=VAlign_Fill}}
  }
  NextBadge_Icon {
   widgetClassPath=&Image
   bIsVariable=true
   parent=NextBadge
   widget=NextBadge_Icon
   properties {visibility=SelfHitTestInvisible}
   slot {class=OverlaySlot, properties {horizontalAlignment=HAlign_Fill, verticalAlignment=VAlign_Fill}}
  }
  Overlay_BonusFeedback {
   widgetClassPath=&Overlay
   parent=BadgesContainer
   widget=Overlay_BonusFeedback
   properties {renderTransform {angle=0, scale=<1.3 1.3>, shear=<0 0>, translation=<0 0>}}
   slot {class=OverlaySlot, properties {horizontalAlignment=HAlign_Fill, verticalAlignment=VAlign_Fill}}
  }
  Image_ElimGlow {
   widgetClassPath=&Image
   bIsVariable=true
   parent=Overlay_BonusFeedback
   widget=Image_ElimGlow
   properties {
    renderOpacity=0
    brush {
     bIsDynamicallyLoaded=false
     drawAs=Image
     imageType=FullColor
     mirroring=NoMirror
     resourceName=None
     resourceObject=/Game/UI/Frontend/Lobby/Habanero/Habanero_Materials/M_UI_Habanero_RecapGlow
     tiling=NoTile
     imageSize=<32 32>
     margin=<ltrb 0 0 0 0>
     outlineSettings {
      bUseBrushTransparency=false
      roundingType=HalfHeightRadius
      width=0
      color {colorUseRule=UseColor_Specified, specifiedColor=<rgba 0 0 0 0>}
      cornerRadii=<0 0 0 0>
     }
     tintColor {colorUseRule=UseColor_Specified, specifiedColor=<rgba 1 1 1 1>}
     uVRegion {bIsValid=false, max=<0 0>, min=<0 0>}
    }
   }
   slot {class=OverlaySlot, properties {horizontalAlignment=HAlign_Fill, verticalAlignment=VAlign_Fill}}
  }
  Image_PlacementGlow {
   widgetClassPath=&Image
   bIsVariable=true
   parent=Overlay_BonusFeedback
   widget=Image_PlacementGlow
   properties {
    renderOpacity=0
    brush {
     bIsDynamicallyLoaded=false
     drawAs=Image
     imageType=FullColor
     mirroring=NoMirror
     resourceName=None
     resourceObject=/Game/UI/Frontend/Lobby/Habanero/Habanero_Materials/MaterialInstances/Flares/MI_UI_Habanero_RecapGlow
     tiling=NoTile
     imageSize=<32 32>
     margin=<ltrb 0 0 0 0>
     outlineSettings {
      bUseBrushTransparency=false
      roundingType=HalfHeightRadius
      width=0
      color {colorUseRule=UseColor_Specified, specifiedColor=<rgba 0 0 0 0>}
      cornerRadii=<0 0 0 0>
     }
     tintColor {colorUseRule=UseColor_Specified, specifiedColor=<rgba 1 1 1 1>}
     uVRegion {bIsValid=false, max=<0 0>, min=<0 0>}
    }
   }
   slot {class=OverlaySlot, properties {horizontalAlignment=HAlign_Fill, verticalAlignment=VAlign_Fill}}
  }
  FortVisualAttachment_Text {
   widgetClassPath=&FortVisualAttachment
   parent=VerticalBox_1
   widget=FortVisualAttachment_Text
   properties {contentAnchor=<0.5 0>}
   slot {class=VerticalBoxSlot, properties {horizontalAlignment=HAlign_Center}}
  }
  RankProtectedBox_Switcher {
   widgetClassPath=&CommonAnimatedSwitcher
   bIsVariable=true
   parent=FortVisualAttachment_Text
   widget=RankProtectedBox_Switcher
   slot {class=SizeBoxSlot}
  }
  TextContainer {
   widgetClassPath=&VerticalBox
   parent=RankProtectedBox_Switcher
   widget=TextContainer
   properties {flowDirectionPreference=Culture, renderTransform {angle=0, scale=<1 1>, shear=<0 0>, translation=<0 -25>}}
   slot {class=WidgetSwitcherSlot}
  }
  Text_TierName {
   widgetClassPath=&CommonTextBlock
   bIsVariable=true
   parent=TextContainer
   widget=Text_TierName
   properties {
    justification=Center
    lineHeightPercentage=0.9
    text=Unreal
    textTransformPolicy=ToUpper
    font {
     bForceMonospaced=false
     bMaterialIsStencil=false
     fontMaterial=null
     fontObject=&Burbank
     letterSpacing=0
     monospacedWidth=1
     size=60
     skewAmount=0
     typefaceFontName=BurbankRegular-Black
     outlineSettings {
      bApplyOutlineToDropShadows=false
      bMiteredCorners=false
      bSeparateFillAlpha=false
      outlineMaterial=null
      outlineSize=0
      outlineColor=<rgba 0 0 0 1>
     }
    }
    shadowOffset=<0 0>
    shapedTextOptions {bOverride_TextFlowDirection=true, bOverride_TextShapingMethod=true, textFlowDirection=Culture, textShapingMethod=Auto}
    strikeBrush {
     bIsDynamicallyLoaded=false
     drawAs=NoDrawType
     imageType=NoImage
     mirroring=NoMirror
     resourceName=None
     resourceObject=null
     tiling=NoTile
     imageSize=<32 32>
     margin=<ltrb 0 0 0 0>
     outlineSettings {
      bUseBrushTransparency=false
      roundingType=HalfHeightRadius
      width=0
      color {colorUseRule=UseColor_Specified, specifiedColor=<rgba 0 0 0 0>}
      cornerRadii=<0 0 0 0>
     }
     tintColor {colorUseRule=UseColor_Specified, specifiedColor=<rgba 1 1 1 1>}
     uVRegion {bIsValid=false, max=<0 0>, min=<0 0>}
    }
   }
   slot {class=VerticalBoxSlot}
  }
  Box_PercentageText {
   widgetClassPath=&HorizontalBox
   parent=TextContainer
   widget=Box_PercentageText
   properties {flowDirectionPreference=Culture}
   slot {class=VerticalBoxSlot, properties {horizontalAlignment=HAlign_Center}}
  }
  Text_PercentProgress {
   widgetClassPath=&CommonTextBlock
   bIsVariable=true
   parent=Box_PercentageText
   widget=Text_PercentProgress
   properties {
    flowDirectionPreference=Culture
    style=/FortUILibrary/Styles/Typography/_TextStyle_SubH4._TextStyle_SubH4_C
    text=82%
    font {
     bForceMonospaced=false
     bMaterialIsStencil=false
     fontMaterial=null
     fontObject=&BurbankSmall
     letterSpacing=0
     monospacedWidth=1
     size=24
     skewAmount=0
     typefaceFontName=Black
     outlineSettings {
      bApplyOutlineToDropShadows=false
      bMiteredCorners=false
      bSeparateFillAlpha=false
      outlineMaterial=null
      outlineSize=0
      outlineColor=<rgba 0 0 0 1>
     }
    }
    shadowOffset=<0 0>
    strikeBrush {
     bIsDynamicallyLoaded=false
     drawAs=NoDrawType
     imageType=NoImage
     mirroring=NoMirror
     resourceName=None
     resourceObject=null
     tiling=NoTile
     imageSize=<32 32>
     margin=<ltrb 0 0 0 0>
     outlineSettings {
      bUseBrushTransparency=false
      roundingType=HalfHeightRadius
      width=0
      color {colorUseRule=UseColor_Specified, specifiedColor=<rgba 0 0 0 0>}
      cornerRadii=<0 0 0 0>
     }
     tintColor {colorUseRule=UseColor_Specified, specifiedColor=<rgba 1 1 1 1>}
     uVRegion {bIsValid=false, max=<0 0>, min=<0 0>}
    }
   }
   slot {class=HorizontalBoxSlot, properties {horizontalAlignment=HAlign_Center, verticalAlignment=VAlign_Bottom}}
  }
  HorizontalBox_RankDelta {
   widgetClassPath=&HorizontalBox
   parent=Box_PercentageText
   widget=HorizontalBox_RankDelta
   properties {flowDirectionPreference=Culture, visibility=Collapsed}
   slot {class=HorizontalBoxSlot, properties {horizontalAlignment=HAlign_Center, verticalAlignment=VAlign_Bottom, padding=<ltrb 5 0 0 3>}}
  }
  Image_PositionChange {
   widgetClassPath=&Image
   bIsVariable=true
   parent=HorizontalBox_RankDelta
   widget=Image_PositionChange
   properties {
    flowDirectionPreference=Culture
    visibility=Collapsed
    brush {
     bIsDynamicallyLoaded=false
     drawAs=Image
     imageType=NoImage
     mirroring=NoMirror
     resourceName=None
     resourceObject=/FortniteCoreUI/Athena/HUD/EndGame/PostGameFlowScreens/DefaultScreens/XPScreen/Art/T_UI_Icon_HabaneroRankDeltaTriangle
     tiling=NoTile
     imageSize=<12 8>
     margin=<ltrb 0 0 0 0>
     outlineSettings {
      bUseBrushTransparency=false
      roundingType=HalfHeightRadius
      width=0
      color {colorUseRule=UseColor_Specified, specifiedColor=<rgba 0 0 0 0>}
      cornerRadii=<0 0 0 0>
     }
     tintColor {colorUseRule=UseColor_Specified, specifiedColor=<rgba 1 1 1 1>}
     uVRegion {bIsValid=false, max=<0 0>, min=<0 0>}
    }
    colorAndOpacity=<rgba 0.9686 1 0.102 1>
   }
   slot {class=HorizontalBoxSlot, properties {horizontalAlignment=HAlign_Center, verticalAlignment=VAlign_Center}}
  }
  Text_PercentChange {
   widgetClassPath=&CommonTextBlock
   bIsVariable=true
   parent=HorizontalBox_RankDelta
   widget=Text_PercentChange
   properties {
    flowDirectionPreference=Culture
    text=+17%
    colorAndOpacity {colorUseRule=UseColor_Specified, specifiedColor=<rgba 0.9686 1 0.102 1>}
    font {
     bForceMonospaced=false
     bMaterialIsStencil=false
     fontMaterial=null
     fontObject=&BurbankSmall
     letterSpacing=0
     monospacedWidth=1
     size=18
     skewAmount=0
     typefaceFontName=Black
     outlineSettings {
      bApplyOutlineToDropShadows=false
      bMiteredCorners=false
      bSeparateFillAlpha=false
      outlineMaterial=null
      outlineSize=0
      outlineColor=<rgba 0 0 0 1>
     }
    }
    shadowOffset=<0 0>
    strikeBrush {
     bIsDynamicallyLoaded=false
     drawAs=NoDrawType
     imageType=NoImage
     mirroring=NoMirror
     resourceName=None
     resourceObject=null
     tiling=NoTile
     imageSize=<32 32>
     margin=<ltrb 0 0 0 0>
     outlineSettings {
      bUseBrushTransparency=false
      roundingType=HalfHeightRadius
      width=0
      color {colorUseRule=UseColor_Specified, specifiedColor=<rgba 0 0 0 0>}
      cornerRadii=<0 0 0 0>
     }
     tintColor {colorUseRule=UseColor_Specified, specifiedColor=<rgba 1 1 1 1>}
     uVRegion {bIsValid=false, max=<0 0>, min=<0 0>}
    }
   }
   slot {class=HorizontalBoxSlot}
  }
  Box_SkillRating {
   widgetClassPath=&HorizontalBox
   bIsVariable=true
   parent=TextContainer
   widget=Box_SkillRating
   properties {renderOpacity=0, visibility=Collapsed}
   slot {class=VerticalBoxSlot, properties {horizontalAlignment=HAlign_Center}}
  }
  Text_SkillRatingName {
   widgetClassPath=&CommonTextBlock
   bIsVariable=true
   parent=Box_SkillRating
   widget=Text_SkillRatingName
   properties {
    justification=Center
    lineHeightPercentage=0.9
    text=Skill Rating
    textTransformPolicy=ToUpper
    colorAndOpacity {colorUseRule=UseColor_Specified, specifiedColor=<rgba 0.5776 0.5776 0.5776 1>}
    font {
     bForceMonospaced=false
     bMaterialIsStencil=false
     fontMaterial=null
     fontObject=&HeadingNow
     letterSpacing=0
     monospacedWidth=1
     size=24
     skewAmount=0
     typefaceFontName=76 Bold
     outlineSettings {
      bApplyOutlineToDropShadows=false
      bMiteredCorners=false
      bSeparateFillAlpha=false
      outlineMaterial=null
      outlineSize=0
      outlineColor=<rgba 0 0 0 1>
     }
    }
    shadowOffset=<0 0>
    shapedTextOptions {bOverride_TextFlowDirection=true, bOverride_TextShapingMethod=true, textFlowDirection=Culture, textShapingMethod=Auto}
    strikeBrush {
     bIsDynamicallyLoaded=false
     drawAs=NoDrawType
     imageType=NoImage
     mirroring=NoMirror
     resourceName=None
     resourceObject=null
     tiling=NoTile
     imageSize=<32 32>
     margin=<ltrb 0 0 0 0>
     outlineSettings {
      bUseBrushTransparency=false
      roundingType=HalfHeightRadius
      width=0
      color {colorUseRule=UseColor_Specified, specifiedColor=<rgba 0 0 0 0>}
      cornerRadii=<0 0 0 0>
     }
     tintColor {colorUseRule=UseColor_Specified, specifiedColor=<rgba 1 1 1 1>}
     uVRegion {bIsValid=false, max=<0 0>, min=<0 0>}
    }
   }
   slot {class=HorizontalBoxSlot}
  }
  Text_SkillRating {
   widgetClassPath=&CommonTextBlock
   bIsVariable=true
   parent=Box_SkillRating
   widget=Text_SkillRating
   properties {
    justification=Center
    lineHeightPercentage=0.9
    text=8888
    textTransformPolicy=ToUpper
    font {
     bForceMonospaced=false
     bMaterialIsStencil=false
     fontMaterial=null
     fontObject=&HeadingNow
     letterSpacing=0
     monospacedWidth=1
     size=24
     skewAmount=0
     typefaceFontName=76 Bold
     outlineSettings {
      bApplyOutlineToDropShadows=false
      bMiteredCorners=false
      bSeparateFillAlpha=false
      outlineMaterial=null
      outlineSize=0
      outlineColor=<rgba 0 0 0 1>
     }
    }
    shadowOffset=<0 0>
    shapedTextOptions {bOverride_TextFlowDirection=true, bOverride_TextShapingMethod=true, textFlowDirection=Culture, textShapingMethod=Auto}
    strikeBrush {
     bIsDynamicallyLoaded=false
     drawAs=NoDrawType
     imageType=NoImage
     mirroring=NoMirror
     resourceName=None
     resourceObject=null
     tiling=NoTile
     imageSize=<32 32>
     margin=<ltrb 0 0 0 0>
     outlineSettings {
      bUseBrushTransparency=false
      roundingType=HalfHeightRadius
      width=0
      color {colorUseRule=UseColor_Specified, specifiedColor=<rgba 0 0 0 0>}
      cornerRadii=<0 0 0 0>
     }
     tintColor {colorUseRule=UseColor_Specified, specifiedColor=<rgba 1 1 1 1>}
     uVRegion {bIsValid=false, max=<0 0>, min=<0 0>}
    }
   }
   slot {class=HorizontalBoxSlot, properties {padding=<ltrb 12 0 0 0>}}
  }
  TextWrapper_RankUp {
   widgetClassPath=&Border
   bIsVariable=true
   parent=TextContainer
   widget=TextWrapper_RankUp
   properties {
    background {
     bIsDynamicallyLoaded=false
     drawAs=NoDrawType
     imageType=NoImage
     mirroring=NoMirror
     resourceName=None
     resourceObject=null
     tiling=NoTile
     imageSize=<32 32>
     margin=<ltrb 0 0 0 0>
     outlineSettings {
      bUseBrushTransparency=false
      roundingType=HalfHeightRadius
      width=0
      color {colorUseRule=UseColor_Specified, specifiedColor=<rgba 0 0 0 0>}
      cornerRadii=<0 0 0 0>
     }
     tintColor {colorUseRule=UseColor_Specified, specifiedColor=<rgba 1 1 1 1>}
     uVRegion {bIsValid=false, max=<0 0>, min=<0 0>}
    }
    contentColorAndOpacity=<rgba 0.008 0.4342 0.8714 1>
    padding=<ltrb 0 0 0 0>
   }
   slot {class=VerticalBoxSlot}
  }
  Text_EventTitle {
   widgetClassPath=&CommonTextBlock
   bIsVariable=true
   parent=TextWrapper_RankUp
   widget=Text_EventTitle
   properties {
    bIsScrollingEnabled=false
    justification=Center
    style=/FortUILibrary/Styles/Typography/_TextStyle_SubH2._TextStyle_SubH2_C
    text="{Rank Event}"
    textTransformPolicy=ToUpper
    font {
     bForceMonospaced=false
     bMaterialIsStencil=false
     fontMaterial=null
     fontObject=&Burbank
     letterSpacing=0
     monospacedWidth=1
     size=33
     skewAmount=0
     typefaceFontName=BurbankRegular-Black
     outlineSettings {
      bApplyOutlineToDropShadows=false
      bMiteredCorners=false
      bSeparateFillAlpha=false
      outlineMaterial=null
      outlineSize=0
      outlineColor=<rgba 0 0 0 1>
     }
    }
    shadowOffset=<0 0>
    strikeBrush {
     bIsDynamicallyLoaded=false
     drawAs=NoDrawType
     imageType=NoImage
     mirroring=NoMirror
     resourceName=None
     resourceObject=null
     tiling=NoTile
     imageSize=<32 32>
     margin=<ltrb 0 0 0 0>
     outlineSettings {
      bUseBrushTransparency=false
      roundingType=HalfHeightRadius
      width=0
      color {colorUseRule=UseColor_Specified, specifiedColor=<rgba 0 0 0 0>}
      cornerRadii=<0 0 0 0>
     }
     tintColor {colorUseRule=UseColor_Specified, specifiedColor=<rgba 1 1 1 1>}
     uVRegion {bIsValid=false, max=<0 0>, min=<0 0>}
    }
   }
   slot {class=BorderSlot, properties {padding=<ltrb 0 0 0 0>}}
  }
  Text_RankUpMessage {
   widgetClassPath=&CommonTextBlock
   bIsVariable=true
   parent=TextContainer
   widget=Text_RankUpMessage
   properties {
    flowDirectionPreference=Culture
    justification=Center
    style=&_TextStyle_SubH4_C
    textTransformPolicy=ToUpper
    wrapTextAt=600
    font {
     bForceMonospaced=false
     bMaterialIsStencil=false
     fontMaterial=null
     fontObject=&BurbankSmall
     letterSpacing=0
     monospacedWidth=1
     size=24
     skewAmount=0
     typefaceFontName=Black
     outlineSettings {
      bApplyOutlineToDropShadows=false
      bMiteredCorners=false
      bSeparateFillAlpha=false
      outlineMaterial=null
      outlineSize=0
      outlineColor=<rgba 0 0 0 1>
     }
    }
    shadowOffset=<0 0>
    strikeBrush {
     bIsDynamicallyLoaded=false
     drawAs=NoDrawType
     imageType=NoImage
     mirroring=NoMirror
     resourceName=None
     resourceObject=null
     tiling=NoTile
     imageSize=<32 32>
     margin=<ltrb 0 0 0 0>
     outlineSettings {
      bUseBrushTransparency=false
      roundingType=HalfHeightRadius
      width=0
      color {colorUseRule=UseColor_Specified, specifiedColor=<rgba 0 0 0 0>}
      cornerRadii=<0 0 0 0>
     }
     tintColor {colorUseRule=UseColor_Specified, specifiedColor=<rgba 1 1 1 1>}
     uVRegion {bIsValid=false, max=<0 0>, min=<0 0>}
    }
   }
   slot {class=VerticalBoxSlot, properties {horizontalAlignment=HAlign_Center, verticalAlignment=VAlign_Bottom, padding=<ltrb 0 45 0 0>}}
  }
  RankProtected_VerticalBox {widgetClassPath=&VerticalBox, parent=RankProtectedBox_Switcher, widget=RankProtected_VerticalBox, slot {class=WidgetSwitcherSlot}}
  WBP_Habanero_Shield {
   widgetClassPath=/Game/UI/Frontend/Lobby/Habanero/WBP_Habanero_Shield
   bIsVariable=true
   parent=RankProtected_VerticalBox
   widget=WBP_Habanero_Shield
   slot {class=VerticalBoxSlot, properties {horizontalAlignment=HAlign_Center, verticalAlignment=VAlign_Center}}
  }
  Text_RankProtected {
   widgetClassPath=&CommonTextBlock
   bIsVariable=true
   parent=RankProtected_VerticalBox
   widget=Text_RankProtected
   properties {
    justification=Center
    lineHeightPercentage=0.9
    text=Rank Protected
    textTransformPolicy=ToUpper
    font {
     bForceMonospaced=false
     bMaterialIsStencil=false
     fontMaterial=null
     fontObject=&Burbank
     letterSpacing=0
     monospacedWidth=1
     size=60
     skewAmount=0
     typefaceFontName=BurbankRegular-Black
     outlineSettings {
      bApplyOutlineToDropShadows=false
      bMiteredCorners=false
      bSeparateFillAlpha=false
      outlineMaterial=null
      outlineSize=0
      outlineColor=<rgba 0 0 0 1>
     }
    }
    shadowOffset=<0 0>
    strikeBrush {
     bIsDynamicallyLoaded=false
     drawAs=NoDrawType
     imageType=NoImage
     mirroring=NoMirror
     resourceName=None
     resourceObject=null
     tiling=NoTile
     imageSize=<32 32>
     margin=<ltrb 0 0 0 0>
     outlineSettings {
      bUseBrushTransparency=false
      roundingType=HalfHeightRadius
      width=0
      color {colorUseRule=UseColor_Specified, specifiedColor=<rgba 0 0 0 0>}
      cornerRadii=<0 0 0 0>
     }
     tintColor {colorUseRule=UseColor_Specified, specifiedColor=<rgba 1 1 1 1>}
     uVRegion {bIsValid=false, max=<0 0>, min=<0 0>}
    }
   }
   slot {class=VerticalBoxSlot, properties {padding=<ltrb 0 -15 0 0>}}
  }
  FortVisualAttachment_Elims {
   widgetClassPath=&FortVisualAttachment
   parent=HorizontalBox_0
   widget=FortVisualAttachment_Elims
   properties {contentAnchor=<0 0.5>, renderTransform {angle=0, scale=<1 1>, shear=<0 0>, translation=<0 15>}}
   slot {class=HorizontalBoxSlot, properties {horizontalAlignment=HAlign_Left, verticalAlignment=VAlign_Center}}
  }
  DisplayTournamentBanner_Switcher {
   widgetClassPath=&CommonVisibilitySwitcher
   bIsVariable=true
   parent=FortVisualAttachment_Elims
   widget=DisplayTournamentBanner_Switcher
   slot {class=SizeBoxSlot}
  }
  StatRightContainer_VerticalBox {
   widgetClassPath=&VerticalBox
   bIsVariable=true
   parent=DisplayTournamentBanner_Switcher
   widget=StatRightContainer_VerticalBox
   slot {class=CommonVisibilitySwitcherSlot}
  }
  Stat_Right_0 {
   widgetClassPath=&WBP_PostMatch_HabaneroSequence_Stat
   bIsVariable=true
   parent=StatRightContainer_VerticalBox
   widget=Stat_Right_0
   slot {class=VerticalBoxSlot}
  }
  Stat_Right_1 {
   widgetClassPath=&WBP_PostMatch_HabaneroSequence_Stat
   bIsVariable=true
   parent=StatRightContainer_VerticalBox
   widget=Stat_Right_1
   properties {visibility=Collapsed}
   slot {class=VerticalBoxSlot}
  }
  WBP_TournamentSmallBanner {
   widgetClassPath=/Game/Athena/HUD/EndGame/PostGameFlowScreens/DefaultScreens/XPScreen/WBP_TournamentSmallBanner
   bIsVariable=true
   parent=DisplayTournamentBanner_Switcher
   widget=WBP_TournamentSmallBanner
   slot {class=CommonVisibilitySwitcherSlot, properties {padding=<ltrb -110 0 0 0>}}
  }
  Overlay_EliminationTimeline {
   widgetClassPath=&Overlay
   bIsVariable=true
   parent=ContentWrapper
   widget=Overlay_EliminationTimeline
   slot {class=OverlaySlot, properties {horizontalAlignment=HAlign_Center, verticalAlignment=VAlign_Fill}}
  }
  SizeBox_0 {
   widgetClassPath=&SizeBox
   parent=Overlay_EliminationTimeline
   widget=SizeBox_0
   properties {bOverride_HeightOverride=true, bOverride_WidthOverride=true, heightOverride=36, widthOverride=750}
   slot {class=OverlaySlot, properties {horizontalAlignment=HAlign_Center, verticalAlignment=VAlign_Center}}
  }
  Timeline {
   widgetClassPath=&Image
   bIsVariable=true
   parent=SizeBox_0
   widget=Timeline
   properties {
    visibility=SelfHitTestInvisible
    brush {
     bIsDynamicallyLoaded=false
     drawAs=Image
     imageType=NoImage
     mirroring=NoMirror
     resourceName=None
     resourceObject=/FortniteCoreUI/Athena/HUD/EndGame/PostGameFlowScreens/DefaultScreens/XPScreen/Art/MI_UI_LinearProgressBar
     tiling=NoTile
     imageSize=<32 32>
     margin=<ltrb 0 0 0 0>
     outlineSettings {
      bUseBrushTransparency=false
      roundingType=HalfHeightRadius
      width=0
      color {colorUseRule=UseColor_Specified, specifiedColor=<rgba 0 0 0 0>}
      cornerRadii=<0 0 0 0>
     }
     tintColor {colorUseRule=UseColor_Specified, specifiedColor=<rgba 1 1 1 1>}
     uVRegion {bIsValid=false, max=<0 0>, min=<0 0>}
    }
   }
   slot {class=SizeBoxSlot}
  }
  Overlay_TimelineEntries {
   widgetClassPath=&Overlay
   bIsVariable=true
   parent=Overlay_EliminationTimeline
   widget=Overlay_TimelineEntries
   slot {class=OverlaySlot, properties {verticalAlignment=VAlign_Center}}
  }
  EliminationEntry {
   widgetClassPath=/Game/Athena/HUD/EndGame/PostGameFlowScreens/DefaultScreens/XPScreen/WBP_EliminationTimelineEntry
   bIsVariable=true
   parent=Overlay_TimelineEntries
   widget=EliminationEntry
   slot {class=OverlaySlot, properties {horizontalAlignment=HAlign_Center, verticalAlignment=VAlign_Center, padding=<ltrb 0 0 0 48>}}
  }
  HorizontalBox_1 {
   widgetClassPath=&HorizontalBox
   parent=Overlay_EliminationTimeline
   widget=HorizontalBox_1
   slot {class=OverlaySlot, properties {horizontalAlignment=HAlign_Center, verticalAlignment=VAlign_Bottom}}
  }
  Minutes {
   widgetClassPath=&CommonTextBlock
   bIsVariable=true
   parent=HorizontalBox_1
   widget=Minutes
   properties {
    autoWrapText=true
    justification=Center
    text=00
    textTransformPolicy=ToUpper
    visibility=HitTestInvisible
    font {
     bForceMonospaced=false
     bMaterialIsStencil=false
     fontMaterial=null
     fontObject=&Burbank
     letterSpacing=0
     monospacedWidth=1
     size=34
     skewAmount=0
     typefaceFontName=Burbank-PlayerPower
     outlineSettings {
      bApplyOutlineToDropShadows=false
      bMiteredCorners=false
      bSeparateFillAlpha=false
      outlineMaterial=null
      outlineSize=0
      outlineColor=<rgba 0 0 0 1>
     }
    }
    shadowOffset=<0 0>
   }
   slot {class=HorizontalBoxSlot}
  }
  Minutes_1 {
   widgetClassPath=&CommonTextBlock
   parent=HorizontalBox_1
   widget=Minutes_1
   properties {
    autoWrapText=true
    justification=Center
    text=:
    textTransformPolicy=ToUpper
    visibility=HitTestInvisible
    font {
     bForceMonospaced=false
     bMaterialIsStencil=false
     fontMaterial=null
     fontObject=&Burbank
     letterSpacing=0
     monospacedWidth=1
     size=34
     skewAmount=0
     typefaceFontName=Burbank-PlayerPower
     outlineSettings {
      bApplyOutlineToDropShadows=false
      bMiteredCorners=false
      bSeparateFillAlpha=false
      outlineMaterial=null
      outlineSize=0
      outlineColor=<rgba 0 0 0 1>
     }
    }
    shadowOffset=<0 0>
   }
   slot {class=HorizontalBoxSlot}
  }
  Seconds {
   widgetClassPath=&CommonTextBlock
   bIsVariable=true
   parent=HorizontalBox_1
   widget=Seconds
   properties {
    autoWrapText=true
    justification=Center
    text=00
    textTransformPolicy=ToUpper
    visibility=HitTestInvisible
    font {
     bForceMonospaced=false
     bMaterialIsStencil=false
     fontMaterial=null
     fontObject=&Burbank
     letterSpacing=0
     monospacedWidth=1
     size=34
     skewAmount=0
     typefaceFontName=Burbank-PlayerPower
     outlineSettings {
      bApplyOutlineToDropShadows=false
      bMiteredCorners=false
      bSeparateFillAlpha=false
      outlineMaterial=null
      outlineSize=0
      outlineColor=<rgba 0 0 0 1>
     }
    }
    shadowOffset=<0 0>
   }
   slot {class=HorizontalBoxSlot}
  }
  SizeBox_Error {
   widgetClassPath=&SizeBox
   bIsVariable=true
   parent=Overlay_0
   widget=SizeBox_Error
   properties {bOverride_WidthOverride=true, visibility=Collapsed, widthOverride=650}
   slot {class=OverlaySlot, properties {horizontalAlignment=HAlign_Center, padding=<ltrb 0 290 0 0>}}
  }
  Text_ErrorText {
   widgetClassPath=&CommonTextBlock
   bIsVariable=true
   parent=SizeBox_Error
   widget=Text_ErrorText
   properties {
    autoWrapText=true
    justification=Center
    text="{This is test text}"
    textTransformPolicy=ToUpper
    font {
     bForceMonospaced=false
     bMaterialIsStencil=false
     fontMaterial=null
     fontObject=&Burbank
     letterSpacing=0
     monospacedWidth=1
     size=30
     skewAmount=0
     typefaceFontName=Burbank-PlayerPower
     outlineSettings {
      bApplyOutlineToDropShadows=false
      bMiteredCorners=false
      bSeparateFillAlpha=false
      outlineMaterial=null
      outlineSize=0
      outlineColor=<rgba 0 0 0 1>
     }
    }
    renderTransform {angle=0, scale=<1 1>, shear=<-8 0>, translation=<0 0>}
    shadowOffset=<0 0>
   }
   slot {class=SizeBoxSlot}
  }
  Image_WaitingSpinner {
   widgetClassPath=&Image
   bIsVariable=true
   parent=Overlay_0
   widget=Image_WaitingSpinner
   properties {
    visibility=Collapsed
    brush {
     bIsDynamicallyLoaded=false
     drawAs=Image
     imageType=NoImage
     mirroring=NoMirror
     resourceName=None
     resourceObject=/Game/UI/DiscoveryShared/Art/MI_UI_Discovery_Throbber
     tiling=NoTile
     imageSize=<192 192>
     margin=<ltrb 0 0 0 0>
     outlineSettings {
      bUseBrushTransparency=false
      roundingType=HalfHeightRadius
      width=0
      color {colorUseRule=UseColor_Specified, specifiedColor=<rgba 0 0 0 0>}
      cornerRadii=<0 0 0 0>
     }
     tintColor {colorUseRule=UseColor_Specified, specifiedColor=<rgba 1 1 1 1>}
     uVRegion {bIsValid=false, max=<0 0>, min=<0 0>}
    }
   }
   slot {class=OverlaySlot, properties {horizontalAlignment=HAlign_Center, verticalAlignment=VAlign_Center}}
  }
  WBP_CaptureForPostBufferUpdate {
   widgetClassPath=&WBP_CaptureForPostBufferUpdate
   bIsVariable=true
   parent=Overlay_0
   slot {class=OverlaySlot, properties {horizontalAlignment=HAlign_Fill, verticalAlignment=VAlign_Fill}}
  }
 ]
}
