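# Ranked end-game recap widget, based on AthenaLastKillerInfoWidget, that shows an "Eliminated By" row with the killer's rank icon and name.
# The row and its Habanero rank icon both start collapsed.
# SetKillerName fills the name text and reveals the row.
# SetRankedProgress updates the icon from the killer's ranked progress data and reveals it only when a matching tier is found.
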WidgetBlueprint /Game/Athena/HUD/EndGame/WBP_RankedRecapKillerInfo
graphs [EventGraph {nodes [
   n1 = Event SetRankedProgress!() -> (OutputDelegate, then, RankedProgress:"Ranked Progress Replicated Data Structure (by ref)") {is=title K2Node_Event}
   > n2 = UpdateDetaislFromTierDataUsingRankedProgressData!(
    execute
    self:WBP Habanero Icon Object Reference=Get WBP_Habanero_Icon() -> (WBP_Habanero_Icon:WBP Habanero Icon Object Reference)
    RankedProgressData:Ranked Progress Replicated Data Structure=@n1.RankedProgress
   ) -> (then, FoundTier:bool)
   > Branch(Condition=@n2)
   > SetVisibility(self=Get WBP_Habanero_Icon(), InVisibility=SelfHitTestInvisible)
   n3 = Event SetKillerName!() -> (OutputDelegate, then, PlayerName:String, PlayerID:String) {is=title K2Node_Event}
   > "SetText (Text)"(self=Get Text_KillerName() -> (Text_KillerName:Common Text Object Reference), InText="To Text (String)"(InString=@n3.PlayerName))
   > SetVisibility(self=Get KillerNameHorizontalBox() -> (KillerNameHorizontalBox:Horizontal Box Object Reference), InVisibility=HitTestInvisible)
  ]}]
variables [RankData {type=FortHabaneroDisplayData, default=/Game/UI/Habanero/HabaneroDisplayData}]
widgets {
 info {inheritedWidgetCount=0, parentClass=/Script/FortniteUI.AthenaLastKillerInfoWidget, rootWidgetClass=&HorizontalBox, widgetCount=5}
 widgets [
  KillerNameHorizontalBox {
   widgetClassPath=&HorizontalBox
   bIsVariable=true
   widget=KillerNameHorizontalBox
   properties {flowDirectionPreference=Culture, visibility=Collapsed}
  }
  CommonTextBlock_88 {
   widgetClassPath=&CommonTextBlock
   parent=KillerNameHorizontalBox
   widget=CommonTextBlock_88
   properties {
    justification=Center
    text=Eliminated By
    textTransformPolicy=ToUpper
    colorAndOpacity {colorUseRule=UseColor_Specified, specifiedColor=<rgba 0.7843 0.7843 0.7843 1>}
    font {
     bForceMonospaced=false
     bMaterialIsStencil=false
     fontMaterial=null
     fontObject=&BurbankSmall
     letterSpacing=0
     monospacedWidth=1
     size=24
     skewAmount=0
     typefaceFontName=Black
     outlineSettings {
      bApplyOutlineToDropShadows=true
      bMiteredCorners=false
      bSeparateFillAlpha=true
      outlineMaterial=&M-FontShadow
      outlineSize=3
      outlineColor=<rgba 0 0 0 0.5>
     }
    }
    shadowOffset=<0 0>
   }
   slot {class=HorizontalBoxSlot, properties {verticalAlignment=VAlign_Center}}
  }
  Spacer_70 {
   widgetClassPath=&Spacer
   parent=KillerNameHorizontalBox
   widget=Spacer_70
   properties {size=<10 1>}
   slot {class=HorizontalBoxSlot}
  }
  WBP_Habanero_Icon {
   widgetClassPath=/Game/UI/Frontend/Lobby/Habanero/WBP_Habanero_Icon
   bIsVariable=true
   parent=KillerNameHorizontalBox
   widget=WBP_Habanero_Icon
   properties {visibility=Collapsed, iconMinSize=<50 50>}
   slot {class=HorizontalBoxSlot}
  }
  Text_KillerName {
   widgetClassPath=&CommonTextBlock
   bIsVariable=true
   parent=KillerNameHorizontalBox
   widget=Text_KillerName
   properties {
    justification=Center
    text="{player}"
    font {
     bForceMonospaced=false
     bMaterialIsStencil=false
     fontMaterial=null
     fontObject=&BurbankSmall
     letterSpacing=0
     monospacedWidth=1
     size=18
     skewAmount=0
     typefaceFontName=Black
     outlineSettings {
      bApplyOutlineToDropShadows=true
      bMiteredCorners=false
      bSeparateFillAlpha=true
      outlineMaterial=&M-FontShadow
      outlineSize=3
      outlineColor=<rgba 0 0 0 0.5>
     }
    }
    shadowOffset=<0 0>
   }
   slot {class=HorizontalBoxSlot, properties {horizontalAlignment=HAlign_Center, verticalAlignment=VAlign_Center}}
  }
 ]
}
