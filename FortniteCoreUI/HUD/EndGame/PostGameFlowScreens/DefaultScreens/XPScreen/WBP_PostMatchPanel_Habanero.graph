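# Post-match panel for ranked (Habanero) play, showing rank progress, match stats, killer info, quest recap and an elimination timeline.
# Tier updates start the child rank sequence once unless stale; error codes start a failure sequence instead.
# Timeline entries and survival accolades fill a scroll box in reverse order, revealed by timer after the sequence finishes.
# A post-game audio bus mix runs while the panel is active and not spectating.
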WidgetBlueprint /FortniteCoreUI/HUD/EndGame/PostGameFlowScreens/DefaultScreens/XPScreen/WBP_PostMatchPanel_Habanero
animations [SequenceFinishedShow {
  binding_count=2
  length_seconds=0.4964
  object=&self:SequenceFinishedShow
  sequence {bindings [
    WBP_RankedRecapKillerInfo {
     id=#1
     possessed_class=WBP_RankedRecapKillerInfo_C
     tracks [{
       type=MovieSceneFloatTrack
       display_name=Render Opacity
       sections [{type=MovieSceneFloatSection, end_frame=10, start_frame=0, channels [None {type=MovieSceneScriptingFloatChannel, keys [[0.0717,0],[10,1]]}]}]
      }]
    }
    Widget_QuestRecap {
     id=#2
     possessed_class=EndOfMatch_QuestProgessRecap_C
     tracks [{
       type=MovieSceneFloatTrack
       display_name=Render Opacity
       sections [{type=MovieSceneFloatSection, end_frame=10, start_frame=0, channels [None {type=MovieSceneScriptingFloatChannel, keys [[0.0717,0],[10,1]]}]}]
      }]
    }
    Overlay_Quests {
     id=#3
     possessed_class=Overlay
     tracks [{
       type=MovieSceneFloatTrack
       display_name=Render Opacity
       sections [{type=MovieSceneFloatSection, end_frame=10, start_frame=0, channels [None {type=MovieSceneScriptingFloatChannel, default=0}]}]
      }]
    }
   ]}
 }]
graphs [
 AddWidgetToList {nodes [
   n1 = AddWidgetToList!() -> (then, NewItem:"WBP Elimination List Entry Object Reference (by ref)") {is=title K2Node_FunctionEntry}
   > Add(
    TargetArray:Array of WBP Elimination List Entry Object References=Get CreatedEliminationEntries() -> (CreatedEliminationEntries:Array of WBP Elimination List Entry Object References)
    NewItem:"WBP Elimination List Entry Object Reference (by ref)"=@n1
   ) {is=title K2Node_CallArrayFunction}
   > SetVisibility(self=@n1, InVisibility=Collapsed)
  ]}
 CreateWidgetForEliminationList {nodes [
   n1 = CreateWidgetForEliminationList!() -> (then, EliminationEntry:Elimination Timeline Entry Structure) {is=title K2Node_FunctionEntry}
   n2 = Break Elimination Timeline Entry!(EliminationTimelineEntry:"Elimination Timeline Entry Structure (by ref)"=@n1) -> (
    PlayerName:String
    PlayerRankType:String
    PlayerRank:int
    PlayerCurrentPosition:int
    PlayerMMR:int
    EliminatedAvatarBrushPtr:Texture 2D Soft Object Reference
    PlayersLeft:int
    PlayersTotal:int
    Time:float=~
    bLocalElimination:bool=~
    bPlayersLeftEntry:bool=~
   ) {is=title K2Node_BreakStruct}
   > n3 = Branch(Condition=@n2.bPlayersLeftEntry)
   n4 = FindBestAccoladeText!(execute=[@n3.then, @n9.then], self, PlayersLeft:int=@n2.PlayersLeft) -> (then, PlayerCountsToDisplay:Array of Integers)
   > Branch(Condition="integer > integer"(A=Length(TargetArray:Array of Integers=@n4)))
   > Set LastPostedAccoladeText(LastPostedAccoladeText:int=@n2.PlayersLeft) -> (Output_Get:int=0)
   > n5 = For Each Loop(Array:Array of Integers=@n4) -> (Array Element:int)
   > n6 = Create Widget!(
    execute:Exec
    Class:User Widget Class Reference=/FortniteCoreUI/Athena/HUD/EndGame/PostGameFlowScreens/DefaultScreens/XPScreen/WBP_EliminationListEntry_SurvivalAccolade.WBP_EliminationListEntry_SurvivalAccolade_C
    OwningPlayer:Player Controller Object Reference
   ) -> (then:Exec, ReturnValue:WBP Elimination List Entry Survival Accolade Object Reference)
   > SetText(
    self:WBP Elimination List Entry Survival Accolade Object Reference=@n6
    InText:Text=Find(
     TargetMap:Map of Integers to Texts=Get PlayerCountToAccoladeText() -> (PlayerCountToAccoladeText:Map of Integers to Texts)
     Key:"Integer (by ref)"=@n5
    ) -> (Value:Text)
   )
   > AddWidgetToList!(execute, self, NewItem:"WBP Elimination List Entry Object Reference (by ref)"=@n6) -> (then) {is=title K2Node_CallFunction}
   n7 = Create Widget(
    execute=@n3.else
    Class=/FortniteCoreUI/Athena/HUD/EndGame/PostGameFlowScreens/DefaultScreens/XPScreen/WBP_EliminationListEntry_RankedRecap.WBP_EliminationListEntry_RankedRecap_C
   ) -> (ReturnValue:WBP Elimination List Entry Ranked Recap Object Reference)
   n8 = Break Elimination Timeline Entry(EliminationTimelineEntry=@n1) -> (Time=~, bLocalElimination=~, bPlayersLeftEntry=~)
   > Initialize!(
    execute
    self:WBP Elimination List Entry Ranked Recap Object Reference=@n7
    Name:String=@n8
    RankType:String=@n8.PlayerRankType
    Rank:int=@n8.PlayerRank
    Current MMR:int=@n8.PlayerMMR
    Current Position:int=@n8.PlayerCurrentPosition
    Avatar:Texture 2D Soft Object Reference=@n8.EliminatedAvatarBrushPtr
    Placement:int=int + int(A=@n8.PlayersLeft)
   ) -> (then)
   > AddWidgetToList(NewItem=@n7)
   n9 = Branch(Condition="integer > integer"(A=@n2.PlayersLeft, B=8))
  ]}
 EventGraph {nodes [
   n1 = "On Skiil Rating Visibility Changed (WBP_PostMatch_HabaneroSequence)"!() -> (OutputDelegate, then, IsVisible:bool) {is=title K2Node_ComponentBoundEvent}
   n2 = IsInSplitScreen!() -> (OutputPin:bool)
   > SetHeightOverride(
    self=[Get SB_Elimination() -> (SB_Elimination:Size Box Object Reference), Get SB_Spacer() -> (SB_Spacer:Size Box Object Reference)]
    InHeightOverride=Select(
     Option 0:float=Get EliminationsBoxHeight() -> (EliminationsBoxHeight:float)
     Option 1:float=Select(Option 0=200, Option 1=90, Index=@n2)
     Index=@n1.IsVisible
    ) -> (ReturnValue:float)
   )
   n3 = HandleOnCMSImageLoaded!() -> (OutputDelegate, then, bValidTexture:bool=~) {is=title K2Node_CustomEvent}
   > Branch(Condition=@n3.bValidTexture)
   n4 = Break Next Tournament Data!(NextTournamentData:"Next Tournament Data Structure (by ref)"=Get NextTournamentData() -> (NextTournamentData:Next Tournament Data Structure)) -> (
    TournamentName:String
    StartTime:Date Time Structure
    EndTime:Date Time Structure
    bIsLoadingImage:bool=~
    TournamentTexture:Texture 2D Object Reference
   ) {is=title K2Node_BreakStruct}
   > Is Valid(InputObject=@n4.TournamentTexture)
   > SetTournamentThumbnail!(
    execute
    self:WBP Post Match Habanero Sequence Object Reference=Get WBP_PostMatch_HabaneroSequence() -> (WBP_PostMatch_HabaneroSequence:WBP Post Match Habanero Sequence Object Reference)
    Texture:Texture Object Reference=@n4.TournamentTexture
   ) -> (then)
   n5 = Event UpdateEliminationTimeline!() -> (OutputDelegate, then, EliminationTimeline:Array of Elimination Timeline Entry Structures, MatchTime:float) {is=title K2Node_Event}
   > n6 = Set Elimination Entries(Elimination Entries:Array of Elimination Timeline Entry Structures=@n5.EliminationTimeline) -> (Output_Get:Array of Elimination Timeline Entry Structures)
   > n7 = Set Total Match Time(Total Match Time:double=@n5.MatchTime) -> (Output_Get:double)
   > Initialize Timeline!(execute, self, EliminationEntries:Array of Elimination Timeline Entry Structures=@n6, TotalMatchTime:double=@n7) -> (then)
   n8 = Event UpdateHabaneroErrorCode!() -> (OutputDelegate, then, ProgressError:EHabaneroProgressError Enum=None, OptionalText:String) {is=title K2Node_Event}
   n9 = Get HasSequenceStarted() -> (HasSequenceStarted:bool)
   > Branch(Condition=AND Boolean(
     A="Not Equal (Enum)"!(A:EHabaneroProgressError Enum=@n8.ProgressError, B:EHabaneroProgressError Enum=None) -> (ReturnValue:bool)
     B=NOT Boolean(A=@n9)
    ))
   > TryStartSequenceFailure()
   Event OnInitialized()
   > SetVisibility(
    self=[
     Get Widget_QuestRecapContainer() -> (Widget_QuestRecapContainer:Overlay Object Reference)
     Get WBP_RankedRecapKillerInfo() -> (WBP_RankedRecapKillerInfo:WBP Ranked Recap Killer Info Object Reference)
    ]
    InVisibility=Collapsed
   )
   > SetVisibility(self=Get Overlay_EliminationList() -> (Overlay_EliminationList:Overlay Object Reference), InVisibility=Hidden)
   > Assign On Sequence Finished!(
    execute:Exec
    self:WBP Post Match Habanero Sequence Object Reference=Get WBP_PostMatch_HabaneroSequence()
    Delegate:"Delegate (by ref)"=Create Event()
   ) -> (then:Exec) {is=title K2Node_AssignDelegate}
   > Bind Event to On CMSImage Loaded(self:Athena Post Match Panel Ranked Recap Object Reference, Delegate=Create Event()) {is=title K2Node_AddDelegate}
   n10 = Select(Option 0=290, Option 1=120, Index=IsInSplitScreen())
   > Set EliminationsBoxHeight(EliminationsBoxHeight:float=@n10) -> (Output_Get:float)
   > SetHeightOverride(self=[Get SB_Elimination(), Get SB_Spacer()], InHeightOverride=@n10)
   Event On Activated()
   > Parent: On Activated() {is=title K2Node_CallParentFunction}
   > RefreshPostGameMixMod()
   n11 = Event UpdateTierDisplay!() -> (
    OutputDelegate
    then
    TierList:Array of Fort Habanero Tier Structures
    RankedGameMode:String
    InitialTierIndex:int=0
    InitialProgress:float
    InitialPlayerPosition:int=0
    InitialMMR:int=0
    bIsInitialTopTier:bool
    CurrentTierIndex:int=0
    CurrentProgress:float
    CurrentPlayerPosition:int=0
    CurrentMMR:int=0
    bIsCurrentTopTier:bool
    bIsStale:bool
    bDebugForceShow:bool
   ) {is=title K2Node_Event}
   n12 = Get HasSequenceStarted()
   > LogMessage!(
    execute
    self:Athena Post Match Panel Ranked Recap Object Reference
    Message:String="To String (Text)"(InText=Format Text!(
      Format="LOCTEXT(\"BP received update.  Is Stale? {stale}  Has sequence already started? {started}\")"
      stale:bool=@n11.bIsStale
      started:bool=@n12
     ) -> (Result))
   ) -> (then)
   n13 = NOT Boolean(A=Get HasSequenceStarted())
   n14 = NOT Boolean(A=@n11.bIsStale)
   > Branch(Condition=OR Boolean(A=AND Boolean(A=@n13, B=@n14), B=@n11.bDebugForceShow))
   > Set HasSequenceStarted(HasSequenceStarted:bool=true) -> (Output_Get:bool)
   > n15 = GetUnrankedTierData!(execute, self:Athena Post Match Panel Ranked Recap Object Reference) -> (then, OutUnrankedTierData:Fort Habanero Tier Structure, ReturnValue:bool)
   > n16 = GetTopTierForGameMode!(
    execute
    self:Fort Habanero Display Data Object Reference=Get HabaneroDisplayData() -> (HabaneroDisplayData:Fort Habanero Display Data Object Reference)
    GameModeName:String=@n11.RankedGameMode
   ) -> (then, ReturnValue:Fort Habanero Tier Structure)
   > StartRankSequence!(
    execute
    self:WBP Post Match Habanero Sequence Object Reference=Get WBP_PostMatch_HabaneroSequence()
    ActiveTierList:Array of Fort Habanero Tier Structures=@n11.TierList
    InitialTierIndex:int=@n11.InitialTierIndex
    InitialProgress:float=@n11.InitialProgress
    InitialPlayerPosition:int=@n11.InitialPlayerPosition
    TargetTierIndex:int=@n11.CurrentTierIndex
    TargetProgress:float=@n11.CurrentProgress
    TargetPlayerPosition:int=@n11.CurrentPlayerPosition
    IsUnrankedTierValid:bool=@n15.ReturnValue
    UnrankedTierData:Fort Habanero Tier Structure=@n15
    EliminationTimeline:Array of Elimination Timeline Entry Structures=Get Elimination Entries() -> (Elimination Entries:Array of Elimination Timeline Entry Structures)
    MatchTime:double=Get Total Match Time() -> (Total Match Time:double)
    IsInitialTopTier:bool=@n11.bIsInitialTopTier
    IsCurrentTopTier:bool=@n11.bIsCurrentTopTier
    InInitialMMR:int=@n11.InitialMMR
    InCurrentMMR:int=@n11.CurrentMMR
    TopTierData:Fort Habanero Tier Structure=@n16
   ) -> (then)
   > InitializeTournamentData!(
    execute
    self:WBP Post Match Habanero Sequence Object Reference=Get WBP_PostMatch_HabaneroSequence()
    NextTournamentData:Next Tournament Data Structure=Get NextTournamentData()
   ) -> (then)
   > DisplayedRankedRecap!(execute, self:Athena Post Match Panel Ranked Recap Object Reference, bSuccessfulSequence:bool=true) -> (then)
   Event OnActivatePanel() {is=title K2Node_Event}
   > RefreshStats()
   > n17 = Branch(Condition=Get HasSequenceStarted())
   LogMessage(execute=@n17.else, Message=BP waiting for start)
   > StartWaitingSequence!(
    execute
    self:WBP Post Match Habanero Sequence Object Reference=Get WBP_PostMatch_HabaneroSequence()
    TimeoutDuration:double=GetFailsafeDelayDuration!(self:Athena Post Match Panel Ranked Recap Object Reference) -> (ReturnValue:float)
   ) -> (then)
   n18 = Event OnSpectatingToggled!() -> (OutputDelegate, then, bIsSpectating:bool) {is=title K2Node_Event}
   > Set IsSpectating(IsSpectating:bool=@n18.bIsSpectating) -> (Output_Get:bool)
   > RefreshPostGameMixMod()
   Event On Deactivated()
   > Parent: On Deactivated() {is=title K2Node_CallParentFunction}
   > RefreshPostGameMixMod()
   n19 = Branch(Condition=ShouldPostMixBeEnabled!(self) -> (IsEnabled:bool))
   n20 = Get PostMatchControlBusMix() -> (PostMatchControlBusMix:Sound Control Bus Mix Object Reference)
   > Activate Control Bus Mix!(execute, Mix:Sound Control Bus Mix Object Reference=@n20) -> (then)
   Deactivate Control Bus Mix!(execute=@n19.else, Mix:Sound Control Bus Mix Object Reference=@n20) -> (then)
   RefreshPostGameMixMod() {is=title K2Node_CustomEvent}
  ]}
 FindBestAccoladeText {nodes [
   FindBestAccoladeText!() -> (then, PlayersLeft:int) {is=title K2Node_FunctionEntry}
   > n1 = Keys!(execute, TargetMap:Map of Integers to Texts=Get PlayerCountToAccoladeText()) -> (then, Keys:Array of Integers)
   > n2 = For Each Loop with Break!(Exec:Exec, Array:Array of Integers=@n1, Break:Exec) -> (LoopBody:Exec, Array Element:int, Array Index:int, Completed:Exec)
   Sort Integer Array!(
    execute=@n2.Completed
    TargetArray:Array of Integers=Get PlayerRemainingCounts() -> (PlayerRemainingCounts:Array of Integers)
    bStableSort:bool=false
    SortOrder:EArraySortOrder Enum=Descending
   ) -> (then)
   > Return Node!(execute, PlayerCountsToDisplay:Array of Integers=Get PlayerRemainingCounts())
   n3 = "integer > integer"(A=Get LastPostedAccoladeText() -> (LastPostedAccoladeText:int=0), B=@n2)
   n4 = "Equal (Integer)"(A=Get LastPostedAccoladeText() -> (LastPostedAccoladeText=0))
   n5 = Get PlayersLeft() -> (PlayersLeft:int=0)
   Branch(execute=@n2.LoopBody, Condition=AND Boolean(A="integer >= integer"(A=@n2, B=@n5), B=OR Boolean(A=@n4, B=@n3)))
   > Add(TargetArray:Array of Integers=Get PlayerRemainingCounts(), NewItem:"Integer (by ref)"=@n2)
  ]}
 Initialize Timeline {nodes [
   Initialize Timeline!() -> (then, EliminationEntries:Array of Elimination Timeline Entry Structures, TotalMatchTime:double=~) {is=title K2Node_FunctionEntry}
   > ClearChildren(self=Get Widget_EliminationListScrollBox() -> (Widget_EliminationListScrollBox:Scroll Box Object Reference))
   > n1 = For Each Loop(Array:Array of Elimination Timeline Entry Structures=Get Elimination Entries()) -> (Array Element:Elimination Timeline Entry Structure)
   n2 = Reverse for Each Loop!(Exec:Exec=@n1.Completed, Array:Array of WBP Elimination List Entry Object References=Get CreatedEliminationEntries()) -> (LoopBody:Exec, ArrayIndex:int, ArrayElement:WBP Elimination List Entry Object Reference, Completed:Exec) {is=title K2Node_MacroInstance}
   > AddChild(self=Get Widget_EliminationListScrollBox(), Content=@n2.ArrayElement)
   CreateWidgetForEliminationList!(execute=@n1.LoopBody, self, EliminationEntry:Elimination Timeline Entry Structure=@n1) -> (then) {is=title K2Node_CallFunction}
  ]}
 IsInSplitScreen {nodes [
   IsInSplitScreen() {is=title K2Node_FunctionEntry}
   n1 = GetFirstLocalFortPlayerController!(bIncludeNonSpectators:bool=true, bIncludeSpectators:bool=false) -> (ReturnValue:Fort Player Controller Object Reference)
   > Return Node!(
    execute
    OutputPin:bool=AND Boolean(A=IsValid(Object=@n1), B=IsInSplitscreen!(self:Fort Player Controller Object Reference=@n1) -> (ReturnValue:bool))
   )
  ]}
 IsInSquad {nodes [
   IsInSquad() {is=title K2Node_FunctionEntry}
   > n1 = Cast To FortPlayerStateAthena!(execute, Object=Get PlayerState!(self:Controller Object Reference=GetOwningPlayer()) -> (PlayerState:Player State Object Reference)) -> (then, CastFailed, AsFort Player State Athena:Fort Player State Athena Object Reference) {is=title K2Node_DynamicCast}
   > Return Node!(execute, IsSquad:bool=IsInSquad!(self:Fort Player State Athena Object Reference=@n1) -> (ReturnValue:bool) {is=title K2Node_CallFunction})
   Return Node(execute=@n1.CastFailed, IsSquad=false)
  ]}
 OnSequenceFinished {nodes [
   OnSequenceFinished() {is=title K2Node_FunctionEntry}
   > OnPanelReady(self:Athena Post Match Panel Object Reference)
   > ShowQuests()
   > PlayAnimation(InAnimation=Get SequenceFinishedShow() -> (SequenceFinishedShow:Widget Animation Object Reference))
   > CheckForPlayerDeath(self:Athena Last Killer Info Widget Object Reference=Get WBP_RankedRecapKillerInfo())
   n1 = Get Widget_EliminationListScrollBox()
   > Branch(Condition="integer > integer"(A=GetChildrenCount!(self:Panel Widget Object Reference=@n1) -> (ReturnValue:int=0)))
   > ShowEliminationList()
  ]}
 PlayQuestIntroAnimation {nodes [
   n1 = PlayQuestIntroAnimation!() -> (then, QuestWidget:End Of Match Quest Entry Auto Object Reference) {is=title K2Node_FunctionEntry}
   > PlayIntroAnimation(self:End Of Match Quest Entry Auto Object Reference=@n1)
  ]}
 RefreshStats {nodes [
   RefreshStats() {is=title K2Node_FunctionEntry}
   > InitializeAllStatsAsCollapsed(self:WBP Post Match Habanero Sequence Object Reference=Get WBP_PostMatch_HabaneroSequence())
   > SetStatAtPosition!(
    execute
    self:WBP Post Match Habanero Sequence Object Reference=Get WBP_PostMatch_HabaneroSequence()
    StatTitle:Text=Select(
     Option 0:Text="LOCTEXT(\"Placement\")"
     Option 1:Text="LOCTEXT(\"Squad Placement\")"
     Index=IsInSquad!(self:Self Object Reference) -> (IsSquad:bool)
    ) -> (ReturnValue:Text)
    StatValue:int=GetLocalPlayerPlacement!(self:Fort HUDContext Object Reference=Get FortHUDContext!() -> (ReturnValue:Fort HUDContext Object Reference) {is=title K2Node_GetBlueprintContextEx}) -> (ReturnValue:int=0)
    StatPosition:ENUM_HabaneroSequence_StatPosition Enum=NewEnumerator0
   ) -> (then)
   > n1 = Cast To FortPlayerStateAthena(Object=Get PlayerState(self=GetOwningPlayer()))
   > SetStatAtPosition(
    self=Get WBP_PostMatch_HabaneroSequence()
    StatTitle=Select(Option 0="LOCTEXT(\"Eliminations\")", Option 1="LOCTEXT(\"Squad Eliminations\")", Index=IsInSquad())
    StatValue=Get TeamKillScore!(self:Fort Player State Athena Object Reference=@n1) -> (TeamKillScore:int=0)
    StatPosition=NewEnumerator1
   )
   Reroute Node()
   > n2 = Branch()
   > n3 = SetStatAtPosition(self=Get WBP_PostMatch_HabaneroSequence(), StatTitle="LOCTEXT(\"Enemy Team Wipes\")", StatValue=0, StatPosition=NewEnumerator3)
   Reroute Node(InputPin=[@n3.then, @n2.else])
   HasTag!(
    TagContainer:"Gameplay Tag Container Structure (by ref)"=GetAthenaPlaylistContextTags!(self:Fort Game State Athena Object Reference=GetGameStateAthena!() -> (ReturnValue:Fort Game State Athena Object Reference)) -> (ReturnValue:Gameplay Tag Container Structure)
    Tag:Gameplay Tag Structure=(TagName="Athena.Playlist.BlastBerry")
    bExactMatch:bool=false
   ) -> (ReturnValue:bool)
  ]}
 ShouldPostMixBeEnabled {nodes [
   ShouldPostMixBeEnabled() {is=title K2Node_FunctionEntry}
   n1 = Get IsSpectating() -> (IsSpectating:bool)
   > Return Node!(
    execute
    IsEnabled:bool=AND Boolean(A=IsActivated!(self:Common Activatable Widget Object Reference) -> (ReturnValue:bool), B=NOT Boolean(A=@n1))
   )
  ]}
 ShowEliminationList {nodes [
   ShowEliminationList() {is=title K2Node_FunctionEntry}
   > SetVisibility(self=Get Overlay_EliminationList(), InVisibility=SelfHitTestInvisible)
   > Set CurrentShowingEliminationEntryIndex(CurrentShowingEliminationEntryIndex:int=-1) -> (Output_Get:int=0)
   > ShowNextEliminationEntry()
  ]}
 ShowNextEliminationEntry {nodes [
   ShowNextEliminationEntry() {is=title K2Node_FunctionEntry}
   n1 = Get CurrentShowingEliminationEntryIndex() -> (CurrentShowingEliminationEntryIndex:int=0)
   > Increment Int(Value=@n1)
   n2 = Get CreatedEliminationEntries()
   > Branch(Condition=Is Valid Index!(TargetArray:Array of WBP Elimination List Entry Object References=@n2, IndexToTest:int=@n1) -> (ReturnValue:bool))
   n3 = "Get (a copy)"(Array:Array of WBP Elimination List Entry Object References=@n2, Dimension 1=@n1) -> (Output:WBP Elimination List Entry Object Reference)
   > SetVisibility(self=@n3, InVisibility=SelfHitTestInvisible)
   > PlayIntroAnimation(self:WBP Elimination List Entry Object Reference=@n3)
   > Set Timer by Event(Delegate=Create Event(), Time=Get TimeBetweenEliminationEntries() -> (TimeBetweenEliminationEntries:double), bLooping=false)
  ]}
 ShowQuests {nodes [
   ShowQuests() {is=title K2Node_FunctionEntry}
   > SetVisibility(self=Get Widget_QuestRecapContainer(), InVisibility=SelfHitTestInvisible)
  ]}
 TryStartSequenceFailure {nodes [
   TryStartSequenceFailure() {is=title K2Node_FunctionEntry}
   > n1 = Branch(Condition=Get HasSequenceStarted())
   LogMessage(execute=@n1.else, Message=BP starting failure sequence)
   > StartFailureSequence!(
    execute
    self:WBP Post Match Habanero Sequence Object Reference=Get WBP_PostMatch_HabaneroSequence()
    ErrorMessage:Text=Get ErrorText() -> (ErrorText:Text)
   ) -> (then)
   > OnPanelReady()
   > ShowQuests() {is=title K2Node_CallFunction}
   > Set HasSequenceStarted(HasSequenceStarted=true)
   > DisplayedRankedRecap(bSuccessfulSequence=false)
  ]}
]
variables [
 CreatedEliminationEntries {type="array<WBP_EliminationListEntry_C>", default []}
 CurrentShowingEliminationEntryIndex {type=integer}
 Elimination Entries {type="array<EliminationTimelineEntry>", default []}
 EliminationsBoxHeight {type=number}
 ErrorText {
  type=string
  default="We were unable to fetch your results but they have been recorded.\r\n\r\nYour rank will update automatically when ready."
 }
 HasSequenceStarted {type=boolean}
 IsSpectating {type=boolean}
 LastPostedAccoladeText {type=integer}
 PlayerCountToAccoladeText {default {10=Survivor III - Top 10 Players, 25=Survivor II - Top 25 Players, 50=Survivor I - Top 50 Players}}
 PostMatchControlBusMix {type=SoundControlBusMix, default=/FortUIAudioBase/Modulation/CBM_UI_PostGameScreen_Active}
 TimeBetweenEliminationEntries {type=number, default=0.5}
 Total Match Time {type=number}
]
widgets {
 info {inheritedWidgetCount=2, parentClass=/Script/FortniteUI.AthenaPostMatchPanel_RankedRecap, rootWidgetClass=&Overlay, widgetCount=20}
 widgets [
  Overlay_0 {widgetClassPath=&Overlay, widget=Overlay_0}
  WBP_PostMatch_HabaneroSequence {
   widgetClassPath=&/WBP_PostMatch_HabaneroSequence
   bIsVariable=true
   parent=Overlay_0
   widget=WBP_PostMatch_HabaneroSequence
   slot {class=OverlaySlot, properties {horizontalAlignment=HAlign_Fill, verticalAlignment=VAlign_Fill}}
  }
  SafeZone_0 {
   widgetClassPath=&SafeZone
   parent=Overlay_0
   widget=SafeZone_0
   properties {padLeft=false, padRight=false, padTop=false}
   slot {class=OverlaySlot, properties {horizontalAlignment=HAlign_Center, verticalAlignment=VAlign_Bottom, padding=<ltrb 0 0 0 100>}}
  }
  VerticalBox_0 {widgetClassPath=&VerticalBox, parent=SafeZone_0, widget=VerticalBox_0, slot {class=SafeZoneSlot}}
  WBP_RankedRecapKillerInfo {
   widgetClassPath=/Game/Athena/HUD/EndGame/WBP_RankedRecapKillerInfo
   bIsVariable=true
   parent=VerticalBox_0
   widget=WBP_RankedRecapKillerInfo
   slot {class=VerticalBoxSlot, properties {horizontalAlignment=HAlign_Center, verticalAlignment=VAlign_Bottom}}
  }
  Spacer_80 {widgetClassPath=&Spacer, parent=VerticalBox_0, widget=Spacer_80, properties {size=<1 35>}, slot {class=VerticalBoxSlot}}
  HorizontalBox_2 {widgetClassPath=&HorizontalBox, parent=VerticalBox_0, widget=HorizontalBox_2, slot {class=VerticalBoxSlot}}
  FortVisualAttachment_1 {
   widgetClassPath=&FortVisualAttachment
   parent=HorizontalBox_2
   widget=FortVisualAttachment_1
   properties {contentAnchor=<1 0>}
   slot {class=HorizontalBoxSlot}
  }
  SB_Spacer {
   widgetClassPath=&SizeBox
   bIsVariable=true
   parent=FortVisualAttachment_1
   widget=SB_Spacer
   properties {bOverride_HeightOverride=true, heightOverride=290}
   slot {class=SizeBoxSlot}
  }
  HorizontalBox_1 {widgetClassPath=&HorizontalBox, parent=SB_Spacer, widget=HorizontalBox_1, slot {class=SizeBoxSlot}}
  Spacer_232 {widgetClassPath=&Spacer, parent=HorizontalBox_1, widget=Spacer_232, properties {size=<32 1>}, slot {class=HorizontalBoxSlot}}
  Overlay_EliminationList {
   widgetClassPath=&Overlay
   bIsVariable=true
   parent=HorizontalBox_2
   widget=Overlay_EliminationList
   slot {class=HorizontalBoxSlot}
  }
  UIKit_Backplate {
   widgetClassPath=&WBP_UIKit_Backplate
   parent=Overlay_EliminationList
   widget=UIKit_Backplate
   properties {brightness=Dark, visibility=SelfHitTestInvisible}
   slot {class=OverlaySlot, properties {horizontalAlignment=HAlign_Fill, verticalAlignment=VAlign_Fill}}
  }
  SB_Elimination {
   widgetClassPath=&SizeBox
   bIsVariable=true
   parent=Overlay_EliminationList
   widget=SB_Elimination
   properties {bOverride_HeightOverride=true, bOverride_WidthOverride=true, heightOverride=290, widthOverride=750}
   slot {class=OverlaySlot}
  }
  Widget_EliminationListScrollBox {
   widgetClassPath=&ScrollBox
   bInherited=true
   bIsVariable=true
   parent=SB_Elimination
   widget=Widget_EliminationListScrollBox
   properties {widgetStyle {
     barThickness=9
     bottomShadowBrush {
      bIsDynamicallyLoaded=false
      drawAs=NoDrawType
      imageType=FullColor
      mirroring=NoMirror
      resourceName=../../../Engine/Content/Slate/Common/ScrollBoxShadowBottom.png
      resourceObject=null
      tiling=NoTile
      imageSize=<16 8>
      margin=<ltrb 0.5 0 0.5 1>
      outlineSettings {
       bUseBrushTransparency=false
       roundingType=HalfHeightRadius
       width=0
       color {colorUseRule=UseColor_Specified, specifiedColor=<rgba 0 0 0 0>}
       cornerRadii=<0 0 0 0>
      }
      tintColor {colorUseRule=UseColor_Specified, specifiedColor=<rgba 1 1 1 1>}
      uVRegion {bIsValid=false, max=<0 0>, min=<0 0>}
     }
     horizontalScrolledContentPadding=<ltrb 0 0 1 0>
     leftShadowBrush {
      bIsDynamicallyLoaded=false
      drawAs=NoDrawType
      imageType=FullColor
      mirroring=NoMirror
      resourceName=../../../Engine/Content/Slate/Common/ScrollBoxShadowLeft.png
      resourceObject=null
      tiling=NoTile
      imageSize=<8 16>
      margin=<ltrb 1 0.5 0 0.5>
      outlineSettings {
       bUseBrushTransparency=false
       roundingType=HalfHeightRadius
       width=0
       color {colorUseRule=UseColor_Specified, specifiedColor=<rgba 0 0 0 0>}
       cornerRadii=<0 0 0 0>
      }
      tintColor {colorUseRule=UseColor_Specified, specifiedColor=<rgba 1 1 1 1>}
      uVRegion {bIsValid=false, max=<0 0>, min=<0 0>}
     }
     rightShadowBrush {
      bIsDynamicallyLoaded=false
      drawAs=NoDrawType
      imageType=FullColor
      mirroring=NoMirror
      resourceName=../../../Engine/Content/Slate/Common/ScrollBoxShadowRight.png
      resourceObject=null
      tiling=NoTile
      imageSize=<8 16>
      margin=<ltrb 0 0.5 1 0.5>
      outlineSettings {
       bUseBrushTransparency=false
       roundingType=HalfHeightRadius
       width=0
       color {colorUseRule=UseColor_Specified, specifiedColor=<rgba 0 0 0 0>}
       cornerRadii=<0 0 0 0>
      }
      tintColor {colorUseRule=UseColor_Specified, specifiedColor=<rgba 1 1 1 1>}
      uVRegion {bIsValid=false, max=<0 0>, min=<0 0>}
     }
     topShadowBrush {
      bIsDynamicallyLoaded=false
      drawAs=NoDrawType
      imageType=FullColor
      mirroring=NoMirror
      resourceName=../../../Engine/Content/Slate/Common/ScrollBoxShadowTop.png
      resourceObject=null
      tiling=NoTile
      imageSize=<16 8>
      margin=<ltrb 0.5 1 0.5 0>
      outlineSettings {
       bUseBrushTransparency=false
       roundingType=HalfHeightRadius
       width=0
       color {colorUseRule=UseColor_Specified, specifiedColor=<rgba 0 0 0 0>}
       cornerRadii=<0 0 0 0>
      }
      tintColor {colorUseRule=UseColor_Specified, specifiedColor=<rgba 1 1 1 1>}
      uVRegion {bIsValid=false, max=<0 0>, min=<0 0>}
     }
     verticalScrolledContentPadding=<ltrb 0 0 0 1>
    }}
   slot {class=SizeBoxSlot, properties {padding=<ltrb 16 12 12 12>}}
  }
  HorizontalBox_WarningMessageContainer {
   widgetClassPath=&HorizontalBox
   parent=Overlay_EliminationList
   widget=HorizontalBox_WarningMessageContainer
   properties {flowDirectionPreference=Culture, visibility=Collapsed, renderTransform {angle=0, scale=<1 1>, shear=<0 0>, translation=<0 290>}}
   slot {class=OverlaySlot, properties {horizontalAlignment=HAlign_Center, padding=<ltrb 3 3 3 3>}}
  }
  Image_WarningIcon {
   widgetClassPath=&Image
   bIsVariable=true
   parent=HorizontalBox_WarningMessageContainer
   widget=Image_WarningIcon
   properties {
    brush {
     bIsDynamicallyLoaded=false
     drawAs=Image
     imageType=FullColor
     mirroring=NoMirror
     resourceName=None
     resourceObject=/FortUILibrary/Textures/MaterialInstances/MI_UIKit_Icon_AgeGatedAlert
     tiling=NoTile
     imageSize=<25 25>
     margin=<ltrb 0 0 0 0>
     outlineSettings {
      bUseBrushTransparency=false
      roundingType=HalfHeightRadius
      width=0
      color {colorUseRule=UseColor_Specified, specifiedColor=<rgba 0 0 0 0>}
      cornerRadii=<0 0 0 0>
     }
     tintColor {colorUseRule=UseColor_Specified, specifiedColor=<rgba 1 1 1 1>}
     uVRegion {bIsValid=false, max=<0 0>, min=<0 0>}
    }
    renderTransform {angle=0, scale=<1 1>, shear=<0 0>, translation=<0 3>}
   }
   slot {class=HorizontalBoxSlot, properties {horizontalAlignment=HAlign_Left, verticalAlignment=VAlign_Center}}
  }
  Text_WarningMessage {
   widgetClassPath=&CommonTextBlock
   bIsVariable=true
   parent=HorizontalBox_WarningMessageContainer
   widget=Text_WarningMessage
   properties {
    justification=Center
    lineHeightPercentage=0.8
    text=Repeat eliminations of the same player have diminishing credit
    wrapTextAt=700
    font {
     bForceMonospaced=false
     bMaterialIsStencil=false
     fontMaterial=null
     fontObject=&BurbankSmall
     letterSpacing=0
     monospacedWidth=1
     size=14.25
     skewAmount=0
     typefaceFontName=Black
     outlineSettings {
      bApplyOutlineToDropShadows=false
      bMiteredCorners=false
      bSeparateFillAlpha=false
      outlineMaterial=null
      outlineSize=0
      outlineColor=<rgba 0 0 0 0.5>
     }
    }
    shadowOffset=<0 0>
   }
   slot {class=HorizontalBoxSlot, properties {verticalAlignment=VAlign_Center, padding=<ltrb 5 0 5 0>}}
  }
  Widget_QuestRecapContainer {
   widgetClassPath=&Overlay
   bIsVariable=true
   parent=Overlay_0
   widget=Widget_QuestRecapContainer
   slot {class=OverlaySlot, properties {verticalAlignment=VAlign_Bottom}}
  }
  Widget_QuestRecap {
   widgetClassPath=/FortniteCoreUI/HUD/EndGame/EndOfMatch_QuestProgessRecap
   bInherited=true
   bIsVariable=true
   parent=Widget_QuestRecapContainer
   widget=Widget_QuestRecap
   properties {exclusiveFilterTags {gameplayTags [{tagName=QuestCategory.BR.Habanero}, {tagName=QuestCategory.BR.Ranked.Urgent}]}}
   slot {class=OverlaySlot, properties {padding=<ltrb 0 0 0 61>}}
  }
 ]
}
