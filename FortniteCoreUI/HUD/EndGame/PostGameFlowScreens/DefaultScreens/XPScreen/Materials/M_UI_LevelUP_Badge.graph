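# Translucent UI material for the animated level-up badge on the post-game XP screen.
# Anim_Silhouette sweeps a color fill vertically over the dark badge backing, with an eased sine hot spot applied by color dodge.
# A distance-field arrow, scaled and moved by Anim_ArrowPosition, gets antialiased edge and soft shadow masks from the badge alpha.
# Anim_ColorIntensity drives the color ramps; named reroutes carry the masks into color and opacity.
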Material /FortniteCoreUI/HUD/EndGame/PostGameFlowScreens/DefaultScreens/XPScreen/Materials/M_UI_LevelUP_Badge
graph {
 properties {bAutomaticallySetUsageInEditor=false, blendMode=BLEND_Translucent, materialDomain=MD_UI, usedShadingModels=Default Lit}
 nodes [
  Constant3Vector()
  Constant() {r=0.65}
  Constant() {r=0.5}
  n1 = Constant3Vector() {constant=<rgba 0.0086 0.0037 0.0409 1>}
  n2 = Constant3Vector() {constant=<rgba 0.013 0.0009 0.0382 1>}
  n3 = TextureCoordinate()
  n4 = NamedRerouteDeclaration(None=MaterialFunctionCall(UVs=@n3, "Rotation Angle (0-1)"=Constant() {r=-0.175}) {materialFunction=&CustomRotator}.Rotated Values) {name=AngleMask, nodeColor=<rgba 0 0 0 1>}
  n5 = TextureSample() {texture=/FortniteCoreUI/HUD/EndGame/PostGameFlowScreens/DefaultScreens/XPScreen/Textures/T_UI_PostMatch_LevelUP_Backing}
  n6 = NamedRerouteUsage()
  n7 = Constant3Vector() {constant=<rgba 0.75 0.25 1 1>}
  n8 = LinearInterpolate(A=@n6, B=@n7) {constAlpha=0.6667}
  n9 = LinearInterpolate(A=@n8) {constAlpha=0.85}
  n10 = ScalarParameter() {parameterName=Anim_ColorIntensity}
  n11 = TextureCoordinate()
  n12 = LinearInterpolate(Alpha=ScalarParameter() {parameterName=Anim_Silhouette}) {constA=1, constB=-1}
  n13 = ComponentMask(None=Add(A=@n11, B=@n12) {constB=0}) {g=true}
  n14 = LinearInterpolate(
   A=LinearInterpolate(A=@n1, B=LinearInterpolate(A=@n1, B=@n2, Alpha=@n4), Alpha=@n5.A)
   B=Saturate(None=MaterialFunctionCall(A=@n6, B=@n8, C=@n9, Alpha=@n10) {materialFunction=&Lerp_3Color}.Result)
   Alpha=Ceil(None=Saturate(None=@n13))
  )
  n15 = Saturate(None=MaterialFunctionCall!(Blend=NamedRerouteUsage(), Base=@n14) {materialFunction=/Engine/Functions/Engine_MaterialFunctions03/Blends/Blend_ColorDodge}.Result)
  n16 = MaterialFunctionCall!(A=ScalarParameter() {parameterName=Anim_Silhouette}, B=Constant()) {materialFunction=/FortUILibrary/MaterialFunctions/Basics/MF_UI_BooleanOperations}
  n17 = Constant3Vector() {constant=<rgba 0.0022 0 0.013 1>}
  n18 = ScalarParameter() {parameterName=Anim_ColorIntensity}
  LinearInterpolate(
   A=LinearInterpolate(A=LinearInterpolate(A=@n14, B=@n15, Alpha=@n16.IsGreater), B=LinearInterpolate(A=@n17, B=@n7, Alpha=@n18), Alpha=NamedRerouteUsage()) {constB=0}
   Alpha=NamedRerouteUsage()
  )
  NamedRerouteDeclaration(None=@n5.RGB) {name=TextureRGB, nodeColor=<rgba 0 0 0 1>}
  n19 = Add(A=@n11) {constB=0.5}
  n20 = ComponentMask(None=Sine(None=Add(A=@n19, B=@n12) {constB=0}) {period=2}) {g=true}
  NamedRerouteDeclaration(None=Saturate(None=MaterialFunctionCall("Value (0-1)"=@n20) {materialFunction=&MF_UI_EaseCurves}.EaseInQuint)) {name=HotSpotMask, nodeColor=<rgba 0 0 0 1>}
  n21 = NamedRerouteDeclaration(None=Saturate(None=SmoothStep(Value=@n5.A) {constMin=0.3333})) {name=BadgeMask, nodeColor=<rgba 0 0 0 1>}
  n22 = NamedRerouteUsage()
  n23 = LinearInterpolate(Alpha=ScalarParameter() {parameterName=Anim_ArrowPosition}) {constA=0.3333, constB=0.6667}
  n24 = Constant()
  n25 = LinearInterpolate(Alpha=ScalarParameter() {parameterName=Anim_ArrowPosition}) {constA=-1.6667, constB=0.2}
  n26 = TextureSample(UVs=Add(
    A=MaterialFunctionCall(Texture Scale=@n23) {materialFunction=&ScaleUVsByCenter}.UVs
    B=AppendVector(A=@n24, B=@n25)
   )) {
   samplerType=SAMPLERTYPE_LinearGrayscale
   texture=/FortniteCoreUI/HUD/EndGame/PostGameFlowScreens/DefaultScreens/XPScreen/Textures/T_UI_PostMatch_LevelUP_ArrowDF
  }
  n27 = MaterialFunctionCall(DistanceField=@n26.R, Threshold=Constant() {r=0.6667}, Smoothing=Constant() {r=0.0125}) {materialFunction=&MF_UI_AntiAliasing}
  NamedRerouteDeclaration(None=Multiply(A=LinearInterpolate(B=@n21, Alpha=@n22) {constA=1}, B=Saturate(None=@n27.Outer))) {name=ArrowMask, nodeColor=<rgba 0 0 0 1>}
  n28 = MaterialFunctionCall("Value (0-1)"=MaterialFunctionCall(DistanceField=@n26.R, Threshold=Constant() {r=0.5}, Smoothing=Constant() {r=1}) {materialFunction=&MF_UI_AntiAliasing}.Outer) {materialFunction=&MF_UI_EaseCurves}
  NamedRerouteDeclaration(None=Multiply(A=@n21, B=Saturate(None=@n28.EaseInSine))) {name=ArrowShadowMask, nodeColor=<rgba 0 0 0 1>}
  n29 = ConstantBiasScale(None=@n3) {bias=-0.5}
  NamedRerouteDeclaration(None=Ceil(None=ComponentMask(None=@n29) {g=true})) {name=DividerMask, nodeColor=<rgba 0 0 0 1>}
  n30 = Multiply(A=NamedRerouteUsage(), B=OneMinus(None=ScalarParameter() {parameterName=Anim_ColorIntensity}))
  n31 = NamedRerouteUsage()
  Saturate(None=LinearInterpolate(A=@n5.A, Alpha=LinearInterpolate(A=@n30, Alpha=@n31)))
 ]
}
