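# Additive UI material that draws a soft purple radial glow, named for text on the post-match XP screen.
# An inverted, saturated circle SDF forms the gradient, which feeds an ease-curve function.
# Anim_Pulse blends between sine-eased and linear falloff to mix two purple tints, which are multiplied by vertex alpha.
# A separate cubic-eased mask also lerps on Anim_Pulse, but its identical endpoints leave it unaffected by the pulse.
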Material /FortniteCoreUI/HUD/EndGame/PostGameFlowScreens/DefaultScreens/XPScreen/Materials/M_UI_PostMatchTextGlow
graph {
 properties {
  bAutomaticallySetUsageInEditor=false
  bUsedWithStaticMesh=false
  blendMode=BLEND_Additive
  materialDomain=MD_UI
  usedShadingModels=Default Lit
 }
 nodes [
  n1 = Constant()
  n2 = Saturate(None=OneMinus(None=MaterialFunctionCall(Size=@n1) {materialFunction=&MF_UI_Circle}.SDF))
  n3 = MaterialFunctionCall("Value (0-1)"=@n2) {materialFunction=&MF_UI_EaseCurves}
  LinearInterpolate(A=@n3.EaseInCubic, B=@n3.EaseInCubic, Alpha=ScalarParameter() {parameterName=Anim_Pulse})
  n4 = ScalarParameter() {parameterName=Anim_Pulse}
  Multiply(
   A=VertexColor().A
   B=LinearInterpolate(
    A=Constant3Vector() {constant=<rgba 0.1433 0.0689 1 1>}
    B=Constant3Vector() {constant=<rgba 0.527 0.0541 1 1>}
    Alpha=LinearInterpolate(A=@n3.EaseInSine, B=@n2, Alpha=@n4)
   )
  )
 ]
}
