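# UI-domain material that adds a shine to text on the post-match XP screen, judging by its name and path.
# A remapped vertical gradient blends white to lavender, then lerps toward a DarkColor parameter by Color Swap.
# A reflected gradient on slightly rotated UVs, positioned by ShineOffset and sharpened with a power, is color-dodged over that base to form the shine band.
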Material /FortniteCoreUI/HUD/EndGame/PostGameFlowScreens/DefaultScreens/XPScreen/Materials/M_UI_PostMatch_FontShine
graph {
 properties {bAutomaticallySetUsageInEditor=false, bUsedWithStaticMesh=false, materialDomain=MD_UI, usedShadingModels=Default Lit}
 nodes [
  Constant4Vector() {constant=<rgba 0.3968 0.5583 1 1>}
  n1 = MaterialFunctionCall!(
   UVs=MaterialFunctionCall(UVs=TextureCoordinate(), "Rotation Angle (0-1)"=Constant() {r=-0.04}) {materialFunction=&CustomRotator}.Rotated Values
   Offset=ScalarParameter() {defaultValue=-1, parameterName=ShineOffset}
   RotationCenter
   RotationAngle
  ) {materialFunction=/FortUILibrary/MaterialFunctions/Gradients/MF_UI_ReflectedGradient}
  n2 = InverseLinearInterpolate(
   A=ScalarParameter() {parameterName=GradientOffsetTop}
   B=ScalarParameter() {defaultValue=0.8, parameterName=GradientOffsetBottom}
   Value=MaterialFunctionCall(UV Channel=TextureCoordinate()) {materialFunction=&LinearGradient}.VGradient
  ) {constB=0.8}
  n3 = LinearInterpolate(
   A=Constant4Vector() {constant=<rgba 1 1 1 1>}.RGB
   B=Constant4Vector() {constant=<rgba 0.6732 0.5 1 1>}.RGB
   Alpha=NamedRerouteDeclaration(None=Saturate(None=@n2)) {name=UV_1, nodeColor=<rgba 1 0.2344 0 1>}
  )
  n4 = VectorParameter() {parameterName=DarkColor, defaultValue=<rgba 0.0965 0.03 0.6 1>}
  n5 = ScalarParameter() {parameterName=Color Swap}
  Multiply(A=MaterialFunctionCall!(Blend=Power(Base=@n1.Result) {constExponent=4}, Base=LinearInterpolate(A=@n3, B=@n4.RGB, Alpha=@n5)) {materialFunction=/Engine/Functions/Engine_MaterialFunctions03/Blends/Blend_ColorDodge}.Result)
 ]
}
