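# One-shot 5.1 MetaSound Source that plays the post-match level-up "In" UI sound.
# OnPlay triggers a six-channel Wave Player, which feeds each surround output and signals OnFinished when the wave ends.
# It uses the Fort_UI sound class with boosted volume, modulated by the level-up control bus.
# A pre-effect send feeds the music-volume sidechain audio bus.
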MetaSoundSource /Game/Sounds/Fort_UI_Sounds/Gameplay/XP/MSS_PostMatch_LevelUp_In
inputs [UE.Source.OnPlay {type=Trigger}]
interfaces [UE.OutputFormat.5dot1:1.0, UE.Source.OneShot:1.0, UE.Source:1.1]
outputs [
 "UE.OutputFormat.5dot1.Audio:0" {type=Audio}
 "UE.OutputFormat.5dot1.Audio:1" {type=Audio}
 "UE.OutputFormat.5dot1.Audio:2" {type=Audio}
 "UE.OutputFormat.5dot1.Audio:3" {type=Audio}
 "UE.OutputFormat.5dot1.Audio:4" {type=Audio}
 "UE.OutputFormat.5dot1.Audio:5" {type=Audio}
 UE.Source.OneShot.OnFinished {type=Trigger}
]
pages [Default {
  input_defaults {UE.Source.OnPlay=null}
  nodes [WavePlayer_0 = UE.Wave Player:5dot1(Play=^UE.Source.OnPlay) {also {defaults {Wave Asset=&/Fort_PostMatch_LevelUp_In}, version [1,1]}}]
  links [
   {dst="@UE.Source.OneShot.OnFinished", src=WavePlayer_0.On Finished}
   {dst="@UE.OutputFormat.5dot1.Audio:2", src=WavePlayer_0.Out Front Center}
   {dst="@UE.OutputFormat.5dot1.Audio:0", src=WavePlayer_0.Out Front Left}
   {dst="@UE.OutputFormat.5dot1.Audio:1", src=WavePlayer_0.Out Front Right}
   {dst="@UE.OutputFormat.5dot1.Audio:3", src=WavePlayer_0.Out Low Frequency}
   {dst="@UE.OutputFormat.5dot1.Audio:4", src=WavePlayer_0.Out Side Left}
   {dst="@UE.OutputFormat.5dot1.Audio:5", src=WavePlayer_0.Out Side Right}
  ]
 }]
properties {
 assetImportData=&self:AssetImportData
 numChannels=6
 outputFormat=FiveDotOne
 sampleRate=48000
 soundClassObject=&Fort_UI
 volume=1.5
 modulationSettings {volumeRouting=Union, volumeModulationDestination {bEnableModulation=true, modulators [/Game/Sounds/Modulation/Buses/CB_FortLevelUp]}}
 preEffectBusSends [{
   audioBus=&AB_Sidechain_FortMusic_Vol
   maxSendDistance=1000
   minSendDistance=100
   minSendLevel=0
   sendLevel=0.3
  }]
}
