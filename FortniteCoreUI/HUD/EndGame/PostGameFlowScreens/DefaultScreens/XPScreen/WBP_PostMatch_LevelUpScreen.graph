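# Post-match XP screen widget that plays a "Leveled Up" celebration with animated badge, glow effects and sounds.
# Construct starts a staged sequence: show with intro sound, main badge animation, then reversed show with outro sound.
# SetNumberOfLevels adds an "xN" suffix above one level and syncs a dummy text block, copying font shape but keeping its own materials and outline.
# Shutdown stops animations, fades intro sound, broadcasts OnVisualizationFinished and removes the widget.
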WidgetBlueprint /FortniteCoreUI/HUD/EndGame/PostGameFlowScreens/DefaultScreens/XPScreen/WBP_PostMatch_LevelUpScreen
animations [
 Anim_Show {
  binding_count=2
  length_seconds=0.25
  object=&self:Anim_Show
  sequence {bindings [
    Overlay_LeveledUp {
     id=#1
     possessed_class=Overlay
     tracks [{
       type=MovieSceneByteTrack
       display_name=Visibility
       sections [{type=MovieSceneByteSection, end_frame=0, start_frame=0, channels [None {type=MovieSceneScriptingByteChannel, const [2,1]}]}]
      }]
    }
    Overlay_IconEffects {
     id=#2
     possessed_class=Overlay
     tracks [{
       type=MovieSceneByteTrack
       display_name=Visibility
       sections [{type=MovieSceneByteSection, end_frame=0, start_frame=0, channels [None {type=MovieSceneScriptingByteChannel, const [2,1]}]}]
      }]
    }
   ]}
 }
 Anim_ShowLevelIcon {
  binding_count=7
  length_seconds=3.2
  object=&self:Anim_ShowLevelIcon
  sequence {bindings [
    Image_LevelUpIcon {
     id=#3
     possessed_class=Image
     tracks [{
       type=MovieScene2DTransformTrack
       sections [{
         type=MovieScene2DTransformSection
         end_frame=64
         start_frame=0
         channels [
          Angle {type=MovieSceneScriptingFloatChannel, keys [[3,-10],[8,0],[60,0],[64,10]]}
          Scale.X {type=MovieSceneScriptingFloatChannel, keys [[3,0],[8,1],[60,1],[64,0]]}
          Scale.Y {type=MovieSceneScriptingFloatChannel, keys [[3,0],[8,1],[60,1],[64,0]]}
         ]
        }]
      }]
    }
    Image_Glow {
     id=#4
     possessed_class=Image
     tracks [{
       type=MovieScene2DTransformTrack
       sections [{
         type=MovieScene2DTransformSection
         end_frame=64
         start_frame=0
         channels [
          Scale.X {type=MovieSceneScriptingFloatChannel, keys [[0,0],[3,1],[60,1],[64,0]]}
          Scale.Y {type=MovieSceneScriptingFloatChannel, keys [[0,0],[3,1],[60,1],[64,0]]}
         ]
        }]
      }]
    }
    Image_Bloom {
     id=#5
     possessed_class=Image
     tracks [
      {
       type=MovieSceneWidgetMaterialTrack
       sections [{
         type=MovieSceneParameterSection
         end_frame=63
         start_frame=0
         channels [
          Streak Size {type=MovieSceneScriptingFloatChannel, keys [[4,0.5],[6,0.75],[61,0.75],[63,0.5]]}
          Star Size {type=MovieSceneScriptingFloatChannel, keys [[3,0.5],[5,1.3],[59,1.3],[63,0.5]]}
         ]
        }]
      }
      {
       type=MovieScene2DTransformTrack
       sections [{
         type=MovieScene2DTransformSection
         end_frame=64
         start_frame=0
         channels [
          Scale.X {type=MovieSceneScriptingFloatChannel, keys [[0,0],[3,1],[62,1],[64,0]]}
          Scale.Y {type=MovieSceneScriptingFloatChannel, keys [[0,0],[3,1],[62,1],[64,0]]}
         ]
        }]
      }
     ]
    }
    Image_TrianglesRight_Large {
     id=#6
     possessed_class=Image
     tracks [
      {type=MovieSceneWidgetMaterialTrack, sections [{type=MovieSceneParameterSection, end_frame=30, start_frame=0}]}
      {
       type=MovieSceneFloatTrack
       display_name=Render Opacity
       sections [{type=MovieSceneFloatSection, end_frame=4, start_frame=0, channels [None {type=MovieSceneScriptingFloatChannel, keys [[0,0],[4,1]]}]}]
      }
      {
       type=MovieScene2DTransformTrack
       sections [{
         type=MovieScene2DTransformSection
         end_frame=4
         start_frame=0
         channels [Translation.X {type=MovieSceneScriptingFloatChannel, keys [[0,-250],[4,0]]}]
        }]
      }
     ]
    }
    Overlay_IconEffects {
     id=#7
     possessed_class=Overlay
     tracks [{
       type=MovieSceneByteTrack
       display_name=Visibility
       sections [{type=MovieSceneByteSection, end_frame=0, start_frame=0, channels [None {type=MovieSceneScriptingByteChannel, const [4,1]}]}]
      }]
    }
    Image_TrianglesLeft_Large {
     id=#8
     possessed_class=Image
     tracks [
      {type=MovieSceneWidgetMaterialTrack, sections [{type=MovieSceneParameterSection, end_frame=30, start_frame=0}]}
      {
       type=MovieScene2DTransformTrack
       sections [{
         type=MovieScene2DTransformSection
         end_frame=4
         start_frame=0
         channels [Translation.X {type=MovieSceneScriptingFloatChannel, keys [[0,250],[4,0]]}]
        }]
      }
      {
       type=MovieSceneFloatTrack
       display_name=Render Opacity
       sections [{type=MovieSceneFloatSection, end_frame=4, start_frame=0, channels [None {type=MovieSceneScriptingFloatChannel, keys [[0,0],[4,1]]}]}]
      }
     ]
    }
    Overlay_LeveledUp {
     id=#9
     possessed_class=Overlay
     tracks [
      {
       type=MovieSceneFloatTrack
       display_name=Render Opacity
       sections [{
         type=MovieSceneFloatSection
         end_frame=64
         start_frame=0
         channels [None {type=MovieSceneScriptingFloatChannel, keys [[0,0],[8,0],[12,1],[60,1],[64,0]]}]
        }]
      }
      {
       type=MovieScene2DTransformTrack
       sections [{
         type=MovieScene2DTransformSection
         end_frame=64
         start_frame=8
         channels [
          Translation.X {type=MovieSceneScriptingFloatChannel, default=0}
          Translation.Y {type=MovieSceneScriptingFloatChannel, keys [[8,-40],[12,0],[60,0],[64,40]]}
         ]
        }]
      }
      {
       type=MovieSceneByteTrack
       display_name=Visibility
       sections [{type=MovieSceneByteSection, end_frame=0, start_frame=0, channels [None {type=MovieSceneScriptingByteChannel, const [4,1]}]}]
      }
     ]
    }
   ]}
 }
 Anim_ShowLevelIcon_2 {
  binding_count=14
  length_seconds=5.0167
  object=&self:Anim_ShowLevelIcon_2
  sequence {
   bindings [
    Image_InwardParticles {
     id=#10
     possessed_class=Image
     tracks [
      {
       type=MovieSceneFloatTrack
       display_name=Render Opacity
       sections [{type=MovieSceneFloatSection, channels [None {type=MovieSceneScriptingFloatChannel, keys [[0,0],[36,1],[66,1],[89.229,0]]}]}]
      }
      {
       type=MovieScene2DTransformTrack
       sections [{
         type=MovieScene2DTransformSection
         channels [
          Scale.X {type=MovieSceneScriptingFloatChannel, keys [[0,1],[33,2],[105,0]]}
          Scale.Y {type=MovieSceneScriptingFloatChannel, keys [[0,1],[33,2],[105,0]]}
         ]
        }]
      }
     ]
    }
    Image_TrianglesLeft_Small {
     id=#11
     possessed_class=Image
     tracks [{
       type=MovieSceneFloatTrack
       display_name=Render Opacity
       sections [{type=MovieSceneFloatSection, channels [None {type=MovieSceneScriptingFloatChannel, keys [[51,0],[57,1]]}]}]
      }]
    }
    Overlay_LeveledUp {
     id=#12
     possessed_class=Overlay
     tracks [{
       type=MovieSceneByteTrack
       display_name=Visibility
       sections [{type=MovieSceneByteSection, end_frame=0, start_frame=0, channels [None {type=MovieSceneScriptingByteChannel, const [4,1]}]}]
      }]
    }
    Image_Glow {
     id=#13
     possessed_class=Image
     tracks [{
       type=MovieSceneFloatTrack
       display_name=Render Opacity
       sections [{type=MovieSceneFloatSection, channels [None {type=MovieSceneScriptingFloatChannel, keys [[0,0],[45,1],[180,0.25]]}]}]
      }]
    }
    Image_Bloom {
     id=#14
     possessed_class=Image
     tracks [
      {
       type=MovieSceneFloatTrack
       display_name=Render Opacity
       sections [{type=MovieSceneFloatSection, channels [None {type=MovieSceneScriptingFloatChannel, keys [[45,0],[54,1],[300,0.025]]}]}]
      }
      {
       type=MovieSceneWidgetMaterialTrack
       sections [{
         type=MovieSceneParameterSection
         channels [
          Star Size {type=MovieSceneScriptingFloatChannel, keys [[45,0],[54,1.5],[240,1.125]]}
          Inner Glow Size {type=MovieSceneScriptingFloatChannel, keys [[54,0.3333],[300,0]]}
         ]
        }]
      }
     ]
    }
    LargeTriangles {
     id=#15
     possessed_class=HorizontalBox
     tracks [
      {
       type=MovieScene2DTransformTrack
       sections [{
         type=MovieScene2DTransformSection
         channels [
          Scale.X {type=MovieSceneScriptingFloatChannel, keys [[51,0],[120,1]]}
          Scale.Y {type=MovieSceneScriptingFloatChannel, keys [[51,0.5],[120,1]]}
         ]
        }]
      }
      {
       type=MovieSceneFloatTrack
       display_name=Render Opacity
       sections [{type=MovieSceneFloatSection, channels [None {type=MovieSceneScriptingFloatChannel, keys [[51,0],[57,1]]}]}]
      }
     ]
    }
    Overlay_IconEffects {
     id=#16
     possessed_class=Overlay
     tracks [{
       type=MovieSceneByteTrack
       display_name=Visibility
       sections [{type=MovieSceneByteSection, end_frame=0, start_frame=0, channels [None {type=MovieSceneScriptingByteChannel, const [4,1]}]}]
      }]
    }
    Image_LeveledUpBackground {
     id=#17
     possessed_class=Image
     tracks [
      {
       type=MovieSceneFloatTrack
       display_name=Render Opacity
       sections [{type=MovieSceneFloatSection, channels [None {type=MovieSceneScriptingFloatChannel, keys [[61,0],[64,1]]}]}]
      }
      {
       type=MovieSceneWidgetMaterialTrack
       sections [{
         type=MovieSceneParameterSection
         end_frame=300
         start_frame=0
         channels [Glow Opacity {type=MovieSceneScriptingFloatChannel, keys [[61,0],[64,1],[226,0]]}]
        }]
      }
      {
       type=MovieScene2DTransformTrack
       sections [{
         type=MovieScene2DTransformSection
         channels [
          Translation.X {type=MovieSceneScriptingFloatChannel, const [0,1]}
          Translation.Y {type=MovieSceneScriptingFloatChannel, const [0,1]}
          Angle {type=MovieSceneScriptingFloatChannel, const [0,1]}
          Scale.X {type=MovieSceneScriptingFloatChannel, keys [[61,0.5],[241,1.25]]}
          Scale.Y {type=MovieSceneScriptingFloatChannel, keys [[61,0.75],[241,1]]}
          Shear.X {type=MovieSceneScriptingFloatChannel, const [0,1]}
          Shear.Y {type=MovieSceneScriptingFloatChannel, const [0,1]}
         ]
        }]
      }
     ]
    }
    Image_TrianglesRight_Small {
     id=#18
     possessed_class=Image
     tracks [{
       type=MovieSceneFloatTrack
       display_name=Render Opacity
       sections [{type=MovieSceneFloatSection, channels [None {type=MovieSceneScriptingFloatChannel, keys [[51,0],[57,1]]}]}]
      }]
    }
    Image_LevelUpIcon {
     id=#19
     possessed_class=Image
     tracks [
      {
       type=MovieSceneFloatTrack
       display_name=Render Opacity
       sections [{type=MovieSceneFloatSection, channels [None {type=MovieSceneScriptingFloatChannel, keys [[0,0],[18,1]]}]}]
      }
      {
       type=MovieSceneWidgetMaterialTrack
       sections [{
         type=MovieSceneParameterSection
         channels [
          Anim_Silhouette {type=MovieSceneScriptingFloatChannel, keys [[0,1],[300,0]]}
          Anim_ColorIntensity {type=MovieSceneScriptingFloatChannel, keys [[54,1],[105,0]]}
          Anim_ArrowPosition {type=MovieSceneScriptingFloatChannel, keys [[60,0],[70,1.05],[120,1]]}
         ]
        }]
      }
      {
       type=MovieScene2DTransformTrack
       sections [{
         type=MovieScene2DTransformSection
         channels [
          Translation.X {type=MovieSceneScriptingFloatChannel, const [0,1]}
          Translation.Y {type=MovieSceneScriptingFloatChannel, const [0,1]}
          Scale.X {type=MovieSceneScriptingFloatChannel, keys [[0,0],[48,1],[54,1.15],[90,1]]}
          Scale.Y {type=MovieSceneScriptingFloatChannel, keys [[0,0],[48,1],[54,1.15],[90,1]]}
         ]
        }]
      }
     ]
    }
    Flares {
     id=#20
     possessed_class=HorizontalBox
     tracks [
      {
       type=MovieSceneFloatTrack
       display_name=Render Opacity
       sections [{type=MovieSceneFloatSection, channels [None {type=MovieSceneScriptingFloatChannel, keys [[48,0],[54,1],[240,0.25]]}]}]
      }
      {
       type=MovieScene2DTransformTrack
       sections [{
         type=MovieScene2DTransformSection
         channels [
          Scale.X {type=MovieSceneScriptingFloatChannel, keys [[48,0],[117,1]]}
          Scale.Y {type=MovieSceneScriptingFloatChannel, keys [[48,0.5],[117,1]]}
         ]
        }]
      }
     ]
    }
    Image_Pulse {
     id=#21
     possessed_class=Image
     tracks [
      {
       type=MovieSceneFloatTrack
       display_name=Render Opacity
       sections [{type=MovieSceneFloatSection, channels [None {type=MovieSceneScriptingFloatChannel, keys [[42,0],[60,1],[174,0]]}]}]
      }
      {
       type=MovieSceneWidgetMaterialTrack
       sections [{type=MovieSceneParameterSection, channels [Inner Radial Manual {type=MovieSceneScriptingFloatChannel, keys [[42,0],[174,1.75]]}]}]
      }
     ]
    }
    Image_BGDarken {
     id=#22
     possessed_class=Image
     tracks [
      {
       type=MovieSceneFloatTrack
       display_name=Render Opacity
       sections [{type=MovieSceneFloatSection, channels [None {type=MovieSceneScriptingFloatChannel, keys [[0,0],[60,1]]}]}]
      }
      {
       type=MovieSceneWidgetMaterialTrack
       sections [{type=MovieSceneParameterSection, channels [Anim_TileBrightness {type=MovieSceneScriptingFloatChannel, keys [[0,-1],[60,1],[301,-0.5]]}]}]
      }
     ]
    }
    TextBlock_LevelUp {
     id=#23
     possessed_class=CommonTextBlock
     tracks [
      {
       type=MovieSceneFloatTrack
       display_name=Render Opacity
       sections [{type=MovieSceneFloatSection, channels [None {type=MovieSceneScriptingFloatChannel, keys [[67,0],[85,1]]}]}]
      }
      {
       type=MovieSceneWidgetMaterialTrack
       display_name=Font.FontMaterial
       sections [{type=MovieSceneParameterSection, channels [ShineOffset {type=MovieSceneScriptingFloatChannel, keys [[67,-1],[285,1]]}]}]
      }
      {
       type=MovieScene2DTransformTrack
       sections [{
         type=MovieScene2DTransformSection
         channels [
          Translation.X {type=MovieSceneScriptingFloatChannel, const [0,1]}
          Translation.Y {type=MovieSceneScriptingFloatChannel, keys [[67,32],[85,0]]}
          Scale.X {type=MovieSceneScriptingFloatChannel, keys [[67,0.5],[85,1]]}
          Scale.Y {type=MovieSceneScriptingFloatChannel, keys [[67,0.5],[85,1]]}
         ]
        }]
      }
      {
       type=MovieSceneWidgetMaterialTrack
       display_name=Font.Outline Settings.OutlineMaterial
       sections [{
         type=MovieSceneParameterSection
         end_frame=300
         start_frame=0
         channels [
          ColorR {type=MovieSceneScriptingFloatChannel, const [0.4538,1]}
          ColorG {type=MovieSceneScriptingFloatChannel, const [0.4615,1]}
          ColorB {type=MovieSceneScriptingFloatChannel, const [1,1]}
          ColorA {type=MovieSceneScriptingFloatChannel, keys [[59,0],[67,0.2],[282,0.05]]}
         ]
        }]
      }
     ]
    }
   ]
   display_rate {numerator=60}
  }
 }
]
graphs [
 EventGraph {nodes [
   Event PreConstruct()
   > SetupDummyText!(
    execute
    self
    Dummy:Text Object Reference=Get TextBlock_DummyLeveledUp() -> (TextBlock_DummyLeveledUp:Text Block Object Reference)
    Real:Text Object Reference=Get TextBlock_LevelUp() -> (TextBlock_LevelUp:Common Text Object Reference)
   ) -> (then)
   Event Construct()
   > ShowLevelUp_VIZ_Part1()
  ]}
 OnVisualizationFinished {nodes [OnVisualizationFinished() {is=title K2Node_FunctionEntry}]}
 SetNumberOfLevels {nodes [
   n1 = SetNumberOfLevels!() -> (then, NumberOfLevels:int) {is=title K2Node_FunctionEntry}
   > Set NumberOfLevels(NumberOfLevels:int=@n1) -> (Output_Get:int=0)
   n2 = Get NumberOfLevels() -> (NumberOfLevels:int=0)
   n3 = Format Text!(Format=Get Text_LevelUpQuantityFormat() -> (Text_LevelUpQuantityFormat:Text), num:Text="To Text (Integer)"(Value=@n2)) -> (Result)
   n4 = "integer > integer"(A=@n2, B=1)
   > "SetText (Text)"(
    self=Get TextBlock_LevelUp()
    InText=Format Text!(Format=Get Text_LevelUp() -> (Text_LevelUp:Text), multiplier:Text=Select(Option 0:Text, Option 1:Text=@n3, Index=@n4) -> (ReturnValue:Text)) -> (Result)
   )
   > SetupDummyText(Dummy=Get TextBlock_DummyLeveledUp(), Real=Get TextBlock_LevelUp())
  ]}
 SetupDummyText {nodes [
   n1 = SetupDummyText!() -> (then, Dummy:Text Object Reference, Real:Text Object Reference) {is=title K2Node_FunctionEntry}
   > "SetText (Text)"(self=@n1, InText="GetText (Text)"!(self:Text Object Reference=@n1.Real) -> (ReturnValue:Text))
   n2 = Break Slate Font Info!(SlateFontInfo:"Slate Font Info Structure (by ref)"=Get Font!(self:Text Object Reference=Get Dummy() -> (Dummy:Text Object Reference)) -> (Font:Slate Font Info Structure)) -> (FontMaterial:Object Reference, OutlineSettings:Font Outline Settings Structure, bMaterialIsStencil:bool=~) {is=title K2Node_BreakStruct}
   n3 = Break Font Outline Settings!(FontOutlineSettings:"Font Outline Settings Structure (by ref)"=@n2.OutlineSettings) -> (
    OutlineSize:int
    bMiteredCorners:bool=~
    bSeparateFillAlpha:bool=~
    bApplyOutlineToDropShadows:bool=~
    OutlineMaterial:Object Reference
    OutlineColor:Linear Color Structure
   ) {is=title K2Node_BreakStruct}
   n4 = Break Slate Font Info!(SlateFontInfo=Get Font(self=Get Real() -> (Real:Text Object Reference))) -> (
    FontObject:Object Reference
    TypefaceFontName:Name
    Size:float=~
    LetterSpacing:int
    SkewAmount:float=~
    bForceMonospaced:bool=~
    MonospacedWidth:float=~
   )
   > SetFont!(
    execute
    self=Get Dummy()
    InFontInfo_FontObject:Object Reference=@n4
    InFontInfo_FontMaterial:Object Reference=@n2
    InFontInfo_OutlineSettings_OutlineSize:int=@n3
    InFontInfo_OutlineSettings_bMiteredCorners:bool=@n3.bMiteredCorners
    InFontInfo_OutlineSettings_bSeparateFillAlpha:bool=@n3.bSeparateFillAlpha
    InFontInfo_OutlineSettings_bApplyOutlineToDropShadows:bool=@n3.bApplyOutlineToDropShadows
    InFontInfo_OutlineSettings_OutlineMaterial:Object Reference=@n3.OutlineMaterial
    InFontInfo_OutlineSettings_OutlineColor:Linear Color Structure=@n3.OutlineColor
    InFontInfo_TypefaceFontName:Name=@n4.TypefaceFontName
    InFontInfo_Size:float=@n4.Size
    InFontInfo_LetterSpacing:int=@n4.LetterSpacing
    InFontInfo_SkewAmount:float=@n4.SkewAmount
    InFontInfo_bForceMonospaced:bool=@n4.bForceMonospaced
    InFontInfo_bMaterialIsStencil:bool=@n2.bMaterialIsStencil
    InFontInfo_MonospacedWidth:float=@n4.MonospacedWidth
   ) -> (then)
  ]}
 Shutdown {nodes [
   Shutdown() {is=title K2Node_FunctionEntry}
   > StopAllAnimations(self:User Widget Object Reference)
   > StopSounds()
   > Call On Visualization Finished!(execute:Exec, self:WBP Post Match Level Up Screen Object Reference) -> (then:Exec) {is=title K2Node_CallDelegate}
   > RemoveFromParent()
  ]}
 StopSounds {nodes [
   StopSounds() {is=title K2Node_FunctionEntry}
   > Set AllowSound(AllowSound:bool=false) -> (Output_Get:bool)
   n1 = Get Sound_LeveledUp_Intro_Component() -> (Sound_LeveledUp_Intro_Component:Audio Component Object Reference)
   > Is Valid(InputObject=@n1)
   > FadeOut!(
    execute
    self:Audio Component Object Reference=@n1
    FadeOutDuration:float=0.2
    FadeVolumeLevel:float=0
    FadeCurve:EAudioFaderCurve Enum=Linear
   ) -> (then)
  ]}
 Visualizations {nodes [
   ShowLevelUp_VIZ_Part3() {is=title K2Node_CustomEvent}
   n1 = Get Anim_Show() -> (Anim_Show:Widget Animation Object Reference)
   > PlayAnimation(InAnimation=@n1, PlayMode=Reverse)
   > BindToAnimationFinished(Animation=@n1, Delegate=Create Event())
   > Branch(Condition=Get AllowSound() -> (AllowSound:bool))
   > n2 = SpawnSound2D!(
    execute
    Sound:Sound Base Object Reference=Get Sound_LeveledUp_Outro() -> (Sound_LeveledUp_Outro:Sound Base Object Reference)
    VolumeMultiplier:float=1
    PitchMultiplier:float=1
    StartTime:float=0
    ConcurrencySettings:Sound Concurrency Object Reference
    bPersistAcrossLevelTransition:bool=false
    bAutoDestroy:bool=true
   ) -> (then, ReturnValue:Audio Component Object Reference)
   > Set Sound_LeveldUp_Outro_Component(Sound_LeveldUp_Outro_Component:Audio Component Object Reference=@n2) -> (Output_Get:Audio Component Object Reference)
   ShowLevelUp_VIZ_Part1() {is=title K2Node_CustomEvent}
   n3 = Get Anim_Show()
   > PlayAnimation(InAnimation=@n3)
   > BindToAnimationFinished(Animation=@n3, Delegate=Create Event())
   > Branch(Condition=Get AllowSound())
   > n4 = SpawnSound2D(
    Sound=Get Sound_LeveledUp_Intro() -> (Sound_LeveledUp_Intro:Sound Base Object Reference)
    VolumeMultiplier=1
    PitchMultiplier=1
    StartTime=0
    bPersistAcrossLevelTransition=false
    bAutoDestroy=true
   )
   > Set Sound_LeveledUp_Intro_Component(Sound_LeveledUp_Intro_Component:Audio Component Object Reference=@n4) -> (Output_Get:Audio Component Object Reference)
   ShowLevelUp_VIZ_Part2() {is=title K2Node_CustomEvent}
   n5 = Get Anim_ShowLevelIcon_2() -> (Anim_ShowLevelIcon_2:Widget Animation Object Reference)
   > PlayAnimation(InAnimation=@n5)
   > Delay(Duration=float - float(A=GetEndTime(self=@n5), B=0.2))
   > ShowLevelUp_VIZ_Part3() {is=title K2Node_CallFunction}
  ]}
]
variables [
 AllowSound {type=boolean, default=true}
 NumberOfLevels {type=integer, default=1}
 OnVisualizationFinished {}
 Sound_LeveldUp_Outro_Component {type=AudioComponent, default=null}
 Sound_LeveledUp_Intro {type=SoundBase, default=/Game/Sounds/Fort_UI_Sounds/Gameplay/XP/MSS_PostMatch_LevelUp_In}
 Sound_LeveledUp_Intro_Component {type=AudioComponent, default=null}
 Sound_LeveledUp_Outro {type=SoundBase, default=/Game/Sounds/Fort_UI_Sounds/Gameplay/XP/MSS_PostMatch_LevelUp_Out}
 Text_LevelUp {type=string, default="Leveled Up{multiplier}"}
 Text_LevelUpQuantityFormat {type=string, default=" x{num}"}
]
widgets {
 info {inheritedWidgetCount=0, parentClass=&UserWidget, rootWidgetClass=&Overlay, widgetCount=24}
 widgets [
  Overlay_0 {widgetClassPath=&Overlay, widget=Overlay_0}
  Image_BGDarken {
   widgetClassPath=&Image
   parent=Overlay_0
   widget=Image_BGDarken
   properties {
    renderOpacity=0
    visibility=SelfHitTestInvisible
    brush {
     bIsDynamicallyLoaded=false
     drawAs=Image
     imageType=FullColor
     mirroring=NoMirror
     resourceName=None
     resourceObject=/FortniteCoreUI/HUD/EndGame/PostGameFlowScreens/DefaultScreens/XPScreen/Materials/MI_PostMatch_Background_LevelUp
     tiling=NoTile
     imageSize=<32 32>
     margin=<ltrb 0 0 0 0>
     outlineSettings {
      bUseBrushTransparency=false
      roundingType=HalfHeightRadius
      width=0
      color {colorUseRule=UseColor_Specified, specifiedColor=<rgba 0 0 0 0>}
      cornerRadii=<0 0 0 0>
     }
     tintColor {colorUseRule=UseColor_Specified, specifiedColor=<rgba 1 1 1 1>}
     uVRegion {bIsValid=false, max=<0 0>, min=<0 0>}
    }
   }
   slot {class=OverlaySlot, properties {horizontalAlignment=HAlign_Fill, verticalAlignment=VAlign_Fill}}
  }
  Overlay_IconEffects {
   widgetClassPath=&Overlay
   parent=Overlay_0
   widget=Overlay_IconEffects
   slot {class=OverlaySlot, properties {horizontalAlignment=HAlign_Center, verticalAlignment=VAlign_Center}}
  }
  Image_Bloom {
   widgetClassPath=&Image
   parent=Overlay_IconEffects
   widget=Image_Bloom
   properties {
    renderOpacity=0.25
    visibility=SelfHitTestInvisible
    brush {
     bIsDynamicallyLoaded=false
     drawAs=Image
     imageType=FullColor
     mirroring=NoMirror
     resourceName=None
     resourceObject=/FortniteCoreUI/HUD/EndGame/PostGameFlowScreens/DefaultScreens/XPScreen/Materials/MI_PostMatch_LevelUpBloom
     tiling=NoTile
     imageSize=<1600 1024>
     margin=<ltrb 0 0 0 0>
     outlineSettings {
      bUseBrushTransparency=false
      roundingType=HalfHeightRadius
      width=0
      color {colorUseRule=UseColor_Specified, specifiedColor=<rgba 0 0 0 0>}
      cornerRadii=<0 0 0 0>
     }
     tintColor {colorUseRule=UseColor_Specified, specifiedColor=<rgba 1 1 1 1>}
     uVRegion {bIsValid=false, max=<0 0>, min=<0 0>}
    }
   }
   slot {class=OverlaySlot, properties {horizontalAlignment=HAlign_Center, verticalAlignment=VAlign_Fill}}
  }
  SmallTriangles {
   widgetClassPath=&HorizontalBox
   parent=Overlay_IconEffects
   widget=SmallTriangles
   properties {renderTransform {angle=0, scale=<1 -1>, shear=<0 0>, translation=<0 0>}}
   slot {class=OverlaySlot, properties {horizontalAlignment=HAlign_Center, verticalAlignment=VAlign_Center}}
  }
  Image_TrianglesLeft_Small {
   widgetClassPath=&Image
   parent=SmallTriangles
   widget=Image_TrianglesLeft_Small
   properties {
    visibility=SelfHitTestInvisible
    brush {
     bIsDynamicallyLoaded=false
     drawAs=Image
     imageType=FullColor
     mirroring=NoMirror
     resourceName=None
     resourceObject=/FortniteCoreUI/HUD/EndGame/PostGameFlowScreens/DefaultScreens/XPScreen/Materials/MI_PostMatch_Triangles_Small
     tiling=NoTile
     imageSize=<320 320>
     margin=<ltrb 0 0 0 0>
     outlineSettings {
      bUseBrushTransparency=false
      roundingType=HalfHeightRadius
      width=0
      color {colorUseRule=UseColor_Specified, specifiedColor=<rgba 0 0 0 0>}
      cornerRadii=<0 0 0 0>
     }
     tintColor {colorUseRule=UseColor_Specified, specifiedColor=<rgba 1 1 1 1>}
     uVRegion {bIsValid=false, max=<0 0>, min=<0 0>}
    }
   }
   slot {class=HorizontalBoxSlot}
  }
  Spacer_1 {widgetClassPath=&Spacer, parent=SmallTriangles, widget=Spacer_1, properties {size=<64 1>}, slot {class=HorizontalBoxSlot}}
  Image_TrianglesRight_Small {
   widgetClassPath=&Image
   parent=SmallTriangles
   widget=Image_TrianglesRight_Small
   properties {
    visibility=SelfHitTestInvisible
    brush {
     bIsDynamicallyLoaded=false
     drawAs=Image
     imageType=FullColor
     mirroring=NoMirror
     resourceName=None
     resourceObject=&MI_PostMatch_Triangles_Small
     tiling=NoTile
     imageSize=<320 320>
     margin=<ltrb 0 0 0 0>
     outlineSettings {
      bUseBrushTransparency=false
      roundingType=HalfHeightRadius
      width=0
      color {colorUseRule=UseColor_Specified, specifiedColor=<rgba 0 0 0 0>}
      cornerRadii=<0 0 0 0>
     }
     tintColor {colorUseRule=UseColor_Specified, specifiedColor=<rgba 1 1 1 1>}
     uVRegion {bIsValid=false, max=<0 0>, min=<0 0>}
    }
    renderTransform {angle=180, scale=<1 1>, shear=<0 0>, translation=<0 0>}
   }
   slot {class=HorizontalBoxSlot, properties {horizontalAlignment=HAlign_Center, verticalAlignment=VAlign_Center}}
  }
  LargeTriangles {
   widgetClassPath=&HorizontalBox
   parent=Overlay_IconEffects
   widget=LargeTriangles
   slot {class=OverlaySlot, properties {horizontalAlignment=HAlign_Center, verticalAlignment=VAlign_Center}}
  }
  Image_TrianglesLeft_Large {
   widgetClassPath=&Image
   parent=LargeTriangles
   widget=Image_TrianglesLeft_Large
   properties {
    visibility=SelfHitTestInvisible
    brush {
     bIsDynamicallyLoaded=false
     drawAs=Image
     imageType=FullColor
     mirroring=NoMirror
     resourceName=None
     resourceObject=/FortniteCoreUI/HUD/EndGame/PostGameFlowScreens/DefaultScreens/XPScreen/Materials/MI_PostMatch_Triangles
     tiling=NoTile
     imageSize=<512 320>
     margin=<ltrb 0 0 0 0>
     outlineSettings {
      bUseBrushTransparency=false
      roundingType=HalfHeightRadius
      width=0
      color {colorUseRule=UseColor_Specified, specifiedColor=<rgba 0 0 0 0>}
      cornerRadii=<0 0 0 0>
     }
     tintColor {colorUseRule=UseColor_Specified, specifiedColor=<rgba 0.212 0.212 0.212 1>}
     uVRegion {bIsValid=false, max=<0 0>, min=<0 0>}
    }
   }
   slot {class=HorizontalBoxSlot}
  }
  Spacer_50 {widgetClassPath=&Spacer, parent=LargeTriangles, widget=Spacer_50, properties {size=<64 1>}, slot {class=HorizontalBoxSlot}}
  Image_TrianglesRight_Large {
   widgetClassPath=&Image
   parent=LargeTriangles
   widget=Image_TrianglesRight_Large
   properties {
    visibility=SelfHitTestInvisible
    brush {
     bIsDynamicallyLoaded=false
     drawAs=Image
     imageType=FullColor
     mirroring=NoMirror
     resourceName=None
     resourceObject=/FortniteCoreUI/HUD/EndGame/PostGameFlowScreens/DefaultScreens/XPScreen/Materials/MI_PostMatch_Triangles_Alternate
     tiling=NoTile
     imageSize=<512 320>
     margin=<ltrb 0 0 0 0>
     outlineSettings {
      bUseBrushTransparency=false
      roundingType=HalfHeightRadius
      width=0
      color {colorUseRule=UseColor_Specified, specifiedColor=<rgba 0 0 0 0>}
      cornerRadii=<0 0 0 0>
     }
     tintColor {colorUseRule=UseColor_Specified, specifiedColor=<rgba 0.212 0.212 0.212 1>}
     uVRegion {bIsValid=false, max=<0 0>, min=<0 0>}
    }
    renderTransform {angle=180, scale=<1 1>, shear=<0 0>, translation=<0 0>}
   }
   slot {class=HorizontalBoxSlot, properties {horizontalAlignment=HAlign_Center, verticalAlignment=VAlign_Center}}
  }
  Flares {
   widgetClassPath=&HorizontalBox
   parent=Overlay_IconEffects
   widget=Flares
   slot {class=OverlaySlot, properties {horizontalAlignment=HAlign_Center, verticalAlignment=VAlign_Center}}
  }
  Image_Flare_L {
   widgetClassPath=&Image
   parent=Flares
   widget=Image_Flare_L
   properties {
    visibility=SelfHitTestInvisible
    brush {
     bIsDynamicallyLoaded=false
     drawAs=Image
     imageType=FullColor
     mirroring=NoMirror
     resourceName=None
     resourceObject=/FortniteCoreUI/HUD/EndGame/PostGameFlowScreens/DefaultScreens/XPScreen/Materials/MI_UI_PostMatch_Flare_LevelUp
     tiling=NoTile
     imageSize=<832 320>
     margin=<ltrb 0 0 0 0>
     outlineSettings {
      bUseBrushTransparency=false
      roundingType=HalfHeightRadius
      width=0
      color {colorUseRule=UseColor_Specified, specifiedColor=<rgba 0 0 0 0>}
      cornerRadii=<0 0 0 0>
     }
     tintColor {colorUseRule=UseColor_Specified, specifiedColor=<rgba 1 1 1 1>}
     uVRegion {bIsValid=false, max=<0 0>, min=<0 0>}
    }
   }
   slot {class=HorizontalBoxSlot, properties {verticalAlignment=VAlign_Center}}
  }
  Spacer {widgetClassPath=&Spacer, parent=Flares, widget=Spacer, properties {size=<192 1>}, slot {class=HorizontalBoxSlot}}
  Image_Flare_R {
   widgetClassPath=&Image
   parent=Flares
   widget=Image_Flare_R
   properties {
    visibility=SelfHitTestInvisible
    brush {
     bIsDynamicallyLoaded=false
     drawAs=Image
     imageType=FullColor
     mirroring=NoMirror
     resourceName=None
     resourceObject=&MI_UI_PostMatch_Flare_LevelUp
     tiling=NoTile
     imageSize=<832 320>
     margin=<ltrb 0 0 0 0>
     outlineSettings {
      bUseBrushTransparency=false
      roundingType=HalfHeightRadius
      width=0
      color {colorUseRule=UseColor_Specified, specifiedColor=<rgba 0 0 0 0>}
      cornerRadii=<0 0 0 0>
     }
     tintColor {colorUseRule=UseColor_Specified, specifiedColor=<rgba 1 1 1 1>}
     uVRegion {bIsValid=false, max=<0 0>, min=<0 0>}
    }
    renderTransform {angle=180, scale=<1 1>, shear=<0 0>, translation=<0 0>}
   }
   slot {class=HorizontalBoxSlot, properties {horizontalAlignment=HAlign_Center, verticalAlignment=VAlign_Center}}
  }
  Image_Glow {
   widgetClassPath=&Image
   parent=Overlay_IconEffects
   widget=Image_Glow
   properties {
    renderOpacity=0.2
    visibility=SelfHitTestInvisible
    brush {
     bIsDynamicallyLoaded=false
     drawAs=Image
     imageType=FullColor
     mirroring=NoMirror
     resourceName=None
     resourceObject=/FortniteCoreUI/HUD/EndGame/PostGameFlowScreens/DefaultScreens/XPScreen/Materials/M_UI_PostMatchTextGlow
     tiling=NoTile
     imageSize=<1024 1024>
     margin=<ltrb 0 0 0 0>
     outlineSettings {
      bUseBrushTransparency=false
      roundingType=HalfHeightRadius
      width=0
      color {colorUseRule=UseColor_Specified, specifiedColor=<rgba 0 0 0 0>}
      cornerRadii=<0 0 0 0>
     }
     tintColor {colorUseRule=UseColor_Specified, specifiedColor=<rgba 1 1 1 1>}
     uVRegion {bIsValid=false, max=<0 0>, min=<0 0>}
    }
    renderTransform {angle=0, scale=<1.25 1.25>, shear=<0 0>, translation=<0 0>}
   }
   slot {class=OverlaySlot, properties {horizontalAlignment=HAlign_Center, verticalAlignment=VAlign_Center}}
  }
  Image_Pulse {
   widgetClassPath=&Image
   parent=Overlay_IconEffects
   widget=Image_Pulse
   properties {
    renderOpacity=0.2
    visibility=SelfHitTestInvisible
    brush {
     bIsDynamicallyLoaded=false
     drawAs=Image
     imageType=FullColor
     mirroring=NoMirror
     resourceName=None
     resourceObject=/FortniteCoreUI/HUD/EndGame/PostGameFlowScreens/DefaultScreens/XPScreen/Materials/MI_PostMatch_LevelUp_PulseGlow
     tiling=NoTile
     imageSize=<1024 1024>
     margin=<ltrb 0 0 0 0>
     outlineSettings {
      bUseBrushTransparency=false
      roundingType=HalfHeightRadius
      width=0
      color {colorUseRule=UseColor_Specified, specifiedColor=<rgba 0 0 0 0>}
      cornerRadii=<0 0 0 0>
     }
     tintColor {colorUseRule=UseColor_Specified, specifiedColor=<rgba 1 1 1 1>}
     uVRegion {bIsValid=false, max=<0 0>, min=<0 0>}
    }
   }
   slot {class=OverlaySlot, properties {horizontalAlignment=HAlign_Center, verticalAlignment=VAlign_Center}}
  }
  Image_InwardParticles {
   widgetClassPath=&Image
   parent=Overlay_IconEffects
   widget=Image_InwardParticles
   properties {
    renderOpacity=0.2
    visibility=SelfHitTestInvisible
    brush {
     bIsDynamicallyLoaded=false
     drawAs=Image
     imageType=FullColor
     mirroring=NoMirror
     resourceName=None
     resourceObject=/Game/UI/DiscoveryShared/Matchmaking/MatchFoundAnimation/MaterialInstances/M_UI_Starburst_MatchFoundAnim
     tiling=NoTile
     imageSize=<1024 1024>
     margin=<ltrb 0 0 0 0>
     outlineSettings {
      bUseBrushTransparency=false
      roundingType=HalfHeightRadius
      width=0
      color {colorUseRule=UseColor_Specified, specifiedColor=<rgba 0 0 0 0>}
      cornerRadii=<0 0 0 0>
     }
     tintColor {colorUseRule=UseColor_Specified, specifiedColor=<rgba 1 1 1 1>}
     uVRegion {bIsValid=false, max=<0 0>, min=<0 0>}
    }
   }
   slot {class=OverlaySlot, properties {horizontalAlignment=HAlign_Center, verticalAlignment=VAlign_Center}}
  }
  Image_LevelUpIcon {
   widgetClassPath=&Image
   parent=Overlay_IconEffects
   widget=Image_LevelUpIcon
   properties {
    visibility=SelfHitTestInvisible
    brush {
     bIsDynamicallyLoaded=false
     drawAs=Image
     imageType=FullColor
     mirroring=NoMirror
     resourceName=None
     resourceObject=/FortniteCoreUI/HUD/EndGame/PostGameFlowScreens/DefaultScreens/XPScreen/Materials/M_UI_LevelUP_Badge
     tiling=NoTile
     imageSize=<320 320>
     margin=<ltrb 0 0 0 0>
     outlineSettings {
      bUseBrushTransparency=false
      roundingType=HalfHeightRadius
      width=0
      color {colorUseRule=UseColor_Specified, specifiedColor=<rgba 0 0 0 0>}
      cornerRadii=<0 0 0 0>
     }
     tintColor {colorUseRule=UseColor_Specified, specifiedColor=<rgba 1 1 1 1>}
     uVRegion {bIsValid=false, max=<0 0>, min=<0 0>}
    }
   }
   slot {class=OverlaySlot, properties {horizontalAlignment=HAlign_Center, verticalAlignment=VAlign_Center}}
  }
  Overlay_LeveledUp {
   widgetClassPath=&Overlay
   parent=Overlay_0
   widget=Overlay_LeveledUp
   slot {class=OverlaySlot, properties {horizontalAlignment=HAlign_Fill, verticalAlignment=VAlign_Center, padding=<ltrb 0 240 0 0>}}
  }
  Image_LeveledUpBackground {
   widgetClassPath=&Image
   parent=Overlay_LeveledUp
   widget=Image_LeveledUpBackground
   properties {
    visibility=SelfHitTestInvisible
    brush {
     bIsDynamicallyLoaded=false
     drawAs=Image
     imageType=FullColor
     mirroring=NoMirror
     resourceName=None
     resourceObject=/FortniteCoreUI/HUD/EndGame/PostGameFlowScreens/DefaultScreens/XPScreen/Materials/MI_PostMatch_TextBackgroundGlow
     tiling=NoTile
     imageSize=<128 128>
     margin=<ltrb 0 0 0 0>
     outlineSettings {
      bUseBrushTransparency=false
      roundingType=HalfHeightRadius
      width=0
      color {colorUseRule=UseColor_Specified, specifiedColor=<rgba 0 0 0 0>}
      cornerRadii=<0 0 0 0>
     }
     tintColor {colorUseRule=UseColor_Specified, specifiedColor=<rgba 1 1 1 1>}
     uVRegion {bIsValid=false, max=<0 0>, min=<0 0>}
    }
   }
   slot {class=OverlaySlot, properties {horizontalAlignment=HAlign_Fill, verticalAlignment=VAlign_Fill, padding=<ltrb 0 -16 0 -16>}}
  }
  TextBlock_DummyLeveledUp {
   widgetClassPath=&UEFN_TextBlock
   bIsVariable=true
   parent=Overlay_LeveledUp
   widget=TextBlock_DummyLeveledUp
   properties {
    justification=Center
    text=Leveled Up
    textTransformPolicy=ToUpper
    visibility=Collapsed
    colorAndOpacity {colorUseRule=UseColor_Specified, specifiedColor=<rgba 1 1 1 0>}
    font {
     bForceMonospaced=false
     bMaterialIsStencil=false
     fontMaterial=null
     fontObject=&HeadingNow
     letterSpacing=-48
     monospacedWidth=1
     size=72
     skewAmount=0
     typefaceFontName=86 Bold Italic
     outlineSettings {
      bApplyOutlineToDropShadows=false
      bMiteredCorners=false
      bSeparateFillAlpha=false
      outlineMaterial=/FortniteCoreUI/HUD/EndGame/PostGameFlowScreens/DefaultScreens/XPScreen/Materials/MI_PostMatch_FontGlow
      outlineSize=48
      outlineColor=<rgba 1 1 1 1>
     }
    }
   }
   slot {class=OverlaySlot, properties {horizontalAlignment=HAlign_Center, verticalAlignment=VAlign_Center}}
  }
  TextBlock_LevelUp {
   widgetClassPath=&CommonTextBlock
   bIsVariable=true
   parent=Overlay_LeveledUp
   widget=TextBlock_LevelUp
   properties {
    justification=Center
    text=Leveled Up
    textTransformPolicy=ToUpper
    font {
     bForceMonospaced=false
     bMaterialIsStencil=false
     fontMaterial=/FortniteCoreUI/HUD/EndGame/PostGameFlowScreens/DefaultScreens/XPScreen/Materials/M_UI_PostMatch_FontShine
     fontObject=&HeadingNow
     letterSpacing=-48
     monospacedWidth=1
     size=72
     skewAmount=0
     typefaceFontName=86 Bold Italic
     outlineSettings {
      bApplyOutlineToDropShadows=false
      bMiteredCorners=true
      bSeparateFillAlpha=false
      outlineMaterial=&MI_PostMatch_FontGlow
      outlineSize=42
      outlineColor=<rgba 1 1 1 1>
     }
    }
    shadowOffset=<0 0>
   }
   slot {class=OverlaySlot, properties {horizontalAlignment=HAlign_Center, verticalAlignment=VAlign_Center}}
  }
 ]
}
