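# Post-match XP screen row, derived from FortMatchXPTotalEntry, showing one rundown category's name and XP total.
# Display sets text, color and value on entry init or at design time, with IsRight mirroring alignment and burst side.
# Parent events drive side-specific intro slides and an outro fade.
# The first XP animation slides the number in; later ones pop it with font-material shine, flare and triangle bursts, both updating the value.
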WidgetBlueprint /FortniteCoreUI/HUD/EndGame/PostGameFlowScreens/DefaultScreens/XPScreen/WBP_PostMatch_RundownCategoryItem
animations [
 IntroAnimation_Right {
  binding_count=2
  length_seconds=0.4
  object=&self:IntroAnimation_Right
  sequence {
   bindings [
    CText_CategoryName {
     id=#1
     possessed_class=CommonTextBlock
     tracks [
      {
       type=MovieScene2DTransformTrack
       sections [{type=MovieScene2DTransformSection, channels [Translation.X {type=MovieSceneScriptingFloatChannel, keys [[0,-200],[12,0]]}]}]
      }
      {
       type=MovieSceneFloatTrack
       display_name=Render Opacity
       sections [{type=MovieSceneFloatSection, channels [None {type=MovieSceneScriptingFloatChannel, keys [[0,0],[12,1]]}]}]
      }
     ]
    }
    HBox_XPTotal {
     id=#2
     possessed_class=HorizontalBox
     tracks [{
       type=MovieSceneFloatTrack
       display_name=Render Opacity
       sections [{type=MovieSceneFloatSection, channels [None {type=MovieSceneScriptingFloatChannel, const [0,1]}]}]
      }]
    }
   ]
   display_rate {numerator=30}
  }
 }
 IntroAnimation_Left {
  binding_count=2
  length_seconds=0.4
  object=&self:IntroAnimation_Left
  sequence {
   bindings [
    CText_CategoryName {
     id=#1
     possessed_class=CommonTextBlock
     tracks [
      {
       type=MovieScene2DTransformTrack
       sections [{type=MovieScene2DTransformSection, channels [Translation.X {type=MovieSceneScriptingFloatChannel, keys [[0,200],[12,0]]}]}]
      }
      {
       type=MovieSceneFloatTrack
       display_name=Render Opacity
       sections [{type=MovieSceneFloatSection, channels [None {type=MovieSceneScriptingFloatChannel, keys [[0,0],[12,1]]}]}]
      }
     ]
    }
    HBox_XPTotal {
     id=#2
     possessed_class=HorizontalBox
     tracks [{
       type=MovieSceneFloatTrack
       display_name=Render Opacity
       sections [{type=MovieSceneFloatSection, channels [None {type=MovieSceneScriptingFloatChannel, const [0,1]}]}]
      }]
    }
   ]
   display_rate {numerator=30}
  }
 }
 OutroAnimation {
  binding_count=1
  length_seconds=0.75
  object=&self:OutroAnimation
  sequence {bindings [Root {
     id=#3
     possessed_class=Overlay
     tracks [{
       type=MovieSceneFloatTrack
       display_name=Render Opacity
       sections [{type=MovieSceneFloatSection, channels [None {type=MovieSceneScriptingFloatChannel, keys [[0,1],[15,0.9]]}]}]
      }]
    }]}
 }
 NumberIntroAnimation_Right {
  binding_count=1
  length_seconds=0.4
  object=&self:NumberIntroAnimation_Right
  sequence {bindings [HBox_XPTotal {
     id=#4
     possessed_class=HorizontalBox
     tracks [
      {
       type=MovieSceneFloatTrack
       display_name=Render Opacity
       sections [{type=MovieSceneFloatSection, channels [None {type=MovieSceneScriptingFloatChannel, keys [[0,0],[8,1]]}]}]
      }
      {
       type=MovieScene2DTransformTrack
       sections [{type=MovieScene2DTransformSection, channels [Translation.X {type=MovieSceneScriptingFloatChannel, keys [[0,-200],[8,0]]}]}]
      }
     ]
    }]}
 }
 NumberIntroAnimation_Left {
  binding_count=1
  length_seconds=0.4
  object=&self:NumberIntroAnimation_Left
  sequence {bindings [HBox_XPTotal {
     id=#5
     possessed_class=HorizontalBox
     tracks [
      {
       type=MovieSceneFloatTrack
       display_name=Render Opacity
       sections [{type=MovieSceneFloatSection, channels [None {type=MovieSceneScriptingFloatChannel, keys [[0,0],[8,1]]}]}]
      }
      {
       type=MovieScene2DTransformTrack
       sections [{type=MovieScene2DTransformSection, channels [Translation.X {type=MovieSceneScriptingFloatChannel, keys [[0,200],[8,0]]}]}]
      }
     ]
    }]}
 }
 NumberUpdateAnimation_Right {
  binding_count=5
  length_seconds=1
  object=&self:NumberUpdateAnimation_Right
  sequence {bindings [
    CNumText_XPTotal {
     id=#6
     possessed_class=CommonNumericTextBlock
     tracks [{
       type=MovieSceneWidgetMaterialTrack
       display_name=Font.FontMaterial
       sections [{type=MovieSceneParameterSection, channels [ShineOffset {type=MovieSceneScriptingFloatChannel, keys [[0,-1],[10,1]]}]}]
      }]
    }
    Img_RightTris {
     id=#7
     possessed_class=Image
     tracks [
      {
       type=MovieSceneFloatTrack
       display_name=Render Opacity
       sections [{type=MovieSceneFloatSection, channels [None {type=MovieSceneScriptingFloatChannel, keys [[0,0],[1,1],[20,0]]}]}]
      }
      {
       type=MovieScene2DTransformTrack
       sections [{
         type=MovieScene2DTransformSection
         channels [
          Scale.X {type=MovieSceneScriptingFloatChannel, keys [[0,1],[20,1.5]]}
          Scale.Y {type=MovieSceneScriptingFloatChannel, keys [[0,1],[20,1.5]]}
         ]
        }]
      }
      {
       type=MovieSceneWidgetMaterialTrack
       sections [{type=MovieSceneParameterSection, channels [TriScale {type=MovieSceneScriptingFloatChannel, keys [[0,0.5],[20,0]]}]}]
      }
     ]
    }
    HBox_XPTotal {
     id=#8
     possessed_class=HorizontalBox
     tracks [
      {
       type=MovieScene2DTransformTrack
       sections [{
         type=MovieScene2DTransformSection
         channels [
          Translation.X {type=MovieSceneScriptingFloatChannel, default=0}
          Translation.Y {type=MovieSceneScriptingFloatChannel, default=0}
          Angle {type=MovieSceneScriptingFloatChannel, default=0}
          Scale.X {type=MovieSceneScriptingFloatChannel, keys [[0,1],[1,1.25],[20,1]]}
          Scale.Y {type=MovieSceneScriptingFloatChannel, keys [[0,1],[1,1.25],[20,1]]}
          Shear.X {type=MovieSceneScriptingFloatChannel, const [0,1]}
          Shear.Y {type=MovieSceneScriptingFloatChannel, const [0,1]}
         ]
        }]
      }
      {
       type=MovieSceneDoubleVectorTrack
       display_name=Pivot
       sections [{
         type=MovieSceneDoubleVectorSection
         channels [
          Vector.X {type=MovieSceneScriptingDoubleChannel, const [0,2]}
          Vector.Y {type=MovieSceneScriptingDoubleChannel, const [0.75,2]}
         ]
        }]
      }
     ]
    }
    Img_RightFlare {
     id=#9
     possessed_class=Image
     tracks [
      {
       type=MovieSceneFloatTrack
       display_name=Render Opacity
       sections [{type=MovieSceneFloatSection, end_frame=20, start_frame=0, channels [None {type=MovieSceneScriptingFloatChannel, keys [[0,0],[1,1],[20,0]]}]}]
      }
      {
       type=MovieScene2DTransformTrack
       sections [{type=MovieScene2DTransformSection, channels [Scale.Y {type=MovieSceneScriptingFloatChannel, keys [[0,0],[20,1]]}]}]
      }
     ]
    }
    RightFXOverlay {
     id=#10
     possessed_class=Overlay
     tracks [{
       type=MovieSceneByteTrack
       display_name=Visibility
       sections [{type=MovieSceneByteSection, channels [None {type=MovieSceneScriptingByteChannel, keys [[0,4],[20,2]]}]}]
      }]
    }
   ]}
 }
 NumberUpdateAnimation_Left {
  binding_count=5
  length_seconds=1
  object=&self:NumberUpdateAnimation_Left
  sequence {bindings [
    HBox_XPTotal {
     id=#11
     possessed_class=HorizontalBox
     tracks [
      {
       type=MovieScene2DTransformTrack
       sections [{
         type=MovieScene2DTransformSection
         channels [
          Translation.X {type=MovieSceneScriptingFloatChannel, default=0}
          Translation.Y {type=MovieSceneScriptingFloatChannel, default=0}
          Angle {type=MovieSceneScriptingFloatChannel, default=0}
          Scale.X {type=MovieSceneScriptingFloatChannel, keys [[0,1],[1,1.25],[20,1]]}
          Scale.Y {type=MovieSceneScriptingFloatChannel, keys [[0,1],[1,1.25],[20,1]]}
          Shear.X {type=MovieSceneScriptingFloatChannel, const [0,1]}
          Shear.Y {type=MovieSceneScriptingFloatChannel, const [0,1]}
         ]
        }]
      }
      {
       type=MovieSceneDoubleVectorTrack
       display_name=Pivot
       sections [{
         type=MovieSceneDoubleVectorSection
         channels [
          Vector.X {type=MovieSceneScriptingDoubleChannel, const [1,2]}
          Vector.Y {type=MovieSceneScriptingDoubleChannel, const [0.75,2]}
         ]
        }]
      }
     ]
    }
    Img_LeftTris {
     id=#12
     possessed_class=Image
     tracks [
      {
       type=MovieSceneFloatTrack
       display_name=Render Opacity
       sections [{type=MovieSceneFloatSection, channels [None {type=MovieSceneScriptingFloatChannel, keys [[0,0],[1,1],[20,0]]}]}]
      }
      {
       type=MovieScene2DTransformTrack
       sections [{
         type=MovieScene2DTransformSection
         channels [
          Scale.X {type=MovieSceneScriptingFloatChannel, keys [[0,1],[20,1.5]]}
          Scale.Y {type=MovieSceneScriptingFloatChannel, keys [[0,1],[20,1.5]]}
         ]
        }]
      }
      {
       type=MovieSceneWidgetMaterialTrack
       sections [{type=MovieSceneParameterSection, channels [TriScale {type=MovieSceneScriptingFloatChannel, keys [[0,0.5],[20,0]]}]}]
      }
     ]
    }
    LeftFXOverlay {
     id=#13
     possessed_class=Overlay
     tracks [{
       type=MovieSceneByteTrack
       display_name=Visibility
       sections [{type=MovieSceneByteSection, channels [None {type=MovieSceneScriptingByteChannel, keys [[0,4],[20,2]]}]}]
      }]
    }
    CNumText_XPTotal {
     id=#14
     possessed_class=CommonNumericTextBlock
     tracks [{
       type=MovieSceneWidgetMaterialTrack
       display_name=Font.FontMaterial
       sections [{type=MovieSceneParameterSection, channels [ShineOffset {type=MovieSceneScriptingFloatChannel, keys [[0,1],[10,-1]]}]}]
      }]
    }
    Img_LeftFlare {
     id=#15
     possessed_class=Image
     tracks [
      {
       type=MovieSceneFloatTrack
       display_name=Render Opacity
       sections [{type=MovieSceneFloatSection, end_frame=20, start_frame=0, channels [None {type=MovieSceneScriptingFloatChannel, keys [[0,0],[1,1],[20,0]]}]}]
      }
      {
       type=MovieScene2DTransformTrack
       sections [{type=MovieScene2DTransformSection, channels [Scale.Y {type=MovieSceneScriptingFloatChannel, keys [[0,0],[20,1]]}]}]
      }
     ]
    }
   ]}
 }
 NumberUpdateAnimation {
  binding_count=2
  length_seconds=1
  object=&self:NumberUpdateAnimation
  sequence {bindings [
    HBox_XPTotal {
     id=#16
     possessed_class=HorizontalBox
     tracks [{
       type=MovieScene2DTransformTrack
       sections [{
         type=MovieScene2DTransformSection
         channels [
          Translation.X {type=MovieSceneScriptingFloatChannel, default=0}
          Translation.Y {type=MovieSceneScriptingFloatChannel, default=0}
          Angle {type=MovieSceneScriptingFloatChannel, default=0}
          Scale.X {type=MovieSceneScriptingFloatChannel, keys [[0,1],[1,1.25],[20,1]]}
          Scale.Y {type=MovieSceneScriptingFloatChannel, keys [[0,1],[1,1.25],[20,1]]}
          Shear.X {type=MovieSceneScriptingFloatChannel, const [0,1]}
          Shear.Y {type=MovieSceneScriptingFloatChannel, const [0,1]}
         ]
        }]
      }]
    }
    CNumText_XPTotal {
     id=#17
     possessed_class=CommonNumericTextBlock
     tracks [{
       type=MovieSceneWidgetMaterialTrack
       display_name=Font.FontMaterial
       sections [{type=MovieSceneParameterSection, channels [ShineOffset {type=MovieSceneScriptingFloatChannel, keys [[0,-1],[20,1]]}]}]
      }]
    }
   ]}
 }
]
graphs [
 Display {nodes [
   Display() {is=title K2Node_FunctionEntry}
   > SetCurrentValue!(
    execute
    self:Common Numeric Text Block Object Reference=Get CNumText_XPTotal() -> (CNumText_XPTotal:Common Numeric Text Block Object Reference)
    NewValue:float="To Float (Integer)"(InInt=Get XPValue() -> (XPValue:int=0))
   ) -> (then)
   > "SetText (Text)"(self=Get CText_CategoryName() -> (CText_CategoryName:Common Text Object Reference), InText=Get XPType() -> (XPType:Text))
   > SetColorAndOpacity!(
    execute
    self=Get CText_CategoryName()
    InColorAndOpacity_SpecifiedColor:Linear Color Structure=Get RundownCategoryColor() -> (RundownCategoryColor:Linear Color Structure)
    InColorAndOpacity_ColorUseRule:ESlateColorStylingMode Enum
   ) -> (then)
   > SetHorizontalAlignment(
    self=[
     SlotAsOverlaySlot(Widget=Get HBox_XPTotal() -> (HBox_XPTotal:Horizontal Box Object Reference))
     SlotAsOverlaySlot(Widget=Get CText_CategoryName())
    ]
    InHorizontalAlignment=Select(Option 0:EHorizontalAlignment Enum=HAlign_Right, Option 1:EHorizontalAlignment Enum=HAlign_Left, Index=Get IsRight() -> (IsRight:bool)) -> (ReturnValue:EHorizontalAlignment Enum)
   )
   > SetVisibility(
    self=Get RightBurst() -> (RightBurst:Fort Visual Attachment Object Reference)
    InVisibility=Select(Option 0:ESlateVisibility Enum=Collapsed, Option 1:ESlateVisibility Enum=SelfHitTestInvisible, Index=Get IsRight()) -> (ReturnValue:ESlateVisibility Enum)
   )
   > SetVisibility(
    self=Get LeftBurst() -> (LeftBurst:Fort Visual Attachment Object Reference)
    InVisibility=Select(Option 0=SelfHitTestInvisible, Option 1=Collapsed, Index=Get IsRight())
   )
  ]}
 EventGraph {nodes [
   Event OnAnimateXPValue() {is=title K2Node_Event}
   > PlayNumber_VIZ()
   Event Construct()
   > SetRenderOpacity(self=[Get CText_CategoryName(), Get HBox_XPTotal()], InOpacity=0)
   Event OnIntroPlayed() {is=title K2Node_Event}
   > PlayIntro_VIZ()
   n1 = Event PreConstruct()
   > Branch(Condition=@n1.IsDesignTime)
   > Display() {is=title K2Node_CallFunction}
   Event OnEntryInitialized() {is=title K2Node_Event}
   > Display()
   Event OnOutroPlayed() {is=title K2Node_Event}
   > PlayOutro_VIZ()
  ]}
 Vizualization Graph {nodes [
   PlayOutro_VIZ() {is=title K2Node_CustomEvent}
   > StopAllAnimations(self:User Widget Object Reference)
   > PlayAnimation(InAnimation=Get OutroAnimation() -> (OutroAnimation:Widget Animation Object Reference))
   PlayIntro_VIZ() {is=title K2Node_CustomEvent}
   > PlayAnimation(InAnimation=Select(
     Option 0:Widget Animation Object Reference=Get IntroAnimation_Left() -> (IntroAnimation_Left:Widget Animation Object Reference)
     Option 1:Widget Animation Object Reference=Get IntroAnimation_Right() -> (IntroAnimation_Right:Widget Animation Object Reference)
     Index=Get IsRight()
    ) -> (ReturnValue:Widget Animation Object Reference))
   PlayNumber_VIZ() {is=title K2Node_CustomEvent}
   > n1 = Branch(Condition=NOT Boolean(A=Get bHasPlayedIntro() -> (bHasPlayedIntro:bool)))
   > Set bHasPlayedIntro(bHasPlayedIntro:bool=true) -> (Output_Get:bool)
   > n2 = PlayAnimation(InAnimation=Select(
     Option 0=Get NumberIntroAnimation_Left() -> (NumberIntroAnimation_Left:Widget Animation Object Reference)
     Option 1=Get NumberIntroAnimation_Right() -> (NumberIntroAnimation_Right:Widget Animation Object Reference)
     Index=Get IsRight()
    ))
   SetCurrentValue(execute=[@n2.then, @n3.then], self=Get CNumText_XPTotal(), NewValue="To Float (Integer)"(InInt=Get XPValue() -> (XPValue=0)))
   n3 = PlayAnimation(
    execute=@n1.else
    InAnimation=Select(
     Option 0=Get NumberUpdateAnimation_Left() -> (NumberUpdateAnimation_Left:Widget Animation Object Reference)
     Option 1=Get NumberUpdateAnimation_Right() -> (NumberUpdateAnimation_Right:Widget Animation Object Reference)
     Index=Get IsRight()
    )
   )
   Get NumberUpdateAnimation() -> (NumberUpdateAnimation:Widget Animation Object Reference)
  ]}
]
variables [
 HasPlayedNumberIntro {type=boolean}
 IsRight {type=boolean}
 RundownCategoryColor {type=LinearColor, default=<rgba 0 0.8714 0.9823 1>}
]
widgets {
 info {inheritedWidgetCount=1, parentClass=/Script/FortniteUI.FortMatchXPTotalEntry, rootWidgetClass=&Overlay, widgetCount=14}
 widgets [
  Root {widgetClassPath=&Overlay, widget=Root}
  LeftBurst {
   widgetClassPath=&FortVisualAttachment
   bIsVariable=true
   parent=Root
   widget=LeftBurst
   properties {visibility=Hidden, contentAnchor=<1 0.5>, renderTransform {angle=0, scale=<1 1>, shear=<-16 0>, translation=<0 0>}}
   slot {class=OverlaySlot, properties {horizontalAlignment=HAlign_Right, verticalAlignment=VAlign_Center}}
  }
  LeftFXOverlay {widgetClassPath=&Overlay, parent=LeftBurst, widget=LeftFXOverlay, properties {visibility=Hidden}, slot {class=SizeBoxSlot}}
  Img_LeftFlare {
   widgetClassPath=&Image
   parent=LeftFXOverlay
   widget=Img_LeftFlare
   properties {
    visibility=SelfHitTestInvisible
    brush {
     bIsDynamicallyLoaded=false
     drawAs=Image
     imageType=FullColor
     mirroring=NoMirror
     resourceName=None
     resourceObject=/FortniteCoreUI/HUD/EndGame/PostGameFlowScreens/DefaultScreens/XPScreen/Materials/MI_UI_PostMatch_Flare_LevelUp
     tiling=NoTile
     imageSize=<1152 128>
     margin=<ltrb 0 0 0 0>
     outlineSettings {
      bUseBrushTransparency=false
      roundingType=HalfHeightRadius
      width=0
      color {colorUseRule=UseColor_Specified, specifiedColor=<rgba 0 0 0 0>}
      cornerRadii=<0 0 0 0>
     }
     tintColor {colorUseRule=UseColor_Specified, specifiedColor=<rgba 1 1 1 1>}
     uVRegion {bIsValid=false, max=<0 0>, min=<0 0>}
    }
   }
   slot {class=OverlaySlot}
  }
  Img_LeftTris {
   widgetClassPath=&Image
   parent=LeftFXOverlay
   widget=Img_LeftTris
   properties {
    visibility=SelfHitTestInvisible
    brush {
     bIsDynamicallyLoaded=false
     drawAs=Image
     imageType=FullColor
     mirroring=NoMirror
     resourceName=None
     resourceObject=/FortniteCoreUI/HUD/EndGame/PostGameFlowScreens/DefaultScreens/XPScreen/Materials/MI_PostMatch_Triangles
     tiling=NoTile
     imageSize=<512 96>
     margin=<ltrb 0 0 0 0>
     outlineSettings {
      bUseBrushTransparency=false
      roundingType=HalfHeightRadius
      width=0
      color {colorUseRule=UseColor_Specified, specifiedColor=<rgba 0 0 0 0>}
      cornerRadii=<0 0 0 0>
     }
     tintColor {colorUseRule=UseColor_Specified, specifiedColor=<rgba 1 1 1 1>}
     uVRegion {bIsValid=false, max=<0 0>, min=<0 0>}
    }
   }
   slot {class=OverlaySlot, properties {horizontalAlignment=HAlign_Right, verticalAlignment=VAlign_Center}}
  }
  RightBurst {
   widgetClassPath=&FortVisualAttachment
   bIsVariable=true
   parent=Root
   widget=RightBurst
   properties {visibility=Hidden, contentAnchor=<0 0.5>, renderTransform {angle=0, scale=<1 1>, shear=<-16 0>, translation=<0 0>}}
   slot {class=OverlaySlot, properties {verticalAlignment=VAlign_Center}}
  }
  RightFXOverlay {widgetClassPath=&Overlay, parent=RightBurst, widget=RightFXOverlay, properties {visibility=Hidden}, slot {class=SizeBoxSlot}}
  Img_RightFlare {
   widgetClassPath=&Image
   parent=RightFXOverlay
   widget=Img_RightFlare
   properties {
    visibility=SelfHitTestInvisible
    brush {
     bIsDynamicallyLoaded=false
     drawAs=Image
     imageType=FullColor
     mirroring=NoMirror
     resourceName=None
     resourceObject=&MI_UI_PostMatch_Flare_LevelUp
     tiling=NoTile
     imageSize=<1152 128>
     margin=<ltrb 0 0 0 0>
     outlineSettings {
      bUseBrushTransparency=false
      roundingType=HalfHeightRadius
      width=0
      color {colorUseRule=UseColor_Specified, specifiedColor=<rgba 0 0 0 0>}
      cornerRadii=<0 0 0 0>
     }
     tintColor {colorUseRule=UseColor_Specified, specifiedColor=<rgba 1 1 1 1>}
     uVRegion {bIsValid=false, max=<0 0>, min=<0 0>}
    }
    renderTransform {angle=0, scale=<-1 1>, shear=<0 0>, translation=<0 0>}
   }
   slot {class=OverlaySlot}
  }
  Img_RightTris {
   widgetClassPath=&Image
   parent=RightFXOverlay
   widget=Img_RightTris
   properties {
    visibility=SelfHitTestInvisible
    brush {
     bIsDynamicallyLoaded=false
     drawAs=Image
     imageType=FullColor
     mirroring=NoMirror
     resourceName=None
     resourceObject=&MI_PostMatch_Triangles
     tiling=NoTile
     imageSize=<512 96>
     margin=<ltrb 0 0 0 0>
     outlineSettings {
      bUseBrushTransparency=false
      roundingType=HalfHeightRadius
      width=0
      color {colorUseRule=UseColor_Specified, specifiedColor=<rgba 0 0 0 0>}
      cornerRadii=<0 0 0 0>
     }
     tintColor {colorUseRule=UseColor_Specified, specifiedColor=<rgba 1 1 1 1>}
     uVRegion {bIsValid=false, max=<0 0>, min=<0 0>}
    }
    renderTransform {angle=180, scale=<1 1>, shear=<0 0>, translation=<0 0>}
   }
   slot {class=OverlaySlot, properties {verticalAlignment=VAlign_Center}}
  }
  CText_CategoryName {
   widgetClassPath=&CommonTextBlock
   bIsVariable=true
   parent=Root
   widget=CText_CategoryName
   properties {
    text=Quests
    textTransformPolicy=ToUpper
    colorAndOpacity {colorUseRule=UseColor_Specified, specifiedColor=<rgba 0 0.8714 0.9823 0.9>}
    font {
     bForceMonospaced=false
     bMaterialIsStencil=false
     fontMaterial=null
     fontObject=&HeadingNow
     letterSpacing=10
     monospacedWidth=1
     size=18
     skewAmount=0
     typefaceFontName=86 Bold
     outlineSettings {
      bApplyOutlineToDropShadows=false
      bMiteredCorners=false
      bSeparateFillAlpha=false
      outlineMaterial=null
      outlineSize=0
      outlineColor=<rgba 0 0 0 1>
     }
    }
    shadowOffset=<0 0>
   }
   slot {class=OverlaySlot, properties {horizontalAlignment=HAlign_Center}}
  }
  HBox_XPTotal {
   widgetClassPath=&HorizontalBox
   bIsVariable=true
   parent=Root
   widget=HBox_XPTotal
   slot {class=OverlaySlot, properties {horizontalAlignment=HAlign_Center, padding=<ltrb 0 16 0 0>}}
  }
  CNumText_XPTotal {
   widgetClassPath=/Script/CommonUI.CommonNumericTextBlock
   bIsVariable=true
   parent=HBox_XPTotal
   widget=CNumText_XPTotal
   properties {
    text=0
    font {
     bForceMonospaced=false
     bMaterialIsStencil=false
     fontMaterial=/FortniteCoreUI/HUD/EndGame/PostGameFlowScreens/DefaultScreens/XPScreen/Materials/M_UI_PostMatch_FontShine
     fontObject=&Burbank
     letterSpacing=0
     monospacedWidth=1
     size=60
     skewAmount=0.2
     typefaceFontName=BurbankRegular-Black
     outlineSettings {
      bApplyOutlineToDropShadows=false
      bMiteredCorners=false
      bSeparateFillAlpha=true
      outlineMaterial=/FortniteCoreUI/HUD/EndGame/PostGameFlowScreens/DefaultScreens/XPScreen/Materials/MI_PostMatch_FontGlow
      outlineSize=36
      outlineColor=<rgba 1 1 1 1>
     }
    }
    margin=<ltrb -40 -24 -24 -24>
    shadowOffset=<0 0>
   }
   slot {class=HorizontalBoxSlot, properties {horizontalAlignment=HAlign_Left, verticalAlignment=VAlign_Top}}
  }
  Text_XPLabel {
   widgetClassPath=&CommonTextBlock
   parent=HBox_XPTotal
   widget=Text_XPLabel
   properties {
    text=XP
    font {
     bForceMonospaced=false
     bMaterialIsStencil=false
     fontMaterial=&M_UI_PostMatch_FontShine
     fontObject=&HeadingNow
     letterSpacing=0
     monospacedWidth=1
     size=30
     skewAmount=0
     typefaceFontName=76 Bold Italic
     outlineSettings {
      bApplyOutlineToDropShadows=false
      bMiteredCorners=false
      bSeparateFillAlpha=false
      outlineMaterial=&MI_PostMatch_FontGlow
      outlineSize=16
      outlineColor=<rgba 1 1 1 1>
     }
    }
    margin=<ltrb -12 0 -12 0>
    shadowOffset=<0 0>
   }
   slot {class=HorizontalBoxSlot, properties {horizontalAlignment=HAlign_Left, verticalAlignment=VAlign_Center, padding=<ltrb 0 0 0 12>}}
  }
  TutorialHighlight {widgetClassPath=&TutorialHighlight}
 ]
}
