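# Post-match XP screen rundown row, based on FortMatchXPTotalEntry, showing a check icon, XP amount and rich-text description.
# Construct collapses and hides the row; Display fills the text on entry initialization or when constructed with positive XP.
# OnIntroPlayed randomizes the particle angle, then plays IntroAnimation_2 (expansion, check burst, glows, sliding text) with sound.
# A sequencer event broadcasts OnIntroFinished with category and amount; separate events dim or hide the row.
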WidgetBlueprint /FortniteCoreUI/HUD/EndGame/PostGameFlowScreens/DefaultScreens/XPScreen/WBP_PostMatch_RundownItem
animations [
 IntroAnimation {
  binding_count=4
  length_seconds=1.5
  object=&self:IntroAnimation
  sequence {
   bindings [
    RootSizeBox {
     id=#1
     possessed_class=SizeBox
     tracks [
      {
       type=MovieSceneFloatTrack
       display_name=Height Override
       sections [{
         type=MovieSceneFloatSection
         end_frame=20
         start_frame=0
         channels [None {type=MovieSceneScriptingFloatChannel, keys [[0,0],[4,50],[20,50]]}]
        }]
      }
      {
       type=MovieSceneFloatTrack
       display_name=Render Opacity
       sections [{type=MovieSceneFloatSection, end_frame=45, start_frame=0, channels [None {type=MovieSceneScriptingFloatChannel, keys [[0,0],[7,1],[45,1]]}]}]
      }
     ]
    }
    Overlay_Root {
     id=#2
     possessed_class=Overlay
     tracks [
      {
       type=MovieScene2DTransformTrack
       sections [{
         type=MovieScene2DTransformSection
         end_frame=20
         start_frame=0
         channels [Translation.X {type=MovieSceneScriptingFloatChannel, keys [[0,-100],[20,0]]}]
        }]
      }
      {
       type=MovieSceneFloatTrack
       display_name=Render Opacity
       sections [{
         type=MovieSceneFloatSection
         end_frame=45
         start_frame=0
         channels [None {type=MovieSceneScriptingFloatChannel, keys [[0,0],[4,0],[6,0.1],[20,1],[30,1],[45,0.5]]}]
        }]
      }
     ]
    }
    Img_Background_Glow_Small {
     id=#3
     possessed_class=Image
     tracks [
      {
       type=MovieScene2DTransformTrack
       sections [{
         type=MovieScene2DTransformSection
         end_frame=20
         start_frame=0
         channels [
          Translation.X {type=MovieSceneScriptingFloatChannel, keys [[0,0],[13,-300],[17,0],[20,0]]}
          Translation.Y {type=MovieSceneScriptingFloatChannel, const [0,1]}
          Angle {type=MovieSceneScriptingFloatChannel, const [0,1]}
          Scale.X {type=MovieSceneScriptingFloatChannel, keys [[0,0],[13,0],[17,1],[20,1]]}
          Scale.Y {type=MovieSceneScriptingFloatChannel, const [1,1]}
          Shear.X {type=MovieSceneScriptingFloatChannel, const [0,1]}
          Shear.Y {type=MovieSceneScriptingFloatChannel, const [0,1]}
         ]
        }]
      }
      {
       type=MovieSceneDoubleVectorTrack
       display_name=Pivot
       sections [{
         type=MovieSceneDoubleVectorSection
         end_frame=20
         start_frame=0
         channels [
          Vector.X {type=MovieSceneScriptingDoubleChannel, keys [[0,1],[20,0]]}
          Vector.Y {type=MovieSceneScriptingDoubleChannel, const [0.5,1]}
         ]
        }]
      }
      {
       type=MovieSceneFloatTrack
       display_name=Render Opacity
       sections [{type=MovieSceneFloatSection, end_frame=45, start_frame=0, channels [None {type=MovieSceneScriptingFloatChannel, keys [[0,1],[20,1],[45,0]]}]}]
      }
     ]
    }
    Img_CheckIcon {
     id=#4
     possessed_class=Image
     tracks [
      {
       type=MovieSceneWidgetMaterialTrack
       sections [{
         type=MovieSceneParameterSection
         end_frame=20
         start_frame=0
         channels [
          MSDF Scale {type=MovieSceneScriptingFloatChannel, keys [[0,0.4966],[13,0.5],[15,1.2],[16.9995,1],[20,1]]}
          Brightness Percentage {type=MovieSceneScriptingFloatChannel, const [1,5]}
         ]
        }]
      }
      {
       type=MovieSceneFloatTrack
       display_name=Render Opacity
       sections [{
         type=MovieSceneFloatSection
         end_frame=20
         start_frame=0
         channels [None {type=MovieSceneScriptingFloatChannel, keys [[0,0],[10,0],[13,1],[17,1]]}]
        }]
      }
     ]
    }
   ]
   display_rate {numerator=30}
  }
 }
 IntroAnimation_2 {
  binding_count=11
  length_seconds=1.5
  object=&self:IntroAnimation_2
  sequence {
   bindings [
    RootSizeBox {
     id=#1
     possessed_class=SizeBox
     tracks [
      {
       type=MovieSceneFloatTrack
       display_name=Height Override
       sections [{type=MovieSceneFloatSection, channels [None {type=MovieSceneScriptingFloatChannel, keys [[0,0],[8,50],[40,50]]}]}]
      }
      {
       type=MovieSceneFloatTrack
       display_name=Render Opacity
       sections [{type=MovieSceneFloatSection, channels [None {type=MovieSceneScriptingFloatChannel, keys [[0,0],[2,1]]}]}]
      }
     ]
    }
    Text_XPLabel {
     id=#5
     possessed_class=CommonTextBlock
     tracks [
      {
       type=MovieScene2DTransformTrack
       sections [{type=MovieScene2DTransformSection, channels [Translation.X {type=MovieSceneScriptingFloatChannel, keys [[22,-80],[51,0]]}]}]
      }
      {
       type=MovieSceneFloatTrack
       display_name=Render Opacity
       sections [{type=MovieSceneFloatSection, channels [None {type=MovieSceneScriptingFloatChannel, keys [[22,0],[28,1]]}]}]
      }
     ]
    }
    Overlay_Root {
     id=#2
     possessed_class=Overlay
     tracks [{
       type=MovieSceneFloatTrack
       display_name=Render Opacity
       sections [{type=MovieSceneFloatSection, channels [None {type=MovieSceneScriptingFloatChannel, keys [[0,0],[3,1],[60,1]]}]}]
      }]
    }
    Img_Background_Glow_Small {
     id=#3
     possessed_class=Image
     tracks [
      {
       type=MovieScene2DTransformTrack
       sections [{
         type=MovieScene2DTransformSection
         channels [
          Translation.X {type=MovieSceneScriptingFloatChannel, keys [[26,-256],[34,0],[40,0]]}
          Translation.Y {type=MovieSceneScriptingFloatChannel, const [0,1]}
          Scale.X {type=MovieSceneScriptingFloatChannel, keys [[26,0],[60,1]]}
          Scale.Y {type=MovieSceneScriptingFloatChannel, const [1,1]}
         ]
        }]
      }
      {
       type=MovieSceneDoubleVectorTrack
       display_name=Pivot
       sections [{
         type=MovieSceneDoubleVectorSection
         channels [
          Vector.X {type=MovieSceneScriptingDoubleChannel, keys [[0,1],[40,0]]}
          Vector.Y {type=MovieSceneScriptingDoubleChannel, const [0.5,1]}
         ]
        }]
      }
      {
       type=MovieSceneFloatTrack
       display_name=Render Opacity
       sections [{type=MovieSceneFloatSection, channels [None {type=MovieSceneScriptingFloatChannel, keys [[26,0],[30,1],[90,0]]}]}]
      }
     ]
    }
    Img_CheckIcon_Burst {
     id=#6
     possessed_class=Image
     tracks [
      {
       type=MovieSceneFloatTrack
       display_name=Render Opacity
       sections [{type=MovieSceneFloatSection, channels [None {type=MovieSceneScriptingFloatChannel, keys [[25,0],[28,1],[64,0]]}]}]
      }
      {
       type=MovieScene2DTransformTrack
       sections [{
         type=MovieScene2DTransformSection
         channels [
          Scale.X {type=MovieSceneScriptingFloatChannel, keys [[28,0.5],[64,10]]}
          Scale.Y {type=MovieSceneScriptingFloatChannel, keys [[28,0.5],[64,10]]}
         ]
        }]
      }
      {
       type=MovieSceneWidgetMaterialTrack
       sections [{
         type=MovieSceneParameterSection
         channels [
          Smoothing {type=MovieSceneScriptingFloatChannel, keys [[28,0],[45,1]]}
          Brightness Percentage {type=MovieSceneScriptingFloatChannel, const [0.25,1]}
          Trans_Focused {type=MovieSceneScriptingFloatChannel, keys [[25,1],[73,0]]}
         ]
        }]
      }
     ]
    }
    Rich_DescriptionText {
     id=#7
     possessed_class=RichTextBlock
     tracks [
      {
       type=MovieSceneFloatTrack
       display_name=Render Opacity
       sections [{type=MovieSceneFloatSection, channels [None {type=MovieSceneScriptingFloatChannel, keys [[30,0],[45,1]]}]}]
      }
      {
       type=MovieScene2DTransformTrack
       sections [{type=MovieScene2DTransformSection, channels [Translation.X {type=MovieSceneScriptingFloatChannel, keys [[30,-80],[66,0]]}]}]
      }
     ]
    }
    Text_XPTotal {
     id=#8
     possessed_class=CommonTextBlock
     tracks [
      {
       type=MovieScene2DTransformTrack
       sections [{type=MovieScene2DTransformSection, channels [Translation.X {type=MovieSceneScriptingFloatChannel, keys [[18,-80],[45,0]]}]}]
      }
      {
       type=MovieSceneFloatTrack
       display_name=Render Opacity
       sections [{type=MovieSceneFloatSection, channels [None {type=MovieSceneScriptingFloatChannel, keys [[18,0],[24,1]]}]}]
      }
     ]
    }
    Img_CheckIcon_Glow {
     id=#9
     possessed_class=Image
     tracks [
      {
       type=MovieScene2DTransformTrack
       sections [{
         type=MovieScene2DTransformSection
         channels [
          Scale.X {type=MovieSceneScriptingFloatChannel, keys [[18,1],[57,8]]}
          Scale.Y {type=MovieSceneScriptingFloatChannel, keys [[18,1],[57,8]]}
         ]
        }]
      }
      {
       type=MovieSceneFloatTrack
       display_name=Render Opacity
       sections [{type=MovieSceneFloatSection, channels [None {type=MovieSceneScriptingFloatChannel, keys [[15,0],[18,1],[33,1],[57,0]]}]}]
      }
     ]
    }
    Img_Background_Glow_Large {
     id=#10
     possessed_class=Image
     tracks [
      {
       type=MovieSceneFloatTrack
       display_name=Render Opacity
       sections [{type=MovieSceneFloatSection, channels [None {type=MovieSceneScriptingFloatChannel, keys [[24,0],[41,1],[90,0]]}]}]
      }
      {
       type=MovieScene2DTransformTrack
       sections [{
         type=MovieScene2DTransformSection
         channels [
          Translation.X {type=MovieSceneScriptingFloatChannel, keys [[24,-512],[90,512]]}
          Translation.Y {type=MovieSceneScriptingFloatChannel, const [0,1]}
          Scale.X {type=MovieSceneScriptingFloatChannel, const [1,1]}
          Scale.Y {type=MovieSceneScriptingFloatChannel, keys [[24,0.5],[30,1],[90,0.5]]}
         ]
        }]
      }
     ]
    }
    Img_CheckIcon {
     id=#4
     possessed_class=Image
     tracks [
      {
       type=MovieSceneWidgetMaterialTrack
       sections [{
         type=MovieSceneParameterSection
         channels [
          Brightness Percentage {type=MovieSceneScriptingFloatChannel, keys [[12,2],[18,2],[60,1]]}
          Trans_Hovered {type=MovieSceneScriptingFloatChannel, keys [[21,0],[23,1],[30,1],[60,0]]}
          Smoothing {type=MovieSceneScriptingFloatChannel, keys [[12,0.5],[24,0.075]]}
          Trans_Focused {type=MovieSceneScriptingFloatChannel, keys [[9,1],[27,0]]}
         ]
        }]
      }
      {
       type=MovieSceneFloatTrack
       display_name=Render Opacity
       sections [{type=MovieSceneFloatSection, channels [None {type=MovieSceneScriptingFloatChannel, thinned_from=4, keys [[12,0],[15,1],[46,1]]}]}]
      }
      {
       type=MovieScene2DTransformTrack
       sections [{
         type=MovieScene2DTransformSection
         channels [
          Translation.X {type=MovieSceneScriptingFloatChannel, default=0}
          Translation.Y {type=MovieSceneScriptingFloatChannel, default=0, const [0,1]}
          Angle {type=MovieSceneScriptingFloatChannel, default=0}
          Scale.X {type=MovieSceneScriptingFloatChannel, keys [[12,16],[24,0.5],[33,2],[57,1]]}
          Scale.Y {type=MovieSceneScriptingFloatChannel, keys [[12,16],[24,0.5],[33,2],[57,1]]}
          Shear.X {type=MovieSceneScriptingFloatChannel, default=0}
          Shear.Y {type=MovieSceneScriptingFloatChannel, default=0}
         ]
        }]
      }
     ]
    }
    Img_CheckIcon_TriangleParticles {
     id=#11
     possessed_class=Image
     tracks [
      {
       type=MovieSceneFloatTrack
       display_name=Render Opacity
       sections [{type=MovieSceneFloatSection, channels [None {type=MovieSceneScriptingFloatChannel, keys [[15,0],[18,1],[30,1],[79,0]]}]}]
      }
      {
       type=MovieScene2DTransformTrack
       sections [{
         type=MovieScene2DTransformSection
         channels [
          Scale.X {type=MovieSceneScriptingFloatChannel, keys [[18,0.5],[79,10]]}
          Scale.Y {type=MovieSceneScriptingFloatChannel, keys [[18,0.5],[79,10]]}
         ]
        }]
      }
      {
       type=MovieSceneWidgetMaterialTrack
       sections [{type=MovieSceneParameterSection, channels [TriScale {type=MovieSceneScriptingFloatChannel, keys [[22,0.3],[79,0]]}]}]
      }
     ]
    }
   ]
   display_rate {numerator=60}
   tracks [{
     type=MovieSceneEventTrack
     sections [{
       type=MovieSceneEventTriggerSection
       channels [None {type=MovieSceneScriptingEventChannel, const ["<Struct 'MovieSceneEvent' (0x467612fbea0) {}>", 1]}]
      }]
    }]
  }
 }
]
graphs [
 Display {nodes [
   Display() {is=title K2Node_FunctionEntry}
   > "SetText (Text)"(self=Get Text_XPTotal() -> (Text_XPTotal:Common Text Object Reference), InText="To Text (Integer)"(Value=Get XPValue() -> (XPValue:int=0)))
   > SetText(self=Get Rich_DescriptionText() -> (Rich_DescriptionText:Rich Text Block Object Reference), InText=Get XPType() -> (XPType:Text))
  ]}
 EventGraph {nodes [
   Event OnIntroPlayed() {is=title K2Node_Event}
   > PlayIntro_VIZ()
   Event Construct()
   > SetHeightOverride(self=Get RootSizeBox() -> (RootSizeBox:Size Box Object Reference), InHeightOverride=0)
   > SetRenderOpacity(self=Get RootSizeBox(), InOpacity=0)
   > Branch(Condition="integer > integer"(A=Get XPValue() -> (XPValue=0)))
   > Display() {is=title K2Node_CallFunction}
   Event OnEntryInitialized() {is=title K2Node_Event}
   > Display()
  ]}
 OnIntroFinished {nodes [OnIntroFinished!() -> (then, XPType:EFortExperienceCategory Enum=None, Amount:int) {is=title K2Node_FunctionEntry}]}
 Sequencer Events {nodes [
   TriggerIntroFinishedEvent() {is=title K2Node_CustomEvent}
   > Call On Intro Finished!(
    execute:Exec
    self:WBP Post Match Rundown Item Object Reference
    XPType:EFortExperienceCategory Enum=Get XPCategory() -> (XPCategory:EFortExperienceCategory Enum=None)
    Amount:int=Get XPValue() -> (XPValue=0)
   ) -> (then:Exec) {is=title K2Node_CallDelegate}
  ]}
 Vizualization Graph {nodes [
   PlayIntro_VIZ() {is=title K2Node_CustomEvent}
   > SetRenderTransformAngle!(
    execute
    self:Widget Object Reference=Get Img_CheckIcon_TriangleParticles() -> (Img_CheckIcon_TriangleParticles:Image Object Reference)
    Angle:float=RandomFloatInRange(Min=0, Max=90)
   ) -> (then)
   > PlayAnimation(InAnimation=Get IntroAnimation_2() -> (IntroAnimation_2:Widget Animation Object Reference))
   > PlaySound2D(Sound=Get LineItemAppear SFX() -> (LineItemAppear SFX:Sound Base Object Reference))
   PlayDim_VIZ() {is=title K2Node_CustomEvent}
   > SetRenderOpacity(self=Get Overlay_Root() -> (Overlay_Root:Overlay Object Reference), InOpacity=0.5)
   PlayHide_VIZ() {is=title K2Node_CustomEvent}
   > SetRenderOpacity(self=Get RootSizeBox(), InOpacity=0)
  ]}
]
variables [
 LineItemAppear SFX {type=SoundBase, default=/Game/Sounds/Fort_UI_Sounds/Gameplay/XP/MSS_PostMatch_XPEntry_LineItemAppear}
 OnIntroFinished {}
 XPCategory {type=EFortExperienceCategory, default=None}
]
widgets {
 info {inheritedWidgetCount=1, parentClass=/Script/FortniteUI.FortMatchXPTotalEntry, rootWidgetClass=&SizeBox, widgetCount=15}
 widgets [
  RootSizeBox {widgetClassPath=&SizeBox, bIsVariable=true, widget=RootSizeBox, properties {bOverride_HeightOverride=true, heightOverride=50}}
  Overlay_Root {widgetClassPath=&Overlay, bIsVariable=true, parent=RootSizeBox, widget=Overlay_Root, slot {class=SizeBoxSlot}}
  Img_Background_Glow_Large {
   widgetClassPath=&Image
   bIsVariable=true
   parent=Overlay_Root
   widget=Img_Background_Glow_Large
   properties {
    renderOpacity=0
    visibility=SelfHitTestInvisible
    brush {
     bIsDynamicallyLoaded=false
     drawAs=Image
     imageType=FullColor
     mirroring=NoMirror
     resourceName=None
     resourceObject=/FortniteCoreUI/HUD/EndGame/PostGameFlowScreens/DefaultScreens/XPScreen/Materials/MI_RundownItemBacking
     tiling=NoTile
     imageSize=<50 50>
     margin=<ltrb 0.5 0.5 0.5 0.5>
     outlineSettings {
      bUseBrushTransparency=false
      roundingType=HalfHeightRadius
      width=0
      color {colorUseRule=UseColor_Specified, specifiedColor=<rgba 0 0 0 0>}
      cornerRadii=<0 0 0 0>
     }
     tintColor {colorUseRule=UseColor_Specified, specifiedColor=<rgba 0.0719 0.0625 0.25 1>}
     uVRegion {bIsValid=false, max=<0 0>, min=<0 0>}
    }
    renderTransform {angle=180, scale=<1 1>, shear=<0 0>, translation=<0 0>}
   }
   slot {class=OverlaySlot, properties {horizontalAlignment=HAlign_Fill, verticalAlignment=VAlign_Fill, padding=<ltrb -256 -64 -256 -64>}}
  }
  Img_Background_Glow_Small {
   widgetClassPath=&Image
   bIsVariable=true
   parent=Overlay_Root
   widget=Img_Background_Glow_Small
   properties {
    visibility=SelfHitTestInvisible
    brush {
     bIsDynamicallyLoaded=false
     drawAs=Box
     imageType=FullColor
     mirroring=NoMirror
     resourceName=None
     resourceObject=&MI_RundownItemBacking
     tiling=NoTile
     imageSize=<50 50>
     margin=<ltrb 0.5 0.5 0.5 0.5>
     outlineSettings {
      bUseBrushTransparency=false
      roundingType=HalfHeightRadius
      width=0
      color {colorUseRule=UseColor_Specified, specifiedColor=<rgba 0 0 0 0>}
      cornerRadii=<0 0 0 0>
     }
     tintColor {colorUseRule=UseColor_Specified, specifiedColor=<rgba 1 1 1 1>}
     uVRegion {bIsValid=false, max=<0 0>, min=<0 0>}
    }
   }
   slot {class=OverlaySlot, properties {horizontalAlignment=HAlign_Fill, verticalAlignment=VAlign_Fill, padding=<ltrb -64 0 -64 0>}}
  }
  HorizontalBox_43 {
   widgetClassPath=&HorizontalBox
   parent=Overlay_Root
   widget=HorizontalBox_43
   slot {class=OverlaySlot, properties {verticalAlignment=VAlign_Center}}
  }
  Overlay_0 {
   widgetClassPath=&Overlay
   parent=HorizontalBox_43
   widget=Overlay_0
   slot {class=HorizontalBoxSlot, properties {verticalAlignment=VAlign_Center, padding=<ltrb 0 0 16 0>}}
  }
  Img_CheckIcon_TriangleParticles {
   widgetClassPath=&Image
   bIsVariable=true
   parent=Overlay_0
   widget=Img_CheckIcon_TriangleParticles
   properties {
    renderOpacity=0
    visibility=SelfHitTestInvisible
    brush {
     bIsDynamicallyLoaded=false
     drawAs=Image
     imageType=FullColor
     mirroring=NoMirror
     resourceName=None
     resourceObject=/FortniteCoreUI/HUD/EndGame/PostGameFlowScreens/DefaultScreens/XPScreen/Materials/MI_PostMatch_Triangles_Slow
     tiling=NoTile
     imageSize=<32 32>
     margin=<ltrb 0 0 0 0>
     outlineSettings {
      bUseBrushTransparency=false
      roundingType=HalfHeightRadius
      width=0
      color {colorUseRule=UseColor_Specified, specifiedColor=<rgba 0 0 0 0>}
      cornerRadii=<0 0 0 0>
     }
     tintColor {colorUseRule=UseColor_Specified, specifiedColor=<rgba 1 1 1 1>}
     uVRegion {bIsValid=false, max=<0 0>, min=<0 0>}
    }
    renderTransform {angle=0, scale=<9.6521 9.6521>, shear=<0 0>, translation=<0 0>}
   }
   slot {class=OverlaySlot, properties {horizontalAlignment=HAlign_Center, verticalAlignment=VAlign_Center}}
  }
  Img_CheckIcon_Burst {
   widgetClassPath=&Image
   bIsVariable=true
   parent=Overlay_0
   widget=Img_CheckIcon_Burst
   properties {
    renderOpacity=0
    visibility=SelfHitTestInvisible
    brush {
     bIsDynamicallyLoaded=false
     drawAs=Image
     imageType=FullColor
     mirroring=NoMirror
     resourceName=None
     resourceObject=/FortniteCoreUI/HUD/EndGame/PostGameFlowScreens/DefaultScreens/XPScreen/Materials/MI_PostMatch_Icon_Check
     tiling=NoTile
     imageSize=<32 32>
     margin=<ltrb 0 0 0 0>
     outlineSettings {
      bUseBrushTransparency=false
      roundingType=HalfHeightRadius
      width=0
      color {colorUseRule=UseColor_Specified, specifiedColor=<rgba 0 0 0 0>}
      cornerRadii=<0 0 0 0>
     }
     tintColor {colorUseRule=UseColor_Specified, specifiedColor=<rgba 1 1 1 1>}
     uVRegion {bIsValid=false, max=<0 0>, min=<0 0>}
    }
   }
   slot {class=OverlaySlot, properties {horizontalAlignment=HAlign_Center, verticalAlignment=VAlign_Center}}
  }
  Img_CheckIcon {
   widgetClassPath=&Image
   bIsVariable=true
   parent=Overlay_0
   widget=Img_CheckIcon
   properties {
    visibility=SelfHitTestInvisible
    brush {
     bIsDynamicallyLoaded=false
     drawAs=Image
     imageType=FullColor
     mirroring=NoMirror
     resourceName=None
     resourceObject=&MI_PostMatch_Icon_Check
     tiling=NoTile
     imageSize=<40 40>
     margin=<ltrb 0 0 0 0>
     outlineSettings {
      bUseBrushTransparency=false
      roundingType=HalfHeightRadius
      width=0
      color {colorUseRule=UseColor_Specified, specifiedColor=<rgba 0 0 0 0>}
      cornerRadii=<0 0 0 0>
     }
     tintColor {colorUseRule=UseColor_Specified, specifiedColor=<rgba 1 1 1 1>}
     uVRegion {bIsValid=false, max=<0 0>, min=<0 0>}
    }
   }
   slot {class=OverlaySlot}
  }
  Img_CheckIcon_Glow {
   widgetClassPath=&Image
   bIsVariable=true
   parent=Overlay_0
   widget=Img_CheckIcon_Glow
   properties {
    renderOpacity=0
    visibility=SelfHitTestInvisible
    brush {
     bIsDynamicallyLoaded=false
     drawAs=Image
     imageType=FullColor
     mirroring=NoMirror
     resourceName=None
     resourceObject=/FortniteCoreUI/HUD/EndGame/PostGameFlowScreens/DefaultScreens/XPScreen/Materials/MI_PostMatch_Bloom_Small
     tiling=NoTile
     imageSize=<32 32>
     margin=<ltrb 0 0 0 0>
     outlineSettings {
      bUseBrushTransparency=false
      roundingType=HalfHeightRadius
      width=0
      color {colorUseRule=UseColor_Specified, specifiedColor=<rgba 0 0 0 0>}
      cornerRadii=<0 0 0 0>
     }
     tintColor {colorUseRule=UseColor_Specified, specifiedColor=<rgba 1 1 1 1>}
     uVRegion {bIsValid=false, max=<0 0>, min=<0 0>}
    }
   }
   slot {class=OverlaySlot, properties {horizontalAlignment=HAlign_Center, verticalAlignment=VAlign_Center}}
  }
  XPHbox {
   widgetClassPath=&HorizontalBox
   parent=HorizontalBox_43
   widget=XPHbox
   slot {class=HorizontalBoxSlot, properties {horizontalAlignment=HAlign_Left, verticalAlignment=VAlign_Top}}
  }
  Text_XPTotal {
   widgetClassPath=&CommonTextBlock
   bIsVariable=true
   parent=XPHbox
   widget=Text_XPTotal
   properties {
    text=0K
    font {
     bForceMonospaced=false
     bMaterialIsStencil=false
     fontMaterial=null
     fontObject=&HeadingNow
     letterSpacing=0
     monospacedWidth=1
     size=24
     skewAmount=0
     typefaceFontName=76 Bold Italic
     outlineSettings {
      bApplyOutlineToDropShadows=false
      bMiteredCorners=false
      bSeparateFillAlpha=false
      outlineMaterial=/FortniteCoreUI/HUD/EndGame/PostGameFlowScreens/DefaultScreens/XPScreen/Materials/MI_PostMatch_FontGlow
      outlineSize=16
      outlineColor=<rgba 1 1 1 1>
     }
    }
    margin=<ltrb -16 -16 -16 0>
    shadowOffset=<0 0>
   }
   slot {class=HorizontalBoxSlot, properties {padding=<ltrb 0 0 8 0>}}
  }
  Text_XPLabel {
   widgetClassPath=&CommonTextBlock
   parent=XPHbox
   widget=Text_XPLabel
   properties {
    text=XP
    font {
     bForceMonospaced=false
     bMaterialIsStencil=false
     fontMaterial=null
     fontObject=&HeadingNow
     letterSpacing=0
     monospacedWidth=1
     size=12.75
     skewAmount=0
     typefaceFontName=76 Bold Italic
     outlineSettings {
      bApplyOutlineToDropShadows=false
      bMiteredCorners=false
      bSeparateFillAlpha=false
      outlineMaterial=&MI_PostMatch_FontGlow
      outlineSize=8
      outlineColor=<rgba 1 1 1 1>
     }
    }
    margin=<ltrb -10 -8 -10 0>
    shadowOffset=<0 0>
   }
   slot {class=HorizontalBoxSlot, properties {horizontalAlignment=HAlign_Left, verticalAlignment=VAlign_Top, padding=<ltrb 0 5 0 0>}}
  }
  Rich_DescriptionText {
   widgetClassPath=/Script/UMG.RichTextBlock
   bIsVariable=true
   parent=HorizontalBox_43
   widget=Rich_DescriptionText
   properties {
    text=TestDo a very descriptive task that takes a while to describe
    textStyleSet=/Game/Athena/HUD/Discoverability/UrgentQuestRichTextStyles
    decoratorClasses [
     /Game/Athena/HUD/QuestIcons/BP_QuestTile_IconDecorator.BP_QuestTile_IconDecorator_C
     /Game/Athena/HUD/Discoverability/UrgentQuestRichTextDecorators.UrgentQuestRichTextDecorators_C
    ]
   }
   slot {class=HorizontalBoxSlot, properties {horizontalAlignment=HAlign_Center, verticalAlignment=VAlign_Center, padding=<ltrb 16 0 0 0>}}
  }
  TutorialHighlight {widgetClassPath=&TutorialHighlight}
 ]
}
