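# Post-match rundown XP entry widget, derived from FortMatchXPTotalEntry, that shows an XP amount with its category label.
# Display fills both texts from XPValue and XPType on entry initialization, XP animation updates and design-time PreConstruct.
# OnIntroPlayed triggers a scale-pop and fade-in through PlayIntro_VIZ.
# A separate PlayOutro_VIZ event, not called here, stops animations and plays the shrink and half-fade outro.
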WidgetBlueprint /FortniteCoreUI/HUD/EndGame/PostGameFlowScreens/DefaultScreens/XPScreen/WBP_PostMatch_RundownReport
animations [
 IntroAnimation {
  binding_count=1
  length_seconds=0.4333
  object=&self:IntroAnimation
  sequence {
   bindings [Overlay_28 {
     id=#1
     possessed_class=Overlay
     tracks [
      {
       type=MovieScene2DTransformTrack
       sections [{
         type=MovieScene2DTransformSection
         end_frame=13
         start_frame=0
         channels [
          Scale.X {type=MovieSceneScriptingFloatChannel, default=1.2, keys [[0,1],[5,1.3],[13,1.2]]}
          Scale.Y {type=MovieSceneScriptingFloatChannel, keys [[0,1],[5,1.3],[13,1.2]]}
         ]
        }]
      }
      {
       type=MovieSceneFloatTrack
       display_name=Render Opacity
       sections [{type=MovieSceneFloatSection, channels [None {type=MovieSceneScriptingFloatChannel, keys [[0,0],[3,1]]}]}]
      }
     ]
    }]
   display_rate {numerator=30}
  }
 }
 OutroAnimation {
  binding_count=1
  length_seconds=0.75
  object=&self:OutroAnimation
  sequence {bindings [Overlay_28 {
     id=#2
     possessed_class=Overlay
     tracks [
      {
       type=MovieSceneFloatTrack
       display_name=Render Opacity
       sections [{type=MovieSceneFloatSection, end_frame=15, start_frame=0, channels [None {type=MovieSceneScriptingFloatChannel, keys [[0,1],[15,0.5]]}]}]
      }
      {
       type=MovieScene2DTransformTrack
       sections [{
         type=MovieScene2DTransformSection
         end_frame=15
         start_frame=0
         channels [
          Scale.X {type=MovieSceneScriptingFloatChannel, keys [[0,1.2],[7,1],[15,1]]}
          Scale.Y {type=MovieSceneScriptingFloatChannel, keys [[0,1.2],[7,1],[15,1]]}
         ]
        }]
      }
     ]
    }]}
 }
]
graphs [
 Display {nodes [
   Display() {is=title K2Node_FunctionEntry}
   > "SetText (Text)"(self=Get Text_XPTotal() -> (Text_XPTotal:Common Text Object Reference), InText="To Text (Integer)"(Value=Get XPValue() -> (XPValue:int=0)))
   > "SetText (Text)"(self=Get Text_XPCategory() -> (Text_XPCategory:Common Text Object Reference), InText=Get XPType() -> (XPType:Text))
  ]}
 EventGraph {nodes [
   n1 = Event OnAnimateXPValue() {is=title K2Node_Event}
   Display(execute=[@n1.then, @n4.then]) {is=title K2Node_CallFunction}
   Event OnIntroPlayed() {is=title K2Node_Event}
   > PlayIntro_VIZ()
   n2 = Event PreConstruct()
   > Branch(Condition=@n2.IsDesignTime)
   > Display()
   n3 = Get IntroAnimation() -> (IntroAnimation:Widget Animation Object Reference)
   > PlayAnimation(InAnimation=@n3, StartAtTime=GetEndTime(self=@n3), PlayMode=Reverse)
   n4 = Event OnEntryInitialized() {is=title K2Node_Event}
  ]}
 Vizualization Graph {nodes [
   PlayOutro_VIZ() {is=title K2Node_CustomEvent}
   > StopAllAnimations(self:User Widget Object Reference)
   > PlayAnimation(InAnimation=Get OutroAnimation() -> (OutroAnimation:Widget Animation Object Reference))
   PlayIntro_VIZ() {is=title K2Node_CustomEvent}
   > PlayAnimation(InAnimation=Get IntroAnimation())
  ]}
]
variables [RundownCategoryColor {type=LinearColor, default=<rgba 0 0.8714 0.9823 1>}]
widgets {
 info {inheritedWidgetCount=1, parentClass=/Script/FortniteUI.FortMatchXPTotalEntry, rootWidgetClass=&Overlay, widgetCount=6}
 widgets [
  Overlay_28 {widgetClassPath=&Overlay, widget=Overlay_28}
  HorizontalBox_43 {widgetClassPath=&HorizontalBox, parent=Overlay_28, widget=HorizontalBox_43, slot {class=OverlaySlot}}
  Text_XPTotal {
   widgetClassPath=&CommonTextBlock
   bIsVariable=true
   parent=HorizontalBox_43
   widget=Text_XPTotal
   properties {
    text="52,000"
    textTransformPolicy=ToUpper
    font {
     bForceMonospaced=false
     bMaterialIsStencil=false
     fontMaterial=null
     fontObject=&HeadingNow
     letterSpacing=0
     monospacedWidth=1
     size=24
     skewAmount=0
     typefaceFontName=86 Bold
     outlineSettings {
      bApplyOutlineToDropShadows=false
      bMiteredCorners=false
      bSeparateFillAlpha=false
      outlineMaterial=null
      outlineSize=0
      outlineColor=<rgba 0 0 0 1>
     }
    }
    shadowOffset=<0 0>
   }
   slot {class=HorizontalBoxSlot, properties {horizontalAlignment=HAlign_Left, verticalAlignment=VAlign_Top, padding=<ltrb 0 0 8 0>}}
  }
  Text_XPLabel {
   widgetClassPath=&CommonTextBlock
   parent=HorizontalBox_43
   widget=Text_XPLabel
   properties {
    text=XP
    font {
     bForceMonospaced=false
     bMaterialIsStencil=false
     fontMaterial=null
     fontObject=&HeadingNow
     letterSpacing=0
     monospacedWidth=1
     size=24
     skewAmount=0
     typefaceFontName=86 Bold
     outlineSettings {
      bApplyOutlineToDropShadows=false
      bMiteredCorners=false
      bSeparateFillAlpha=false
      outlineMaterial=null
      outlineSize=0
      outlineColor=<rgba 0 0 0 1>
     }
    }
    shadowOffset=<0 0>
   }
   slot {class=HorizontalBoxSlot, properties {horizontalAlignment=HAlign_Left, verticalAlignment=VAlign_Center}}
  }
  Text_XPCategory {
   widgetClassPath=&CommonTextBlock
   bIsVariable=true
   parent=Overlay_28
   widget=Text_XPCategory
   properties {
    text="52,000"
    colorAndOpacity {colorUseRule=UseColor_Specified, specifiedColor=<rgba 0.9 0.9 0.9 0.75>}
    font {
     bForceMonospaced=false
     bMaterialIsStencil=false
     fontMaterial=null
     fontObject=&HeadingNow
     letterSpacing=0
     monospacedWidth=1
     size=18
     skewAmount=0
     typefaceFontName=74 Regular
     outlineSettings {
      bApplyOutlineToDropShadows=false
      bMiteredCorners=false
      bSeparateFillAlpha=false
      outlineMaterial=null
      outlineSize=0
      outlineColor=<rgba 0 0 0 1>
     }
    }
    shadowOffset=<0 0>
   }
   slot {class=OverlaySlot, properties {padding=<ltrb 0 40 0 0>}}
  }
  TutorialHighlight {widgetClassPath=&TutorialHighlight}
 ]
}
