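# Additive UI material that draws the XP logo on the post-match XP screen with an animated shine.
# A grayscale logo texture serves as a distance field, anti-aliased into crisp and soft edges driven by Anim_Softness.
# That mask blends purple vertical gradients, and Color Switch fades the result toward dark navy.
# A rotated reflected gradient, moved by Anim_Shine and ShineOffset, adds the shine through Color Dodge.
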Material /FortniteCoreUI/HUD/EndGame/PostGameFlowScreens/DefaultScreens/XPScreen/Materials/M_UI_PostMatchXPLogo
graph {
 properties {
  bAutomaticallySetUsageInEditor=false
  bUsedWithStaticMesh=false
  blendMode=BLEND_Additive
  materialDomain=MD_UI
  usedShadingModels=Default Lit
 }
 nodes [
  Constant()
  n1 = NamedRerouteUsage()
  LinearInterpolate(
   A=VectorParameter() {parameterName=Color1, defaultValue=<rgba 0.8963 0.9301 1 1>}.RGB
   B=VectorParameter() {parameterName=Color2, defaultValue=<rgba 0.8963 0.9301 1 1>}.RGB
   Alpha=Saturate(None=Multiply(A=@n1) {constB=4})
  )
  n2 = Constant3Vector() {constant=<rgba 0.0266 0.0062 0.25 1>}
  n3 = Constant3Vector() {constant=<rgba 0.0323 0.0125 0.25 1>}
  n4 = MaterialFunctionCall() {materialFunction=&LinearGradient}
  n5 = TextureSample() {
   samplerType=SAMPLERTYPE_Grayscale
   texture=/FortniteCoreUI/HUD/EndGame/PostGameFlowScreens/DefaultScreens/XPScreen/Textures/T_UI_Icon_PostMatchXP
  }
  n6 = LinearInterpolate(A=Constant3Vector() {constant=<rgba 0.0325 0.025 0.25 0>}, B=Constant3Vector() {constant=<rgba 0.1042 0.025 0.5 1>}, Alpha=@n4.VGradient)
  n7 = Constant()
  n8 = ScalarParameter() {parameterName=Anim_Softness}
  n9 = Constant() {r=0.5}
  n10 = Constant() {r=1}
  n11 = Add(
   A=MaterialFunctionCall(DistanceField=@n5.R, Threshold=Constant() {r=0.5}, Smoothing=LinearInterpolate(A=@n7, Alpha=@n8) {constB=0.75}) {materialFunction=&MF_UI_AntiAliasing}.Outer
   B=Multiply(A=MaterialFunctionCall(DistanceField=@n5.R, Threshold=@n9, Smoothing=@n10) {materialFunction=&MF_UI_AntiAliasing}.Outer) {constB=0.5}
  )
  n12 = LinearInterpolate(A=Multiply(A=LinearInterpolate(A=@n2, B=@n3, Alpha=@n4.VGradient), B=@n5.R), B=LinearInterpolate(A=@n6, B=@n6, Alpha=@n11), Alpha=@n11)
  n13 = LinearInterpolate(A=@n12, B=Constant3Vector() {constant=<rgba 0.0076 0.003 0.125 1>}, Alpha=ScalarParameter() {parameterName=Color Switch})
  n14 = TextureCoordinate()
  n15 = LinearInterpolate(Alpha=ScalarParameter() {parameterName=Anim_Shine}) {constA=1, constB=-1}
  n16 = MaterialFunctionCall!(
   UVs=MaterialFunctionCall(UVs=Add(A=@n14, B=@n15), "Rotation Angle (0-1)"=Constant() {r=-0.09}) {materialFunction=&CustomRotator}.Rotated Values
   Offset=ScalarParameter() {parameterName=ShineOffset}
   RotationCenter
   RotationAngle
  ) {materialFunction=/FortUILibrary/MaterialFunctions/Gradients/MF_UI_ReflectedGradient}
  LinearInterpolate(
   A=@n13
   B=MaterialFunctionCall!(Blend=Multiply(A=@n16.Result) {constB=0.75}, Base=@n13) {materialFunction=/Engine/Functions/Engine_MaterialFunctions03/Blends/Blend_ColorDodge}.Result
   Alpha=@n12
  )
  Multiply(A=@n11, B=LinearInterpolate(Alpha=ScalarParameter() {parameterName=Anim_Softness}) {constA=0.5, constB=0}) {constB=0.25}
  Power(Base=@n16.Result) {constExponent=1.5}
  TextureCoordinate()
 ]
}
