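# UI alpha-composite material for an animatable glow band in the post-game XP recap screen.
# A sine of the V texture coordinate feeds chained ease curves, shaping a soft falloff and a much tighter core.
# The tightened curve drives a three-color blend and its intensity mask, multiplied by vertex alpha, while a separate branch scales the softer curve at low strength.
# Anim_Opacity scales both branches and shifts each blended color between neighboring color parameters.
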Material /FortniteCoreUI/HUD/EndGame/PostGameFlowScreens/DefaultScreens/XPScreen/Materials/M_UI_RecapSineGlow
graph {
 properties {
  bAutomaticallySetUsageInEditor=false
  bUsedWithStaticMesh=false
  blendMode=BLEND_AlphaComposite
  materialDomain=MD_UI
  usedShadingModels=Default Lit
 }
 nodes [
  n1 = TextureCoordinate()
  n2 = MaterialFunctionCall("Value (0-1)"=ComponentMask(None=Sine(None=@n1) {period=2}) {g=true}) {materialFunction=&MF_UI_EaseCurves}
  Multiply(A=@n2.EaseInQuint, B=LinearInterpolate(Alpha=ScalarParameter() {defaultValue=1, parameterName=Anim_Opacity}) {constB=0.25}) {constB=0.25}
  n3 = VectorParameter() {parameterName=ColorFalloff, defaultValue=<rgba 0.025 0.025 0.5 1>}
  n4 = LinearInterpolate(
   A=VectorParameter() {parameterName=ColorFalloff2, defaultValue=<rgba 0 0 0.5 1>}.RGB
   B=@n3.RGB
   Alpha=ScalarParameter() {defaultValue=1, parameterName=Anim_Opacity}
  ) {constB=0.25}
  n5 = VectorParameter() {parameterName=ColorCore, defaultValue=<rgba 0.1292 0.05 1 1>}
  n6 = LinearInterpolate(A=@n3.RGB, B=@n5.RGB, Alpha=ScalarParameter() {defaultValue=1, parameterName=Anim_Opacity}) {constB=0.25}
  n7 = LinearInterpolate(
   A=@n5.RGB
   B=VectorParameter() {parameterName=Color, defaultValue=<rgba 0.25 0.1 1 1>}.RGB
   Alpha=ScalarParameter() {defaultValue=1, parameterName=Anim_Opacity}
  ) {constB=0.25}
  n8 = MaterialFunctionCall("Value (0-1)"=MaterialFunctionCall("Value (0-1)"=@n2.EaseInQuint) {materialFunction=&MF_UI_EaseCurves}.EaseInQuint) {materialFunction=&MF_UI_EaseCurves}
  n9 = ScalarParameter() {defaultValue=1, parameterName=Anim_Opacity}
  Multiply(
   A=VertexColor().A
   B=Multiply(
    A=MaterialFunctionCall(A=@n4, B=@n6, C=@n7, Alpha=@n8.EaseInCubic) {materialFunction=&3ColorBlend}.Result
    B=LinearInterpolate(B=@n8.EaseInSine, Alpha=@n9) {constB=0.25}
   )
  )
 ]
}
