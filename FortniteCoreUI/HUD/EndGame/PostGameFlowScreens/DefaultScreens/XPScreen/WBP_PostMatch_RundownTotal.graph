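# Post-match XP screen widget, derived from FortMatchXPTotalEntry, that shows an XP total with its category label.
# Display sets number and label on PreConstruct and entry initialization; OnAnimateXPValue interpolates the numeric text to XPValue.
# Intro and outro events play short scale animations on the main overlay, with a fade on outro.
# PlayRecapAnimation plays a longer reveal driving transforms, opacity and image and font material parameters for beam, glows, flares and shine.
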WidgetBlueprint /FortniteCoreUI/HUD/EndGame/PostGameFlowScreens/DefaultScreens/XPScreen/WBP_PostMatch_RundownTotal
animations [
 IntroAnimation {
  binding_count=1
  length_seconds=0.4333
  object=&self:IntroAnimation
  sequence {
   bindings [Overlay_28 {
     id=#1
     possessed_class=Overlay
     tracks [{
       type=MovieScene2DTransformTrack
       sections [{
         type=MovieScene2DTransformSection
         end_frame=13
         start_frame=0
         channels [
          Scale.X {type=MovieSceneScriptingFloatChannel, default=1.2, keys [[0,1],[5,1.3],[13,1.2]]}
          Scale.Y {type=MovieSceneScriptingFloatChannel, keys [[0,1],[5,1.3],[13,1.2]]}
         ]
        }]
      }]
    }]
   display_rate {numerator=30}
  }
 }
 OutroAnimation {
  binding_count=1
  length_seconds=0.75
  object=&self:OutroAnimation
  sequence {bindings [Overlay_28 {
     id=#2
     possessed_class=Overlay
     tracks [
      {
       type=MovieSceneFloatTrack
       display_name=Render Opacity
       sections [{type=MovieSceneFloatSection, end_frame=15, start_frame=0, channels [None {type=MovieSceneScriptingFloatChannel, keys [[0,1],[15,0.5]]}]}]
      }
      {
       type=MovieScene2DTransformTrack
       sections [{
         type=MovieScene2DTransformSection
         end_frame=15
         start_frame=0
         channels [
          Scale.X {type=MovieSceneScriptingFloatChannel, keys [[0,1.2],[7,1],[15,1]]}
          Scale.Y {type=MovieSceneScriptingFloatChannel, keys [[0,1.2],[7,1],[15,1]]}
         ]
        }]
      }
     ]
    }]}
 }
 Intro_XPRecap {
  binding_count=14
  length_seconds=4.25
  object=&self:Intro_XPRecap
  sequence {
   bindings [
    Glows {
     id=#3
     possessed_class=Overlay
     tracks [{
       type=MovieScene2DTransformTrack
       sections [{
         type=MovieScene2DTransformSection
         channels [
          Scale.X {type=MovieSceneScriptingFloatChannel, keys [[19,0],[164,1]]}
          Scale.Y {type=MovieSceneScriptingFloatChannel, keys [[19,0.5],[164,1]]}
         ]
        }]
      }]
    }
    Image_IconXP_Pulse {
     id=#4
     possessed_class=Image
     tracks [
      {
       type=MovieSceneWidgetMaterialTrack
       sections [{type=MovieSceneParameterSection, channels [Anim_Softness {type=MovieSceneScriptingFloatChannel, keys [[25,1],[29,0],[72,1]]}]}]
      }
      {
       type=MovieScene2DTransformTrack
       sections [{
         type=MovieScene2DTransformSection
         channels [
          Scale.X {type=MovieSceneScriptingFloatChannel, keys [[29,1],[166,2]]}
          Scale.Y {type=MovieSceneScriptingFloatChannel, keys [[29,1],[166,2]]}
         ]
        }]
      }
     ]
    }
    Triangles {
     id=#5
     possessed_class=HorizontalBox
     tracks [
      {
       type=MovieScene2DTransformTrack
       sections [{
         type=MovieScene2DTransformSection
         channels [
          Scale.X {type=MovieSceneScriptingFloatChannel, keys [[18,0],[162,1.5]]}
          Scale.Y {type=MovieSceneScriptingFloatChannel, keys [[18,0],[162,1.5]]}
         ]
        }]
      }
      {
       type=MovieSceneFloatTrack
       display_name=Render Opacity
       sections [{type=MovieSceneFloatSection, channels [None {type=MovieSceneScriptingFloatChannel, keys [[18,0],[21,1]]}]}]
      }
     ]
    }
    Image_GlowSmall {
     id=#6
     possessed_class=Image
     tracks [{
       type=MovieSceneFloatTrack
       display_name=Render Opacity
       sections [{type=MovieSceneFloatSection, channels [None {type=MovieSceneScriptingFloatChannel, keys [[19,0],[26,1],[105,1],[255,0]]}]}]
      }]
    }
    Content {
     id=#7
     possessed_class=HorizontalBox
     tracks [
      {
       type=MovieScene2DTransformTrack
       sections [{
         type=MovieScene2DTransformSection
         channels [
          Translation.X {type=MovieSceneScriptingFloatChannel, const [0,1]}
          Translation.Y {type=MovieSceneScriptingFloatChannel, const [0,1]}
          Angle {type=MovieSceneScriptingFloatChannel, const [0,1]}
          Scale.X {type=MovieSceneScriptingFloatChannel, keys [[15,4],[25,0.9],[43,1],[210,1],[219,0.8]]}
          Scale.Y {type=MovieSceneScriptingFloatChannel, keys [[15,4],[25,0.9],[43,1],[210,1],[219,0.8]]}
          Shear.X {type=MovieSceneScriptingFloatChannel, const [0,1]}
          Shear.Y {type=MovieSceneScriptingFloatChannel, const [0,1]}
         ]
        }]
      }
      {
       type=MovieSceneFloatTrack
       display_name=Render Opacity
       sections [{type=MovieSceneFloatSection, channels [None {type=MovieSceneScriptingFloatChannel, keys [[15,0],[18,1],[210,1],[219,0]]}]}]
      }
     ]
    }
    Triangles_Left {
     id=#8
     possessed_class=Image
     tracks [{
       type=MovieSceneWidgetMaterialTrack
       sections [{type=MovieSceneParameterSection, channels [TriScale {type=MovieSceneScriptingFloatChannel, keys [[20,0.75],[227,0]]}]}]
      }]
    }
    Text_XP_Label {
     id=#9
     possessed_class=CommonTextBlock
     tracks [{
       type=MovieSceneWidgetMaterialTrack
       display_name=Font.FontMaterial
       sections [{
         type=MovieSceneParameterSection
         channels [
          Color Swap {type=MovieSceneScriptingFloatChannel, default=1, keys [[18,0],[27,1],[180,0]]}
          ShineOffset {type=MovieSceneScriptingFloatChannel, keys [[75,-1],[150,1]]}
         ]
        }]
      }]
    }
    Flares {
     id=#10
     possessed_class=HorizontalBox
     tracks [
      {
       type=MovieSceneFloatTrack
       display_name=Render Opacity
       sections [{type=MovieSceneFloatSection, channels [None {type=MovieSceneScriptingFloatChannel, keys [[21,0],[24,1],[195,0]]}]}]
      }
      {
       type=MovieScene2DTransformTrack
       sections [{
         type=MovieScene2DTransformSection
         channels [
          Scale.X {type=MovieSceneScriptingFloatChannel, keys [[21,0],[166,1]]}
          Scale.Y {type=MovieSceneScriptingFloatChannel, keys [[21,0.5],[166,1]]}
         ]
        }]
      }
     ]
    }
    Text_XPCategory {
     id=#11
     possessed_class=CommonTextBlock
     tracks [{
       type=MovieSceneWidgetMaterialTrack
       display_name=Font.FontMaterial
       sections [{
         type=MovieSceneParameterSection
         channels [
          Color Swap {type=MovieSceneScriptingFloatChannel, default=1, keys [[18,0],[27,1],[180,0]]}
          ShineOffset {type=MovieSceneScriptingFloatChannel, keys [[60,-1],[106,1]]}
         ]
        }]
      }]
    }
    Image_IconXP {
     id=#12
     possessed_class=Image
     tracks [
      {
       type=MovieSceneFloatTrack
       display_name=Render Opacity
       sections [{type=MovieSceneFloatSection, channels [None {type=MovieSceneScriptingFloatChannel, keys [[18,0],[42,1]]}]}]
      }
      {
       type=MovieScene2DTransformTrack
       sections [{
         type=MovieScene2DTransformSection
         channels [
          Translation.X {type=MovieSceneScriptingFloatChannel, const [0,1]}
          Translation.Y {type=MovieSceneScriptingFloatChannel, const [0,1]}
          Angle {type=MovieSceneScriptingFloatChannel, const [0,1]}
          Scale.X {type=MovieSceneScriptingFloatChannel, keys [[18,0],[45,1],[210,1],[225,0]]}
          Scale.Y {type=MovieSceneScriptingFloatChannel, keys [[18,0],[45,1],[210,1],[225,0]]}
          Shear.X {type=MovieSceneScriptingFloatChannel, const [0,1]}
          Shear.Y {type=MovieSceneScriptingFloatChannel, const [0,1]}
         ]
        }]
      }
      {
       type=MovieSceneWidgetMaterialTrack
       sections [{
         type=MovieSceneParameterSection
         channels [
          Anim_Softness {type=MovieSceneScriptingFloatChannel, keys [[18,1],[42,0],[225,1]]}
          Anim_Shine {type=MovieSceneScriptingFloatChannel, keys [[18,0],[73,1]]}
         ]
        }]
      }
     ]
    }
    Image_GlowLarge {
     id=#13
     possessed_class=Image
     tracks [{
       type=MovieSceneFloatTrack
       display_name=Render Opacity
       sections [{type=MovieSceneFloatSection, channels [None {type=MovieSceneScriptingFloatChannel, keys [[19,0],[22,1],[255,0]]}]}]
      }]
    }
    Beam {
     id=#14
     possessed_class=Image
     tracks [
      {
       type=MovieSceneWidgetMaterialTrack
       sections [{type=MovieSceneParameterSection, channels [Anim_Opacity {type=MovieSceneScriptingFloatChannel, keys [[0,0],[9,1],[31,1],[181,0]]}]}]
      }
      {
       type=MovieScene2DTransformTrack
       sections [{type=MovieScene2DTransformSection, channels [Scale.Y {type=MovieSceneScriptingFloatChannel, keys [[0,0],[15,0.25],[30,2],[181,0]]}]}]
      }
     ]
    }
    Triangles_Right {
     id=#15
     possessed_class=Image
     tracks [{
       type=MovieSceneWidgetMaterialTrack
       sections [{type=MovieSceneParameterSection, channels [TriScale {type=MovieSceneScriptingFloatChannel, keys [[20,0.75],[227,0]]}]}]
      }]
    }
    NText_XPTotal_Number {
     id=#16
     possessed_class=CommonNumericTextBlock
     tracks [{
       type=MovieSceneWidgetMaterialTrack
       display_name=Font.FontMaterial
       sections [{
         type=MovieSceneParameterSection
         channels [
          Color Swap {type=MovieSceneScriptingFloatChannel, keys [[18,0],[27,1],[180,0]]}
          ShineOffset {type=MovieSceneScriptingFloatChannel, keys [[45,-1],[76,1]]}
         ]
        }]
      }]
    }
   ]
   display_rate {numerator=60}
  }
 }
]
graphs [
 Display {nodes [
   Display() {is=title K2Node_FunctionEntry}
   > SetCurrentValue!(
    execute
    self:Common Numeric Text Block Object Reference=Get NText_XPTotal_Number() -> (NText_XPTotal_Number:Common Numeric Text Block Object Reference)
    NewValue:float="To Float (Integer)"(InInt=Get XPValue() -> (XPValue:int=0))
   ) -> (then)
   > "SetText (Text)"(self=Get Text_XPCategory() -> (Text_XPCategory:Common Text Object Reference), InText=Get XPType() -> (XPType:Text))
  ]}
 EventGraph {nodes [
   Event OnAnimateXPValue() {is=title K2Node_Event}
   > InterpolateToValue!(
    execute
    self:Common Numeric Text Block Object Reference=Get NText_XPTotal_Number()
    TargetValue:float="To Float (Integer)"(InInt=Get XPValue() -> (XPValue=0))
    MaximumInterpolationDuration:float=0.85
    MinimumChangeRate:float=1
    OutroOffset:float=0
   ) -> (then)
   Event OnIntroPlayed() {is=title K2Node_Event}
   > PlayIntro_VIZ()
   Event PreConstruct()
   > Display() {is=title K2Node_CallFunction}
   Event OnEntryInitialized() {is=title K2Node_Event}
   > Display()
  ]}
 PlayRecapAnimation {nodes [
   PlayRecapAnimation() {is=title K2Node_FunctionEntry}
   > PlayAnimation(InAnimation=Get Intro_XPRecap() -> (Intro_XPRecap:Widget Animation Object Reference))
  ]}
 Vizualization Graph {nodes [
   PlayOutro_VIZ() {is=title K2Node_CustomEvent}
   > StopAllAnimations(self:User Widget Object Reference)
   > PlayAnimation(InAnimation=Get OutroAnimation() -> (OutroAnimation:Widget Animation Object Reference))
   PlayIntro_VIZ() {is=title K2Node_CustomEvent}
   > PlayAnimation(InAnimation=Get IntroAnimation() -> (IntroAnimation:Widget Animation Object Reference))
  ]}
]
variables [RundownCategoryColor {type=LinearColor, default=<rgba 0 0.8714 0.9823 1>}]
widgets {
 info {inheritedWidgetCount=1, parentClass=/Script/FortniteUI.FortMatchXPTotalEntry, rootWidgetClass=&VerticalBox, widgetCount=23}
 widgets [
  VerticalBox_0 {widgetClassPath=&VerticalBox, widget=VerticalBox_0}
  Overlay_28 {
   widgetClassPath=&Overlay
   parent=VerticalBox_0
   widget=Overlay_28
   slot {class=VerticalBoxSlot, properties {padding=<ltrb 0 64 0 0>, size {sizeRule=Fill, value=1}}}
  }
  Beam {
   widgetClassPath=&Image
   bIsVariable=true
   parent=Overlay_28
   widget=Beam
   properties {
    visibility=SelfHitTestInvisible
    brush {
     bIsDynamicallyLoaded=false
     drawAs=Image
     imageType=FullColor
     mirroring=NoMirror
     resourceName=None
     resourceObject=/FortniteCoreUI/HUD/EndGame/PostGameFlowScreens/DefaultScreens/XPScreen/Materials/M_UI_RecapSineGlow
     tiling=NoTile
     imageSize=<0 640>
     margin=<ltrb 0 0 0 0>
     outlineSettings {
      bUseBrushTransparency=false
      roundingType=HalfHeightRadius
      width=0
      color {colorUseRule=UseColor_Specified, specifiedColor=<rgba 0 0 0 0>}
      cornerRadii=<0 0 0 0>
     }
     tintColor {colorUseRule=UseColor_Specified, specifiedColor=<rgba 1 1 1 1>}
     uVRegion {bIsValid=false, max=<0 0>, min=<0 0>}
    }
   }
   slot {class=OverlaySlot, properties {horizontalAlignment=HAlign_Fill, verticalAlignment=VAlign_Center}}
  }
  FortVisualAttachment_0 {
   widgetClassPath=&FortVisualAttachment
   parent=Overlay_28
   widget=FortVisualAttachment_0
   properties {contentAnchor=<0.5 0.475>}
   slot {class=OverlaySlot, properties {horizontalAlignment=HAlign_Center, verticalAlignment=VAlign_Center}}
  }
  Overlay_0 {widgetClassPath=&Overlay, parent=FortVisualAttachment_0, widget=Overlay_0, slot {class=SizeBoxSlot}}
  Image_IconXP_Pulse {
   widgetClassPath=&Image
   parent=Overlay_0
   widget=Image_IconXP_Pulse
   properties {
    visibility=SelfHitTestInvisible
    brush {
     bIsDynamicallyLoaded=false
     drawAs=Image
     imageType=FullColor
     mirroring=NoMirror
     resourceName=None
     resourceObject=/FortniteCoreUI/HUD/EndGame/PostGameFlowScreens/DefaultScreens/XPScreen/Materials/M_UI_PostMatchXPLogo
     tiling=NoTile
     imageSize=<320 320>
     margin=<ltrb 0 0 0 0>
     outlineSettings {
      bUseBrushTransparency=false
      roundingType=HalfHeightRadius
      width=0
      color {colorUseRule=UseColor_Specified, specifiedColor=<rgba 0 0 0 0>}
      cornerRadii=<0 0 0 0>
     }
     tintColor {colorUseRule=UseColor_Specified, specifiedColor=<rgba 1 1 1 1>}
     uVRegion {bIsValid=false, max=<0 0>, min=<0 0>}
    }
   }
   slot {class=OverlaySlot, properties {horizontalAlignment=HAlign_Fill, verticalAlignment=VAlign_Fill}}
  }
  Image_IconXP {
   widgetClassPath=&Image
   parent=Overlay_0
   widget=Image_IconXP
   properties {
    visibility=SelfHitTestInvisible
    brush {
     bIsDynamicallyLoaded=false
     drawAs=Image
     imageType=FullColor
     mirroring=NoMirror
     resourceName=None
     resourceObject=&M_UI_PostMatchXPLogo
     tiling=NoTile
     imageSize=<320 320>
     margin=<ltrb 0 0 0 0>
     outlineSettings {
      bUseBrushTransparency=false
      roundingType=HalfHeightRadius
      width=0
      color {colorUseRule=UseColor_Specified, specifiedColor=<rgba 0 0 0 0>}
      cornerRadii=<0 0 0 0>
     }
     tintColor {colorUseRule=UseColor_Specified, specifiedColor=<rgba 1 1 1 1>}
     uVRegion {bIsValid=false, max=<0 0>, min=<0 0>}
    }
   }
   slot {class=OverlaySlot, properties {horizontalAlignment=HAlign_Fill, verticalAlignment=VAlign_Fill}}
  }
  Overlay_1 {
   widgetClassPath=&Overlay
   parent=Overlay_28
   widget=Overlay_1
   slot {class=OverlaySlot, properties {horizontalAlignment=HAlign_Center, verticalAlignment=VAlign_Center}}
  }
  Glows {
   widgetClassPath=&Overlay
   parent=Overlay_1
   widget=Glows
   slot {class=OverlaySlot, properties {horizontalAlignment=HAlign_Fill, verticalAlignment=VAlign_Fill}}
  }
  Image_GlowLarge {
   widgetClassPath=&Image
   parent=Glows
   widget=Image_GlowLarge
   properties {
    visibility=SelfHitTestInvisible
    brush {
     bIsDynamicallyLoaded=false
     drawAs=Image
     imageType=FullColor
     mirroring=NoMirror
     resourceName=None
     resourceObject=/FortniteCoreUI/HUD/EndGame/PostGameFlowScreens/DefaultScreens/XPScreen/Materials/M_UI_PostMatchTextGlow
     tiling=NoTile
     imageSize=<64 64>
     margin=<ltrb 0 0 0 0>
     outlineSettings {
      bUseBrushTransparency=false
      roundingType=HalfHeightRadius
      width=0
      color {colorUseRule=UseColor_Specified, specifiedColor=<rgba 0 0 0 0>}
      cornerRadii=<0 0 0 0>
     }
     tintColor {colorUseRule=UseColor_Specified, specifiedColor=<rgba 1 1 1 1>}
     uVRegion {bIsValid=false, max=<0 0>, min=<0 0>}
    }
   }
   slot {class=OverlaySlot, properties {horizontalAlignment=HAlign_Fill, verticalAlignment=VAlign_Fill, padding=<ltrb -512 -128 -512 -128>}}
  }
  Image_GlowSmall {
   widgetClassPath=&Image
   parent=Glows
   widget=Image_GlowSmall
   properties {
    visibility=SelfHitTestInvisible
    brush {
     bIsDynamicallyLoaded=false
     drawAs=Image
     imageType=FullColor
     mirroring=NoMirror
     resourceName=None
     resourceObject=&M_UI_PostMatchTextGlow
     tiling=NoTile
     imageSize=<64 64>
     margin=<ltrb 0 0 0 0>
     outlineSettings {
      bUseBrushTransparency=false
      roundingType=HalfHeightRadius
      width=0
      color {colorUseRule=UseColor_Specified, specifiedColor=<rgba 0 0 0 0>}
      cornerRadii=<0 0 0 0>
     }
     tintColor {colorUseRule=UseColor_Specified, specifiedColor=<rgba 0.5 0.5 0.5 1>}
     uVRegion {bIsValid=false, max=<0 0>, min=<0 0>}
    }
   }
   slot {class=OverlaySlot, properties {horizontalAlignment=HAlign_Fill, verticalAlignment=VAlign_Fill, padding=<ltrb -320 -80 -310 -80>}}
  }
  Flares {
   widgetClassPath=&HorizontalBox
   parent=Overlay_1
   widget=Flares
   slot {class=OverlaySlot, properties {horizontalAlignment=HAlign_Fill, verticalAlignment=VAlign_Center, padding=<ltrb -384 0 -384 0>}}
  }
  Flare_Left {
   widgetClassPath=&Image
   parent=Flares
   widget=Flare_Left
   properties {
    visibility=SelfHitTestInvisible
    brush {
     bIsDynamicallyLoaded=false
     drawAs=Image
     imageType=FullColor
     mirroring=NoMirror
     resourceName=None
     resourceObject=/FortniteCoreUI/HUD/EndGame/PostGameFlowScreens/DefaultScreens/XPScreen/Materials/MI_UI_PostMatch_XPFlare
     tiling=NoTile
     imageSize=<32 56>
     margin=<ltrb 0 0 0 0>
     outlineSettings {
      bUseBrushTransparency=false
      roundingType=HalfHeightRadius
      width=0
      color {colorUseRule=UseColor_Specified, specifiedColor=<rgba 0 0 0 0>}
      cornerRadii=<0 0 0 0>
     }
     tintColor {colorUseRule=UseColor_Specified, specifiedColor=<rgba 1 1 1 1>}
     uVRegion {bIsValid=false, max=<0 0>, min=<0 0>}
    }
   }
   slot {class=HorizontalBoxSlot, properties {padding=<ltrb 0 0 -8 0>, size {sizeRule=Fill, value=1}}}
  }
  Flare_Right {
   widgetClassPath=&Image
   parent=Flares
   widget=Flare_Right
   properties {
    visibility=SelfHitTestInvisible
    brush {
     bIsDynamicallyLoaded=false
     drawAs=Image
     imageType=FullColor
     mirroring=NoMirror
     resourceName=None
     resourceObject=&MI_UI_PostMatch_XPFlare
     tiling=NoTile
     imageSize=<32 56>
     margin=<ltrb 0 0 0 0>
     outlineSettings {
      bUseBrushTransparency=false
      roundingType=HalfHeightRadius
      width=0
      color {colorUseRule=UseColor_Specified, specifiedColor=<rgba 0 0 0 0>}
      cornerRadii=<0 0 0 0>
     }
     tintColor {colorUseRule=UseColor_Specified, specifiedColor=<rgba 1 1 1 1>}
     uVRegion {bIsValid=false, max=<0 0>, min=<0 0>}
    }
    renderTransform {angle=0, scale=<-1 -1>, shear=<0 0>, translation=<0 0>}
   }
   slot {class=HorizontalBoxSlot, properties {padding=<ltrb -8 0 0 0>, size {sizeRule=Fill, value=1}}}
  }
  Triangles {
   widgetClassPath=&HorizontalBox
   parent=Overlay_1
   widget=Triangles
   slot {class=OverlaySlot, properties {horizontalAlignment=HAlign_Fill, verticalAlignment=VAlign_Fill, padding=<ltrb -128 -40 -128 -40>}}
  }
  Triangles_Left {
   widgetClassPath=&Image
   parent=Triangles
   widget=Triangles_Left
   properties {
    visibility=SelfHitTestInvisible
    brush {
     bIsDynamicallyLoaded=false
     drawAs=Image
     imageType=FullColor
     mirroring=NoMirror
     resourceName=None
     resourceObject=/FortniteCoreUI/HUD/EndGame/PostGameFlowScreens/DefaultScreens/XPScreen/Materials/MI_PostMatch_Triangles
     tiling=NoTile
     imageSize=<500 40>
     margin=<ltrb 0 0 0 0>
     outlineSettings {
      bUseBrushTransparency=false
      roundingType=HalfHeightRadius
      width=0
      color {colorUseRule=UseColor_Specified, specifiedColor=<rgba 0 0 0 0>}
      cornerRadii=<0 0 0 0>
     }
     tintColor {colorUseRule=UseColor_Specified, specifiedColor=<rgba 1 1 1 1>}
     uVRegion {bIsValid=false, max=<0 0>, min=<0 0>}
    }
   }
   slot {class=HorizontalBoxSlot, properties {size {sizeRule=Fill, value=1}}}
  }
  Triangles_Right {
   widgetClassPath=&Image
   parent=Triangles
   widget=Triangles_Right
   properties {
    visibility=SelfHitTestInvisible
    brush {
     bIsDynamicallyLoaded=false
     drawAs=Image
     imageType=FullColor
     mirroring=NoMirror
     resourceName=None
     resourceObject=/FortniteCoreUI/HUD/EndGame/PostGameFlowScreens/DefaultScreens/XPScreen/Materials/MI_PostMatch_Triangles_Alternate
     tiling=NoTile
     imageSize=<500 40>
     margin=<ltrb 0 0 0 0>
     outlineSettings {
      bUseBrushTransparency=false
      roundingType=HalfHeightRadius
      width=0
      color {colorUseRule=UseColor_Specified, specifiedColor=<rgba 0 0 0 0>}
      cornerRadii=<0 0 0 0>
     }
     tintColor {colorUseRule=UseColor_Specified, specifiedColor=<rgba 1 1 1 1>}
     uVRegion {bIsValid=false, max=<0 0>, min=<0 0>}
    }
    renderTransform {angle=0, scale=<-1 -1>, shear=<0 0>, translation=<0 0>}
   }
   slot {class=HorizontalBoxSlot, properties {size {sizeRule=Fill, value=1}}}
  }
  Content {widgetClassPath=&HorizontalBox, parent=Overlay_1, widget=Content, slot {class=OverlaySlot}}
  NText_XPTotal_Number {
   widgetClassPath=/Script/CommonUI.CommonNumericTextBlock
   bIsVariable=true
   parent=Content
   widget=NText_XPTotal_Number
   properties {
    currentNumericValue=4293.9458
    text=0
    font {
     bForceMonospaced=false
     bMaterialIsStencil=false
     fontMaterial=/FortniteCoreUI/HUD/EndGame/PostGameFlowScreens/DefaultScreens/XPScreen/Materials/M_UI_PostMatch_FontShine
     fontObject=&HeadingNow
     letterSpacing=0
     monospacedWidth=1
     size=54
     skewAmount=0
     typefaceFontName=86 Bold Italic
     outlineSettings {
      bApplyOutlineToDropShadows=false
      bMiteredCorners=false
      bSeparateFillAlpha=true
      outlineMaterial=/FortniteCoreUI/HUD/EndGame/PostGameFlowScreens/DefaultScreens/XPScreen/Materials/MI_PostMatch_FontGlow
      outlineSize=32
      outlineColor=<rgba 1 1 1 1>
     }
    }
    margin=<ltrb -36 -36 -36 -36>
    shadowOffset=<0 0>
   }
   slot {class=HorizontalBoxSlot, properties {horizontalAlignment=HAlign_Left, verticalAlignment=VAlign_Center, padding=<ltrb 0 0 28 0>}}
  }
  Text_XPCategory {
   widgetClassPath=&CommonTextBlock
   bIsVariable=true
   parent=Content
   widget=Text_XPCategory
   properties {
    text=Total
    textTransformPolicy=ToUpper
    font {
     bForceMonospaced=false
     bMaterialIsStencil=false
     fontMaterial=&M_UI_PostMatch_FontShine
     fontObject=&HeadingNow
     letterSpacing=0
     monospacedWidth=1
     size=54
     skewAmount=0
     typefaceFontName=86 Bold Italic
     outlineSettings {
      bApplyOutlineToDropShadows=false
      bMiteredCorners=false
      bSeparateFillAlpha=true
      outlineMaterial=&MI_PostMatch_FontGlow
      outlineSize=32
      outlineColor=<rgba 1 1 1 1>
     }
    }
    margin=<ltrb -36 -36 -36 -36>
    shadowColorAndOpacity=<rgba 1 1 1 0>
    shadowOffset=<0 0>
    strikeBrush {
     bIsDynamicallyLoaded=false
     drawAs=Image
     imageType=FullColor
     mirroring=NoMirror
     resourceName=None
     resourceObject=null
     tiling=NoTile
     imageSize=<32 32>
     margin=<ltrb 0 0 0 0>
     outlineSettings {
      bUseBrushTransparency=false
      roundingType=HalfHeightRadius
      width=0
      color {colorUseRule=UseColor_Specified, specifiedColor=<rgba 0 0 0 0>}
      cornerRadii=<0 0 0 0>
     }
     tintColor {colorUseRule=UseColor_Specified, specifiedColor=<rgba 1 1 1 1>}
     uVRegion {bIsValid=false, max=<0 0>, min=<0 0>}
    }
   }
   slot {class=HorizontalBoxSlot, properties {horizontalAlignment=HAlign_Left, verticalAlignment=VAlign_Center, padding=<ltrb 0 0 28 0>}}
  }
  Text_XP_Label {
   widgetClassPath=&CommonTextBlock
   parent=Content
   widget=Text_XP_Label
   properties {
    text=XP
    textTransformPolicy=ToUpper
    font {
     bForceMonospaced=false
     bMaterialIsStencil=false
     fontMaterial=&M_UI_PostMatch_FontShine
     fontObject=&HeadingNow
     letterSpacing=0
     monospacedWidth=1
     size=54
     skewAmount=0
     typefaceFontName=86 Bold Italic
     outlineSettings {
      bApplyOutlineToDropShadows=false
      bMiteredCorners=false
      bSeparateFillAlpha=true
      outlineMaterial=&MI_PostMatch_FontGlow
      outlineSize=32
      outlineColor=<rgba 1 1 1 1>
     }
    }
    margin=<ltrb -36 -36 -36 -36>
    shadowColorAndOpacity=<rgba 1 1 1 0>
    shadowOffset=<0 0>
    strikeBrush {
     bIsDynamicallyLoaded=false
     drawAs=Image
     imageType=FullColor
     mirroring=NoMirror
     resourceName=None
     resourceObject=null
     tiling=NoTile
     imageSize=<32 32>
     margin=<ltrb 0 0 0 0>
     outlineSettings {
      bUseBrushTransparency=false
      roundingType=HalfHeightRadius
      width=0
      color {colorUseRule=UseColor_Specified, specifiedColor=<rgba 0 0 0 0>}
      cornerRadii=<0 0 0 0>
     }
     tintColor {colorUseRule=UseColor_Specified, specifiedColor=<rgba 1 1 1 1>}
     uVRegion {bIsValid=false, max=<0 0>, min=<0 0>}
    }
   }
   slot {class=HorizontalBoxSlot, properties {horizontalAlignment=HAlign_Left, verticalAlignment=VAlign_Center}}
  }
  Text_XPLabel {
   widgetClassPath=&CommonTextBlock
   parent=Content
   widget=Text_XPLabel
   properties {
    text=XP
    visibility=Collapsed
    font {
     bForceMonospaced=false
     bMaterialIsStencil=false
     fontMaterial=null
     fontObject=&HeadingNow
     letterSpacing=0
     monospacedWidth=1
     size=27
     skewAmount=0
     typefaceFontName=76 Bold Italic
     outlineSettings {
      bApplyOutlineToDropShadows=false
      bMiteredCorners=false
      bSeparateFillAlpha=false
      outlineMaterial=null
      outlineSize=0
      outlineColor=<rgba 0 0 0 1>
     }
    }
    shadowOffset=<0 0>
   }
   slot {class=HorizontalBoxSlot, properties {horizontalAlignment=HAlign_Left, verticalAlignment=VAlign_Center}}
  }
  TutorialHighlight {widgetClassPath=&TutorialHighlight}
 ]
}
