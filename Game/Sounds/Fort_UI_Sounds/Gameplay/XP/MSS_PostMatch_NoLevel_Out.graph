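# Stereo one-shot MetaSound Source for post-match XP UI audio, named for the no-level-up outro.
# It passes its Sounds array, defaulting to a single XP reveal wave, into a nested MSS_PlayRandom_Oneshot that supplies both output channels.
# OnPlay links directly to OnFinished, and no trigger link reaches the nested node.
# The UI-class sound takes volume modulation from the level-up control bus and sends pre-effect audio to a music sidechain bus.
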MetaSoundSource /Game/Sounds/Fort_UI_Sounds/Gameplay/XP/MSS_PostMatch_NoLevel_Out
inputs [
 Sounds {type=WaveAsset:Array}
 UE.Source.OnPlay {type=Trigger}
]
interfaces [UE.OutputFormat.Stereo:1.0, UE.Source.OneShot:1.0, UE.Source:1.1]
outputs [
 "UE.OutputFormat.Stereo.Audio:0" {type=Audio}
 "UE.OutputFormat.Stereo.Audio:1" {type=Audio}
 UE.Source.OneShot.OnFinished {type=Trigger}
]
pages [Default {
  input_defaults {UE.Source.OnPlay=null, Sounds [&/Fort_PostMatch_XP_Reveal]}
  nodes [MSS_PlayRandom_Oneshot_0 = /Metasound/DefaultAssets/Sources/MSS_PlayRandom_Oneshot(Sounds=UE.Reroute(Value=^Sounds).Value)]
  links [
   {dst="@UE.Source.OneShot.OnFinished", src="@UE.Source.OnPlay"}
   {dst="@UE.OutputFormat.Stereo.Audio:0", src=MSS_PlayRandom_Oneshot_0.UE.OutputFormat.Stereo.Audio:0}
   {dst="@UE.OutputFormat.Stereo.Audio:1", src=MSS_PlayRandom_Oneshot_0.UE.OutputFormat.Stereo.Audio:1}
  ]
 }]
properties {
 assetImportData=&self:AssetImportData
 numChannels=2
 outputFormat=Stereo
 sampleRate=48000
 soundClassObject=&Fort_UI
 modulationSettings {volumeRouting=Union, volumeModulationDestination {bEnableModulation=true, modulators [/Game/Sounds/Modulation/Buses/CB_FortLevelUp]}}
 preEffectBusSends [{
   audioBus=&AB_Sidechain_FortMusic_Vol
   maxSendDistance=1000
   minSendDistance=100
   minSendLevel=0
   sendLevel=0.3
  }]
}
