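# Post-match XP recap widget that animates gameplay and quest XP rundowns, then can show a Return to Island button that ends the minigame.
# Builds reversed entry lists, with helpers to dim and hide entries and play ticker audio.
# Async-loads battle pass and XP Everywhere widgets by tag alias, destroying the battle pass after a cancellable delay.
# Registers two extension slots and uses the global event router for focus, map and XP events.
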WidgetBlueprint /FortniteCoreUI/HUD/EndGame/PostGameFlowScreens/DefaultScreens/XPScreen/WBP_MatchXPRecap_Rundown
animations [
 Anim_Intro {
  binding_count=5
  length_seconds=0.5
  object=&self:Anim_Intro
  sequence {bindings [
    Ovr_QuestArea {
     id=#1
     possessed_class=Overlay
     tracks [{
       type=MovieSceneFloatTrack
       display_name=Render Opacity
       sections [{type=MovieSceneFloatSection, end_frame=10, start_frame=0, channels [None {type=MovieSceneScriptingFloatChannel, keys [[0,0],[1,1],[10,1]]}]}]
      }]
    }
    Ovr_GameplayArea {
     id=#2
     possessed_class=Overlay
     tracks [{
       type=MovieSceneFloatTrack
       display_name=Render Opacity
       sections [{type=MovieSceneFloatSection, end_frame=10, start_frame=0, channels [None {type=MovieSceneScriptingFloatChannel, keys [[0,0],[1,1],[10,1]]}]}]
      }]
    }
    Att_RundownAndTotal {
     id=#3
     possessed_class=FortVisualAttachment
     tracks [
      {
       type=MovieSceneFloatTrack
       display_name=Render Opacity
       sections [{type=MovieSceneFloatSection, end_frame=10, start_frame=0, channels [None {type=MovieSceneScriptingFloatChannel, keys [[0,0],[10,1]]}]}]
      }
      {
       type=MovieScene2DTransformTrack
       sections [{
         type=MovieScene2DTransformSection
         end_frame=10
         start_frame=0
         channels [Translation.Y {type=MovieSceneScriptingFloatChannel, keys [[0,-50],[10,0]]}]
        }]
      }
     ]
    }
    GPanel_XPRundown {
     id=#4
     possessed_class=GridPanel
     tracks [{
       type=MovieSceneFloatTrack
       display_name=Render Opacity
       sections [{type=MovieSceneFloatSection, end_frame=2, start_frame=0, channels [None {type=MovieSceneScriptingFloatChannel, keys [[0,0],[2,1]]}]}]
      }]
    }
    Overlay_XPRecap {
     id=#5
     possessed_class=Overlay
     tracks [{
       type=MovieSceneFloatTrack
       display_name=Render Opacity
       sections [{type=MovieSceneFloatSection, end_frame=10, start_frame=0, channels [None {type=MovieSceneScriptingFloatChannel, keys [[0,0],[4,1],[10,1]]}]}]
      }]
    }
   ]}
 }
 Anim_IntroTransition {
  binding_count=2
  length_seconds=3
  object=&self:Anim_IntroTransition
  sequence {bindings [
    WBP_PostMatch_RundownTotal {
     id=#6
     possessed_class=WBP_PostMatch_RundownTotal_C
     tracks [{
       type=MovieSceneFloatTrack
       display_name=Render Opacity
       sections [{type=MovieSceneFloatSection, end_frame=60, start_frame=0, channels [None {type=MovieSceneScriptingFloatChannel, const [1,2]}]}]
      }]
    }
    Att_RundownAndTotal {
     id=#7
     possessed_class=FortVisualAttachment
     tracks [
      {
       type=MovieSceneFloatTrack
       display_name=Render Opacity
       sections [{type=MovieSceneFloatSection, end_frame=5, start_frame=0, channels [None {type=MovieSceneScriptingFloatChannel, keys [[0,1],[5,0]]}]}]
      }
      {
       type=MovieScene2DTransformTrack
       sections [{
         type=MovieScene2DTransformSection
         end_frame=5
         start_frame=0
         channels [Translation.Y {type=MovieSceneScriptingFloatChannel, keys [[0,0],[5,-50]]}]
        }]
      }
     ]
    }
   ]}
 }
 Anim_Finish {
  binding_count=7
  length_seconds=0.25
  object=&self:Anim_Finish
  sequence {
   bindings [
    Ovr_GameplayArea {
     id=#8
     possessed_class=Overlay
     tracks [
      {
       type=MovieSceneFloatTrack
       display_name=Render Opacity
       sections [{type=MovieSceneFloatSection, end_frame=15, start_frame=0, channels [None {type=MovieSceneScriptingFloatChannel, const [0,2]}]}]
      }
      {
       type=MovieScene2DTransformTrack
       sections [{
         type=MovieScene2DTransformSection
         end_frame=15
         start_frame=0
         channels [
          Translation.X {type=MovieSceneScriptingFloatChannel, const [0,2]}
          Translation.Y {type=MovieSceneScriptingFloatChannel, const [0,2]}
          Angle {type=MovieSceneScriptingFloatChannel, const [0,2]}
          Scale.X {type=MovieSceneScriptingFloatChannel, const [1,2]}
          Scale.Y {type=MovieSceneScriptingFloatChannel, const [1,2]}
          Shear.X {type=MovieSceneScriptingFloatChannel, const [0,2]}
          Shear.Y {type=MovieSceneScriptingFloatChannel, const [0,2]}
         ]
        }]
      }
     ]
    }
    Att_RundownAndTotal {
     id=#9
     possessed_class=FortVisualAttachment
     tracks [
      {
       type=MovieSceneFloatTrack
       display_name=Render Opacity
       sections [{type=MovieSceneFloatSection, end_frame=15, start_frame=0, channels [None {type=MovieSceneScriptingFloatChannel, const [0,2]}]}]
      }
      {
       type=MovieScene2DTransformTrack
       sections [{
         type=MovieScene2DTransformSection
         end_frame=15
         start_frame=0
         channels [
          Translation.X {type=MovieSceneScriptingFloatChannel, const [0,2]}
          Translation.Y {type=MovieSceneScriptingFloatChannel, const [0,2]}
          Angle {type=MovieSceneScriptingFloatChannel, const [0,2]}
          Scale.X {type=MovieSceneScriptingFloatChannel, const [1,2]}
          Scale.Y {type=MovieSceneScriptingFloatChannel, const [1,2]}
          Shear.X {type=MovieSceneScriptingFloatChannel, const [0,2]}
          Shear.Y {type=MovieSceneScriptingFloatChannel, const [0,2]}
         ]
        }]
      }
     ]
    }
    Ovr_QuestArea {
     id=#10
     possessed_class=Overlay
     tracks [
      {
       type=MovieSceneFloatTrack
       display_name=Render Opacity
       sections [{type=MovieSceneFloatSection, end_frame=15, start_frame=0, channels [None {type=MovieSceneScriptingFloatChannel, const [0,2]}]}]
      }
      {
       type=MovieScene2DTransformTrack
       sections [{
         type=MovieScene2DTransformSection
         end_frame=15
         start_frame=0
         channels [
          Translation.X {type=MovieSceneScriptingFloatChannel, const [0,2]}
          Translation.Y {type=MovieSceneScriptingFloatChannel, const [0,2]}
          Angle {type=MovieSceneScriptingFloatChannel, const [0,2]}
          Scale.X {type=MovieSceneScriptingFloatChannel, const [1,2]}
          Scale.Y {type=MovieSceneScriptingFloatChannel, const [1,2]}
          Shear.X {type=MovieSceneScriptingFloatChannel, const [0,2]}
          Shear.Y {type=MovieSceneScriptingFloatChannel, const [0,2]}
         ]
        }]
      }
     ]
    }
    NSlot_BattlepassInjection {
     id=#11
     possessed_class=NamedSlot
     tracks [
      {
       type=MovieSceneFloatTrack
       display_name=Render Opacity
       sections [{type=MovieSceneFloatSection, end_frame=15, start_frame=0, channels [None {type=MovieSceneScriptingFloatChannel, keys [[0,0],[15,1]]}]}]
      }
      {
       type=MovieScene2DTransformTrack
       sections [{
         type=MovieScene2DTransformSection
         end_frame=15
         start_frame=0
         channels [Translation.Y {type=MovieSceneScriptingFloatChannel, keys [[0,100],[15,0]]}]
        }]
      }
      {
       type=MovieSceneByteTrack
       display_name=Visibility
       sections [{type=MovieSceneByteSection, end_frame=15, start_frame=0, channels [None {type=MovieSceneScriptingByteChannel, keys [[0,1],[1,4],[15,4]]}]}]
      }
     ]
    }
    VisSwitcher_BottomCenter {
     id=#12
     possessed_class=CommonVisibilitySwitcher
     tracks [
      {
       type=MovieSceneFloatTrack
       display_name=Render Opacity
       sections [{type=MovieSceneFloatSection, end_frame=15, start_frame=0, channels [None {type=MovieSceneScriptingFloatChannel, keys [[0,0],[15,1]]}]}]
      }
      {
       type=MovieScene2DTransformTrack
       sections [{type=MovieScene2DTransformSection, end_frame=15, start_frame=0, channels [Translation.Y {type=MovieSceneScriptingFloatChannel, const [0,2]}]}]
      }
     ]
    }
    SBox_CenterContent {
     id=#13
     possessed_class=SizeBox
     tracks [
      {
       type=MovieSceneFloatTrack
       display_name=Height Override
       sections [{type=MovieSceneFloatSection, end_frame=15, start_frame=0, channels [None {type=MovieSceneScriptingFloatChannel, keys [[0,400],[15,300]]}]}]
      }
      {
       type=MovieSceneFloatTrack
       display_name=Width Override
       sections [{type=MovieSceneFloatSection, end_frame=15, start_frame=0, channels [None {type=MovieSceneScriptingFloatChannel, keys [[0,400],[15,300]]}]}]
      }
      {
       type=MovieScene2DTransformTrack
       sections [{
         type=MovieScene2DTransformSection
         end_frame=15
         start_frame=0
         channels [Translation.Y {type=MovieSceneScriptingFloatChannel, keys [[0,0],[15,-50]]}]
        }]
      }
     ]
    }
    VerticalBox_PrimaryMatchXPExtension {
     id=#14
     possessed_class=VerticalBox
     tracks [
      {
       type=MovieSceneFloatTrack
       display_name=Render Opacity
       sections [{type=MovieSceneFloatSection, end_frame=15, start_frame=0, channels [None {type=MovieSceneScriptingFloatChannel, keys [[0,0],[15,1]]}]}]
      }
      {
       type=MovieScene2DTransformTrack
       sections [{
         type=MovieScene2DTransformSection
         end_frame=15
         start_frame=0
         channels [Translation.X {type=MovieSceneScriptingFloatChannel, keys [[0,-50],[15,0]]}]
        }]
      }
     ]
    }
   ]
   display_rate {numerator=60}
  }
 }
 Anim_BattlepassFocused {
  binding_count=1
  length_seconds=0.25
  object=&self:Anim_BattlepassFocused
  sequence {bindings [GPanel_XPRundown {
     id=#15
     possessed_class=GridPanel
     tracks [
      {
       type=MovieSceneFloatTrack
       display_name=Render Opacity
       sections [{type=MovieSceneFloatSection, end_frame=5, start_frame=0, channels [None {type=MovieSceneScriptingFloatChannel, keys [[0,1],[5,0]]}]}]
      }
      {
       type=MovieScene2DTransformTrack
       sections [{
         type=MovieScene2DTransformSection
         end_frame=5
         start_frame=0
         channels [
          Translation.X {type=MovieSceneScriptingFloatChannel, const [0,2]}
          Translation.Y {type=MovieSceneScriptingFloatChannel, keys [[0,0],[5,-25]]}
         ]
        }]
      }
     ]
    }]}
 }
 Anim_RundownQuestXPCategory {
  binding_count=4
  length_seconds=0.25
  object=&self:Anim_RundownQuestXPCategory
  sequence {bindings [
    MatchXP_ChallengeXP {
     id=#16
     possessed_class=WBP_PostMatch_RundownCategoryItem_C
     tracks [{
       type=MovieScene2DTransformTrack
       sections [{
         type=MovieScene2DTransformSection
         end_frame=5
         start_frame=0
         channels [
          Scale.X {type=MovieSceneScriptingFloatChannel, keys [[0,1],[5,1.1]]}
          Scale.Y {type=MovieSceneScriptingFloatChannel, keys [[0,1],[5,1.1]]}
         ]
        }]
      }]
    }
    Ovr_QuestXPEffects {
     id=#17
     possessed_class=Overlay
     tracks [
      {
       type=MovieSceneFloatTrack
       display_name=Render Opacity
       sections [{type=MovieSceneFloatSection, end_frame=5, start_frame=0, channels [None {type=MovieSceneScriptingFloatChannel, keys [[0,0],[5,1]]}]}]
      }
      {
       type=MovieScene2DTransformTrack
       sections [{
         type=MovieScene2DTransformSection
         end_frame=5
         start_frame=0
         channels [
          Translation.X {type=MovieSceneScriptingFloatChannel, const [0,1]}
          Translation.Y {type=MovieSceneScriptingFloatChannel, const [0,1]}
          Angle {type=MovieSceneScriptingFloatChannel, const [0,1]}
          Scale.X {type=MovieSceneScriptingFloatChannel, const [1,1]}
          Scale.Y {type=MovieSceneScriptingFloatChannel, keys [[0,0.5],[5,1]]}
          Shear.X {type=MovieSceneScriptingFloatChannel, const [0,1]}
          Shear.Y {type=MovieSceneScriptingFloatChannel, const [0,1]}
         ]
        }]
      }
     ]
    }
    Ovr_QuestTotals {
     id=#18
     possessed_class=Overlay
     tracks [{
       type=MovieScene2DTransformTrack
       sections [{
         type=MovieScene2DTransformSection
         end_frame=5
         start_frame=0
         channels [
          Scale.X {type=MovieSceneScriptingFloatChannel, keys [[0,1],[5,1.1]]}
          Scale.Y {type=MovieSceneScriptingFloatChannel, keys [[0,1],[5,1.1]]}
         ]
        }]
      }]
    }
    Img_XPEffects_Glow {
     id=#19
     possessed_class=Image
     tracks [{
       type=MovieSceneFloatTrack
       display_name=Render Opacity
       sections [{type=MovieSceneFloatSection, end_frame=5, start_frame=0, channels [None {type=MovieSceneScriptingFloatChannel, keys [[0,0],[5,1]]}]}]
      }]
    }
   ]}
 }
 Anim_RundownGameplayXPCategory {
  binding_count=3
  length_seconds=0.25
  object=&self:Anim_RundownGameplayXPCategory
  sequence {bindings [
    Img_XPEffects_Glow_1 {
     id=#20
     possessed_class=Image
     tracks [{
       type=MovieSceneFloatTrack
       display_name=Render Opacity
       sections [{type=MovieSceneFloatSection, end_frame=5, start_frame=0, channels [None {type=MovieSceneScriptingFloatChannel, keys [[0,0],[5,1]]}]}]
      }]
    }
    MatchXP_GameplayXp {
     id=#21
     possessed_class=WBP_PostMatch_RundownCategoryItem_C
     tracks [{
       type=MovieScene2DTransformTrack
       sections [{
         type=MovieScene2DTransformSection
         end_frame=5
         start_frame=0
         channels [
          Scale.X {type=MovieSceneScriptingFloatChannel, keys [[0,1],[5,1.1]]}
          Scale.Y {type=MovieSceneScriptingFloatChannel, keys [[0,1],[5,1.1]]}
         ]
        }]
      }]
    }
    Ovr_GameplayEffects {
     id=#22
     possessed_class=Overlay
     tracks [
      {
       type=MovieSceneFloatTrack
       display_name=Render Opacity
       sections [{type=MovieSceneFloatSection, end_frame=5, start_frame=0, channels [None {type=MovieSceneScriptingFloatChannel, keys [[0,0],[5,1]]}]}]
      }
      {
       type=MovieScene2DTransformTrack
       sections [{
         type=MovieScene2DTransformSection
         end_frame=5
         start_frame=0
         channels [Scale.Y {type=MovieSceneScriptingFloatChannel, keys [[0,0.5],[5,1]]}]
        }]
      }
     ]
    }
   ]}
 }
 Anim_HideXPEverywhere {
  binding_count=1
  length_seconds=0.25
  object=&self:Anim_HideXPEverywhere
  sequence {bindings [SBox_CenterContent {
     id=#23
     possessed_class=SizeBox
     tracks [{
       type=MovieSceneFloatTrack
       display_name=Render Opacity
       sections [{type=MovieSceneFloatSection, end_frame=5, start_frame=0, channels [None {type=MovieSceneScriptingFloatChannel, keys [[0,1],[5,0]]}]}]
      }]
    }]}
 }
 Anim_ShowQuestDetails {
  binding_count=1
  length_seconds=0.4
  object=&self:Anim_ShowQuestDetails
  sequence {bindings [Ovr_QuestTotals {
     id=#24
     possessed_class=Overlay
     tracks [
      {
       type=MovieSceneFloatTrack
       display_name=Render Opacity
       sections [{
         type=MovieSceneFloatSection
         end_frame=4
         start_frame=0
         channels [None {type=MovieSceneScriptingFloatChannel, keys [[0,0],[1.3333,0],[4.6667,1]]}]
        }]
      }
      {
       type=MovieScene2DTransformTrack
       sections [{
         type=MovieScene2DTransformSection
         end_frame=8
         start_frame=0
         channels [
          Translation.X {type=MovieSceneScriptingFloatChannel, keys [[0,200],[4.6667,10],[8,0]]}
          Translation.Y {type=MovieSceneScriptingFloatChannel, const [0,1]}
          Scale.X {type=MovieSceneScriptingFloatChannel, const [1,2]}
          Scale.Y {type=MovieSceneScriptingFloatChannel, keys [[0,0],[4.6667,1],[8,1]]}
          Shear.X {type=MovieSceneScriptingFloatChannel, keys [[0,15],[4.6667,0],[8,0]]}
          Shear.Y {type=MovieSceneScriptingFloatChannel, const [0,2]}
         ]
        }]
      }
     ]
    }]}
 }
]
graphs [
 AddAlias_Battlepass {nodes [
   Inputs!() -> (In:Exec)
   > Branch(Condition=NOT Boolean(A=Get IsSeasonPassCreated() -> (IsSeasonPassCreated:bool)))
   > Set IsSeasonPassCreated(IsSeasonPassCreated:bool=true) -> (Output_Get:bool)
   n1 = GetSoftWidgetClassForAlias!(
    self:Fort UIData Configuration Context Object Reference=Get FortUIDataConfigurationContext!() -> (ReturnValue:Fort UIData Configuration Context Object Reference) {is=title K2Node_GetBlueprintContextEx}
    WidgetAlias:"Gameplay Tag Structure (by ref)"=MakeLiteralGameplayTag!(Value:Gameplay Tag Structure=(TagName="DynamicUI.HUD.Fort.Athena.PostMatch.BottomTray")) -> (ReturnValue:Gameplay Tag Structure)
   ) -> (ReturnValue:User Widget Soft Class Reference)
   > Branch(Condition=IsValidSoftClassReference!(SoftClassReference:"Object Soft Class Reference (by ref)"=@n1) -> (ReturnValue:bool))
   > n2 = CreateWidgetAsync!(
    execute:Exec
    UserWidgetSoftClass:User Widget Soft Class Reference=@n1
    OwningPlayer:Player Controller Object Reference
    bSuspendInputUntilComplete:bool=true
   ) -> (then:Exec, OnComplete:Exec, OnFailure:Exec, UserWidget:User Widget Object Reference) {is=title K2Node_AsyncAction}
   n3 = Cast To CommonActivatableWidget!(execute=@n2.OnComplete, Object=@n2) -> (then, CastFailed, AsCommon Activatable Widget:Common Activatable Widget Object Reference) {is=title K2Node_DynamicCast}
   > AddChild(self=Get NSlot_BattlepassInjection() -> (NSlot_BattlepassInjection:Named Slot Object Reference), Content=@n3)
   > Outputs!(BattlepassInitialized:Exec)
  ]}
 AddAlias_XPEverywhere {nodes [
   Inputs()
   > Branch(Condition=NOT Boolean(A=Get IsXPEverywhereInitialized() -> (IsXPEverywhereInitialized:bool)))
   n1 = GetSoftWidgetClassForAlias(
    self=Get FortUIDataConfigurationContext()
    WidgetAlias=MakeLiteralGameplayTag(Value=(TagName="DynamicUI.HUD.Fort.Athena.PostMatch.XPEverywhere"))
   )
   > Branch(Condition=IsValidSoftClassReference(SoftClassReference=@n1))
   > n2 = CreateWidgetAsync(UserWidgetSoftClass=@n1, bSuspendInputUntilComplete=true)
   AddChild(execute=@n2.OnComplete, self=Get NSlot_XPEverywhere() -> (NSlot_XPEverywhere:Named Slot Object Reference), Content=@n2)
   > Set IsXPEverywhereInitialized(IsXPEverywhereInitialized:bool=true) -> (Output_Get:bool)
   > Outputs!(XPEverywhereInitialized:Exec)
  ]}
 AddRundownWidgets {nodes [
   AddRundownWidgets() {is=title K2Node_FunctionEntry}
   > Branch(Condition=NOT Boolean(A=Get bIsSuppressingBattlepass() -> (bIsSuppressingBattlepass:bool)))
   > Clear(TargetArray:Array of Fort Match XPTotal Entry Object References=Get QuestXPWidgets() -> (QuestXPWidgets:Array of Fort Match XPTotal Entry Object References))
   > Clear(TargetArray=Get GameplayXPWidgets() -> (GameplayXPWidgets:Array of Fort Match XPTotal Entry Object References))
   > ClearChildren(self=Get VBox_RundownEntries() -> (VBox_RundownEntries:Vertical Box Object Reference))
   > n1 = For Each Loop(Array:Array of Rundown Entry Display Structures=Get GameplayRundownEntries() -> (GameplayRundownEntries:Array of Rundown Entry Display Structures)) -> (Array Element:Rundown Entry Display Structure)
   n2 = Break Rundown Entry Display!(RundownEntryDisplay:"Rundown Entry Display Structure (by ref)"=@n1) -> (EarnedXP:int, Description:Text) {is=title K2Node_BreakStruct}
   > n3 = Create Widget!(
    execute:Exec
    Class:User Widget Class Reference=&/WBP_PostMatch_RundownItem.WBP_PostMatch_RundownItem_C
    OwningPlayer:Player Controller Object Reference
    XPType:Text=@n2.Description
    XPValue:int=@n2
   ) -> (then:Exec, ReturnValue:WBP Post Match Rundown Item Object Reference)
   > Set XPCategory!(execute, XPCategory:EFortExperienceCategory Enum=Gameplay, self:WBP Post Match Rundown Item Object Reference=@n3) -> (then, Output_Get:EFortExperienceCategory Enum=None)
   > Add(
    TargetArray:Array of Fort Match XPTotal Entry Object References=Get GameplayXPWidgets()
    NewItem:"Fort Match XPTotal Entry Object Reference (by ref)"=@n3
   ) {is=title K2Node_CallArrayFunction}
   n4 = For Each Loop(Exec=@n1.Completed, Array=Get CompletedRundownEntries() -> (CompletedRundownEntries:Array of Rundown Entry Display Structures))
   n5 = Break Rundown Entry Display(RundownEntryDisplay=@n4)
   > n6 = Create Widget(
    Class=&WBP_PostMatch_RundownItem_C
    XPType=@n5.Description
    XPValue=@n5
   )
   > Set XPCategory(XPCategory=Quest, self=@n6) -> (Output_Get=None)
   > Add(TargetArray=Get QuestXPWidgets(), NewItem=@n6)
   n7 = Get GameplayXPWidgets()
   Reverse!(execute=@n4.Completed, TargetArray:Array of Fort Match XPTotal Entry Object References=@n7) -> (then) {is=title K2Node_CallArrayFunction}
   > n8 = For Each Loop(Array:Array of Fort Match XPTotal Entry Object References=@n7) -> (Array Element:Widget Object Reference)
   > n9 = AddChildToVerticalBox(self=Get VBox_RundownEntries(), Content=@n8)
   > SetHorizontalAlignment(self:Vertical Box Slot Object Reference=@n9, InHorizontalAlignment=HAlign_Center)
   n10 = Get QuestXPWidgets()
   Reverse(execute=@n8.Completed, TargetArray=@n10)
   > n11 = For Each Loop(Array=@n10)
   > n12 = AddChildToVerticalBox(self=Get VBox_RundownEntries(), Content=@n11)
   > SetHorizontalAlignment(self=@n12, InHorizontalAlignment=HAlign_Center)
  ]}
 AddRundownWidgets_old {nodes [AddRundownWidgets_old() {is=title K2Node_FunctionEntry}]}
 ApplyFadeToLowestEntries {nodes [
   ApplyFadeToLowestEntries() {is=title K2Node_FunctionEntry}
   n1 = Get MaxEntriesToDisplay() -> (MaxEntriesToDisplay:int=0)
   > n2 = FindNextValidEntry!(
    execute
    self
    StartingIndex:int=int + int(A=Get CurrentXpEntryIndex() -> (CurrentXpEntryIndex:int=0), B=int - int(A=@n1, B=2))
    StartWithGameplayWidgets:bool=Get QuestsFinished() -> (QuestsFinished:bool)
   ) -> (then, IsSuccess:bool, IsGameplayItem:bool, FoundIndex:int=0, EntryWidget:WBP Post Match Rundown Item Object Reference)
   > Branch(Condition=@n2)
   > PlayDim_VIZ(self:WBP Post Match Rundown Item Object Reference=@n2.EntryWidget)
   > n3 = FindNextValidEntry(StartingIndex=@n2.FoundIndex, StartWithGameplayWidgets=@n2.IsGameplayItem) -> (FoundIndex=0)
   > Branch(Condition=@n3)
   > PlayHide_VIZ(self:WBP Post Match Rundown Item Object Reference=@n3.EntryWidget)
  ]}
 AudioAddTickerComponent {nodes [
   n1 = AudioAddTickerComponent!() -> (then, IsFinal:bool=~) {is=title K2Node_FunctionEntry}
   > n2 = SpawnSound2D!(
    execute
    Sound:Sound Base Object Reference=Select(
     Option 0:Sound Base Object Reference=Get XPEntry_Ticker() -> (XPEntry_Ticker:Sound Base Object Reference)
     Option 1:Sound Base Object Reference=Get XPEntry_TickerFinal() -> (XPEntry_TickerFinal:Sound Base Object Reference)
     Index=@n1
    ) -> (ReturnValue:Sound Base Object Reference)
    VolumeMultiplier:float=1
    PitchMultiplier:float=1
    StartTime:float=0
    ConcurrencySettings:Sound Concurrency Object Reference
    bPersistAcrossLevelTransition:bool=false
    bAutoDestroy:bool=true
   ) -> (then, ReturnValue:Audio Component Object Reference)
   > Set NewAudioComponent(NewAudioComponent:Audio Component Object Reference=@n2) -> (Output_Get:Audio Component Object Reference)
   > n3 = For Each Loop with Break!(
    Exec:Exec
    Array:Array of Audio Component Object References=Get XPEntry_Ticker_Components() -> (XPEntry_Ticker_Components:Array of Audio Component Object References)
    Break:Exec=@n6.then
   ) -> (LoopBody:Exec, Array Element:Audio Component Object Reference, Array Index:int, Completed:Exec)
   n4 = Branch(execute=@n3.Completed, Condition="Equal (Integer)"(A=Get EmptyIndex() -> (EmptyIndex:int=0), B=-1))
   > Add(
    TargetArray:Array of Audio Component Object References=Get XPEntry_Ticker_Components()
    NewItem:"Audio Component Object Reference (by ref)"=Get NewAudioComponent() -> (NewAudioComponent:Audio Component Object Reference)
   )
   Set Array Elem!(
    execute=@n4.else
    TargetArray:Array of Audio Component Object References=Get XPEntry_Ticker_Components()
    Index:int=Get EmptyIndex() -> (EmptyIndex=0)
    Item:"Audio Component Object Reference (by ref)"=Get NewAudioComponent()
    bSizeToFit:bool=false
   ) -> (then)
   n5 = Is Valid(exec=@n3.LoopBody, InputObject=@n3)
   n6 = Set EmptyIndex(execute=@n5.Is Not Valid, EmptyIndex:int=@n3.Array Index) -> (Output_Get:int=0)
  ]}
 BindToIntro {nodes [
   n1 = BindToIntro!() -> (then, Object:Object Reference) {is=title K2Node_FunctionEntry}
   > n2 = Cast To WBP_PostMatch_RundownItem!(execute, Object=@n1) -> (then, CastFailed, AsWBP Post Match Rundown Item:WBP Post Match Rundown Item Object Reference) {is=title K2Node_DynamicCast}
   > Bind Event to On Intro Finished(self:WBP Post Match Rundown Item Object Reference=@n2, Delegate=Create Event()) {is=title K2Node_AddDelegate}
  ]}
 CacheMinigame {nodes [
   CacheMinigame() {is=title K2Node_FunctionEntry}
   > n1 = Cast To FortPlayerControllerAthena(Object=GetOwningPlayer())
   n2 = GetMinigame!(self:Fort Player Controller Gameplay Object Reference=@n1) -> (ReturnValue:Fort Minigame Object Reference)
   > Is Valid(InputObject=@n2)
   > Set Minigame(Minigame:Fort Minigame Object Reference=@n2) -> (Output_Get:Fort Minigame Object Reference)
  ]}
 CanChangeMinigameState {nodes [
   CanChangeMinigameState() {is=title K2Node_FunctionEntry}
   > n1 = Cast To FortGameStateAthena!(execute, Object=GetGameState!() -> (ReturnValue:Game State Base Object Reference)) -> (then, CastFailed, AsFort Game State Athena:Fort Game State Athena Object Reference) {is=title K2Node_DynamicCast}
   > n2 = Branch(Condition=IsExperiencePlayOnly!(self:Fort Game State Athena Object Reference=@n1) -> (ReturnValue:bool))
   Return Node!(execute=[@n3.CastFailed, @n2.then, @n1.CastFailed], CanChange:bool=false)
   n3 = Cast To FortPlayerControllerAthena(execute=@n2.else, Object=GetOwningPlayer())
   > Return Node(CanChange=CanOperateOnMinigame!(PlayerController:Fort Player Controller Object Reference=@n3) -> (ReturnValue:bool))
  ]}
 EventGraph {nodes [
   n1 = "On Clicked (ReturnToIslandButton)"!() -> (OutputDelegate, then, Button:Common Button Base Object Reference) {is=title K2Node_ComponentBoundEvent}
   n2 = Sequence(execute=[@n1.then, @n22.then])
   > n3 = Cast To FortPlayerControllerAthena(Object=GetOwningPlayer())
   > ServerEndMinigame!(execute, self:Fort Player Controller Athena Object Reference=@n3, bAbandon:bool=true) -> (then)
   ResetPostMatch(execute=@n2.then_1)
   InitializeXPRundown() {is=title K2Node_CustomEvent}
   > Branch(Condition=NOT Boolean(A=Get IsPostmatchInitialized() -> (IsPostmatchInitialized:bool)))
   > Set IsPostmatchInitialized(IsPostmatchInitialized:bool=true) -> (Output_Get:bool)
   > AddRundownWidgets() {is=title K2Node_CallFunction}
   > InitializeXP()
   > SetVisibility(
    self=[
     Get MatchXPScoreboard() -> (MatchXPScoreboard:WBP Match XPScoreboard Object Reference)
     Get ReturnToIslandButton() -> (ReturnToIslandButton:WBP UIKit Button Loud Object Reference)
    ]
    InVisibility=Collapsed
   )
   > ReturnToIslandButtonShown!(execute, self:Fort Match XPRecap Object Reference, bShown:bool=false) -> (then)
   > CacheMinigame() {is=title K2Node_CallFunction}
   > n4 = Branch(Condition=NOT Boolean(A=Get IsSkippingVisualization() -> (IsSkippingVisualization:bool)))
   > Intro_VIZ()
   HandleVisualizationSkipped(execute=@n4.else)
   OnAllXPAnimationsFinished() {is=title K2Node_CustomEvent}
   > n5 = Sequence!(execute) -> (then_0, then_1, then_2:Exec, then_3:Exec)
   > Branch(Condition=CanChangeMinigameState!() -> (CanChange:bool) {is=title K2Node_CallFunction})
   > SetVisibility(self=Get ReturnToIslandButton(), InVisibility=SelfHitTestInvisible)
   > ReturnToIslandButtonShown(bShown=true)
   > SetActiveWidget(
    self=Get VisSwitcher_BottomCenter() -> (VisSwitcher_BottomCenter:Common Visibility Switcher Object Reference)
    Widget=Get ReturnToIslandButton()
   )
   n6 = Update Show Scoreboard!(execute=@n5.then_1, self) -> (then, ShowingScoreboard:bool)
   > Branch(Condition=@n6)
   > FadeOutOverlay(self:Fort Match XPRecap Object Reference)
   Branch(execute=@n5.then_2, Condition=NOT Boolean(A=Get IsSkippingVisualization()))
   > XPTotalsFinished_VIZ()
   n7 = TimerEvent_DestroyBattlepassInjection() {is=title K2Node_CustomEvent}
   > ClearChildren(self=Get NSlot_BattlepassInjection())
   > Set IsSeasonPassCreated(IsSeasonPassCreated=false)
   > Clear and Invalidate Timer by Handle(Handle=Get BattlepassInjectionDestructionTimer() -> (BattlepassInjectionDestructionTimer:Timer Handle Structure))
   CreateBattlepassInjectionOrCancelDestruction() {is=title K2Node_CustomEvent}
   > Branch(Condition=NOT Boolean(A=Get bIsSuppressingBattlepass()))
   n8 = Get BattlepassInjectionDestructionTimer()
   > n9 = Branch(Condition=Is Timer Active by Handle!(Handle:Timer Handle Structure=@n8) -> (ReturnValue:bool))
   > Clear and Invalidate Timer by Handle(Handle=@n8)
   Branch(execute=@n9.else, Condition=Get IsPostmatchInitialized())
   > Add Alias Battlepass!(In:Exec) -> (BattlepassInitialized:Exec) {is=title K2Node_MacroInstance}
   DestroyBattlepassInjectionAfterTimer() {is=title K2Node_CustomEvent}
   > Branch(Condition=NOT Boolean(A=Get bIsSuppressingBattlepass()))
   > n10 = Set Timer by Event(Delegate=@n7, Time=5, bLooping=false)
   > Set BattlepassInjectionDestructionTimer(BattlepassInjectionDestructionTimer:Timer Handle Structure=@n10) -> (Output_Get:Timer Handle Structure)
   n11 = HandleBattlepassOpened!() -> (OutputDelegate, then, IsBattlepassOpen:bool=~) {is=title K2Node_CustomEvent}
   > n12 = Branch(Condition=@n11.IsBattlepassOpen)
   > BattlepassActivated_VIZ()
   BattlepassDeactivated_VIZ!(execute=@n12.else, self, bSkipAnimation:bool=false) -> (then)
   HandleAllVizualizationsFinished() {is=title K2Node_CustomEvent}
   > Branch(Condition=NOT Boolean(A=Get IsResettingAnimation() -> (IsResettingAnimation:bool)))
   > Call On XPRundown Animations Finished!(execute:Exec, self:Fort Match XPRecap Object Reference) -> (then:Exec) {is=title K2Node_CallDelegate}
   n13 = OnToggleFullscreenMap!() -> (OutputDelegate, then, bIsMapVisible:bool=~) {is=title K2Node_CustomEvent}
   > n14 = Branch(Condition=@n13.bIsMapVisible)
   > DestroyBattlepassInjectionAfterTimer() {is=title K2Node_CallFunction}
   CreateBattlepassInjectionOrCancelDestruction(execute=@n14.else) {is=title K2Node_CallFunction}
   Event On Activated()
   > CreateBattlepassInjectionOrCancelDestruction()
   > SetVisibility(
    self=[
     Get MatchXP_UntilNextLvlReport() -> (MatchXP_UntilNextLvlReport:WBP Post Match Rundown Report Object Reference)
     Get WBP_PostMatch_RundownTotal() -> (WBP_PostMatch_RundownTotal:WBP Post Match Rundown Total Object Reference)
     Get MatchXP_TotalEntry() -> (MatchXP_TotalEntry:Fort Match XPTotal Entry Object Reference)
    ]
    InVisibility=SelfHitTestInvisible
   )
   > BattlepassDeactivated_VIZ(bSkipAnimation=true)
   Event On Deactivated()
   > DestroyBattlepassInjectionAfterTimer()
   n15 = Event SetXPAnimationState!() -> (OutputDelegate, then, IsActive:bool) {is=title K2Node_Event}
   > SetXPAnimationActive!(execute, self, IsAnimationActive:bool=@n15.IsActive) -> (then)
   Event Post Minigame Reset() {is=title K2Node_Event}
   > ResetPostMatch()
   n16 = Event On Recap Bar XP Event Created!() -> (OutputDelegate, then, XPEvent:Fort XPEvent Entry Structure, DeltaStep:int=0) {is=title K2Node_Event}
   > ProgressEvent_VIZ!(
    execute
    self
    DeltaStep:int=@n16.DeltaStep
    XPEvent:Fort XPEvent Entry Structure=@n16.XPEvent
    Duration:double=Get Duration_ProgressAnimation_CONST() -> (Duration_ProgressAnimation_CONST:double)
   ) -> (then)
   > Broadcast Event!(
    execute
    self:Gameplay Event Router Component Object Reference=GetGlobalEventRouter()
    EventData:XPEverywhere Event Manually Push Event Structure=MakeXPEverywherePushEvent!(EventData:Fort XPEvent Entry Structure=@n16.XPEvent) -> (PushEvent:XPEverywhere Event Manually Push Event Structure)
    EventContext:Object Reference=Self-Reference()
   ) -> (then, ReturnValue:bool)
   Event OnDebugAnimationReset() {is=title K2Node_Event}
   > Set IsResettingAnimation(IsResettingAnimation:bool=true) -> (Output_Get:bool)
   > Set IsPostmatchInitialized(IsPostmatchInitialized=false)
   > Set GameplayIntroPlayed(GameplayIntroPlayed:bool=false) -> (Output_Get:bool)
   > Set IsXPEverywhereInitialized(IsXPEverywhereInitialized=false)
   > Set IsSeasonPassCreated(IsSeasonPassCreated=false)
   > DEBUG_ResetAnimationState(self:Fort Match XPTotal Entry Object Reference=[
     Get MatchXP_GameplayXp() -> (MatchXP_GameplayXp:Fort Match XPTotal Entry Object Reference)
     Get MatchXP_ChallengeXp() -> (MatchXP_ChallengeXp:Fort Match XPTotal Entry Object Reference)
    ])
   n17 = Get Anim_IntroTransition() -> (Anim_IntroTransition:Widget Animation Object Reference)
   > PlayAnimation(InAnimation=@n17, StartAtTime=GetEndTime(self=@n17), PlayMode=Reverse)
   n18 = Get Anim_Finish() -> (Anim_Finish:Widget Animation Object Reference)
   > PlayAnimation(InAnimation=@n18, StartAtTime=GetEndTime(self=@n18), PlayMode=Reverse)
   > n19 = Update Post Match Injections!(In:Exec, IsShuttingDown:bool=true) -> (AllInjectionsUpdated:Exec, XPEverywhereUpdatedOnly:Exec, AllCleared:Exec) {is=title K2Node_MacroInstance}
   DEBUG PlayCosmeticIntro(execute=@n19.AllCleared, self:Fort Match XPRecap Object Reference)
   Event OnInitialized()
   > SetVisibility(self=[Get MatchXPScoreboard(), Get ReturnToIslandButton()], InVisibility=Collapsed)
   > ReturnToIslandButtonShown(bShown=false)
   > InitializeRundownLineItemDuration()
   Event On Toggle Spectating() {is=title K2Node_Event}
   > Update Show Scoreboard()
   > n20 = Branch(Condition=IsSpectating!(self:Fort Match XPRecap Object Reference) -> (ReturnValue:bool))
   > DestroyBattlepassInjectionAfterTimer()
   CreateBattlepassInjectionOrCancelDestruction(execute=@n20.else)
   Event AnimateXPTotals() {is=title K2Node_Event}
   > Branch(Condition="integer > integer"(A=Get ShowingPlayerLevel() -> (ShowingPlayerLevel:int=0)))
   > n21 = Update Post Match Injections()
   InitializeXPRundown(execute=[@n21.AllInjectionsUpdated, @n21.XPEverywhereUpdatedOnly]) {is=title K2Node_CallFunction}
   n22 = Event OnReturnToIslandInputAction() {is=title K2Node_Event}
   Event Destruct()
   n23 = Get MatchRecapLeftSide_ExtensionSlot() -> (MatchRecapLeftSide_ExtensionSlot:Gameplay Tag Structure)
   > UnregisterWidgetHost!(execute, WidgetHost:Widget Object Reference, WidgetHostTag:"Gameplay Tag Structure (by ref)"=@n23) -> (then)
   n24 = Get MatchRecapBottomCenter_ExtensionSlot() -> (MatchRecapBottomCenter_ExtensionSlot:Gameplay Tag Structure)
   > UnregisterWidgetHost(WidgetHostTag=@n24)
   Event BP_OnVisualizationSkipped() {is=title K2Node_Event}
   > HandleVisualizationSkipped()
   n25 = Event Add Content To Host Widget!() -> (
    OutputDelegate
    then
    WidgetHostTag:"Gameplay Tag Structure (by ref)"
    ContentWidget:Widget Object Reference
    Padding:"Margin Structure (by ref)"
    bFill:bool
   ) {is=title K2Node_Event}
   > n26 = Branch(Condition=MatchesTag!(
     TagOne:Gameplay Tag Structure=@n25.WidgetHostTag
     TagTwo:Gameplay Tag Structure=Get MatchRecapLeftSide_ExtensionSlot()
     bExactMatch:bool=true
    ) -> (ReturnValue:bool))
   > AddChildToVerticalBox(
    self=Get VerticalBox_PrimaryMatchXPExtension() -> (VerticalBox_PrimaryMatchXPExtension:Vertical Box Object Reference)
    Content=@n25.ContentWidget
   )
   Branch(execute=@n26.else, Condition=MatchesTag(TagOne=@n25.WidgetHostTag, TagTwo=Get MatchRecapBottomCenter_ExtensionSlot(), bExactMatch=true))
   > AddChildToOverlay!(
    execute
    self:Overlay Object Reference=Get Overlay_BottomCenterExtension() -> (Overlay_BottomCenterExtension:Overlay Object Reference)
    Content:Widget Object Reference=@n25.ContentWidget
   ) -> (then, ReturnValue:Overlay Slot Object Reference)
   > SetActiveWidget(self=Get VisSwitcher_BottomCenter(), Widget=Get Overlay_BottomCenterExtension())
   Event Construct()
   > RegisterWidgetHost!(execute, WidgetHost:Widget Object Reference, WidgetHostTag:"Gameplay Tag Structure (by ref)"=@n23, IsPlayerSpecific:bool=true) -> (then)
   > RegisterWidgetHost(WidgetHostTag=@n24, IsPlayerSpecific=true)
   > InitializeDisplay()
   > n27 = StartListeningToEvent!(
    execute:Exec
    Target:Gameplay Event Router Component Object Reference=GetGlobalEventRouter()
    EventType:Script Struct Object Reference=/Script/FortniteUI.MatchXPScreenEvent_FocusBattlepass
    Context:Object Reference
    EventBubblingRule:EEventBubblingRule Enum
    None:%=Self-Reference()
   ) -> (then:Exec, AsyncTaskProxy:Async Action Start Listening To Event Object Reference, OnEventReceived:Exec, Payload_IsBattlepassFocused:bool=~)
   > n28 = StartListeningToEvent!(
    execute
    Target=GetGlobalEventRouter()
    EventType=/Script/FortniteUI.MatchXPScreenEvent_ToggledFullscreenMap
    Context
    EventBubblingRule
    None:%=Self-Reference()
   ) -> (then, AsyncTaskProxy, OnEventReceived, Payload_bFullscreenMapVisible:bool=~)
   OnToggleFullscreenMap!(execute=@n28.OnEventReceived, self, bIsMapVisible:bool=@n28.Payload_bFullscreenMapVisible) -> (then) {is=title K2Node_CallFunction}
   HandleBattlepassOpened!(execute=@n27.OnEventReceived, self, IsBattlepassOpen:bool=@n27.Payload_IsBattlepassFocused) -> (then) {is=title K2Node_CallFunction}
   n29 = "To Float (Integer)"(InInt=0)
   n30 = Get Duration_ProgressAnimation_CONST()
   Broadcast Event(
    self=GetGlobalEventRouter()
    EventData:S UI Set Rundown Progress Update Duration Structure=Make S UI Set Rundown Progress Update Duration!(RundownItemDuration_2_#25:double=MapRangeClamped(Value=@n29, InRangeA=5000, InRangeB=100000, OutRangeA=@n30, OutRangeB=1.2)) -> (S_UI_SetRundownProgressUpdateDuration:S UI Set Rundown Progress Update Duration Structure) {is=title K2Node_MakeStruct}
   )
  ]}
 FindNextValidEntry {nodes [
   n1 = FindNextValidEntry!() -> (then, StartingIndex:int, StartWithGameplayWidgets:bool=~) {is=title K2Node_FunctionEntry}
   > n2 = Set CurrentIndex(CurrentIndex:int=@n1) -> (Output_Get:int=0)
   > Increment Int(Value=@n2)
   > n3 = Branch(Condition=@n1.StartWithGameplayWidgets)
   n4 = Get GameplayXPWidgets()
   > n5 = Branch(Condition="integer <= integer"(
     A=Get CurrentIndex() -> (CurrentIndex:int=0)
     B=Last Index!(TargetArray:Array of Fort Match XPTotal Entry Object References=@n4) -> (ReturnValue=0)
    ))
   > n6 = For Loop(FirstIndex=Get CurrentIndex() -> (CurrentIndex=0), LastIndex=Last Index(TargetArray=Get GameplayXPWidgets()) -> (ReturnValue=0))
   n7 = Set CurrentIndex(execute=@n6.Completed, CurrentIndex=0) -> (Output_Get=0)
   n8 = For Loop(
    execute=[@n7.then, @n14.then, @n3.else]
    FirstIndex=Get CurrentIndex() -> (CurrentIndex=0)
    LastIndex=Last Index(TargetArray=Get QuestXPWidgets()) -> (ReturnValue=0)
   )
   Return Node!(
    execute=@n8.Completed
    IsSuccess:bool=false
    IsGameplayItem:bool=false
    FoundIndex:int=0
    EntryWidget:WBP Post Match Rundown Item Object Reference
   )
   > Branch(Condition=Is Valid Index!(TargetArray:Array of Fort Match XPTotal Entry Object References=Get QuestXPWidgets(), IndexToTest:int=Get CurrentIndex() -> (CurrentIndex=0)) -> (ReturnValue:bool))
   > n9 = Set FoundWidget(FoundWidget:Fort Match XPTotal Entry Object Reference="Get (a copy)"(Array:Array of Fort Match XPTotal Entry Object References=Get QuestXPWidgets(), Dimension 1=Get CurrentIndex() -> (CurrentIndex=0)) -> (Output:Fort Match XPTotal Entry Object Reference)) -> (Output_Get:Fort Match XPTotal Entry Object Reference)
   > Branch(Condition=IsEntryValid!(self, EntryItem:"Fort Match XPTotal Entry Object Reference (by ref)"=@n9) -> (IsValid:bool))
   > n10 = Cast To WBP_PostMatch_RundownItem(Object=Get FoundWidget() -> (FoundWidget:Fort Match XPTotal Entry Object Reference))
   > Return Node(IsSuccess=true, IsGameplayItem=false, FoundIndex=Get CurrentIndex() -> (CurrentIndex=0), EntryWidget=@n10)
   Branch(execute=@n6.LoopBody, Condition=Is Valid Index(TargetArray=Get GameplayXPWidgets(), IndexToTest=@n6))
   > n11 = Set FoundWidget(FoundWidget="Get (a copy)"(Array=Get GameplayXPWidgets(), Dimension 1=Get CurrentIndex() -> (CurrentIndex=0)))
   > Branch(Condition=IsEntryValid(EntryItem=@n11))
   > n12 = Cast To WBP_PostMatch_RundownItem(Object=Get FoundWidget())
   > Return Node(IsSuccess=true, IsGameplayItem=true, FoundIndex=Get CurrentIndex() -> (CurrentIndex=0), EntryWidget=@n12)
   n13 = Get GameplayXPWidgets()
   n14 = Set CurrentIndex(
    execute=@n5.else
    CurrentIndex=int - int(A=Get CurrentIndex() -> (CurrentIndex=0), B=Length(TargetArray:Array of Fort Match XPTotal Entry Object References=@n13))
   ) -> (Output_Get=0)
  ]}
 FindPreviousValidEntry {nodes [
   n1 = FindPreviousValidEntry!() -> (then, StartingIndex:int) {is=title K2Node_FunctionEntry}
   > n2 = Set CurrentIndex(CurrentIndex=@n1) -> (Output_Get=0)
   n3 = Decrement Int(" "=[@n2.then, @n6.else], Value=@n2)
   > n4 = Branch(Condition="integer >= integer"(A=@n3))
   > n5 = GetEntryAtIndex!(execute, self, Index:int=Get CurrentIndex() -> (CurrentIndex=0)) -> (then, IsValid:bool, EntryWidget:Fort Match XPTotal Entry Object Reference)
   > n6 = Branch(Condition=@n5)
   > Return Node!(
    execute
    IsSuccess=true
    FoundIndex=Get CurrentIndex() -> (CurrentIndex=0)
    EntryWidget:Fort Match XPTotal Entry Object Reference=@n5.EntryWidget
   )
   Return Node(execute=@n4.else, IsSuccess=false, FoundIndex=0)
  ]}
 GetCategoryFromXPEvent {nodes [
   n1 = GetCategoryFromXPEvent!() -> (
    then
    Event Data_OldXPValue:int
    Event Data_OldLevel:int
    Event Data_NewXPValue:int
    Event Data_NewLevel:int
    Event Data_EventXpValue:int
    Event Data_RemainingRestXP:int
    Event Data_RestedXpConsumed:int
    Event Data_ExperienceCategory:EFortExperienceCategory Enum
   ) {is=title K2Node_FunctionEntry}
   > Return Node!(execute, Category:EFortExperienceCategory Enum=@n1.Event Data_ExperienceCategory)
  ]}
 GetEntryAtIndex {nodes [
   n1 = GetEntryAtIndex!() -> (then, Index:int) {is=title K2Node_FunctionEntry}
   n2 = Select(
    Option 0:Array of Fort Match XPTotal Entry Object References=Get QuestXPWidgets()
    Option 1:Array of Fort Match XPTotal Entry Object References=Get GameplayXPWidgets()
    Index=Get QuestsFinished()
   ) -> (ReturnValue:Array of Fort Match XPTotal Entry Object References)
   > n3 = Branch(Condition=Is Valid Index(TargetArray=@n2, IndexToTest=@n1))
   n4 = "Get (a copy)"(Array=@n2, Dimension 1=Get Index() -> (Index:int=0))
   > n5 = Branch(Condition=IsEntryValid(EntryItem=@n4))
   > Return Node!(execute, IsValid:bool=true, EntryWidget:Fort Match XPTotal Entry Object Reference=@n4)
   Return Node(execute=[@n5.else, @n3.else], IsValid=false)
  ]}
 HandleVisualizationSkipped {nodes [
   HandleVisualizationSkipped() {is=title K2Node_FunctionEntry}
   > Set IsSkippingVisualization(IsSkippingVisualization:bool=true) -> (Output_Get:bool)
   > n1 = Sequence!(execute) -> (then_0, then_1, then_2, then_3, then_4:Exec, then_5:Exec)
   > StopAllAnimations(self:User Widget Object Reference)
   > StopAllAnimations(self=Get WBP_PostMatch_RundownTotal())
   > SetXPValue!(
    execute
    self:Fort Match XPTotal Entry Object Reference=Get WBP_PostMatch_RundownTotal()
    XPValue:int=Get XPValue!(self:Fort Match XPTotal Entry Object Reference=Get MatchXP_TotalEntry()) -> (XPValue:int=0)
   ) -> (then)
   > StopAllAudio()
   n2 = Cast To WBP_PostMatch_LevelUpScreen!(
    execute=@n1.then_1
    Object=GetChildAt!(self:Panel Widget Object Reference=Get NSlot_LevelUpInjection() -> (NSlot_LevelUpInjection:Named Slot Object Reference), Index:int=0) -> (ReturnValue:Widget Object Reference)
   ) -> (then, CastFailed, AsWBP Post Match Level Up Screen:WBP Post Match Level Up Screen Object Reference) {is=title K2Node_DynamicCast}
   > Shutdown(self:WBP Post Match Level Up Screen Object Reference=@n2)
   OnLastXPAnimationFinished(execute=@n1.then_2)
   SetXPEverywhereToFinalState(execute=@n1.then_3)
   > Clear and Invalidate Timer by Handle(Handle=Get XPRecapTimer() -> (XPRecapTimer:Timer Handle Structure))
   n3 = Get Anim_Finish()
   n4 = GetEndTime(self=@n3)
   PlayAnimationTimeRange!(
    execute=@n1.then_4
    self:User Widget Object Reference
    InAnimation:Widget Animation Object Reference=@n3
    StartAtTime:float=float - float(A=@n4, B=0.01)
    EndAtTime:float=@n4
    NumLoopsToPlay:int=1
    PlayMode:EUMGSequencePlayMode Enum
    PlaybackSpeed:float=1
    bRestoreState:bool=false
   ) -> (then, ReturnValue:Widget Animation Handle Structure)
   > SetRenderOpacity(
    self=[
     Get SBox_CenterContent() -> (SBox_CenterContent:Size Box Object Reference)
     Get GPanel_XPRundown() -> (GPanel_XPRundown:Grid Panel Object Reference)
    ]
    InOpacity=1
   )
   Set IsResettingAnimation(execute=@n1.then_5, IsResettingAnimation=false)
   > HandleAllVizualizationsFinished() {is=title K2Node_CallFunction}
  ]}
 InitializeDisplay {nodes [
   InitializeDisplay() {is=title K2Node_FunctionEntry}
   > SetRenderOpacity(self=Get GPanel_XPRundown(), InOpacity=0)
  ]}
 InitializeRundownLineItemDuration {nodes [
   InitializeRundownLineItemDuration() {is=title K2Node_FunctionEntry}
   > n1 = GetConsoleVariableFloatValue!(execute, VariableName:String=Athena.MatchXPScreenWidget.S39XPRundownAnim_LineItemDuration) -> (then, ReturnValue:float)
   > Branch(Condition="float > float"(A=@n1))
   > Set Duration_ProgressAnimation_CONST(Duration_ProgressAnimation_CONST:double=@n1) -> (Output_Get:double)
  ]}
 InitializeXP {nodes [
   InitializeXP() {is=title K2Node_FunctionEntry}
   > Set TotalGameplayXP(TotalGameplayXP:int=Get XPValue(self=Get MatchXP_GameplayXp()) -> (XPValue=0)) -> (Output_Get:int=0)
   > Set TotalQuestXP(TotalQuestXP:int=Get XPValue(self=Get MatchXP_ChallengeXp()) -> (XPValue=0)) -> (Output_Get:int=0)
   > SetXPValue(self=[Get MatchXP_ChallengeXp(), Get MatchXP_GameplayXp()], XPValue=0)
   > SetRenderOpacity(self=Get Ovr_QuestTotals() -> (Ovr_QuestTotals:Overlay Object Reference), InOpacity=0)
   n1 = Get TotalCompletedQuests() -> (TotalCompletedQuests:int=0)
   > SetVisibility(
    self=Get Ovr_QuestTotals()
    InVisibility=Select(Option 0:ESlateVisibility Enum=Collapsed, Option 1:ESlateVisibility Enum=SelfHitTestInvisible, Index="integer > integer"(A=@n1)) -> (ReturnValue:ESlateVisibility Enum)
   )
   > "SetText (Text)"(
    self=Get Txt_CompletedQuests() -> (Txt_CompletedQuests:Common Text Object Reference)
    InText="To Text (Integer)"(Value=Get TotalCompletedQuests() -> (TotalCompletedQuests=0))
   )
   > "SetText (Text)"(
    self=Get Txt_InProgressQuests() -> (Txt_InProgressQuests:Common Text Object Reference)
    InText="To Text (Integer)"(Value=Get TotalInProgressQuests() -> (TotalInProgressQuests:int=0))
   )
   > SetXPValue(self=[Get MatchXP_TotalEntry(), Get WBP_PostMatch_RundownTotal()], XPValue=Get XPValue(self=Get MatchXP_TotalEntry()) -> (XPValue=0))
   > SetXPValue(self=Get MatchXP_UntilNextLvlReport(), XPValue=Get TotalXPRemainingUntilNextLevel() -> (TotalXPRemainingUntilNextLevel:int=0))
   n2 = Get ShowingPlayerXP_Current() -> (ShowingPlayerXP_Current:int=0)
   n3 = Get ShowingPlayerLevel() -> (ShowingPlayerLevel=0)
   > Broadcast Event(
    self=GetGlobalEventRouter()
    EventData:XPEverywhere Event Manually Initialize Bar Structure=Make XPEverywhere Event Manually Initialize Bar!(
     InitializeData_OldXPValue:int=@n2
     InitializeData_OldLevel:int=@n3
     InitializeData_NewXPValue:int=@n2
     InitializeData_NewLevel:int=@n3
     InitializeData_EventXpValue:int=@n2
     InitializeData_RemainingRestXP:int=Get ShowingPlayerXP_Remaining() -> (ShowingPlayerXP_Remaining:int=0)
     InitializeData_RestedXpConsumed:int=0
     InitializeData_ExperienceCategory:EFortExperienceCategory Enum=None
     RundownLineItemDuration:float=Get Duration_ProgressAnimation_CONST()
    ) -> (XPEverywhereEvent_ManuallyInitializeBar:XPEverywhere Event Manually Initialize Bar Structure) {is=title K2Node_MakeStruct}
    EventContext=Self-Reference()
   )
  ]}
 IsEntryValid {nodes [
   n1 = IsEntryValid!() -> (then, EntryItem:"Fort Match XPTotal Entry Object Reference (by ref)") {is=title K2Node_FunctionEntry}
   > n2 = Is Valid(InputObject=@n1)
   Return Node!(execute=@n2.Is Not Valid, IsValid=false)
   > Return Node(IsValid="integer > integer"(A=GetXPValue!(self:Fort Match XPTotal Entry Object Reference=@n1) -> (ReturnValue:int=0), B=0) {is=title K2Node_CallFunction})
  ]}
 IsWidgetClassUnique {nodes [
   n1 = IsWidgetClassUnique!() -> (then, ExistingWidgets:Array of Widget Object References, NewWidget:Widget Object Reference) {is=title K2Node_FunctionEntry}
   > n2 = For Each Loop(Array:Array of Widget Object References=@n1)
   Return Node!(execute=@n2.Completed, IsUnique:bool=true)
   n3 = Get NewWidget() -> (NewWidget:Widget Object Reference)
   > Branch(Condition=IsA!(Object:Object Reference=@n2, ObjectClass:Object Class Reference=Get Class(Object=@n3) -> (ReturnValue:Widget Class Reference)) -> (ReturnValue:bool))
   > Return Node(IsUnique=false)
  ]}
 MakeXPEverywherePushEvent {nodes [
   n1 = MakeXPEverywherePushEvent!() -> (then, EventData:Fort XPEvent Entry Structure) {is=title K2Node_FunctionEntry}
   > Return Node!(execute, PushEvent_EventData:Fort XPEvent Entry Structure=@n1)
  ]}
 OnLastXPAnimationFinished {nodes [
   OnLastXPAnimationFinished() {is=title K2Node_FunctionEntry}
   > n1 = Branch(Condition=Get IsSkippingVisualization())
   > OnAllXPAnimationsFinished() {is=title K2Node_CallFunction}
   PlayGameplayRundown_VIZ!(execute=@n1.else, self, IsEnding:bool=true) -> (then)
   > Set Timer by Event(Delegate=Create Event(), Time=Get Delay_PostAfterRundownPreTransition() -> (Delay_PostAfterRundownPreTransition:float), bLooping=false)
  ]}
 OnNavigateFromBattlepass {nodes [OnNavigateFromBattlepass() {is=title K2Node_FunctionEntry}]}
 RemoveContentFromWidgetHost {nodes [
   n1 = Remove Content From Host Widget!() -> (
    then
    WidgetHostTag:"Gameplay Tag Structure (by ref)"
    ContentWidget:Widget Object Reference
    RemovalType:EWidgetSlottingRemovalType Enum=Unparent
   ) {is=title K2Node_FunctionEntry}
   > Branch(Condition=MatchesTag(TagOne=@n1, TagTwo=Get MatchRecapLeftSide_ExtensionSlot(), bExactMatch=true))
   > n2 = Switch on EWidgetSlottingRemovalType!(execute, Selection:EWidgetSlottingRemovalType Enum=@n1.RemovalType) -> (Unparent:Exec, Collapse:Exec) {is=title K2Node_SwitchEnum}
   SetVisibility(execute=@n2.Collapse, self=@n1.ContentWidget, InVisibility=Collapsed)
   n3 = RemoveChild(execute=@n2.Unparent, self=Get VerticalBox_PrimaryMatchXPExtension(), Content=@n1.ContentWidget)
   > Return Node!(execute, ReturnValue:bool=@n3)
  ]}
 ResetPostMatch {nodes [
   ResetPostMatch() {is=title K2Node_FunctionEntry}
   > PopContentWidget!(execute) -> (then)
   > RecoverFromPostGameMenu(self:Athena HUDContext Object Reference=Get AthenaHUDContext!() -> (ReturnValue:Athena HUDContext Object Reference) {is=title K2Node_GetBlueprintContextEx})
   n1 = Get FortHUDContext!() -> (ReturnValue:Fort HUDContext Object Reference) {is=title K2Node_GetBlueprintContextEx}
   > SetCursorModeLocked!(execute, self:Fort HUDContext Object Reference=@n1, bLocked:bool=false) -> (then)
   > EnterCameraMode(self:Fort HUDContext Object Reference=@n1)
  ]}
 SetXPAnimationActive {nodes [
   n1 = SetXPAnimationActive!() -> (then, IsAnimationActive:bool=~) {is=title K2Node_FunctionEntry}
   > n2 = Set IsXpAnimationActive(IsXpAnimationActive:bool=@n1) -> (Output_Get:bool)
   > n3 = Branch(Condition=@n2)
   n4 = Get AnimationSequenceTimerHandle() -> (AnimationSequenceTimerHandle:Timer Handle Structure)
   > Unpause Timer by Handle!(execute, Handle:Timer Handle Structure=@n4) -> (then)
   Pause Timer by Handle!(execute=@n3.else, Handle:Timer Handle Structure=@n4) -> (then)
  ]}
 SetXPEverywhereToFinalState {nodes [
   SetXPEverywhereToFinalState() {is=title K2Node_FunctionEntry}
   n1 = Get FinalCurrentXP() -> (FinalCurrentXP:int=0)
   n2 = Get FinalLevelDisplay() -> (FinalLevelDisplay:int=0)
   > Broadcast Event(
    self=GetGlobalEventRouter()
    EventData=Make XPEverywhere Event Manually Initialize Bar(
     InitializeData_OldXPValue=@n1
     InitializeData_OldLevel=@n2
     InitializeData_NewXPValue=@n1
     InitializeData_NewLevel=@n2
     InitializeData_EventXpValue=@n1
     InitializeData_RemainingRestXP=Get FinalRemainingXP() -> (FinalRemainingXP:int=0)
     InitializeData_RestedXpConsumed=0
     InitializeData_ExperienceCategory=None
     RundownLineItemDuration=Get Duration_ProgressAnimation_CONST()
    )
    EventContext=Self-Reference()
   )
  ]}
 ShouldVisualizeEntry {nodes [
   n1 = ShouldVisualizeEntry!() -> (then, XPEntry:Fort Match XPTotal Entry Object Reference) {is=title K2Node_FunctionEntry}
   n2 = GetXPValue(self=@n1) -> (ReturnValue=0)
   > Return Node!(execute, ShouldVisualize:bool="integer > integer"(A=@n2) {is=title K2Node_PromotableOperator}, XPTotal:int=@n2)
  ]}
 StopAllAudio {nodes [
   StopAllAudio() {is=title K2Node_FunctionEntry}
   > n1 = For Each Loop(Array:Array of Audio Component Object References=Get XPEntry_Ticker_Components()) -> (Array Element:Audio Component Object Reference)
   n2 = Get Recap_Outro_NoLevelUp_Component() -> (Recap_Outro_NoLevelUp_Component:Audio Component Object Reference)
   Is Valid(exec=@n1.Completed, InputObject=@n2)
   > FadeOut!(
    execute
    self:Audio Component Object Reference=@n2
    FadeOutDuration:float=0.2
    FadeVolumeLevel:float=0
    FadeCurve:EAudioFaderCurve Enum=Linear
   ) -> (then)
   Is Valid(exec=@n1.LoopBody, InputObject=@n1)
   > FadeOut(self=@n1, FadeOutDuration=0.2, FadeVolumeLevel=0, FadeCurve=Linear)
  ]}
 Update Show Scoreboard {nodes [
   Update Show Scoreboard() {is=title K2Node_FunctionEntry}
   > n1 = Sequence!(execute) -> (then_0, then_1, then_2)
   n2 = GetMinigameSettingsMachine!(ContextActor:Actor Object Reference=Get Minigame() -> (Minigame:Fort Minigame Object Reference)) -> (ReturnValue:Fort Minigame Settings Building Object Reference)
   > Is Valid(InputObject=@n2)
   n3 = Break Minigame Post Game Settings!(MinigamePostGameSettings:"Minigame Post Game Settings Structure (by ref)"=GetPostGameSettings!(self:Fort Minigame Settings Building Object Reference=@n2) -> (ReturnValue:"Minigame Post Game Settings Structure (by ref)")) -> (
    PostGameType:EMinigameGameEndPostGameType Enum
    PostGameDelay:float=~
    bCustomShowScoreboard:bool=~
    VictoryAnimationSettings:Minigame Post Game Animation Settings Structure
    DefeatAnimationSettings:Minigame Post Game Animation Settings Structure
    TieAnimationSettings:Minigame Post Game Animation Settings Structure
   ) {is=title K2Node_BreakStruct}
   n4 = IsSpectating()
   > Set bShowScoreboard(bShowScoreboard:bool=AND Boolean!(
     A="Equal (Enum)"(A:EMinigameGameEndPostGameType Enum=@n3, B:EMinigameGameEndPostGameType Enum=Custom)
     B=@n3.bCustomShowScoreboard
     C:bool=NOT Boolean(A=@n4)
    ) -> (ReturnValue)) -> (Output_Get:bool)
   n5 = Branch(execute=@n1.then_1, Condition=Get bShowScoreboard() -> (bShowScoreboard:bool))
   n6 = Get MatchXpScoreboard() -> (MatchXpScoreboard:WBP Match XPScoreboard Object Reference)
   > DoShowUI!(execute:Exec, self:Object Reference=@n6) -> (then:Exec) {is=title K2Node_Message}
   DoHideUI!(execute:Exec=@n5.else, self:Object Reference=@n6) -> (then:Exec) {is=title K2Node_Message}
   Return Node!(execute=@n1.then_2, ShowingScoreboard:bool=Get bShowScoreboard())
  ]}
 UpdateNextEntry {nodes [
   UpdateNextEntry() {is=title K2Node_FunctionEntry}
   > n1 = GetEntryAtIndex(Index=Get CurrentXpEntryIndex() -> (CurrentXpEntryIndex=0)) {is=title K2Node_CallFunction}
   > n2 = Branch(Condition=@n1)
   > Set CurrentEntryWidget(CurrentEntryWidget:Fort Match XPTotal Entry Object Reference=@n1.EntryWidget) -> (Output_Get:Fort Match XPTotal Entry Object Reference)
   > Return Node!(execute, WasFound:bool=true)
   n3 = FindPreviousValidEntry!(execute=@n2.else, self, StartingIndex:int=Get CurrentXpEntryIndex() -> (CurrentXpEntryIndex=0)) -> (then, IsSuccess:bool, FoundIndex:int=0, EntryWidget:Fort Match XPTotal Entry Object Reference) {is=title K2Node_CallFunction}
   > n4 = Branch(Condition=@n3)
   > Set CurrentXpEntryIndex(CurrentXpEntryIndex:int=@n3.FoundIndex) -> (Output_Get:int=0)
   > Set CurrentEntryWidget(CurrentEntryWidget=@n3.EntryWidget)
   > Return Node(WasFound=true)
   Return Node(execute=@n4.else, WasFound=false)
  ]}
 UpdatePostMatchInjections {nodes [
   n1 = Inputs!() -> (In, IsShuttingDown:bool=~)
   > n2 = Branch(Condition=AND Boolean!(A=Get bIsUsingPostmatchUpdates() -> (bIsUsingPostmatchUpdates:bool), B=NOT Boolean(A=@n1)) -> (ReturnValue))
   > Branch(Condition=NOT Boolean(A=Get IsPostmatchInitialized()))
   > Add Alias XPEverywhere!(In:Exec) -> (XPEverywhereInitialized:Exec) {is=title K2Node_MacroInstance}
   > n3 = Branch(Condition=NOT Boolean(A=Get bIsSuppressingBattlepass()))
   > Add Alias Battlepass()
   > SetPadding!(
    execute
    self:Overlay Slot Object Reference=SlotAsOverlaySlot(Widget=Get GPanel_XPRundown())
    InPadding_Left:float=0
    InPadding_Top:float=0
    InPadding_Right:float=0
    InPadding_Bottom:float=80
   ) -> (then)
   > Outputs!(AllInjectionsUpdated:Exec, XPEverywhereUpdatedOnly:Exec=@n4.then, AllCleared:Exec=@n5.then)
   n4 = SetPadding(
    execute=@n3.else
    self=SlotAsOverlaySlot(Widget=Get GPanel_XPRundown())
    InPadding_Left=0
    InPadding_Top=0
    InPadding_Right=0
    InPadding_Bottom=0
   )
   n5 = ClearChildren(execute=@n2.else, self=[Get NSlot_BattlepassInjection(), Get NSlot_XPEverywhere()])
  ]}
 Visualizations {nodes [
   n1 = PlayGameplayRundown_VIZ!() -> (OutputDelegate, then, IsEnding:bool=~) {is=title K2Node_CustomEvent}
   > Branch(Condition="integer > integer"(A=Get TotalGameplayXP() -> (TotalGameplayXP:int=0)))
   > n2 = Branch(Condition=@n1.IsEnding)
   > PlayAnimation(PlayMode=Reverse)
   > PlayOutro(self:Fort Match XPTotal Entry Object Reference=Get MatchXP_GameplayXp())
   PlayAnimation(execute=@n2.else)
   > SetAnimatedXPValue!(execute, self:Fort Match XPTotal Entry Object Reference=Get MatchXP_GameplayXp(), XPValue:int=0) -> (then)
   n3 = PlayQuestRundown_VIZ!() -> (OutputDelegate, then, IsEnding:bool=~) {is=title K2Node_CustomEvent}
   > Branch(Condition="integer > integer"(A=Get TotalQuestXP() -> (TotalQuestXP:int=0)))
   > n4 = Branch(Condition=@n3.IsEnding)
   > PlayAnimation(PlayMode=Reverse)
   > PlayOutro(self=Get MatchXP_ChallengeXp())
   > Branch(Condition="integer > integer"(A=Get TotalCompletedQuests() -> (TotalCompletedQuests=0)))
   > PlayAnimation(InAnimation=Get Anim_ShowQuestDetails() -> (Anim_ShowQuestDetails:Widget Animation Object Reference))
   PlayAnimation(execute=@n4.else)
   > SetAnimatedXPValue(self=Get MatchXP_ChallengeXp(), XPValue=0)
   PlayNextEntry_VIZ() {is=title K2Node_CustomEvent}
   > Delay(Duration=Get Delay_BetweenXPEntries() -> (Delay_BetweenXPEntries:float))
   n5 = Get GameplayIntroPlayed() -> (GameplayIntroPlayed:bool)
   n6 = Get IsSkippingVisualization()
   > n7 = Branch(Condition=AND Boolean!(A=Get QuestsFinished(), B=NOT Boolean(A=@n5), C:bool=NOT Boolean(A=@n6)) -> (ReturnValue))
   > BeginGameplayRundown_VIZ()
   > n8 = Delay(Duration=Get Delay_PostCategoryNumPreCategoryItems() -> (Delay_PostCategoryNumPreCategoryItems:float))
   n9 = UpdateNextEntry!(execute=[@n8.then, @n7.else], self) -> (then, WasFound:bool) {is=title K2Node_CallFunction}
   > n10 = Branch(Condition=@n9)
   n11 = ShouldVisualizeEntry!(XPEntry:Fort Match XPTotal Entry Object Reference=Get CurrentEntryWidget() -> (CurrentEntryWidget:Fort Match XPTotal Entry Object Reference)) -> (ShouldVisualize:bool, XPTotal:int=0) {is=title K2Node_CallFunction}
   > n12 = Branch(Condition=@n11)
   n13 = Get CurrentEntryWidget()
   > PlayIntro(self:Fort Match XPTotal Entry Object Reference=@n13)
   > BindToIntro!(execute, self, Object:Object Reference=@n13) -> (then) {is=title K2Node_CallFunction}
   > ApplyFadeToLowestEntries() {is=title K2Node_CallFunction}
   > Delay(Duration=Get Delay_PostEntryPreRadialProgress() -> (Delay_PostEntryPreRadialProgress:float))
   > n14 = PushManualRecapXPEvents!(
    execute
    self:Fort Match XPRecap Object Reference
    DeltaXP:int=@n11.XPTotal
    XPType:EFortExperienceCategory Enum=Select(Option 0:EFortExperienceCategory Enum=Quest, Option 1:EFortExperienceCategory Enum=Gameplay, Index=Get QuestsFinished()) -> (ReturnValue:EFortExperienceCategory Enum)
   ) -> (then)
   n15 = FindPreviousValidEntry(execute=[@n14.then, @n12.else], StartingIndex=Get CurrentXpEntryIndex() -> (CurrentXpEntryIndex=0)) -> (FoundIndex=0)
   > Set CurrentXpEntryIndex(CurrentXpEntryIndex=@n15.FoundIndex) -> (Output_Get=0)
   > n16 = Branch(Condition=@n15)
   > n17 = Set IsXpAnimationActive(IsXpAnimationActive=true)
   AudioAddTickerComponent!(execute=[@n17.then, @n19.then], self, IsFinal:bool=false) -> (then) {is=title K2Node_CallFunction}
   n18 = Branch(execute=[@n16.else, @n10.else], Condition=NOT Boolean(A=Get QuestsFinished()))
   > Set CurrentXpEntryIndex(CurrentXpEntryIndex=Last Index(TargetArray=Get GameplayXPWidgets()) -> (ReturnValue=0)) -> (Output_Get=0)
   > Set QuestsFinished(QuestsFinished:bool=true) -> (Output_Get:bool)
   > n19 = Set IsXpAnimationActive(IsXpAnimationActive=true)
   Set IsXpAnimationActive(execute=@n18.else, IsXpAnimationActive=false)
   n20 = Get GameplayXPWidgets()
   > Branch(Condition="integer > integer"(A=Length(TargetArray=@n20)))
   > AudioAddTickerComponent(IsFinal=true)
   BeginGameplayRundown_VIZ() {is=title K2Node_CustomEvent}
   > n21 = Branch(Condition="Not Equal (Integer)"(A=Get TotalCompletedQuests() -> (TotalCompletedQuests=0)))
   > n22 = PlayQuestRundown_VIZ!(execute, self, IsEnding:bool=true) -> (then) {is=title K2Node_CallFunction}
   n23 = Branch(execute=[@n22.then, @n21.else], Condition="integer > integer"(A=Get TotalGameplayXP() -> (TotalGameplayXP=0)))
   > PlayGameplayRundown_VIZ(IsEnding=false) {is=title K2Node_CallFunction}
   > Set GameplayIntroPlayed(GameplayIntroPlayed=true)
   OnLastXPAnimationFinished(execute=@n23.else) {is=title K2Node_CallFunction}
   n24 = RingUpXP_VIZ!() -> (OutputDelegate, then, Type:EFortExperienceCategory Enum=None, XP:int) {is=title K2Node_CustomEvent}
   n25 = Get MatchXP_GameplayXp()
   n26 = Select!(
    None:Fort Match XPTotal Entry Object Reference=@n25
    Gameplay:Fort Match XPTotal Entry Object Reference=@n25
    Quest:Fort Match XPTotal Entry Object Reference=Get MatchXP_ChallengeXp()
    Index:EFortExperienceCategory Enum=@n24.Type
   ) -> (ReturnValue:Fort Match XPTotal Entry Object Reference)
   > SetAnimatedXPValue(self=@n26, XPValue=int + int(A=@n24.XP, B=Get XPValue(self=@n26) -> (XPValue=0)))
   XPTotalsFinished_VIZ() {is=title K2Node_CustomEvent}
   > PlayAnimation(InAnimation=Get Anim_IntroTransition())
   > PlayRecapAnimation(self:WBP Post Match Rundown Total Object Reference=Get WBP_PostMatch_RundownTotal())
   > PlaySound2D(Sound=Get XPEntry_ShowTotal() -> (XPEntry_ShowTotal:Sound Base Object Reference))
   > n27 = Branch(Condition="integer > integer"(A=Get LevelsGained() -> (LevelsGained:int=0)))
   n28 = GetEndTime(self=Get Anim_IntroTransition())
   > Delay(Duration=@n28)
   > Branch(Condition=NOT Boolean(A=Get IsSkippingVisualization()))
   > n29 = CreateWidgetAsync(
    UserWidgetSoftClass=&/WBP_PostMatch_LevelUpScreen.WBP_PostMatch_LevelUpScreen_C
    OwningPlayer=GetOwningPlayer()
    bSuspendInputUntilComplete=true
   )
   n30 = Cast To WBP_PostMatch_LevelUpScreen(execute=@n29.OnComplete, Object=@n29)
   > Bind Event to On Visualization Finished(self:WBP Post Match Level Up Screen Object Reference=@n30, Delegate=Create Event()) {is=title K2Node_AddDelegate}
   > SetNumberOfLevels!(execute, self:WBP Post Match Level Up Screen Object Reference=@n30, NumberOfLevels:int=Get LevelsGained() -> (LevelsGained=0)) -> (then)
   > AddChild(self=Get NSlot_LevelUpInjection(), Content=@n29)
   > PlayAnimation(InAnimation=Get Anim_HideXPEverywhere() -> (Anim_HideXPEverywhere:Widget Animation Object Reference))
   Delay(execute=@n27.else, Duration=@n28)
   > Branch(Condition=NOT Boolean(A=Get IsSkippingVisualization()))
   > RundownFinished_VIZ()
   RundownFinished_VIZ() {is=title K2Node_CustomEvent}
   n31 = Get Anim_Finish()
   > PlayAnimation(InAnimation=@n31)
   > Set IsResettingAnimation(IsResettingAnimation=false)
   > BindToAnimationFinished(Animation=@n31, Delegate=Create Event())
   > n32 = Branch(Condition="integer > integer"(A=Get LevelsGained() -> (LevelsGained=0)))
   > PlayAnimation(InAnimation=Get Anim_HideXPEverywhere(), PlayMode=Reverse)
   n33 = SpawnSound2D(
    execute=@n32.else
    Sound=Get Recap_Outro_NoLevelUp() -> (Recap_Outro_NoLevelUp:Sound Base Object Reference)
    VolumeMultiplier=1
    PitchMultiplier=1
    StartTime=0
    bPersistAcrossLevelTransition=false
    bAutoDestroy=true
   )
   > Set Recap_Outro_NoLevelUp_Component(Recap_Outro_NoLevelUp_Component:Audio Component Object Reference=@n33) -> (Output_Get:Audio Component Object Reference)
   Intro_VIZ() {is=title K2Node_CustomEvent}
   > PlayAnimation(InAnimation=Get Anim_Intro() -> (Anim_Intro:Widget Animation Object Reference))
   > n34 = SpawnSound2D(
    Sound=Get Recap_Intro() -> (Recap_Intro:Sound Base Object Reference)
    VolumeMultiplier=1
    PitchMultiplier=1
    StartTime=0
    bPersistAcrossLevelTransition=false
    bAutoDestroy=true
   )
   > Set Recap_Intro_Component(Recap_Intro_Component:Audio Component Object Reference=@n34) -> (Output_Get:Audio Component Object Reference)
   n35 = Get Anim_Intro()
   > Delay(Duration=float + float(A=Get Delay_PostIntroPreCategoryIntros() -> (Delay_PostIntroPreCategoryIntros:double), B=GetEndTime(self=@n35)))
   > PlayIntro(self=[Get MatchXP_GameplayXp(), Get MatchXP_ChallengeXp()])
   > KickoffXPRundown_VIZ()
   n36 = BattlepassDeactivated_VIZ!() -> (OutputDelegate, then, bSkipAnimation:bool=~) {is=title K2Node_CustomEvent}
   n37 = Get Anim_BattlepassFocused() -> (Anim_BattlepassFocused:Widget Animation Object Reference)
   n38 = Get Anim_BattlepassFocused()
   > PlayAnimation(InAnimation=@n38, StartAtTime=SelectFloat(A=GetEndTime(self=@n37), B=0, bPickA=@n36.bSkipAnimation), PlayMode=Reverse)
   BattlepassActivated_VIZ() {is=title K2Node_CustomEvent}
   > PlayAnimation(InAnimation=@n38)
   ProgressEventsFinished_VIZ() {is=title K2Node_CustomEvent}
   > n39 = Branch(Condition=Get IsXpAnimationActive() -> (IsXpAnimationActive:bool))
   > DelayUntilNextTick!(execute) -> (then)
   > PlayNextEntry_VIZ() {is=title K2Node_CallFunction}
   OnLastXPAnimationFinished(execute=@n39.else)
   KickoffXPRundown_VIZ() {is=title K2Node_CustomEvent}
   > Set QuestsFinished(QuestsFinished="Equal (Integer)"(A=Get TotalCompletedQuests() -> (TotalCompletedQuests=0)))
   n40 = Get QuestXPWidgets()
   n41 = Get QuestsFinished()
   n42 = Get GameplayXPWidgets()
   > Set CurrentXpEntryIndex(CurrentXpEntryIndex=Last Index(TargetArray=Select!(
      Option 0:Array of Fort Match XPTotal Entry Object References=@n40
      Option 1:Array of Fort Match XPTotal Entry Object References=@n42
      Index:bool=@n41
     ) -> (ReturnValue:Array of Fort Match XPTotal Entry Object References)) -> (ReturnValue=0)) -> (Output_Get=0)
   > Delay(Duration=Get Delay_PostCategoryIntros_PreCategoryNum() -> (Delay_PostCategoryIntros_PreCategoryNum:float))
   > n43 = Branch(Condition=NOT Boolean(A=Get QuestsFinished()))
   > PlayQuestRundown_VIZ(IsEnding=false)
   > n44 = Delay(Duration=Get Delay_PostCategoryNumPreCategoryItems())
   PlayNextEntry_VIZ(execute=[@n44.then, @n43.else])
   n45 = ProgressEvent_VIZ!() -> (OutputDelegate, then, DeltaStep:int, XPEvent:Fort XPEvent Entry Structure, Duration:double=~) {is=title K2Node_CustomEvent}
   > n46 = Branch(Condition=NOT Boolean(A=Get IsSkippingVisualization()))
   > n47 = Branch(Condition=Is Timer Active by Handle(Handle=Get XPRecapTimer()))
   n48 = Get XPRecapTimer()
   > n49 = Set RundownDuration(RundownDuration:double=float + float(
     A=Get Timer Remaining Time by Handle!(Handle:Timer Handle Structure=@n48) -> (ReturnValue:float) {is=title K2Node_CallFunction}
     B=@n45.Duration
    )) -> (Output_Get:double)
   n50 = Set Timer by Event(execute=[@n49.then, @n51.then], Delegate=Create Event(), Time=Get RundownDuration() -> (RundownDuration:double), bLooping=false)
   > Set XPRecapTimer(XPRecapTimer:Timer Handle Structure=@n50) -> (Output_Get:Timer Handle Structure)
   n51 = Set RundownDuration(execute=@n47.else, RundownDuration=@n45.Duration)
   Clear and Invalidate Timer by Handle(execute=@n46.else, Handle=Get XPRecapTimer())
   Get Anim_RundownQuestXPCategory() -> (Anim_RundownQuestXPCategory:Widget Animation Object Reference)
   Get Anim_RundownQuestXPCategory()
   Get Anim_RundownGameplayXPCategory() -> (Anim_RundownGameplayXPCategory:Widget Animation Object Reference)
   Get Anim_RundownGameplayXPCategory()
  ]}
]
variables [
 AnimationSequenceTimerHandle {type=TimerHandle}
 BattlepassInjectionDestructionTimer {type=TimerHandle}
 CurrentEntryWidget {type=FortMatchXPTotalEntry, default=null}
 CurrentXpEntryIndex {type=integer}
 Delay_BetweenXPEntries {type=number, default=0.05}
 Delay_PostAfterRundownPreTransition {type=number, default=1.25}
 Delay_PostCategoryIntros_PreCategoryNum {type=number, default=0.5}
 Delay_PostCategoryNumPreCategoryItems {type=number}
 Delay_PostEntryPreRadialProgress {type=number}
 Delay_PostIntroPreCategoryIntros {type=number, default=0.5}
 Duration_ProgressAnimation_CONST {type=number, default=1.4}
 GameplayIntroPlayed {type=boolean}
 IsPostmatchInitialized {type=boolean}
 IsResettingAnimation {type=boolean}
 IsSeasonPassCreated {type=boolean}
 IsSkippingVisualization {type=boolean}
 IsUsingS39PostmatchUpdates {type=boolean}
 IsXpAnimationActive {type=boolean}
 IsXPEverywhereInitialized {type=boolean}
 MatchRecapBottomCenter_ExtensionSlot {type=GameplayTag, default {tagName=DynamicUI.HUD.Fort.Athena.PostMatch.BottomCenter}}
 MatchRecapLeftSide_ExtensionSlot {type=GameplayTag, default {tagName=DynamicUI.HUD.Fort.Athena.PostMatch.Slot}}
 MaxEntriesToDisplay {type=integer, default=4}
 Minigame {type=FortMinigame, default=null}
 OnNavigateFromBattlepass {}
 PARENT_PADDING {type=Margin, default=<ltrb 0 0 0 0>}
 QuestRundownItems {type="array<WBP_PostMatch_RundownItem_C>", default []}
 QuestsFinished {type=boolean}
 QueueAnimateXPTotals {type=boolean}
 Recap_Intro {type=SoundBase, default=/Game/Sounds/Fort_UI_Sounds/Gameplay/XP/MSS_PostMatch_XP_Reveal}
 Recap_Intro_Component {type=AudioComponent, default=null}
 Recap_Outro_NoLevelUp {type=SoundBase, default=/Game/Sounds/Fort_UI_Sounds/Gameplay/XP/MSS_PostMatch_NoLevel_Out}
 Recap_Outro_NoLevelUp_Component {type=AudioComponent, default=null}
 RundownDuration {type=number}
 TotalGameplayXP {type=integer}
 TotalQuestXP {type=integer}
 XPEntry_ShowTotal {type=SoundBase, default=/Game/Sounds/Fort_UI_Sounds/Gameplay/XP/MSS_PostMatch_XPEntry_Ticker_ShowTotal}
 XPEntry_Ticker {type=SoundBase, default=/Game/Sounds/Fort_UI_Sounds/Gameplay/XP/MSS_PostMatch_XPEntry_Ticker}
 XPEntry_Ticker_Components {type="array<AudioComponent>", default [null,null,null]}
 XPEntry_TickerFinal {type=SoundBase, default=/Game/Sounds/Fort_UI_Sounds/Gameplay/XP/MSS_PostMatch_XPEntry_TickerFinal}
 XPRecapTimer {type=TimerHandle}
]
widgets {
 info {inheritedWidgetCount=9, parentClass=/Script/FortniteUI.FortMatchXPRecap, rootWidgetClass=&Overlay, widgetCount=57}
 widgets [
  Overlay_XPRecap {widgetClassPath=&Overlay, widget=Overlay_XPRecap}
  RootSafeZone {
   widgetClassPath=&SafeZone
   parent=Overlay_XPRecap
   widget=RootSafeZone
   slot {class=OverlaySlot, properties {horizontalAlignment=HAlign_Fill, verticalAlignment=VAlign_Fill}}
  }
  Overlay_Safezone {widgetClassPath=&Overlay, parent=RootSafeZone, widget=Overlay_Safezone, slot {class=SafeZoneSlot}}
  VerticalBox_0 {
   widgetClassPath=&VerticalBox
   parent=Overlay_Safezone
   widget=VerticalBox_0
   slot {class=OverlaySlot, properties {horizontalAlignment=HAlign_Fill, verticalAlignment=VAlign_Center}}
  }
  Spacer_48 {widgetClassPath=&Spacer, parent=VerticalBox_0, widget=Spacer_48, properties {size=<1 64>}, slot {class=VerticalBoxSlot}}
  GPanel_XPRundown {
   widgetClassPath=&GridPanel
   bIsVariable=true
   parent=VerticalBox_0
   widget=GPanel_XPRundown
   properties {columnFill [1,0,1]}
   slot {class=VerticalBoxSlot}
  }
  SBox_LeftSideContent {
   widgetClassPath=&SizeBox
   parent=GPanel_XPRundown
   widget=SBox_LeftSideContent
   properties {bOverride_WidthOverride=true, widthOverride=700}
   slot {class=GridSlot, properties {horizontalAlignment=HAlign_Right, verticalAlignment=VAlign_Center}}
  }
  Overlay_251 {widgetClassPath=&Overlay, parent=SBox_LeftSideContent, widget=Overlay_251, slot {class=SizeBoxSlot}}
  Attach_QuestEffects {
   widgetClassPath=&FortVisualAttachment
   parent=Overlay_251
   widget=Attach_QuestEffects
   properties {contentAnchor=<0.5 0.5>}
   slot {class=OverlaySlot, properties {horizontalAlignment=HAlign_Center, verticalAlignment=VAlign_Center}}
  }
  Ovr_QuestXPEffects {
   widgetClassPath=&Overlay
   bIsVariable=true
   parent=Attach_QuestEffects
   widget=Ovr_QuestXPEffects
   properties {renderOpacity=0}
   slot {class=SizeBoxSlot, properties {horizontalAlignment=HAlign_Right, verticalAlignment=VAlign_Center}}
  }
  Img_XPEffects_Glow {
   widgetClassPath=&Image
   parent=Ovr_QuestXPEffects
   widget=Img_XPEffects_Glow
   properties {
    brush {
     bIsDynamicallyLoaded=false
     drawAs=Image
     imageType=FullColor
     mirroring=NoMirror
     resourceName=None
     resourceObject=/FortniteCoreUI/HUD/EndGame/PostGameFlowScreens/DefaultScreens/XPScreen/Materials/MI_UI_PostMatch_XPFlare
     tiling=NoTile
     imageSize=<700 190>
     margin=<ltrb 0 0 0 0>
     outlineSettings {
      bUseBrushTransparency=false
      roundingType=HalfHeightRadius
      width=0
      color {colorUseRule=UseColor_Specified, specifiedColor=<rgba 0 0 0 0>}
      cornerRadii=<0 0 0 0>
     }
     tintColor {colorUseRule=UseColor_Specified, specifiedColor=<rgba 1 1 1 1>}
     uVRegion {bIsValid=false, max=<0 0>, min=<0 0>}
    }
    colorAndOpacity=<rgba 1 1 1 0.3>
   }
   slot {class=OverlaySlot, properties {horizontalAlignment=HAlign_Center, verticalAlignment=VAlign_Center}}
  }
  Ovr_QuestArea {
   widgetClassPath=&Overlay
   parent=Overlay_251
   widget=Ovr_QuestArea
   slot {class=OverlaySlot, properties {horizontalAlignment=HAlign_Right, verticalAlignment=VAlign_Center, padding=<ltrb 0 0 40 0>}}
  }
  Attach_QuestTotals {
   widgetClassPath=&FortVisualAttachment
   parent=Ovr_QuestArea
   widget=Attach_QuestTotals
   properties {contentAnchor=<0.5 0>}
   slot {class=OverlaySlot, properties {horizontalAlignment=HAlign_Center, verticalAlignment=VAlign_Bottom}}
  }
  Ovr_QuestTotals {
   widgetClassPath=&Overlay
   bIsVariable=true
   parent=Attach_QuestTotals
   widget=Ovr_QuestTotals
   properties {renderOpacity=0}
   slot {class=SizeBoxSlot, properties {horizontalAlignment=HAlign_Right, verticalAlignment=VAlign_Bottom}}
  }
  HorizontalBox_305 {widgetClassPath=&HorizontalBox, parent=Ovr_QuestTotals, widget=HorizontalBox_305, slot {class=OverlaySlot}}
  Txt_CompletedQuests {
   widgetClassPath=&CommonTextBlock
   bIsVariable=true
   parent=HorizontalBox_305
   widget=Txt_CompletedQuests
   properties {
    text=0
    textTransformPolicy=ToUpper
    font {
     bForceMonospaced=false
     bMaterialIsStencil=false
     fontMaterial=null
     fontObject=&HeadingNow
     letterSpacing=0
     monospacedWidth=1
     size=22.5
     skewAmount=0
     typefaceFontName=76 Bold
     outlineSettings {
      bApplyOutlineToDropShadows=false
      bMiteredCorners=false
      bSeparateFillAlpha=false
      outlineMaterial=null
      outlineSize=0
      outlineColor=<rgba 0 0 0 1>
     }
    }
    shadowOffset=<0 0>
   }
   slot {class=HorizontalBoxSlot, properties {verticalAlignment=VAlign_Center, padding=<ltrb 0 0 10 0>}}
  }
  CommonTextBlock {
   widgetClassPath=&CommonTextBlock
   parent=HorizontalBox_305
   widget=CommonTextBlock
   properties {
    text=Completed
    textTransformPolicy=ToUpper
    colorAndOpacity {colorUseRule=UseColor_Specified, specifiedColor=<rgba 0.3906 0.3906 0.3906 1>}
    font {
     bForceMonospaced=false
     bMaterialIsStencil=false
     fontMaterial=null
     fontObject=&HeadingNow
     letterSpacing=0
     monospacedWidth=1
     size=18.75
     skewAmount=0
     typefaceFontName=75 Medium
     outlineSettings {
      bApplyOutlineToDropShadows=false
      bMiteredCorners=false
      bSeparateFillAlpha=false
      outlineMaterial=null
      outlineSize=0
      outlineColor=<rgba 0 0 0 1>
     }
    }
    shadowOffset=<0 0>
   }
   slot {class=HorizontalBoxSlot, properties {verticalAlignment=VAlign_Center}}
  }
  HorizontalBox {
   widgetClassPath=&HorizontalBox
   parent=Ovr_QuestTotals
   widget=HorizontalBox
   slot {class=OverlaySlot, properties {padding=<ltrb 0 40 0 0>}}
  }
  Txt_InProgressQuests {
   widgetClassPath=&CommonTextBlock
   bIsVariable=true
   parent=HorizontalBox
   widget=Txt_InProgressQuests
   properties {
    text=0
    textTransformPolicy=ToUpper
    font {
     bForceMonospaced=false
     bMaterialIsStencil=false
     fontMaterial=null
     fontObject=&HeadingNow
     letterSpacing=0
     monospacedWidth=1
     size=22.5
     skewAmount=0
     typefaceFontName=76 Bold
     outlineSettings {
      bApplyOutlineToDropShadows=false
      bMiteredCorners=false
      bSeparateFillAlpha=false
      outlineMaterial=null
      outlineSize=0
      outlineColor=<rgba 0 0 0 1>
     }
    }
    shadowOffset=<0 0>
   }
   slot {class=HorizontalBoxSlot, properties {verticalAlignment=VAlign_Center, padding=<ltrb 0 0 10 0>}}
  }
  CommonTextBlock_1 {
   widgetClassPath=&CommonTextBlock
   parent=HorizontalBox
   widget=CommonTextBlock_1
   properties {
    text=In Progress
    textTransformPolicy=ToUpper
    colorAndOpacity {colorUseRule=UseColor_Specified, specifiedColor=<rgba 0.3906 0.3906 0.3906 1>}
    font {
     bForceMonospaced=false
     bMaterialIsStencil=false
     fontMaterial=null
     fontObject=&HeadingNow
     letterSpacing=0
     monospacedWidth=1
     size=18.75
     skewAmount=0
     typefaceFontName=75 Medium
     outlineSettings {
      bApplyOutlineToDropShadows=false
      bMiteredCorners=false
      bSeparateFillAlpha=false
      outlineMaterial=null
      outlineSize=0
      outlineColor=<rgba 0 0 0 1>
     }
    }
    shadowOffset=<0 0>
   }
   slot {class=HorizontalBoxSlot, properties {verticalAlignment=VAlign_Center}}
  }
  MatchXP_ChallengeXP {
   widgetClassPath=&/WBP_PostMatch_RundownCategoryItem
   bInherited=true
   bIsVariable=true
   parent=Ovr_QuestArea
   widget=MatchXP_ChallengeXP
   properties {bCollapseWhenEmpty=false, hiddenVisibility=SelfHitTestInvisible, elementTag {tagName=HUD.Athena.XPWidget.PostGameFeedEntry}}
   slot {class=OverlaySlot}
  }
  FortVisualAttachment_151 {
   widgetClassPath=&FortVisualAttachment
   parent=Overlay_251
   widget=FortVisualAttachment_151
   properties {contentAnchor=<0 0.5>}
   slot {class=OverlaySlot}
  }
  VerticalBox_PrimaryMatchXPExtension {
   widgetClassPath=&VerticalBox
   bIsVariable=true
   parent=FortVisualAttachment_151
   widget=VerticalBox_PrimaryMatchXPExtension
   properties {renderOpacity=0}
   slot {class=SizeBoxSlot, properties {horizontalAlignment=HAlign_Center, verticalAlignment=VAlign_Top}}
  }
  SBox_CenterContent {
   widgetClassPath=&SizeBox
   bIsVariable=true
   parent=GPanel_XPRundown
   widget=SBox_CenterContent
   properties {bOverride_HeightOverride=true, bOverride_WidthOverride=true, heightOverride=400, widthOverride=400}
   slot {class=GridSlot, properties {column=1, horizontalAlignment=HAlign_Center, layer=3, verticalAlignment=VAlign_Center}}
  }
  Overlay_253 {widgetClassPath=&Overlay, parent=SBox_CenterContent, widget=Overlay_253, slot {class=SizeBoxSlot}}
  Att_PartyLeft {
   widgetClassPath=&FortVisualAttachment
   parent=Overlay_253
   widget=Att_PartyLeft
   properties {contentAnchor=<0.5 1>}
   slot {class=OverlaySlot, properties {horizontalAlignment=HAlign_Center}}
  }
  Text_MemberOptedOut {
   widgetClassPath=&CommonTextBlock
   parent=Att_PartyLeft
   widget=Text_MemberOptedOut
   properties {
    autoWrapText=true
    justification=Center
    text=Your party left the team
    textTransformPolicy=ToUpper
    visibility=Collapsed
    font {
     bForceMonospaced=false
     bMaterialIsStencil=false
     fontMaterial=null
     fontObject=&Burbank
     letterSpacing=0
     monospacedWidth=1
     size=34
     skewAmount=0
     typefaceFontName=Burbank-PlayerPower
     outlineSettings {
      bApplyOutlineToDropShadows=false
      bMiteredCorners=false
      bSeparateFillAlpha=false
      outlineMaterial=null
      outlineSize=0
      outlineColor=<rgba 0 0 0 1>
     }
    }
    renderTransform {angle=0, scale=<1 1>, shear=<-8 0>, translation=<0 0>}
    shadowOffset=<0 0>
   }
   slot {class=SizeBoxSlot, properties {horizontalAlignment=HAlign_Center, verticalAlignment=VAlign_Top}}
  }
  ScaleBox_0 {
   widgetClassPath=&ScaleBox
   parent=Overlay_253
   widget=ScaleBox_0
   slot {class=OverlaySlot, properties {horizontalAlignment=HAlign_Fill, verticalAlignment=VAlign_Fill}}
  }
  NSlot_XPEverywhere {
   widgetClassPath=&NamedSlot
   bIsVariable=true
   parent=ScaleBox_0
   widget=NSlot_XPEverywhere
   properties {bExposeOnInstanceOnly=true}
   slot {class=ScaleBoxSlot}
  }
  Att_BottomCenterContent {
   widgetClassPath=&FortVisualAttachment
   parent=Overlay_253
   widget=Att_BottomCenterContent
   properties {contentAnchor=<0.5 0>}
   slot {class=OverlaySlot, properties {horizontalAlignment=HAlign_Center, verticalAlignment=VAlign_Bottom}}
  }
  VisSwitcher_BottomCenter {
   widgetClassPath=&CommonVisibilitySwitcher
   bIsVariable=true
   parent=Att_BottomCenterContent
   widget=VisSwitcher_BottomCenter
   properties {bOverride_Cursor=true, renderOpacity=0}
   slot {class=SizeBoxSlot, properties {horizontalAlignment=HAlign_Center, verticalAlignment=VAlign_Bottom, padding=<ltrb 0 20 0 20>}}
  }
  HBox_CenterXPRundownContent {
   widgetClassPath=&HorizontalBox
   bIsVariable=true
   parent=VisSwitcher_BottomCenter
   widget=HBox_CenterXPRundownContent
   properties {bOverride_Cursor=true}
   slot {class=CommonVisibilitySwitcherSlot}
  }
  MatchXP_TotalEntry {
   widgetClassPath=&/WBP_PostMatch_RundownReport
   bInherited=true
   bIsVariable=true
   parent=HBox_CenterXPRundownContent
   widget=MatchXP_TotalEntry
   properties {
    xPType=Total Match XP
    elementTag {tagName=HUD.Athena.XPWidget.PostGameFeedEntry}
    renderTransform {angle=0, scale=<0.7 0.7>, shear=<0 0>, translation=<0 0>}
   }
   slot {class=HorizontalBoxSlot}
  }
  MatchXP_UntilNextLvlReport {
   widgetClassPath=&WBP_PostMatch_RundownReport
   bIsVariable=true
   parent=HBox_CenterXPRundownContent
   widget=MatchXP_UntilNextLvlReport
   properties {elementTag {tagName=HUD.Athena.XPWidget.PostGameFeedEntry}, renderTransform {angle=0, scale=<0.7 0.7>, shear=<0 0>, translation=<0 0>}}
   slot {class=HorizontalBoxSlot}
  }
  ReturnToIslandButton {
   widgetClassPath=/FortUILibrary/UIKit/Buttons/WBP_UIKit_Button_Loud
   bIsVariable=true
   parent=VisSwitcher_BottomCenter
   widget=ReturnToIslandButton
   properties {
    bIsFocusable=false
    buttonSize=Large
    navigation=&self:WidgetTree.ReturnToIslandButton.WidgetNavigation_0
    text=RETURN TO ISLAND
    visibility=Collapsed
    triggeringInputAction {dataTable=&InputActionDataTable, rowName=CreativeScoreboard_Abort}
   }
   slot {class=CommonVisibilitySwitcherSlot}
  }
  Overlay_BottomCenterExtension {
   widgetClassPath=&Overlay
   bIsVariable=true
   parent=VisSwitcher_BottomCenter
   widget=Overlay_BottomCenterExtension
   slot {class=CommonVisibilitySwitcherSlot}
  }
  Att_RundownAndTotal {
   widgetClassPath=&FortVisualAttachment
   parent=Overlay_253
   widget=Att_RundownAndTotal
   properties {bOverride_HeightOverride=true, bOverride_WidthOverride=true, heightOverride=195, widthOverride=1500, contentAnchor=<0.5 0>}
   slot {class=OverlaySlot, properties {horizontalAlignment=HAlign_Center, verticalAlignment=VAlign_Bottom}}
  }
  VBox_RundownEntries {
   widgetClassPath=&VerticalBox
   bIsVariable=true
   parent=Att_RundownAndTotal
   widget=VBox_RundownEntries
   slot {class=SizeBoxSlot, properties {horizontalAlignment=HAlign_Center, verticalAlignment=VAlign_Top}}
  }
  SBox_RightSideContent {
   widgetClassPath=&SizeBox
   parent=GPanel_XPRundown
   widget=SBox_RightSideContent
   properties {bOverride_WidthOverride=true, widthOverride=700}
   slot {class=GridSlot, properties {column=2, horizontalAlignment=HAlign_Left, verticalAlignment=VAlign_Center}}
  }
  Overlay {widgetClassPath=&Overlay, parent=SBox_RightSideContent, widget=Overlay, slot {class=SizeBoxSlot}}
  Attatch_GameplayEffects {
   widgetClassPath=&FortVisualAttachment
   parent=Overlay
   widget=Attatch_GameplayEffects
   properties {contentAnchor=<0.5 0.5>}
   slot {class=OverlaySlot, properties {horizontalAlignment=HAlign_Center, verticalAlignment=VAlign_Center}}
  }
  Ovr_GameplayEffects {
   widgetClassPath=&Overlay
   bIsVariable=true
   parent=Attatch_GameplayEffects
   widget=Ovr_GameplayEffects
   properties {renderOpacity=0, renderTransform {angle=0, scale=<-1 1>, shear=<0 0>, translation=<0 0>}}
   slot {class=SizeBoxSlot, properties {horizontalAlignment=HAlign_Left, verticalAlignment=VAlign_Center}}
  }
  Img_XPEffects_Glow_1 {
   widgetClassPath=&Image
   parent=Ovr_GameplayEffects
   widget=Img_XPEffects_Glow_1
   properties {
    visibility=SelfHitTestInvisible
    brush {
     bIsDynamicallyLoaded=false
     drawAs=Image
     imageType=FullColor
     mirroring=NoMirror
     resourceName=None
     resourceObject=/FortniteCoreUI/HUD/EndGame/PostGameFlowScreens/DefaultScreens/XPScreen/Materials/MI_UI_PostMatch_XPFlare_Gameplay
     tiling=NoTile
     imageSize=<700 190>
     margin=<ltrb 0 0 0 0>
     outlineSettings {
      bUseBrushTransparency=false
      roundingType=HalfHeightRadius
      width=0
      color {colorUseRule=UseColor_Specified, specifiedColor=<rgba 0 0 0 0>}
      cornerRadii=<0 0 0 0>
     }
     tintColor {colorUseRule=UseColor_Specified, specifiedColor=<rgba 1 1 1 1>}
     uVRegion {bIsValid=false, max=<0 0>, min=<0 0>}
    }
    colorAndOpacity=<rgba 1 1 1 0.3>
   }
   slot {class=OverlaySlot, properties {horizontalAlignment=HAlign_Center, verticalAlignment=VAlign_Center}}
  }
  Ovr_GameplayArea {
   widgetClassPath=&Overlay
   parent=Overlay
   widget=Ovr_GameplayArea
   slot {class=OverlaySlot, properties {verticalAlignment=VAlign_Center, padding=<ltrb 40 0 0 0>}}
  }
  MatchXP_GameplayXp {
   widgetClassPath=&WBP_PostMatch_RundownCategoryItem
   bInherited=true
   bIsVariable=true
   parent=Ovr_GameplayArea
   widget=MatchXP_GameplayXp
   properties {
    bCollapseWhenEmpty=false
    hiddenVisibility=SelfHitTestInvisible
    isRight=true
    xPType=Gameplay
    elementTag {tagName=HUD.Athena.XPWidget.PostGameFeedEntry}
    rundownCategoryColor=<rgba 0.9647 0.7379 0.4342 1>
   }
   slot {class=OverlaySlot}
  }
  WBP_PostMatch_RundownTotal {
   widgetClassPath=&/WBP_PostMatch_RundownTotal
   bIsVariable=true
   parent=GPanel_XPRundown
   widget=WBP_PostMatch_RundownTotal
   properties {renderOpacity=0, elementTag {tagName=HUD.Athena.XPWidget.PostGameFeedEntry}}
   slot {class=GridSlot, properties {columnSpan=3, row=1, padding=<ltrb -60 -250 -60 -250>}}
  }
  SizeBox_0 {
   widgetClassPath=&SizeBox
   parent=Overlay_Safezone
   widget=SizeBox_0
   properties {bOverride_HeightOverride=true, bOverride_WidthOverride=true, heightOverride=550, widthOverride=1200}
   slot {class=OverlaySlot, properties {padding=<ltrb 12 140 0 0>}}
  }
  MatchXPScoreboard {
   widgetClassPath=/FortniteCoreUI/Athena/HUD/EndGame/PostGameFlowScreens/DefaultScreens/XPScreen/WBP_MatchXPScoreboard
   bIsVariable=true
   parent=SizeBox_0
   widget=MatchXPScoreboard
   properties {bShowHeader=false, elementTag {tagName=Athena.HideForCapture}, hUDElementTag {gameplayTags []}}
   slot {class=SizeBoxSlot}
  }
  AthenaWinnerList {
   widgetClassPath=/Game/Athena/HUD/EndGame/AthenaWinnerList
   parent=Overlay_Safezone
   widget=AthenaWinnerList
   properties {visibility=Collapsed}
   slot {class=OverlaySlot, properties {horizontalAlignment=HAlign_Center, padding=<ltrb 0 355 0 0>}}
  }
  NSlot_LevelUpInjection {
   widgetClassPath=&NamedSlot
   bIsVariable=true
   parent=Overlay_Safezone
   widget=NSlot_LevelUpInjection
   slot {class=OverlaySlot, properties {horizontalAlignment=HAlign_Fill, verticalAlignment=VAlign_Fill}}
  }
  NSlot_BattlepassInjection {
   widgetClassPath=&NamedSlot
   bIsVariable=true
   parent=Overlay_XPRecap
   widget=NSlot_BattlepassInjection
   properties {renderOpacity=0}
   slot {class=OverlaySlot, properties {horizontalAlignment=HAlign_Fill, verticalAlignment=VAlign_Fill, padding=<ltrb 60 30 60 130>}}
  }
  MatchXP_MedalEntry {widgetClassPath=/Script/FortniteUI.FortMatchXPTotalEntry, bInherited=true, widget=null}
  MatchXP_SurvivalBonus {widgetClassPath=&FortMatchXPTotalEntry, bInherited=true, widget=null}
  MatchXP_CombatBonus {widgetClassPath=&FortMatchXPTotalEntry, bInherited=true, widget=null}
  MatchXP_MatchXp {widgetClassPath=&FortMatchXPTotalEntry, bInherited=true, widget=null}
  Widget_QuestRecap {widgetClassPath=/Script/FortniteUI.EndOfMatchQuestUpdateWidget, bInherited=true, widget=null}
  MatchCash_TotalEntry {widgetClassPath=&FortMatchXPTotalEntry, bInherited=true, widget=null}
 ]
}
