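# Post-match panel widget for the match recap XP rundown screen, wrapping an XPRecap child that also supplies the focus target.
# On init, S39 console variables can enable post-game updates or suppress the rundown by marking it skipped.
# Activation skips the visualization or starts XP animations, and deactivation stops them.
# When animations finish, the panel signals ready if auto-advance is valid or S39 updates are enabled.
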WidgetBlueprint /FortniteCoreUI/HUD/EndGame/PostGameFlowScreens/DefaultScreens/XPScreen/WBP_PostMatchPanel_MatchRecap_Rundown
graphs [
 BP_GetDesiredFocusTarget {nodes [
   Get Desired Focus Target()
   > Return Node!(
    execute
    ReturnValue:Widget Object Reference=GetDesiredFocusTarget!(self:Common Activatable Widget Object Reference=Get XPRecap() -> (XPRecap:Fort Match XPRecap Object Reference)) -> (ReturnValue:Widget Object Reference)
   )
  ]}
 EventGraph {nodes [
   OnXpAnimationsFinished() {is=title K2Node_CustomEvent}
   > Branch(Condition=OR Boolean(
     A=IsAutoAdvanceTabValid!(self:Athena Post Match Panel Object Reference) -> (ReturnValue:bool)
     B=Get S39EnablePostGameUpdates() -> (S39EnablePostGameUpdates:bool)
    ))
   > OnPanelReady(self:Athena Post Match Panel Object Reference)
   Event OnSkipVisualization() {is=title K2Node_Event}
   > Set IsSkipping(IsSkipping:bool=true) -> (Output_Get:bool)
   > SkipVisualization(self:Fort Match XPRecap Object Reference=Get XPRecap())
   Event On Deactivated()
   > SetXPAnimationState!(execute, self:Fort Match XPRecap Object Reference=Get XPRecap(), IsActive:bool=false) -> (then)
   Event OnInitialized()
   > InitializeS39PostMatch()
   Event OnActivatePanel() {is=title K2Node_Event}
   > n1 = Branch(Condition=Get IsSkipping() -> (IsSkipping:bool))
   > SkipVisualization(self=Get XPRecap())
   SetXPAnimationState(execute=@n1.else, self=Get XPRecap(), IsActive=true)
   Event Construct()
   > Parent: Construct()
   > Bind Event to On XPRundown Animations Finished(self:Fort Match XPRecap Object Reference=Get XPRecap(), Delegate=Create Event()) {is=title K2Node_AddDelegate}
  ]}
 InitializeS39PostMatch {nodes [
   InitializeS39PostMatch() {is=title K2Node_FunctionEntry}
   > n1 = GetConsoleVariableBoolValue!(execute, VariableName:String=Athena.MatchXPScreenWidget.S39EnablePostGameUpdates) -> (then, ReturnValue:bool)
   > Set S39EnablePostGameUpdates(S39EnablePostGameUpdates:bool=@n1) -> (Output_Get:bool)
   > n2 = GetConsoleVariableBoolValue(VariableName=Athena.MatchXPScreenWidget.S39SuppressXPRundown)
   > n3 = Set IsSkipping(IsSkipping=@n2)
   > Branch(Condition=@n3)
   > n4 = Cast To WBP_MatchXPRecap_Rundown!(execute, Object=Get XPRecap()) -> (then, CastFailed, AsWBP Match XPRecap Rundown:WBP Match XPRecap Rundown Object Reference) {is=title K2Node_DynamicCast}
   > Set IsSkippingVisualization!(execute, IsSkippingVisualization:bool=true, self:WBP Match XPRecap Rundown Object Reference=@n4) -> (then, Output_Get:bool)
  ]}
 OnBlockMapActionByKey {nodes [
   OnBlockMapActionByKey!() -> (then, InKey:"Key Structure (by ref)") {is=title K2Node_FunctionEntry}
   > Return Node!(execute, bOutMapActionBlocked:bool=false, ReturnValue:bool=false)
  ]}
 OnHandleMapActionKey {nodes [
   OnHandleMapActionKey!() -> (then, InKey:"Key Structure (by ref)", bKeyDown:bool) {is=title K2Node_FunctionEntry}
   > Return Node!(execute, bOutMapActionHandled:bool=false, ReturnValue=false)
  ]}
 OnHandleMapBackAction {nodes [
   OnHandleMapBackAction() {is=title K2Node_FunctionEntry}
   > Return Node!(execute, bOutBackActionHandled:bool=false, ReturnValue=false)
  ]}
 OnShouldCreateMapScreenTab {nodes [
   OnShouldCreateMapScreenTab!() -> (then, WorldContextObject:Object Reference) {is=title K2Node_FunctionEntry}
   > Return Node!(execute, bOutShouldCreateMapScreenTab:bool=false, ReturnValue=false)
  ]}
 OnShouldDisplayBackActionInActionBar {nodes [
   OnShouldDisplayBackActionInActionBar() {is=title K2Node_FunctionEntry}
   > Return Node!(execute, bOutShouldDisplayBackActionInActionBar:bool=false, ReturnValue=false)
  ]}
 OnShouldShowTabOnMapScreenForMinigame {nodes [
   OnShouldShowTabOnMapScreenForMinigame!() -> (then, Minigame:Fort Minigame Object Reference) {is=title K2Node_FunctionEntry}
   > Return Node!(execute, bOutShouldShowTabOnMapScreenForMinigame:bool=false, ReturnValue=false)
  ]}
]
variables [
 IsSkipping {type=boolean}
 S39EnablePostGameUpdates {type=boolean}
]
widgets {
 info {inheritedWidgetCount=1, parentClass=/Script/FortniteUI.AthenaPostMatchPanel_MatchRecap, rootWidgetClass=&Overlay, widgetCount=2}
 widgets [
  Overlay {widgetClassPath=&Overlay, widget=Overlay, properties {bOverride_Cursor=true}}
  XPRecap {
   widgetClassPath=&/WBP_MatchXPRecap_Rundown
   bInherited=true
   bIsVariable=true
   parent=Overlay
   widget=XPRecap
   slot {class=OverlaySlot, properties {horizontalAlignment=HAlign_Fill, verticalAlignment=VAlign_Fill}}
  }
 ]
}
