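# Post-match XP screen widget that shows an "Available Rewards!" banner with a claim-count highlight below it.
# Two stacked headlines sit on glow and streak material backgrounds, each doubled by a transparent-fill copy whose outline material creates a glow behind the text.
# Anim_Show fades and drops in both headlines and the highlight in a short stagger.
# The event graph holds only unconnected stubs, so nothing here plays the animation or replaces the placeholder text.
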WidgetBlueprint /FortniteCoreUI/HUD/EndGame/PostGameFlowScreens/DefaultScreens/XPScreen/WBP_PostMatch_AvailableRewards
animations [Anim_Show {
  binding_count=3
  length_seconds=0.45
  object=&self:Anim_Show
  sequence {bindings [
    Overlay_Available {
     id=#1
     possessed_class=Overlay
     tracks [
      {
       type=MovieScene2DTransformTrack
       sections [{
         type=MovieScene2DTransformSection
         end_frame=5
         start_frame=0
         channels [
          Translation.X {type=MovieSceneScriptingFloatChannel, const [0,1]}
          Translation.Y {type=MovieSceneScriptingFloatChannel, keys [[0,-20],[5,0]]}
         ]
        }]
      }
      {
       type=MovieSceneFloatTrack
       display_name=Render Opacity
       sections [{type=MovieSceneFloatSection, end_frame=5, start_frame=0, channels [None {type=MovieSceneScriptingFloatChannel, keys [[0,0],[5,1]]}]}]
      }
     ]
    }
    Overlay_Rewards {
     id=#2
     possessed_class=Overlay
     tracks [
      {
       type=MovieScene2DTransformTrack
       sections [{
         type=MovieScene2DTransformSection
         end_frame=8
         start_frame=0
         channels [
          Translation.X {type=MovieSceneScriptingFloatChannel, const [0,1]}
          Translation.Y {type=MovieSceneScriptingFloatChannel, keys [[2,-20],[7,0]]}
         ]
        }]
      }
      {
       type=MovieSceneFloatTrack
       display_name=Render Opacity
       sections [{type=MovieSceneFloatSection, end_frame=8, start_frame=0, channels [None {type=MovieSceneScriptingFloatChannel, keys [[2,0],[7,1]]}]}]
      }
     ]
    }
    WBP_UIKit_Highlight_59 {
     id=#3
     possessed_class=WBP_UIKit_Highlight_C
     tracks [
      {
       type=MovieSceneFloatTrack
       display_name=Render Opacity
       sections [{type=MovieSceneFloatSection, end_frame=9, start_frame=0, channels [None {type=MovieSceneScriptingFloatChannel, keys [[4,0],[9,1]]}]}]
      }
      {
       type=MovieScene2DTransformTrack
       sections [{
         type=MovieScene2DTransformSection
         end_frame=9
         start_frame=0
         channels [
          Translation.X {type=MovieSceneScriptingFloatChannel, const [0,1]}
          Translation.Y {type=MovieSceneScriptingFloatChannel, keys [[4,-20],[9,0]]}
         ]
        }]
      }
     ]
    }
   ]}
 }]
graphs [EventGraph {nodes [
   Event PreConstruct()
   Event Construct()
   Event Tick()
  ]}]
variables []
widgets {
 info {inheritedWidgetCount=0, parentClass=&UserWidget, rootWidgetClass=&Overlay, widgetCount=12}
 widgets [
  Overlay_AvailableRewards {widgetClassPath=&Overlay, widget=Overlay_AvailableRewards}
  Image_AvailableRewardsBackground {
   widgetClassPath=&Image
   parent=Overlay_AvailableRewards
   widget=Image_AvailableRewardsBackground
   properties {brush {
     bIsDynamicallyLoaded=false
     drawAs=Box
     imageType=FullColor
     mirroring=NoMirror
     resourceName=None
     resourceObject=/FortniteCoreUI/HUD/EndGame/PostGameFlowScreens/DefaultScreens/XPScreen/Materials/MI_PostMatch_TextBackgroundGlow2
     tiling=NoTile
     imageSize=<128 128>
     margin=<ltrb 0 0 0 0>
     outlineSettings {
      bUseBrushTransparency=false
      roundingType=HalfHeightRadius
      width=0
      color {colorUseRule=UseColor_Specified, specifiedColor=<rgba 0 0 0 0>}
      cornerRadii=<0 0 0 0>
     }
     tintColor {colorUseRule=UseColor_Specified, specifiedColor=<rgba 1 1 1 1>}
     uVRegion {bIsValid=false, max=<0 0>, min=<0 0>}
    }}
   slot {class=OverlaySlot, properties {horizontalAlignment=HAlign_Fill, verticalAlignment=VAlign_Fill, padding=<ltrb 0 -200 0 -200>}}
  }
  VBox_AvailableRewards {
   widgetClassPath=&VerticalBox
   parent=Overlay_AvailableRewards
   widget=VBox_AvailableRewards
   properties {bOverride_Cursor=true}
   slot {class=OverlaySlot}
  }
  Overlay_Available {
   widgetClassPath=&Overlay
   parent=VBox_AvailableRewards
   widget=Overlay_Available
   slot {class=VerticalBoxSlot, properties {horizontalAlignment=HAlign_Center}}
  }
  Image_AvailableBackground {
   widgetClassPath=&Image
   parent=Overlay_Available
   widget=Image_AvailableBackground
   properties {brush {
     bIsDynamicallyLoaded=false
     drawAs=Image
     imageType=FullColor
     mirroring=NoMirror
     resourceName=None
     resourceObject=/FortniteCoreUI/HUD/EndGame/PostGameFlowScreens/DefaultScreens/XPScreen/Materials/MI_PostMatch_TextBackgroundStreak
     tiling=NoTile
     imageSize=<32 32>
     margin=<ltrb 0 0 0 0>
     outlineSettings {
      bUseBrushTransparency=false
      roundingType=HalfHeightRadius
      width=0
      color {colorUseRule=UseColor_Specified, specifiedColor=<rgba 0 0 0 0>}
      cornerRadii=<0 0 0 0>
     }
     tintColor {colorUseRule=UseColor_Specified, specifiedColor=<rgba 1 1 1 1>}
     uVRegion {bIsValid=false, max=<0 0>, min=<0 0>}
    }}
   slot {class=OverlaySlot, properties {horizontalAlignment=HAlign_Fill, verticalAlignment=VAlign_Fill, padding=<ltrb -80 20 -80 20>}}
  }
  Text_DummyAvailable {
   widgetClassPath=&UEFN_TextBlock
   bIsVariable=true
   parent=Overlay_Available
   widget=Text_DummyAvailable
   properties {
    justification=Center
    text="{Available}"
    textTransformPolicy=ToUpper
    colorAndOpacity {colorUseRule=UseColor_Specified, specifiedColor=<rgba 1 1 1 0>}
    font {
     bForceMonospaced=false
     bMaterialIsStencil=false
     fontMaterial=null
     fontObject=&HeadingNow
     letterSpacing=0
     monospacedWidth=1
     size=30
     skewAmount=0
     typefaceFontName=86 Bold Italic
     outlineSettings {
      bApplyOutlineToDropShadows=false
      bMiteredCorners=true
      bSeparateFillAlpha=false
      outlineMaterial=/FortniteCoreUI/HUD/EndGame/PostGameFlowScreens/DefaultScreens/XPScreen/Materials/MI_PostMatch_FontGlow
      outlineSize=22
      outlineColor=<rgba 1 1 1 1>
     }
    }
   }
   slot {class=OverlaySlot, properties {horizontalAlignment=HAlign_Fill, verticalAlignment=VAlign_Center}}
  }
  Text_Available {
   widgetClassPath=&UEFN_TextBlock
   bIsVariable=true
   parent=Overlay_Available
   widget=Text_Available
   properties {
    justification=Center
    text="{Available}"
    textTransformPolicy=ToUpper
    font {
     bForceMonospaced=false
     bMaterialIsStencil=false
     fontMaterial=null
     fontObject=&HeadingNow
     letterSpacing=0
     monospacedWidth=1
     size=30
     skewAmount=0
     typefaceFontName=86 Bold Italic
     outlineSettings {
      bApplyOutlineToDropShadows=false
      bMiteredCorners=true
      bSeparateFillAlpha=false
      outlineMaterial=null
      outlineSize=2
      outlineColor=<rgba 0 0 0 0.1>
     }
    }
   }
   slot {class=OverlaySlot, properties {horizontalAlignment=HAlign_Fill, verticalAlignment=VAlign_Center}}
  }
  Overlay_Rewards {
   widgetClassPath=&Overlay
   parent=VBox_AvailableRewards
   widget=Overlay_Rewards
   slot {class=VerticalBoxSlot, properties {padding=<ltrb 0 -96 0 -24>}}
  }
  Image_RewardsBackground {
   widgetClassPath=&Image
   parent=Overlay_Rewards
   widget=Image_RewardsBackground
   properties {brush {
     bIsDynamicallyLoaded=false
     drawAs=Image
     imageType=FullColor
     mirroring=NoMirror
     resourceName=None
     resourceObject=&MI_PostMatch_TextBackgroundStreak
     tiling=NoTile
     imageSize=<32 32>
     margin=<ltrb 0 0 0 0>
     outlineSettings {
      bUseBrushTransparency=false
      roundingType=HalfHeightRadius
      width=0
      color {colorUseRule=UseColor_Specified, specifiedColor=<rgba 0 0 0 0>}
      cornerRadii=<0 0 0 0>
     }
     tintColor {colorUseRule=UseColor_Specified, specifiedColor=<rgba 1 1 1 1>}
     uVRegion {bIsValid=false, max=<0 0>, min=<0 0>}
    }}
   slot {class=OverlaySlot, properties {horizontalAlignment=HAlign_Fill, verticalAlignment=VAlign_Fill, padding=<ltrb -80 50 -80 50>}}
  }
  Text_DummyRewards {
   widgetClassPath=&UEFN_TextBlock
   bIsVariable=true
   parent=Overlay_Rewards
   widget=Text_DummyRewards
   properties {
    justification=Center
    text="{Rewards!}"
    textTransformPolicy=ToUpper
    colorAndOpacity {colorUseRule=UseColor_Specified, specifiedColor=<rgba 1 1 1 0>}
    font {
     bForceMonospaced=false
     bMaterialIsStencil=false
     fontMaterial=null
     fontObject=&HeadingNow
     letterSpacing=-48
     monospacedWidth=1
     size=72
     skewAmount=0
     typefaceFontName=86 Bold Italic
     outlineSettings {
      bApplyOutlineToDropShadows=false
      bMiteredCorners=true
      bSeparateFillAlpha=false
      outlineMaterial=&MI_PostMatch_FontGlow
      outlineSize=48
      outlineColor=<rgba 1 1 1 1>
     }
    }
   }
   slot {class=OverlaySlot, properties {horizontalAlignment=HAlign_Center, verticalAlignment=VAlign_Center}}
  }
  Text_Rewards {
   widgetClassPath=&UEFN_TextBlock
   bIsVariable=true
   parent=Overlay_Rewards
   widget=Text_Rewards
   properties {
    justification=Center
    text="{Rewards!}"
    textTransformPolicy=ToUpper
    font {
     bForceMonospaced=false
     bMaterialIsStencil=false
     fontMaterial=null
     fontObject=&HeadingNow
     letterSpacing=-48
     monospacedWidth=1
     size=72
     skewAmount=0
     typefaceFontName=86 Bold Italic
     outlineSettings {
      bApplyOutlineToDropShadows=false
      bMiteredCorners=true
      bSeparateFillAlpha=false
      outlineMaterial=null
      outlineSize=2
      outlineColor=<rgba 0 0 0 0.1>
     }
    }
   }
   slot {class=OverlaySlot, properties {horizontalAlignment=HAlign_Center, verticalAlignment=VAlign_Center}}
  }
  WBP_UIKit_Highlight_59 {
   widgetClassPath=/FortUILibrary/UIKit/Highlight/WBP_UIKit_Highlight
   bIsVariable=true
   parent=VBox_AvailableRewards
   widget=WBP_UIKit_Highlight_59
   properties {highlightText=10 to claim, highlightType=Attention}
   slot {class=VerticalBoxSlot, properties {horizontalAlignment=HAlign_Center}}
  }
 ]
}
